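# Pirate-themed skydiving contrail Niagara system that attaches ribbon trails and a flag mesh to the pawn's skeletal mesh.
# SkyContrails splits one burst into three ribbons, rebuilt each update from per-group bone/socket positions plus index-based offsets, with curves shaping alpha and width.
# User.ParaGlideLeanAlpha adjusts ribbon spread, length and panner speed.
# Pirate_Flag follows its own socket, facing system X, with randomized and timeline-driven material parameters.
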NiagaraSystem /BRCosmetics/Athena/Effects/Particles/Skydiving/NS_Contrail_ARRRGHHH
dependencies {
 usedDataInterfaces [&NiagaraDataInterfaceCurve]
 usedDynamicInputs []
 usedModules [&SystemState, &Timeline]
}
emitters {
 Pirate_Flag {
  emitterUpdateScript [
   &EmitterState(
    Inactive Response=::NewEnumerator1
    Life Cycle Mode=::NewEnumerator1
    Loop Behavior=::NewEnumerator1
    Loop Duration Mode=::NewEnumerator1
    UseLoopDelay=false
   )
   &SpawnBurst_Instantaneous(Use Spawn Probability=false, UseLoopCountLimit=false)
  ]
  particleSpawnScript [
   &InitializeParticle(Color Mode=::NewEnumerator0, Lifetime=5, UsePositionOffset=false, Write Material Random=true)
   &AddVelocity(Velocity=<0 0 0.1>)
  ]
  particleUpdateScript [
   &ParticleState(Kill Particles When Lifetime Has Elapsed=false, Let Infinitely Lived Particles Die When Emitter Deactivates=true)
   &UpdateMeshOrientation(
    Facing Direction=$Engine.Owner.SystemXAxis
    Facing Interpolation=false
    Orientation Method=::NewEnumerator1
    Override Quat=false
    Stabilization Mode:ENiagara_OrientVectorCount=::NewEnumerator0
    TransformFacing=false
   )
   &SkeletalMeshLocation(
    Bone / Socket Index=3
    Bone Sampling Mode=::NewEnumerator4
    MaskSpawnGroup=false
    Orientation Sampling=::NewEnumerator1
    Skeletal Mesh=$User.Pawn
   )
   &SolveForcesAndVelocity(Clamp Velocity=false, Limit Acceleration=false)
   &DynamicMaterialParameters(
    Index 0 Param 1=&RandomRangeFloat(Minimum=0.62, Maximum=0.9)
    Index 0 Param 2=&RandomRangeFloat(Minimum=0.75, Maximum=1.25)
    Index 0 Param 4=&FloatFromCurve(
     FloatCurve=DataInterface() {propertyValues {Curve {keys [
         {arriveTangent=1.1335, interpMode=RCIM_Cubic, leaveTangent=1.1335, tangentMode=RCTM_User, time=0, value=1}
         {arriveTangent=1.1188, interpMode=RCIM_Cubic, leaveTangent=1.1188, tangentMode=RCTM_User, time=1, value=0}
        ]}}}
     CurveIndex=$System.Timeline.NormalizedTime
    )
    Module.Param0WriteEnabled=true
    Module.Param1WriteEnabled=true
    Module.Param2WriteEnabled=true
    Module.Param3WriteEnabled=true
   )
  ]
  rendererClasses [&NiagaraMeshRendererProperties]
  renderers [{rendererClass=&NiagaraMeshRendererProperties, rendererIndex=0}]
 }
 SkyContrails {
  emitterUpdateScript [
   &EmitterState(Life Cycle Mode=::NewEnumerator1, Loop Behavior=::NewEnumerator1, Loop Duration Mode=::NewEnumerator1, UseLoopDelay=false)
   &SpawnBurst_Instantaneous(Spawn Count=30, Use Spawn Probability=false, UseLoopCountLimit=false)
   /SetVariables(Emitter.PannerSpeed:Vector3f=&MakeVector(X=&Lerp_Float(A=-2, B=-5, Alpha=$User.ParaGlideLeanAlpha), Y=0.1, Z=0)) {isSetParameters=true}
  ]
  particleSpawnScript [
   &FN_SetSpawnGroup(Normalize Spawn Group=true, Number of Spawn Groups=3, Particles Per Group=10, Set Ribbon ID and Ribbon Link Order=true)
   &InitializeParticle(
    Color=<rgba 1 1 1 0.075>
    Lifetime=0
    Ribbon Width=&MaskFloatBySpawnGroup(Masked Float=-10, Passthrough Float=10)
    Ribbon Width Mode=::NewEnumerator1
    UsePositionOffset=false
    Write Material Random=true
   )
  ]
  particleUpdateScript [
   &ParticleState(Kill Particles When Lifetime Has Elapsed=false, Let Infinitely Lived Particles Die When Emitter Deactivates=true)
   /SetVariables(Particles.RibbonOffsets:Vector3f=&MakeVector(
     X=&Lerp_Float(
      A=0
      B=&Lerp_Float(A=&MaskFloatBySpawnGroup(Masked Float=25, Passthrough Float=-25), B=0, Alpha=$User.ParaGlideLeanAlpha)
      Alpha=&FloatFromCurve(
       FloatCurve=DataInterface() {propertyValues {Curve {keys [
           {arriveTangent=2.4069, interpMode=RCIM_Cubic, leaveTangent=2.4069, leaveTangentWeight=0.8748, tangentMode=RCTM_User, time=0.15, value=0}
           {arriveTangentWeight=0.859, interpMode=RCIM_Cubic, tangentMode=RCTM_User, time=1, value=1}
          ]}}}
       CurveIndex=&ReturnNormalizedExecIndex()
      )
     )
     Y=0
     Z=&Lerp_Float(
      A=-20
      B=&Lerp_Float(A=100, B=150, Alpha=$User.ParaGlideLeanAlpha)
      Alpha=&ReturnNormalizedExecIndex()
     )
    )) {isSetParameters=true}
   &SkeletalMeshLocation(
    Bone / Socket Index=&SpawnGroup()
    Bone Sampling Mode=::NewEnumerator4
    MaskSpawnGroup=false
    Sampled Position Offset=$Particles.RibbonOffsets
    Sampled Position Offset Coordinate Space:ENiagaraSkelMeshTransforms=::NewEnumerator2
    Skeletal Mesh=$User.Pawn
   )
   &DynamicMaterialParameters(
    Index 0 Param 2=1
    Index 0 Param 3=2
    Module.Param0WriteEnabled=true
    Module.Param1WriteEnabled=true
    Module.Param2WriteEnabled=true
    Module.Param3WriteEnabled=true
   )
   &ScaleColor(
    Scale Alpha=&FloatFromCurve(
     FloatCurve=DataInterface() {propertyValues {Curve {keys [
         {arriveTangent=15.0922, interpMode=RCIM_Cubic, leaveTangent=15.0922, tangentMode=RCTM_User, time=0.1, value=0}
         <at 0.29 1 RCIM_Cubic>
         <at 1 0 RCIM_Cubic>
        ]}}}
     CurveIndex=&ReturnNormalizedExecIndex()
    )
    ScaleA=true
    ScaleRGB=true
   )
   &ScaleRibbonWidth(Ribbon Width Scale=&FloatFromCurve(
     FloatCurve=DataInterface() {propertyValues {Curve {keys [
         {arriveTangent=1.5004, interpMode=RCIM_Cubic, leaveTangent=1.5004, leaveTangentWeight=0.3333, tangentMode=RCTM_User, time=0, value=0.5}
         {arriveTangentWeight=0.3381, interpMode=RCIM_Cubic, tangentMode=RCTM_User, time=1, value=2}
        ]}}}
     CurveIndex=&ReturnNormalizedExecIndex()
    ))
  ]
  rendererClasses [&NiagaraRibbonRendererProperties]
  renderers [{rendererClass=&NiagaraRibbonRendererProperties}]
 }
}
summary {
 systemName=NS_Contrail_ARRRGHHH
 emitters [
  {emitterName=SkyContrails, rendererClasses [&NiagaraRibbonRendererProperties]}
  {emitterName=Pirate_Flag, rendererClasses [&NiagaraMeshRendererProperties]}
 ]
 userVariables [
  User.ParaGlideLeanAlpha {type=NiagaraFloat, defaultValue=0, description=""}
  User.Pawn {
   type=NiagaraDataInterfaceSkeletalMesh
   defaultValue {
    struct=NiagaraExt_VariableValue_DataInterface
    value {dataInterface=&self:NiagaraDataInterfaceSkeletalMesh_0}
   }
  }
 ]
}
