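# Cascade-converted Niagara skydiving contrail cosmetic that surrounds and trails the character with bats, smoke and ribbons.
# Three mesh emitters make orbiting bats: one around the character, one attached at sampled bones, one trailing.
# Hands and Feet emit bone-sampled ribbons split by spawn-group ribbon IDs, while SmokeLoop re-samples hand and foot bones every update.
# Front-end float modifiers add upward velocity outside gameplay and change the surrounding bats' spawn rate.
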NiagaraSystem /BRCosmetics/Athena/Effects/Particles/Skydiving/NS_Contrail_Bats_Converted
dependencies {
 usedDataInterfaces [&NiagaraDataInterfaceCurve, &NiagaraDataInterfaceVectorCurve]
 usedDynamicInputs []
 usedModules [
  &self:SystemSpawnScript.SystemScriptSource.SystemScriptGraph.NiagaraNodeAssignment_0.SetVariables_#1
  &SystemState
 ]
}
emitters {
 BatsTrail1001 {
  emitterUpdateScript [
   &EmitterState(Inactive Response=::NewEnumerator1, Life Cycle Mode=::NewEnumerator1, Loop Duration=1, UseLoopDelay=true)
   &SpawnRate(SpawnRate=15, Use Spawn Probability=false)
  ]
  particleSpawnScript [
   /SetVariables() {isSetParameters=true}
   /SetVariables(Particles.MeshOrientation=/CascadeToNiagaraConverter/NiagaraScripts/DynamicInputScripts/MakeQuatFromEuler(Euler:Vector3f=$Particles.MeshRotationEuler, Scale:NiagaraFloat=360)) {isSetParameters=true}
   /SetVariables(Particles.Acceleration:Vector3f=&RandomRangeVector(Minimum=<-100 -100 -100>, Maximum=<100 100 100>, Evaluation Type=::NewEnumerator1)) {isSetParameters=true}
   &InitializeParticle(
    Lifetime=&RandomRangeFloat(Minimum=1, Maximum=1.5, Evaluation Type=::NewEnumerator1)
    Mesh Scale=&VectorFromFloat(Value=&RandomRangeFloat(Minimum=0.25, Maximum=0.8, Evaluation Type=::NewEnumerator1))
    Mesh Scale Mode=::NewEnumerator4
    UsePositionOffset=false
    Write Material Random=true
   )
   &DynamicMaterialParameters(
    Index 0 Param 1=&RandomRangeFloat(Minimum=0, Maximum=3, Evaluation Type=::NewEnumerator1)
    Index 0 Param 2=&RandomRangeFloat(Minimum=1, Maximum=1.2, Evaluation Type=::NewEnumerator1)
    Module.Param0WriteEnabled=true
    Module.Param1WriteEnabled=true
    Module.Param2WriteEnabled=false
    Module.Param3WriteEnabled=false
   )
   &InheritVelocity(Inherited Velocity Amount Scale=<-0.05 -0.05 -0.05>, Inherited Velocity Speed Limit=0, Speed Limit=false, UseSpeedThreshold=false)
   &SkeletalMeshLocation(
    Bone / Socket Index=/CascadeToNiagaraConverter/NiagaraScripts/DynamicInputScripts/SequentialInt(Use Array Data:NiagaraBool=false, First Number:NiagaraInt32=0, Sequence:NiagaraDataInterfaceArrayInt32=?, Last Number:NiagaraInt32=1)
    Bone Sampling Mode=::NewEnumerator4
    MaskSpawnGroup=false
    Orientation Sampling=::NewEnumerator1
    Sampled Position Offset=<0 0 -15>
    Sampled Position Offset Coordinate Space:ENiagaraSkelMeshTransforms=::NewEnumerator2
    Skeletal Mesh=$System.Skeletal Mesh Reader
   ) {module=SkeletalMeshLocation001}
   &AddVelocity(Velocity=&MakeVector(X=0, Y=0, Z=&FN_FrontEndFloatModifier(Gameplay=0, Front End=150)))
  ]
  particleUpdateScript [
   &ParticleState()
   &ScaleMeshSize(Scale Factor=&VectorFromFloat(Value=&FloatFromCurve(FloatCurve=DataInterface() {propertyValues {Curve {keys [
          {arriveTangent=-0.0152, interpMode=RCIM_Cubic, leaveTangent=-0.0152, tangentMode=RCTM_User, time=0.0047, value=0.0057}
          {arriveTangent=0.0317, interpMode=RCIM_Cubic, leaveTangent=0.0317, tangentMode=RCTM_User, time=0.1171, value=1}
          {arriveTangent=-1.1326, interpMode=RCIM_Cubic, leaveTangent=-1.1326, time=0.61, value=0.9915}
          <at 1 0 RCIM_Cubic>
         ]}}})))
   &Color(Color=&MakeLinearColorFromVectorAndFloat(
     "Vector (RGB)"=&VectorFromCurve(VectorCurve=DataInterface() {propertyValues {XCurve {keys [<at 0 1>,<at 1 1>]}, YCurve {keys [<at 0 1>,<at 1 1>]}, ZCurve {keys [<at 0 1>,<at 1 1>]}}})
     "Float (Alpha)"=&ClampFloat(Float=&FloatFromCurve(FloatCurve=DataInterface() {propertyValues {Curve {keys [
           <at 0 0 RCIM_Cubic>
           {arriveTangent=0.7433, interpMode=RCIM_Cubic, leaveTangent=0.7433, time=0.102, value=0.5537}
           {arriveTangent=-0.6166, interpMode=RCIM_Cubic, leaveTangent=-0.6166, time=0.7404, value=0.5504}
           <at 1 0>
          ]}}}))
    ))
   &AccelerationForce(Acceleration=$Particles.Acceleration)
   &Drag(Drag=-0.5, Use Linear Drag=true, Use Rotational Drag=true)
   &SolveForcesAndVelocity(Clamp Velocity=false, Limit Acceleration=false)
   &DynamicMaterialParameters(
    Index 0 Param 3=1
    Index 0 Param 4=1
    Module.Param0WriteEnabled=false
    Module.Param1WriteEnabled=false
    Module.Param2WriteEnabled=true
    Module.Param3WriteEnabled=true
   ) {module=DynamicMaterialParameters001}
   /Niagara/Modules/CascadeConversion/CascadeConversion_Orbit(
    Offset:Vector3f=&RandomRangeVector(Minimum=<0 0 -25>, Maximum=<0 0 25>)
    Rotation:Vector3f=&RandomRangeVector(Minimum=<0 0 0>, Maximum=<1 1 1>)
    Rotation Rate:Vector3f=&RandomRangeVector(Minimum=<0 0 0>, Maximum=<0.5 0.5 0.5>)
   )
   /Niagara/Modules/CascadeConversion/CascadeConversion_SolveOrbit()
   &CameraOffset(Camera Offset Amount=-25)
  ]
  rendererClasses [&NiagaraMeshRendererProperties]
  renderers [{rendererClass=&NiagaraMeshRendererProperties, rendererIndex=0}]
 }
 Bats_Around {
  emitterUpdateScript [
   &EmitterState(Inactive Response=::NewEnumerator1, Life Cycle Mode=::NewEnumerator1, Loop Duration=1, UseLoopDelay=true)
   &SpawnRate(SpawnRate=&FN_FrontEndFloatModifier(Gameplay=12, Front End=5), Use Spawn Probability=false)
  ]
  particleSpawnScript [
   &InitializeParticle(
    Lifetime=&RandomRangeFloat(Minimum=0.5, Maximum=1, Evaluation Type=::NewEnumerator1)
    Mesh Scale=&VectorFromFloat(Value=&RandomRangeFloat(Minimum=0.5, Maximum=1.25, Evaluation Type=::NewEnumerator1))
    Mesh Scale Mode=::NewEnumerator4
    UsePositionOffset=false
    Write Material Random=true
   )
   &DynamicMaterialParameters(
    Index 0 Param 1=&RandomRangeFloat(Minimum=0, Maximum=3, Evaluation Type=::NewEnumerator1)
    Index 0 Param 2=&RandomRangeFloat(Minimum=1, Maximum=1.2, Evaluation Type=::NewEnumerator1)
    Module.Param0WriteEnabled=true
    Module.Param1WriteEnabled=true
    Module.Param2WriteEnabled=false
    Module.Param3WriteEnabled=false
   )
   /CascadeToNiagaraConverter/NiagaraScripts/ModuleScripts/AddVectorToPosition(Vector:Vector3f=&TransformVector(Vector=<0 0 50>, Destination Space=::Simulation))
   &InheritVelocity(Inherited Velocity Amount Scale=<-0.3 -0.3 -0.1>, Inherited Velocity Speed Limit=0, Speed Limit=false, UseSpeedThreshold=false)
  ]
  particleUpdateScript [
   &ParticleState()
   &ScaleMeshSize(Scale Factor=&VectorFromFloat(Value=&FloatFromCurve(FloatCurve=DataInterface() {propertyValues {Curve {keys [
          {arriveTangent=-0.0152, interpMode=RCIM_Cubic, leaveTangent=-0.0152, tangentMode=RCTM_User, time=0.0047, value=0.0057}
          {arriveTangent=0.0317, interpMode=RCIM_Cubic, leaveTangent=0.0317, tangentMode=RCTM_User, time=0.1171, value=1}
          {arriveTangent=-1.1326, interpMode=RCIM_Cubic, leaveTangent=-1.1326, time=0.61, value=0.9915}
          <at 1 0 RCIM_Cubic>
         ]}}})))
   &Color(Color=&MakeLinearColorFromVectorAndFloat(
     "Vector (RGB)"=&VectorFromCurve(VectorCurve=DataInterface() {propertyValues {XCurve {keys [<at 0 1>,<at 1 1>]}, YCurve {keys [<at 0 1>,<at 1 1>]}, ZCurve {keys [<at 0 1>,<at 1 1>]}}})
     "Float (Alpha)"=&ClampFloat(Float=&FloatFromCurve(FloatCurve=DataInterface() {propertyValues {Curve {keys [
           <at 0 0 RCIM_Cubic>
           {arriveTangent=0.7433, interpMode=RCIM_Cubic, leaveTangent=0.7433, time=0.102, value=0.5537}
           {arriveTangent=-0.6166, interpMode=RCIM_Cubic, leaveTangent=-0.6166, time=0.7404, value=0.5504}
           <at 1 0>
          ]}}}))
    ))
   &DynamicMaterialParameters(
    Index 0 Param 3=1
    Index 0 Param 4=1
    Module.Param0WriteEnabled=false
    Module.Param1WriteEnabled=false
    Module.Param2WriteEnabled=true
    Module.Param3WriteEnabled=true
   ) {module=DynamicMaterialParameters001}
   &Drag(Use Linear Drag=true, Use Rotational Drag=true)
   &SolveForcesAndVelocity(Clamp Velocity=false, Limit Acceleration=false)
   &CascadeConversion_Orbit(
    Offset:Vector3f=<0 0 150>
    Rotation:Vector3f=&RandomRangeVector(Minimum=<0 0 0>, Maximum=<1 1 1>)
    Rotation Rate:Vector3f=&RandomRangeVector(Minimum=<0 0 0>, Maximum=<0 0 0.5>)
   )
   &CascadeConversion_SolveOrbit()
   &CameraOffset(Camera Offset Amount=-25)
  ]
  rendererClasses [&NiagaraMeshRendererProperties]
  renderers [{rendererClass=&NiagaraMeshRendererProperties, rendererIndex=0}]
 }
 Bats_Attached {
  emitterUpdateScript [
   &EmitterState(Inactive Response=::NewEnumerator1, Life Cycle Mode=::NewEnumerator1, Loop Duration=1, UseLoopDelay=true)
   &SpawnRate(SpawnRate=10, Use Spawn Probability=false)
  ]
  particleSpawnScript [
   &InitializeParticle(
    Lifetime=&RandomRangeFloat(Minimum=0.5, Maximum=1, Evaluation Type=::NewEnumerator1)
    Mesh Scale=&VectorFromFloat(Value=&RandomRangeFloat(Minimum=0.5, Maximum=1.25, Evaluation Type=::NewEnumerator1))
    Mesh Scale Mode=::NewEnumerator4
    UsePositionOffset=false
    Write Material Random=true
   )
   &DynamicMaterialParameters(
    Index 0 Param 1=&RandomRangeFloat(Minimum=0, Maximum=3, Evaluation Type=::NewEnumerator1)
    Index 0 Param 2=&RandomRangeFloat(Minimum=1, Maximum=1.2, Evaluation Type=::NewEnumerator1)
    Module.Param0WriteEnabled=true
    Module.Param1WriteEnabled=true
    Module.Param2WriteEnabled=false
    Module.Param3WriteEnabled=false
   )
   &SkeletalMeshLocation(
    Bone / Socket Index=&SequentialInt(Use Array Data:NiagaraBool=false, First Number:NiagaraInt32=0, Sequence:NiagaraDataInterfaceArrayInt32=?, Last Number:NiagaraInt32=3)
    Bone Sampling Mode=::NewEnumerator4
    MaskSpawnGroup=false
    Orientation Sampling=::NewEnumerator1
    Sampled Position Offset Coordinate Space:ENiagaraSkelMeshTransforms=::NewEnumerator2
    Skeletal Mesh=$System.Skeletal Mesh Reader
   ) {module=SkeletalMeshLocation001}
   &SkeletalMeshLocation(
    Bone / Socket Index=&SequentialInt(Use Array Data:NiagaraBool=false, First Number:NiagaraInt32=0, Sequence:NiagaraDataInterfaceArrayInt32=?, Last Number:NiagaraInt32=3)
    Bone Sampling Mode=::NewEnumerator4
    MaskSpawnGroup=false
    Orientation Sampling=::NewEnumerator1
    Skeletal Mesh=DataInterface() {propertyValues {
      DefaultMesh=null
      ExcludeBoneName=root
      OptionalComponentClass=null
      PreviewMesh=null
      SkinningMode=SkinOnTheFly
      SoftSourceActor=null
      SourceMode=Default
      UvSetIndex=0
      WholeMeshLOD=-1
      bExcludeBone=false
      bReadDeformedGeometry=true
      bRequireCurrentFrameData=true
      ComponentTags []
      FilteredBones [ball_r,ball_l,hand_r,hand_l]
      FilteredSockets []
      MeshUserParameter {parameter {name=None, typeDefHandle {registeredTypeIndex=1}, varData []}}
      SamplingRegions []
     }}
   ) {enabled=false}
   &AddVectorToPosition(Vector:Vector3f=&TransformVector(
     Vector=&RandomRangeVector(Minimum=<0 0 -5>, Maximum=<0 0 -20>, Evaluation Type=::NewEnumerator1)
     Destination Space=::Simulation
    ))
   &AddVelocity(
    Cone Angle=&RandomRangeFloat(Minimum=0, Maximum=1, Evaluation Type=::NewEnumerator1)
    Cone Angle Mode:ENiagara_AngleInput=::NewEnumerator1
    Cone Axis=<0 0 1>
    Module.UseConeFalloff=false
    Module.UseVelDistribution=false
    Override Seed=false
    Velocity Mode=::NewEnumerator2
    Velocity Speed=&RandomRangeFloat(Minimum=2, Maximum=4, Evaluation Type=::NewEnumerator1)
   )
   &AddVelocity(Velocity=&MakeVector(X=0, Y=0, Z=&FN_FrontEndFloatModifier(Gameplay=0, Front End=70))) {module=AddVelocity001}
  ]
  particleUpdateScript [
   &ParticleState()
   &ScaleMeshSize(Scale Factor=&VectorFromFloat(Value=&FloatFromCurve(FloatCurve=DataInterface() {propertyValues {Curve {keys [
          {arriveTangent=-0.0152, interpMode=RCIM_Cubic, leaveTangent=-0.0152, tangentMode=RCTM_User, time=0.0047, value=0.0057}
          {arriveTangent=0.0317, interpMode=RCIM_Cubic, leaveTangent=0.0317, tangentMode=RCTM_User, time=0.1171, value=1}
          {arriveTangent=-1.1326, interpMode=RCIM_Cubic, leaveTangent=-1.1326, time=0.61, value=0.9915}
          <at 1 0 RCIM_Cubic>
         ]}}})))
   &Color(Color=&MakeLinearColorFromVectorAndFloat(
     "Vector (RGB)"=&VectorFromCurve(VectorCurve=DataInterface() {propertyValues {XCurve {keys [<at 0 1>,<at 1 1>]}, YCurve {keys [<at 0 1>,<at 1 1>]}, ZCurve {keys [<at 0 1>,<at 1 1>]}}})
     "Float (Alpha)"=&ClampFloat(Float=&FloatFromCurve(FloatCurve=DataInterface() {propertyValues {Curve {keys [
           <at 0 0 RCIM_Cubic>
           {arriveTangent=0.7433, interpMode=RCIM_Cubic, leaveTangent=0.7433, time=0.102, value=0.5537}
           {arriveTangent=-0.6166, interpMode=RCIM_Cubic, leaveTangent=-0.6166, time=0.7404, value=0.5504}
           <at 1 0>
          ]}}}))
    ))
   &DynamicMaterialParameters(
    Index 0 Param 3=1
    Index 0 Param 4=1
    Module.Param0WriteEnabled=false
    Module.Param1WriteEnabled=false
    Module.Param2WriteEnabled=true
    Module.Param3WriteEnabled=true
   ) {module=DynamicMaterialParameters001}
   &CascadeConversion_Orbit(
    Offset:Vector3f=<0 0 10>
    Rotation:Vector3f=&RandomRangeVector(Minimum=<0 0 0>, Maximum=<1 1 1>)
    Rotation Rate:Vector3f=&RandomRangeVector(Minimum=<0 0 0.5>, Maximum=<0 0 1>)
   )
   &CascadeConversion_SolveOrbit()
   &CameraOffset(Camera Offset Amount=-25)
   &SolveForcesAndVelocity(Clamp Velocity=false, Limit Acceleration=false)
  ]
  rendererClasses [&NiagaraMeshRendererProperties]
  renderers [{rendererClass=&NiagaraMeshRendererProperties, rendererIndex=0}]
 }
 Feet {
  emitterUpdateScript [
   &EmitterState(Inactive Response=::NewEnumerator1, Life Cycle Mode=::NewEnumerator1, Loop Duration=1, UseLoopDelay=true)
   &SpawnRate(SpawnRate=80, Use Spawn Probability=false)
  ]
  particleSpawnScript [
   &FN_SetSpawnGroup()
   /SetVariables(Particles.RibbonID=/Niagara/DynamicInputs/Meshes/MakeRibbonIDFromSpawnGroup()) {isSetParameters=true, module=SetVariables_#2}
   &InitializeParticle(
    Lifetime=0.4
    Ribbon Width=&Multiply_Float(A=18, B=2)
    Ribbon Width Mode=::NewEnumerator1
    UsePositionOffset=false
    Write Material Random=true
   )
   &Color(Color=&MakeLinearColorFromVectorAndFloat("Vector (RGB)"=<0.3802 0.3802 0.3802>, "Float (Alpha)"=2))
   &SkeletalMeshLocation(
    Bone / Socket Index=&SequentialInt(Use Array Data:NiagaraBool=false, First Number:NiagaraInt32=2, Sequence:NiagaraDataInterfaceArrayInt32=?, Last Number:NiagaraInt32=3)
    Bone Sampling Mode=::NewEnumerator4
    MaskSpawnGroup=false
    Orientation Sampling=::NewEnumerator1
    Sampled Position Offset=<0 0 -15>
    Sampled Position Offset Coordinate Space:ENiagaraSkelMeshTransforms=::NewEnumerator2
    Skeletal Mesh=$System.Skeletal Mesh Reader
   ) {module=SkeletalMeshLocation001}
   &AddVectorToPosition(Vector:Vector3f=&TransformVector(Vector=<0 0 0>, Destination Space=::Simulation))
   &AddVelocity(Velocity=&MakeVector(X=0, Y=0, Z=&FN_FrontEndFloatModifier(Gameplay=0, Front End=1200)))
  ]
  particleUpdateScript [
   &ParticleState()
   &ScaleColor(
    Scale Alpha=&FloatFromCurve(FloatCurve=DataInterface() {propertyValues {Curve {keys [
         {arriveTangent=2.0102, time=0, value=0}
         {arriveTangent=-0.5102, leaveTangent=-0.5102, time=0.0349, value=2.0102}
         {arriveTangent=-1.5, leaveTangent=-1.5, time=0.4185, value=1.5}
         <at 0.7 0>
        ]}}})
    Scale RGB=&VectorFromFloat(Value=&FloatFromCurve(FloatCurve=DataInterface() {propertyValues {Curve {keys [
          {arriveTangent=-5.5, interpMode=RCIM_Cubic, leaveTangent=-51.4373, tangentMode=RCTM_Break, time=0, value=8}
          {arriveTangent=-14.8011, interpMode=RCIM_Cubic, leaveTangent=-13.3728, tangentMode=RCTM_Break, time=0.1, value=3}
          <at 0.19 0>
         ]}}}))
    ScaleA=true
    ScaleRGB=true
   )
   &DynamicMaterialParameters(
    Index 0 Param 1=0.5
    Index 0 Param 2=15
    Index 0 Param 3=&FloatFromCurve(FloatCurve=DataInterface() {propertyValues {Curve {keys [
         {arriveTangent=-1.8399, interpMode=RCIM_Cubic, leaveTangent=-1.8399, tangentMode=RCTM_User, time=-0.0058, value=1.367}
         {arriveTangent=-3.8231, interpMode=RCIM_Cubic, leaveTangent=-3.8231, time=0.2238, value=0.7705}
         {arriveTangent=-0.9926, interpMode=RCIM_Cubic, leaveTangent=-0.9926, time=0.5993, value=-0.9465}
         <at 1 0>
        ]}}})
    Index 0 Param 4=1
    Module.Param0WriteEnabled=true
    Module.Param1WriteEnabled=true
    Module.Param2WriteEnabled=true
    Module.Param3WriteEnabled=true
   )
   &SolveForcesAndVelocity(Clamp Velocity=false, Limit Acceleration=false)
  ]
  rendererClasses [&NiagaraRibbonRendererProperties]
  renderers [{rendererClass=&NiagaraRibbonRendererProperties}]
 }
 Hands {
  emitterUpdateScript [
   &EmitterState(Inactive Response=::NewEnumerator1, Life Cycle Mode=::NewEnumerator1, Loop Duration=1, UseLoopDelay=true)
   &SpawnRate(SpawnRate=80, Use Spawn Probability=false)
  ]
  particleSpawnScript [
   &FN_SetSpawnGroup()
   /SetVariables(Particles.RibbonID=&MakeRibbonIDFromSpawnGroup()) {isSetParameters=true, module=SetVariables_#2}
   &InitializeParticle(
    Lifetime=0.4
    Ribbon Width=&Multiply_Float(A=18, B=2)
    Ribbon Width Mode=::NewEnumerator1
    UsePositionOffset=false
    Write Material Random=true
   )
   &Color(Color=&MakeLinearColorFromVectorAndFloat("Vector (RGB)"=<0.3802 0.3802 0.3802>, "Float (Alpha)"=2))
   &SkeletalMeshLocation(
    Bone / Socket Index=&SequentialInt(Use Array Data:NiagaraBool=false, First Number:NiagaraInt32=0, Sequence:NiagaraDataInterfaceArrayInt32=?, Last Number:NiagaraInt32=1)
    Bone Sampling Mode=::NewEnumerator4
    MaskSpawnGroup=false
    Orientation Sampling=::NewEnumerator1
    Sampled Position Offset=<0 0 -15>
    Sampled Position Offset Coordinate Space:ENiagaraSkelMeshTransforms=::NewEnumerator2
    Skeletal Mesh=$System.Skeletal Mesh Reader
   ) {module=SkeletalMeshLocation001}
   &AddVectorToPosition(Vector:Vector3f=&TransformVector(Vector=<0 0 0>, Destination Space=::Simulation))
   &AddVelocity(Velocity=&MakeVector(X=0, Y=0, Z=&FN_FrontEndFloatModifier(Gameplay=0, Front End=1200)))
  ]
  particleUpdateScript [
   &ParticleState()
   &ScaleColor(
    Scale Alpha=&FloatFromCurve(FloatCurve=DataInterface() {propertyValues {Curve {keys [
         {arriveTangent=2.0102, time=0, value=0}
         {arriveTangent=-0.5102, leaveTangent=-0.5102, time=0.0349, value=2.0102}
         {arriveTangent=-1.5, leaveTangent=-1.5, time=0.4185, value=1.5}
         <at 0.7 0>
        ]}}})
    Scale RGB=&VectorFromFloat(Value=&FloatFromCurve(FloatCurve=DataInterface() {propertyValues {Curve {keys [
          {arriveTangent=-5.5, interpMode=RCIM_Cubic, leaveTangent=-51.4373, tangentMode=RCTM_Break, time=0, value=8}
          {arriveTangent=-14.8011, interpMode=RCIM_Cubic, leaveTangent=-13.3728, tangentMode=RCTM_Break, time=0.1, value=3}
          <at 0.19 0>
         ]}}}))
    ScaleA=true
    ScaleRGB=true
   )
   &DynamicMaterialParameters(
    Index 0 Param 1=0.5
    Index 0 Param 2=15
    Index 0 Param 3=&FloatFromCurve(FloatCurve=DataInterface() {propertyValues {Curve {keys [
         {arriveTangent=-1.8399, interpMode=RCIM_Cubic, leaveTangent=-1.8399, tangentMode=RCTM_User, time=-0.0058, value=1.367}
         {arriveTangent=-3.8231, interpMode=RCIM_Cubic, leaveTangent=-3.8231, time=0.2238, value=0.7705}
         {arriveTangent=-0.9926, interpMode=RCIM_Cubic, leaveTangent=-0.9926, time=0.5993, value=-0.9465}
         <at 1 0>
        ]}}})
    Index 0 Param 4=1
    Module.Param0WriteEnabled=true
    Module.Param1WriteEnabled=true
    Module.Param2WriteEnabled=true
    Module.Param3WriteEnabled=true
   )
   &SolveForcesAndVelocity(Clamp Velocity=false, Limit Acceleration=false)
  ]
  rendererClasses [&NiagaraRibbonRendererProperties]
  renderers [{rendererClass=&NiagaraRibbonRendererProperties}]
 }
 SmokeLoop {
  emitterUpdateScript [
   &EmitterState(
    Delay First Loop Only=true
    Inactive Response=::NewEnumerator1
    Life Cycle Mode=::NewEnumerator1
    Loop Delay=0.2
    Loop Duration=0.5
    UseLoopDelay=true
   )
   &SpawnRate(SpawnRate=12, Use Spawn Probability=false)
  ]
  particleSpawnScript [
   /SetVariables(Particles.SpriteRotationRate=&RandomRangeFloat(Minimum=0.05, Maximum=-0.05, Evaluation Type=::NewEnumerator1)) {isSetParameters=true}
   &InitializeParticle(
    Lifetime=&RandomRangeFloat(Minimum=0.5, Maximum=1, Evaluation Type=::NewEnumerator1)
    Sprite Rotation Angle=&RandomRangeFloat(Minimum=1, Maximum=0, Evaluation Type=::NewEnumerator1)
    Sprite Rotation Mode=::NewEnumerator1
    Sprite Size=<20 20>
    Sprite Size Mode=::NewEnumerator4
    UsePositionOffset=false
    Write Material Random=true
   )
   &Color(Color=&MakeLinearColorFromVectorAndFloat("Vector (RGB)"=<0.12 0.0447 0.0195>, "Float (Alpha)"=1))
   &DynamicMaterialParameters(
    Index 0 Param 2=&RandomRangeFloat(Minimum=0, Maximum=5, Evaluation Type=::NewEnumerator1)
    Module.Param0WriteEnabled=false
    Module.Param1WriteEnabled=true
    Module.Param2WriteEnabled=false
    Module.Param3WriteEnabled=false
   ) {module=DynamicMaterialParameters001}
  ]
  particleUpdateScript [
   &ParticleState()
   &SpriteRotationRate(Rotation Rate=&AngleConversion(
     Angle:NiagaraFloat=$Particles.SpriteRotationRate
     Angle Input:ENiagara_AngleInput=::NewEnumerator1
     Angle Output:ENiagara_AngleInput=::NewEnumerator0
    ))
   &ScaleSpriteSize(
    Scale Sprite Size Mode=::NewEnumerator0
    Uniform Scale Factor=&FloatFromCurve(FloatCurve=DataInterface() {propertyValues {Curve {keys [<at 0.0014 0.4978 RCIM_Cubic>,<at 0.992 3.4972 RCIM_Cubic>]}}})
   )
   &ScaleColor(
    Scale Alpha=&FloatFromCurve(FloatCurve=DataInterface() {propertyValues {Curve {keys [
         {arriveTangent=2.6336, interpMode=RCIM_Cubic, leaveTangent=2.6336, tangentMode=RCTM_User, time=0.003, value=-0.0044}
         {arriveTangent=0.1001, interpMode=RCIM_Cubic, leaveTangent=0.1001, tangentMode=RCTM_User, time=0.0779, value=2.2738}
         {arriveTangent=-3.4049, interpMode=RCIM_Cubic, leaveTangent=-2.5614, tangentMode=RCTM_Break, time=0.5867, value=0.8506}
         <at 1 0>
        ]}}})
    Scale RGB=&VectorFromFloat(Value=&FloatFromCurve(FloatCurve=DataInterface() {propertyValues {Curve {keys [
          <at -0.0009 0.0043 RCIM_Cubic>
          {arriveTangent=-0.0043, interpMode=RCIM_Cubic, leaveTangent=-0.0043, time=0.2049, value=1}
          <at 1 0 RCIM_Cubic>
         ]}}}))
    ScaleA=true
    ScaleRGB=true
   )
   &SkeletalMeshLocation(
    Bone / Socket Index=&SequentialInt(Use Array Data:NiagaraBool=false, First Number:NiagaraInt32=0, Sequence:NiagaraDataInterfaceArrayInt32=?, Last Number:NiagaraInt32=3)
    Bone Sampling Mode=::NewEnumerator4
    MaskSpawnGroup=false
    Orientation Sampling=::NewEnumerator1
    Skeletal Mesh=DataInterface() {propertyValues {
      DefaultMesh=null
      ExcludeBoneName=root
      OptionalComponentClass=null
      PreviewMesh=null
      SkinningMode=SkinOnTheFly
      SoftSourceActor=null
      SourceMode=Default
      UvSetIndex=0
      WholeMeshLOD=-1
      bExcludeBone=false
      bReadDeformedGeometry=true
      bRequireCurrentFrameData=true
      ComponentTags []
      FilteredBones [ball_r,ball_l,hand_r,hand_l]
      FilteredSockets []
      MeshUserParameter {parameter {name=None, typeDefHandle {registeredTypeIndex=1}, varData []}}
      SamplingRegions []
     }}
   )
   &DynamicMaterialParameters(
    Index 0 Param 1=&FloatFromCurve(FloatCurve=DataInterface() {propertyValues {Curve {keys [{arriveTangent=3.783, interpMode=RCIM_Cubic, leaveTangent=3.783, tangentMode=RCTM_User, time=0, value=0}, <at 1 1.3296 RCIM_Cubic>]}}})
    Index 0 Param 3=1
    Index 0 Param 4=1
    Module.Param0WriteEnabled=true
    Module.Param1WriteEnabled=false
    Module.Param2WriteEnabled=true
    Module.Param3WriteEnabled=true
   )
  ]
  rendererClasses [&NiagaraSpriteRendererProperties]
  renderers [{rendererClass=&NiagaraSpriteRendererProperties}]
 }
}
summary {
 systemName=NS_Contrail_Bats_Converted
 emitters [
  {emitterName=SmokeLoop, rendererClasses [&NiagaraSpriteRendererProperties]}
  {emitterName=Bats_Around, rendererClasses [&NiagaraMeshRendererProperties]}
  {emitterName=Bats_Attached, rendererClasses [&NiagaraMeshRendererProperties]}
  {emitterName=Hands, rendererClasses [&NiagaraRibbonRendererProperties]}
  {emitterName=Feet, rendererClasses [&NiagaraRibbonRendererProperties]}
  {emitterName=BatsTrail1001, rendererClasses [&NiagaraMeshRendererProperties]}
 ]
 userVariables [
  User.bIsFrontEnd {type=NiagaraBool}
  User.bIsFrontEndPreview {type=NiagaraBool}
  User.New Skeletal Mesh {
   type=NiagaraDataInterfaceSkeletalMesh
   defaultValue {
    struct=NiagaraExt_VariableValue_DataInterface
    value {dataInterface=&self:NiagaraDataInterfaceSkeletalMesh_1}
   }
  }
  User.ParaGlideLeanAlpha {type=NiagaraFloat, defaultValue=0, description=""}
  User.SK_Mesh {
   type=NiagaraDataInterfaceSkeletalMesh
   defaultValue {
    struct=NiagaraExt_VariableValue_DataInterface
    value {dataInterface=&self:NiagaraDataInterfaceSkeletalMesh_0}
   }
  }
 ]
}
