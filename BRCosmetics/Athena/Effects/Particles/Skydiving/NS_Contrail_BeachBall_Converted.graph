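# Niagara skydiving contrail cosmetic, converted from Cascade, that trails beach balls and glints behind a falling character.
# BeachBall slowly spawns tumbling, orbiting meshes with upward acceleration, scale pop and alpha fade, randomizing dynamic material parameters.
# FallingDust keeps small fading meshes pinned to skeletal bones or sockets by resampling them every update.
# HandLights, the only GPU emitter, launches bright orange sprites from sampled bones or sockets, flickering through a size curve.
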NiagaraSystem /BRCosmetics/Athena/Effects/Particles/Skydiving/NS_Contrail_BeachBall_Converted
dependencies {
 usedDataInterfaces [&NiagaraDataInterfaceCurve, &NiagaraDataInterfaceVectorCurve, &NiagaraDataInterfaceSkeletalMesh]
 usedDynamicInputs []
 usedModules [
  &SystemState
  &self:SystemSpawnScript.SystemScriptSource.SystemScriptGraph.NiagaraNodeAssignment_0.SetVariables_#1
 ]
}
emitters {
 BeachBall {
  emitterUpdateScript [
   &EmitterState(Inactive Response=::NewEnumerator1, Life Cycle Mode=::NewEnumerator1, Loop Duration=1, UseLoopDelay=true)
   &SpawnRate(SpawnRate=&Multiply_Float(A=3, B=0.45), Use Spawn Probability=false)
  ]
  particleSpawnScript [
   /SetVariables(Particles.MeshRotationEuler:Vector3f=&RandomRangeVector(Minimum=<0 0 0>, Maximum=<1 1 1>, Evaluation Type=::NewEnumerator1)) {isSetParameters=true}
   /SetVariables(Particles.MeshOrientation=/CascadeToNiagaraConverter/NiagaraScripts/DynamicInputScripts/MakeQuatFromEuler(Euler:Vector3f=$Particles.MeshRotationEuler, Scale:NiagaraFloat=360)) {isSetParameters=true}
   /SetVariables(Particles.MeshRotationRateBase:Vector3f=&RandomRangeVector(Minimum=<-1 -1 -1>, Maximum=<1 1 1>, Evaluation Type=::NewEnumerator1)) {isSetParameters=true}
   /SetVariables(Particles.Acceleration:Vector3f=<0 0 50>) {isSetParameters=true}
   &InitializeParticle(
    Lifetime=&RandomRangeFloat(Minimum=2, Maximum=4, Evaluation Type=::NewEnumerator1)
    Mesh Scale=&VectorFromFloat(Value=&RandomRangeFloat(Minimum=0.6, Maximum=0.9, Evaluation Type=::NewEnumerator1))
    Mesh Scale Mode=::NewEnumerator4
    UsePositionOffset=false
    Write Material Random=true
   )
   &InheritVelocity(Inherited Velocity Amount Scale=<-0.01 -0.01 -0.01>, Inherited Velocity Speed Limit=0, Speed Limit=false, UseSpeedThreshold=false)
   &Color(Color=&MakeLinearColorFromVectorAndFloat("Vector (RGB)"=<0.6684 0.017 1>, "Float (Alpha)"=0.8))
   &DynamicMaterialParameters(
    Index 0 Param 1=&RandomRangeFloat(Minimum=0, Maximum=10, Evaluation Type=::NewEnumerator1)
    Index 0 Param 2=&RandomRangeFloat(Minimum=4, Maximum=7, Evaluation Type=::NewEnumerator1)
    Index 0 Param 3=&RandomRangeFloat(Minimum=0.8, Maximum=1.5, Evaluation Type=::NewEnumerator1)
    Index 0 Param 4=&RandomRangeFloat(Minimum=-0.6, Maximum=0.6, Evaluation Type=::NewEnumerator1)
    Module.Param0WriteEnabled=true
    Module.Param1WriteEnabled=true
    Module.Param2WriteEnabled=true
    Module.Param3WriteEnabled=true
   )
   /CascadeToNiagaraConverter/NiagaraScripts/ModuleScripts/AddVectorToPosition(Vector:Vector3f=&TransformVector(Vector=<0 0 0>, Destination Space=::Simulation))
   &AddVelocity()
  ]
  particleUpdateScript [
   /SetVariables(Particles.MeshRotationRate:Vector3f=$Particles.MeshRotationRateBase) {isSetParameters=true}
   &ParticleState()
   &ScaleMeshSize(Scale Factor=&VectorFromFloat(Value=&FloatFromCurve(FloatCurve=DataInterface() {propertyValues {Curve {keys [
          <at 0 0 RCIM_Cubic>
          {arriveTangent=5.1804, interpMode=RCIM_Cubic, leaveTangent=5.1804, time=0.0947, value=1.391}
          {arriveTangent=-1.4759, interpMode=RCIM_Cubic, leaveTangent=-1.4759, tangentMode=RCTM_User, time=0.1928, value=0.9989}
          {arriveTangent=1.8075, interpMode=RCIM_Cubic, leaveTangent=20.6058, tangentMode=RCTM_Break, time=0.8896, value=0.9999}
          {arriveTangent=0.0709, interpMode=RCIM_Cubic, leaveTangent=0.0709, tangentMode=RCTM_User, time=1, value=0.0047}
         ]}}})))
   &AccelerationForce(Acceleration=$Particles.Acceleration)
   &SolveForcesAndVelocity(Clamp Velocity=false, Limit Acceleration=false)
   &ScaleColor(
    Scale Alpha=&FloatFromCurve(FloatCurve=DataInterface() {propertyValues {Curve {keys [<at 0 0 RCIM_Cubic>,<at 0.5 1 RCIM_Cubic>,<at 1 0 RCIM_Cubic>]}}})
    Scale RGB=&VectorFromCurve(VectorCurve=DataInterface() {propertyValues {XCurve {keys [<at 0 1>,<at 1 1>]}, YCurve {keys [<at 0 1>,<at 1 1>]}, ZCurve {keys [<at 0 1>,<at 1 1>]}}})
    ScaleA=true
    ScaleRGB=true
   )
   /Niagara/Modules/CascadeConversion/CascadeConversion_Orbit(
    Offset:Vector3f=&RandomRangeVector(Minimum=<0 80 95>, Maximum=<0 95 110>)
    Rotation:Vector3f=&RandomRangeVector(Minimum=<0 0 0>, Maximum=<0 0 1>)
    Rotation Rate:Vector3f=&VectorFromCurve(VectorCurve=DataInterface() {propertyValues {
       XCurve {keys [{interpMode=RCIM_Cubic, tangentMode=RCTM_User, time=0, value=0}, {interpMode=RCIM_Cubic, tangentMode=RCTM_User, time=0.997, value=0}]}
       YCurve {keys [{interpMode=RCIM_Cubic, tangentMode=RCTM_User, time=0, value=0}, {interpMode=RCIM_Cubic, tangentMode=RCTM_User, time=0.997, value=0}]}
       ZCurve {keys [
         {arriveTangent=0.0683, interpMode=RCIM_Cubic, leaveTangent=0.0683, tangentMode=RCTM_User, time=0, value=0.0998}
         {arriveTangent=0.0407, interpMode=RCIM_Cubic, leaveTangent=0.0407, tangentMode=RCTM_User, time=0.997, value=0.1254}
        ]}
      }})
   )
   /Niagara/Modules/CascadeConversion/CascadeConversion_SolveOrbit()
   &CameraOffset(Camera Offset Amount=-25)
   /CascadeToNiagaraConverter/NiagaraScripts/ModuleScripts/UpdateMeshOrientationByEuler(RotationRate:Vector3f=$Particles.MeshRotationRate)
  ]
  rendererClasses [&NiagaraMeshRendererProperties]
  renderers [{rendererClass=&NiagaraMeshRendererProperties, rendererIndex=0}]
 }
 FallingDust {
  emitterUpdateScript [
   &EmitterState(Inactive Response=::NewEnumerator1, Life Cycle Mode=::NewEnumerator1, Loop Duration=1, UseLoopDelay=true)
   &SpawnRate(SpawnRate=4.5, Use Spawn Probability=false)
  ]
  particleSpawnScript [
   &InitializeParticle(
    Lifetime=&RandomRangeFloat(Minimum=0.5, Maximum=2, Evaluation Type=::NewEnumerator1)
    Mesh Scale=&VectorFromFloat(Value=&RandomRangeFloat(Minimum=0.03, Maximum=0.07, Evaluation Type=::NewEnumerator1))
    Mesh Scale Mode=::NewEnumerator4
    UsePositionOffset=false
    Write Material Random=true
   )
   &Color(Color=&MakeLinearColorFromVectorAndFloat("Vector (RGB)"=<1 1 1>, "Float (Alpha)"=1))
   &InitialMeshOrientation(Mesh Orientation Mode=::NewEnumerator2, Orientation Vector=<0 0 -1>, Use Rotation Vector=false)
  ]
  particleUpdateScript [
   &ParticleState()
   &ScaleColor(
    Scale Alpha=&FloatFromCurve(FloatCurve=DataInterface() {propertyValues {Curve {keys [
         <at 0 0 RCIM_Cubic>
         {arriveTangent=1.4286, interpMode=RCIM_Cubic, leaveTangent=1.4286, time=0.3, value=0.6}
         {arriveTangent=-0.8571, interpMode=RCIM_Cubic, leaveTangent=-0.8571, time=0.7, value=1}
         <at 1 0>
        ]}}})
    ScaleA=true
    ScaleRGB=true
   )
   &SkeletalMeshLocation(
    Bone / Socket Index=/CascadeToNiagaraConverter/NiagaraScripts/DynamicInputScripts/SequentialInt(Use Array Data:NiagaraBool=false, First Number:NiagaraInt32=0, Sequence:NiagaraDataInterfaceArrayInt32=?, Last Number:NiagaraInt32=1)
    Bone Sampling Mode=::NewEnumerator4
    MaskSpawnGroup=false
    Orientation Sampling=::NewEnumerator1
    Skeletal Mesh=$System.Skeletal Mesh Reader
   )
  ]
  rendererClasses [&NiagaraMeshRendererProperties]
  renderers [{rendererClass=&NiagaraMeshRendererProperties, rendererIndex=0}]
 }
 HandLights {
  simTarget=GPUComputeSim
  emitterUpdateScript [
   &EmitterState(Inactive Response=::NewEnumerator1, Life Cycle Mode=::NewEnumerator1, Loop Duration=1, UseLoopDelay=true)
   &SpawnRate(SpawnRate=30, Use Spawn Probability=false)
  ]
  particleSpawnScript [
   &InitializeParticle(
    Lifetime=&RandomRangeFloat(Minimum=0.5, Maximum=1, Evaluation Type=::NewEnumerator1)
    Sprite Size=&Vector2DFromFloat(Value=&RandomRangeFloat(Minimum=2.5, Maximum=5, Evaluation Type=::NewEnumerator1))
    Sprite Size Mode=::NewEnumerator4
    UsePositionOffset=false
    Write Material Random=true
   )
   &Color(Color=&MakeLinearColorFromVectorAndFloat("Vector (RGB)"=<1 0.3669 0>, "Float (Alpha)"=0.5))
   &SkeletalMeshLocation(
    Bone / Socket Index=&SequentialInt(Use Array Data:NiagaraBool=false, First Number:NiagaraInt32=0, Sequence:NiagaraDataInterfaceArrayInt32=?, Last Number:NiagaraInt32=1)
    Bone Sampling Mode=::NewEnumerator4
    MaskSpawnGroup=false
    Orientation Sampling=::NewEnumerator1
    Skeletal Mesh=$System.Skeletal Mesh Reader
   )
   &AddVelocity(Velocity=/Niagara/DynamicInputs/CascadeConversion/SplitVector(
     Double Component Vector Channels:ECascadeNiagaraTwoVectorChannels=::NewEnumerator0
     Double Component:NiagaraFloat=&RandomRangeFloat(Minimum=0, Maximum=0, Evaluation Type=::NewEnumerator1)
     Single Component:NiagaraFloat=&RandomRangeFloat(Minimum=50, Maximum=300, Evaluation Type=::NewEnumerator1)
    ))
   &ShapeLocation(Hemisphere Distribution=<1 0.5>, MaskSpawnGroup=false, Override Seed=false, Sphere Radius=2, UseOwnerScale=true)
   /CascadeToNiagaraConverter/NiagaraScripts/ModuleScripts/FixShapeVector()
   &AddVelocity(Velocity=&Multiply_VectorByFloat(Vector=$Output.FixShapeVector.ShapeVector, Float=5)) {module=AddVelocity001}
  ]
  particleUpdateScript [
   &ParticleState()
   &ScaleSpriteSize(
    Scale Factor=&Vector2DFromFloat(Value=&FloatFromCurve(FloatCurve=DataInterface() {propertyValues {Curve {keys [
          <at 0 0 RCIM_Cubic>
          {arriveTangent=2.1263, interpMode=RCIM_Cubic, leaveTangent=2.1263, time=0.1034, value=1.1346}
          {arriveTangent=-0.4991, interpMode=RCIM_Cubic, leaveTangent=-0.4991, time=0.2184, value=0.4643}
          {arriveTangent=-1.1951, interpMode=RCIM_Cubic, leaveTangent=-1.1951, time=0.3019, value=1.0356}
          {arriveTangent=0.1149, interpMode=RCIM_Cubic, leaveTangent=0.1149, time=0.506, value=0.1205}
          {arriveTangent=-0.2793, interpMode=RCIM_Cubic, leaveTangent=-0.2793, tangentMode=RCTM_User, time=0.6432, value=1.0748}
          {arriveTangent=-2.7506, interpMode=RCIM_Cubic, leaveTangent=-2.7506, tangentMode=RCTM_User, time=0.8021, value=0.2488}
          <at 1 0 RCIM_Cubic>
         ]}}}))
    Scale Sprite Size Mode=::NewEnumerator2
   )
   &ScaleColor(
    Scale Alpha=&FloatFromCurve(FloatCurve=DataInterface() {propertyValues {Curve {keys [
         {arriveTangent=9.9173, interpMode=RCIM_Cubic, leaveTangent=9.9173, tangentMode=RCTM_User, time=0, value=3}
         {arriveTangent=-14.9864, interpMode=RCIM_Cubic, leaveTangent=7.9419, tangentMode=RCTM_Break, time=0.3, value=1}
         {arriveTangent=-13.5796, interpMode=RCIM_Cubic, leaveTangent=1.4747, tangentMode=RCTM_Break, time=0.4562, value=1.0768}
         <at 1 0 RCIM_Cubic>
        ]}}})
    Scale RGB=&VectorFromCurve(VectorCurve=DataInterface() {propertyValues {
       XCurve {keys [<at 0 18>,<at 0.3 9>,<at 1 0.6>]}
       YCurve {keys [<at 0 18>,<at 0.3 9>,<at 1 0.6>]}
       ZCurve {keys [<at 0 18>,<at 0.3 9>,<at 1 0.6>]}
      }})
    ScaleA=true
    ScaleRGB=true
   )
   &SolveForcesAndVelocity(Clamp Velocity=false, Limit Acceleration=false)
  ]
  rendererClasses [&NiagaraSpriteRendererProperties]
  renderers [{rendererClass=&NiagaraSpriteRendererProperties}]
 }
}
summary {
 systemName=NS_Contrail_BeachBall_Converted
 emitters [
  {emitterName=BeachBall, rendererClasses [&NiagaraMeshRendererProperties]}
  {emitterName=FallingDust, rendererClasses [&NiagaraMeshRendererProperties]}
  {emitterName=HandLights, simTarget=GPUComputeSim, rendererClasses [&NiagaraSpriteRendererProperties]}
 ]
 userVariables []
}
