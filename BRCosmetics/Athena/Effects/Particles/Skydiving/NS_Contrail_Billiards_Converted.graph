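# Billiards skydiving contrail, a Niagara system converted from Cascade, that trails tumbling pool balls and faint wind meshes from a skeletal mesh.
# PoolBalls, WindRibbon and WindRings share a system mesh reader, alternating between two sampled bone or socket indices.
# Balls randomize between two size and motion variants, tag their spawn group, and use converted orbit modules.
# Several older per-hand ring, boost and vortex emitters are present but disabled.
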NiagaraSystem /BRCosmetics/Athena/Effects/Particles/Skydiving/NS_Contrail_Billiards_Converted
dependencies {
 usedDataInterfaces [
  &NiagaraDataInterfaceSkeletalMesh
  &NiagaraDataInterfaceArrayFloat3
  &NiagaraDataInterfaceCurve
  &NiagaraDataInterfaceVectorCurve
 ]
 usedDynamicInputs []
 usedModules [
  &self:SystemSpawnScript.SystemScriptSource.SystemScriptGraph.NiagaraNodeAssignment_0.SetVariables_#1
  &SystemState
 ]
}
emitters {
 PoolBalls {
  emitterUpdateScript [
   &EmitterState(
    Enable Distance Culling=true
    Life Cycle Mode=::NewEnumerator1
    Loop Behavior=::NewEnumerator1
    Loop Duration Mode=::NewEnumerator1
    Module.bUseMaxDistance=false
    Module.bUseMinDistance=false
    Scalability Mode=::NewEnumerator1
    UseLoopDelay=false
   )
   &SpawnRate(SpawnGroup=1, SpawnRate=2, Use Spawn Probability=false) {module=SpawnRate001}
   &SpawnRate(SpawnRate=10, Use Spawn Probability=false)
   &SpawnBurst_Instantaneous(Spawn Count=5, Use Spawn Probability=false, UseLoopCountLimit=false)
  ]
  particleSpawnScript [
   /SetVariables(
    Particles.Rand:NiagaraFloat=&UniformAOrBFloat(A=0, Evaluation Type=::NewEnumerator0, Recalculate Random Each Loop=true, Override Seed=false)
    Particles.VisibilityTag=&SpawnGroup()
   ) {isSetParameters=true}
   /SetVariables(Particles.MeshRotationRateBase:Vector3f=&RandomRangeVector(Minimum=<0 0 0>, Maximum=<0.6 0.6 0.6>, Evaluation Type=::NewEnumerator1)) {isSetParameters=true}
   /SetVariables(Particles.MeshRotationEuler:Vector3f=&RandomRangeVector(Minimum=<0 0 0>, Maximum=<1 1 1>, Evaluation Type=::NewEnumerator1)) {isSetParameters=true}
   /SetVariables(Particles.MeshOrientation=/CascadeToNiagaraConverter/NiagaraScripts/DynamicInputScripts/MakeQuatFromEuler(Euler:Vector3f=$Particles.MeshRotationEuler, Scale:NiagaraFloat=360)) {isSetParameters=true}
   /SetVariables(Particles.Acceleration:Vector3f=&Lerp_Vector(
     A=&RandomRangeVector(Minimum=<0 0 100>, Maximum=<0 0 700>, Evaluation Type=::NewEnumerator1)
     B=&RandomRangeVector(Minimum=<0 0 50>, Maximum=<0 0 350>, Evaluation Type=::NewEnumerator1)
     Alpha=$Particles.Rand
    )) {isSetParameters=true}
   /SetVariables(Particles.SkeletonSamplingIndex:NiagaraInt32=/CascadeToNiagaraConverter/NiagaraScripts/DynamicInputScripts/SequentialInt(Use Array Data:NiagaraBool=false, First Number:NiagaraInt32=0, Sequence:NiagaraDataInterfaceArrayInt32=?, Last Number:NiagaraInt32=1)) {isSetParameters=true}
   &InitializeParticle(
    Lifetime=&Lerp_Float(
     A=&RandomRangeFloat(Minimum=2, Maximum=3, Evaluation Type=::NewEnumerator1)
     B=&RandomRangeFloat(Minimum=1, Maximum=2, Evaluation Type=::NewEnumerator1)
     Alpha=$Particles.Rand
    )
    Mesh Scale Mode=::NewEnumerator3
    Mesh Uniform Scale=&Lerp_Float(A=0.12, B=0.08, Alpha=$Particles.Rand)
    UsePositionOffset=false
    Write Material Random=true
   )
   &AddVelocity(Velocity=&Lerp_Vector(
     A=&RandomRangeVector(Minimum=<0 0 150>, Maximum=<0 0 350>, Evaluation Type=::NewEnumerator1)
     B=&RandomRangeVector(Minimum=<0 0 160>, Maximum=<0 0 450>, Evaluation Type=::NewEnumerator1)
     Alpha=$Particles.Rand
    ))
   &SkeletalMeshLocation(
    Bone / Socket Index=$Particles.SkeletonSamplingIndex
    Bone Sampling Mode=::NewEnumerator4
    MaskSpawnGroup=false
    Orientation Sampling=::NewEnumerator1
    Sampled Position Offset=<6 0 -20>
    Sampled Position Offset Coordinate Space:ENiagaraSkelMeshTransforms=::NewEnumerator2
    Skeletal Mesh=$System.Skeletal Mesh Reader
   )
   &DynamicMaterialParameters(
    Index 0 Param 1=&RandomRangeFloat(Minimum=0, Maximum=1, Evaluation Type=::NewEnumerator1)
    Index 0 Param 2=&RandomRangeFloat(Minimum=0, Maximum=1, Evaluation Type=::NewEnumerator1)
    Index 0 Param 3=1
    Index 0 Param 4=1
    Module.Index0Float0Write=true
    Module.Index0Float1Write=true
    Module.Index0Float2Write=true
    Module.Index0Float3Write=true
   )
   /CascadeToNiagaraConverter/NiagaraScripts/ModuleScripts/AddVectorToPosition(Vector:Vector3f=&TransformVector(
     Vector=/Niagara/DynamicInputs/CascadeConversion/SplitVector(
      Double Component Vector Channels:ECascadeNiagaraTwoVectorChannels=::NewEnumerator0
      Double Component:NiagaraFloat=&RandomRangeFloat(Minimum=-10, Maximum=10, Evaluation Type=::NewEnumerator1)
      Single Component:NiagaraFloat=&RandomRangeFloat(Minimum=0, Maximum=0, Evaluation Type=::NewEnumerator1)
     )
     Destination Space=::Simulation
    ))
   &InheritVelocity(Inherited Velocity Amount Scale=<0 0 -0.09>, Inherited Velocity Speed Limit=0, Speed Limit=false, UseSpeedThreshold=false)
  ]
  particleUpdateScript [
   /SetVariables(Particles.MeshRotationRate:Vector3f=$Particles.MeshRotationRateBase) {isSetParameters=true}
   &ParticleState()
   &ScaleMeshSize(Scale Factor=&VectorFromFloat(Value=&FloatFromCurve(FloatCurve=DataInterface() {propertyValues {Curve {keys [
          <at 0.034 0>
          <at 0.035 0.3>
          <at 0.06 1.2>
          <at 0.08 1>
          <at 0.9 1>
          {arriveTangent=-20, interpMode=RCIM_Cubic, leaveTangent=-20, tangentMode=RCTM_User, time=1, value=0}
         ]}}})))
   &Color(Color=&MakeLinearColorFromVectorAndFloat(
     "Vector (RGB)"=&VectorFromCurve(VectorCurve=DataInterface() {propertyValues {XCurve {keys [<at 0 1>,<at 1 1>]}, YCurve {keys [<at 0 1>,<at 1 1>]}, ZCurve {keys [<at 0 1>,<at 1 1>]}}})
     "Float (Alpha)"=&ClampFloat(Float=&FloatFromCurve(FloatCurve=DataInterface() {propertyValues {Curve {keys [{arriveTangent=-1, time=0, value=1}, <at 1 0>]}}}))
    ))
   &AccelerationForce(Acceleration=$Particles.Acceleration)
   &SolveForcesAndVelocity(Clamp Velocity=false, Limit Acceleration=false)
   /Niagara/Modules/CascadeConversion/CascadeConversion_Orbit(Offset:Vector3f=&Lerp_Vector(
     A=&VectorFromCurve(VectorCurve=DataInterface() {propertyValues {
        XCurve {keys [
          <at 0 0 RCIM_Cubic>
          {arriveTangent=70.4057, interpMode=RCIM_Cubic, leaveTangent=70.4057, tangentMode=RCTM_User, time=0.1215, value=6.514}
          {arriveTangent=-34.4584, interpMode=RCIM_Cubic, leaveTangent=-34.4584, tangentMode=RCTM_User, time=0.3227, value=3.2668}
          {arriveTangent=-59.0723, interpMode=RCIM_Cubic, leaveTangent=-59.0723, tangentMode=RCTM_User, time=0.7413, value=-0.0282}
         ]}
        YCurve {keys [
          <at 0 0 RCIM_Cubic>
          {arriveTangent=56.1778, interpMode=RCIM_Cubic, leaveTangent=56.1778, tangentMode=RCTM_User, time=0.1215, value=2.3869}
          {arriveTangent=-40.4172, interpMode=RCIM_Cubic, leaveTangent=-40.4172, tangentMode=RCTM_User, time=0.3227, value=2.4715}
          {arriveTangent=-54.728, interpMode=RCIM_Cubic, leaveTangent=-54.728, tangentMode=RCTM_User, time=0.7413, value=0.1189}
         ]}
        ZCurve {keys [
          <at 0 0 RCIM_Cubic>
          {interpMode=RCIM_Cubic, tangentMode=RCTM_User, time=0.1215, value=0}
          {interpMode=RCIM_Cubic, tangentMode=RCTM_User, time=0.3227, value=0}
          {arriveTangent=-0.0996, interpMode=RCIM_Cubic, leaveTangent=-0.0996, tangentMode=RCTM_User, time=0.7413, value=0}
         ]}
       }})
     B=&VectorFromCurve(VectorCurve=DataInterface() {propertyValues {
        XCurve {keys [
          <at 0 0 RCIM_Cubic>
          {arriveTangent=309.9124, interpMode=RCIM_Cubic, leaveTangent=309.9124, tangentMode=RCTM_User, time=0.1105, value=9.6992}
          {arriveTangent=-14.0097, interpMode=RCIM_Cubic, leaveTangent=-14.0097, tangentMode=RCTM_User, time=0.324, value=3.6649}
          {arriveTangent=-59.0723, interpMode=RCIM_Cubic, leaveTangent=-59.0723, tangentMode=RCTM_User, time=0.7413, value=-0.0282}
         ]}
        YCurve {keys [
          <at 0 0 RCIM_Cubic>
          {arriveTangent=184.3597, interpMode=RCIM_Cubic, leaveTangent=184.3597, tangentMode=RCTM_User, time=0.1105, value=5.8565}
          {arriveTangent=-40.4172, interpMode=RCIM_Cubic, leaveTangent=-40.4172, tangentMode=RCTM_User, time=0.324, value=6.453}
          {arriveTangent=-54.728, interpMode=RCIM_Cubic, leaveTangent=-54.728, tangentMode=RCTM_User, time=0.7413, value=0.1189}
         ]}
        ZCurve {keys [
          <at 0 0 RCIM_Cubic>
          {interpMode=RCIM_Cubic, tangentMode=RCTM_User, time=0.1105, value=0}
          {interpMode=RCIM_Cubic, tangentMode=RCTM_User, time=0.324, value=0}
          {arriveTangent=-0.0996, interpMode=RCIM_Cubic, leaveTangent=-0.0996, tangentMode=RCTM_User, time=0.7413, value=0}
         ]}
       }})
     Alpha=$Particles.Rand
    ))
   &CascadeConversion_Orbit(
    Offset:Vector3f=&Lerp_Vector(
     A=&VectorFromCurve(VectorCurve=DataInterface() {propertyValues {
        XCurve {keys [{arriveTangent=0.0132, time=0, value=0}, <at 0.0051 0.0132>]}
        YCurve {keys [{arriveTangent=0.0113, time=0, value=0}, <at 0.0051 0.0113>]}
        ZCurve {keys [{arriveTangent=-0.0117, time=0, value=0}, <at 0.0051 -0.0117>]}
       }})
     B=&VectorFromCurve(VectorCurve=DataInterface() {propertyValues {
        XCurve {keys [{arriveTangent=0.0132, time=0, value=0}, <at 0.0051 0.0132>]}
        YCurve {keys [{arriveTangent=0.0113, time=0, value=0}, <at 0.0051 0.0113>]}
        ZCurve {keys [{arriveTangent=-0.0198, time=0, value=0}, <at 0.0051 -0.0198>]}
       }})
     Alpha=$Particles.Rand
    )
    Rotation:Vector3f=&RandomRangeVector(Minimum=<0 0 0>, Maximum=<0 0 0.9954>)
    Rotation Rate:Vector3f=<0 0 0.5>
   ) {module=CascadeConversion_Orbit001}
   /Niagara/Modules/CascadeConversion/CascadeConversion_SolveOrbit()
   /CascadeToNiagaraConverter/NiagaraScripts/ModuleScripts/UpdateMeshOrientationByEuler(RotationRate:Vector3f=$Particles.MeshRotationRate)
  ]
  rendererClasses [&NiagaraMeshRendererProperties]
  renderers [
   {rendererClass=&NiagaraMeshRendererProperties, rendererIndex=0}
   {rendererClass=&NiagaraMeshRendererProperties, rendererIndex=1}
  ]
 }
 WindRibbon {
  emitterUpdateScript [
   &EmitterState(
    Inactive Response=::NewEnumerator1
    Life Cycle Mode=::NewEnumerator1
    Loop Behavior=::NewEnumerator1
    Loop Duration Mode=::NewEnumerator1
    UseLoopDelay=true
   )
   &SpawnBurst_Instantaneous(Spawn Count=4, Use Spawn Probability=false, UseLoopCountLimit=false)
  ]
  particleSpawnScript [
   /SetVariables(Particles.SkeletonSamplingIndex:NiagaraInt32=&SequentialInt(Use Array Data:NiagaraBool=false, First Number:NiagaraInt32=0, Sequence:NiagaraDataInterfaceArrayInt32=?, Last Number:NiagaraInt32=1)) {isSetParameters=true}
   &InitializeParticle(
    Mesh Scale=&RandomRangeVector(Minimum=<1 7 5.5>, Maximum=<1.5 8 5.5>, Evaluation Type=::NewEnumerator1)
    Mesh Scale Mode=::NewEnumerator4
    UsePositionOffset=false
    Write Material Random=true
   )
   &Color(Color=&MakeLinearColorFromVectorAndFloat("Vector (RGB)"=<1 1 1>, "Float (Alpha)"=0.0185))
   &DynamicMaterialParameters(
    Index 0 Param 1=&RandomRangeFloat(Minimum=0, Maximum=1, Evaluation Type=::NewEnumerator1)
    Index 0 Param 2=1
    Index 0 Param 3=1
    Index 0 Param 4=1
    Module.Param0WriteEnabled=true
    Module.Param1WriteEnabled=true
    Module.Param2WriteEnabled=true
    Module.Param3WriteEnabled=true
   )
   &InitialMeshOrientation(Mesh Orientation Mode=::NewEnumerator2, Orientation Vector=<0 0 1>, Use Rotation Vector=false)
  ]
  particleUpdateScript [
   &ParticleState(Kill Particles When Lifetime Has Elapsed=false, Let Infinitely Lived Particles Die When Emitter Deactivates=true)
   &SkeletalMeshLocation(
    Bone / Socket Index=$Particles.SkeletonSamplingIndex
    Bone Sampling Mode=::NewEnumerator4
    MaskSpawnGroup=false
    Orientation Sampling=::NewEnumerator1
    Sampled Position Offset=<-10 0 0>
    Skeletal Mesh=$System.Skeletal Mesh Reader
   )
  ]
  rendererClasses [&NiagaraMeshRendererProperties]
  renderers [{rendererClass=&NiagaraMeshRendererProperties, rendererIndex=0}]
 }
 WindRings {
  emitterUpdateScript [
   &EmitterState(Inactive Response=::NewEnumerator1, Life Cycle Mode=::NewEnumerator1, Loop Duration=2, UseLoopDelay=true)
   &SpawnRate(SpawnRate=16, Use Spawn Probability=false)
  ]
  particleSpawnScript [
   &InitializeParticle(
    Lifetime=&RandomRangeFloat(Minimum=0.6, Maximum=1, Evaluation Type=::NewEnumerator1)
    Mesh Scale=&SplitVector(
     Double Component Vector Channels:ECascadeNiagaraTwoVectorChannels=::NewEnumerator0
     Double Component:NiagaraFloat=&RandomRangeFloat(Minimum=-0.3, Maximum=0.3, Evaluation Type=::NewEnumerator1)
     Single Component:NiagaraFloat=&RandomRangeFloat(Minimum=-0.3, Maximum=0.3, Evaluation Type=::NewEnumerator1)
    )
    Mesh Scale Mode=::NewEnumerator4
    UsePositionOffset=false
    Write Material Random=true
   )
   &AddVelocity(Velocity=&RandomRangeVector(Minimum=<0 0 200>, Maximum=<0 0 350>, Evaluation Type=::NewEnumerator1))
   &Color(Color=&MakeLinearColorFromVectorAndFloat(
     "Vector (RGB)"=<0.9486 0.9489 0.9688>
     "Float (Alpha)"=&RandomRangeFloat(Minimum=0.01, Maximum=0.05, Evaluation Type=::NewEnumerator1)
    ))
   &SkeletalMeshLocation(
    Bone / Socket Index=&SequentialInt(Use Array Data:NiagaraBool=false, First Number:NiagaraInt32=0, Sequence:NiagaraDataInterfaceArrayInt32=?, Last Number:NiagaraInt32=1)
    Bone Sampling Mode=::NewEnumerator4
    MaskSpawnGroup=false
    Orientation Sampling=::NewEnumerator1
    Skeletal Mesh=$System.Skeletal Mesh Reader
   )
   &AddVectorToPosition(Vector:Vector3f=&TransformVector(
     Vector=&RandomRangeVector(Minimum=<0 0 -15>, Maximum=<0 0 20>, Evaluation Type=::NewEnumerator1)
     Destination Space=::Simulation
    ))
   &InitialMeshOrientation(Mesh Axis To Orient=<0 1 0>, Mesh Orientation Mode=::NewEnumerator2, Orientation Vector=<0 0 1>, Use Rotation Vector=false)
  ]
  particleUpdateScript [
   &ParticleState()
   &UpdateMeshOrientation(Rotation Rate=&RandomRangeFloat(Minimum=1, Maximum=3), Rotation Vector=<0 1 0>)
   &ScaleMeshSize(Scale Factor=&VectorFromCurve(
     VectorCurve=DataInterface() {propertyValues {XCurve {keys [<at 0 0.5>,<at 1 1>]}, YCurve {keys [<at 0 0.5>,<at 1 10>]}, ZCurve {keys [<at 0 0.5>,<at 1 1>]}}}
     CurveIndex=&UniformAOrBFloat(
      A=&Multiply_Float(A=$Particles.NormalizedAge, B=1)
      B=&Multiply_Float(A=$Particles.NormalizedAge, B=1.4)
      Evaluation Type=::NewEnumerator0
      Recalculate Random Each Loop=true
      Override Seed=false
     )
     Scale Curve=<5 5 5>
    )) {module=ScaleMeshSize001}
   &SolveForcesAndVelocity(Clamp Velocity=false, Limit Acceleration=false)
   &ScaleColor(
    Scale Alpha=&FloatFromCurve(FloatCurve=DataInterface() {propertyValues {Curve {keys [<at 0 0>,<at 0.2 0.5>,<at 0.8 0.5>,<at 1 0>]}}})
    Scale RGB=&VectorFromCurve(VectorCurve=DataInterface() {propertyValues {XCurve {keys [<at 0 1>,<at 1 1>]}, YCurve {keys [<at 0 1>,<at 1 1>]}, ZCurve {keys [<at 0 1>,<at 1 1>]}}})
    ScaleA=true
    ScaleRGB=true
   )
   &DynamicMaterialParameters(
    Index 0 Param 1=2
    Index 0 Param 2=&FloatFromCurve(FloatCurve=DataInterface() {propertyValues {Curve {keys [<at 0 0>,<at 1 1>]}}})
    Index 0 Param 3=&RandomRangeFloat(Minimum=0, Maximum=1, Evaluation Type=::NewEnumerator1)
    Index 0 Param 4=&RandomRangeFloat(Minimum=0.4, Maximum=0.8, Evaluation Type=::NewEnumerator1)
    Module.Index0Float0Write=true
    Module.Index0Float1Write=true
    Module.Index0Float2Write=true
    Module.Index0Float3Write=true
   ) {module=DynamicMaterialParameters001}
  ]
  rendererClasses [&NiagaraMeshRendererProperties]
  renderers [{rendererClass=&NiagaraMeshRendererProperties, rendererIndex=0}]
 }
 WindRings_FL2 {
  bEnabled=false
  emitterUpdateScript [
   &EmitterState(Inactive Response=::NewEnumerator1, Life Cycle Mode=::NewEnumerator1, Loop Duration=2, UseLoopDelay=true)
   &SpawnRate(SpawnRate=3, Use Spawn Probability=false)
  ]
  particleSpawnScript [
   /SetVariables(Particles.MeshRotationEuler:Vector3f=&RandomRangeVector(Minimum=<0 0 0>, Maximum=<0 0 1>, Evaluation Type=::NewEnumerator1)) {isSetParameters=true}
   /SetVariables(Particles.MeshOrientation=&MakeQuatFromEuler(Euler:Vector3f=$Particles.MeshRotationEuler, Scale:NiagaraFloat=360)) {isSetParameters=true}
   /SetVariables(Particles.MeshRotationRateBase:Vector3f=&RandomRangeVector(Minimum=<0 0 1>, Maximum=<0 0 3>, Evaluation Type=::NewEnumerator1)) {isSetParameters=true}
   &InitializeParticle(
    Lifetime=&RandomRangeFloat(Minimum=0.6, Maximum=1, Evaluation Type=::NewEnumerator1)
    Mesh Scale=&SplitVector(
     Double Component Vector Channels:ECascadeNiagaraTwoVectorChannels=::NewEnumerator0
     Double Component:NiagaraFloat=&RandomRangeFloat(Minimum=-0.3, Maximum=0.3, Evaluation Type=::NewEnumerator1)
     Single Component:NiagaraFloat=&RandomRangeFloat(Minimum=-0.3, Maximum=0.3, Evaluation Type=::NewEnumerator1)
    )
    Mesh Scale Mode=::NewEnumerator4
    UsePositionOffset=false
    Write Material Random=true
   )
   &AddVelocity(Velocity=&RandomRangeVector(Minimum=<0 0 350>, Maximum=<0 0 500>, Evaluation Type=::NewEnumerator1))
   &Color(Color=&MakeLinearColorFromVectorAndFloat(
     "Vector (RGB)"=<0.9486 0.9489 0.9688>
     "Float (Alpha)"=&RandomRangeFloat(Minimum=0.01, Maximum=0.05, Evaluation Type=::NewEnumerator1)
    ))
   &SkeletalMeshLocation(
    Bone / Socket Index=&SequentialInt(Use Array Data:NiagaraBool=false, First Number:NiagaraInt32=0, Sequence:NiagaraDataInterfaceArrayInt32=?, Last Number:NiagaraInt32=0)
    Bone Sampling Mode=::NewEnumerator4
    MaskSpawnGroup=false
    Orientation Sampling=::NewEnumerator1
    Skeletal Mesh=DataInterface() {propertyValues {
      DefaultMesh=null
      ExcludeBoneName=root
      OptionalComponentClass=null
      PreviewMesh=null
      SkinningMode=SkinOnTheFly
      SoftSourceActor=null
      SourceMode=Default
      UvSetIndex=0
      WholeMeshLOD=-1
      bExcludeBone=false
      bReadDeformedGeometry=true
      bRequireCurrentFrameData=true
      ComponentTags []
      FilteredBones [hand_l]
      FilteredSockets []
      MeshUserParameter {parameter {name=None, typeDefHandle {registeredTypeIndex=1}, varData []}}
      SamplingRegions []
     }}
   )
   &DynamicMaterialParameters(
    Index 0 Param 1=2
    Index 0 Param 3=&RandomRangeFloat(Minimum=0, Maximum=1, Evaluation Type=::NewEnumerator1)
    Index 0 Param 4=&RandomRangeFloat(Minimum=0.4, Maximum=0.8, Evaluation Type=::NewEnumerator1)
    Module.Param0WriteEnabled=true
    Module.Param1WriteEnabled=false
    Module.Param2WriteEnabled=true
    Module.Param3WriteEnabled=true
   )
   &AddVectorToPosition(Vector:Vector3f=&TransformVector(
     Vector=&RandomRangeVector(Minimum=<0 0 -15>, Maximum=<0 0 20>, Evaluation Type=::NewEnumerator1)
     Destination Space=::Simulation
    ))
  ]
  particleUpdateScript [
   /SetVariables(Particles.MeshRotationRate:Vector3f=$Particles.MeshRotationRateBase) {isSetParameters=true}
   &ParticleState()
   &ScaleMeshSize(Scale Factor=&VectorFromCurve(VectorCurve=DataInterface() {propertyValues {XCurve {keys [<at 0 0.5>,<at 0.6 0.5>]}, YCurve {keys [<at 0 0.5>,<at 0.6 0.5>]}, ZCurve {keys [<at 0 0>,<at 0.6 10>]}}})) {module=ScaleMeshSize001}
   &ScaleMeshSize(Scale Factor=<5 5 5>)
   &UpdateMeshOrientationByEuler(RotationRate:Vector3f=$Particles.MeshRotationRate)
   &SolveForcesAndVelocity(Clamp Velocity=false, Limit Acceleration=false)
   &ScaleColor(
    Scale Alpha=&FloatFromCurve(FloatCurve=DataInterface() {propertyValues {Curve {keys [<at 0 0>,<at 0.2 0.5>,<at 0.8 0.5>,<at 1 0>]}}})
    Scale RGB=&VectorFromCurve(VectorCurve=DataInterface() {propertyValues {XCurve {keys [<at 0 1>,<at 1 1>]}, YCurve {keys [<at 0 1>,<at 1 1>]}, ZCurve {keys [<at 0 1>,<at 1 1>]}}})
    ScaleA=true
    ScaleRGB=true
   )
   &DynamicMaterialParameters(
    Index 0 Param 2=&FloatFromCurve(FloatCurve=DataInterface() {propertyValues {Curve {keys [<at 0 0>,<at 1 1>]}}})
    Module.Param0WriteEnabled=false
    Module.Param1WriteEnabled=true
    Module.Param2WriteEnabled=false
    Module.Param3WriteEnabled=false
   ) {module=DynamicMaterialParameters001}
  ]
  rendererClasses [&NiagaraMeshRendererProperties]
  renderers [{rendererClass=&NiagaraMeshRendererProperties, rendererIndex=0}]
 }
 WindRings_FR {
  bEnabled=false
  emitterUpdateScript [
   &EmitterState(Inactive Response=::NewEnumerator1, Life Cycle Mode=::NewEnumerator1, Loop Duration=2, UseLoopDelay=true)
   &SpawnRate(SpawnRate=3, Use Spawn Probability=false)
  ]
  particleSpawnScript [
   /SetVariables(Particles.MeshRotationEuler:Vector3f=&RandomRangeVector(Minimum=<0 0 0>, Maximum=<0 0 1>, Evaluation Type=::NewEnumerator1)) {isSetParameters=true}
   /SetVariables(Particles.MeshOrientation=&MakeQuatFromEuler(Euler:Vector3f=$Particles.MeshRotationEuler, Scale:NiagaraFloat=360)) {isSetParameters=true}
   /SetVariables(Particles.MeshRotationRateBase:Vector3f=&RandomRangeVector(Minimum=<0 0 1>, Maximum=<0 0 3>, Evaluation Type=::NewEnumerator1)) {isSetParameters=true}
   &InitializeParticle(
    Lifetime=&RandomRangeFloat(Minimum=0.6, Maximum=1, Evaluation Type=::NewEnumerator1)
    Mesh Scale=&SplitVector(
     Double Component Vector Channels:ECascadeNiagaraTwoVectorChannels=::NewEnumerator0
     Double Component:NiagaraFloat=&RandomRangeFloat(Minimum=-0.3, Maximum=0.3, Evaluation Type=::NewEnumerator1)
     Single Component:NiagaraFloat=&RandomRangeFloat(Minimum=-0.3, Maximum=0.3, Evaluation Type=::NewEnumerator1)
    )
    Mesh Scale Mode=::NewEnumerator4
    UsePositionOffset=false
    Write Material Random=true
   )
   &AddVelocity(Velocity=&RandomRangeVector(Minimum=<0 0 200>, Maximum=<0 0 350>, Evaluation Type=::NewEnumerator1))
   &Color(Color=&MakeLinearColorFromVectorAndFloat(
     "Vector (RGB)"=<0.9486 0.9489 0.9688>
     "Float (Alpha)"=&RandomRangeFloat(Minimum=0.01, Maximum=0.05, Evaluation Type=::NewEnumerator1)
    ))
   &SkeletalMeshLocation(
    Bone / Socket Index=&SequentialInt(Use Array Data:NiagaraBool=false, First Number:NiagaraInt32=0, Sequence:NiagaraDataInterfaceArrayInt32=?, Last Number:NiagaraInt32=0)
    Bone Sampling Mode=::NewEnumerator4
    MaskSpawnGroup=false
    Orientation Sampling=::NewEnumerator1
    Skeletal Mesh=DataInterface() {propertyValues {
      DefaultMesh=null
      ExcludeBoneName=root
      OptionalComponentClass=null
      PreviewMesh=null
      SkinningMode=SkinOnTheFly
      SoftSourceActor=null
      SourceMode=Default
      UvSetIndex=0
      WholeMeshLOD=-1
      bExcludeBone=false
      bReadDeformedGeometry=true
      bRequireCurrentFrameData=true
      ComponentTags []
      FilteredBones [hand_r]
      FilteredSockets []
      MeshUserParameter {parameter {name=None, typeDefHandle {registeredTypeIndex=1}, varData []}}
      SamplingRegions []
     }}
   )
   &DynamicMaterialParameters(
    Index 0 Param 1=2
    Index 0 Param 3=&RandomRangeFloat(Minimum=0, Maximum=1, Evaluation Type=::NewEnumerator1)
    Index 0 Param 4=&RandomRangeFloat(Minimum=0.4, Maximum=0.8, Evaluation Type=::NewEnumerator1)
    Module.Param0WriteEnabled=true
    Module.Param1WriteEnabled=false
    Module.Param2WriteEnabled=true
    Module.Param3WriteEnabled=true
   )
   &AddVectorToPosition(Vector:Vector3f=&TransformVector(
     Vector=&RandomRangeVector(Minimum=<0 0 -15>, Maximum=<0 0 20>, Evaluation Type=::NewEnumerator1)
     Destination Space=::Simulation
    ))
  ]
  particleUpdateScript [
   /SetVariables(Particles.MeshRotationRate:Vector3f=$Particles.MeshRotationRateBase) {isSetParameters=true}
   &ParticleState()
   &ScaleMeshSize(Scale Factor=&VectorFromCurve(VectorCurve=DataInterface() {propertyValues {XCurve {keys [<at 0 0.5>,<at 1 1>]}, YCurve {keys [<at 0 0.5>,<at 1 1>]}, ZCurve {keys [<at 0 0>,<at 1 10>]}}})) {module=ScaleMeshSize001}
   &ScaleMeshSize(Scale Factor=<5 5 5>)
   &UpdateMeshOrientationByEuler(RotationRate:Vector3f=$Particles.MeshRotationRate)
   &SolveForcesAndVelocity(Clamp Velocity=false, Limit Acceleration=false)
   &ScaleColor(
    Scale Alpha=&FloatFromCurve(FloatCurve=DataInterface() {propertyValues {Curve {keys [<at 0 0>,<at 0.2 0.5>,<at 0.8 0.5>,<at 1 0>]}}})
    Scale RGB=&VectorFromCurve(VectorCurve=DataInterface() {propertyValues {XCurve {keys [<at 0 1>,<at 1 1>]}, YCurve {keys [<at 0 1>,<at 1 1>]}, ZCurve {keys [<at 0 1>,<at 1 1>]}}})
    ScaleA=true
    ScaleRGB=true
   )
   &DynamicMaterialParameters(
    Index 0 Param 2=&FloatFromCurve(FloatCurve=DataInterface() {propertyValues {Curve {keys [<at 0 0>,<at 1 1>]}}})
    Module.Param0WriteEnabled=false
    Module.Param1WriteEnabled=true
    Module.Param2WriteEnabled=false
    Module.Param3WriteEnabled=false
   ) {module=DynamicMaterialParameters001}
  ]
  rendererClasses [&NiagaraMeshRendererProperties]
  renderers [{rendererClass=&NiagaraMeshRendererProperties, rendererIndex=0}]
 }
 WindRings_FR2 {
  bEnabled=false
  emitterUpdateScript [
   &EmitterState(Inactive Response=::NewEnumerator1, Life Cycle Mode=::NewEnumerator1, Loop Duration=2, UseLoopDelay=true)
   &SpawnRate(SpawnRate=3, Use Spawn Probability=false)
  ]
  particleSpawnScript [
   /SetVariables(Particles.MeshRotationEuler:Vector3f=&RandomRangeVector(Minimum=<0 0 0>, Maximum=<0 0 1>, Evaluation Type=::NewEnumerator1)) {isSetParameters=true}
   /SetVariables(Particles.MeshOrientation=&MakeQuatFromEuler(Euler:Vector3f=$Particles.MeshRotationEuler, Scale:NiagaraFloat=360)) {isSetParameters=true}
   /SetVariables(Particles.MeshRotationRateBase:Vector3f=&RandomRangeVector(Minimum=<0 0 1>, Maximum=<0 0 3>, Evaluation Type=::NewEnumerator1)) {isSetParameters=true}
   &InitializeParticle(
    Lifetime=&RandomRangeFloat(Minimum=0.6, Maximum=1, Evaluation Type=::NewEnumerator1)
    Mesh Scale=&SplitVector(
     Double Component Vector Channels:ECascadeNiagaraTwoVectorChannels=::NewEnumerator0
     Double Component:NiagaraFloat=&RandomRangeFloat(Minimum=-0.3, Maximum=0.3, Evaluation Type=::NewEnumerator1)
     Single Component:NiagaraFloat=&RandomRangeFloat(Minimum=-0.3, Maximum=0.3, Evaluation Type=::NewEnumerator1)
    )
    Mesh Scale Mode=::NewEnumerator4
    UsePositionOffset=false
    Write Material Random=true
   )
   &AddVelocity(Velocity=&RandomRangeVector(Minimum=<0 0 200>, Maximum=<0 0 350>, Evaluation Type=::NewEnumerator1))
   &Color(Color=&MakeLinearColorFromVectorAndFloat(
     "Vector (RGB)"=<0.9486 0.9489 0.9688>
     "Float (Alpha)"=&RandomRangeFloat(Minimum=0.01, Maximum=0.05, Evaluation Type=::NewEnumerator1)
    ))
   &SkeletalMeshLocation(
    Bone / Socket Index=&SequentialInt(Use Array Data:NiagaraBool=false, First Number:NiagaraInt32=0, Sequence:NiagaraDataInterfaceArrayInt32=?, Last Number:NiagaraInt32=0)
    Bone Sampling Mode=::NewEnumerator4
    MaskSpawnGroup=false
    Orientation Sampling=::NewEnumerator1
    Skeletal Mesh=DataInterface() {propertyValues {
      DefaultMesh=null
      ExcludeBoneName=root
      OptionalComponentClass=null
      PreviewMesh=null
      SkinningMode=SkinOnTheFly
      SoftSourceActor=null
      SourceMode=Default
      UvSetIndex=0
      WholeMeshLOD=-1
      bExcludeBone=false
      bReadDeformedGeometry=true
      bRequireCurrentFrameData=true
      ComponentTags []
      FilteredBones [hand_r]
      FilteredSockets []
      MeshUserParameter {parameter {name=None, typeDefHandle {registeredTypeIndex=1}, varData []}}
      SamplingRegions []
     }}
   )
   &DynamicMaterialParameters(
    Index 0 Param 1=2
    Index 0 Param 3=&RandomRangeFloat(Minimum=0, Maximum=1, Evaluation Type=::NewEnumerator1)
    Index 0 Param 4=&RandomRangeFloat(Minimum=0.4, Maximum=0.8, Evaluation Type=::NewEnumerator1)
    Module.Param0WriteEnabled=true
    Module.Param1WriteEnabled=false
    Module.Param2WriteEnabled=true
    Module.Param3WriteEnabled=true
   )
   &AddVectorToPosition(Vector:Vector3f=&TransformVector(
     Vector=&RandomRangeVector(Minimum=<0 0 -15>, Maximum=<0 0 20>, Evaluation Type=::NewEnumerator1)
     Destination Space=::Simulation
    ))
  ]
  particleUpdateScript [
   /SetVariables(Particles.MeshRotationRate:Vector3f=$Particles.MeshRotationRateBase) {isSetParameters=true}
   &ParticleState()
   &ScaleMeshSize(Scale Factor=&VectorFromCurve(VectorCurve=DataInterface() {propertyValues {XCurve {keys [<at 0 0.5>,<at 0.6 0.5>]}, YCurve {keys [<at 0 0.5>,<at 0.6 0.5>]}, ZCurve {keys [<at 0 0>,<at 0.6 10>]}}})) {module=ScaleMeshSize001}
   &ScaleMeshSize(Scale Factor=<5 5 5>)
   &UpdateMeshOrientationByEuler(RotationRate:Vector3f=$Particles.MeshRotationRate)
   &SolveForcesAndVelocity(Clamp Velocity=false, Limit Acceleration=false)
   &ScaleColor(
    Scale Alpha=&FloatFromCurve(FloatCurve=DataInterface() {propertyValues {Curve {keys [<at 0 0>,<at 0.2 0.5>,<at 0.8 0.5>,<at 1 0>]}}})
    Scale RGB=&VectorFromCurve(VectorCurve=DataInterface() {propertyValues {XCurve {keys [<at 0 1>,<at 1 1>]}, YCurve {keys [<at 0 1>,<at 1 1>]}, ZCurve {keys [<at 0 1>,<at 1 1>]}}})
    ScaleA=true
    ScaleRGB=true
   )
   &DynamicMaterialParameters(
    Index 0 Param 2=&FloatFromCurve(FloatCurve=DataInterface() {propertyValues {Curve {keys [<at 0 0>,<at 1 1>]}}})
    Module.Param0WriteEnabled=false
    Module.Param1WriteEnabled=true
    Module.Param2WriteEnabled=false
    Module.Param3WriteEnabled=false
   ) {module=DynamicMaterialParameters001}
  ]
  rendererClasses [&NiagaraMeshRendererProperties]
  renderers [{rendererClass=&NiagaraMeshRendererProperties, rendererIndex=0}]
 }
 poolBall_Boost {
  bEnabled=false
  emitterUpdateScript [
   &EmitterState(Inactive Response=::NewEnumerator1, Life Cycle Mode=::NewEnumerator1, Loop Behavior=::NewEnumerator1, Loop Duration=0.1, UseLoopDelay=true)
   &SpawnRate(SpawnRate=250, Use Spawn Probability=false)
  ]
  particleSpawnScript [
   /SetVariables(Particles.MeshRotationRateBase:Vector3f=&RandomRangeVector(Minimum=<0 0 0>, Maximum=<0.6 0.6 0.6>, Evaluation Type=::NewEnumerator1)) {isSetParameters=true}
   /SetVariables(Particles.MeshRotationEuler:Vector3f=&RandomRangeVector(Minimum=<0 0 0>, Maximum=<1 1 1>, Evaluation Type=::NewEnumerator1)) {isSetParameters=true}
   /SetVariables(Particles.MeshOrientation=&MakeQuatFromEuler(Euler:Vector3f=$Particles.MeshRotationEuler, Scale:NiagaraFloat=360)) {isSetParameters=true}
   /SetVariables() {isSetParameters=true}
   /SetVariables(Particles.SkeletonSamplingIndex:NiagaraInt32=&SequentialInt(Use Array Data:NiagaraBool=false, First Number:NiagaraInt32=0, Sequence:NiagaraDataInterfaceArrayInt32=?, Last Number:NiagaraInt32=1)) {isSetParameters=true}
   &InitializeParticle(
    Lifetime=&RandomRangeFloat(Minimum=0.5, Maximum=1, Evaluation Type=::NewEnumerator1)
    Mesh Scale=<0.08 0.08 0.08>
    Mesh Scale Mode=::NewEnumerator4
    UsePositionOffset=false
    Write Material Random=true
   )
   &AddVelocity(Velocity=&RandomRangeVector(Minimum=<0 0 300>, Maximum=<0 0 1200>, Evaluation Type=::NewEnumerator1))
   &DynamicMaterialParameters(
    Index 0 Param 1=&RandomRangeFloat(Minimum=0, Maximum=1, Evaluation Type=::NewEnumerator1)
    Index 0 Param 2=&RandomRangeFloat(Minimum=0, Maximum=1, Evaluation Type=::NewEnumerator1)
    Module.Param0WriteEnabled=true
    Module.Param1WriteEnabled=true
    Module.Param2WriteEnabled=false
    Module.Param3WriteEnabled=false
   )
   &SkeletalMeshLocation(
    Bone / Socket Index=$Particles.SkeletonSamplingIndex
    Bone Sampling Mode=::NewEnumerator4
    MaskSpawnGroup=false
    Orientation Sampling=::NewEnumerator1
    Sampled Position Offset=&SelectVectorFromArray(Direct Array Index=$Particles.SkeletonSamplingIndex, Vector Selection Array=DataInterface() {propertyValues {FloatData [<6 0 0>, <0 0 0>]}})
    Skeletal Mesh=DataInterface() {propertyValues {
      DefaultMesh=null
      ExcludeBoneName=root
      OptionalComponentClass=null
      PreviewMesh=null
      SkinningMode=SkinOnTheFly
      SoftSourceActor=null
      SourceMode=Default
      UvSetIndex=0
      WholeMeshLOD=-1
      bExcludeBone=false
      bReadDeformedGeometry=true
      bRequireCurrentFrameData=true
      ComponentTags []
      FilteredBones [hand_l,hand_r]
      FilteredSockets []
      MeshUserParameter {parameter {name=None, typeDefHandle {registeredTypeIndex=1}, varData []}}
      SamplingRegions []
     }}
   )
   &AddVectorToPosition(Vector:Vector3f=&TransformVector(
     Vector=&SplitVector(
      Double Component Vector Channels:ECascadeNiagaraTwoVectorChannels=::NewEnumerator0
      Double Component:NiagaraFloat=&RandomRangeFloat(Minimum=-15, Maximum=15, Evaluation Type=::NewEnumerator1)
      Single Component:NiagaraFloat=&RandomRangeFloat(Minimum=0, Maximum=50, Evaluation Type=::NewEnumerator1)
     )
     Destination Space=::Simulation
    ))
   &InheritVelocity(Inherited Velocity Amount Scale=<0 0 -0.09>, Inherited Velocity Speed Limit=0, Speed Limit=false, UseSpeedThreshold=false)
  ]
  particleUpdateScript [
   /SetVariables(Particles.MeshRotationRate:Vector3f=$Particles.MeshRotationRateBase) {isSetParameters=true}
   &ParticleState()
   &ScaleMeshSize(Scale Factor=&VectorFromFloat(Value=&FloatFromCurve(FloatCurve=DataInterface() {propertyValues {Curve {keys [
          <at 0 0.3>
          <at 0.035 1.2>
          <at 0.08 1>
          <at 0.9 1>
          {arriveTangent=-20, interpMode=RCIM_Cubic, leaveTangent=-20, tangentMode=RCTM_User, time=1, value=0}
         ]}}})))
   &Color(Color=&MakeLinearColorFromVectorAndFloat(
     "Vector (RGB)"=&VectorFromCurve(VectorCurve=DataInterface() {propertyValues {XCurve {keys [<at 0 1>,<at 1 1>]}, YCurve {keys [<at 0 1>,<at 1 1>]}, ZCurve {keys [<at 0 1>,<at 1 1>]}}})
     "Float (Alpha)"=&ClampFloat(Float=&FloatFromCurve(FloatCurve=DataInterface() {propertyValues {Curve {keys [{arriveTangent=-1, time=0, value=1}, <at 1 0>]}}}))
    ))
   &AccelerationForce(Acceleration=$Particles.Acceleration)
   &Drag(
    Drag=&RandomRangeFloat(Minimum=0, Maximum=2, Evaluation Type=::NewEnumerator1)
    Use Linear Drag=true
    Use Rotational Drag=true
   )
   &SolveForcesAndVelocity(Clamp Velocity=false, Limit Acceleration=false)
   &CascadeConversion_Orbit(Offset:Vector3f=&VectorFromCurve(VectorCurve=DataInterface() {propertyValues {
       XCurve {keys [
         <at 0 0 RCIM_Cubic>
         {arriveTangent=309.9124, interpMode=RCIM_Cubic, leaveTangent=309.9124, tangentMode=RCTM_User, time=0.1105, value=9.6992}
         {arriveTangent=-14.0097, interpMode=RCIM_Cubic, leaveTangent=-14.0097, tangentMode=RCTM_User, time=0.324, value=3.6649}
         {arriveTangent=-59.0723, interpMode=RCIM_Cubic, leaveTangent=-59.0723, tangentMode=RCTM_User, time=0.7413, value=-0.0282}
        ]}
       YCurve {keys [
         <at 0 0 RCIM_Cubic>
         {arriveTangent=184.3597, interpMode=RCIM_Cubic, leaveTangent=184.3597, tangentMode=RCTM_User, time=0.1105, value=5.8565}
         {arriveTangent=-40.4172, interpMode=RCIM_Cubic, leaveTangent=-40.4172, tangentMode=RCTM_User, time=0.324, value=6.453}
         {arriveTangent=-54.728, interpMode=RCIM_Cubic, leaveTangent=-54.728, tangentMode=RCTM_User, time=0.7413, value=0.1189}
        ]}
       ZCurve {keys [
         <at 0 0 RCIM_Cubic>
         {interpMode=RCIM_Cubic, tangentMode=RCTM_User, time=0.1105, value=0}
         {interpMode=RCIM_Cubic, tangentMode=RCTM_User, time=0.324, value=0}
         {arriveTangent=-0.0996, interpMode=RCIM_Cubic, leaveTangent=-0.0996, tangentMode=RCTM_User, time=0.7413, value=0}
        ]}
      }}))
   &CascadeConversion_Orbit(
    Offset:Vector3f=&VectorFromCurve(VectorCurve=DataInterface() {propertyValues {
       XCurve {keys [{arriveTangent=0.0132, time=0, value=0}, <at 0.0051 0.0132>]}
       YCurve {keys [{arriveTangent=0.0113, time=0, value=0}, <at 0.0051 0.0113>]}
       ZCurve {keys [{arriveTangent=-0.0198, time=0, value=0}, <at 0.0051 -0.0198>]}
      }})
    Rotation:Vector3f=&RandomRangeVector(Minimum=<0 0 0>, Maximum=<0 0 0.9954>)
    Rotation Rate:Vector3f=<0 0 0.5>
   ) {module=CascadeConversion_Orbit001}
   &CascadeConversion_SolveOrbit()
   &UpdateMeshOrientationByEuler(RotationRate:Vector3f=$Particles.MeshRotationRate)
   &DynamicMaterialParameters(
    Index 0 Param 3=1
    Index 0 Param 4=1
    Module.Param0WriteEnabled=false
    Module.Param1WriteEnabled=false
    Module.Param2WriteEnabled=true
    Module.Param3WriteEnabled=true
   ) {module=DynamicMaterialParameters001}
  ]
  rendererClasses [&NiagaraMeshRendererProperties]
  renderers [{rendererClass=&NiagaraMeshRendererProperties, rendererIndex=0}]
 }
 poolBall_Vortex_L {
  bEnabled=false
  emitterUpdateScript [
   &EmitterState(Delay First Loop Only=true, Inactive Response=::NewEnumerator1, Life Cycle Mode=::NewEnumerator1, Loop Duration=1, UseLoopDelay=true)
   &SpawnRate(
    SpawnRate=&Multiply_Float(A=4, B=&RandomRangeFloat(Minimum=0.9, Maximum=1.3, Evaluation Type=::NewEnumerator1))
    Use Spawn Probability=false
   )
   &SpawnPerUnit(
    Module.Use Max Movement Threshold=true
    Module.Use Movement Tolerance=true
    Movement Tolerance=0.1
    Spawn Spacing:NiagaraFloat=&Divide_Float(A=1, B=&Divide_Float(A=1, B=1000))
    Use Spawn Probability=false
   )
  ]
  particleSpawnScript [
   /SetVariables(Particles.MeshRotationRateBase:Vector3f=&RandomRangeVector(Minimum=<0 0 0>, Maximum=<0.6 0.6 0.6>, Evaluation Type=::NewEnumerator1)) {isSetParameters=true}
   /SetVariables(Particles.MeshRotationEuler:Vector3f=&RandomRangeVector(Minimum=<0 0 0>, Maximum=<1 1 1>, Evaluation Type=::NewEnumerator1)) {isSetParameters=true}
   /SetVariables(Particles.MeshOrientation=&MakeQuatFromEuler(Euler:Vector3f=$Particles.MeshRotationEuler, Scale:NiagaraFloat=360)) {isSetParameters=true}
   /SetVariables(Particles.Acceleration:Vector3f=&RandomRangeVector(Minimum=<0 0 50>, Maximum=<0 0 350>, Evaluation Type=::NewEnumerator1)) {isSetParameters=true}
   /SetVariables(Particles.SkeletonSamplingIndex:NiagaraInt32=&SequentialInt(Use Array Data:NiagaraBool=false, First Number:NiagaraInt32=0, Sequence:NiagaraDataInterfaceArrayInt32=?, Last Number:NiagaraInt32=0)) {isSetParameters=true}
   &InitializeParticle(
    Lifetime=&RandomRangeFloat(Minimum=1, Maximum=2, Evaluation Type=::NewEnumerator1)
    Mesh Scale=<0.08 0.08 0.08>
    Mesh Scale Mode=::NewEnumerator4
    UsePositionOffset=false
    Write Material Random=true
   )
   &AddVelocity(Velocity=&RandomRangeVector(Minimum=<0 0 160>, Maximum=<0 0 450>, Evaluation Type=::NewEnumerator1))
   &DynamicMaterialParameters(
    Index 0 Param 1=&RandomRangeFloat(Minimum=0, Maximum=1, Evaluation Type=::NewEnumerator1)
    Index 0 Param 2=&RandomRangeFloat(Minimum=0, Maximum=1, Evaluation Type=::NewEnumerator1)
    Module.Param0WriteEnabled=true
    Module.Param1WriteEnabled=true
    Module.Param2WriteEnabled=false
    Module.Param3WriteEnabled=false
   )
   &SkeletalMeshLocation(
    Bone / Socket Index=$Particles.SkeletonSamplingIndex
    Bone Sampling Mode=::NewEnumerator4
    MaskSpawnGroup=false
    Orientation Sampling=::NewEnumerator1
    Sampled Position Offset=&SelectVectorFromArray(Direct Array Index=$Particles.SkeletonSamplingIndex, Vector Selection Array=DataInterface() {propertyValues {FloatData [<6 0 0>]}})
    Skeletal Mesh=DataInterface() {propertyValues {
      DefaultMesh=null
      ExcludeBoneName=root
      OptionalComponentClass=null
      PreviewMesh=null
      SkinningMode=SkinOnTheFly
      SoftSourceActor=null
      SourceMode=Default
      UvSetIndex=0
      WholeMeshLOD=-1
      bExcludeBone=false
      bReadDeformedGeometry=true
      bRequireCurrentFrameData=true
      ComponentTags []
      FilteredBones [hand_l]
      FilteredSockets []
      MeshUserParameter {parameter {name=None, typeDefHandle {registeredTypeIndex=1}, varData []}}
      SamplingRegions []
     }}
   )
   &AddVectorToPosition(Vector:Vector3f=&TransformVector(
     Vector=&SplitVector(
      Double Component Vector Channels:ECascadeNiagaraTwoVectorChannels=::NewEnumerator0
      Double Component:NiagaraFloat=&RandomRangeFloat(Minimum=-10, Maximum=10, Evaluation Type=::NewEnumerator1)
      Single Component:NiagaraFloat=&RandomRangeFloat(Minimum=0, Maximum=0, Evaluation Type=::NewEnumerator1)
     )
     Destination Space=::Simulation
    ))
   &InheritVelocity(Inherited Velocity Amount Scale=<0 0 -0.09>, Inherited Velocity Speed Limit=0, Speed Limit=false, UseSpeedThreshold=false)
  ]
  particleUpdateScript [
   /SetVariables(Particles.MeshRotationRate:Vector3f=$Particles.MeshRotationRateBase) {isSetParameters=true}
   &ParticleState()
   &ScaleMeshSize(Scale Factor=&VectorFromFloat(Value=&FloatFromCurve(FloatCurve=DataInterface() {propertyValues {Curve {keys [
          <at 0.034 0>
          <at 0.035 0.3>
          <at 0.06 1.2>
          <at 0.08 1>
          <at 0.9 1>
          {arriveTangent=-20, interpMode=RCIM_Cubic, leaveTangent=-20, tangentMode=RCTM_User, time=1, value=0}
         ]}}})))
   &Color(Color=&MakeLinearColorFromVectorAndFloat(
     "Vector (RGB)"=&VectorFromCurve(VectorCurve=DataInterface() {propertyValues {XCurve {keys [<at 0 1>,<at 1 1>]}, YCurve {keys [<at 0 1>,<at 1 1>]}, ZCurve {keys [<at 0 1>,<at 1 1>]}}})
     "Float (Alpha)"=&ClampFloat(Float=&FloatFromCurve(FloatCurve=DataInterface() {propertyValues {Curve {keys [{arriveTangent=-1, time=0, value=1}, <at 1 0>]}}}))
    ))
   &AccelerationForce(Acceleration=$Particles.Acceleration)
   &SolveForcesAndVelocity(Clamp Velocity=false, Limit Acceleration=false)
   &CascadeConversion_Orbit(Offset:Vector3f=&VectorFromCurve(VectorCurve=DataInterface() {propertyValues {
       XCurve {keys [
         <at 0 0 RCIM_Cubic>
         {arriveTangent=309.9124, interpMode=RCIM_Cubic, leaveTangent=309.9124, tangentMode=RCTM_User, time=0.1105, value=9.6992}
         {arriveTangent=-14.0097, interpMode=RCIM_Cubic, leaveTangent=-14.0097, tangentMode=RCTM_User, time=0.324, value=3.6649}
         {arriveTangent=-59.0723, interpMode=RCIM_Cubic, leaveTangent=-59.0723, tangentMode=RCTM_User, time=0.7413, value=-0.0282}
        ]}
       YCurve {keys [
         <at 0 0 RCIM_Cubic>
         {arriveTangent=184.3597, interpMode=RCIM_Cubic, leaveTangent=184.3597, tangentMode=RCTM_User, time=0.1105, value=5.8565}
         {arriveTangent=-40.4172, interpMode=RCIM_Cubic, leaveTangent=-40.4172, tangentMode=RCTM_User, time=0.324, value=6.453}
         {arriveTangent=-54.728, interpMode=RCIM_Cubic, leaveTangent=-54.728, tangentMode=RCTM_User, time=0.7413, value=0.1189}
        ]}
       ZCurve {keys [
         <at 0 0 RCIM_Cubic>
         {interpMode=RCIM_Cubic, tangentMode=RCTM_User, time=0.1105, value=0}
         {interpMode=RCIM_Cubic, tangentMode=RCTM_User, time=0.324, value=0}
         {arriveTangent=-0.0996, interpMode=RCIM_Cubic, leaveTangent=-0.0996, tangentMode=RCTM_User, time=0.7413, value=0}
        ]}
      }}))
   &CascadeConversion_Orbit(
    Offset:Vector3f=&VectorFromCurve(VectorCurve=DataInterface() {propertyValues {
       XCurve {keys [{arriveTangent=0.0132, time=0, value=0}, <at 0.0051 0.0132>]}
       YCurve {keys [{arriveTangent=0.0113, time=0, value=0}, <at 0.0051 0.0113>]}
       ZCurve {keys [{arriveTangent=-0.0198, time=0, value=0}, <at 0.0051 -0.0198>]}
      }})
    Rotation:Vector3f=&RandomRangeVector(Minimum=<0 0 0>, Maximum=<0 0 0.9954>)
    Rotation Rate:Vector3f=<0 0 0.5>
   ) {module=CascadeConversion_Orbit001}
   &CascadeConversion_SolveOrbit()
   &UpdateMeshOrientationByEuler(RotationRate:Vector3f=$Particles.MeshRotationRate)
   &DynamicMaterialParameters(
    Index 0 Param 3=1
    Index 0 Param 4=1
    Module.Param0WriteEnabled=false
    Module.Param1WriteEnabled=false
    Module.Param2WriteEnabled=true
    Module.Param3WriteEnabled=true
   ) {module=DynamicMaterialParameters001}
  ]
  rendererClasses [&NiagaraMeshRendererProperties]
  renderers [{rendererClass=&NiagaraMeshRendererProperties, rendererIndex=0}]
 }
 poolBall_Vortex_L001 {
  bEnabled=false
  emitterUpdateScript [
   &EmitterState(
    Delay First Loop Only=true
    Inactive Response=::NewEnumerator1
    Life Cycle Mode=::NewEnumerator1
    Loop Duration=&RandomRangeFloat(Minimum=0.3, Maximum=1.3, Evaluation Type=::NewEnumerator1)
    Recalculate Duration Each Loop=true
    UseLoopDelay=true
   )
   &SpawnBurst_Instantaneous(Use Spawn Probability=false, UseLoopCountLimit=false)
  ]
  particleSpawnScript [
   /SetVariables(Particles.MeshRotationRateBase:Vector3f=&RandomRangeVector(Minimum=<0 0 0>, Maximum=<0.6 0.6 0.6>, Evaluation Type=::NewEnumerator1)) {isSetParameters=true}
   /SetVariables(Particles.MeshRotationEuler:Vector3f=&RandomRangeVector(Minimum=<0 0 0>, Maximum=<1 1 1>, Evaluation Type=::NewEnumerator1)) {isSetParameters=true}
   /SetVariables(Particles.MeshOrientation=&MakeQuatFromEuler(Euler:Vector3f=$Particles.MeshRotationEuler, Scale:NiagaraFloat=360)) {isSetParameters=true}
   /SetVariables(Particles.Acceleration:Vector3f=&RandomRangeVector(Minimum=<0 0 100>, Maximum=<0 0 700>, Evaluation Type=::NewEnumerator1)) {isSetParameters=true}
   /SetVariables(Particles.SkeletonSamplingIndex:NiagaraInt32=&SequentialInt(Use Array Data:NiagaraBool=false, First Number:NiagaraInt32=0, Sequence:NiagaraDataInterfaceArrayInt32=?, Last Number:NiagaraInt32=0)) {isSetParameters=true}
   &InitializeParticle(
    Lifetime=&RandomRangeFloat(Minimum=2, Maximum=3, Evaluation Type=::NewEnumerator1)
    Mesh Scale=<0.12 0.12 0.12>
    Mesh Scale Mode=::NewEnumerator4
    UsePositionOffset=false
    Write Material Random=true
   )
   &AddVelocity(Velocity=&RandomRangeVector(Minimum=<0 0 150>, Maximum=<0 0 350>, Evaluation Type=::NewEnumerator1))
   &DynamicMaterialParameters(
    Index 0 Param 1=&RandomRangeFloat(Minimum=0, Maximum=1, Evaluation Type=::NewEnumerator1)
    Index 0 Param 2=&RandomRangeFloat(Minimum=0, Maximum=1, Evaluation Type=::NewEnumerator1)
    Module.Param0WriteEnabled=true
    Module.Param1WriteEnabled=true
    Module.Param2WriteEnabled=false
    Module.Param3WriteEnabled=false
   )
   &SkeletalMeshLocation(
    Bone / Socket Index=$Particles.SkeletonSamplingIndex
    Bone Sampling Mode=::NewEnumerator4
    MaskSpawnGroup=false
    Orientation Sampling=::NewEnumerator1
    Sampled Position Offset=&SelectVectorFromArray(Direct Array Index=$Particles.SkeletonSamplingIndex, Vector Selection Array=DataInterface() {propertyValues {FloatData [<6 0 0>]}})
    Skeletal Mesh=DataInterface() {propertyValues {
      DefaultMesh=null
      ExcludeBoneName=root
      OptionalComponentClass=null
      PreviewMesh=null
      SkinningMode=SkinOnTheFly
      SoftSourceActor=null
      SourceMode=Default
      UvSetIndex=0
      WholeMeshLOD=-1
      bExcludeBone=false
      bReadDeformedGeometry=true
      bRequireCurrentFrameData=true
      ComponentTags []
      FilteredBones [hand_l]
      FilteredSockets []
      MeshUserParameter {parameter {name=None, typeDefHandle {registeredTypeIndex=1}, varData []}}
      SamplingRegions []
     }}
   )
   &AddVectorToPosition(Vector:Vector3f=&TransformVector(
     Vector=&SplitVector(
      Double Component Vector Channels:ECascadeNiagaraTwoVectorChannels=::NewEnumerator0
      Double Component:NiagaraFloat=&RandomRangeFloat(Minimum=-10, Maximum=10, Evaluation Type=::NewEnumerator1)
      Single Component:NiagaraFloat=&RandomRangeFloat(Minimum=0, Maximum=0, Evaluation Type=::NewEnumerator1)
     )
     Destination Space=::Simulation
    ))
  ]
  particleUpdateScript [
   /SetVariables(Particles.MeshRotationRate:Vector3f=$Particles.MeshRotationRateBase) {isSetParameters=true}
   &ParticleState()
   &ScaleMeshSize(Scale Factor=&VectorFromFloat(Value=&FloatFromCurve(FloatCurve=DataInterface() {propertyValues {Curve {keys [
          <at 0.034 0>
          <at 0.035 0.3>
          <at 0.06 1.2>
          <at 0.08 1>
          <at 0.9 1>
          {arriveTangent=-20, interpMode=RCIM_Cubic, leaveTangent=-20, tangentMode=RCTM_User, time=1, value=0}
         ]}}})))
   &Color(Color=&MakeLinearColorFromVectorAndFloat(
     "Vector (RGB)"=&VectorFromCurve(VectorCurve=DataInterface() {propertyValues {XCurve {keys [<at 0 1>,<at 1 1>]}, YCurve {keys [<at 0 1>,<at 1 1>]}, ZCurve {keys [<at 0 1>,<at 1 1>]}}})
     "Float (Alpha)"=&ClampFloat(Float=&FloatFromCurve(FloatCurve=DataInterface() {propertyValues {Curve {keys [{arriveTangent=-1, time=0, value=1}, <at 1 0>]}}}))
    ))
   &AccelerationForce(Acceleration=$Particles.Acceleration)
   &SolveForcesAndVelocity(Clamp Velocity=false, Limit Acceleration=false)
   &CascadeConversion_Orbit(Offset:Vector3f=&VectorFromCurve(VectorCurve=DataInterface() {propertyValues {
       XCurve {keys [
         <at 0 0 RCIM_Cubic>
         {arriveTangent=70.4057, interpMode=RCIM_Cubic, leaveTangent=70.4057, tangentMode=RCTM_User, time=0.1215, value=6.514}
         {arriveTangent=-34.4584, interpMode=RCIM_Cubic, leaveTangent=-34.4584, tangentMode=RCTM_User, time=0.3227, value=3.2668}
         {arriveTangent=-59.0723, interpMode=RCIM_Cubic, leaveTangent=-59.0723, tangentMode=RCTM_User, time=0.7413, value=-0.0282}
        ]}
       YCurve {keys [
         <at 0 0 RCIM_Cubic>
         {arriveTangent=56.1778, interpMode=RCIM_Cubic, leaveTangent=56.1778, tangentMode=RCTM_User, time=0.1215, value=2.3869}
         {arriveTangent=-40.4172, interpMode=RCIM_Cubic, leaveTangent=-40.4172, tangentMode=RCTM_User, time=0.3227, value=2.4715}
         {arriveTangent=-54.728, interpMode=RCIM_Cubic, leaveTangent=-54.728, tangentMode=RCTM_User, time=0.7413, value=0.1189}
        ]}
       ZCurve {keys [
         <at 0 0 RCIM_Cubic>
         {interpMode=RCIM_Cubic, tangentMode=RCTM_User, time=0.1215, value=0}
         {interpMode=RCIM_Cubic, tangentMode=RCTM_User, time=0.3227, value=0}
         {arriveTangent=-0.0996, interpMode=RCIM_Cubic, leaveTangent=-0.0996, tangentMode=RCTM_User, time=0.7413, value=0}
        ]}
      }}))
   &CascadeConversion_Orbit(
    Offset:Vector3f=&VectorFromCurve(VectorCurve=DataInterface() {propertyValues {
       XCurve {keys [{arriveTangent=0.0132, time=0, value=0}, <at 0.0051 0.0132>]}
       YCurve {keys [{arriveTangent=0.0113, time=0, value=0}, <at 0.0051 0.0113>]}
       ZCurve {keys [{arriveTangent=-0.0117, time=0, value=0}, <at 0.0051 -0.0117>]}
      }})
    Rotation:Vector3f=&RandomRangeVector(Minimum=<0 0 0>, Maximum=<0 0 0.9954>)
    Rotation Rate:Vector3f=<0 0 0.5>
   ) {module=CascadeConversion_Orbit001}
   &CascadeConversion_SolveOrbit()
   &UpdateMeshOrientationByEuler(RotationRate:Vector3f=$Particles.MeshRotationRate)
   &DynamicMaterialParameters(
    Index 0 Param 3=1
    Index 0 Param 4=1
    Module.Param0WriteEnabled=false
    Module.Param1WriteEnabled=false
    Module.Param2WriteEnabled=true
    Module.Param3WriteEnabled=true
   ) {module=DynamicMaterialParameters001}
  ]
  rendererClasses [&NiagaraMeshRendererProperties]
  renderers [{rendererClass=&NiagaraMeshRendererProperties, rendererIndex=0}]
 }
 poolBall_Vortex_R {
  bEnabled=false
  emitterUpdateScript [
   &EmitterState(Delay First Loop Only=true, Inactive Response=::NewEnumerator1, Life Cycle Mode=::NewEnumerator1, Loop Duration=1, UseLoopDelay=true)
   &SpawnRate(
    SpawnRate=&Multiply_Float(A=4, B=&RandomRangeFloat(Minimum=0.9, Maximum=1.3, Evaluation Type=::NewEnumerator1))
    Use Spawn Probability=false
   )
   &SpawnPerUnit(
    Module.Use Max Movement Threshold=true
    Module.Use Movement Tolerance=true
    Movement Tolerance=0.1
    Spawn Spacing:NiagaraFloat=&Divide_Float(A=1, B=&Divide_Float(A=1, B=1000))
    Use Spawn Probability=false
   )
  ]
  particleSpawnScript [
   /SetVariables(Particles.MeshRotationRateBase:Vector3f=&RandomRangeVector(Minimum=<0 0 0>, Maximum=<0.6 0.6 0.6>, Evaluation Type=::NewEnumerator1)) {isSetParameters=true}
   /SetVariables(Particles.MeshRotationEuler:Vector3f=&RandomRangeVector(Minimum=<0 0 0>, Maximum=<1 1 1>, Evaluation Type=::NewEnumerator1)) {isSetParameters=true}
   /SetVariables(Particles.MeshOrientation=&MakeQuatFromEuler(Euler:Vector3f=$Particles.MeshRotationEuler, Scale:NiagaraFloat=360)) {isSetParameters=true}
   /SetVariables(Particles.Acceleration:Vector3f=&RandomRangeVector(Minimum=<0 0 200>, Maximum=<0 0 350>, Evaluation Type=::NewEnumerator1)) {isSetParameters=true}
   /SetVariables(Particles.SkeletonSamplingIndex:NiagaraInt32=&SequentialInt(Use Array Data:NiagaraBool=false, First Number:NiagaraInt32=0, Sequence:NiagaraDataInterfaceArrayInt32=?, Last Number:NiagaraInt32=0)) {isSetParameters=true}
   &InitializeParticle(
    Lifetime=&RandomRangeFloat(Minimum=1, Maximum=2, Evaluation Type=::NewEnumerator1)
    Mesh Scale=<0.08 0.08 0.08>
    Mesh Scale Mode=::NewEnumerator4
    UsePositionOffset=false
    Write Material Random=true
   )
   &AddVelocity(Velocity=&RandomRangeVector(Minimum=<0 0 160>, Maximum=<0 0 450>, Evaluation Type=::NewEnumerator1))
   &DynamicMaterialParameters(
    Index 0 Param 1=&RandomRangeFloat(Minimum=0, Maximum=1, Evaluation Type=::NewEnumerator1)
    Index 0 Param 2=&RandomRangeFloat(Minimum=0, Maximum=1, Evaluation Type=::NewEnumerator1)
    Module.Param0WriteEnabled=true
    Module.Param1WriteEnabled=true
    Module.Param2WriteEnabled=false
    Module.Param3WriteEnabled=false
   )
   &SkeletalMeshLocation(
    Bone / Socket Index=$Particles.SkeletonSamplingIndex
    Bone Sampling Mode=::NewEnumerator4
    MaskSpawnGroup=false
    Orientation Sampling=::NewEnumerator1
    Sampled Position Offset=&SelectVectorFromArray(Direct Array Index=$Particles.SkeletonSamplingIndex, Vector Selection Array=DataInterface() {propertyValues {FloatData [<-6 0 0>]}})
    Skeletal Mesh=DataInterface() {propertyValues {
      DefaultMesh=null
      ExcludeBoneName=root
      OptionalComponentClass=null
      PreviewMesh=null
      SkinningMode=SkinOnTheFly
      SoftSourceActor=null
      SourceMode=Default
      UvSetIndex=0
      WholeMeshLOD=-1
      bExcludeBone=false
      bReadDeformedGeometry=true
      bRequireCurrentFrameData=true
      ComponentTags []
      FilteredBones [hand_r]
      FilteredSockets []
      MeshUserParameter {parameter {name=None, typeDefHandle {registeredTypeIndex=1}, varData []}}
      SamplingRegions []
     }}
   )
   &AddVectorToPosition(Vector:Vector3f=&TransformVector(
     Vector=&SplitVector(
      Double Component Vector Channels:ECascadeNiagaraTwoVectorChannels=::NewEnumerator0
      Double Component:NiagaraFloat=&RandomRangeFloat(Minimum=-10, Maximum=10, Evaluation Type=::NewEnumerator1)
      Single Component:NiagaraFloat=&RandomRangeFloat(Minimum=0, Maximum=0, Evaluation Type=::NewEnumerator1)
     )
     Destination Space=::Simulation
    ))
   &InheritVelocity(Inherited Velocity Amount Scale=<0 0 -0.09>, Inherited Velocity Speed Limit=0, Speed Limit=false, UseSpeedThreshold=false)
  ]
  particleUpdateScript [
   /SetVariables(Particles.MeshRotationRate:Vector3f=$Particles.MeshRotationRateBase) {isSetParameters=true}
   &ParticleState()
   &ScaleMeshSize(Scale Factor=&VectorFromFloat(Value=&FloatFromCurve(FloatCurve=DataInterface() {propertyValues {Curve {keys [
          <at 0.034 0>
          <at 0.035 0.3>
          <at 0.06 1.2>
          <at 0.08 1>
          <at 0.9 1>
          {arriveTangent=-20, interpMode=RCIM_Cubic, leaveTangent=-20, tangentMode=RCTM_User, time=1, value=0}
         ]}}})))
   &Color(Color=&MakeLinearColorFromVectorAndFloat(
     "Vector (RGB)"=&VectorFromCurve(VectorCurve=DataInterface() {propertyValues {XCurve {keys [<at 0 1>,<at 1 1>]}, YCurve {keys [<at 0 1>,<at 1 1>]}, ZCurve {keys [<at 0 1>,<at 1 1>]}}})
     "Float (Alpha)"=&ClampFloat(Float=&FloatFromCurve(FloatCurve=DataInterface() {propertyValues {Curve {keys [{arriveTangent=-1, time=0, value=1}, <at 1 0>]}}}))
    ))
   &AccelerationForce(Acceleration=$Particles.Acceleration)
   &SolveForcesAndVelocity(Clamp Velocity=false, Limit Acceleration=false)
   &CascadeConversion_Orbit(Offset:Vector3f=&VectorFromCurve(VectorCurve=DataInterface() {propertyValues {
       XCurve {keys [
         <at 0 0 RCIM_Cubic>
         {arriveTangent=309.9124, interpMode=RCIM_Cubic, leaveTangent=309.9124, tangentMode=RCTM_User, time=0.1105, value=9.6992}
         {arriveTangent=-14.0097, interpMode=RCIM_Cubic, leaveTangent=-14.0097, tangentMode=RCTM_User, time=0.324, value=3.6649}
         {arriveTangent=-59.0723, interpMode=RCIM_Cubic, leaveTangent=-59.0723, tangentMode=RCTM_User, time=0.7413, value=-0.0282}
        ]}
       YCurve {keys [
         <at 0 0 RCIM_Cubic>
         {arriveTangent=184.3597, interpMode=RCIM_Cubic, leaveTangent=184.3597, tangentMode=RCTM_User, time=0.1105, value=5.8565}
         {arriveTangent=-40.4172, interpMode=RCIM_Cubic, leaveTangent=-40.4172, tangentMode=RCTM_User, time=0.324, value=6.453}
         {arriveTangent=-54.728, interpMode=RCIM_Cubic, leaveTangent=-54.728, tangentMode=RCTM_User, time=0.7413, value=0.1189}
        ]}
       ZCurve {keys [
         <at 0 0 RCIM_Cubic>
         {interpMode=RCIM_Cubic, tangentMode=RCTM_User, time=0.1105, value=0}
         {interpMode=RCIM_Cubic, tangentMode=RCTM_User, time=0.324, value=0}
         {arriveTangent=-0.0996, interpMode=RCIM_Cubic, leaveTangent=-0.0996, tangentMode=RCTM_User, time=0.7413, value=0}
        ]}
      }}))
   &CascadeConversion_Orbit(
    Offset:Vector3f=&VectorFromCurve(VectorCurve=DataInterface() {propertyValues {
       XCurve {keys [{arriveTangent=0.0132, time=0, value=0}, <at 0.0051 0.0132>]}
       YCurve {keys [{arriveTangent=0.0113, time=0, value=0}, <at 0.0051 0.0113>]}
       ZCurve {keys [{arriveTangent=-0.0198, time=0, value=0}, <at 0.0051 -0.0198>]}
      }})
    Rotation:Vector3f=&RandomRangeVector(Minimum=<0 0 0>, Maximum=<0 0 0.9954>)
    Rotation Rate:Vector3f=<0 0 0.5>
   ) {module=CascadeConversion_Orbit001}
   &CascadeConversion_SolveOrbit()
   &UpdateMeshOrientationByEuler(RotationRate:Vector3f=$Particles.MeshRotationRate)
   &DynamicMaterialParameters(
    Index 0 Param 3=1
    Index 0 Param 4=1
    Module.Param0WriteEnabled=false
    Module.Param1WriteEnabled=false
    Module.Param2WriteEnabled=true
    Module.Param3WriteEnabled=true
   ) {module=DynamicMaterialParameters001}
  ]
  rendererClasses [&NiagaraMeshRendererProperties]
  renderers [{rendererClass=&NiagaraMeshRendererProperties, rendererIndex=0}]
 }
 poolBall_Vortex_R001 {
  bEnabled=false
  emitterUpdateScript [
   &EmitterState(
    Delay First Loop Only=true
    Inactive Response=::NewEnumerator1
    Life Cycle Mode=::NewEnumerator1
    Loop Duration=&RandomRangeFloat(Minimum=0.3, Maximum=1.3, Evaluation Type=::NewEnumerator1)
    Recalculate Duration Each Loop=true
    UseLoopDelay=true
   )
   &SpawnBurst_Instantaneous(Use Spawn Probability=false, UseLoopCountLimit=false)
  ]
  particleSpawnScript [
   /SetVariables(Particles.MeshRotationRateBase:Vector3f=&RandomRangeVector(Minimum=<0 0 0>, Maximum=<0.6 0.6 0.6>, Evaluation Type=::NewEnumerator1)) {isSetParameters=true}
   /SetVariables(Particles.MeshRotationEuler:Vector3f=&RandomRangeVector(Minimum=<0 0 0>, Maximum=<1 1 1>, Evaluation Type=::NewEnumerator1)) {isSetParameters=true}
   /SetVariables(Particles.MeshOrientation=&MakeQuatFromEuler(Euler:Vector3f=$Particles.MeshRotationEuler, Scale:NiagaraFloat=360)) {isSetParameters=true}
   /SetVariables(Particles.Acceleration:Vector3f=&RandomRangeVector(Minimum=<0 0 100>, Maximum=<0 0 700>, Evaluation Type=::NewEnumerator1)) {isSetParameters=true}
   /SetVariables(Particles.SkeletonSamplingIndex:NiagaraInt32=&SequentialInt(Use Array Data:NiagaraBool=false, First Number:NiagaraInt32=0, Sequence:NiagaraDataInterfaceArrayInt32=?, Last Number:NiagaraInt32=0)) {isSetParameters=true}
   &InitializeParticle(
    Lifetime=&RandomRangeFloat(Minimum=2, Maximum=3, Evaluation Type=::NewEnumerator1)
    Mesh Scale=<0.12 0.12 0.12>
    Mesh Scale Mode=::NewEnumerator4
    UsePositionOffset=false
    Write Material Random=true
   )
   &AddVelocity(Velocity=&RandomRangeVector(Minimum=<0 0 150>, Maximum=<0 0 350>, Evaluation Type=::NewEnumerator1))
   &DynamicMaterialParameters(
    Index 0 Param 1=&RandomRangeFloat(Minimum=0, Maximum=1, Evaluation Type=::NewEnumerator1)
    Index 0 Param 2=&RandomRangeFloat(Minimum=0, Maximum=1, Evaluation Type=::NewEnumerator1)
    Module.Param0WriteEnabled=true
    Module.Param1WriteEnabled=true
    Module.Param2WriteEnabled=false
    Module.Param3WriteEnabled=false
   )
   &SkeletalMeshLocation(
    Bone / Socket Index=$Particles.SkeletonSamplingIndex
    Bone Sampling Mode=::NewEnumerator4
    MaskSpawnGroup=false
    Orientation Sampling=::NewEnumerator1
    Sampled Position Offset=&SelectVectorFromArray(Direct Array Index=$Particles.SkeletonSamplingIndex, Vector Selection Array=DataInterface() {propertyValues {FloatData [<-6 0 0>]}})
    Skeletal Mesh=DataInterface() {propertyValues {
      DefaultMesh=null
      ExcludeBoneName=root
      OptionalComponentClass=null
      PreviewMesh=null
      SkinningMode=SkinOnTheFly
      SoftSourceActor=null
      SourceMode=Default
      UvSetIndex=0
      WholeMeshLOD=-1
      bExcludeBone=false
      bReadDeformedGeometry=true
      bRequireCurrentFrameData=true
      ComponentTags []
      FilteredBones [hand_r]
      FilteredSockets []
      MeshUserParameter {parameter {name=None, typeDefHandle {registeredTypeIndex=1}, varData []}}
      SamplingRegions []
     }}
   )
   &AddVectorToPosition(Vector:Vector3f=&TransformVector(
     Vector=&SplitVector(
      Double Component Vector Channels:ECascadeNiagaraTwoVectorChannels=::NewEnumerator0
      Double Component:NiagaraFloat=&RandomRangeFloat(Minimum=-10, Maximum=10, Evaluation Type=::NewEnumerator1)
      Single Component:NiagaraFloat=&RandomRangeFloat(Minimum=0, Maximum=0, Evaluation Type=::NewEnumerator1)
     )
     Destination Space=::Simulation
    ))
  ]
  particleUpdateScript [
   /SetVariables(Particles.MeshRotationRate:Vector3f=$Particles.MeshRotationRateBase) {isSetParameters=true}
   &ParticleState()
   &ScaleMeshSize(Scale Factor=&VectorFromFloat(Value=&FloatFromCurve(FloatCurve=DataInterface() {propertyValues {Curve {keys [
          <at 0.034 0>
          <at 0.035 0.3>
          <at 0.06 1.2>
          <at 0.08 1>
          <at 0.9 1>
          {arriveTangent=-20, interpMode=RCIM_Cubic, leaveTangent=-20, tangentMode=RCTM_User, time=1, value=0}
         ]}}})))
   &Color(Color=&MakeLinearColorFromVectorAndFloat(
     "Vector (RGB)"=&VectorFromCurve(VectorCurve=DataInterface() {propertyValues {XCurve {keys [<at 0 1>,<at 1 1>]}, YCurve {keys [<at 0 1>,<at 1 1>]}, ZCurve {keys [<at 0 1>,<at 1 1>]}}})
     "Float (Alpha)"=&ClampFloat(Float=&FloatFromCurve(FloatCurve=DataInterface() {propertyValues {Curve {keys [{arriveTangent=-1, time=0, value=1}, <at 1 0>]}}}))
    ))
   &AccelerationForce(Acceleration=$Particles.Acceleration)
   &SolveForcesAndVelocity(Clamp Velocity=false, Limit Acceleration=false)
   &CascadeConversion_Orbit(Offset:Vector3f=&VectorFromCurve(VectorCurve=DataInterface() {propertyValues {
       XCurve {keys [
         <at 0 0 RCIM_Cubic>
         {arriveTangent=70.4057, interpMode=RCIM_Cubic, leaveTangent=70.4057, tangentMode=RCTM_User, time=0.1215, value=6.514}
         {arriveTangent=-34.4584, interpMode=RCIM_Cubic, leaveTangent=-34.4584, tangentMode=RCTM_User, time=0.3227, value=3.2668}
         {arriveTangent=-59.0723, interpMode=RCIM_Cubic, leaveTangent=-59.0723, tangentMode=RCTM_User, time=0.7413, value=-0.0282}
        ]}
       YCurve {keys [
         <at 0 0 RCIM_Cubic>
         {arriveTangent=56.1778, interpMode=RCIM_Cubic, leaveTangent=56.1778, tangentMode=RCTM_User, time=0.1215, value=2.3869}
         {arriveTangent=-40.4172, interpMode=RCIM_Cubic, leaveTangent=-40.4172, tangentMode=RCTM_User, time=0.3227, value=2.4715}
         {arriveTangent=-54.728, interpMode=RCIM_Cubic, leaveTangent=-54.728, tangentMode=RCTM_User, time=0.7413, value=0.1189}
        ]}
       ZCurve {keys [
         <at 0 0 RCIM_Cubic>
         {interpMode=RCIM_Cubic, tangentMode=RCTM_User, time=0.1215, value=0}
         {interpMode=RCIM_Cubic, tangentMode=RCTM_User, time=0.3227, value=0}
         {arriveTangent=-0.0996, interpMode=RCIM_Cubic, leaveTangent=-0.0996, tangentMode=RCTM_User, time=0.7413, value=0}
        ]}
      }}))
   &CascadeConversion_Orbit(
    Offset:Vector3f=&VectorFromCurve(VectorCurve=DataInterface() {propertyValues {
       XCurve {keys [{arriveTangent=0.0132, time=0, value=0}, <at 0.0051 0.0132>]}
       YCurve {keys [{arriveTangent=0.0113, time=0, value=0}, <at 0.0051 0.0113>]}
       ZCurve {keys [{arriveTangent=-0.0117, time=0, value=0}, <at 0.0051 -0.0117>]}
      }})
    Rotation:Vector3f=&RandomRangeVector(Minimum=<0 0 0>, Maximum=<0 0 0.9954>)
    Rotation Rate:Vector3f=<0 0 0.5>
   ) {module=CascadeConversion_Orbit001}
   &CascadeConversion_SolveOrbit()
   &UpdateMeshOrientationByEuler(RotationRate:Vector3f=$Particles.MeshRotationRate)
   &DynamicMaterialParameters(
    Index 0 Param 3=1
    Index 0 Param 4=1
    Module.Param0WriteEnabled=false
    Module.Param1WriteEnabled=false
    Module.Param2WriteEnabled=true
    Module.Param3WriteEnabled=true
   ) {module=DynamicMaterialParameters001}
  ]
  rendererClasses [&NiagaraMeshRendererProperties]
  renderers [{rendererClass=&NiagaraMeshRendererProperties, rendererIndex=0}]
 }
 windRibbon_FR {
  bEnabled=false
  emitterUpdateScript [
   &EmitterState(Inactive Response=::NewEnumerator1, Life Cycle Mode=::NewEnumerator1, Loop Behavior=::NewEnumerator1, Loop Duration=1, UseLoopDelay=true)
   &SpawnBurst_Instantaneous(
    Spawn Count=&MakeIntFromFloat(
     Float=&Multiply_Float(A=2, B=&MakeFloatFromInt(Int=1))
     Conversion Mode=::NewEnumerator2
    )
    Use Spawn Probability=false
    UseLoopCountLimit=false
   )
  ]
  particleSpawnScript [
   /SetVariables(Particles.SkeletonSamplingIndex:NiagaraInt32=&SequentialInt(Use Array Data:NiagaraBool=false, First Number:NiagaraInt32=0, Sequence:NiagaraDataInterfaceArrayInt32=?, Last Number:NiagaraInt32=0)) {isSetParameters=true}
   &InitializeParticle(
    Mesh Scale=&RandomRangeVector(Minimum=<1 7 5.5>, Maximum=<1.5 8 5.5>, Evaluation Type=::NewEnumerator1)
    Mesh Scale Mode=::NewEnumerator4
    UsePositionOffset=false
    Write Material Random=true
   )
   &Color(Color=&MakeLinearColorFromVectorAndFloat("Vector (RGB)"=<1 1 1>, "Float (Alpha)"=0.0185))
   &DynamicMaterialParameters(
    Index 0 Param 1=&RandomRangeFloat(Minimum=0, Maximum=1, Evaluation Type=::NewEnumerator1)
    Module.Param0WriteEnabled=true
    Module.Param1WriteEnabled=false
    Module.Param2WriteEnabled=false
    Module.Param3WriteEnabled=false
   )
  ]
  particleUpdateScript [
   &ParticleState(Kill Particles When Lifetime Has Elapsed=false, Let Infinitely Lived Particles Die When Emitter Deactivates=true)
   &SkeletalMeshLocation(
    Bone / Socket Index=$Particles.SkeletonSamplingIndex
    Bone Sampling Mode=::NewEnumerator4
    MaskSpawnGroup=false
    Orientation Sampling=::NewEnumerator1
    Sampled Position Offset=&SelectVectorFromArray(Direct Array Index=$Particles.SkeletonSamplingIndex, Vector Selection Array=DataInterface() {propertyValues {FloatData [<10 0 0>]}})
    Skeletal Mesh=DataInterface() {propertyValues {
      DefaultMesh=null
      ExcludeBoneName=root
      OptionalComponentClass=null
      PreviewMesh=null
      SkinningMode=SkinOnTheFly
      SoftSourceActor=null
      SourceMode=Default
      UvSetIndex=0
      WholeMeshLOD=-1
      bExcludeBone=false
      bReadDeformedGeometry=true
      bRequireCurrentFrameData=true
      ComponentTags []
      FilteredBones [hand_r]
      FilteredSockets []
      MeshUserParameter {parameter {name=None, typeDefHandle {registeredTypeIndex=1}, varData []}}
      SamplingRegions []
     }}
   )
   &DynamicMaterialParameters(
    Index 0 Param 2=1
    Index 0 Param 3=1
    Index 0 Param 4=1
    Module.Param0WriteEnabled=false
    Module.Param1WriteEnabled=true
    Module.Param2WriteEnabled=true
    Module.Param3WriteEnabled=true
   ) {module=DynamicMaterialParameters001}
  ]
  rendererClasses [&NiagaraMeshRendererProperties]
  renderers [{rendererClass=&NiagaraMeshRendererProperties, rendererIndex=0}]
 }
}
summary {
 systemName=NS_Contrail_Billiards_Converted
 emitters [
  {bEnabled=false, emitterName=poolBall_Boost, rendererClasses [&NiagaraMeshRendererProperties]}
  {bEnabled=false, emitterName=poolBall_Vortex_L, rendererClasses [&NiagaraMeshRendererProperties]}
  {bEnabled=false, emitterName=poolBall_Vortex_L001, rendererClasses [&NiagaraMeshRendererProperties]}
  {bEnabled=false, emitterName=poolBall_Vortex_R, rendererClasses [&NiagaraMeshRendererProperties]}
  {bEnabled=false, emitterName=poolBall_Vortex_R001, rendererClasses [&NiagaraMeshRendererProperties]}
  {emitterName=WindRibbon, rendererClasses [&NiagaraMeshRendererProperties]}
  {bEnabled=false, emitterName=windRibbon_FR, rendererClasses [&NiagaraMeshRendererProperties]}
  {emitterName=WindRings, rendererClasses [&NiagaraMeshRendererProperties]}
  {bEnabled=false, emitterName=WindRings_FL2, rendererClasses [&NiagaraMeshRendererProperties]}
  {bEnabled=false, emitterName=WindRings_FR, rendererClasses [&NiagaraMeshRendererProperties]}
  {bEnabled=false, emitterName=WindRings_FR2, rendererClasses [&NiagaraMeshRendererProperties]}
  {emitterName=PoolBalls, rendererClasses [&NiagaraMeshRendererProperties]}
 ]
 userVariables []
}
