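# Niagara skydiving contrail cosmetic that trails bubble sprites from a character's skeletal mesh.
# A dense small-bubble emitter and an intermittent larger-bubble emitter cycle through four bone or socket indices, offset by sphere locations.
# Bubbles wobble and pop via size curves, inherit curve-scaled source movement and swirl in a vortex around their sampled point.
# A front-end modifier adds upward drift in menus, replacing gameplay's zero velocity.
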NiagaraSystem /BRCosmetics/Athena/Effects/Particles/Skydiving/NS_Contrail_Bubbles
dependencies {
 usedDataInterfaces [&NiagaraDataInterfaceCurve, &NiagaraDataInterfaceSkeletalMesh]
 usedDynamicInputs []
 usedModules [
  &self:SystemSpawnScript.SystemScriptSource.SystemScriptGraph.NiagaraNodeAssignment_0.SetVariables_#1
  &SystemState
 ]
}
emitters {
 Bubbles_Hand_Bigger {
  emitterUpdateScript [
   &EmitterState(
    Delay First Loop Only=true
    Inactive Response=::NewEnumerator1
    Life Cycle Mode=::NewEnumerator1
    Loop Delay=0.5
    Loop Duration=&RandomRangeFloat(Minimum=0.3, Maximum=1, Evaluation Type=::NewEnumerator1)
    Recalculate Duration Each Loop=true
    UseLoopDelay=true
   )
   &SpawnRate(
    SpawnRate=&RandomRangeFloat(Minimum=2, Maximum=15, Evaluation Type=::NewEnumerator1)
    Use Spawn Probability=false
   )
  ]
  particleSpawnScript [
   &InitializeParticle(
    Lifetime=&RandomRangeFloat(Minimum=0.5, Maximum=1, Evaluation Type=::NewEnumerator1)
    Sprite Size=&Vector2DFromFloat(Value=&RandomRangeFloat(Minimum=20, Maximum=25, Evaluation Type=::NewEnumerator1))
    Sprite Size Mode=::NewEnumerator4
    UsePositionOffset=false
    Write Material Random=true
   )
   &Color(Color=&MakeLinearColorFromVectorAndFloat("Vector (RGB)"=<1 1 1>, "Float (Alpha)"=1))
   &SkeletalMeshLocation(
    Bone / Socket Index=/CascadeToNiagaraConverter/NiagaraScripts/DynamicInputScripts/SequentialInt(Use Array Data:NiagaraBool=false, First Number:NiagaraInt32=0, Sequence:NiagaraDataInterfaceArrayInt32=?, Last Number:NiagaraInt32=3)
    Bone Sampling Mode=::NewEnumerator4
    MaskSpawnGroup=false
    Orientation Sampling=::NewEnumerator1
    Skeletal Mesh=$System.Skeletal Mesh Reader
   )
   &ShapeLocation(MaskSpawnGroup=false, Override Seed=false, Sphere Radius=5, UseOwnerScale=true)
   &AddVelocity(Velocity=&FN_FrontEndVector3Modifier(Gameplay=<0 0 0>, Front End=<0 -200 600>))
  ]
  particleUpdateScript [
   &ParticleState()
   &ScaleSpriteSize(
    Scale Sprite Size Mode=::NewEnumerator0
    Uniform Scale Factor=&FloatFromCurve(FloatCurve=DataInterface() {propertyValues {Curve {keys [
         <at 0 0 RCIM_Cubic>
         {arriveTangent=3.8762, interpMode=RCIM_Cubic, leaveTangent=3.8762, time=0.1034, value=0.9634}
         {arriveTangent=0.4344, interpMode=RCIM_Cubic, leaveTangent=0.4344, time=0.2227, value=0.8631}
         {arriveTangent=0.2382, interpMode=RCIM_Cubic, leaveTangent=0.2382, time=0.2988, value=1.0483}
         {arriveTangent=0.0856, interpMode=RCIM_Cubic, leaveTangent=0.0856, time=0.8126, value=1.0037}
         {arriveTangent=-5.3567, interpMode=RCIM_Cubic, leaveTangent=-5.3567, time=0.8837, value=1.0984}
         <at 1 0 RCIM_Cubic>
        ]}}})
   )
   &ScaleColor(ScaleA=true, ScaleRGB=true)
   &InheritSourceMovement(
    Applied Position Scale=&Multiply_VectorByFloat(
     Vector=<0.2 0.2 1>
     Float=&FloatFromCurve(
      FloatCurve=DataInterface() {propertyValues {Curve {
         defaultValue=1
         keys [
          {interpMode=RCIM_Cubic, tangentMode=RCTM_User, time=0.1678, value=-0.1}
          {interpMode=RCIM_Cubic, tangentMode=RCTM_User, time=0.3286, value=0.4}
         ]
        }}}
      Scale Curve=-1
     )
    )
    ApplyForce=false
    ApplyPosition=true
    ApplyPositionFalloff=false
   )
   &VortexForce(UseOriginOffset=false, Vortex Origin:NiagaraPosition=$Particles.SkeletalMeshLocation.SampledPosition)
   &SolveForcesAndVelocity(Clamp Velocity=false, Limit Acceleration=false)
  ]
  rendererClasses [&NiagaraSpriteRendererProperties]
  renderers [{rendererClass=&NiagaraSpriteRendererProperties}]
 }
 Bubbles_Hand_L {
  emitterUpdateScript [
   &EmitterState(Inactive Response=::NewEnumerator1, Life Cycle Mode=::NewEnumerator1, Loop Duration=1, UseLoopDelay=true)
   &SpawnRate(SpawnRate=80, Use Spawn Probability=false)
  ]
  particleSpawnScript [
   &InitializeParticle(
    Lifetime=&RandomRangeFloat(Minimum=0.5, Maximum=1, Evaluation Type=::NewEnumerator1)
    Sprite Size=&Vector2DFromFloat(Value=&RandomRangeFloat(Minimum=5, Maximum=9, Evaluation Type=::NewEnumerator1))
    Sprite Size Mode=::NewEnumerator4
    UsePositionOffset=false
    Write Material Random=true
   )
   &Color(Color=&MakeLinearColorFromVectorAndFloat("Vector (RGB)"=<1 1 1>, "Float (Alpha)"=1))
   &SkeletalMeshLocation(
    Bone / Socket Index=&SequentialInt(Use Array Data:NiagaraBool=false, First Number:NiagaraInt32=0, Sequence:NiagaraDataInterfaceArrayInt32=?, Last Number:NiagaraInt32=3)
    Bone Sampling Mode=::NewEnumerator4
    MaskSpawnGroup=false
    Orientation Sampling=::NewEnumerator1
    Skeletal Mesh=$System.Skeletal Mesh Reader
   )
   &ShapeLocation(MaskSpawnGroup=false, Non Uniform Scale=<1 1 0>, Override Seed=false, Sphere Radius=35, UseOwnerScale=true)
   &AddVelocity(Velocity=&FN_FrontEndVector3Modifier(Gameplay=<0 0 0>, Front End=<0 -200 600>))
  ]
  particleUpdateScript [
   &ParticleState()
   &ScaleSpriteSize(
    Scale Sprite Size Mode=::NewEnumerator0
    Uniform Scale Factor=&FloatFromCurve(FloatCurve=DataInterface() {propertyValues {Curve {keys [
         <at 0 0 RCIM_Cubic>
         {arriveTangent=3.8762, interpMode=RCIM_Cubic, leaveTangent=3.8762, time=0.1034, value=0.9347}
         {arriveTangent=0.5812, interpMode=RCIM_Cubic, leaveTangent=0.5812, time=0.2227, value=0.8631}
         {arriveTangent=0.2382, interpMode=RCIM_Cubic, leaveTangent=0.2382, time=0.2988, value=1.0483}
         {arriveTangent=0.0856, interpMode=RCIM_Cubic, leaveTangent=0.0856, time=0.8126, value=1.0037}
         {arriveTangent=-5.3567, interpMode=RCIM_Cubic, leaveTangent=-5.3567, time=0.8837, value=1.0984}
         <at 1 0 RCIM_Cubic>
        ]}}})
   )
   &ScaleColor(ScaleA=true, ScaleRGB=true)
   &InheritSourceMovement(
    Applied Position Scale=&Multiply_VectorByFloat(
     Vector=<0.2 0.2 1>
     Float=&FloatFromCurve(
      FloatCurve=DataInterface() {propertyValues {Curve {
         defaultValue=1
         keys [
          {interpMode=RCIM_Cubic, tangentMode=RCTM_User, time=0.1678, value=-0.1}
          {interpMode=RCIM_Cubic, tangentMode=RCTM_User, time=0.3286, value=0.4}
         ]
        }}}
      Scale Curve=-1
     )
    )
    ApplyForce=false
    ApplyPosition=true
    ApplyPositionFalloff=false
   )
   &VortexForce(UseOriginOffset=false, Vortex Origin:NiagaraPosition=$Particles.SkeletalMeshLocation.SampledPosition)
   &SolveForcesAndVelocity(Clamp Velocity=false, Limit Acceleration=false)
  ]
  rendererClasses [&NiagaraSpriteRendererProperties]
  renderers [{rendererClass=&NiagaraSpriteRendererProperties}]
 }
}
summary {
 systemName=NS_Contrail_Bubbles
 emitters [
  {emitterName=Bubbles_Hand_L, rendererClasses [&NiagaraSpriteRendererProperties]}
  {emitterName=Bubbles_Hand_Bigger, rendererClasses [&NiagaraSpriteRendererProperties]}
 ]
 userVariables [
  User.bIsFrontEnd {type=NiagaraBool}
  User.bIsFrontEndPreview {type=NiagaraBool}
 ]
}
