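# Niagara skydiving contrail for the Celestial cosmetic, attaching sparkles, flares, glows, smoke and ribbons to bones or sockets on User.PlayerPawn.
# Spawn groups double as socket indices, so one emitter covers several attachment points.
# User.ParaGlideLeanAlpha lengthens hand ribbons via an age-lerped offset and shifts the background sprites' kill volume.
# A separate Source emitter adds falling speed-stretched streaks, while Source_Ribbons has its spawn modules disabled.
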NiagaraSystem /BRCosmetics/Athena/Effects/Particles/Skydiving/NS_Contrail_Celestial
dependencies {
 usedDataInterfaces [
  &NiagaraDataInterfaceCurve
  &NiagaraDataInterfaceArrayFloat3
  &NiagaraDataInterfaceVectorCurve
  &NiagaraDataInterfaceParticleRead
 ]
 usedDynamicInputs []
 usedModules [&SystemState]
}
emitters {
 Emitting_Points {
  emitterUpdateScript [
   &EmitterState()
   &SpawnRate(SpawnRate=10, Use Spawn Probability=false)
  ]
  particleSpawnScript [
   &InitializeParticle(
    Color Maximum=<rgba 0.4 1 0.75 1>
    Color Minimum=<rgba 0 0.1887 0.5573 0.2>
    Color Mode=::NewEnumerator2
    Lifetime Max=3
    Lifetime Min=0.5
    Lifetime Mode=::NewEnumerator1
    Sprite Size Mode=::NewEnumerator1
    Uniform Sprite Size Max=7
    Uniform Sprite Size Min=1
    UsePositionOffset=false
    Write Material Random=true
   )
   &SkeletalMeshLocation(
    Bone / Socket Index=4
    Bone Sampling Mode=::NewEnumerator4
    MaskSpawnGroup=false
    Sampled Position Offset=<0 0 10>
    Sampled Position Offset Coordinate Space:ENiagaraSkelMeshTransforms=::NewEnumerator0
    Skeletal Mesh=$User.PlayerPawn
   )
   &ShapeLocation(MaskSpawnGroup=false, Override Seed=false, UseOwnerScale=true)
   &AddVelocity(Velocity=&Multiply_VectorByFloat(Vector=$Particles.ShapeLocation.ShapeNormal, Float=25))
   &AddVelocity(Velocity=&RandomRangeVector(Minimum=<0 0 10>, Maximum=<0 0 175>)) {module=AddVelocity001}
  ]
  particleUpdateScript [
   &ParticleState()
   &InheritVelocity(Inherited Velocity Amount Scale=<-0.2 -0.2 -0.2>, Speed Limit=true, UseSpeedThreshold=false)
   &SolveForcesAndVelocity(Clamp Velocity=false, Limit Acceleration=false)
   &ScaleColor(
    Scale Alpha=&FloatFromCurve(FloatCurve=DataInterface() {propertyValues {Curve {keys [
         {arriveTangent=10.7395, interpMode=RCIM_Cubic, leaveTangent=10.7395, tangentMode=RCTM_User, time=0, value=3}
         {arriveTangent=-4.2154, interpMode=RCIM_Cubic, leaveTangent=12.8937, tangentMode=RCTM_Break, time=0.3, value=1}
         {arriveTangent=-15.3617, interpMode=RCIM_Cubic, leaveTangent=4.0988, tangentMode=RCTM_Break, time=0.4562, value=1.0768}
         <at 1 0 RCIM_Cubic>
        ]}}})
    Scale RGB=&VectorFromCurve(VectorCurve=DataInterface() {propertyValues {
       XCurve {keys [<at 0 18 RCIM_Cubic>, {arriveTangent=-17.7313, interpMode=RCIM_Cubic, leaveTangent=-17.7313, time=0.3, value=9}, <at 1 0.2688 RCIM_Cubic>]}
       YCurve {keys [<at 0 18 RCIM_Cubic>, {arriveTangent=-17.7202, interpMode=RCIM_Cubic, leaveTangent=-17.7202, time=0.3, value=9}, <at 1 0.2798 RCIM_Cubic>]}
       ZCurve {keys [<at 0 18 RCIM_Cubic>, {arriveTangent=-17.4, interpMode=RCIM_Cubic, leaveTangent=-17.4, time=0.3, value=11.8641}, <at 1 0.6 RCIM_Cubic>]}
      }})
    ScaleA=true
    ScaleRGB=true
   )
   &ScaleSpriteSize(Uniform Curve Sprite Scale=DataInterface() {propertyValues {Curve {keys [
        <at 0 0 RCIM_Cubic>
        {arriveTangent=2.1263, interpMode=RCIM_Cubic, leaveTangent=2.1263, time=0.1034, value=1.1346}
        {arriveTangent=-0.4991, interpMode=RCIM_Cubic, leaveTangent=-0.4991, time=0.2184, value=0.4643}
        {arriveTangent=-1.1951, interpMode=RCIM_Cubic, leaveTangent=-1.1951, time=0.3019, value=1.0356}
        {arriveTangent=0.1149, interpMode=RCIM_Cubic, leaveTangent=0.1149, time=0.506, value=0.1205}
        {arriveTangent=0.4332, interpMode=RCIM_Cubic, leaveTangent=0.4332, time=0.6432, value=1.0748}
        {arriveTangent=-3.012, interpMode=RCIM_Cubic, leaveTangent=-3.012, time=0.8021, value=0.2488}
        <at 1 0 RCIM_Cubic>
       ]}}})
  ]
  rendererClasses [&NiagaraSpriteRendererProperties]
  renderers [{rendererClass=&NiagaraSpriteRendererProperties}]
 }
 Flare_Complex {
  simTarget=GPUComputeSim
  emitterUpdateScript [
   &EmitterState()
   &SpawnRate(SpawnRate=8, Use Spawn Probability=false)
   &SpawnRate(SpawnGroup=1, SpawnRate=8, Use Spawn Probability=false) {module=SpawnRate001}
   &SpawnRate(SpawnGroup=2, SpawnRate=8, Use Spawn Probability=false) {module=SpawnRate002}
   &SpawnRate(SpawnGroup=3, SpawnRate=8, Use Spawn Probability=false) {module=SpawnRate003}
  ]
  particleSpawnScript [
   &InitializeParticle(
    Color=&MakeLinearColorFromVectorAndFloat(
     "Vector (RGB)"=&MakeVectorFromLinearColorRGB(Color=<rgba 0.1042 1 0.9234 1>)
     "Float (Alpha)"=&RandomRangeFloat(Minimum=0.1, Maximum=0.15)
    )
    Lifetime Max=1.5
    Lifetime Min=0.5
    Lifetime Mode=::NewEnumerator1
    Sprite Size Max=&MakeCustomVector2DFromBool(
     Bool=&SetBoolByIntComparison(A=&SpawnGroup(), B=2, Comparison Type=::NewEnumerator2)
     True Vector2D=<100 50>
     False Vector2D=<150 75>
    )
    Sprite Size Min=&MakeCustomVector2DFromBool(
     Bool=&SetBoolByIntComparison(A=&SpawnGroup(), B=2, Comparison Type=::NewEnumerator2)
     True Vector2D=<0 0>
     False Vector2D=<50 15>
    )
    Sprite Size Mode=::NewEnumerator2
    UsePositionOffset=false
    Write Material Random=true
   )
   &DynamicMaterialParameters(
    Index 0 Param 1=1
    Index 0 Param 2=1
    Index 0 Param 3=1
    Index 0 Param 4=1
    Module.Param0WriteEnabled=true
    Module.Param1WriteEnabled=true
    Module.Param2WriteEnabled=true
    Module.Param3WriteEnabled=true
   )
   /SetVariables(Particles.RoundFlareSize:Vector2f=&Vector2DFromFloat(Value=&MakeCustomFloatFromBool(
      Bool=&SetBoolByIntComparison(A=&SpawnGroup(), B=2, Comparison Type=::NewEnumerator2)
      True Float=60
      False Float=120
     ))) {isSetParameters=true}
   &CameraOffset(Camera Offset Amount=20)
  ]
  particleUpdateScript [
   &ParticleState()
   &SkeletalMeshLocation(
    Bone / Socket Index=&SpawnGroup()
    Bone Sampling Mode=::NewEnumerator4
    MaskSpawnGroup=false
    Sampled Position Offset=&SelectVectorFromArray(
     Direct Array Index=&SpawnGroup()
     Vector Selection Array=DataInterface() {propertyValues {FloatData [<7 0 0>, <-7 0 0>, <0 0 0>, <0 0 0>]}}
    )
    Skeletal Mesh=$User.PlayerPawn
   )
   &ScaleColor(
    Scale Alpha=&FloatFromCurve(FloatCurve=DataInterface() {propertyValues {Curve {keys [
         <at 0 0 RCIM_Cubic>
         {arriveTangent=3.0935, interpMode=RCIM_Cubic, leaveTangent=3.0935, time=0.1663, value=1}
         {arriveTangent=5.1051, interpMode=RCIM_Cubic, leaveTangent=5.1051, time=0.2122, value=0.6565}
         {arriveTangent=-2.1687, interpMode=RCIM_Cubic, leaveTangent=-2.1687, time=0.2722, value=1.5407}
         {arriveTangent=-4.0351, interpMode=RCIM_Cubic, leaveTangent=-4.0351, time=0.3914, value=0.2679}
         {arriveTangent=1.1112, interpMode=RCIM_Cubic, leaveTangent=1.1112, time=0.4747, value=0.7234}
         {arriveTangent=1.5431, interpMode=RCIM_Cubic, leaveTangent=1.5431, time=0.5722, value=0.4689}
         {arriveTangent=-2.0132, interpMode=RCIM_Cubic, leaveTangent=-2.0132, time=0.657, value=1.0048}
         {arriveTangent=0.0259, interpMode=RCIM_Cubic, leaveTangent=0.0259, time=0.7253, value=0.1608}
         {arriveTangent=-0.5852, interpMode=RCIM_Cubic, leaveTangent=-0.5852, time=0.7705, value=1.0077}
         <at 1 0 RCIM_Cubic>
        ]}}})
    Scale RGB=<1.2 1.2 1.2>
    ScaleA=true
    ScaleRGB=true
   )
   /SetVariables(Particles.RoundFlareColor:LinearColor=&MultiplyLinearColor(
     Linear Color=<rgba 1 0.4078 0.7956 0.1>
     Scale Factor=&MakeFloat4FromVectorAndFloat(
      "Vector (RGB)"=<1.2 1.2 1.2>
      "Float (Alpha)"=&FloatFromCurve(FloatCurve=DataInterface() {propertyValues {Curve {keys [
           <at 0 0 RCIM_Cubic>
           {arriveTangent=3.0935, interpMode=RCIM_Cubic, leaveTangent=3.0935, time=0.1663, value=1}
           {arriveTangent=5.1051, interpMode=RCIM_Cubic, leaveTangent=5.1051, time=0.2122, value=0.6565}
           {arriveTangent=-2.1687, interpMode=RCIM_Cubic, leaveTangent=-2.1687, time=0.2722, value=1.5407}
           {arriveTangent=-4.0351, interpMode=RCIM_Cubic, leaveTangent=-4.0351, time=0.3914, value=0.2679}
           {arriveTangent=1.1112, interpMode=RCIM_Cubic, leaveTangent=1.1112, time=0.4747, value=0.7234}
           {arriveTangent=1.5431, interpMode=RCIM_Cubic, leaveTangent=1.5431, time=0.5722, value=0.4689}
           {arriveTangent=-2.0132, interpMode=RCIM_Cubic, leaveTangent=-2.0132, time=0.657, value=1.0048}
           {arriveTangent=0.0259, interpMode=RCIM_Cubic, leaveTangent=0.0259, time=0.7253, value=0.1608}
           {arriveTangent=-0.5852, interpMode=RCIM_Cubic, leaveTangent=-0.5852, time=0.7705, value=1.0077}
           <at 1 0 RCIM_Cubic>
          ]}}})
     )
    )) {isSetParameters=true}
  ]
  rendererClasses [&NiagaraSpriteRendererProperties]
  renderers [
   {rendererClass=&NiagaraSpriteRendererProperties}
   {rendererClass=&NiagaraSpriteRendererProperties, rendererIndex=1}
  ]
 }
 Glows {
  emitterUpdateScript [
   &EmitterState()
   &SpawnRate(SpawnRate=40, Use Spawn Probability=false)
   &SpawnRate(SpawnGroup=1, SpawnRate=40, Use Spawn Probability=false) {module=SpawnRate001}
   &SpawnRate(SpawnGroup=2, SpawnRate=40, Use Spawn Probability=false) {module=SpawnRate002}
   &SpawnRate(SpawnGroup=3, SpawnRate=40, Use Spawn Probability=false) {module=SpawnRate003}
  ]
  particleSpawnScript [
   &InitializeParticle(
    Color=<rgba 0.4232 0.0659 1 0.17>
    Lifetime Max=0.2
    Lifetime Min=0.12
    Lifetime Mode=::NewEnumerator1
    Sprite Size Mode=::NewEnumerator3
    Uniform Sprite Size=&MakeCustomFloatFromBool(
     Bool=&SetBoolByIntComparison(A=&SpawnGroup(), B=2, Comparison Type=::NewEnumerator2)
     True Float=45
     False Float=75
    )
    UsePositionOffset=false
    Write Material Random=true
   )
   &CameraOffset(Camera Offset Amount=5)
   &DynamicMaterialParameters(
    Index 0 Param 1=1
    Index 0 Param 2=0.5
    Index 0 Param 4=1
    Module.Param0WriteEnabled=true
    Module.Param1WriteEnabled=true
    Module.Param2WriteEnabled=true
    Module.Param3WriteEnabled=true
   )
  ]
  particleUpdateScript [
   &ParticleState()
   &SkeletalMeshLocation(
    Bone / Socket Index=&SpawnGroup()
    Bone Sampling Mode=::NewEnumerator4
    MaskSpawnGroup=false
    Sampled Position Offset=&SelectVectorFromArray(
     Direct Array Index=&SpawnGroup()
     Vector Selection Array=DataInterface() {propertyValues {FloatData [<7 0 0>, <-7 0 0>, <0 0 0>, <0 0 0>]}}
    )
    Skeletal Mesh=$User.PlayerPawn
   )
   &ScaleColor(
    Scale Alpha=&FloatFromCurve(FloatCurve=DataInterface() {propertyValues {Curve {keys [
         <at 0 0 RCIM_Cubic>
         {arriveTangentWeight=0.1067, interpMode=RCIM_Cubic, leaveTangentWeight=0.3202, tangentMode=RCTM_User, time=0.2, value=1}
         {arriveTangentWeight=0.3202, interpMode=RCIM_Cubic, leaveTangentWeight=0.1067, tangentMode=RCTM_User, time=0.8, value=1}
         <at 1 0 RCIM_Cubic>
        ]}}})
    ScaleA=true
    ScaleRGB=true
   )
  ]
  rendererClasses [&NiagaraSpriteRendererProperties]
  renderers [{rendererClass=&NiagaraSpriteRendererProperties}]
 }
 Hand_Emitting_Points {
  simTarget=GPUComputeSim
  emitterUpdateScript [
   &EmitterState()
   &SpawnRate(SpawnRate=15, Use Spawn Probability=false)
   &SpawnRate(SpawnGroup=1, SpawnRate=15, Use Spawn Probability=false) {module=SpawnRate001}
  ]
  particleSpawnScript [
   &InitializeParticle(
    Color Maximum=<rgba 0.1333 0.3294 0.3098 0.5>
    Color Minimum=<rgba 0.0405 0.0524 0.1 0.5>
    Color Mode=::NewEnumerator2
    Lifetime Max=1.5
    Lifetime Min=0.75
    Lifetime Mode=::NewEnumerator1
    Sprite Size Mode=::NewEnumerator1
    Uniform Sprite Size Max=6
    Uniform Sprite Size Min=2
    UsePositionOffset=false
    Write Material Random=true
   )
   &SkeletalMeshLocation(
    Bone / Socket Index=&SpawnGroup()
    Bone Sampling Mode=::NewEnumerator4
    MaskSpawnGroup=false
    Skeletal Mesh=$User.PlayerPawn
   )
   &ShapeLocation(Hemisphere Distribution=<0.5 1>, MaskSpawnGroup=false, Override Seed=false, Sphere Radius=10, UseOwnerScale=true)
   &AddVelocity(Velocity=&Multiply_VectorByFloat(Vector=$Particles.ShapeLocation.ShapeNormal, Float=15))
   &AddVelocity(Velocity=<0 0 500>) {module=AddVelocity001}
  ]
  particleUpdateScript [
   &ParticleState()
   &InheritVelocity(
    Inherited Velocity Amount Scale=&RandomRangeVector(Minimum=<0.75 0.75 -0.1>, Maximum=<0.75 0.75 -0.3>)
    Speed Limit=false
    UseSpeedThreshold=false
   )
   &Drag(Drag=5, Use Linear Drag=true, Use Rotational Drag=false)
   &SolveForcesAndVelocity(Clamp Velocity=false, Limit Acceleration=false)
   &ScaleColor(
    Scale Alpha=&FloatFromCurve(FloatCurve=DataInterface() {propertyValues {Curve {keys [
         {arriveTangent=10.7395, interpMode=RCIM_Cubic, leaveTangent=10.7395, tangentMode=RCTM_User, time=0, value=3}
         {arriveTangent=-4.2154, interpMode=RCIM_Cubic, leaveTangent=12.8937, tangentMode=RCTM_Break, time=0.3, value=1}
         {arriveTangent=-15.3617, interpMode=RCIM_Cubic, leaveTangent=4.0988, tangentMode=RCTM_Break, time=0.4562, value=1.0768}
         <at 1 0 RCIM_Cubic>
        ]}}})
    Scale RGB=&VectorFromFloat(Value=&FloatFromCurve(FloatCurve=DataInterface() {propertyValues {Curve {keys [<at 0 18>,<at 0.3 9>,<at 1 0.6>]}}}))
    ScaleA=true
    ScaleRGB=true
   )
   &ScaleSpriteSize(Uniform Curve Sprite Scale=DataInterface() {propertyValues {Curve {keys [
        <at 0 0 RCIM_Cubic>
        {arriveTangent=2.1263, interpMode=RCIM_Cubic, leaveTangent=2.1263, time=0.1034, value=1.1346}
        {arriveTangent=-0.4991, interpMode=RCIM_Cubic, leaveTangent=-0.4991, time=0.2184, value=0.4643}
        {arriveTangent=-1.1951, interpMode=RCIM_Cubic, leaveTangent=-1.1951, time=0.3019, value=1.0356}
        {arriveTangent=0.1149, interpMode=RCIM_Cubic, leaveTangent=0.1149, time=0.506, value=0.1205}
        {arriveTangent=0.4332, interpMode=RCIM_Cubic, leaveTangent=0.4332, time=0.6432, value=1.0748}
        {arriveTangent=-3.012, interpMode=RCIM_Cubic, leaveTangent=-3.012, time=0.8021, value=0.2488}
        <at 1 0 RCIM_Cubic>
       ]}}})
  ]
  rendererClasses [&NiagaraSpriteRendererProperties]
  renderers [{rendererClass=&NiagaraSpriteRendererProperties}]
 }
 Hand_Ribbons {
  emitterUpdateScript [
   &EmitterState()
   &SpawnRate(SpawnRate=300, Use Spawn Probability=false)
   &SpawnRate(SpawnGroup=1, SpawnRate=300, Use Spawn Probability=false) {module=SpawnRate001}
  ]
  particleSpawnScript [&InitializeRibbon(Color=<rgba 0.3802 0.3802 0.3802 3>, Lifetime=0.4, Ribbon Width=30, UsePositionOffset=false, Write Material Random=true)]
  particleUpdateScript [
   &ParticleState()
   &SkeletalMeshLocation(
    Bone / Socket Index=&SpawnGroup()
    Bone Sampling Mode=::NewEnumerator4
    MaskSpawnGroup=false
    Orientation Sampling=::NewEnumerator1
    Sampled Position Offset=&Lerp_Vector(
     A=<0 0 0>
     B=&Lerp_Vector(A=<0 0 300>, B=<0 0 800>, Alpha=$User.ParaGlideLeanAlpha)
     Alpha=$Particles.NormalizedAge
    )
    Sampled Position Offset Coordinate Space:ENiagaraSkelMeshTransforms=::NewEnumerator0
    Skeletal Mesh=$User.PlayerPawn
   )
   &DynamicMaterialParameters(
    Index 0 Param 1=0.5
    Index 0 Param 2=15
    Index 0 Param 3=&FloatFromCurve(FloatCurve=DataInterface() {propertyValues {Curve {keys [
         <at 0 1.367 RCIM_Cubic>
         {arriveTangent=-5.4611, interpMode=RCIM_Cubic, leaveTangent=-5.4611, tangentMode=RCTM_User, time=0.2238, value=0.7705}
         {arriveTangentWeight=0.1764, interpMode=RCIM_Cubic, leaveTangentWeight=0.1882, tangentMode=RCTM_User, time=0.5993, value=-0.9465}
         <at 1 0 RCIM_Cubic>
        ]}}})
    Index 0 Param 4=1
    Module.Param0WriteEnabled=true
    Module.Param1WriteEnabled=true
    Module.Param2WriteEnabled=true
    Module.Param3WriteEnabled=true
   )
   &ScaleColor(
    Scale Alpha=&FloatFromCurve(FloatCurve=DataInterface() {propertyValues {Curve {keys [
         {arriveTangent=63.5336, interpMode=RCIM_Cubic, leaveTangent=63.5336, tangentMode=RCTM_User, time=0, value=0}
         {arriveTangent=1.4107, interpMode=RCIM_Cubic, leaveTangent=1.4107, time=0.0349, value=2.0102}
         {arriveTangent=-3.0491, interpMode=RCIM_Cubic, leaveTangent=-3.0491, time=0.8126, value=1.1463}
         {arriveTangent=-9.1281, interpMode=RCIM_Cubic, leaveTangent=-9.1281, tangentMode=RCTM_User, time=1, value=-0.9324}
        ]}}})
    ScaleA=true
    ScaleRGB=true
   )
  ]
  rendererClasses [&NiagaraRibbonRendererProperties]
  renderers [{rendererClass=&NiagaraRibbonRendererProperties}]
 }
 Ribbon_BKGD {
  emitterUpdateScript [
   &EmitterState()
   &SpawnRate(SpawnRate=40, Use Spawn Probability=false)
   &SpawnRate(SpawnGroup=1, SpawnRate=40, Use Spawn Probability=false) {module=SpawnRate001}
  ]
  particleSpawnScript [
   &InitializeParticle(
    Color=&MakeLinearColorFromVectorAndFloat(
     "Vector (RGB)"=&MakeVectorFromLinearColorRGB(Color=<rgba 1.3 2.2 3 1>)
     "Float (Alpha)"=&RandomRangeFloat(Minimum=0.1, Maximum=0.25)
    )
    Lifetime=0.5
    Sprite Rotation Mode=::NewEnumerator2
    Sprite Size Mode=::NewEnumerator1
    Uniform Sprite Size Max=40
    Uniform Sprite Size Min=25
    UsePositionOffset=false
    Write Material Random=true
   )
   &CameraOffset(Camera Offset Amount=-5)
   &SkeletalMeshLocation(
    Bone / Socket Index=&SpawnGroup()
    Bone Sampling Mode=::NewEnumerator4
    MaskSpawnGroup=false
    Skeletal Mesh=$User.PlayerPawn
   )
   &AddVelocity(Velocity=<0 0 500>) {module=AddVelocity001}
  ]
  particleUpdateScript [
   &ParticleState()
   &InheritVelocity(
    Inherited Velocity Amount Scale=&RandomRangeVector(Minimum=<0.75 0.75 -0.1>, Maximum=<0.75 0.75 -0.3>)
    Speed Limit=false
    UseSpeedThreshold=false
   )
   &Drag(Drag=5, Use Linear Drag=true, Use Rotational Drag=false)
   &SolveForcesAndVelocity(Clamp Velocity=false, Limit Acceleration=false)
   &ScaleColor(
    Scale Alpha=&FloatFromCurve(FloatCurve=DataInterface() {propertyValues {Curve {keys [
         {arriveTangent=7.0976, interpMode=RCIM_Cubic, leaveTangent=7.0976, tangentMode=RCTM_User, time=0, value=0}
         {arriveTangent=-1.1259, interpMode=RCIM_Cubic, leaveTangent=-1.1259, tangentMode=RCTM_User, time=0.4038, value=0.9426}
         <at 1 0.2629 RCIM_Cubic>
        ]}}})
    ScaleA=true
    ScaleRGB=true
   )
   &ScaleSpriteSize(Uniform Curve Sprite Scale=DataInterface() {propertyValues {Curve {keys [
        {arriveTangent=3.1998, interpMode=RCIM_Cubic, leaveTangent=3.1998, tangentMode=RCTM_User, time=0, value=0.828}
        {arriveTangentWeight=0.2015, interpMode=RCIM_Cubic, leaveTangentWeight=0.2015, tangentMode=RCTM_User, time=0.5, value=1.5901}
        {arriveTangent=-0.2672, interpMode=RCIM_Cubic, leaveTangent=-0.2672, tangentMode=RCTM_User, time=1, value=1.5072}
       ]}}})
   &DynamicMaterialParameters(
    Index 0 Param 1=&FloatFromCurve(FloatCurve=DataInterface() {propertyValues {Curve {keys [<at 0.5 1>,<at 0.7 0>]}}})
    Index 0 Param 2=&FloatFromCurve(FloatCurve=DataInterface() {propertyValues {Curve {keys [<at 0 1>, {arriveTangent=-1, interpMode=RCIM_Cubic, leaveTangent=-1, time=0.3, value=1.0101}, <at 1 0>]}}})
    Module.Param0WriteEnabled=true
    Module.Param1WriteEnabled=true
    Module.Param2WriteEnabled=true
    Module.Param3WriteEnabled=true
   )
   &KillParticlesInVolume(
    Invert Volume=true
    Kill Shape=::NewEnumerator2
    Origin Offset:Vector3f=&Lerp_Vector(A=<0 0 300>, B=<0 0 800>, Alpha=$User.ParaGlideLeanAlpha)
   )
  ]
  rendererClasses [&NiagaraSpriteRendererProperties]
  renderers [{rendererClass=&NiagaraSpriteRendererProperties}]
 }
 Smoke {
  emitterUpdateScript [
   &EmitterState()
   &SpawnRate(SpawnRate=10, Use Spawn Probability=false)
   &SpawnRate(SpawnGroup=1, SpawnRate=10, Use Spawn Probability=false) {module=SpawnRate001}
  ]
  particleSpawnScript [
   &InitializeParticle(
    Color=&MakeLinearColorFromVectorAndFloat(
     "Vector (RGB)"=&MakeVectorFromLinearColorRGB(Color=<rgba 0.0573 0.0573 0.0573 1>)
     "Float (Alpha)"=&RandomRangeFloat(Minimum=0.2, Maximum=0.5)
    )
    Sprite Rotation Mode=::NewEnumerator2
    Sprite Size Mode=::NewEnumerator1
    Uniform Sprite Size Max=100
    Uniform Sprite Size Min=50
    UsePositionOffset=false
    Write Material Random=true
   )
   &CameraOffset(Camera Offset Amount=-20)
  ]
  particleUpdateScript [
   &ParticleState()
   &SkeletalMeshLocation(
    Bone / Socket Index=&SpawnGroup()
    Bone Sampling Mode=::NewEnumerator4
    MaskSpawnGroup=false
    Sampled Position Offset=&SelectVectorFromArray(
     Direct Array Index=&SpawnGroup()
     Vector Selection Array=DataInterface() {propertyValues {FloatData [<7 0 0>, <-7 0 0>]}}
    )
    Skeletal Mesh=$User.PlayerPawn
   )
   &SubUVAnimation(Sub UV Play Rate=0.2, SubUV Animation Mode:ENiagara_SubUVLookupMode=::NewEnumerator3)
   &ScaleColor(
    Scale Alpha=&FloatFromCurve(FloatCurve=DataInterface() {propertyValues {Curve {keys [
         {arriveTangent=3.8929, interpMode=RCIM_Cubic, leaveTangent=3.8929, tangentMode=RCTM_User, time=0, value=0}
         <at 0.25 1 RCIM_Cubic>
         <at 1 0 RCIM_Cubic>
        ]}}})
    ScaleA=true
    ScaleRGB=true
   )
   &DynamicMaterialParameters(
    Index 0 Param 1=&FloatFromCurve(FloatCurve=DataInterface() {propertyValues {Curve {keys [<at 0.5 0.5>,<at 1 -0.1>]}}})
    Index 0 Param 4=1
    Module.Param0WriteEnabled=true
    Module.Param1WriteEnabled=true
    Module.Param2WriteEnabled=true
    Module.Param3WriteEnabled=true
   )
   &ScaleSpriteSize(Uniform Curve Sprite Scale=DataInterface() {propertyValues {Curve {keys [<at 0 0.5 RCIM_Cubic>, {arriveTangentWeight=0.4788, interpMode=RCIM_Cubic, tangentMode=RCTM_User, time=1, value=0.7652}]}}})
  ]
  rendererClasses [&NiagaraSpriteRendererProperties]
  renderers [{rendererClass=&NiagaraSpriteRendererProperties}]
 }
 Source {
  emitterUpdateScript [
   &EmitterState()
   &SpawnRate(SpawnRate=2, Use Spawn Probability=false)
   &SpawnBurst_Instantaneous(Spawn Count=3, Use Spawn Probability=false, UseLoopCountLimit=false) {enabled=false}
  ]
  particleSpawnScript [
   &InitializeParticle(Color=<rgba 0 6.7708 20 1>, Sprite Size Mode=::NewEnumerator3, Uniform Sprite Size=5, UsePositionOffset=false, Write Material Random=true)
   &ShapeLocation(
    Hemisphere Distribution=<1 0.5>
    MaskSpawnGroup=false
    Offset=<0 0 50>
    Offset Mode=::NewEnumerator0
    Override Seed=false
    Sphere Radius=200
    UseOwnerScale=true
   )
   &AddVelocity(Velocity=<0 0 -1200>)
  ]
  particleUpdateScript [
   &ParticleState()
   &InheritVelocity(
    Inherited Velocity Amount Scale=&RandomRangeVector(Minimum=<-0.1 -0.1 -1>, Maximum=<0.1 0.1 -1>)
    Speed Limit=true
    UseSpeedThreshold=false
   )
   &SolveForcesAndVelocity(Clamp Velocity=false, Limit Acceleration=false)
   &ScaleSpriteSizeBySpeed(Max Scale Factor=<1 10>)
   &GenerateLocationEvent(Use Event Probability=false, UseEventDelay=false)
  ]
  rendererClasses [&NiagaraSpriteRendererProperties]
  renderers [{rendererClass=&NiagaraSpriteRendererProperties}]
 }
 Source_Ribbons {
  emitterUpdateScript [
   &EmitterState()
   &SpawnRate(SpawnRate=3, Use Spawn Probability=false) {enabled=false}
   &SpawnParticlesFromOtherEmitter(
    Attribute Reader=DataInterface() {propertyValues {EmitterBinding {bindingMode=Other, emitterName=Source}}}
    Use Maximum Age=false
    Use Minimum Age=false
   ) {enabled=false}
  ]
  particleSpawnScript [
   &InitializeParticle(
    Color=<rgba 0 3.3854 10 0.15>
    Lifetime=0.2
    Sprite Size Mode=::NewEnumerator1
    Uniform Sprite Size Max=5
    Uniform Sprite Size Min=2
    UsePositionOffset=false
    Write Material Random=true
   ) {enabled=false}
   &InitializeRibbon(Lifetime=0.4, UsePositionOffset=false, Write Material Random=true)
  ]
  particleUpdateScript [
   &ParticleState()
   &ScaleColor(
    Scale Alpha=&FloatFromCurve(FloatCurve=DataInterface() {propertyValues {Curve {keys [
         {arriveTangent=65.2399, interpMode=RCIM_Cubic, leaveTangent=65.2399, tangentMode=RCTM_User, time=0, value=0}
         <at 0.03 1 RCIM_Cubic>
         <at 1 0 RCIM_Cubic>
        ]}}})
    Scale RGB=&VectorFromFloat(Value=&FloatFromCurve(
      FloatCurve=DataInterface() {propertyValues {Curve {keys [{interpMode=RCIM_Cubic, tangentMode=RCTM_User, time=0, value=1}, {interpMode=RCIM_Cubic, tangentMode=RCTM_User, time=1, value=0}]}}}
      Scale Curve=20
     ))
    ScaleA=true
    ScaleRGB=true
   )
   &ScaleRibbonWidth(Ribbon Width Scale=&FloatFromCurve(FloatCurve=DataInterface() {propertyValues {Curve {keys [<at 0 1>,<at 1 0>]}}}))
  ]
  rendererClasses [&NiagaraRibbonRendererProperties]
  renderers [{rendererClass=&NiagaraRibbonRendererProperties}]
 }
}
summary {
 systemName=NS_Contrail_Celestial
 emitters [
  {emitterName=Hand_Ribbons, rendererClasses [&NiagaraRibbonRendererProperties]}
  {emitterName=Flare_Complex, simTarget=GPUComputeSim, rendererClasses [&NiagaraSpriteRendererProperties]}
  {emitterName=Glows, rendererClasses [&NiagaraSpriteRendererProperties]}
  {emitterName=Hand_Emitting_Points, simTarget=GPUComputeSim, rendererClasses [&NiagaraSpriteRendererProperties]}
  {emitterName=Smoke, rendererClasses [&NiagaraSpriteRendererProperties]}
  {emitterName=Emitting_Points, rendererClasses [&NiagaraSpriteRendererProperties]}
  {emitterName=Source_Ribbons, rendererClasses [&NiagaraRibbonRendererProperties]}
  {emitterName=Source, rendererClasses [&NiagaraSpriteRendererProperties]}
  {emitterName=Ribbon_BKGD, rendererClasses [&NiagaraSpriteRendererProperties]}
 ]
 userVariables [
  User.bIsFrontEnd {type=NiagaraBool}
  User.bIsFrontEndPreview {type=NiagaraBool}
  User.ParaGlideLeanAlpha {type=NiagaraFloat, defaultValue=0, description=""}
  User.PlayerPawn {
   type=NiagaraDataInterfaceSkeletalMesh
   defaultValue {
    struct=NiagaraExt_VariableValue_DataInterface
    value {dataInterface=&self:NiagaraDataInterfaceSkeletalMesh_0}
   }
  }
 ]
}
