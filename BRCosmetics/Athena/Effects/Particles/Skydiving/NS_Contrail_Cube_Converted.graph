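# Cascade-converted Niagara skydiving contrail for the Cube cosmetic, layering purple lightning, smoke and sparks with blue-green runes on the character.
# Mesh lightning and sprite/light glows follow both hand bones, while sparks, small squares and hand runes also spawn at the foot balls.
# Ribbon, smoke and trail runes sample sockets on the system's skeletal mesh, alongside a shared FN_Contrail_Setup module.
# Sequential indices alternate attachment points, and curves animate color, size and material parameters.
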NiagaraSystem /BRCosmetics/Athena/Effects/Particles/Skydiving/NS_Contrail_Cube_Converted
dependencies {
 usedDataInterfaces [
  &NiagaraDataInterfaceCurve
  &NiagaraDataInterfaceVectorCurve
  &NiagaraDataInterfaceArrayFloat3
  &NiagaraDataInterfaceSpriteRendererInfo
  &NiagaraDataInterfaceSkeletalMesh
  &NiagaraDataInterfaceColorCurve
  &NiagaraDataInterfaceVector2DCurve
 ]
 usedDynamicInputs [&Lerp_Float]
 usedModules [
  &self:SystemSpawnScript.SystemScriptSource.SystemScriptGraph.NiagaraNodeAssignment_0.SetVariables_#1
  &SystemState
  /Game/Effects/Niagara/Modules/Contrails/FN_Contrail_Setup
  &FN_Integrator
 ]
}
emitters {
 BaseLightning {
  emitterUpdateScript [
   &EmitterState(Inactive Response=::NewEnumerator1, Life Cycle Mode=::NewEnumerator1, Loop Duration=1, UseLoopDelay=true)
   &SpawnRate(SpawnRate=24, Use Spawn Probability=false)
  ]
  particleSpawnScript [
   /SetVariables(Particles.MeshRotationEuler:Vector3f=&RandomRangeVector(Minimum=<0 0 0>, Maximum=<1 0 0>, Evaluation Type=::NewEnumerator1)) {isSetParameters=true}
   /SetVariables(Particles.MeshOrientation=/CascadeToNiagaraConverter/NiagaraScripts/DynamicInputScripts/MakeQuatFromEuler(Euler:Vector3f=$Particles.MeshRotationEuler, Scale:NiagaraFloat=360)) {isSetParameters=true}
   /SetVariables(Particles.SkeletonSamplingIndex:NiagaraInt32=/CascadeToNiagaraConverter/NiagaraScripts/DynamicInputScripts/SequentialInt(Use Array Data:NiagaraBool=false, First Number:NiagaraInt32=0, Sequence:NiagaraDataInterfaceArrayInt32=?, Last Number:NiagaraInt32=1)) {isSetParameters=true}
   &InitializeParticle(
    Lifetime=&RandomRangeFloat(Minimum=0.14, Maximum=0.25, Evaluation Type=::NewEnumerator1)
    Mesh Scale=&RandomRangeVector(Minimum=<1.2 1 1>, Maximum=<1.7 1.8 1.7>, Evaluation Type=::NewEnumerator1)
    Mesh Scale Mode=::NewEnumerator4
    UsePositionOffset=false
    Write Material Random=true
   )
   &Color(Color=&MakeLinearColorFromVectorAndFloat("Vector (RGB)"=<13.4002 0.7031 15>, "Float (Alpha)"=1))
   &DynamicMaterialParameters(
    Index 0 Param 1=&RandomRangeFloat(Minimum=2, Maximum=3, Evaluation Type=::NewEnumerator1)
    Index 0 Param 3=&RandomRangeFloat(Minimum=0.2, Maximum=2, Evaluation Type=::NewEnumerator1)
    Index 0 Param 4=&RandomRangeFloat(Minimum=0, Maximum=1, Evaluation Type=::NewEnumerator1)
    Module.Param0WriteEnabled=true
    Module.Param1WriteEnabled=false
    Module.Param2WriteEnabled=true
    Module.Param3WriteEnabled=true
   )
   &SkeletalMeshLocation(
    Bone / Socket Index=$Particles.SkeletonSamplingIndex
    Bone Sampling Mode=::NewEnumerator4
    MaskSpawnGroup=false
    Orientation Sampling=::NewEnumerator1
    Sampled Position Offset=&SelectVectorFromArray(
     Direct Array Index=$Particles.SkeletonSamplingIndex
     Vector Selection Array=DataInterface() {propertyValues {FloatData [<-10 10 1>, <10 -10 1>]}}
    )
    Skeletal Mesh=DataInterface() {propertyValues {
      DefaultMesh=null
      ExcludeBoneName=root
      OptionalComponentClass=null
      PreviewMesh=null
      SkinningMode=SkinOnTheFly
      SoftSourceActor=null
      SourceMode=Default
      UvSetIndex=0
      WholeMeshLOD=-1
      bExcludeBone=false
      bReadDeformedGeometry=true
      bRequireCurrentFrameData=true
      ComponentTags []
      FilteredBones [hand_r,hand_l]
      FilteredSockets []
      MeshUserParameter {parameter {name=None, typeDefHandle {registeredTypeIndex=1}, varData []}}
      SamplingRegions []
     }}
   )
   &ShapeLocation(MaskSpawnGroup=false, Override Seed=false, Sphere Radius=5, UseOwnerScale=true)
   /CascadeToNiagaraConverter/NiagaraScripts/ModuleScripts/FixShapeVector()
   &AddVelocity(Velocity=&Multiply_VectorByFloat(
     Vector=$Output.FixShapeVector.ShapeVector
     Float=&RandomRangeFloat(Minimum=1, Maximum=25, Evaluation Type=::NewEnumerator1)
    ))
  ]
  particleUpdateScript [
   &ParticleState()
   &ScaleMeshSize(Scale Factor=&VectorFromCurve(VectorCurve=DataInterface() {propertyValues {
       XCurve {keys [<at 0 1 RCIM_Constant>,<at 0.1918 0.7415 RCIM_Constant>,<at 0.4053 0.5034 RCIM_Constant>,<at 0.6009 0.2891 RCIM_Constant>,<at 0.8 0.1293>]}
       YCurve {keys [<at 0 0.6528 RCIM_Constant>,<at 0.1918 0.6027 RCIM_Constant>,<at 0.4053 0.3331 RCIM_Constant>,<at 0.6009 0.6577 RCIM_Constant>,<at 0.8 1>]}
       ZCurve {keys [<at 0 0.8843 RCIM_Constant>,<at 0.1918 0.4904 RCIM_Constant>,<at 0.4053 0.4534 RCIM_Constant>,<at 0.6009 0.5861 RCIM_Constant>,<at 0.8 1>]}
      }}))
   &DynamicMaterialParameters(
    Index 0 Param 2=&FloatFromCurve(FloatCurve=DataInterface() {propertyValues {Curve {keys [{arriveTangent=-1.8894, interpMode=RCIM_Cubic, leaveTangent=-1.8894, tangentMode=RCTM_User, time=0, value=1}, <at 1 -0.25>]}}})
    Module.Param0WriteEnabled=false
    Module.Param1WriteEnabled=true
    Module.Param2WriteEnabled=false
    Module.Param3WriteEnabled=false
   ) {module=DynamicMaterialParameters001}
   &ScaleColor(
    Scale RGB=&VectorFromFloat(Value=&FloatFromCurve(FloatCurve=DataInterface() {propertyValues {Curve {keys [<at 0 50 RCIM_Cubic>, {arriveTangent=-45, interpMode=RCIM_Cubic, leaveTangent=-45, time=0.5, value=30}, <at 1 5>]}}}))
    ScaleA=true
    ScaleRGB=true
   )
   &CameraOffset(Camera Offset Amount=-80)
   &SolveForcesAndVelocity(Clamp Velocity=false, Limit Acceleration=false)
  ]
  rendererClasses [&NiagaraMeshRendererProperties]
  renderers [{rendererClass=&NiagaraMeshRendererProperties, rendererIndex=0}]
 }
 HandGlow {
  emitterUpdateScript [
   &EmitterState(Inactive Response=::NewEnumerator1, Life Cycle Mode=::NewEnumerator1, Loop Duration=0.125, UseLoopDelay=true)
   &SpawnRate(SpawnRate=6, Use Spawn Probability=false)
   &SpawnRate(SpawnGroup=1, SpawnRate=3, Use Spawn Probability=false) {module=SpawnRate001}
  ]
  particleSpawnScript [
   /SetVariables(
    Particles.CoronaCameraOffset=100
    Particles.CoronaColor:LinearColor=&MakeLinearColorFromVectorAndFloat("Vector (RGB)"=<1.547 0.5332 2>, "Float (Alpha)"=1)
    Particles.CoronaSize:Vector2f=<38 38>
    Particles.VisibilityTag=&SpawnGroup()
   ) {isSetParameters=true}
   /SetVariables(Particles.SkeletonSamplingIndex:NiagaraInt32=&SequentialInt(Use Array Data:NiagaraBool=false, First Number:NiagaraInt32=0, Sequence:NiagaraDataInterfaceArrayInt32=?, Last Number:NiagaraInt32=1)) {isSetParameters=true}
   &InitializeParticle(
    Lifetime=&MaskFloatBySpawnGroup(
     Masked Float=1.2
     Passthrough Float=&RandomRangeFloat(Minimum=0.4, Maximum=1, Evaluation Type=::NewEnumerator1)
     Spawn Group Mask=1
    )
    Sprite Size=<80 80>
    Sprite Size Mode=::NewEnumerator4
    UsePositionOffset=false
    Write Material Random=true
   )
   &Color(Color=&MakeLinearColorFromVectorAndFloat("Vector (RGB)"=<0.5587 0.217 1>, "Float (Alpha)"=0.4))
  ]
  particleUpdateScript [
   /SetVariables(Particles.LightExponent=3) {isSetParameters=true}
   &ParticleState()
   &ScaleSpriteSize(
    Scale Factor=&Vector2DFromFloat(Value=&FloatFromCurve(FloatCurve=DataInterface() {propertyValues {Curve {keys [<at 0 0.6>,<at 1 1.8>]}}}))
    Scale Sprite Size Mode=::NewEnumerator2
   )
   /SetVariables(
    Particles.CoronaColor:LinearColor=&MultiplyLinearColor(
     Linear Color=$Particles.Initial.CoronaColor
     Scale Factor=&Vector4FromCurve(Vector4Curve=DataInterface() {propertyValues {
        AlphaCurve {
         defaultValue=1
         keys [{arriveTangent=-2.4776, interpMode=RCIM_Cubic, leaveTangent=-2.4776, tangentMode=RCTM_User, time=0, value=1}, <at 1 0>]
        }
        BlueCurve {keys [<at 0 1>,<at 1 1>]}
        GreenCurve {keys [<at 0 1>,<at 1 1>]}
        RedCurve {keys [<at 0 1>,<at 1 1>]}
       }})
    )
    Particles.CoronaSize:Vector2f=&Multiply_Vector2D(
     A=$Particles.Initial.CoronaSize
     B=&Vector2DFromFloat(Value=&FloatFromCurve(FloatCurve=DataInterface() {propertyValues {Curve {keys [
           {arriveTangent=1.5255, interpMode=RCIM_Cubic, leaveTangent=1.5255, tangentMode=RCTM_User, time=0, value=0.65}
           {arriveTangent=0.2191, interpMode=RCIM_Cubic, leaveTangent=0.2191, tangentMode=RCTM_User, time=1, value=1}
          ]}}}))
    )
   ) {isSetParameters=true}
   &CameraOffset(Camera Offset Amount=&RandomRangeFloat(Minimum=10, Maximum=40, Evaluation Type=::NewEnumerator1))
   &ScaleColor(
    Scale Alpha=&FloatFromCurve(FloatCurve=DataInterface() {propertyValues {Curve {keys [<at 0 0 RCIM_Cubic>,<at 0.2321 0.9177 RCIM_Cubic>,<at 1 0>]}}})
    Scale RGB=&VectorFromCurve(VectorCurve=DataInterface() {propertyValues {
       XCurve {keys [<at 0 3 RCIM_Cubic>,<at 0.8939 0.141 RCIM_Cubic>]}
       YCurve {keys [<at 0 0.07 RCIM_Cubic>,<at 0.8939 0.0568 RCIM_Cubic>]}
       ZCurve {keys [<at 0 7 RCIM_Cubic>,<at 0.8939 0.5473 RCIM_Cubic>]}
      }})
    ScaleA=true
    ScaleRGB=true
   )
   &SkeletalMeshLocation(
    Bone / Socket Index=$Particles.SkeletonSamplingIndex
    Bone Sampling Mode=::NewEnumerator4
    MaskSpawnGroup=false
    Orientation Sampling=::NewEnumerator1
    Sampled Position Offset=&SelectVectorFromArray(
     Direct Array Index=$Particles.SkeletonSamplingIndex
     Vector Selection Array=DataInterface() {propertyValues {FloatData [<10 -5 0>, <-10 5 0>]}}
    )
    Skeletal Mesh=DataInterface() {propertyValues {
      DefaultMesh=null
      ExcludeBoneName=root
      OptionalComponentClass=null
      PreviewMesh=null
      SkinningMode=SkinOnTheFly
      SoftSourceActor=null
      SourceMode=Default
      UvSetIndex=0
      WholeMeshLOD=-1
      bExcludeBone=false
      bReadDeformedGeometry=true
      bRequireCurrentFrameData=true
      ComponentTags []
      FilteredBones [hand_l,hand_r]
      FilteredSockets []
      MeshUserParameter {parameter {name=None, typeDefHandle {registeredTypeIndex=1}, varData []}}
      SamplingRegions []
     }}
   )
   &CascadeConversion_LightProperties(LightBrightnessScale=1, LightColorScale=<10 1 25>, LightRadiusScale=1)
   &DynamicMaterialParameters(
    Index 0 Param 1=&FloatFromCurve(FloatCurve=DataInterface() {propertyValues {Curve {keys [{arriveTangent=-1.5232, interpMode=RCIM_Cubic, leaveTangent=-1.5232, tangentMode=RCTM_User, time=0, value=1.5}, <at 1 0.5 RCIM_Cubic>]}}})
    Index 0 Param 2=0.4
    Index 0 Param 3=1
    Index 0 Param 4=1
    Module.Param0WriteEnabled=true
    Module.Param1WriteEnabled=true
    Module.Param2WriteEnabled=true
    Module.Param3WriteEnabled=true
   )
  ]
  rendererClasses [&NiagaraSpriteRendererProperties, &NiagaraLightRendererProperties]
  renderers [
   {rendererClass=&NiagaraSpriteRendererProperties}
   {rendererClass=&NiagaraLightRendererProperties, rendererIndex=1}
   {rendererClass=&NiagaraSpriteRendererProperties, rendererIndex=2}
  ]
 }
 Hand_R {
  emitterUpdateScript [
   &EmitterState(Inactive Response=::NewEnumerator1, Life Cycle Mode=::NewEnumerator1, Loop Duration=1, UseLoopDelay=true)
   &SpawnRate(SpawnRate=120, Use Spawn Probability=false)
  ]
  particleSpawnScript [
   &InitializeParticle(
    Lifetime=0.4
    Ribbon Width=&Multiply_Float(A=20, B=2)
    Ribbon Width Mode=::NewEnumerator1
    UsePositionOffset=false
    Write Material Random=true
   )
   &Color(Color=&MakeLinearColorFromVectorAndFloat("Vector (RGB)"=<0.3802 0.3802 0.3802>, "Float (Alpha)"=2))
   &SkeletalMeshLocation(
    Bone / Socket Index=&SequentialInt(Use Array Data:NiagaraBool=false, First Number:NiagaraInt32=0, Sequence:NiagaraDataInterfaceArrayInt32=?, Last Number:NiagaraInt32=1)
    MaskSpawnGroup=false
    Mesh Sampling Type=::NewEnumerator5
    Orientation Sampling=::NewEnumerator1
    Skeletal Mesh=$System.Skeletal Mesh
    Socket Sampling Mode=::NewEnumerator5
   )
   /CascadeToNiagaraConverter/NiagaraScripts/ModuleScripts/AddVectorToPosition(Vector:Vector3f=&TransformVector(Vector=<0 0 0>, Destination Space=::Simulation))
  ]
  particleUpdateScript [
   &ParticleState()
   &ScaleColor(
    Scale Alpha=&FloatFromCurve(FloatCurve=DataInterface() {propertyValues {Curve {keys [
         {arriveTangent=2.0102, time=0, value=0}
         {arriveTangent=-0.5102, leaveTangent=-0.5102, time=0.0349, value=1}
         {arriveTangent=-1.5, leaveTangent=-1.5, time=0.4185, value=0.75}
         <at 0.7 0>
        ]}}})
    Scale RGB=&VectorFromFloat(Value=&FloatFromCurve(FloatCurve=DataInterface() {propertyValues {Curve {keys [
          {arriveTangent=-5.5, interpMode=RCIM_Cubic, leaveTangent=-51.4373, tangentMode=RCTM_Break, time=0, value=5}
          {arriveTangent=-14.8011, interpMode=RCIM_Cubic, leaveTangent=-13.3728, tangentMode=RCTM_Break, time=0.1, value=5}
          <at 0.19 0>
         ]}}}))
    ScaleA=true
    ScaleRGB=true
   )
   &DynamicMaterialParameters(
    Index 0 Param 1=0.5
    Index 0 Param 2=15
    Index 0 Param 3=&FloatFromCurve(FloatCurve=DataInterface() {propertyValues {Curve {keys [
         {arriveTangent=-1.8399, interpMode=RCIM_Cubic, leaveTangent=-1.8399, tangentMode=RCTM_User, time=-0.0058, value=1.367}
         {arriveTangent=-3.8231, interpMode=RCIM_Cubic, leaveTangent=-3.8231, time=0.2238, value=0.7705}
         {arriveTangent=-0.9926, interpMode=RCIM_Cubic, leaveTangent=-0.9926, time=0.5993, value=-0.9465}
         <at 1 0>
        ]}}})
    Index 0 Param 4=1
    Module.Param0WriteEnabled=true
    Module.Param1WriteEnabled=true
    Module.Param2WriteEnabled=true
    Module.Param3WriteEnabled=true
   )
  ]
  rendererClasses [&NiagaraRibbonRendererProperties]
  renderers [{rendererClass=&NiagaraRibbonRendererProperties}]
 }
 Lightning {
  emitterUpdateScript [
   &EmitterState(Inactive Response=::NewEnumerator1, Life Cycle Mode=::NewEnumerator1, Loop Duration=1, UseLoopDelay=true)
   &SpawnRate(SpawnRate=0.9, Use Spawn Probability=false)
  ]
  particleSpawnScript [
   /SetVariables(Particles.MeshRotationEuler:Vector3f=&RandomRangeVector(Minimum=<0 0 0>, Maximum=<1 0 0>, Evaluation Type=::NewEnumerator1)) {isSetParameters=true}
   /SetVariables(Particles.MeshOrientation=&MakeQuatFromEuler(Euler:Vector3f=$Particles.MeshRotationEuler, Scale:NiagaraFloat=360)) {isSetParameters=true}
   /SetVariables(Particles.SkeletonSamplingIndex:NiagaraInt32=&SequentialInt(Use Array Data:NiagaraBool=false, First Number:NiagaraInt32=0, Sequence:NiagaraDataInterfaceArrayInt32=?, Last Number:NiagaraInt32=1)) {isSetParameters=true}
   &InitializeParticle(
    Lifetime=&RandomRangeFloat(Minimum=0.15, Maximum=0.25, Evaluation Type=::NewEnumerator1)
    Mesh Scale=&RandomRangeVector(Minimum=<4.2 2.2 2.4>, Maximum=<5 4 4>, Evaluation Type=::NewEnumerator1)
    Mesh Scale Mode=::NewEnumerator4
    UsePositionOffset=false
    Write Material Random=true
   )
   &Color(Color=&MakeLinearColorFromVectorAndFloat("Vector (RGB)"=<13.4002 0.7031 15>, "Float (Alpha)"=1))
   &DynamicMaterialParameters(
    Index 0 Param 1=&RandomRangeFloat(Minimum=2, Maximum=3, Evaluation Type=::NewEnumerator1)
    Index 0 Param 3=&RandomRangeFloat(Minimum=0.2, Maximum=2, Evaluation Type=::NewEnumerator1)
    Index 0 Param 4=&RandomRangeFloat(Minimum=0, Maximum=1, Evaluation Type=::NewEnumerator1)
    Module.Param0WriteEnabled=true
    Module.Param1WriteEnabled=false
    Module.Param2WriteEnabled=true
    Module.Param3WriteEnabled=true
   )
   &SkeletalMeshLocation(
    Bone / Socket Index=$Particles.SkeletonSamplingIndex
    Bone Sampling Mode=::NewEnumerator4
    MaskSpawnGroup=false
    Orientation Sampling=::NewEnumerator1
    Sampled Position Offset=&SelectVectorFromArray(
     Direct Array Index=$Particles.SkeletonSamplingIndex
     Vector Selection Array=DataInterface() {propertyValues {FloatData [<-10 10 1>, <10 -10 1>]}}
    )
    Skeletal Mesh=DataInterface() {propertyValues {
      DefaultMesh=null
      ExcludeBoneName=root
      OptionalComponentClass=null
      PreviewMesh=null
      SkinningMode=SkinOnTheFly
      SoftSourceActor=null
      SourceMode=Default
      UvSetIndex=0
      WholeMeshLOD=-1
      bExcludeBone=false
      bReadDeformedGeometry=true
      bRequireCurrentFrameData=true
      ComponentTags []
      FilteredBones [hand_r,hand_l]
      FilteredSockets []
      MeshUserParameter {parameter {name=None, typeDefHandle {registeredTypeIndex=1}, varData []}}
      SamplingRegions []
     }}
   )
   &ShapeLocation(MaskSpawnGroup=false, Override Seed=false, Sphere Radius=1.2, UseOwnerScale=true)
   &FixShapeVector()
   &AddVelocity(Velocity=&Multiply_VectorByFloat(
     Vector=$Output.FixShapeVector.ShapeVector
     Float=&RandomRangeFloat(Minimum=5, Maximum=25, Evaluation Type=::NewEnumerator1)
    ))
  ]
  particleUpdateScript [
   &ParticleState()
   &ScaleMeshSize(Scale Factor=&VectorFromCurve(VectorCurve=DataInterface() {propertyValues {
       XCurve {keys [<at 0 1 RCIM_Constant>,<at 0.1918 0.7415 RCIM_Constant>,<at 0.4053 0.5034 RCIM_Constant>,<at 0.6009 0.2891 RCIM_Constant>,<at 0.8 0.1293>]}
       YCurve {keys [<at 0 0.6528 RCIM_Constant>,<at 0.1918 0.6027 RCIM_Constant>,<at 0.4053 0.3331 RCIM_Constant>,<at 0.6009 0.6577 RCIM_Constant>,<at 0.8 1>]}
       ZCurve {keys [<at 0 0.8843 RCIM_Constant>,<at 0.1918 0.4904 RCIM_Constant>,<at 0.4053 0.4534 RCIM_Constant>,<at 0.6009 0.5861 RCIM_Constant>,<at 0.8 1>]}
      }}))
   &DynamicMaterialParameters(
    Index 0 Param 2=&FloatFromCurve(FloatCurve=DataInterface() {propertyValues {Curve {keys [{arriveTangent=-1.8894, interpMode=RCIM_Cubic, leaveTangent=-1.8894, tangentMode=RCTM_User, time=0, value=1}, <at 1 -0.25>]}}})
    Module.Param0WriteEnabled=false
    Module.Param1WriteEnabled=true
    Module.Param2WriteEnabled=false
    Module.Param3WriteEnabled=false
   ) {module=DynamicMaterialParameters001}
   &ScaleColor(
    Scale RGB=&VectorFromFloat(Value=&FloatFromCurve(FloatCurve=DataInterface() {propertyValues {Curve {keys [<at 0 50 RCIM_Cubic>, {arriveTangent=-45, interpMode=RCIM_Cubic, leaveTangent=-45, time=0.5, value=30}, <at 1 5>]}}}))
    ScaleA=true
    ScaleRGB=true
   )
   &CameraOffset(Camera Offset Amount=-80)
   &SolveForcesAndVelocity(Clamp Velocity=false, Limit Acceleration=false)
  ]
  rendererClasses [&NiagaraMeshRendererProperties]
  renderers [{rendererClass=&NiagaraMeshRendererProperties, rendererIndex=0}]
 }
 RunesHand {
  emitterUpdateScript [
   &EmitterState(Inactive Response=::NewEnumerator1, Life Cycle Mode=::NewEnumerator1, Loop Duration=1, UseLoopDelay=true)
   &SpawnRate(SpawnRate=12, Use Spawn Probability=false)
  ]
  particleSpawnScript [
   &InitializeParticle(
    Lifetime=0.7
    Sprite Size=&Vector2DFromFloat(Value=&RandomRangeFloat(Minimum=6, Maximum=10, Evaluation Type=::NewEnumerator1))
    Sprite Size Mode=::NewEnumerator4
    UsePositionOffset=false
    Write Material Random=true
   )
   &AddVectorToPosition(Vector:Vector3f=&TransformVector(
     Vector=&RandomRangeVector(Minimum=<-50 -50 -50>, Maximum=<50 50 50>, Evaluation Type=::NewEnumerator1)
     Destination Space=::Simulation
    ))
   &Color(Color=&MakeLinearColorFromVectorAndFloat(
     "Vector (RGB)"=&RandomRangeVector(Minimum=<1 1.2 8>, Maximum=<0 1 4>, Evaluation Type=::NewEnumerator1)
     "Float (Alpha)"=0.4
    ))
   &SkeletalMeshLocation(
    Bone / Socket Index=&SequentialInt(Use Array Data:NiagaraBool=false, First Number:NiagaraInt32=0, Sequence:NiagaraDataInterfaceArrayInt32=?, Last Number:NiagaraInt32=3)
    Bone Sampling Mode=::NewEnumerator4
    MaskSpawnGroup=false
    Orientation Sampling=::NewEnumerator1
    Skeletal Mesh=DataInterface() {propertyValues {
      DefaultMesh=null
      ExcludeBoneName=root
      OptionalComponentClass=null
      PreviewMesh=null
      SkinningMode=SkinOnTheFly
      SoftSourceActor=null
      SourceMode=Default
      UvSetIndex=0
      WholeMeshLOD=-1
      bExcludeBone=false
      bReadDeformedGeometry=true
      bRequireCurrentFrameData=true
      ComponentTags []
      FilteredBones [hand_r,hand_l,ball_r,ball_l]
      FilteredSockets []
      MeshUserParameter {parameter {name=None, typeDefHandle {registeredTypeIndex=1}, varData []}}
      SamplingRegions []
     }}
   )
   &ShapeLocation(
    Cylinder Height=50
    Cylinder Radius=12
    MaskSpawnGroup=false
    Override Seed=false
    Rotation Mode=::NewEnumerator2
    Shape Primitive=::NewEnumerator1
    Surface Only=false
    UseOwnerScale=true
   )
  ]
  particleUpdateScript [
   &ParticleState()
   &ScaleColor(
    Scale Alpha=&FloatFromCurve(FloatCurve=DataInterface() {propertyValues {Curve {keys [{arriveTangent=-2.2945, interpMode=RCIM_Cubic, leaveTangent=-2.2945, tangentMode=RCTM_User, time=0, value=1}, <at 1 0>]}}})
    Scale RGB=&RandomRangeVector(Minimum=<1 18 0.2>, Maximum=<0 4 4>, Evaluation Type=::NewEnumerator1)
    ScaleA=true
    ScaleRGB=true
   )
   &CameraOffset(Camera Offset Amount=5)
  ]
  rendererClasses [&NiagaraSpriteRendererProperties]
  renderers [{rendererClass=&NiagaraSpriteRendererProperties}]
 }
 RunesTrail {
  emitterUpdateScript [
   &EmitterState(Inactive Response=::NewEnumerator1, Life Cycle Mode=::NewEnumerator1, Loop Duration=1, UseLoopDelay=true)
   &SpawnRate(SpawnRate=10, Use Spawn Probability=false)
   &SpawnRate(SpawnGroup=1, SpawnRate=12, Use Spawn Probability=false) {module=SpawnRate001}
  ]
  particleSpawnScript [
   /SetVariables(Particles.SkeletonSamplingIndex:NiagaraInt32=&SequentialInt(Use Array Data:NiagaraBool=false, First Number:NiagaraInt32=0, Sequence:NiagaraDataInterfaceArrayInt32=?, Last Number:NiagaraInt32=1)) {isSetParameters=true}
   &InitializeParticle(
    Lifetime=&RandomRangeFloat(Minimum=0.55, Maximum=0.85, Evaluation Type=::NewEnumerator1)
    Sprite Size=&Vector2DFromFloat(Value=&RandomRangeFloat(Minimum=6, Maximum=10, Evaluation Type=::NewEnumerator1))
    Sprite Size Mode=::NewEnumerator4
    UsePositionOffset=false
    Write Material Random=true
   )
   &AddVectorToPosition(Vector:Vector3f=&TransformVector(
     Vector=&RandomRangeVector(Minimum=<-50 -50 -50>, Maximum=<50 50 50>, Evaluation Type=::NewEnumerator1)
     Destination Space=::Simulation
    ))
   &Color(Color=&MakeLinearColorFromVectorAndFloat(
     "Vector (RGB)"=&RandomRangeVector(Minimum=<1 1.2 8>, Maximum=<0 1 4>, Evaluation Type=::NewEnumerator1)
     "Float (Alpha)"=&RandomRangeFloat(Minimum=0.2, Maximum=0.4, Evaluation Type=::NewEnumerator1)
    ))
   &SkeletalMeshLocation(
    Bone / Socket Index=$Particles.SkeletonSamplingIndex
    MaskSpawnGroup=false
    Mesh Sampling Type=::NewEnumerator5
    Orientation Sampling=::NewEnumerator1
    Sampled Position Offset=&SelectVectorFromArray(Direct Array Index=$Particles.SkeletonSamplingIndex, Vector Selection Array=DataInterface() {propertyValues {FloatData [<0 15 0>, <0 0 0>]}})
    Skeletal Mesh=$System.Skeletal Mesh
    Socket Sampling Mode=::NewEnumerator5
   )
   &InheritVelocity(Inherited Velocity Amount Scale=<-0.14 -0.14 -0.2>, Inherited Velocity Speed Limit=0, Speed Limit=false, UseSpeedThreshold=false)
   &ShapeLocation(MaskSpawnGroup=false, Override Seed=false, Sphere Radius=10, UseOwnerScale=true)
   &FixShapeVector()
   &AddVelocity(Velocity=&Multiply_VectorByFloat(Vector=$Output.FixShapeVector.ShapeVector, Float=1))
  ]
  particleUpdateScript [
   &ParticleState()
   &ScaleColor(
    Scale Alpha=&FloatFromCurve(FloatCurve=DataInterface() {propertyValues {Curve {keys [{arriveTangent=-2.2945, interpMode=RCIM_Cubic, leaveTangent=-2.2945, tangentMode=RCTM_User, time=0.3, value=0}, <at 0.5 1>, <at 1 0>]}}})
    Scale RGB=&RandomRangeVector(Minimum=<1 20 0.2>, Maximum=<0 6 4>, Evaluation Type=::NewEnumerator1)
    ScaleA=true
    ScaleRGB=true
   )
   &CameraOffset(Camera Offset Amount=5)
   &SolveForcesAndVelocity(Clamp Velocity=false, Limit Acceleration=false)
  ]
  rendererClasses [&NiagaraSpriteRendererProperties]
  renderers [{rendererClass=&NiagaraSpriteRendererProperties}]
 }
 SmokeHand {
  emitterUpdateScript [
   &EmitterState(Inactive Response=::NewEnumerator1, Life Cycle Mode=::NewEnumerator1, Loop Duration=1, UseLoopDelay=true)
   &SpawnRate(SpawnRate=16, Use Spawn Probability=false)
  ]
  particleSpawnScript [
   /SetVariables(Particles.SkeletonSamplingIndex:NiagaraInt32=&SequentialInt(Use Array Data:NiagaraBool=false, First Number:NiagaraInt32=0, Sequence:NiagaraDataInterfaceArrayInt32=?, Last Number:NiagaraInt32=1)) {isSetParameters=true}
   /SetVariables(Particles.SpriteRotationRate=&RandomRangeFloat(Minimum=-0.2, Maximum=-0.25, Evaluation Type=::NewEnumerator1)) {isSetParameters=true}
   &InitializeParticle(
    Lifetime=&RandomRangeFloat(Minimum=0.7, Maximum=1.5, Evaluation Type=::NewEnumerator1)
    Sprite Rotation Angle=&RandomRangeFloat(Minimum=0, Maximum=1, Evaluation Type=::NewEnumerator1)
    Sprite Rotation Mode=::NewEnumerator1
    Sprite Size=&Vector2DFromFloat(Value=&RandomRangeFloat(Minimum=30, Maximum=52, Evaluation Type=::NewEnumerator1))
    Sprite Size Mode=::NewEnumerator4
    UsePositionOffset=false
    Write Material Random=true
   )
   &Color(Color=&MakeLinearColorFromVectorAndFloat(
     "Vector (RGB)"=<0.3694 0.263 0.5>
     "Float (Alpha)"=&RandomRangeFloat(Minimum=1, Maximum=1.4, Evaluation Type=::NewEnumerator1)
    ))
   &CameraOffset(Camera Offset Amount=&RandomRangeFloat(Minimum=10, Maximum=25, Evaluation Type=::NewEnumerator1))
   &DynamicMaterialParameters(
    Index 0 Param 1=&RandomRangeFloat(Minimum=0, Maximum=1, Evaluation Type=::NewEnumerator1)
    Index 0 Param 2=&RandomRangeFloat(Minimum=1, Maximum=2, Evaluation Type=::NewEnumerator1)
    Module.Param0WriteEnabled=true
    Module.Param1WriteEnabled=true
    Module.Param2WriteEnabled=false
    Module.Param3WriteEnabled=false
   )
  ]
  particleUpdateScript [
   &ParticleState()
   &SpriteRotationRate(Rotation Rate=&AngleConversion(
     Angle:NiagaraFloat=$Particles.SpriteRotationRate
     Angle Input:ENiagara_AngleInput=::NewEnumerator1
     Angle Output:ENiagara_AngleInput=::NewEnumerator0
    ))
   &ScaleSpriteSize(
    Scale Sprite Size Mode=::NewEnumerator0
    Uniform Scale Factor=&FloatFromCurve(FloatCurve=DataInterface() {propertyValues {Curve {keys [
         {arriveTangent=5.9454, interpMode=RCIM_Cubic, leaveTangent=5.9454, tangentMode=RCTM_User, time=0, value=0.1}
         {arriveTangent=-2.2918, interpMode=RCIM_Cubic, leaveTangent=-2.2918, tangentMode=RCTM_User, time=0.3891, value=0.8744}
         <at 1 0.3 RCIM_Cubic>
        ]}}})
   )
   &ScaleColor(
    Scale Alpha=&FloatFromCurve(FloatCurve=DataInterface() {propertyValues {Curve {keys [{arriveTangent=9.1933, interpMode=RCIM_Cubic, leaveTangent=9.1933, tangentMode=RCTM_User, time=0, value=0}, <at 0.2 0.8 RCIM_Cubic>]}}})
    Scale RGB=&VectorFromCurve(VectorCurve=DataInterface() {propertyValues {
       XCurve {keys [{arriveTangent=-418.9295, interpMode=RCIM_Cubic, leaveTangent=-418.9295, tangentMode=RCTM_User, time=0, value=47.1681}, <at 0.2 0.2422>]}
       YCurve {keys [{arriveTangent=-82.3078, interpMode=RCIM_Cubic, leaveTangent=-82.3078, tangentMode=RCTM_User, time=0, value=8.304}, <at 0.2 0.3234>]}
       ZCurve {keys [{arriveTangent=-402.2961, interpMode=RCIM_Cubic, leaveTangent=-402.2961, tangentMode=RCTM_User, time=0, value=50}, <at 0.2 0.5>]}
      }})
    ScaleA=true
    ScaleRGB=true
   )
   &SkeletalMeshLocation(
    Bone / Socket Index=$Particles.SkeletonSamplingIndex
    MaskSpawnGroup=false
    Mesh Sampling Type=::NewEnumerator5
    Orientation Sampling=::NewEnumerator1
    Sampled Position Offset=&SelectVectorFromArray(
     Direct Array Index=$Particles.SkeletonSamplingIndex
     Vector Selection Array=DataInterface() {propertyValues {FloatData [<7 -1 5>, <7 -1 -5>]}}
    )
    Skeletal Mesh=$System.Skeletal Mesh
    Socket Sampling Mode=::NewEnumerator5
   )
   &SubUVAnimation(
    End Frame Range Override=61
    Output Alternate Renderer Binding=true
    "Sprite Renderer "=DataInterface() {propertyValues {SpriteRenderer=&self:SmokeHand_0.NiagaraSpriteRendererProperties_0}}
    Start Frame Range Override=17
    UseEndFrame=false
    UseStartFrame=false
   )
   &DynamicMaterialParameters(
    Index 0 Param 4=0.5
    Module.Param0WriteEnabled=false
    Module.Param1WriteEnabled=false
    Module.Param2WriteEnabled=true
    Module.Param3WriteEnabled=true
   ) {module=DynamicMaterialParameters001}
   &Drag(Drag=1.5, Use Linear Drag=true, Use Rotational Drag=true) {enabled=false}
   &AccelerationForce(Acceleration=<0 0 -25>) {enabled=false}
   &SolveForcesAndVelocity(Clamp Velocity=false, Limit Acceleration=false)
  ]
  rendererClasses [&NiagaraSpriteRendererProperties]
  renderers [{rendererClass=&NiagaraSpriteRendererProperties}]
 }
 Sparks {
  emitterUpdateScript [
   &EmitterState(Inactive Response=::NewEnumerator1, Life Cycle Mode=::NewEnumerator1, Loop Duration=0.5, UseLoopDelay=true)
   &SpawnRate(SpawnRate=8, Use Spawn Probability=false)
  ]
  particleSpawnScript [
   /SetVariables(Particles.SkeletonSamplingIndex:NiagaraInt32=&SequentialInt(Use Array Data:NiagaraBool=false, First Number:NiagaraInt32=0, Sequence:NiagaraDataInterfaceArrayInt32=?, Last Number:NiagaraInt32=3)) {isSetParameters=true}
   &InitializeParticle(
    Lifetime=&RandomRangeFloat(Minimum=0.2, Maximum=0.75, Evaluation Type=::NewEnumerator1)
    Sprite Size=&RandomRangeVector2D(Minimum=<1.8 2>, Maximum=<2.5 3.5>, Evaluation Type=::NewEnumerator1)
    Sprite Size Mode=::NewEnumerator4
    UsePositionOffset=false
    Write Material Random=true
   )
   &ShapeLocation(MaskSpawnGroup=false, Override Seed=false, Sphere Radius=5, UseOwnerScale=true)
   &FixShapeVector()
   &AddVelocity(Velocity=&Multiply_VectorByFloat(
     Vector=$Output.FixShapeVector.ShapeVector
     Float=&RandomRangeFloat(Minimum=5, Maximum=10, Evaluation Type=::NewEnumerator1)
    ))
   &Color(Color=&MakeLinearColorFromVectorAndFloat("Vector (RGB)"=<31.8287 3.4828 40>, "Float (Alpha)"=1))
   &InheritVelocity(Inherited Velocity Amount Scale=<-0.06 -0.05 -0.15>, Inherited Velocity Speed Limit=0, Speed Limit=false, UseSpeedThreshold=false)
   &SkeletalMeshLocation(
    Bone / Socket Index=$Particles.SkeletonSamplingIndex
    Bone Sampling Mode=::NewEnumerator4
    MaskSpawnGroup=false
    Orientation Sampling=::NewEnumerator1
    Sampled Position Offset=&SelectVectorFromArray(
     Direct Array Index=$Particles.SkeletonSamplingIndex
     Vector Selection Array=DataInterface() {propertyValues {FloatData [<-10 10 1>, <10 -10 1>, <0 0 0>, <0 0 0>]}}
    )
    Skeletal Mesh=DataInterface() {propertyValues {
      DefaultMesh=null
      ExcludeBoneName=root
      OptionalComponentClass=null
      PreviewMesh=null
      SkinningMode=SkinOnTheFly
      SoftSourceActor=null
      SourceMode=Default
      UvSetIndex=0
      WholeMeshLOD=-1
      bExcludeBone=false
      bReadDeformedGeometry=true
      bRequireCurrentFrameData=true
      ComponentTags []
      FilteredBones [hand_r,hand_l,ball_r,ball_l]
      FilteredSockets []
      MeshUserParameter {parameter {name=None, typeDefHandle {registeredTypeIndex=1}, varData []}}
      SamplingRegions []
     }}
   )
  ]
  particleUpdateScript [
   &ParticleState()
   &ScaleSpriteSize(
    Scale Factor=&Vector2DFromCurve(
     Vector2Curve=DataInterface() {propertyValues {
       XCurve {keys [
         {interpMode=RCIM_Cubic, tangentMode=RCTM_User, time=0, value=1.5}
         {arriveTangent=-0.5, interpMode=RCIM_Cubic, leaveTangent=-0.5, tangentMode=RCTM_User, time=0.3019, value=1.2624}
         <at 1 1 RCIM_Cubic>
        ]}
       YCurve {keys [
         {arriveTangent=1.7792, interpMode=RCIM_Cubic, leaveTangent=1.7792, tangentMode=RCTM_User, time=0, value=1}
         {arriveTangent=-0.4733, interpMode=RCIM_Cubic, leaveTangent=-0.4733, tangentMode=RCTM_User, time=0.3019, value=1.1}
         <at 1 0 RCIM_Cubic>
        ]}
      }}
     ScaleCurve=<1 1>
    )
    Scale Sprite Size Mode=::NewEnumerator2
   )
   &SolveForcesAndVelocity(Clamp Velocity=false, Limit Acceleration=false)
   &ScaleColor(
    Scale RGB=&VectorFromCurve(VectorCurve=DataInterface() {propertyValues {
       XCurve {keys [<at 0 1 RCIM_Cubic>,<at 1 0.3333>]}
       YCurve {keys [<at 0 1 RCIM_Cubic>,<at 1 0.026>]}
       ZCurve {keys [<at 0 1 RCIM_Cubic>,<at 1 0.1205>]}
      }})
    ScaleA=true
    ScaleRGB=true
   )
   /Niagara/Modules/CascadeConversion/CascadeConversion_Orbit(
    Offset:Vector3f=&RandomRangeVector(Minimum=<-10 -10 0>, Maximum=<10 10 0>)
    Rotation:Vector3f=&RandomRangeVector(Minimum=<0 0 0>, Maximum=<1 1 1>)
    Rotation Rate:Vector3f=&RandomRangeVector(Minimum=<0 0 0>, Maximum=<1 1 1>)
   )
   /Niagara/Modules/CascadeConversion/CascadeConversion_SolveOrbit()
  ]
  rendererClasses [&NiagaraSpriteRendererProperties]
  renderers [{rendererClass=&NiagaraSpriteRendererProperties}]
 }
 SquaresRefract {
  emitterUpdateScript [
   &EmitterState(Inactive Response=::NewEnumerator1, Life Cycle Mode=::NewEnumerator1, Loop Duration=0.2, UseLoopDelay=true)
   &SpawnRate(SpawnRate=20, Use Spawn Probability=false)
  ]
  particleSpawnScript [
   &InitializeParticle(
    Lifetime=&RandomRangeFloat(Minimum=0.2, Maximum=0.45, Evaluation Type=::NewEnumerator1)
    Sprite Size=&Vector2DFromFloat(Value=&RandomRangeFloat(Minimum=2, Maximum=5, Evaluation Type=::NewEnumerator1))
    Sprite Size Mode=::NewEnumerator4
    UsePositionOffset=false
    Write Material Random=true
   )
   &Color(Color=&MakeLinearColorFromVectorAndFloat(
     "Vector (RGB)"=<0.6671 0.099 1>
     "Float (Alpha)"=&RandomRangeFloat(Minimum=0.5, Maximum=1, Evaluation Type=::NewEnumerator1)
    ))
   &SkeletalMeshLocation(
    Bone / Socket Index=&SequentialInt(Use Array Data:NiagaraBool=false, First Number:NiagaraInt32=0, Sequence:NiagaraDataInterfaceArrayInt32=?, Last Number:NiagaraInt32=3)
    Bone Sampling Mode=::NewEnumerator4
    MaskSpawnGroup=false
    Orientation Sampling=::NewEnumerator1
    Skeletal Mesh=DataInterface() {propertyValues {
      DefaultMesh=null
      ExcludeBoneName=root
      OptionalComponentClass=null
      PreviewMesh=null
      SkinningMode=SkinOnTheFly
      SoftSourceActor=null
      SourceMode=Default
      UvSetIndex=0
      WholeMeshLOD=-1
      bExcludeBone=false
      bReadDeformedGeometry=true
      bRequireCurrentFrameData=true
      ComponentTags []
      FilteredBones [hand_r,hand_l,ball_r,ball_l]
      FilteredSockets []
      MeshUserParameter {parameter {name=None, typeDefHandle {registeredTypeIndex=1}, varData []}}
      SamplingRegions []
     }}
   )
   &AddVectorToPosition(Vector:Vector3f=&TransformVector(Vector=<4 0 0>, Destination Space=::Simulation))
   &InheritVelocity(Inherited Velocity Amount Scale=<-0.05 -0.05 -0.2>, Inherited Velocity Speed Limit=0, Speed Limit=false, UseSpeedThreshold=false)
   &ShapeLocation(MaskSpawnGroup=false, Override Seed=false, Sphere Radius=10, Sphere Surface Distribution=1, UseOwnerScale=true)
   &FixShapeVector()
   &AddVelocity(Velocity=&Multiply_VectorByFloat(Vector=$Output.FixShapeVector.ShapeVector, Float=5))
  ]
  particleUpdateScript [
   &ParticleState()
   &ScaleColor(
    Scale Alpha=&FloatFromCurve(FloatCurve=DataInterface() {propertyValues {Curve {keys [{arriveTangent=-2.2945, interpMode=RCIM_Cubic, leaveTangent=-2.2945, tangentMode=RCTM_User, time=0, value=1}, <at 1 0>]}}})
    Scale RGB=&VectorFromFloat(Value=&FloatFromCurve(FloatCurve=DataInterface() {propertyValues {Curve {keys [{arriveTangent=-3.7201, interpMode=RCIM_Cubic, leaveTangent=-3.7201, tangentMode=RCTM_User, time=0, value=2}, <at 0.7 0>]}}}))
    ScaleA=true
    ScaleRGB=true
   )
   &Drag(
    Drag=&RandomRangeFloat(Minimum=6, Maximum=10, Evaluation Type=::NewEnumerator1)
    Use Linear Drag=true
    Use Rotational Drag=true
   )
   &SolveForcesAndVelocity(Clamp Velocity=false, Limit Acceleration=false)
   &CameraOffset(Camera Offset Amount=5)
  ]
  rendererClasses [&NiagaraSpriteRendererProperties]
  renderers [{rendererClass=&NiagaraSpriteRendererProperties}]
 }
}
summary {
 systemName=NS_Contrail_Cube_Converted
 emitters [
  {emitterName=SmokeHand, rendererClasses [&NiagaraSpriteRendererProperties]}
  {emitterName=Hand_R, rendererClasses [&NiagaraRibbonRendererProperties]}
  {emitterName=RunesHand, rendererClasses [&NiagaraSpriteRendererProperties]}
  {emitterName=RunesTrail, rendererClasses [&NiagaraSpriteRendererProperties]}
  {emitterName=SquaresRefract, rendererClasses [&NiagaraSpriteRendererProperties]}
  {emitterName=BaseLightning, rendererClasses [&NiagaraMeshRendererProperties]}
  {emitterName=Lightning, rendererClasses [&NiagaraMeshRendererProperties]}
  {emitterName=HandGlow, rendererClasses [&NiagaraSpriteRendererProperties, &NiagaraLightRendererProperties]}
  {emitterName=Sparks, rendererClasses [&NiagaraSpriteRendererProperties]}
 ]
 userVariables [
  User.bIsFrontEnd {type=NiagaraBool}
  User.bIsFrontEndPreview {type=NiagaraBool}
  User.ParaGlideLeanAlpha {type=NiagaraFloat, defaultValue=0, description=""}
 ]
}
