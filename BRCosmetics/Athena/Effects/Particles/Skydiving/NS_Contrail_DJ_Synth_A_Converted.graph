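# Cascade-converted Niagara skydiving contrail for the DJ Synth cosmetic, attaching equalizer-style effects to four skeletal points, apparently hands and feet.
# HandEQ spawns fading meshes stretched by per-point curves, while HandFlatEQ adds flattened meshes driven by material parameters.
# Rotating ray sprites cycle through rainbow colors, and Ribbon_R_Front bursts two persistent ribbon meshes on the first two points.
# Every emitter picks points via a sequential index, re-sampling skeletal location each update with per-point offsets.
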NiagaraSystem /BRCosmetics/Athena/Effects/Particles/Skydiving/NS_Contrail_DJ_Synth_A_Converted
dependencies {
 usedDataInterfaces [
  &NiagaraDataInterfaceVectorCurve
  &NiagaraDataInterfaceCurve
  &NiagaraDataInterfaceArrayFloat3
  &NiagaraDataInterfaceSkeletalMesh
 ]
 usedDynamicInputs []
 usedModules [
  &SystemState
  &self:SystemSpawnScript.SystemScriptSource.SystemScriptGraph.NiagaraNodeAssignment_0.SetVariables_#1
 ]
}
emitters {
 HandEQ {
  emitterUpdateScript [
   &EmitterState(Inactive Response=::NewEnumerator1, Life Cycle Mode=::NewEnumerator1, Loop Duration=1, UseLoopDelay=true)
   &SpawnRate(SpawnRate=20, Use Spawn Probability=false)
  ]
  particleSpawnScript [
   /SetVariables(Particles.SkeletonSamplingIndex:NiagaraInt32=/CascadeToNiagaraConverter/NiagaraScripts/DynamicInputScripts/SequentialInt(Use Array Data:NiagaraBool=false, First Number:NiagaraInt32=0, Sequence:NiagaraDataInterfaceArrayInt32=?, Last Number:NiagaraInt32=3)) {isSetParameters=true}
   &InitializeParticle(
    Lifetime=0.5
    Mesh Scale=&MergeHelper_VectorSwitch(
     0=/Niagara/DynamicInputs/CascadeConversion/SplitVector(
      Double Component Vector Channels:ECascadeNiagaraTwoVectorChannels=::NewEnumerator0
      Double Component:NiagaraFloat=&RandomRangeFloat(Minimum=0.035, Maximum=0.05, Evaluation Type=::NewEnumerator1)
      Single Component:NiagaraFloat=&RandomRangeFloat(Minimum=0.2, Maximum=0.2, Evaluation Type=::NewEnumerator1)
     )
     1=&SplitVector(
      Double Component Vector Channels:ECascadeNiagaraTwoVectorChannels=::NewEnumerator0
      Double Component:NiagaraFloat=&RandomRangeFloat(Minimum=0.035, Maximum=0.05, Evaluation Type=::NewEnumerator1)
      Single Component:NiagaraFloat=&RandomRangeFloat(Minimum=0.2, Maximum=0.2, Evaluation Type=::NewEnumerator1)
     )
     2=&SplitVector(
      Double Component Vector Channels:ECascadeNiagaraTwoVectorChannels=::NewEnumerator0
      Double Component:NiagaraFloat=&RandomRangeFloat(Minimum=0.03, Maximum=0.04, Evaluation Type=::NewEnumerator1)
      Single Component:NiagaraFloat=&RandomRangeFloat(Minimum=0.085, Maximum=0.225, Evaluation Type=::NewEnumerator1)
     )
     3=&SplitVector(
      Double Component Vector Channels:ECascadeNiagaraTwoVectorChannels=::NewEnumerator0
      Double Component:NiagaraFloat=&RandomRangeFloat(Minimum=0.03, Maximum=0.04, Evaluation Type=::NewEnumerator1)
      Single Component:NiagaraFloat=&RandomRangeFloat(Minimum=0.085, Maximum=0.225, Evaluation Type=::NewEnumerator1)
     )
     Index=$Particles.SkeletonSamplingIndex
     Num=::NewEnumerator2
    )
    Mesh Scale Mode=::NewEnumerator4
    UsePositionOffset=false
    Write Material Random=true
   )
  ]
  particleUpdateScript [
   &ParticleState()
   &ScaleMeshSize(Scale Factor=&MergeHelper_VectorSwitch(
     0=&VectorFromCurve(VectorCurve=DataInterface() {propertyValues {XCurve {keys [<at 0 1.5>,<at 1 1>]}, YCurve {keys [<at 0 1.5>,<at 1 1>]}, ZCurve {keys [<at 0 0.5>,<at 1 2.5>]}}})
     1=&VectorFromCurve(VectorCurve=DataInterface() {propertyValues {XCurve {keys [<at 0 1.5>,<at 1 1>]}, YCurve {keys [<at 0 1.5>,<at 1 1>]}, ZCurve {keys [<at 0 0.5>,<at 1 2.5>]}}})
     2=&VectorFromCurve(VectorCurve=DataInterface() {propertyValues {
        XCurve {keys [<at 0 1.4>,<at 0 1.4>,<at 1 0.8>]}
        YCurve {keys [<at 0 1.4>,<at 0 1.4>,<at 1 0.8>]}
        ZCurve {keys [<at 0 0.25>,<at 0 0.6>,<at 1 1.5>]}
       }})
     3=&VectorFromCurve(VectorCurve=DataInterface() {propertyValues {
        XCurve {keys [<at 0 1.4>,<at 0 1.4>,<at 1 0.8>]}
        YCurve {keys [<at 0 1.4>,<at 0 1.4>,<at 1 0.8>]}
        ZCurve {keys [<at 0 0.25>,<at 0 0.6>,<at 1 1.5>]}
       }})
     Index=$Particles.SkeletonSamplingIndex
     Num=::NewEnumerator2
    ))
   &Color(Color=&MakeLinearColorFromVectorAndFloat(
     "Vector (RGB)"=&VectorFromCurve(VectorCurve=DataInterface() {propertyValues {XCurve {keys [<at 0 1>,<at 1 1>]}, YCurve {keys [<at 0 1>,<at 1 1>]}, ZCurve {keys [<at 0 1>,<at 1 1>]}}})
     "Float (Alpha)"=&ClampFloat(Float=&FloatFromCurve(FloatCurve=DataInterface() {propertyValues {Curve {keys [{arriveTangent=-1, time=0, value=1}, <at 1 0>]}}}))
    ))
   &DynamicMaterialParameters(Write Parameter Index 0=::NewEnumerator0)
   &SkeletalMeshLocation(
    Bone / Socket Index=$Particles.SkeletonSamplingIndex
    Bone Sampling Mode=::NewEnumerator4
    MaskSpawnGroup=false
    Orientation Sampling=::NewEnumerator1
    Sampled Position Offset=&SelectVectorFromArray(
     Direct Array Index=$Particles.SkeletonSamplingIndex
     Vector Selection Array=DataInterface() {propertyValues {FloatData [<13 0 0>, <-13 0 0>, <10 -5 0>, <-10 5 0>]}}
    )
    Skeletal Mesh=$System.Skeletal Mesh Reader
   )
  ]
  rendererClasses [&NiagaraMeshRendererProperties]
  renderers [{rendererClass=&NiagaraMeshRendererProperties, rendererIndex=0}]
 }
 HandFlatEQ {
  emitterUpdateScript [
   &EmitterState(
    Inactive Response=::NewEnumerator1
    Life Cycle Mode=::NewEnumerator1
    Loop Behavior=::NewEnumerator1
    Loop Duration Mode=::NewEnumerator1
    UseLoopDelay=true
   )
   &SpawnRate(SpawnRate=10, Use Spawn Probability=false)
  ]
  particleSpawnScript [
   /SetVariables(Particles.SkeletonSamplingIndex:NiagaraInt32=&SequentialInt(Use Array Data:NiagaraBool=false, First Number:NiagaraInt32=0, Sequence:NiagaraDataInterfaceArrayInt32=?, Last Number:NiagaraInt32=3)) {isSetParameters=true}
   &InitializeParticle(
    Mesh Scale=&SelectVectorFromArray(
     Direct Array Index=$Particles.SkeletonSamplingIndex
     Vector Selection Array=DataInterface() {propertyValues {FloatData [<0.06 0.06 0.0001>, <0.06 0.06 0.0001>, <0.05 0.05 0.0001>, <0.05 0.05 0.0001>]}}
    )
    Mesh Scale Mode=::NewEnumerator4
    UsePositionOffset=false
    Write Material Random=true
   )
  ]
  particleUpdateScript [
   &ParticleState(Kill Particles When Lifetime Has Elapsed=false, Let Infinitely Lived Particles Die When Emitter Deactivates=true)
   &Color(Color=&MakeLinearColorFromVectorAndFloat(
     "Vector (RGB)"=&VectorFromCurve(VectorCurve=DataInterface() {propertyValues {XCurve {keys [<at 0 1>,<at 1 1>]}, YCurve {keys [<at 0 1>,<at 1 1>]}, ZCurve {keys [<at 0 1>,<at 1 1>]}}})
     "Float (Alpha)"=&ClampFloat(Float=&FloatFromCurve(FloatCurve=DataInterface() {propertyValues {Curve {keys [{arriveTangent=-1, time=0, value=1}, <at 1 0>]}}}))
    ))
   &DynamicMaterialParameters(
    Index 0 Param 1=&FloatFromCurve(FloatCurve=DataInterface() {propertyValues {Curve {keys [<at 0 2>,<at 0.3 0.8>,<at 1.5 1>]}}})
    Index 0 Param 2=1
    Index 0 Param 3=1
    Index 0 Param 4=1
    Module.Index0Float0Write=true
    Module.Index0Float1Write=true
    Module.Index0Float2Write=true
    Module.Index0Float3Write=true
   )
   &SkeletalMeshLocation(
    Bone / Socket Index=$Particles.SkeletonSamplingIndex
    Bone Sampling Mode=::NewEnumerator4
    MaskSpawnGroup=false
    Orientation Sampling=::NewEnumerator1
    Sampled Position Offset=&SelectVectorFromArray(
     Direct Array Index=$Particles.SkeletonSamplingIndex
     Vector Selection Array=DataInterface() {propertyValues {FloatData [<13 0 0>, <-13 0 0>, <10 -5 0>, <-10 5 0>]}}
    )
    Skeletal Mesh=$System.Skeletal Mesh Reader
   )
  ]
  rendererClasses [&NiagaraMeshRendererProperties]
  renderers [{rendererClass=&NiagaraMeshRendererProperties, rendererIndex=0}]
 }
 RaysConstant_HandsAndFeet {
  emitterUpdateScript [
   &EmitterState(Delay First Loop Only=true, Inactive Response=::NewEnumerator1, Life Cycle Mode=::NewEnumerator1, Loop Duration=1, UseLoopDelay=true)
   &SpawnBurst_Instantaneous(Spawn Count=0, Use Spawn Probability=false, UseLoopCountLimit=false)
   &SpawnRate(SpawnRate=5, Use Spawn Probability=false)
  ]
  particleSpawnScript [
   /SetVariables(Particles.SpriteRotationRate=&RandomRangeFloat(Minimum=-0.2, Maximum=0.2, Evaluation Type=::NewEnumerator1)) {isSetParameters=true}
   /SetVariables(Particles.SkeletonSamplingIndex:NiagaraInt32=&SequentialInt(Use Array Data:NiagaraBool=false, First Number:NiagaraInt32=0, Sequence:NiagaraDataInterfaceArrayInt32=?, Last Number:NiagaraInt32=3)) {isSetParameters=true}
   &InitializeParticle(
    Sprite Rotation Angle=&RandomRangeFloat(Minimum=0, Maximum=1, Evaluation Type=::NewEnumerator1)
    Sprite Rotation Mode=::NewEnumerator1
    Sprite Size=&Vector2DFromFloat(Value=&RandomRangeFloat(Minimum=100, Maximum=60, Evaluation Type=::NewEnumerator1))
    Sprite Size Mode=::NewEnumerator4
    UsePositionOffset=false
    Write Material Random=true
   )
   /CascadeToNiagaraConverter/NiagaraScripts/ModuleScripts/AddVectorToPosition(Vector:Vector3f=&TransformVector(Vector=<0 35 70>, Destination Space=::Simulation))
   &CameraOffset(Camera Offset Amount=&RandomRangeFloat(Minimum=65, Maximum=75, Evaluation Type=::NewEnumerator1))
  ]
  particleUpdateScript [
   &ParticleState()
   &SpriteRotationRate(Rotation Rate=&AngleConversion(
     Angle:NiagaraFloat=$Particles.SpriteRotationRate
     Angle Input:ENiagara_AngleInput=::NewEnumerator1
     Angle Output:ENiagara_AngleInput=::NewEnumerator0
    ))
   &ScaleSpriteSize(
    Scale Sprite Size Mode=::NewEnumerator0
    Uniform Scale Factor=&FloatFromCurve(FloatCurve=DataInterface() {propertyValues {Curve {keys [<at 0 1>,<at 1 1.3>]}}})
   )
   &Color(Color=&MakeLinearColorFromVectorAndFloat(
     "Vector (RGB)"=&VectorFromCurve(VectorCurve=DataInterface() {propertyValues {
        XCurve {keys [
          <at 0 1 RCIM_Cubic>
          {arriveTangent=-3.5, interpMode=RCIM_Cubic, leaveTangent=-3.5, time=0.1429, value=1}
          {arriveTangent=-3.5, interpMode=RCIM_Cubic, leaveTangent=-3.5, time=0.2857, value=0}
          <at 0.4286 0 RCIM_Cubic>
          {arriveTangent=3.5, interpMode=RCIM_Cubic, leaveTangent=3.5, time=0.5714, value=0}
          {arriveTangent=3.5, interpMode=RCIM_Cubic, leaveTangent=3.5, time=0.7143, value=1}
          <at 0.8571 1 RCIM_Cubic>
          <at 1 1 RCIM_Cubic>
         ]}
        YCurve {keys [
          <at 0 0 RCIM_Cubic>
          {arriveTangent=3.5, interpMode=RCIM_Cubic, leaveTangent=3.5, time=0.1429, value=1}
          <at 0.2857 1 RCIM_Cubic>
          {arriveTangent=-3.5, interpMode=RCIM_Cubic, leaveTangent=-3.5, time=0.4286, value=1}
          {arriveTangent=-3.5, interpMode=RCIM_Cubic, leaveTangent=-3.5, time=0.5714, value=0}
          <at 0.7143 0 RCIM_Cubic>
          {arriveTangent=3.5, interpMode=RCIM_Cubic, leaveTangent=3.5, time=0.8571, value=0}
          <at 1 1 RCIM_Cubic>
         ]}
        ZCurve {keys [
          <at 0 0 RCIM_Cubic>
          <at 0.1429 0 RCIM_Cubic>
          {arriveTangent=3.5, interpMode=RCIM_Cubic, leaveTangent=3.5, time=0.2857, value=0}
          {arriveTangent=3.5, interpMode=RCIM_Cubic, leaveTangent=3.5, time=0.4286, value=1}
          <at 0.5714 1 RCIM_Cubic>
          {arriveTangent=-3.5, interpMode=RCIM_Cubic, leaveTangent=-3.5, time=0.7143, value=1}
          {arriveTangent=-3.5, interpMode=RCIM_Cubic, leaveTangent=-3.5, time=0.8571, value=0}
          <at 1 0 RCIM_Cubic>
         ]}
       }})
     "Float (Alpha)"=&ClampFloat(Float=&FloatFromCurve(FloatCurve=DataInterface() {propertyValues {Curve {keys [{arriveTangent=0.1, time=0, value=0}, <at 0.1429 0.1>, {arriveTangent=-0.1, leaveTangent=-0.1, time=0.8571, value=0.1}, <at 1 0>]}}}))
    ))
   &ScaleColor(
    Color Value To Scale=$Particles.Color
    Scale Alpha=5
    Scale RGB=&VectorFromCurve(VectorCurve=DataInterface() {propertyValues {XCurve {keys [<at 0 1>,<at 1 1>]}, YCurve {keys [<at 0 1>,<at 1 1>]}, ZCurve {keys [<at 0 1>,<at 1 1>]}}})
    ScaleA=true
    ScaleRGB=true
   )
   &SkeletalMeshLocation(
    Bone / Socket Index=$Particles.SkeletonSamplingIndex
    Bone Sampling Mode=::NewEnumerator4
    MaskSpawnGroup=false
    Orientation Sampling=::NewEnumerator1
    Sampled Position Offset=&SelectVectorFromArray(
     Direct Array Index=$Particles.SkeletonSamplingIndex
     Vector Selection Array=DataInterface() {propertyValues {FloatData [<12 0 0>, <-12 0 0>, <10 -7 0>, <-10 7 0>]}}
    )
    Skeletal Mesh=$System.Skeletal Mesh Reader
   )
  ]
  rendererClasses [&NiagaraSpriteRendererProperties]
  renderers [{rendererClass=&NiagaraSpriteRendererProperties}]
 }
 Ribbon_R_Front {
  emitterUpdateScript [
   &EmitterState(
    Inactive Response=::NewEnumerator1
    Life Cycle Mode=::NewEnumerator1
    Loop Behavior=::NewEnumerator1
    Loop Duration Mode=::NewEnumerator1
    UseLoopDelay=true
   )
   &SpawnBurst_Instantaneous(Spawn Count=2, Use Spawn Probability=false, UseLoopCountLimit=false)
  ]
  particleSpawnScript [
   /SetVariables(Particles.SkeletonSamplingIndex:NiagaraInt32=&SequentialInt(Use Array Data:NiagaraBool=false, First Number:NiagaraInt32=0, Sequence:NiagaraDataInterfaceArrayInt32=?, Last Number:NiagaraInt32=1)) {isSetParameters=true}
   &InitializeParticle(Mesh Scale=<0.525 10 10>, Mesh Scale Mode=::NewEnumerator4, UsePositionOffset=false, Write Material Random=true)
   &Color(Color=&MakeLinearColorFromVectorAndFloat("Vector (RGB)"=<1 1 1>, "Float (Alpha)"=1))
   &DynamicMaterialParameters(
    Index 0 Param 1=&RandomRangeFloat(Minimum=0, Maximum=1, Evaluation Type=::NewEnumerator1)
    Module.Param0WriteEnabled=true
    Module.Param1WriteEnabled=false
    Module.Param2WriteEnabled=false
    Module.Param3WriteEnabled=false
   )
   &InitialMeshOrientation(Mesh Orientation Mode=::NewEnumerator2, Orientation Vector=<0 0 1>, Use Rotation Vector=false)
  ]
  particleUpdateScript [
   &ParticleState(Kill Particles When Lifetime Has Elapsed=false, Let Infinitely Lived Particles Die When Emitter Deactivates=true)
   &SkeletalMeshLocation(
    Bone / Socket Index=$Particles.SkeletonSamplingIndex
    Bone Sampling Mode=::NewEnumerator4
    MaskSpawnGroup=false
    Orientation Sampling=::NewEnumerator1
    Sampled Position Offset=&SelectVectorFromArray(
     Direct Array Index=$Particles.SkeletonSamplingIndex
     Vector Selection Array=DataInterface() {propertyValues {FloatData [<10 0 0>, <-10 0 0>]}}
    )
    Skeletal Mesh=$System.Skeletal Mesh Reader
   )
   &DynamicMaterialParameters(
    Index 0 Param 2=0.98
    Index 0 Param 3=1
    Index 0 Param 4=1
    Module.Param0WriteEnabled=false
    Module.Param1WriteEnabled=true
    Module.Param2WriteEnabled=true
    Module.Param3WriteEnabled=true
   ) {module=DynamicMaterialParameters001}
  ]
  rendererClasses [&NiagaraMeshRendererProperties]
  renderers [{rendererClass=&NiagaraMeshRendererProperties, rendererIndex=0}]
 }
}
summary {
 systemName=NS_Contrail_DJ_Synth_A_Converted
 emitters [
  {emitterName=HandEQ, rendererClasses [&NiagaraMeshRendererProperties]}
  {emitterName=HandFlatEQ, rendererClasses [&NiagaraMeshRendererProperties]}
  {emitterName=RaysConstant_HandsAndFeet, rendererClasses [&NiagaraSpriteRendererProperties]}
  {emitterName=Ribbon_R_Front, rendererClasses [&NiagaraMeshRendererProperties]}
 ]
 userVariables []
}
