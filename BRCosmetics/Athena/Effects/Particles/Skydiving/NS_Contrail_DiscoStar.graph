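# Niagara cosmetic skydiving contrail (DiscoStar) that trails ribbons, sparkles and spinning star sprites from the character's skeletal mesh.
# All three emitters cycle through bone or socket indices on a shared mesh reader, with ribbon IDs derived from spawn groups.
# Ribbons widen over life, and User.ParaGlideLeanAlpha drives their alpha scale in gameplay.
# Front-end modifiers add velocity to simulate motion in menus and fix the ribbon alpha scale.
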NiagaraSystem /BRCosmetics/Athena/Effects/Particles/Skydiving/NS_Contrail_DiscoStar
dependencies {
 usedDataInterfaces [&NiagaraDataInterfaceCurve, &NiagaraDataInterfaceVectorCurve, &NiagaraDataInterfaceSkeletalMesh]
 usedDynamicInputs []
 usedModules [
  &self:SystemSpawnScript.SystemScriptSource.SystemScriptGraph.NiagaraNodeAssignment_0.SetVariables_#1
  &SystemState
 ]
}
emitters {
 Ribbon_Left_Foot {
  emitterUpdateScript [
   &EmitterState(Inactive Response=::NewEnumerator1, Life Cycle Mode=::NewEnumerator1, Loop Duration=1, UseLoopDelay=false)
   &SpawnRate(SpawnRate=160, Use Spawn Probability=false)
  ]
  particleSpawnScript [
   &FN_SetSpawnGroup()
   /SetVariables(Particles.RibbonID=/Niagara/DynamicInputs/Meshes/MakeRibbonIDFromSpawnGroup()) {isSetParameters=true}
   /SetVariables(Particles.Acceleration:Vector3f=&RandomRangeVector(Minimum=<-3 -3 0>, Maximum=<3 3 0>, Evaluation Type=::NewEnumerator1)) {isSetParameters=true}
   &InitializeParticle(
    Color=<rgba 0.797 0 2 1>
    Lifetime=0.3
    Ribbon Width=&Multiply_Float(A=3, B=2)
    Ribbon Width Mode=::NewEnumerator1
    UsePositionOffset=false
    Write Material Random=true
   )
   &SkeletalMeshLocation(
    Bone / Socket Index=/CascadeToNiagaraConverter/NiagaraScripts/DynamicInputScripts/SequentialInt(Use Array Data:NiagaraBool=false, First Number:NiagaraInt32=0, Sequence:NiagaraDataInterfaceArrayInt32=?, Last Number:NiagaraInt32=3)
    Bone Sampling Mode=::NewEnumerator4
    MaskSpawnGroup=false
    Orientation Sampling=::NewEnumerator1
    Sampled Position Offset=&MergeHelper_VectorSwitch(0=<0 0 0>, 1=<0 0 0>, 2=<0 0 0>, 3=<0 0 0>, Index=&SpawnGroup(), Num=::NewEnumerator2)
    Skeletal Mesh=$System.Skeletal Mesh Reader
   )
   /CascadeToNiagaraConverter/NiagaraScripts/ModuleScripts/AddVectorToPosition(Vector:Vector3f=&TransformVector(Vector=<0 0 0>, Destination Space=::Simulation))
   &AddVelocity(Velocity=&FN_FrontEndVector3Modifier(Gameplay=<0 0 0>, Front End=<0 -200 800>))
  ]
  particleUpdateScript [
   /SetVariables(Particles.ScaledRibbonWidth:NiagaraFloat=&Multiply_Float(
     A=$Particles.RibbonWidth
     B=&FloatFromCurve(FloatCurve=DataInterface() {propertyValues {Curve {keys [
          {arriveTangent=37.7088, interpMode=RCIM_Cubic, leaveTangent=37.7088, tangentMode=RCTM_User, time=-0.0035, value=-0.0003}
          {interpMode=RCIM_Cubic, tangentMode=RCTM_User, time=0.0921, value=2.3001}
          {arriveTangent=0.4859, interpMode=RCIM_Cubic, leaveTangent=0.4859, tangentMode=RCTM_User, time=1, value=3.9862}
         ]}}})
    )) {isSetParameters=true}
   &ParticleState()
   &ScaleColor(
    Scale Alpha=&FloatFromCurve(
     FloatCurve=DataInterface() {propertyValues {Curve {keys [<at 0.1 0 RCIM_Cubic>,<at 0.3 1 RCIM_Cubic>,<at 1 0>]}}}
     Scale Curve=&FN_FrontEndFloatModifier(Gameplay=&Lerp_Float(A=0.15, B=0.7, Alpha=$User.ParaGlideLeanAlpha), Front End=0.15)
    )
    ScaleA=true
    ScaleRGB=true
   )
   &AccelerationForce(Acceleration=$Particles.Acceleration)
   &SolveForcesAndVelocity(Clamp Velocity=false, Limit Acceleration=false)
   &DynamicMaterialParameters(
    Index 0 Param 2=&RandomRangeFloat(Minimum=1, Maximum=1)
    Index 0 Param 4=1
    Module.Index0Float0Write=true
    Module.Index0Float1Write=true
    Module.Index0Float2Write=true
    Module.Index0Float3Write=true
   )
  ]
  rendererClasses [&NiagaraRibbonRendererProperties]
  renderers [{rendererClass=&NiagaraRibbonRendererProperties}]
 }
 SparklesHandL {
  emitterUpdateScript [
   &EmitterState(Inactive Response=::NewEnumerator1, Life Cycle Mode=::NewEnumerator1, Loop Duration=1, UseLoopDelay=true)
   &SpawnRate(SpawnRate=100, Use Spawn Probability=false)
  ]
  particleSpawnScript [
   /SetVariables(Particles.SpriteRotationRate=&RandomRangeFloat(Minimum=0, Maximum=1, Evaluation Type=::NewEnumerator1)) {isSetParameters=true}
   &InitializeParticle(
    Lifetime=&RandomRangeFloat(Minimum=0.5, Maximum=1, Evaluation Type=::NewEnumerator1)
    Sprite Rotation Angle=&RandomRangeFloat(Minimum=0, Maximum=1, Evaluation Type=::NewEnumerator1)
    Sprite Rotation Mode=::NewEnumerator1
    Sprite Size=&Vector2DFromFloat(Value=&RandomRangeFloat(Minimum=2, Maximum=8, Evaluation Type=::NewEnumerator1))
    Sprite Size Mode=::NewEnumerator4
    UsePositionOffset=false
    Write Material Random=true
   )
   &Color(Color=&MakeLinearColorFromVectorAndFloat("Vector (RGB)"=<0 37.745 45>, "Float (Alpha)"=1))
   &SkeletalMeshLocation(
    Bone / Socket Index=&SequentialInt(Use Array Data:NiagaraBool=false, First Number:NiagaraInt32=0, Sequence:NiagaraDataInterfaceArrayInt32=?, Last Number:NiagaraInt32=1)
    Bone Sampling Mode=::NewEnumerator4
    MaskSpawnGroup=false
    Orientation Sampling=::NewEnumerator1
    Skeletal Mesh=$System.Skeletal Mesh Reader
   )
   &ShapeLocation(MaskSpawnGroup=false, Override Seed=false, Sphere Radius=8, UseOwnerScale=true)
   /CascadeToNiagaraConverter/NiagaraScripts/ModuleScripts/FixShapeVector()
   &AddVelocity(Velocity=&Multiply_VectorByFloat(Vector=$Output.FixShapeVector.ShapeVector, Float=15))
   &AddVelocity(Velocity=<0 0 -200>) {module=AddVelocity001}
   &AddVelocity(Velocity=&FN_FrontEndVector3Modifier(Gameplay=<0 0 0>, Front End=<0 -200 900>)) {module=AddVelocity002}
  ]
  particleUpdateScript [
   &ParticleState()
   &SpriteRotationRate(Rotation Rate=&AngleConversion(
     Angle:NiagaraFloat=$Particles.SpriteRotationRate
     Angle Input:ENiagara_AngleInput=::NewEnumerator1
     Angle Output:ENiagara_AngleInput=::NewEnumerator0
    ))
   &ScaleSpriteSize(
    Scale Sprite Size Mode=::NewEnumerator0
    Uniform Scale Factor=&FloatFromCurve(FloatCurve=DataInterface() {propertyValues {Curve {keys [
         <at 0 0 RCIM_Cubic>
         {arriveTangent=3.8762, interpMode=RCIM_Cubic, leaveTangent=3.8762, time=0.1034, value=1.1346}
         {arriveTangent=-0.4418, interpMode=RCIM_Cubic, leaveTangent=-0.4418, time=0.2227, value=0.8631}
         {arriveTangent=0.2382, interpMode=RCIM_Cubic, leaveTangent=0.2382, time=0.2988, value=1.0483}
         {arriveTangent=0.0856, interpMode=RCIM_Cubic, leaveTangent=0.0856, time=0.8126, value=1.0037}
         {arriveTangent=-5.3567, interpMode=RCIM_Cubic, leaveTangent=-5.3567, time=0.8837, value=1.0984}
         <at 1 0 RCIM_Cubic>
        ]}}})
   )
   &ScaleColor(
    Scale RGB=&VectorFromCurve(VectorCurve=DataInterface() {propertyValues {XCurve {keys [<at 0 15>,<at 0.1 1>,<at 1 0>]}, YCurve {keys [<at 0 15>,<at 0.1 1>,<at 1 0>]}, ZCurve {keys [<at 0 15>,<at 0.1 1>,<at 1 0>]}}})
    ScaleA=true
    ScaleRGB=true
   )
   &SolveForcesAndVelocity(Clamp Velocity=false, Limit Acceleration=false)
  ]
  rendererClasses [&NiagaraSpriteRendererProperties]
  renderers [{rendererClass=&NiagaraSpriteRendererProperties}]
 }
 StarsGPU_L {
  emitterUpdateScript [
   &EmitterState(Inactive Response=::NewEnumerator1, Life Cycle Mode=::NewEnumerator1, Loop Duration=1, UseLoopDelay=true)
   &SpawnRate(SpawnRate=70, Use Spawn Probability=false)
  ]
  particleSpawnScript [
   /SetVariables(Particles.SpriteRotationRate=&RandomRangeFloat(Minimum=0.25, Maximum=1, Evaluation Type=::NewEnumerator1)) {isSetParameters=true}
   &InitializeParticle(
    Sprite Rotation Angle=&RandomRangeFloat(Minimum=0, Maximum=1, Evaluation Type=::NewEnumerator1)
    Sprite Rotation Mode=::NewEnumerator1
    Sprite Size=&Vector2DFromFloat(Value=&RandomRangeFloat(Minimum=5, Maximum=25, Evaluation Type=::NewEnumerator1))
    Sprite Size Mode=::NewEnumerator4
    UsePositionOffset=false
    Write Material Random=true
   )
   &AddVelocity(Velocity=<0 0 -300>)
   &Color(Color=&MakeLinearColorFromVectorAndFloat("Vector (RGB)"=<1 5 15>, "Float (Alpha)"=1))
   &SkeletalMeshLocation(
    Bone / Socket Index=&SequentialInt(Use Array Data:NiagaraBool=false, First Number:NiagaraInt32=0, Sequence:NiagaraDataInterfaceArrayInt32=?, Last Number:NiagaraInt32=1)
    Bone Sampling Mode=::NewEnumerator4
    MaskSpawnGroup=false
    Orientation Sampling=::NewEnumerator1
    Skeletal Mesh=$System.Skeletal Mesh Reader
   )
   &ShapeLocation(
    MaskSpawnGroup=false
    Override Seed=false
    Sphere Radius=&RandomRangeFloat(Minimum=1, Maximum=8, Evaluation Type=::NewEnumerator1)
    UseOwnerScale=true
   )
   &FixShapeVector()
   &AddVelocity(Velocity=&Multiply_VectorByFloat(Vector=$Output.FixShapeVector.ShapeVector, Float=25)) {module=AddVelocity001}
   &AddVelocity(Velocity=&FN_FrontEndVector3Modifier(Gameplay=<0 0 0>, Front End=<0 -200 900>)) {module=AddVelocity002}
  ]
  particleUpdateScript [
   &ParticleState()
   &SpriteRotationRate(Rotation Rate=&AngleConversion(
     Angle:NiagaraFloat=$Particles.SpriteRotationRate
     Angle Input:ENiagara_AngleInput=::NewEnumerator1
     Angle Output:ENiagara_AngleInput=::NewEnumerator0
    ))
   &ScaleSpriteSize(
    Scale Sprite Size Mode=::NewEnumerator0
    Uniform Scale Factor=&FloatFromCurve(FloatCurve=DataInterface() {propertyValues {Curve {keys [
         <at 0 0 RCIM_Cubic>
         {arriveTangent=9.64, interpMode=RCIM_Cubic, leaveTangent=9.64, time=0.0654, value=1.1713}
         {arriveTangent=-0.5426, interpMode=RCIM_Cubic, leaveTangent=-0.5426, time=0.106, value=1.0214}
         {arriveTangent=-0.1819, interpMode=RCIM_Cubic, leaveTangent=-0.1819, time=0.5487, value=0.909}
         {arriveTangent=-2.0257, interpMode=RCIM_Cubic, leaveTangent=-2.0257, time=0.6946, value=0.9143}
         <at 0.9975 0>
        ]}}})
   )
   &SolveForcesAndVelocity(Clamp Velocity=false, Limit Acceleration=false)
  ]
  rendererClasses [&NiagaraSpriteRendererProperties]
  renderers [{rendererClass=&NiagaraSpriteRendererProperties}]
 }
}
summary {
 systemName=NS_Contrail_DiscoStar
 emitters [
  {emitterName=Ribbon_Left_Foot, rendererClasses [&NiagaraRibbonRendererProperties]}
  {emitterName=StarsGPU_L, rendererClasses [&NiagaraSpriteRendererProperties]}
  {emitterName=SparklesHandL, rendererClasses [&NiagaraSpriteRendererProperties]}
 ]
 userVariables [
  User.bIsFrontEnd {type=NiagaraBool}
  User.bIsFrontEndPreview {type=NiagaraBool}
  User.ParaGlideLeanAlpha {type=NiagaraFloat, defaultValue=0, description=""}
 ]
}
