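# Niagara skydiving contrail cosmetic with a fiber-optic look, apparently converted from Cascade, attached to two skeletal mesh bones per emitter.
# Two mesh emitters draw the fibers through dynamic material parameters, one bursting two scale variants, the other spawning continuously at partitioned bone offsets.
# Soft glow sprites follow the hand bones, while color-shifting sparkle dust drifts under curl noise, its vertical launch velocity boosted by dive amount.
# Front-end modifiers adjust glow and fiber brightness outside gameplay.
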NiagaraSystem /BRCosmetics/Athena/Effects/Particles/Skydiving/NS_Contrail_FiberOptics_Converted
dependencies {
 usedDataInterfaces [&NiagaraDataInterfaceCurve, &NiagaraDataInterfaceVectorCurve, &NiagaraDataInterfaceArrayFloat3]
 usedDynamicInputs []
 usedModules [
  &self:SystemSpawnScript.SystemScriptSource.SystemScriptGraph.NiagaraNodeAssignment_0.SetVariables_#1
  &SystemState
 ]
}
emitters {
 FN_SparkleDust {
  emitterUpdateScript [
   &EmitterState()
   &SpawnRate(SpawnRate=16, Use Spawn Probability=false)
   /Game/Effects/Niagara/Modules/Skydiving/FN_DivingAttributes/FN_EmitterDivingAttributes_v1(SkeletalMesh:NiagaraDataInterfaceSkeletalMesh=$User.SkeletalMesh)
  ]
  particleSpawnScript [
   &FN_SetSpawnGroup_v2(Number of Spawn Groups=2)
   /Niagara/Modules/Spawn/Initialization/InitializeParticle(
    Lifetime=&UniformRangedFloat(Minimum=0.6, Maximum=0.8)
    Mesh Scale=&Multiply_Vector(
     A=<1 1 1>
     B=&VectorFromFloat(Value=&UniformRangedFloat(Minimum=0.2, Maximum=1))
    )
    Module.Write Color=false
    Module.Write Lifetime=true
    Module.Write Mass=true
    Module.Write Material Random=true
    Module.Write Position=false
    Module.Write Scale=false
    Module.Write SpriteRotation=true
    Module.Write SpriteSize=true
    Position=&SimulationPosition()
    Sprite Rotation=&UniformAOrBFloat(A=0, B=360)
    Sprite Size=&Vector2DFromFloat(Value=&UniformRangedFloat(Minimum=3, Maximum=6))
   )
   &SkeletalMeshLocation(
    Bone / Socket Index=$Particles.SpawnGroup
    Bone Sampling Mode=::NewEnumerator4
    Inherit Velocity=false
    MaskSpawnGroup=false
    Skeletal Mesh=$User.SkeletalMesh
   )
   &AddVelocity(
    Coordinate Space=::World
    Scale Added Velocity=&Lerp_Vector(A=<1 1 1>, B=<1 1 3>, Alpha=$Emitter.Dive)
    Velocity=&UniformRangedVector(Minimum=<0 0 125>, Maximum=<1 1 225>)
   )
   /Niagara/Modules/Spawn/Location/SphereLocation(MaskSpawnGroup=false, Module.Override Randomness=false, Sphere Radius=12.5, Surface Only=false)
  ]
  particleUpdateScript [
   &ParticleState()
   &Drag(Drag=0.5, Use Linear Drag=true, Use Rotational Drag=false)
   /Niagara/Modules/Update/Forces/CurlNoiseForce(
    Noise Frequency:NiagaraFloat=&UniformRangedFloat(Minimum=0.5, Maximum=1)
    Noise Strength=&UniformRangedFloat(Minimum=300, Maximum=500)
    Pan Noise=false
    Pan Noise Field=<1 1 0>
   )
   &SolveForcesAndVelocity(Clamp Velocity=false, Limit Acceleration=false)
   &ScaleSpriteSize(Scale Factor=&Multiply_Vector2(
     A=<1 1>
     B=&Vector2DFromFloat(Value=&FloatFromCurve(FloatCurve=DataInterface() {propertyValues {Curve {keys [<at -0.0046 3>,<at 1.0011 0.002>]}}}))
    ))
   &ScaleColor(
    Scale Alpha=&FloatFromCurve(FloatCurve=DataInterface() {propertyValues {Curve {keys [<at 0 0>,<at 0.12 1>,<at 0.6 0>,<at 1 0>]}}})
    Scale RGB=&VectorFromCurve(VectorCurve=DataInterface() {propertyValues {
       XCurve {keys [<at 0 3>,<at 0.15 4>,<at 0.5 0>]}
       YCurve {keys [<at 0 1.8323>,<at 0.15 1.2168>,<at 0.5 2.379>]}
       ZCurve {keys [<at 0 0>,<at 0.15 0>,<at 0.5 8>]}
      }})
    ScaleA=true
    ScaleRGB=true
   )
  ]
  rendererClasses [&NiagaraSpriteRendererProperties]
  renderers [{rendererClass=&NiagaraSpriteRendererProperties}]
 }
 Glow {
  emitterUpdateScript [
   &EmitterState(Inactive Response=::NewEnumerator1, Life Cycle Mode=::NewEnumerator1, Loop Duration=0.125, UseLoopDelay=true)
   &SpawnBurst_Instantaneous(
    Spawn Count=&RandomRangeInt(Minimum=1, Maximum=2)
    Use Spawn Probability=false
    UseLoopCountLimit=false
   )
   &SpawnRate(SpawnRate=2, Use Spawn Probability=false)
  ]
  particleSpawnScript [
   /SetVariables(Particles.SkeletonSamplingIndex:NiagaraInt32=/CascadeToNiagaraConverter/NiagaraScripts/DynamicInputScripts/SequentialInt(Use Array Data:NiagaraBool=false, First Number:NiagaraInt32=0, Sequence:NiagaraDataInterfaceArrayInt32=?, Last Number:NiagaraInt32=1)) {isSetParameters=true}
   /Niagara/Modules/Spawn/Initialization/V2/InitializeParticle(Lifetime=1.5, Sprite Size=<90 90>, Sprite Size Mode=::NewEnumerator4, UsePositionOffset=false, Write Material Random=true)
   &CameraOffset(Camera Offset Amount=&RandomRangeFloat(Minimum=10, Maximum=30, Evaluation Type=::NewEnumerator1))
   &Color(Color=&MakeLinearColorFromVectorAndFloat(
     "Vector (RGB)"=<0.1856 0.7152 1>
     "Float (Alpha)"=&FN_FrontEndFloatModifier(Gameplay=0.22, Front End=0.2)
    ))
  ]
  particleUpdateScript [
   &ParticleState()
   &ScaleSpriteSize(
    Scale Sprite Size Mode=::NewEnumerator0
    Uniform Scale Factor=&FloatFromCurve(FloatCurve=DataInterface() {propertyValues {Curve {keys [<at 0 0.6>,<at 1 1.8>]}}})
   )
   &ScaleColor(
    Scale Alpha=&FloatFromCurve(FloatCurve=DataInterface() {propertyValues {Curve {keys [
         <at 0 0 RCIM_Cubic>
         {arriveTangent=0.5535, interpMode=RCIM_Cubic, leaveTangent=0.5535, tangentMode=RCTM_User, time=0.0706, value=0.7783}
         {arriveTangent=-0.8375, interpMode=RCIM_Cubic, leaveTangent=-0.8375, time=0.8172, value=0.8553}
         <at 1 0>
        ]}}})
    Scale RGB=&VectorFromCurve(VectorCurve=DataInterface() {propertyValues {
       XCurve {keys [<at 0 0 RCIM_Cubic>,<at 0.8939 0.0713 RCIM_Cubic>]}
       YCurve {keys [<at 0 1.0706 RCIM_Cubic>,<at 0.8939 0.1749 RCIM_Cubic>]}
       ZCurve {keys [<at 0 1.2 RCIM_Cubic>,<at 0.8939 0.5473 RCIM_Cubic>]}
      }})
    ScaleA=true
    ScaleRGB=true
   )
   &SkeletalMeshLocation(
    Bone / Socket Index=$Particles.SkeletonSamplingIndex
    Bone Sampling Mode=::NewEnumerator4
    MaskSpawnGroup=false
    Orientation Sampling=::NewEnumerator1
    Sampled Position Offset=&SelectVectorFromArray(Direct Array Index=$Particles.SkeletonSamplingIndex, Vector Selection Array=DataInterface() {propertyValues {FloatData [<7 0 0>, <-7 0 0>]}})
    Skeletal Mesh=DataInterface() {propertyValues {
      DefaultMesh=null
      ExcludeBoneName=root
      OptionalComponentClass=null
      PreviewMesh=null
      SkinningMode=SkinOnTheFly
      SoftSourceActor=null
      SourceMode=Default
      UvSetIndex=0
      WholeMeshLOD=-1
      bExcludeBone=false
      bReadDeformedGeometry=true
      bRequireCurrentFrameData=true
      ComponentTags []
      FilteredBones [hand_l,hand_r]
      FilteredSockets []
      MeshUserParameter {parameter {name=None, typeDefHandle {registeredTypeIndex=1}, varData []}}
      SamplingRegions []
     }}
   )
  ]
  rendererClasses [&NiagaraSpriteRendererProperties]
  renderers [{rendererClass=&NiagaraSpriteRendererProperties}]
 }
 fiber_a002 {
  emitterUpdateScript [
   &EmitterState(Inactive Response=::NewEnumerator1, Life Cycle Mode=::NewEnumerator1, Loop Duration=1, UseLoopDelay=true)
   &SpawnBurst_Instantaneous(Spawn Count=4, Use Spawn Probability=false, UseLoopCountLimit=false)
   &SpawnBurst_Instantaneous(Spawn Count=4, Spawn Group=1, Use Spawn Probability=false, UseLoopCountLimit=false) {module=SpawnBurst_Instantaneous001}
  ]
  particleSpawnScript [
   /SetVariables(Particles.VisibilityTag=&SpawnGroup()) {isSetParameters=true}
   &InitializeParticle(
    Mesh Scale=&MergeHelper_VectorSwitch(
     0=&RandomRangeVector(Minimum=<0.9 0.92 2.8>, Maximum=<1.35 1.35 4.2>, Evaluation Type=::NewEnumerator1)
     1=&RandomRangeVector(Minimum=<0.42 0.42 5>, Maximum=<0.57 0.57 6>, Evaluation Type=::NewEnumerator1)
     2=?
     3=?
     Index=&SpawnGroup()
     Num=::NewEnumerator0
    )
    Mesh Scale Mode=::NewEnumerator4
    UsePositionOffset=false
    Write Material Random=true
   )
   &Color(Color=&MakeLinearColorFromVectorAndFloat(
     "Vector (RGB)"=<1 1 1>
     "Float (Alpha)"=&MergeHelper_FloatSwitch(
      0=&FN_FrontEndFloatModifier(Gameplay=1, Front End=0.35)
      1=&FN_FrontEndFloatModifier(Gameplay=1, Front End=0.35)
      2=?
      3=?
      Index=&SpawnGroup()
      Num=::NewEnumerator0
     )
    ))
   &DynamicMaterialParameters(
    Index 0 Param 1=&MergeHelper_FloatSwitch(
     0=&RandomRangeFloat(Minimum=0.69, Maximum=0.9, Evaluation Type=::NewEnumerator1)
     1=&RandomRangeFloat(Minimum=0.62, Maximum=0.9, Evaluation Type=::NewEnumerator1)
     2=?
     3=?
     Index=&SpawnGroup()
     Num=::NewEnumerator0
    )
    Index 0 Param 2=&MergeHelper_FloatSwitch(
     0=&RandomRangeFloat(Minimum=0.7, Maximum=1, Evaluation Type=::NewEnumerator1)
     1=&RandomRangeFloat(Minimum=0.7, Maximum=1, Evaluation Type=::NewEnumerator1)
     2=?
     3=?
     Index=&SpawnGroup()
     Num=::NewEnumerator0
    )
    Module.Index0Float0Write=true
    Module.Index0Float1Write=true
    Module.Index0Float2Write=true
    Module.Index0Float3Write=true
   ) {module=DynamicMaterialParameters003}
  ]
  particleUpdateScript [
   &ParticleState()
   &SkeletalMeshLocation(
    Bone / Socket Index=&SequentialInt(Use Array Data:NiagaraBool=false, First Number:NiagaraInt32=0, Sequence:NiagaraDataInterfaceArrayInt32=?, Last Number:NiagaraInt32=1)
    Bone Sampling Mode=::NewEnumerator4
    MaskSpawnGroup=false
    Orientation Sampling=::NewEnumerator1
    Sampled Position Offset=&FN_FrontEndVector3Modifier(Gameplay=<0 0 0>, Front End=<0 0 0>)
    Sampled Position Offset Coordinate Space:ENiagaraSkelMeshTransforms=::NewEnumerator2
    Skeletal Mesh=$System.Skeletal Mesh Reader
   ) {module=SkeletalMeshLocation001}
   &DynamicMaterialParameters(
    Index 0 Param 4=&MergeHelper_FloatSwitch(0=420, 1=210, 2=?, 3=?, Index=&SpawnGroup(), Num=::NewEnumerator0)
    Module.Index0Float0Write=false
    Module.Index0Float1Write=false
    Module.Index0Float2Write=true
    Module.Index0Float3Write=true
   ) {module=DynamicMaterialParameters001}
   &ScaleColor(
    Scale Alpha=&FN_FrontEndFloatModifier(Gameplay=1, Front End=0.75)
    Scale RGB=&FN_FrontEndVector3Modifier(Gameplay=<2.2 2.2 2.2>, Front End=&Multiply_VectorByFloat(Vector=<1 1 1>, Float=2))
    ScaleA=true
    ScaleRGB=true
   )
   &SolveForcesAndVelocity(Clamp Velocity=false, Limit Acceleration=false)
  ]
  rendererClasses [&NiagaraMeshRendererProperties]
  renderers [
   {rendererClass=&NiagaraMeshRendererProperties, rendererIndex=0}
   {rendererClass=&NiagaraMeshRendererProperties, rendererIndex=1}
  ]
 }
 fiber_c {
  emitterUpdateScript [
   &EmitterState(Inactive Response=::NewEnumerator1, Life Cycle Mode=::NewEnumerator1, Loop Duration=1, UseLoopDelay=true)
   &SpawnRate(SpawnRate=10, Use Spawn Probability=false)
  ]
  particleSpawnScript [
   /SetVariables(Particles.VisibilityTag=&SpawnGroup()) {isSetParameters=true}
   &InitializeParticle(
    Lifetime=&RandomRangeFloat(Minimum=3.3, Maximum=0.7, Evaluation Type=::NewEnumerator1)
    Mesh Scale=/Niagara/DynamicInputs/CascadeConversion/SplitVector(
     Double Component Vector Channels:ECascadeNiagaraTwoVectorChannels=::NewEnumerator0
     Double Component:NiagaraFloat=&RandomRangeFloat(Minimum=0.05, Maximum=0.1, Evaluation Type=::NewEnumerator1)
     Single Component:NiagaraFloat=&RandomRangeFloat(Minimum=1, Maximum=2.2, Evaluation Type=::NewEnumerator1)
    )
    Mesh Scale Mode=::NewEnumerator4
    UsePositionOffset=false
    Write Material Random=true
   ) {enabled=false, module=InitializeParticle001}
   &InitializeParticle(
    Lifetime=&RandomRangeFloat(Minimum=3.3, Maximum=0.7, Evaluation Type=::NewEnumerator1)
    Mesh Scale=&SplitVector(
     Double Component Vector Channels:ECascadeNiagaraTwoVectorChannels=::NewEnumerator0
     Double Component:NiagaraFloat=&RandomRangeFloat(Minimum=0.05, Maximum=0.1, Evaluation Type=::NewEnumerator1)
     Single Component:NiagaraFloat=&RandomRangeFloat(Minimum=1, Maximum=2.2, Evaluation Type=::NewEnumerator1)
    )
    Mesh Scale Mode=::NewEnumerator4
    UsePositionOffset=false
    Write Material Random=true
   )
   &PartitionParticles(Number Of Partitions=5)
  ]
  particleUpdateScript [
   &ParticleState()
   &Color(Color=&MakeLinearColorFromVectorAndFloat(
     "Vector (RGB)"=&VectorFromCurve(VectorCurve=DataInterface() {propertyValues {XCurve {keys [<at 0 1>,<at 1 1>]}, YCurve {keys [<at 0 1>,<at 1 1>]}, ZCurve {keys [<at 0 1>,<at 1 1>]}}})
     "Float (Alpha)"=&ClampFloat(Float=&FloatFromCurve(FloatCurve=DataInterface() {propertyValues {Curve {keys [{arriveTangent=-1, time=0, value=1}, <at 1 0>]}}}))
    ))
   &SkeletalMeshLocation(
    Bone / Socket Index=&MergeHelper_IntSwitch(
     0=0
     1=1
     2=?
     3=?
     4=?
     5=?
     Index=&SequentialInt(Use Array Data:NiagaraBool=false, First Number:NiagaraInt32=0, Sequence:NiagaraDataInterfaceArrayInt32=?, Last Number:NiagaraInt32=1)
     Num=::NewEnumerator0
    )
    Bone Sampling Mode=::NewEnumerator4
    MaskSpawnGroup=false
    Orientation Sampling=::NewEnumerator1
    Sampled Position Offset=&MergeHelper_VectorSwitch(
     0=&MergeHelper_VectorSwitch(0=<10 0 0>, 1=<-8 0 -12>, 2=<0 10 9>, 3=<5 10 10>, 4=<8 8 0>, Index=$Particles.PartitionParticles.Partition, Num=::NewEnumerator3)
     1=&MergeHelper_VectorSwitch(0=<10 0 0>, 1=<-8 0 -12>, 2=<0 10 9>, 3=<5 10 10>, 4=<8 8 0>, Index=$Particles.PartitionParticles.Partition, Num=::NewEnumerator3)
     2=?
     3=?
     Index=&SpawnGroup()
     Num=::NewEnumerator0
    )
    Sampled Position Offset Coordinate Space:ENiagaraSkelMeshTransforms=::NewEnumerator2
    Skeletal Mesh=$System.Skeletal Mesh Reader
   )
   &DynamicMaterialParameters(
    Index 0 Param 1=&RandomRangeFloat(Minimum=-1.25, Maximum=0.8)
    Index 0 Param 2=0.5
    Index 0 Param 3=1
    Index 0 Param 4=400
    Module.Index0Float0Write=true
    Module.Index0Float1Write=true
    Module.Index0Float2Write=true
    Module.Index0Float3Write=true
   ) {module=DynamicMaterialParameters001}
   &ScaleColor(
    Color Value To Scale=$Particles.Color
    Scale Alpha=&FN_FrontEndFloatModifier(
     Gameplay=&FloatFromCurve(FloatCurve=DataInterface() {propertyValues {Curve {keys [<at 0 0>,<at 0.2 1>,<at 0.4 1>,<at 1 0>]}}})
     Front End=&FloatFromCurve(FloatCurve=DataInterface() {propertyValues {Curve {keys [<at 0 0>,<at 0.2 0.35>,<at 0.4 0.35>,<at 1 0>]}}})
    )
    Scale RGB=<14 14 14>
    ScaleA=true
    ScaleRGB=true
   )
   &Drag(Use Linear Drag=true, Use Rotational Drag=true)
   &SolveForcesAndVelocity(Clamp Velocity=false, Limit Acceleration=false)
  ]
  rendererClasses [&NiagaraMeshRendererProperties]
  renderers [{rendererClass=&NiagaraMeshRendererProperties, rendererIndex=0}]
 }
}
summary {
 systemName=NS_Contrail_FiberOptics_Converted
 emitters [
  {emitterName=Glow, rendererClasses [&NiagaraSpriteRendererProperties]}
  {emitterName=fiber_c, rendererClasses [&NiagaraMeshRendererProperties]}
  {emitterName=fiber_a002, rendererClasses [&NiagaraMeshRendererProperties]}
  {emitterName=FN_SparkleDust, rendererClasses [&NiagaraSpriteRendererProperties]}
 ]
 userVariables [
  User.bIsFrontEnd {type=NiagaraBool}
  User.bIsFrontEndPreview {type=NiagaraBool}
  User.PlayerSpeed {type=NiagaraFloat, defaultValue=0, description=""}
  User.SkeletalMesh {
   type=NiagaraDataInterfaceSkeletalMesh
   defaultValue {
    struct=NiagaraExt_VariableValue_DataInterface
    value {dataInterface=&self:NiagaraDataInterfaceSkeletalMesh_0}
   }
  }
 ]
}
