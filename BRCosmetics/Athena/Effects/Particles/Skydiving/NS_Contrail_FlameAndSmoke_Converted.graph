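# Niagara skydiving contrail, apparently converted from Cascade, that trails flame and smoke from character bones.
# Flare sprites on the pinky and ball-of-foot bones and short-lived, owner-velocity-influenced mesh flames use bright HDR oranges that darken over life.
# Despite its name, Ribbon_Left_Foot cycles through four bone indices, using spawn-group ribbon IDs for separate trails that darken and fade.
# Front-end modifiers change mesh and ribbon offset, velocity or alpha for menu display.
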NiagaraSystem /BRCosmetics/Athena/Effects/Particles/Skydiving/NS_Contrail_FlameAndSmoke_Converted
dependencies {
 usedDataInterfaces [&NiagaraDataInterfaceCurve, &NiagaraDataInterfaceVectorCurve, &NiagaraDataInterfaceSkeletalMesh]
 usedDynamicInputs []
 usedModules [
  &self:SystemSpawnScript.SystemScriptSource.SystemScriptGraph.NiagaraNodeAssignment_0.SetVariables_#1
  &SystemState
 ]
}
emitters {
 Flare {
  emitterUpdateScript [
   &EmitterState(Inactive Response=::NewEnumerator1, Life Cycle Mode=::NewEnumerator1, Loop Duration=0.1, UseLoopDelay=true)
   &SpawnRate(SpawnRate=12, Use Spawn Probability=false)
  ]
  particleSpawnScript [
   &InitializeParticle(
    Lifetime=&RandomRangeFloat(Minimum=0.1, Maximum=0.2, Evaluation Type=::NewEnumerator1)
    Sprite Size=&Vector2DFromFloat(Value=&RandomRangeFloat(Minimum=40, Maximum=60, Evaluation Type=::NewEnumerator1))
    Sprite Size Mode=::NewEnumerator4
    UsePositionOffset=false
    Write Material Random=true
   )
   &Color(Color=&MakeLinearColorFromVectorAndFloat(
     "Vector (RGB)"=&RandomRangeVector(Minimum=<16 1.7176 0.4428>, Maximum=<24 4.6283 1.5139>, Evaluation Type=::NewEnumerator1)
     "Float (Alpha)"=&RandomRangeFloat(Minimum=0.8, Maximum=1, Evaluation Type=::NewEnumerator1)
    ))
  ]
  particleUpdateScript [
   &ParticleState()
   &SkeletalMeshLocation(
    Bone / Socket Index=/CascadeToNiagaraConverter/NiagaraScripts/DynamicInputScripts/SequentialInt(Use Array Data:NiagaraBool=false, First Number:NiagaraInt32=0, Sequence:NiagaraDataInterfaceArrayInt32=?, Last Number:NiagaraInt32=3)
    Bone Sampling Mode=::NewEnumerator4
    MaskSpawnGroup=false
    Orientation Sampling=::NewEnumerator1
    Skeletal Mesh=DataInterface() {propertyValues {
      DefaultMesh=null
      ExcludeBoneName=root
      OptionalComponentClass=null
      PreviewMesh=null
      SkinningMode=SkinOnTheFly
      SoftSourceActor=null
      SourceMode=Default
      UvSetIndex=0
      WholeMeshLOD=-1
      bExcludeBone=false
      bReadDeformedGeometry=true
      bRequireCurrentFrameData=true
      ComponentTags []
      FilteredBones [pinky_01_r,pinky_01_l,ball_r,ball_l]
      FilteredSockets []
      MeshUserParameter {parameter {name=None, typeDefHandle {registeredTypeIndex=1}, varData []}}
      SamplingRegions []
     }}
   )
   &ScaleColor(
    Scale Alpha=&FloatFromCurve(FloatCurve=DataInterface() {propertyValues {Curve {keys [
         <at 0.0006 -0.0023 RCIM_Cubic>
         {arriveTangent=1.7677, interpMode=RCIM_Cubic, leaveTangent=1.7677, time=0.2472, value=0.24}
         {arriveTangent=-0.0001, interpMode=RCIM_Cubic, leaveTangent=-0.0001, time=0.5, value=0.8804}
         {arriveTangent=-1.7608, interpMode=RCIM_Cubic, leaveTangent=-1.7608, time=0.7566, value=0.24}
         <at 1 0 RCIM_Cubic>
        ]}}})
    Scale RGB=&VectorFromCurve(VectorCurve=DataInterface() {propertyValues {
       XCurve {keys [
         <at 0.0553 3 RCIM_Cubic>
         {arriveTangent=-2.9005, interpMode=RCIM_Cubic, leaveTangent=-2.9005, time=0.2995, value=1}
         <at 1 0.2598 RCIM_Cubic>
        ]}
       YCurve {keys [<at 0.0553 3 RCIM_Cubic>, {arriveTangent=-3.098, interpMode=RCIM_Cubic, leaveTangent=-3.098, time=0.2995, value=1}, <at 1 0.0733 RCIM_Cubic>]}
       ZCurve {keys [
         <at 0.0553 3 RCIM_Cubic>
         {arriveTangent=-3.1474, interpMode=RCIM_Cubic, leaveTangent=-3.1474, time=0.2995, value=1}
         <at 1 0.0266 RCIM_Cubic>
        ]}
      }})
    ScaleA=true
    ScaleRGB=true
   )
   &CameraOffset(Camera Offset Amount=&RandomRangeFloat(Minimum=30, Maximum=40, Evaluation Type=::NewEnumerator1))
  ]
  rendererClasses [&NiagaraSpriteRendererProperties]
  renderers [{rendererClass=&NiagaraSpriteRendererProperties}]
 }
 Particle_Emitter {
  emitterUpdateScript [
   &EmitterState(Inactive Response=::NewEnumerator1, Life Cycle Mode=::NewEnumerator1, Loop Duration=0.1, UseLoopDelay=true)
   &SpawnRate(SpawnRate=60, Use Spawn Probability=false)
  ]
  particleSpawnScript [
   &InitializeParticle(
    Lifetime=&RandomRangeFloat(Minimum=0.1, Maximum=0.2, Evaluation Type=::NewEnumerator1)
    Mesh Scale=&RandomRangeVector(Minimum=<4 -2 -2>, Maximum=<3 1.5 1.5>, Evaluation Type=::NewEnumerator1)
    Mesh Scale Mode=::NewEnumerator4
    UsePositionOffset=false
    Write Material Random=true
   )
   &SkeletalMeshLocation(
    Bone / Socket Index=&SequentialInt(Use Array Data:NiagaraBool=false, First Number:NiagaraInt32=0, Sequence:NiagaraDataInterfaceArrayInt32=?, Last Number:NiagaraInt32=3)
    Bone Sampling Mode=::NewEnumerator4
    MaskSpawnGroup=false
    Orientation Sampling=::NewEnumerator1
    Sampled Position Offset=&FN_FrontEndVector3Modifier(Gameplay=<0 0 -10>, Front End=<0 0 0>)
    Sampled Position Offset Coordinate Space:ENiagaraSkelMeshTransforms=::NewEnumerator2
    Skeletal Mesh=$System.Skeletal Mesh Reader
   )
   &Color(Color=&MakeLinearColorFromVectorAndFloat(
     "Vector (RGB)"=&RandomRangeVector(Minimum=<8 0.7049 0.0417>, Maximum=<100 5.9748 1.9948>, Evaluation Type=::NewEnumerator1)
     "Float (Alpha)"=&RandomRangeFloat(Minimum=3, Maximum=6, Evaluation Type=::NewEnumerator1)
    ))
   &CameraOffset(Camera Offset Amount=&FloatFromCurve(FloatCurve=DataInterface() {propertyValues {Curve {keys [<at 0 10>,<at 1 30>]}}}))
   /CascadeToNiagaraConverter/NiagaraScripts/ModuleScripts/AddVectorToPosition(Vector:Vector3f=&TransformVector(
     Vector=&RandomRangeVector(Minimum=<-2 -2 -2>, Maximum=<2 2 2>, Evaluation Type=::NewEnumerator1)
     Destination Space=::Simulation
    ))
   &AddVelocity(Velocity=&FN_FrontEndVector3Modifier(Gameplay=<0 0 0>, Front End=<0 -0.3 1>))
   &InheritVelocity(
    Inherited Velocity Amount Scale=<0.05 0.05 0.05>
    Inherited Velocity Speed Limit=0
    Speed Limit=false
    UseSpeedThreshold=false
    Velocity Source=&Multiply_VectorByFloat(
     Vector=&TransformVector(Vector=$Engine.Owner.Velocity, Source Space=::World, Destination Space=::Local)
     Float=-1
    )
   )
  ]
  particleUpdateScript [
   &ParticleState()
   &ScaleMeshSize(Scale Factor=&VectorFromCurve(VectorCurve=DataInterface() {propertyValues {
       XCurve {keys [{arriveTangent=3.8758, interpMode=RCIM_Cubic, leaveTangent=3.8758, tangentMode=RCTM_User, time=0, value=1}, <at 1 3 RCIM_Cubic>]}
       YCurve {keys [{interpMode=RCIM_Cubic, tangentMode=RCTM_User, time=0, value=1}, <at 1 1 RCIM_Cubic>]}
       ZCurve {keys [{interpMode=RCIM_Cubic, tangentMode=RCTM_User, time=0, value=1}, <at 1 1 RCIM_Cubic>]}
      }}))
   &ScaleColor(
    Scale Alpha=&FloatFromCurve(FloatCurve=DataInterface() {propertyValues {Curve {keys [
         <at 0.0006 -0.0023 RCIM_Cubic>
         {arriveTangent=2.3038, interpMode=RCIM_Cubic, leaveTangent=2.3038, time=0.223, value=1}
         {arriveTangent=-1.9066, interpMode=RCIM_Cubic, leaveTangent=-1.9066, time=0.3838, value=0.8804}
         {arriveTangent=-1.4288, interpMode=RCIM_Cubic, leaveTangent=-1.4288, time=0.6216, value=0.24}
         <at 1 0 RCIM_Cubic>
        ]}}})
    Scale RGB=&VectorFromCurve(VectorCurve=DataInterface() {propertyValues {
       XCurve {keys [<at 0.0553 3 RCIM_Cubic>, {arriveTangent=-3.0697, interpMode=RCIM_Cubic, leaveTangent=-3.0697, time=0.2995, value=1}, <at 1 0.1 RCIM_Cubic>]}
       YCurve {keys [
         <at 0.0553 3 RCIM_Cubic>
         {arriveTangent=-3.1662, interpMode=RCIM_Cubic, leaveTangent=-3.1662, time=0.2995, value=1}
         <at 1 0.0088 RCIM_Cubic>
        ]}
       ZCurve {keys [<at 0.0553 3 RCIM_Cubic>, {arriveTangent=-3.175, interpMode=RCIM_Cubic, leaveTangent=-3.175, time=0.2995, value=1}, <at 1 0.0005 RCIM_Cubic>]}
      }})
    ScaleA=true
    ScaleRGB=true
   )
   &SolveForcesAndVelocity(Clamp Velocity=false, Limit Acceleration=false)
   &DynamicMaterialParameters(
    Index 0 Param 1=&RandomRangeFloat(Minimum=0, Maximum=0.7, Evaluation Type=::NewEnumerator1)
    Index 0 Param 2=&RandomRangeFloat(Minimum=80, Maximum=90, Evaluation Type=::NewEnumerator1)
    Index 0 Param 3=1
    Index 0 Param 4=1
    Module.Index0Float0Write=true
    Module.Index0Float1Write=true
    Module.Index0Float2Write=true
    Module.Index0Float3Write=true
   )
   &UpdateMeshOrientation(Facing Interpolation=false, Orientation Method=::NewEnumerator1, Override Quat=false, TransformFacing=false, TransformUp=false)
  ]
  rendererClasses [&NiagaraMeshRendererProperties]
  renderers [{rendererClass=&NiagaraMeshRendererProperties, rendererIndex=0}]
 }
 Ribbon_Left_Foot {
  emitterUpdateScript [
   &EmitterState(Inactive Response=::NewEnumerator1, Life Cycle Mode=::NewEnumerator1, Loop Duration=1, UseLoopDelay=true)
   &SpawnRate(SpawnRate=120, Use Spawn Probability=false)
  ]
  particleSpawnScript [
   &FN_SetSpawnGroup()
   /SetVariables(Particles.RibbonID=/Niagara/DynamicInputs/Meshes/MakeRibbonIDFromSpawnGroup()) {isSetParameters=true}
   &InitializeParticle(
    Lifetime=0.19
    Ribbon Width=&Multiply_Float(A=15, B=2)
    Ribbon Width Mode=::NewEnumerator1
    UsePositionOffset=false
    Write Material Random=true
   )
   &SkeletalMeshLocation(
    Bone / Socket Index=&SequentialInt(Use Array Data:NiagaraBool=false, First Number:NiagaraInt32=0, Sequence:NiagaraDataInterfaceArrayInt32=?, Last Number:NiagaraInt32=3)
    Bone Sampling Mode=::NewEnumerator4
    MaskSpawnGroup=false
    Orientation Sampling=::NewEnumerator1
    Sampled Position Offset=<0 0 -10>
    Sampled Position Offset Coordinate Space:ENiagaraSkelMeshTransforms=::NewEnumerator2
    Skeletal Mesh=$System.Skeletal Mesh Reader
   )
   &AddVectorToPosition(Vector:Vector3f=&TransformVector(Vector=<0 0 0>, Destination Space=::Simulation))
   &Color(Color=&MakeLinearColorFromVectorAndFloat("Vector (RGB)"=<1 1 1>, "Float (Alpha)"=&FN_FrontEndFloatModifier(Gameplay=1, Front End=0.7)))
   &AddVelocity(Velocity=&FN_FrontEndVector3Modifier(Gameplay=<0 0 0>, Front End=<0 -400 1500>))
  ]
  particleUpdateScript [
   /SetVariables(Particles.ScaledRibbonWidth:NiagaraFloat=&Multiply_Float(
     A=$Particles.RibbonWidth
     B=&FN_FrontEndFloatModifier(
      Gameplay=&FloatFromCurve(FloatCurve=DataInterface() {propertyValues {Curve {keys [<at 0 0.6>,<at 0.8 0.85>,<at 1 1>]}}})
      Front End=&FloatFromCurve(FloatCurve=DataInterface() {propertyValues {Curve {keys [<at 0 0.6>,<at 0.8 0.85>,<at 1 1>]}}})
     )
    )) {isSetParameters=true}
   &ParticleState()
   &DynamicMaterialParameters(
    Index 0 Param 1=1
    Index 0 Param 2=&FloatFromCurve(FloatCurve=DataInterface() {propertyValues {Curve {keys [<at 0.05 0>,<at 0.25 1>]}}})
    Index 0 Param 3=&FloatFromCurve(FloatCurve=DataInterface() {propertyValues {Curve {keys [<at 0.5 0>,<at 1 1.5>]}}})
    Index 0 Param 4=1
    Module.Index0Float0Write=true
    Module.Index0Float1Write=true
    Module.Index0Float2Write=true
    Module.Index0Float3Write=true
   )
   &ScaleColor(
    Scale Alpha=&FN_FrontEndFloatModifier(
     Gameplay=&FloatFromCurve(FloatCurve=DataInterface() {propertyValues {Curve {keys [<at 0.6 1>,<at 1 0.3>]}}})
     Front End=&FloatFromCurve(FloatCurve=DataInterface() {propertyValues {Curve {keys [<at 0.6 1>,<at 0.6371 0.4821>,<at 1 0.3>]}}})
    )
    Scale RGB=&VectorFromCurve(VectorCurve=DataInterface() {propertyValues {XCurve {keys [<at 0.4 1>,<at 0.8 0.0156>]}, YCurve {keys [<at 0.4 1>,<at 0.8 0.0156>]}, ZCurve {keys [<at 0.4 1>,<at 0.8 0.0156>]}}})
    ScaleA=true
    ScaleRGB=true
   )
   &SolveForcesAndVelocity(Clamp Velocity=false, Limit Acceleration=false)
  ]
  rendererClasses [&NiagaraRibbonRendererProperties]
  renderers [{rendererClass=&NiagaraRibbonRendererProperties}]
 }
}
summary {
 systemName=NS_Contrail_FlameAndSmoke_Converted
 emitters [
  {emitterName=Ribbon_Left_Foot, rendererClasses [&NiagaraRibbonRendererProperties]}
  {emitterName=Particle_Emitter, rendererClasses [&NiagaraMeshRendererProperties]}
  {emitterName=Flare, rendererClasses [&NiagaraSpriteRendererProperties]}
 ]
 userVariables [
  User.bIsFrontEnd {type=NiagaraBool}
  User.bIsFrontEndPreview {type=NiagaraBool}
  User.ParaGlideLeanAlpha {type=NiagaraFloat, defaultValue=0, description=""}
 ]
}
