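# Niagara skydiving contrail with a lightning look, apparently converted from Cascade, that trails ribbons, sparks and energy sprites from character bones.
# Four emitters place particles through a shared skeletal mesh reader, using partitions or spawn groups to pick the bone or socket.
# Ribbons widen over life, sparks spawn by distance moved, and streak and distortion sprites re-sample bone positions every update.
# Front-end modifiers add rate-based sparks and ribbon velocity for non-gameplay display.
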NiagaraSystem /BRCosmetics/Athena/Effects/Particles/Skydiving/NS_Contrail_FlashLightning_Converted
dependencies {
 usedDataInterfaces [
  &NiagaraDataInterfaceCurve
  &NiagaraDataInterfaceVectorCurve
  &NiagaraDataInterfaceArrayFloat3
  &NiagaraDataInterfaceVector2DCurve
  &NiagaraDataInterfaceSkeletalMesh
 ]
 usedDynamicInputs []
 usedModules [
  &self:SystemSpawnScript.SystemScriptSource.SystemScriptGraph.NiagaraNodeAssignment_0.SetVariables_#1
  &SystemState
 ]
}
emitters {
 Blue_Sparks_Hand_L {
  emitterUpdateScript [
   &EmitterState(
    Inactive Response=::NewEnumerator1
    Life Cycle Mode=::NewEnumerator1
    Loop Duration=&RandomRangeFloat(Minimum=0.7, Maximum=2.7, Evaluation Type=::NewEnumerator1)
    Recalculate Duration Each Loop=true
    UseLoopDelay=true
   )
   &SpawnRate(SpawnRate=&FN_FrontEndFloatModifier(Gameplay=0, Front End=30), Use Spawn Probability=false)
   &SpawnPerUnit(
    Module.Use Max Movement Threshold=true
    Module.Use Movement Tolerance=true
    Movement Tolerance=0.1
    Spawn Spacing:NiagaraFloat=&Divide_Float(A=1, B=&Divide_Float(A=1, B=50))
    Use Spawn Probability=false
   )
  ]
  particleSpawnScript [
   &PartitionParticles()
   &InitializeParticle(
    Lifetime=&RandomRangeFloat(Minimum=0.2, Maximum=1, Evaluation Type=::NewEnumerator1)
    Sprite Size=&RandomRangeVector2D(Minimum=<1 5>, Maximum=<2 10>, Evaluation Type=::NewEnumerator1)
    Sprite Size Mode=::NewEnumerator4
    UsePositionOffset=false
    Write Material Random=true
   )
   &SkeletalMeshLocation(
    Bone / Socket Index=&MergeHelper_IntSwitch(0=0, 1=1, 2=?, 3=?, 4=?, 5=?, Index=$Particles.PartitionParticles.Partition, Num=::NewEnumerator0)
    Bone Sampling Mode=::NewEnumerator4
    MaskSpawnGroup=false
    Orientation Sampling=::NewEnumerator1
    Skeletal Mesh=$System.Skeletal Mesh Reader
   )
   &ShapeLocation(MaskSpawnGroup=false, Override Seed=false, Sphere Radius=10, UseOwnerScale=true)
   /CascadeToNiagaraConverter/NiagaraScripts/ModuleScripts/FixShapeVector()
   &AddVelocity(Velocity=&Multiply_VectorByFloat(Vector=$Output.FixShapeVector.ShapeVector, Float=2))
   &AddVelocity(Velocity=&RandomRangeVector(Minimum=<-200 -200 550>, Maximum=<200 200 800>, Evaluation Type=::NewEnumerator1)) {module=AddVelocity001}
   &Color(Color=&MakeLinearColorFromVectorAndFloat(
     "Vector (RGB)"=&SelectVectorFromArray(
      Array Sampling Mode=::NewEnumerator0
      Direct Array Index=?
      Direct Array Mode=::NewEnumerator1
      Vector Selection Array=DataInterface() {propertyValues {FloatData [<1.8016 0 50>, <1.9495 2 0>]}}
      Random Seed=0
     )
     "Float (Alpha)"=1
    ))
  ]
  particleUpdateScript [
   &ParticleState()
   &ScaleSpriteSize(
    Scale Factor=&Vector2DFromCurve(Vector2Curve=DataInterface() {propertyValues {XCurve {keys [<at 0 1>,<at 1 0>]}, YCurve {keys [<at 0 1>,<at 1 0>]}}}, ScaleCurve=<1 1>)
    Scale Sprite Size Mode=::NewEnumerator2
   )
   &AccelerationForce(Acceleration=<0 0 1250>)
   &SolveForcesAndVelocity(Clamp Velocity=false, Limit Acceleration=false)
  ]
  rendererClasses [&NiagaraSpriteRendererProperties]
  renderers [{rendererClass=&NiagaraSpriteRendererProperties}]
 }
 EnergyDistortion_Hand_L {
  emitterUpdateScript [
   &EmitterState(Inactive Response=::NewEnumerator1, Life Cycle Mode=::NewEnumerator1, Loop Duration=1, UseLoopDelay=true)
   &SpawnRate(SpawnGroup=1, SpawnRate=3, Use Spawn Probability=false) {enabled=false, module=SpawnRate001}
   &SpawnRate(
    SpawnRate=&RandomRangeFloat(Minimum=8, Maximum=4, Evaluation Type=::NewEnumerator1)
    Use Spawn Probability=false
   )
  ]
  particleSpawnScript [
   /SetVariables(Particles.VisibilityTag=&SpawnGroup()) {enabled=false, isSetParameters=true}
   &PartitionParticles(Number Of Partitions=4)
   &InitializeParticle(
    Lifetime=&RandomRangeFloat(Minimum=1, Maximum=1.5, Evaluation Type=::NewEnumerator1)
    Sprite Size=<30 30>
    Sprite Size Mode=::NewEnumerator4
    UsePositionOffset=false
    Write Material Random=true
   )
   &Color(Color=&MakeLinearColorFromVectorAndFloat(
     "Vector (RGB)"=<0 0 1>
     "Float (Alpha)"=&MergeHelper_FloatSwitch(0=1, 1=1, 2=0.1, 3=0.1, Index=$Particles.PartitionParticles.Partition, Num=::NewEnumerator2)
    ))
  ]
  particleUpdateScript [
   &ParticleState()
   &CameraOffset(Camera Offset Amount=20)
   &SkeletalMeshLocation(
    Bone / Socket Index=&MergeHelper_IntSwitch(0=0, 1=1, 2=2, 3=3, 4=?, 5=?, Index=$Particles.PartitionParticles.Partition, Num=::NewEnumerator2)
    Bone Sampling Mode=::NewEnumerator4
    MaskSpawnGroup=false
    Orientation Sampling=::NewEnumerator1
    Skeletal Mesh=$System.Skeletal Mesh Reader
   )
   &ScaleColor(
    Scale Alpha=&FloatFromCurve(FloatCurve=DataInterface() {propertyValues {Curve {keys [<at 0 0>,<at 0.5 1>,<at 1 0>]}}})
    Scale RGB=<10 10 10>
    ScaleA=true
    ScaleRGB=true
   )
   &Light_Attributes(
    Color=1
    Falloff Exponent=2
    Light Exponent=32
    Scale Alpha:NiagaraFloat=&FloatFromCurve(FloatCurve=DataInterface() {propertyValues {Curve {keys [<at 0 0 RCIM_Cubic>,<at 0.4212 0.9946 RCIM_Cubic>,<at 1 0 RCIM_Cubic>]}}})
    Scale RGB:Vector3f=&VectorFromCurve(VectorCurve=DataInterface() {propertyValues {
       XCurve {keys [<at 0 0 RCIM_Cubic>,<at 0.25 1 RCIM_Cubic>,<at 0.5 0 RCIM_Cubic>,<at 0.75 1 RCIM_Cubic>,<at 1 0>]}
       YCurve {keys [<at 0 0 RCIM_Cubic>,<at 0.25 1 RCIM_Cubic>,<at 0.5 0 RCIM_Cubic>,<at 0.75 1 RCIM_Cubic>,<at 1 0>]}
       ZCurve {keys [<at 0 1 RCIM_Cubic>,<at 0.25 0 RCIM_Cubic>,<at 0.5 1 RCIM_Cubic>,<at 0.75 0 RCIM_Cubic>,<at 1 1>]}
      }})
    ScaleA=true
    ScaleRGB=true
   ) {enabled=false}
  ]
  rendererClasses [&NiagaraSpriteRendererProperties, &NiagaraLightRendererProperties]
  renderers [
   {rendererClass=&NiagaraSpriteRendererProperties}
   {rendererClass=&NiagaraLightRendererProperties, rendererIndex=1}
  ]
 }
 Hand_Stk_Blue_L {
  emitterUpdateScript [
   &EmitterState(Inactive Response=::NewEnumerator1, Life Cycle Mode=::NewEnumerator1, Loop Duration=1, UseLoopDelay=true)
   &SpawnRate(
    SpawnGroup=3
    SpawnRate=&RandomRangeFloat(Minimum=4, Maximum=10, Evaluation Type=::NewEnumerator1)
    Use Spawn Probability=false
   ) {module=SpawnRate002}
   &SpawnRate(
    SpawnGroup=2
    SpawnRate=&RandomRangeFloat(Minimum=4, Maximum=10, Evaluation Type=::NewEnumerator1)
    Use Spawn Probability=false
   ) {module=SpawnRate001}
   &SpawnRate(
    SpawnGroup=1
    SpawnRate=&RandomRangeFloat(Minimum=20, Maximum=40, Evaluation Type=::NewEnumerator1)
    Use Spawn Probability=false
   )
   &SpawnRate(
    SpawnRate=&RandomRangeFloat(Minimum=20, Maximum=40, Evaluation Type=::NewEnumerator1)
    Use Spawn Probability=false
   ) {module=SpawnRate003}
  ]
  particleSpawnScript [
   /SetVariables(Particles.SpriteRotationRate=&RandomRangeFloat(Minimum=-2, Maximum=2, Evaluation Type=::NewEnumerator1)) {isSetParameters=true}
   &InitializeParticle(
    Lifetime=&RandomRangeFloat(Minimum=0.05, Maximum=0.2, Evaluation Type=::NewEnumerator1)
    Sprite Rotation Angle=&RandomRangeFloat(Minimum=0, Maximum=1, Evaluation Type=::NewEnumerator1)
    Sprite Rotation Mode=::NewEnumerator1
    Sprite Size=&Vector2DFromFloat(Value=&RandomRangeFloat(Minimum=15, Maximum=35, Evaluation Type=::NewEnumerator1))
    Sprite Size Mode=::NewEnumerator4
    UsePositionOffset=false
    Write Material Random=true
   )
   &Color(Color=&MakeLinearColorFromVectorAndFloat(
     "Vector (RGB)"=&SelectVectorFromArray(
      Array Sampling Mode=::NewEnumerator0
      Direct Array Index=?
      Vector Selection Array=DataInterface() {propertyValues {FloatData [<1 1 50>, <2 2 1>]}}
      Random Seed=0
     )
     "Float (Alpha)"=1
    ))
  ]
  particleUpdateScript [
   &ParticleState()
   &SpriteRotationRate(Rotation Rate=&AngleConversion(
     Angle:NiagaraFloat=$Particles.SpriteRotationRate
     Angle Input:ENiagara_AngleInput=::NewEnumerator1
     Angle Output:ENiagara_AngleInput=::NewEnumerator0
    ))
   &SkeletalMeshLocation(
    Bone / Socket Index=&MergeHelper_IntSwitch(0=0, 1=1, 2=2, 3=3, 4=?, 5=?, Index=&SpawnGroup(), Num=::NewEnumerator2)
    Bone Sampling Mode=::NewEnumerator4
    MaskSpawnGroup=false
    Orientation Sampling=::NewEnumerator1
    Skeletal Mesh=$System.Skeletal Mesh Reader
   )
   &CameraOffset(Camera Offset Amount=&FloatFromCurve(FloatCurve=DataInterface() {propertyValues {Curve {keys [<at 0 20>,<at 1 -20>]}}}))
  ]
  rendererClasses [&NiagaraSpriteRendererProperties]
  renderers [{rendererClass=&NiagaraSpriteRendererProperties}]
 }
 Ribbon_Left_Hand {
  emitterUpdateScript [
   &EmitterState(Inactive Response=::NewEnumerator1, Life Cycle Mode=::NewEnumerator1, Loop Duration=1, UseLoopDelay=true)
   &SpawnRate(SpawnGroup=1, SpawnRate=3, Use Spawn Probability=false) {enabled=false, module=SpawnRate001}
   &SpawnBurst_Instantaneous(Spawn Count=2, Use Spawn Probability=false, UseLoopCountLimit=false) {enabled=false}
   &SpawnRate(SpawnRate=40, Use Spawn Probability=false)
  ]
  particleSpawnScript [
   /SetVariables(Particles.VisibilityTag=&SpawnGroup()) {enabled=false, isSetParameters=true}
   &PartitionParticles()
   &InitializeParticle(
    Lifetime=0.3
    Ribbon Width=&Multiply_Float(A=10, B=2)
    Ribbon Width Mode=::NewEnumerator1
    UsePositionOffset=false
    Write Material Random=true
   )
   &SkeletalMeshLocation(
    Bone / Socket Index=&MergeHelper_IntSwitch(0=0, 1=1, 2=?, 3=?, 4=?, 5=?, Index=$Particles.PartitionParticles.Partition, Num=::NewEnumerator0)
    Bone Sampling Mode=::NewEnumerator4
    MaskSpawnGroup=true
    Orientation Sampling=::NewEnumerator1
    Skeletal Mesh=$System.Skeletal Mesh Reader
   )
   &ShapeLocation(MaskSpawnGroup=false, Override Seed=false, Sphere Radius=10, UseOwnerScale=true)
   &FixShapeVector()
   &Color(Color=&MakeLinearColorFromVectorAndFloat(
     "Vector (RGB)"=&RandomRangeVector(Minimum=<1 1 0.75>, Maximum=<0 0 2>, Evaluation Type=::NewEnumerator1)
     "Float (Alpha)"=1
    ))
   &AddVelocity(Velocity=&FN_FrontEndVector3Modifier(
     Gameplay=<0 0 0>
     Front End=&MaskVector(Mask Source=0, Mask Parameter=?, Mask=0, Masked Vector=<0 0 1000>, Passthrough Vector=<0 0 0>)
    ))
  ]
  particleUpdateScript [
   /SetVariables(Particles.ScaledRibbonWidth:NiagaraFloat=&Multiply_Float(
     A=$Particles.RibbonWidth
     B=&FloatFromCurve(FloatCurve=DataInterface() {propertyValues {Curve {keys [
          {arriveTangent=14.9114, interpMode=RCIM_Cubic, leaveTangent=14.9114, tangentMode=RCTM_User, time=-0.0035, value=-0.0003}
          {arriveTangent=10.1549, interpMode=RCIM_Cubic, leaveTangent=10.1549, time=0.0441, value=1.0328}
          {arriveTangent=3.0898, interpMode=RCIM_Cubic, leaveTangent=3.0898, time=0.2337, value=2.4085}
          {arriveTangent=0.4859, interpMode=RCIM_Cubic, leaveTangent=0.4859, tangentMode=RCTM_User, time=1, value=3.9862}
         ]}}})
    )) {isSetParameters=true}
   &ParticleState()
   &DynamicMaterialParameters(
    Index 0 Param 1=&RandomRangeFloat(Minimum=1, Maximum=3, Evaluation Type=::NewEnumerator1)
    Index 0 Param 2=&RandomRangeFloat(Minimum=0, Maximum=1, Evaluation Type=::NewEnumerator1)
    Index 0 Param 3=1
    Index 0 Param 4=1
    Module.Param0WriteEnabled=true
    Module.Param1WriteEnabled=true
    Module.Param2WriteEnabled=true
    Module.Param3WriteEnabled=true
   ) {module=DynamicMaterialParameters001}
   &SolveForcesAndVelocity(Clamp Velocity=false, Limit Acceleration=false)
   &SkeletalMeshLocation(
    Bone / Socket Index=&MergeHelper_IntSwitch(0=0, 1=1, 2=?, 3=?, 4=?, 5=?, Index=$Particles.PartitionParticles.Partition, Num=::NewEnumerator0)
    Bone Sampling Mode=::NewEnumerator4
    MaskSpawnGroup=true
    Orientation Sampling=::NewEnumerator1
    Position Sampling=::NewEnumerator2
    Skeletal Mesh=$System.Skeletal Mesh Reader
    Spawn Group Mask=1
   ) {enabled=false, module=SkeletalMeshLocation001}
   &ScaleColor(Scale RGB=<2 2 2>, ScaleA=true, ScaleRGB=true)
   &Light_Attributes(
    Color=1
    Falloff Exponent=2
    Light Exponent=32
    Scale Alpha:NiagaraFloat=&FloatFromCurve(FloatCurve=DataInterface() {propertyValues {Curve {keys [<at 0 0 RCIM_Cubic>,<at 0.4212 0.9946 RCIM_Cubic>,<at 1 0 RCIM_Cubic>]}}})
    Scale RGB:Vector3f=&VectorFromCurve(VectorCurve=DataInterface() {propertyValues {
       XCurve {keys [<at 0 0 RCIM_Cubic>,<at 0.25 1 RCIM_Cubic>,<at 0.5 0 RCIM_Cubic>,<at 0.75 1 RCIM_Cubic>,<at 1 0>]}
       YCurve {keys [<at 0 0 RCIM_Cubic>,<at 0.25 1 RCIM_Cubic>,<at 0.5 0 RCIM_Cubic>,<at 0.75 1 RCIM_Cubic>,<at 1 0>]}
       ZCurve {keys [<at 0 1 RCIM_Cubic>,<at 0.25 0 RCIM_Cubic>,<at 0.5 1 RCIM_Cubic>,<at 0.75 0 RCIM_Cubic>,<at 1 1>]}
      }})
    ScaleA=true
    ScaleRGB=true
   ) {enabled=false}
  ]
  rendererClasses [&NiagaraRibbonRendererProperties, &NiagaraLightRendererProperties]
  renderers [
   {rendererClass=&NiagaraRibbonRendererProperties}
   {rendererClass=&NiagaraLightRendererProperties, rendererIndex=1}
  ]
 }
}
summary {
 systemName=NS_Contrail_FlashLightning_Converted
 emitters [
  {
   emitterName=Ribbon_Left_Hand
   rendererClasses [&NiagaraRibbonRendererProperties, &NiagaraLightRendererProperties]
  }
  {
   emitterName=EnergyDistortion_Hand_L
   rendererClasses [&NiagaraSpriteRendererProperties, &NiagaraLightRendererProperties]
  }
  {emitterName=Hand_Stk_Blue_L, rendererClasses [&NiagaraSpriteRendererProperties]}
  {emitterName=Blue_Sparks_Hand_L, rendererClasses [&NiagaraSpriteRendererProperties]}
 ]
 userVariables [
  User.bIsFrontEnd {type=NiagaraBool}
  User.bIsFrontEndPreview {type=NiagaraBool}
 ]
}
