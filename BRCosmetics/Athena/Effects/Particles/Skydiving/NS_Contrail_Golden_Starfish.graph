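# Niagara skydiving contrail for the Golden Starfish cosmetic, trailing gold-toned effects from four bones or sockets on the user skeletal mesh.
# Combines four ribbons, smoke, pulsing and persistent glows, electric arcs and GPU sparkles, with Low variants of smoke and glow.
# In gameplay, smoke drifts against owner velocity and ribbons and sparkles get downward velocity.
# Front-end modifiers switch these to upward motion, with smoke and sparkles scaled by the system Dive value.
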NiagaraSystem /BRCosmetics/Athena/Effects/Particles/Skydiving/NS_Contrail_Golden_Starfish
dependencies {
 usedDataInterfaces [
  &NiagaraDataInterfaceCurve
  &NiagaraDataInterfaceColorCurve
  &NiagaraDataInterfaceVectorCurve
  &NiagaraDataInterfaceVector2DCurve
 ]
 usedDynamicInputs []
 usedModules [&SystemState, /Game/Effects/Niagara/Modules/Skydiving/FN_DivingAttributes/v5/FN_DivingAttributes]
}
emitters {
 ElectricArcs {
  emitterUpdateScript [
   &EmitterState()
   &SpawnRate(
    Spawn Probability=0.25
    SpawnRate=&RandomRangeFloat(Minimum=2, Maximum=1, Evaluation Type=::NewEnumerator1)
    Use Spawn Probability=true
   )
   &SpawnBurst_Instantaneous(Spawn Count=&RandomRangeInt(Minimum=0, Maximum=7), Use Spawn Probability=false)
  ]
  particleSpawnScript [
   &InitializeParticle(
    Lifetime=0.2
    Sprite Size Max=<10 20>
    Sprite Size Min=<15 40>
    Sprite Size Mode=::NewEnumerator2
    UsePositionOffset=false
    Write Material Random=true
   )
   &AddVelocity(Velocity=&RandomRangeVector(Minimum=<-5 -5 5>, Maximum=<5 5 30>, Evaluation Type=::NewEnumerator1))
   &Color(Color=&MakeLinearColorFromVectorAndFloat(
     "Vector (RGB)"=&RandomRangeVector(Minimum=<0 0 0>, Maximum=<4 4 4>, Evaluation Type=::NewEnumerator1)
     "Float (Alpha)"=&ClampFloat(Float=1)
    ))
   &SkeletalMeshLocation(
    Bone / Socket Index=&RandomRangeInt(Minimum=0, Maximum=3)
    Bone Sampling Mode=::NewEnumerator4
    Inherit Velocity=false
    MaskSpawnGroup=false
    Skeletal Mesh=$User.SkeletalMesh
   ) {module=SkeletalMeshLocation001}
   /Niagara/Modules/Spawn/Location/SphereLocation(MaskSpawnGroup=false, Module.Override Randomness=false, Sphere Radius=2, Surface Only=true)
   &AddVelocity(Velocity=&Multiply_VectorByFloat(Vector=$Output.SphereLocation.SphereVector, Float=35)) {module=AddVelocity001}
  ]
  particleUpdateScript [
   &ParticleState()
   &DynamicMaterialParameters(
    Index 0 Param 1=&RandomRangeFloat(Minimum=1, Maximum=5)
    Index 0 Param 2=&FloatFromCurve(FloatCurve=DataInterface() {propertyValues {Curve {keys [<at 0 1>,<at 1 0>]}}})
    Index 0 Param 3=&RandomRangeFloat(Minimum=0.5, Maximum=1)
    Index 0 Param 4=&RandomRangeFloat(Minimum=0, Maximum=1)
    Module.Param0WriteEnabled=true
    Module.Param1WriteEnabled=true
    Module.Param2WriteEnabled=true
    Module.Param3WriteEnabled=true
   ) {module=DynamicMaterialParameters001}
   &ScaleSpriteSize(Scale Factor=&Vector2DFromCurve(
     Vector2Curve=DataInterface() {propertyValues {XCurve {keys [<at 0 0>,<at 0.5 1>,<at 0.7 1>,<at 1 0>]}, YCurve {keys [<at 0 0>,<at 0.5 1>]}}}
     ScaleCurve=<1 1>
    ))
   &SolveForcesAndVelocity(Clamp Velocity=false, Limit Acceleration=false)
   &ScaleColor(
    Scale RGB=&VectorFromCurve(VectorCurve=DataInterface() {propertyValues {XCurve {keys [<at 0.3 16>,<at 1 0.25>]}, YCurve {keys [<at 0.3 8.4738>,<at 1 0.0597>]}, ZCurve {keys [<at 0.3 0.25>,<at 1 0>]}}})
    ScaleA=true
    ScaleRGB=true
   )
  ]
  rendererClasses [&NiagaraSpriteRendererProperties]
  renderers [{rendererClass=&NiagaraSpriteRendererProperties}]
 }
 FN_Contrails_Ribbon_v4 {
  emitterUpdateScript [
   &EmitterState()
   &FN_SpawnRibbon_v2(Number of Ribbons=4, Spawn Mode=::NewEnumerator0, SpawnRate=60)
  ]
  particleSpawnScript [
   &FN_InitializeRibbon_v3(Color:LinearColor=<rgba 4 4 4 1>, Enable Skeletal Position=true, Lifetime=0.3, Ribbon Width=30)
   /SetVariables(Particles.RibbonIndex:NiagaraInt32=&self:GetRibbonIndex()) {isSetParameters=true}
   &AddVelocity(Velocity=&FN_FrontEndVector3Modifier(Gameplay=<0 0 -80>, Front End=<0 0 1000>))
   &DynamicMaterialParameters(
    Index 0 Param 1=&RandomRangeFloat(Minimum=0, Maximum=1, Randomness Mode=::NewEnumerator1, Random Seed=$Particles.RibbonIndex, Fixed Random Seed=true)
    Index 0 Param 2=&RandomRangeFloat(Minimum=0, Maximum=1, Randomness Mode=::NewEnumerator1, Random Seed=$Particles.RibbonIndex, Fixed Random Seed=true)
    Index 0 Param 3=1
    Index 0 Param 4=1
    Module.Param0WriteEnabled=true
    Module.Param1WriteEnabled=true
    Module.Param2WriteEnabled=true
    Module.Param3WriteEnabled=true
   )
  ]
  particleUpdateScript [
   &SolveForcesAndVelocity(Clamp Velocity=false, Limit Acceleration=false)
   &FN_UpdateRibbon(
    Enable Skeletal Position=true
    Initial Color:LinearColor=$Particles.Initial.Color
    Initial RibbonWidth:NiagaraFloat=$Particles.Initial.RibbonWidth
   )
   &ScaleColor(
    Scale Alpha=&FloatFromCurve(FloatCurve=DataInterface() {propertyValues {Curve {keys [<at 0.1 0 RCIM_Cubic>,<at 0.3 1.3 RCIM_Cubic>,<at 1 0>]}}})
    Scale RGB=&VectorFromCurve(VectorCurve=DataInterface() {propertyValues {
       XCurve {keys [<at 0 12 RCIM_Cubic>,<at 0.4 1 RCIM_Cubic>]}
       YCurve {keys [<at 0 12 RCIM_Cubic>,<at 0.4 1 RCIM_Cubic>]}
       ZCurve {keys [<at 0 12 RCIM_Cubic>,<at 0.4 1 RCIM_Cubic>]}
      }})
    ScaleA=true
    ScaleRGB=true
   )
   &ScaleRibbonWidth(Ribbon Width Scale=&FloatFromCurve(FloatCurve=DataInterface() {propertyValues {Curve {keys [
         <at -0.0035 1 RCIM_Cubic>
         {arriveTangent=0.5985, interpMode=RCIM_Cubic, leaveTangent=0.5985, tangentMode=RCTM_User, time=0.1014, value=1.2046}
         <at 0.9769 1.4197 RCIM_Cubic>
        ]}}}))
  ]
  rendererClasses [&NiagaraRibbonRendererProperties]
  renderers [{rendererClass=&NiagaraRibbonRendererProperties}]
 }
 Glow {
  emitterUpdateScript [
   &EmitterState()
   &SpawnRate(SpawnRate=10, Use Spawn Probability=false)
   &SpawnBurst_Instantaneous(Spawn Count=&RandomRangeInt(Minimum=0, Maximum=2), Use Spawn Probability=false)
  ]
  particleSpawnScript [
   /Niagara/Modules/Spawn/Initialization/InitializeParticle(
    Module.Write Color=true
    Module.Write Lifetime=true
    Module.Write Mass=true
    Module.Write Material Random=true
    Module.Write Position=false
    Module.Write Scale=false
    Module.Write SpriteRotation=true
    Module.Write SpriteSize=true
    Sprite Rotation=&RandomRangeFloat(Minimum=0, Maximum=360)
    Sprite Size=<90 90>
   )
   &Color(Color=&MakeLinearColorFromVectorAndFloat(
     "Vector (RGB)"=&RandomRangeVector(Minimum=<2 0.6785 0.1741>, Maximum=<1 1 1>, Evaluation Type=::NewEnumerator1)
     "Float (Alpha)"=&ClampFloat(Float=0.5)
    ))
   &DynamicMaterialParameters(
    Index 0 Param 1=500
    Module.Param0WriteEnabled=true
    Module.Param1WriteEnabled=true
    Module.Param2WriteEnabled=true
    Module.Param3WriteEnabled=true
   )
  ]
  particleUpdateScript [
   &ParticleState()
   &CameraOffset(Camera Offset Amount=&RandomRangeFloat(Minimum=10, Maximum=40))
   &ScaleColor(
    Scale Alpha=&FloatFromCurve(FloatCurve=DataInterface() {propertyValues {Curve {keys [
         <at 0 0 RCIM_Cubic>
         {arriveTangent=0.9286, interpMode=RCIM_Cubic, leaveTangent=0.9286, time=0.2321, value=0.9177}
         {arriveTangent=-1.2352, interpMode=RCIM_Cubic, leaveTangent=-1.2352, tangentMode=RCTM_User, time=0.7, value=0.65}
         <at 1 0>
        ]}}})
    Scale RGB=&VectorFromCurve(VectorCurve=DataInterface() {propertyValues {
       XCurve {keys [<at 0 1.2 RCIM_Cubic>,<at 0.8939 0.5473 RCIM_Cubic>]}
       YCurve {keys [<at 0 0.5735 RCIM_Cubic>,<at 0.8939 0.0179 RCIM_Cubic>]}
       ZCurve {keys [<at 0 0.1427 RCIM_Cubic>,<at 0.8939 0 RCIM_Cubic>]}
      }})
    ScaleA=true
    ScaleRGB=true
   )
   &ScaleSpriteSize(Scale Factor=&Vector2DFromFloat(Value=&FloatFromCurve(FloatCurve=DataInterface() {propertyValues {Curve {keys [<at 0 0.6>,<at 1 1.8>]}}})))
   &SkeletalMeshLocation(
    Bone / Socket Index=&ModuloInt(A=&ReturnParticleIDAttributes(), B=4)
    Bone Sampling Mode=::NewEnumerator4
    Inherit Velocity=false
    Interpolate Mesh Sampling=false
    MaskSpawnGroup=false
    Skeletal Mesh=$User.SkeletalMesh
   ) {module=SkeletalMeshLocation001}
  ]
  rendererClasses [&NiagaraSpriteRendererProperties]
  renderers [{rendererClass=&NiagaraSpriteRendererProperties}]
 }
 GlowLow {
  emitterUpdateScript [
   &EmitterState()
   &SpawnBurst_Instantaneous(
    Spawn Count=4
    Spawn Probability=&MakeCustomFloatFromBool(
     Bool=&SetBoolByIntComparison(A=$Emitter.LoopCount, B=0, Comparison Type=::NewEnumerator4)
     True Float=1
     False Float=0
    )
    Use Spawn Probability=true
   )
  ]
  particleSpawnScript [
   &InitializeParticle(
    Module.Write Color=true
    Module.Write Lifetime=true
    Module.Write Mass=true
    Module.Write Material Random=true
    Module.Write Position=false
    Module.Write Scale=false
    Module.Write SpriteRotation=true
    Module.Write SpriteSize=true
    Sprite Rotation=&RandomRangeFloat(Minimum=0, Maximum=360)
    Sprite Size=<90 90>
   )
   &Color(Color=&MakeLinearColorFromVectorAndFloat(
     "Vector (RGB)"=&Multiply_Vector(
      A=&RandomRangeVector(Minimum=<2 0.6785 0.1741>, Maximum=<1 1 1>, Evaluation Type=::NewEnumerator1)
      B=<1.2 0.57 0.14>
     )
     "Float (Alpha)"=0.25
    ))
   &DynamicMaterialParameters(
    Index 0 Param 1=500
    Module.Param0WriteEnabled=true
    Module.Param1WriteEnabled=true
    Module.Param2WriteEnabled=true
    Module.Param3WriteEnabled=true
   )
   /SetVariables(Particles.ExecIndex:NiagaraInt32=&ReturnExecIndex()) {isSetParameters=true}
   &CameraOffset(Camera Offset Amount=40)
  ]
  particleUpdateScript [
   &ParticleState(Kill Particles When Lifetime Has Elapsed=false, Let Infinitely Lived Particles Die When Emitter Deactivates=true)
   &SkeletalMeshLocation(
    Bone / Socket Index=$Particles.ExecIndex
    Bone Sampling Mode=::NewEnumerator4
    Inherit Velocity=false
    Interpolate Mesh Sampling=false
    MaskSpawnGroup=false
    Skeletal Mesh=$User.SkeletalMesh
   ) {module=SkeletalMeshLocation001}
  ]
  rendererClasses [&NiagaraSpriteRendererProperties]
  renderers [{rendererClass=&NiagaraSpriteRendererProperties}]
 }
 Smoke {
  emitterUpdateScript [
   &EmitterState()
   &SpawnRate(
    SpawnRate=&FN_FrontEndFloatModifier(Gameplay=75, Front End=&Lerp_Float(A=75, B=200, Alpha=$System.Dive))
    Use Spawn Probability=false
   )
  ]
  particleSpawnScript [
   &InitializeParticle(
    Lifetime=&FN_FrontEndFloatModifier(
     Gameplay=&RandomRangeFloat(Minimum=0.5, Maximum=1, Evaluation Type=::NewEnumerator1)
     Front End=&Lerp_Float(
      A=&RandomRangeFloat(Minimum=0.5, Maximum=1, Evaluation Type=::NewEnumerator1)
      B=&RandomRangeFloat(Minimum=0.25, Maximum=0.45, Evaluation Type=::NewEnumerator1)
      Alpha=$System.Dive
     )
    )
    Module.Write Color=true
    Module.Write Lifetime=true
    Module.Write Mass=true
    Module.Write Material Random=true
    Module.Write Position=false
    Module.Write Scale=false
    Module.Write SpriteRotation=true
    Module.Write SpriteSize=true
    Sprite Rotation=&NormalizedAngleToDegrees(NormalizedAngle=&RandomRangeFloat(Minimum=-1, Maximum=1, Evaluation Type=::NewEnumerator1))
    Sprite Size=&Vector2DFromFloat(Value=&RandomRangeFloat(Minimum=30, Maximum=60))
   )
   &Color(Color=&MakeLinearColorFromVectorAndFloat("Vector (RGB)"=<1 0.8353 0.6244>, "Float (Alpha)"=&ClampFloat(Float=0.5)))
   /Niagara/Modules/Update/SubUV/SubUVAnimation(SubUV Animation Mode=::NewEnumerator1)
   /SetVariables(
    Transient.FrontDotLerp:Vector3f=&Lerp_Vector(A=<0 0 0>, B=<0 1 0>, Alpha=$User.ForwardDot)
    Transient.RightDotLerp:Vector3f=&Lerp_Vector(A=<0 0 0>, B=<1 0 0>, Alpha=$User.RightDot)
   ) {enabled=false, isSetParameters=true}
   /SetVariables(Particles.DirectionalVector:Vector3f=&RotateVectorByQuaternion(
     Vector To Rotate=&Multiply_Vector(
      A=&Add_Vector(A=$Transient.FrontDotLerp, B=$Transient.RightDotLerp)
      B=&RandomRangeVector(Minimum=<1 1 0>, Maximum=<1000 1000 0>)
     )
     Quaternion=$Engine.Owner.Rotation
    )) {enabled=false, isSetParameters=true}
   /SetVariables(Particles.DirectionalVector:Vector3f=&FN_FrontEndVector3Modifier(
     Gameplay=/Niagara/DynamicInputs/Math/Clamp_Vector(
      Vector:Vector3f=&Multiply_Vector(A=$Engine.Owner.Velocity, B=<-0.1 -0.1 -0.1>)
      Min:Vector3f=<-100 -100 -5>
      Max:Vector3f=<100 100 200>
     )
     Front End=&Lerp_Vector(
      A=&RandomRangeVector(Minimum=<-20 -20 75>, Maximum=<20 20 150>)
      B=&RandomRangeVector(Minimum=<-20 -20 100>, Maximum=<20 20 250>)
      Alpha=$System.Dive
     )
    )) {isSetParameters=true, module=SetVariables_#1}
  ]
  particleUpdateScript [
   &ParticleState()
   &CameraOffset(Camera Offset Amount=-25) {enabled=false}
   &ScaleColor(
    Scale Alpha=&FloatFromCurve(FloatCurve=DataInterface() {propertyValues {Curve {keys [
         <at 0 0 RCIM_Cubic>
         {arriveTangent=-2.4588, interpMode=RCIM_Cubic, leaveTangent=-2.4588, tangentMode=RCTM_User, time=0.1085, value=0.7977}
         <at 1 0>
        ]}}})
    ScaleA=true
    ScaleRGB=false
   )
   &ScaleSpriteSize(Scale Factor=&Vector2DFromFloat(Value=&FloatFromCurve(FloatCurve=DataInterface() {propertyValues {Curve {keys [
          {arriveTangent=2.3611, interpMode=RCIM_Cubic, leaveTangent=2.3611, tangentMode=RCTM_User, time=0, value=0.1}
          {interpMode=RCIM_Cubic, tangentMode=RCTM_User, time=1, value=1}
         ]}}})))
   &DynamicMaterialParameters(
    Index 0 Param 1=&FloatFromCurve(FloatCurve=DataInterface() {propertyValues {Curve {keys [{interpMode=RCIM_Cubic, tangentMode=RCTM_User, time=0, value=0.2}, <at 1 5>]}}})
    Index 0 Param 2=0.05
    Module.Param0WriteEnabled=true
    Module.Param1WriteEnabled=true
    Module.Param2WriteEnabled=true
    Module.Param3WriteEnabled=true
   )
   &SpriteRotationRate(Rotation Rate=&NormalizedAngleToDegrees(NormalizedAngle=&RandomRangeFloat(Minimum=-0.2, Maximum=0.2)))
   &SkeletalMeshLocation(
    Bone / Socket Index=&RandomRangeInt(Minimum=0, Maximum=3)
    Bone Sampling Mode=::NewEnumerator4
    Inherit Velocity=false
    Interpolate Mesh Sampling=false
    MaskSpawnGroup=false
    Sampled Position Offset=&Lerp_Vector(A=<0 0 0>, B=$Particles.DirectionalVector, Alpha=&Power(a=$Particles.NormalizedAge, b=1.3))
    Sampled Position Offset Coordinate Space:ENiagaraMeshTransforms=::NewEnumerator1
    Skeletal Mesh=$User.SkeletalMesh
   )
   &Color(Color=&ColorFromCurve(
     ColorCurve=DataInterface() {propertyValues {
       AlphaCurve {keys [<at 0 1 RCIM_Cubic>]}
       BlueCurve {keys [<at 0 0 RCIM_Cubic>,<at 0.4401 0>,<at 1 1 RCIM_Cubic>]}
       GreenCurve {keys [<at 0 0 RCIM_Cubic>,<at 0.4401 1>,<at 1 0 RCIM_Cubic>]}
       RedCurve {keys [<at 0 1 RCIM_Cubic>,<at 0.4401 0>,<at 1 0 RCIM_Cubic>]}
      }}
     CurveIndex=$Particles.Dive
    )) {enabled=false, module=Color001}
  ]
  rendererClasses [&NiagaraSpriteRendererProperties]
  renderers [{rendererClass=&NiagaraSpriteRendererProperties}]
 }
 SmokeLow {
  emitterUpdateScript [
   &EmitterState()
   &SpawnRate(
    SpawnRate=&FN_FrontEndFloatModifier(Gameplay=25, Front End=&Lerp_Float(A=25, B=50, Alpha=$System.Dive))
    Use Spawn Probability=false
   )
  ]
  particleSpawnScript [
   &InitializeParticle(
    Lifetime=&FN_FrontEndFloatModifier(
     Gameplay=&RandomRangeFloat(Minimum=0.5, Maximum=1, Evaluation Type=::NewEnumerator1)
     Front End=&Lerp_Float(
      A=&RandomRangeFloat(Minimum=0.5, Maximum=1, Evaluation Type=::NewEnumerator1)
      B=&RandomRangeFloat(Minimum=0.25, Maximum=0.45, Evaluation Type=::NewEnumerator1)
      Alpha=$System.Dive
     )
    )
    Module.Write Color=true
    Module.Write Lifetime=true
    Module.Write Mass=true
    Module.Write Material Random=true
    Module.Write Position=false
    Module.Write Scale=false
    Module.Write SpriteRotation=true
    Module.Write SpriteSize=true
    Sprite Rotation=&NormalizedAngleToDegrees(NormalizedAngle=&RandomRangeFloat(Minimum=-1, Maximum=1, Evaluation Type=::NewEnumerator1))
    Sprite Size=&Vector2DFromFloat(Value=&RandomRangeFloat(Minimum=30, Maximum=60))
   )
   &Color(Color=&MakeLinearColorFromVectorAndFloat("Vector (RGB)"=<1 0.8353 0.6244>, "Float (Alpha)"=&ClampFloat(Float=0.5)))
   &SubUVAnimation(SubUV Animation Mode=::NewEnumerator1)
   /SetVariables(Particles.DirectionalVector:Vector3f=&FN_FrontEndVector3Modifier(
     Gameplay=&Clamp_Vector(
      Vector:Vector3f=&Multiply_Vector(A=$Engine.Owner.Velocity, B=<-0.051 -0.05 -0.05>)
      Min:Vector3f=<-50 -50 -5>
      Max:Vector3f=<50 50 100>
     )
     Front End=&Lerp_Vector(
      A=&RandomRangeVector(Minimum=<-10 -10 50>, Maximum=<10 10 75>)
      B=&RandomRangeVector(Minimum=<-10 -10 50>, Maximum=<10 10 125>)
      Alpha=$System.Dive
     )
    )) {isSetParameters=true, module=SetVariables_#1}
  ]
  particleUpdateScript [
   &ParticleState()
   &CameraOffset(Camera Offset Amount=-25) {enabled=false}
   &ScaleColor(
    Scale Alpha=&FloatFromCurve(FloatCurve=DataInterface() {propertyValues {Curve {keys [
         <at 0 0 RCIM_Cubic>
         {arriveTangent=-2.4588, interpMode=RCIM_Cubic, leaveTangent=-2.4588, tangentMode=RCTM_User, time=0.1085, value=0.7977}
         <at 1 0>
        ]}}})
    ScaleA=true
    ScaleRGB=false
   )
   &ScaleSpriteSize(Scale Factor=&Vector2DFromFloat(Value=&FloatFromCurve(FloatCurve=DataInterface() {propertyValues {Curve {keys [
          {arriveTangent=2.3611, interpMode=RCIM_Cubic, leaveTangent=2.3611, tangentMode=RCTM_User, time=0, value=0.1}
          {interpMode=RCIM_Cubic, tangentMode=RCTM_User, time=1, value=1}
         ]}}})))
   &DynamicMaterialParameters(
    Index 0 Param 1=&FloatFromCurve(FloatCurve=DataInterface() {propertyValues {Curve {keys [{interpMode=RCIM_Cubic, tangentMode=RCTM_User, time=0, value=0.2}, <at 1 5>]}}})
    Index 0 Param 2=0.05
    Module.Param0WriteEnabled=true
    Module.Param1WriteEnabled=true
    Module.Param2WriteEnabled=true
    Module.Param3WriteEnabled=true
   )
   &SpriteRotationRate(Rotation Rate=&NormalizedAngleToDegrees(NormalizedAngle=&RandomRangeFloat(Minimum=-0.2, Maximum=0.2)))
   &SkeletalMeshLocation(
    Bone / Socket Index=&ModuloInt(A=&ReturnParticleIDAttributes(), B=4)
    Bone Sampling Mode=::NewEnumerator4
    Inherit Velocity=false
    Interpolate Mesh Sampling=false
    MaskSpawnGroup=false
    Sampled Position Offset=&Lerp_Vector(A=<0 0 0>, B=$Particles.DirectionalVector, Alpha=&Power(a=$Particles.NormalizedAge, b=1.3))
    Sampled Position Offset Coordinate Space:ENiagaraMeshTransforms=::NewEnumerator1
    Skeletal Mesh=$User.SkeletalMesh
   )
  ]
  rendererClasses [&NiagaraSpriteRendererProperties]
  renderers [{rendererClass=&NiagaraSpriteRendererProperties}]
 }
 Sparkles {
  simTarget=GPUComputeSim
  emitterUpdateScript [
   &EmitterState()
   &SpawnRate(SpawnRate=100, Use Spawn Probability=false)
  ]
  particleSpawnScript [
   &InitializeParticle(
    Lifetime=&RandomRangeFloat(Minimum=0.3, Maximum=0.6, Evaluation Type=::NewEnumerator1)
    Module.Write Color=true
    Module.Write Lifetime=true
    Module.Write Mass=true
    Module.Write Material Random=true
    Module.Write Position=false
    Module.Write Scale=false
    Module.Write SpriteRotation=true
    Module.Write SpriteSize=true
    Sprite Rotation=&RandomRangeFloat(Minimum=4, Maximum=360)
    Sprite Size=&RandomRangeVector2D(Minimum=<2 2>, Maximum=<10 10>, Evaluation Type=::NewEnumerator1)
   )
   &Color(Color=&MakeLinearColorFromVectorAndFloat("Vector (RGB)"=<1 0.3669 0>, "Float (Alpha)"=&ClampFloat(Float=1)))
   &SkeletalMeshLocation(
    Bone / Socket Index=&RandomRangeInt(Minimum=0, Maximum=3)
    Bone Sampling Mode=::NewEnumerator4
    Inherit Velocity=false
    MaskSpawnGroup=false
    Sampled Position Offset=&RandomVector(Vector Scale=10)
    Skeletal Mesh=$User.SkeletalMesh
   )
   &AddVelocity(Velocity=&Multiply_VectorByFloat(
     Vector=&FN_FrontEndVector3Modifier(Gameplay=<0 0 -80>, Front End=&Lerp_Vector(A=<0 0 300>, B=<0 0 750>, Alpha=$System.Dive))
     Float=&RandomRangeFloat(Minimum=0.75, Maximum=1.2)
    ))
   &InheritVelocity(
    Module.Limit Speed=false
    Velocity Limit=0
    Velocity Scale=&RandomRangeVector(Minimum=<0 0 0>, Maximum=<1 1 1>, Evaluation Type=::NewEnumerator1)
   )
  ]
  particleUpdateScript [
   &ParticleState()
   &ScaleColor(
    Scale Alpha=&FloatFromCurve(FloatCurve=DataInterface() {propertyValues {Curve {keys [
         {arriveTangent=9.9173, interpMode=RCIM_Cubic, time=0, value=3}
         {arriveTangent=-4.2155, interpMode=RCIM_Cubic, leaveTangent=-4.2155, time=0.3, value=1}
         {arriveTangent=-1.4286, interpMode=RCIM_Cubic, leaveTangent=-1.4286, time=0.4562, value=1.0768}
         <at 1 0 RCIM_Cubic>
        ]}}})
    Scale RGB=&VectorFromCurve(VectorCurve=DataInterface() {propertyValues {
       XCurve {keys [<at 0 32>,<at 0.3 1.5>,<at 1 0>]}
       YCurve {keys [<at 0 32>,<at 0.3 1.5>,<at 1 0>]}
       ZCurve {keys [<at 0 32>,<at 0.3 1.5>,<at 1 0>]}
      }})
    ScaleA=true
    ScaleRGB=true
   )
   &ScaleSpriteSize(Scale Factor=&Vector2DFromFloat(Value=&FloatFromCurve(FloatCurve=DataInterface() {propertyValues {Curve {keys [
          <at 0 0 RCIM_Cubic>
          {arriveTangent=3.8762, interpMode=RCIM_Cubic, leaveTangent=3.8762, time=0.1034, value=1.1346}
          {arriveTangent=-0.4418, interpMode=RCIM_Cubic, leaveTangent=-0.4418, time=0.2227, value=0.8631}
          {arriveTangent=0.4021, interpMode=RCIM_Cubic, leaveTangent=0.4021, time=0.2988, value=1.0483}
          {arriveTangent=0.1455, interpMode=RCIM_Cubic, leaveTangent=0.1455, time=0.5721, value=1.0037}
          {arriveTangent=-3.2827, interpMode=RCIM_Cubic, leaveTangent=-3.2827, time=0.6432, value=1.0984}
          {arriveTangent=-3.0783, interpMode=RCIM_Cubic, leaveTangent=-3.0783, time=0.8021, value=0.2488}
          <at 1 0 RCIM_Cubic>
         ]}}})))
   &SolveForcesAndVelocity(Clamp Velocity=false, Limit Acceleration=false)
  ]
  rendererClasses [&NiagaraSpriteRendererProperties]
  renderers [{rendererClass=&NiagaraSpriteRendererProperties}]
 }
}
summary {
 systemName=NS_Contrail_Golden_Starfish
 emitters [
  {emitterName=Smoke, rendererClasses [&NiagaraSpriteRendererProperties]}
  {emitterName=Sparkles, simTarget=GPUComputeSim, rendererClasses [&NiagaraSpriteRendererProperties]}
  {emitterName=Glow, rendererClasses [&NiagaraSpriteRendererProperties]}
  {emitterName=ElectricArcs, rendererClasses [&NiagaraSpriteRendererProperties]}
  {emitterName=FN_Contrails_Ribbon_v4, rendererClasses [&NiagaraRibbonRendererProperties]}
  {emitterName=GlowLow, rendererClasses [&NiagaraSpriteRendererProperties]}
  {emitterName=SmokeLow, rendererClasses [&NiagaraSpriteRendererProperties]}
 ]
 userVariables [
  User.bIsFrontEnd {type=NiagaraBool}
  User.bIsFrontEndPreview {type=NiagaraBool}
  User.ForwardDot {type=NiagaraFloat, defaultValue=0, description=""}
  User.ParaGlideLeanAlpha {type=NiagaraFloat, defaultValue=0, description=""}
  User.RibbonIDOffset {type=NiagaraInt32}
  User.RightDot {type=NiagaraFloat, defaultValue=0, description=""}
  User.SkeletalMesh {
   type=NiagaraDataInterfaceSkeletalMesh
   defaultValue {
    struct=NiagaraExt_VariableValue_DataInterface
    value {dataInterface=&self:NiagaraDataInterfaceSkeletalMesh_1}
   }
  }
 ]
}
