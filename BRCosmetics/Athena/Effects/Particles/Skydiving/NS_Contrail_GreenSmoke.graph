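# Niagara skydiving contrail cosmetic (GreenSmoke) that trails smoke, ribbons and glow from the character's skeletal mesh.
# Emitters sample bones or sockets via a shared mesh reader: looping, mirrored counter-rotating flipbook puffs, short-lived ribbons, pale teal glow sprites and lights.
# Distance-based spawning adds velocity-inheriting puffs, one layer stretched by speed.
# Front-end modifiers swap velocities on the ribbons and idle smoke to simulate motion in menus.
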NiagaraSystem /BRCosmetics/Athena/Effects/Particles/Skydiving/NS_Contrail_GreenSmoke
dependencies {
 usedDataInterfaces [
  &NiagaraDataInterfaceCurve
  &NiagaraDataInterfaceSpriteRendererInfo
  &NiagaraDataInterfaceVector2DCurve
  &NiagaraDataInterfaceArrayFloat3
  &NiagaraDataInterfaceSkeletalMesh
 ]
 usedDynamicInputs []
 usedModules [
  &self:SystemSpawnScript.SystemScriptSource.SystemScriptGraph.NiagaraNodeAssignment_0.SetVariables_#1
  &SystemState
 ]
}
emitters {
 Darksmoke_hand_l001 {
  emitterUpdateScript [
   &EmitterState(Inactive Response=::NewEnumerator1, Life Cycle Mode=::NewEnumerator1, Loop Duration=1, UseLoopDelay=true)
   &SpawnRate(SpawnRate=240, Use Spawn Probability=false)
  ]
  particleSpawnScript [
   &InitializeParticle(
    Lifetime=0.08
    Ribbon Width=&Multiply_Float(A=8, B=2)
    Ribbon Width Mode=::NewEnumerator1
    UsePositionOffset=false
    Write Material Random=true
   )
   &AddVelocity(Velocity=&FN_FrontEndVector3Modifier(Gameplay=<0 0 -80>, Front End=<0 -200 1100>))
   &SkeletalMeshLocation(
    Bone / Socket Index=/CascadeToNiagaraConverter/NiagaraScripts/DynamicInputScripts/SequentialInt(Use Array Data:NiagaraBool=false, First Number:NiagaraInt32=2, Sequence:NiagaraDataInterfaceArrayInt32=?, Last Number:NiagaraInt32=3)
    Bone Sampling Mode=::NewEnumerator4
    MaskSpawnGroup=false
    Orientation Sampling=::NewEnumerator1
    Skeletal Mesh=$System.Skeletal Mesh Reader
   )
   /CascadeToNiagaraConverter/NiagaraScripts/ModuleScripts/AddVectorToPosition(Vector:Vector3f=&TransformVector(Vector=<0 0 -4>, Destination Space=::Simulation))
   &Color(Color=&MakeLinearColorFromVectorAndFloat("Vector (RGB)"=<1 1 1>, "Float (Alpha)"=4))
  ]
  particleUpdateScript [
   &ParticleState()
   &SolveForcesAndVelocity(Clamp Velocity=false, Limit Acceleration=false)
   &CameraOffset(Camera Offset Amount=-20)
  ]
  rendererClasses [&NiagaraRibbonRendererProperties]
  renderers [{rendererClass=&NiagaraRibbonRendererProperties}]
 }
 Darksmoke_hand_r {
  emitterUpdateScript [
   &EmitterState(Inactive Response=::NewEnumerator1, Life Cycle Mode=::NewEnumerator1, Loop Duration=1, UseLoopDelay=true)
   &SpawnRate(SpawnRate=240, Use Spawn Probability=false)
  ]
  particleSpawnScript [
   /SetVariables(Particles.SkeletonSamplingIndex:NiagaraInt32=&SequentialInt(Use Array Data:NiagaraBool=false, First Number:NiagaraInt32=0, Sequence:NiagaraDataInterfaceArrayInt32=?, Last Number:NiagaraInt32=1)) {isSetParameters=true}
   &InitializeParticle(
    Lifetime=0.3
    Ribbon Width=&Multiply_Float(A=12, B=2)
    Ribbon Width Mode=::NewEnumerator1
    UsePositionOffset=false
    Write Material Random=true
   )
   &AddVelocity(Velocity=&FN_FrontEndVector3Modifier(Gameplay=<0 0 -80>, Front End=<0 -200 1100>))
   &SkeletalMeshLocation(
    Bone / Socket Index=$Particles.SkeletonSamplingIndex
    Bone Sampling Mode=::NewEnumerator4
    MaskSpawnGroup=false
    Orientation Sampling=::NewEnumerator1
    Skeletal Mesh=$System.Skeletal Mesh Reader
   )
   &AddVectorToPosition(Vector:Vector3f=&TransformVector(Vector=<0 0 -5>, Destination Space=::Simulation))
   &Color(Color=&MakeLinearColorFromVectorAndFloat("Vector (RGB)"=<1 1 1>, "Float (Alpha)"=4))
  ]
  particleUpdateScript [
   &ParticleState()
   &SolveForcesAndVelocity(Clamp Velocity=false, Limit Acceleration=false)
  ]
  rendererClasses [&NiagaraRibbonRendererProperties]
  renderers [{rendererClass=&NiagaraRibbonRendererProperties}]
 }
 Glow {
  emitterUpdateScript [
   &EmitterState(Inactive Response=::NewEnumerator1, Life Cycle Mode=::NewEnumerator1, Loop Duration=0.125, UseLoopDelay=true)
   &SpawnRate(SpawnRate=6, Use Spawn Probability=false)
   &SpawnRate(SpawnGroup=1, SpawnRate=6, Use Spawn Probability=false) {module=SpawnRate001}
   &SpawnRate(SpawnGroup=2, SpawnRate=6, Use Spawn Probability=false) {module=SpawnRate002}
   &SpawnRate(SpawnGroup=3, SpawnRate=6, Use Spawn Probability=false) {module=SpawnRate003}
  ]
  particleSpawnScript [
   /SetVariables(Particles.SkeletonSamplingIndex:NiagaraInt32=&SequentialInt(Use Array Data:NiagaraBool=false, First Number:NiagaraInt32=0, Sequence:NiagaraDataInterfaceArrayInt32=?, Last Number:NiagaraInt32=1)) {enabled=false, isSetParameters=true}
   &InitializeParticle(
    Lifetime=&RandomRangeFloat(Minimum=0.25, Maximum=0.5, Evaluation Type=::NewEnumerator1)
    Sprite Size=<100 100>
    Sprite Size Mode=::NewEnumerator4
    UsePositionOffset=false
    Write Material Random=true
   )
   &Color(Color=&MakeLinearColorFromVectorAndFloat("Vector (RGB)"=<0.3722 0.5485 0.558>, "Float (Alpha)"=0.1))
  ]
  particleUpdateScript [
   &ParticleState()
   &ScaleSpriteSize(
    Scale Sprite Size Mode=::NewEnumerator0
    Uniform Scale Factor=&FloatFromCurve(FloatCurve=DataInterface() {propertyValues {Curve {keys [<at 0 0.6>,<at 1 1.8>]}}})
   )
   &CameraOffset(Camera Offset Amount=40)
   &ScaleColor(
    Scale Alpha=&FloatFromCurve(FloatCurve=DataInterface() {propertyValues {Curve {keys [<at 0 0 RCIM_Cubic>,<at 0.5 1 RCIM_Cubic>,<at 1 0>]}}})
    ScaleA=true
    ScaleRGB=true
   )
   &SkeletalMeshLocation(
    Bone / Socket Index=&SpawnGroup()
    Bone Sampling Mode=::NewEnumerator4
    MaskSpawnGroup=false
    Orientation Sampling=::NewEnumerator1
    Sampled Position Coordinate Space=::Local
    Sampled Position Offset=&MergeHelper_VectorSwitch(0=<15 0 0>, 1=<-15 0 0>, 2=<-8 0 0>, 3=<8 0 0>, Index=&SpawnGroup(), Num=::NewEnumerator2)
    Skeletal Mesh=$System.Skeletal Mesh Reader
   )
  ]
  rendererClasses [&NiagaraSpriteRendererProperties]
  renderers [{rendererClass=&NiagaraSpriteRendererProperties}]
 }
 Light {
  emitterUpdateScript [
   &EmitterState(Inactive Response=::NewEnumerator1, Life Cycle Mode=::NewEnumerator1, Loop Duration=1, UseLoopDelay=true)
   &SpawnRate(SpawnRate=2, Use Spawn Probability=false)
  ]
  particleSpawnScript [
   /SetVariables(Particles.SkeletonSamplingIndex:NiagaraInt32=&SequentialInt(Use Array Data:NiagaraBool=false, First Number:NiagaraInt32=0, Sequence:NiagaraDataInterfaceArrayInt32=?, Last Number:NiagaraInt32=1)) {isSetParameters=true}
   &InitializeParticle(Sprite Size=<100 100>, Sprite Size Mode=::NewEnumerator4, UsePositionOffset=false, Write Material Random=true)
   &AddVectorToPosition(Vector:Vector3f=&TransformVector(Vector=<0 0 -2>, Destination Space=::Simulation))
   &Color(Color=&MakeLinearColorFromVectorAndFloat("Vector (RGB)"=<0.3722 0.5485 0.558>, "Float (Alpha)"=0.1))
  ]
  particleUpdateScript [
   /SetVariables(Particles.LightExponent=3) {isSetParameters=true}
   &ParticleState()
   &SkeletalMeshLocation(
    Bone / Socket Index=$Particles.SkeletonSamplingIndex
    Bone Sampling Mode=::NewEnumerator4
    MaskSpawnGroup=false
    Orientation Sampling=::NewEnumerator1
    Sampled Position Coordinate Space=::Local
    Sampled Position Offset=&SelectVectorFromArray(
     Direct Array Index=$Particles.SkeletonSamplingIndex
     Vector Selection Array=DataInterface() {propertyValues {FloatData [<15 0 0>, <-15 0 0>]}}
    )
    Skeletal Mesh=$System.Skeletal Mesh Reader
   )
   &ScaleColor(
    Scale Alpha=&FloatFromCurve(FloatCurve=DataInterface() {propertyValues {Curve {keys [<at 0 0>,<at 0.31 1>,<at 0.6 1>,<at 1 0>]}}})
    Scale RGB=<100 100 100>
    ScaleA=true
    ScaleRGB=true
   )
   &CascadeConversion_LightProperties(LightBrightnessScale=1, LightColorScale=<1 1 1>, LightRadiusScale=1)
  ]
  rendererClasses [&NiagaraLightRendererProperties]
  renderers [{rendererClass=&NiagaraLightRendererProperties, rendererIndex=0}]
 }
 SmokeLoop {
  emitterUpdateScript [
   &EmitterState(Inactive Response=::NewEnumerator1, Life Cycle Mode=::NewEnumerator1, Loop Duration=1, UseLoopDelay=true)
   &SpawnRate(
    SpawnRate=&RandomRangeFloat(Minimum=5, Maximum=10, Evaluation Type=::NewEnumerator1)
    Use Spawn Probability=false
   )
  ]
  particleSpawnScript [
   /SetVariables(Particles.SpriteRotationRate=&RandomRangeFloat(Minimum=-0.6, Maximum=-0.25, Evaluation Type=::NewEnumerator1)) {isSetParameters=true}
   /SetVariables() {isSetParameters=true}
   &InitializeParticle(
    Lifetime=&RandomRangeFloat(Minimum=1.7, Maximum=2.2, Evaluation Type=::NewEnumerator1)
    Sprite Rotation Angle=&RandomRangeFloat(Minimum=0, Maximum=1, Evaluation Type=::NewEnumerator1)
    Sprite Rotation Mode=::NewEnumerator1
    Sprite Size=&Vector2DFromFloat(Value=&RandomRangeFloat(Minimum=45, Maximum=75, Evaluation Type=::NewEnumerator1))
    Sprite Size Mode=::NewEnumerator4
    UsePositionOffset=false
    Write Material Random=true
   )
   &Color(Color=&MakeLinearColorFromVectorAndFloat(
     "Vector (RGB)"=<1 1 1>
     "Float (Alpha)"=&RandomRangeFloat(Minimum=1, Maximum=1.25, Evaluation Type=::NewEnumerator1)
    ))
   &SkeletalMeshLocation(
    Bone / Socket Index=&SequentialInt(Use Array Data:NiagaraBool=false, First Number:NiagaraInt32=0, Sequence:NiagaraDataInterfaceArrayInt32=?, Last Number:NiagaraInt32=3)
    Bone Sampling Mode=::NewEnumerator4
    MaskSpawnGroup=false
    Orientation Sampling=::NewEnumerator1
    Sampled Position Coordinate Space=::Local
    Skeletal Mesh=$System.Skeletal Mesh Reader
   )
   &CameraOffset(Camera Offset Amount=&RandomRangeFloat(Minimum=15, Maximum=30, Evaluation Type=::NewEnumerator1))
   &DynamicMaterialParameters(
    Index 0 Param 1=&RandomRangeFloat(Minimum=0, Maximum=1, Evaluation Type=::NewEnumerator1)
    Index 0 Param 2=&RandomRangeFloat(Minimum=0, Maximum=1, Evaluation Type=::NewEnumerator1)
    Module.Param0WriteEnabled=true
    Module.Param1WriteEnabled=true
    Module.Param2WriteEnabled=false
    Module.Param3WriteEnabled=false
   )
  ]
  particleUpdateScript [
   &ParticleState()
   &ScaleSpriteSize(
    Scale Sprite Size Mode=::NewEnumerator0
    Uniform Scale Factor=&FloatFromCurve(
     FloatCurve=DataInterface() {propertyValues {Curve {keys [{arriveTangent=1.9916, interpMode=RCIM_Cubic, leaveTangent=1.9916, tangentMode=RCTM_User, time=0, value=0.2}, <at 1 1.5 RCIM_Cubic>]}}}
     Scale Curve=0.9
    )
   )
   &SpriteRotationRate(Rotation Rate=&AngleConversion(
     Angle:NiagaraFloat=$Particles.SpriteRotationRate
     Angle Input:ENiagara_AngleInput=::NewEnumerator1
     Angle Output:ENiagara_AngleInput=::NewEnumerator0
    ))
   &ScaleColor(
    Scale Alpha=&FloatFromCurve(FloatCurve=DataInterface() {propertyValues {Curve {keys [<at 0 0 RCIM_Cubic>,<at 0.5 1 RCIM_Cubic>,<at 1 0 RCIM_Cubic>]}}}, Scale Curve=0.8)
    Scale RGB=&VectorFromFloat(Value=&FloatFromCurve(
      FloatCurve=DataInterface() {propertyValues {Curve {keys [{arriveTangent=-2.1797, interpMode=RCIM_Cubic, leaveTangent=-2.1797, tangentMode=RCTM_User, time=0, value=1}, <at 1 0.0987>]}}}
      Scale Curve=0.7
     ))
    ScaleA=true
    ScaleRGB=true
   )
   &SubUVAnimation(
    End Frame Range Override=59
    "Sprite Renderer "=DataInterface() {propertyValues {SpriteRenderer=&self:SmokeLoop_0.NiagaraSpriteRendererProperties_0}}
    Start Frame Range Override=14
    UseEndFrame=true
    UseStartFrame=true
   )
   &DynamicMaterialParameters(
    Index 0 Param 4=&FloatFromCurve(FloatCurve=DataInterface() {propertyValues {Curve {keys [<at 0 1>,<at 0.5 0>]}}})
    Module.Param0WriteEnabled=false
    Module.Param1WriteEnabled=false
    Module.Param2WriteEnabled=true
    Module.Param3WriteEnabled=true
   ) {module=DynamicMaterialParameters001}
   &AccelerationForce(Acceleration=$Particles.Acceleration)
   &SolveForcesAndVelocity(Clamp Velocity=false, Limit Acceleration=false)
  ]
  rendererClasses [&NiagaraSpriteRendererProperties]
  renderers [{rendererClass=&NiagaraSpriteRendererProperties}]
 }
 SmokeLoopIdle001 {
  emitterUpdateScript [
   &EmitterState(Inactive Response=::NewEnumerator1, Life Cycle Mode=::NewEnumerator1, Loop Duration=1, UseLoopDelay=true)
   &SpawnRate(SpawnRate=4, Use Spawn Probability=false)
   &SpawnPerUnit(
    Module.Use Max Movement Threshold=true
    Module.Use Movement Tolerance=true
    Movement Tolerance=0.1
    Spawn Spacing:NiagaraFloat=&Divide_Float(
     A=1
     B=&Divide_Float(A=&RandomRangeFloat(Minimum=1, Maximum=2, Evaluation Type=::NewEnumerator1), B=150)
    )
    Use Spawn Probability=false
   )
  ]
  particleSpawnScript [
   &InitializeParticle(
    Lifetime=&RandomRangeFloat(Minimum=0.75, Maximum=1, Evaluation Type=::NewEnumerator1)
    Sprite Rotation Angle=&RandomRangeFloat(Minimum=0.45, Maximum=0.55, Evaluation Type=::NewEnumerator1)
    Sprite Rotation Mode=::NewEnumerator1
    Sprite Size=&RandomRangeVector2D(Minimum=<50 50>, Maximum=<90 90>, Evaluation Type=::NewEnumerator1)
    Sprite Size Mode=::NewEnumerator4
    Sprite UV Mode=::NewEnumerator3
    UsePositionOffset=false
    Write Material Random=true
   )
   &Color(Color=&MakeLinearColorFromVectorAndFloat(
     "Vector (RGB)"=<1 1 1>
     "Float (Alpha)"=&RandomRangeFloat(Minimum=1, Maximum=1.25, Evaluation Type=::NewEnumerator1)
    ))
   &AddVelocity(Velocity=&FN_FrontEndVector3Modifier(Gameplay=<0 0 0>, Front End=<0 -300 1200>))
   &SkeletalMeshLocation(
    Bone / Socket Index=&SequentialInt(Use Array Data:NiagaraBool=false, First Number:NiagaraInt32=0, Sequence:NiagaraDataInterfaceArrayInt32=?, Last Number:NiagaraInt32=1)
    Bone Sampling Mode=::NewEnumerator4
    MaskSpawnGroup=false
    Orientation Sampling=::NewEnumerator1
    Skeletal Mesh=$System.Skeletal Mesh Reader
   )
   &CameraOffset(Camera Offset Amount=&RandomRangeFloat(Minimum=15, Maximum=30, Evaluation Type=::NewEnumerator1))
   &DynamicMaterialParameters(
    Index 0 Param 1=&RandomRangeFloat(Minimum=0, Maximum=1, Evaluation Type=::NewEnumerator1)
    Module.Param0WriteEnabled=true
    Module.Param1WriteEnabled=false
    Module.Param2WriteEnabled=false
    Module.Param3WriteEnabled=false
   )
   &InheritVelocity(Inherited Velocity Amount Scale=<0.75 0.75 0.75>, Inherited Velocity Speed Limit=0, Speed Limit=false, UseSpeedThreshold=false)
  ]
  particleUpdateScript [
   &ParticleState()
   &ScaleSpriteSize(
    Scale Factor=&Vector2DFromCurve(
     Vector2Curve=DataInterface() {propertyValues {
       XCurve {keys [
         {arriveTangent=3.8842, interpMode=RCIM_Cubic, leaveTangent=3.8842, tangentMode=RCTM_User, time=0, value=0.05}
         {arriveTangent=1.7776, interpMode=RCIM_Cubic, leaveTangent=1.7776, time=0.2344, value=0.7419}
         {arriveTangent=0.5984, interpMode=RCIM_Cubic, leaveTangent=0.5984, time=0.5916, value=1.1017}
         <at 1 1.2 RCIM_Cubic>
        ]}
       YCurve {keys [
         {arriveTangent=4.6606, interpMode=RCIM_Cubic, leaveTangent=4.6606, tangentMode=RCTM_User, time=0, value=0.1}
         {arriveTangent=2.17, interpMode=RCIM_Cubic, leaveTangent=2.17, time=0.2344, value=1.0157}
         {arriveTangent=0.6325, interpMode=RCIM_Cubic, leaveTangent=0.6325, time=0.5916, value=1.3838}
         <at 1 1.5 RCIM_Cubic>
        ]}
      }}
     ScaleCurve=<1 1>
    )
    Scale Sprite Size Mode=::NewEnumerator2
   )
   &ScaleColor(
    Scale Alpha=&FloatFromCurve(FloatCurve=DataInterface() {propertyValues {Curve {keys [<at 0 0 RCIM_Cubic>,<at 0.5 1 RCIM_Cubic>,<at 1 0 RCIM_Cubic>]}}})
    Scale RGB=&VectorFromFloat(Value=&FloatFromCurve(FloatCurve=DataInterface() {propertyValues {Curve {keys [{arriveTangent=-2.1797, interpMode=RCIM_Cubic, leaveTangent=-2.1797, tangentMode=RCTM_User, time=0, value=1}, <at 1 0.0987>]}}}))
    ScaleA=true
    ScaleRGB=true
   )
   &SubUVAnimation(
    End Frame Range Override=57
    "Sprite Renderer "=DataInterface() {propertyValues {SpriteRenderer=&self:SmokeLoopIdle001_0.NiagaraSpriteRendererProperties_12}}
    Start Frame Range Override=4
    SubUV Animation Mode:ENiagara_SubUVLookupModeV2=::NewEnumerator1
    SubUV Curve=DataInterface() {propertyValues {Curve {
       defaultValue=0
       keys [
        {arriveTangent=5.2886, interpMode=RCIM_Cubic, leaveTangent=5.2886, tangentMode=RCTM_User, time=0, value=0}
        {arriveTangent=1.4914, interpMode=RCIM_Cubic, leaveTangent=1.4914, tangentMode=RCTM_User, time=0.16, value=0.5}
        {arriveTangent=0.0224, interpMode=RCIM_Cubic, leaveTangent=0.0224, tangentMode=RCTM_User, time=1, value=1}
       ]
      }}}
    UseEndFrame=true
    UseStartFrame=true
   )
   &DynamicMaterialParameters(
    Index 0 Param 2=1
    Index 0 Param 3=&FloatFromCurve(FloatCurve=DataInterface() {propertyValues {Curve {keys [
         <at 0.5 1 RCIM_Cubic>
         {arriveTangent=-5.4992, interpMode=RCIM_Cubic, leaveTangent=-5.4992, tangentMode=RCTM_User, time=0.6598, value=0.5089}
         <at 0.8 0 RCIM_Cubic>
        ]}}})
    Index 0 Param 4=&FloatFromCurve(FloatCurve=DataInterface() {propertyValues {Curve {keys [<at 0 1>,<at 0.5 0>]}}})
    Module.Param0WriteEnabled=false
    Module.Param1WriteEnabled=true
    Module.Param2WriteEnabled=true
    Module.Param3WriteEnabled=true
   ) {module=DynamicMaterialParameters001}
   &SolveForcesAndVelocity(Clamp Velocity=false, Limit Acceleration=false)
  ]
  rendererClasses [&NiagaraSpriteRendererProperties]
  renderers [{rendererClass=&NiagaraSpriteRendererProperties}]
 }
 SmokeLoopReversed {
  emitterUpdateScript [
   &EmitterState(Inactive Response=::NewEnumerator1, Life Cycle Mode=::NewEnumerator1, Loop Duration=1, UseLoopDelay=true)
   &SpawnRate(
    SpawnRate=&RandomRangeFloat(Minimum=5, Maximum=10, Evaluation Type=::NewEnumerator1)
    Use Spawn Probability=false
   )
  ]
  particleSpawnScript [
   /SetVariables(Particles.SpriteRotationRate=&RandomRangeFloat(Minimum=0.2, Maximum=0.5, Evaluation Type=::NewEnumerator1)) {isSetParameters=true}
   /SetVariables() {isSetParameters=true}
   &InitializeParticle(
    Lifetime=&RandomRangeFloat(Minimum=1.7, Maximum=2.2, Evaluation Type=::NewEnumerator1)
    Sprite Rotation Angle=&RandomRangeFloat(Minimum=0, Maximum=1, Evaluation Type=::NewEnumerator1)
    Sprite Rotation Mode=::NewEnumerator1
    Sprite Size=&Vector2DFromFloat(Value=&RandomRangeFloat(Minimum=45, Maximum=75, Evaluation Type=::NewEnumerator1))
    Sprite Size Mode=::NewEnumerator4
    Sprite UV Mode=::NewEnumerator4
    Sprite UV Scale:Vector2f=<-1 1>
    UsePositionOffset=false
    Write Material Random=true
   )
   &Color(Color=&MakeLinearColorFromVectorAndFloat(
     "Vector (RGB)"=<1 1 1>
     "Float (Alpha)"=&RandomRangeFloat(Minimum=1, Maximum=1.25, Evaluation Type=::NewEnumerator1)
    ))
   &SkeletalMeshLocation(
    Bone / Socket Index=&SequentialInt(Use Array Data:NiagaraBool=false, First Number:NiagaraInt32=0, Sequence:NiagaraDataInterfaceArrayInt32=?, Last Number:NiagaraInt32=3)
    Bone Sampling Mode=::NewEnumerator4
    MaskSpawnGroup=false
    Orientation Sampling=::NewEnumerator1
    Sampled Position Coordinate Space=::Local
    Skeletal Mesh=$System.Skeletal Mesh Reader
   )
   &CameraOffset(Camera Offset Amount=&RandomRangeFloat(Minimum=15, Maximum=30, Evaluation Type=::NewEnumerator1))
   &DynamicMaterialParameters(
    Index 0 Param 1=&RandomRangeFloat(Minimum=0, Maximum=1, Evaluation Type=::NewEnumerator1)
    Index 0 Param 2=&RandomRangeFloat(Minimum=0, Maximum=1, Evaluation Type=::NewEnumerator1)
    Module.Param0WriteEnabled=true
    Module.Param1WriteEnabled=true
    Module.Param2WriteEnabled=false
    Module.Param3WriteEnabled=false
   )
  ]
  particleUpdateScript [
   &ParticleState()
   &ScaleSpriteSize(
    Scale Sprite Size Mode=::NewEnumerator0
    Uniform Scale Factor=&FloatFromCurve(
     FloatCurve=DataInterface() {propertyValues {Curve {keys [{arriveTangent=1.9916, interpMode=RCIM_Cubic, leaveTangent=1.9916, tangentMode=RCTM_User, time=0, value=0.2}, <at 1 1.5 RCIM_Cubic>]}}}
     Scale Curve=0.9
    )
   )
   &SpriteRotationRate(Rotation Rate=&AngleConversion(
     Angle:NiagaraFloat=$Particles.SpriteRotationRate
     Angle Input:ENiagara_AngleInput=::NewEnumerator1
     Angle Output:ENiagara_AngleInput=::NewEnumerator0
    ))
   &ScaleColor(
    Scale Alpha=&FloatFromCurve(FloatCurve=DataInterface() {propertyValues {Curve {keys [<at 0 0 RCIM_Cubic>,<at 0.5 1 RCIM_Cubic>,<at 1 0 RCIM_Cubic>]}}}, Scale Curve=0.8)
    Scale RGB=&VectorFromFloat(Value=&FloatFromCurve(
      FloatCurve=DataInterface() {propertyValues {Curve {keys [{arriveTangent=-2.1797, interpMode=RCIM_Cubic, leaveTangent=-2.1797, tangentMode=RCTM_User, time=0, value=1}, <at 1 0.0987>]}}}
      Scale Curve=0.7
     ))
    ScaleA=true
    ScaleRGB=true
   )
   &SubUVAnimation(
    End Frame Range Override=59
    "Sprite Renderer "=DataInterface() {propertyValues {SpriteRenderer=&self:SmokeLoopReversed_0.NiagaraSpriteRendererProperties_1}}
    Start Frame Range Override=14
    UseEndFrame=true
    UseStartFrame=true
   )
   &DynamicMaterialParameters(
    Index 0 Param 4=&FloatFromCurve(FloatCurve=DataInterface() {propertyValues {Curve {keys [<at 0 1>,<at 0.5 0>]}}})
    Module.Param0WriteEnabled=false
    Module.Param1WriteEnabled=false
    Module.Param2WriteEnabled=true
    Module.Param3WriteEnabled=true
   ) {module=DynamicMaterialParameters001}
   &AccelerationForce(Acceleration=$Particles.Acceleration)
   &SolveForcesAndVelocity(Clamp Velocity=false, Limit Acceleration=false)
  ]
  rendererClasses [&NiagaraSpriteRendererProperties]
  renderers [{rendererClass=&NiagaraSpriteRendererProperties}]
 }
 SmokeLoopVelocity {
  emitterUpdateScript [
   &EmitterState(Inactive Response=::NewEnumerator1, Life Cycle Mode=::NewEnumerator1, Loop Duration=1, UseLoopDelay=true)
   &SpawnPerUnit(
    Module.Use Max Movement Threshold=true
    Module.Use Movement Tolerance=true
    Movement Tolerance=0.1
    Spawn Spacing:NiagaraFloat=&Divide_Float(
     A=1
     B=&Divide_Float(A=&RandomRangeFloat(Minimum=0, Maximum=1, Evaluation Type=::NewEnumerator1), B=150)
    )
    Use Spawn Probability=false
   )
  ]
  particleSpawnScript [
   &InitializeParticle(
    Lifetime=&RandomRangeFloat(Minimum=0.4, Maximum=0.6, Evaluation Type=::NewEnumerator1)
    Sprite Size=&RandomRangeVector2D(Minimum=<45 45>, Maximum=<90 90>, Evaluation Type=::NewEnumerator1)
    Sprite Size Mode=::NewEnumerator4
    Sprite UV Mode=::NewEnumerator3
    UsePositionOffset=false
    Write Material Random=true
   )
   &Color(Color=&MakeLinearColorFromVectorAndFloat(
     "Vector (RGB)"=<1 1 1>
     "Float (Alpha)"=&RandomRangeFloat(Minimum=1, Maximum=1.25, Evaluation Type=::NewEnumerator1)
    ))
   &SkeletalMeshLocation(
    Bone / Socket Index=&SequentialInt(Use Array Data:NiagaraBool=false, First Number:NiagaraInt32=0, Sequence:NiagaraDataInterfaceArrayInt32=?, Last Number:NiagaraInt32=1)
    Bone Sampling Mode=::NewEnumerator4
    MaskSpawnGroup=false
    Orientation Sampling=::NewEnumerator1
    Skeletal Mesh=$System.Skeletal Mesh Reader
   )
   &CameraOffset(Camera Offset Amount=&RandomRangeFloat(Minimum=15, Maximum=30, Evaluation Type=::NewEnumerator1))
   &DynamicMaterialParameters(
    Index 0 Param 1=&RandomRangeFloat(Minimum=0, Maximum=1, Evaluation Type=::NewEnumerator1)
    Module.Param0WriteEnabled=true
    Module.Param1WriteEnabled=false
    Module.Param2WriteEnabled=false
    Module.Param3WriteEnabled=false
   )
   &InheritVelocity(Inherited Velocity Amount Scale=<0.75 0.75 0.75>, Inherited Velocity Speed Limit=0, Speed Limit=false, UseSpeedThreshold=false)
  ]
  particleUpdateScript [
   &SolveForcesAndVelocity(Clamp Velocity=false, Limit Acceleration=false)
   &ParticleState()
   &ScaleSpriteSize(
    Scale Factor=&Vector2DFromCurve(
     Vector2Curve=DataInterface() {propertyValues {
       XCurve {keys [
         {arriveTangent=3.8842, interpMode=RCIM_Cubic, leaveTangent=3.8842, tangentMode=RCTM_User, time=0, value=0.05}
         {arriveTangent=1.7776, interpMode=RCIM_Cubic, leaveTangent=1.7776, time=0.2344, value=0.7419}
         {arriveTangent=0.5984, interpMode=RCIM_Cubic, leaveTangent=0.5984, time=0.5916, value=1.1017}
         <at 1 1.2 RCIM_Cubic>
        ]}
       YCurve {keys [
         {arriveTangent=4.6606, interpMode=RCIM_Cubic, leaveTangent=4.6606, tangentMode=RCTM_User, time=0, value=0.1}
         {arriveTangent=2.17, interpMode=RCIM_Cubic, leaveTangent=2.17, time=0.2344, value=1.0157}
         {arriveTangent=0.6325, interpMode=RCIM_Cubic, leaveTangent=0.6325, time=0.5916, value=1.3838}
         <at 1 1.5 RCIM_Cubic>
        ]}
      }}
     ScaleCurve=<1 1>
    )
    Scale Sprite Size Mode=::NewEnumerator2
   )
   &ScaleSpriteSizeBySpeed(Max Scale Factor=<10 20>, Source Velocity:Vector3f=&Multiply_Vector(A=<0 0.05 1>, B=$Particles.Velocity))
   &ScaleColor(
    Scale Alpha=&FloatFromCurve(FloatCurve=DataInterface() {propertyValues {Curve {keys [<at 0 0 RCIM_Cubic>,<at 0.5 1 RCIM_Cubic>,<at 1 0 RCIM_Cubic>]}}})
    Scale RGB=&VectorFromFloat(Value=&FloatFromCurve(FloatCurve=DataInterface() {propertyValues {Curve {keys [{arriveTangent=-2.1797, interpMode=RCIM_Cubic, leaveTangent=-2.1797, tangentMode=RCTM_User, time=0, value=1}, <at 1 0.0987>]}}}))
    ScaleA=true
    ScaleRGB=true
   )
   &SubUVAnimation(
    End Frame Range Override=38
    "Sprite Renderer "=DataInterface() {propertyValues {SpriteRenderer=&self:SmokeLoopVelocity_0.NiagaraSpriteRendererProperties_2}}
    Start Frame Range Override=9
    SubUV Animation Mode:ENiagara_SubUVLookupModeV2=::NewEnumerator1
    SubUV Curve=DataInterface() {propertyValues {Curve {keys [
        {arriveTangent=0.7943, interpMode=RCIM_Cubic, leaveTangent=0.7943, tangentMode=RCTM_User, time=0, value=0}
        {arriveTangent=1.1449, interpMode=RCIM_Cubic, leaveTangent=1.1449, tangentMode=RCTM_User, time=0.29, value=0.7}
        {arriveTangent=0.2086, interpMode=RCIM_Cubic, leaveTangent=0.2086, tangentMode=RCTM_User, time=1, value=1}
       ]}}}
    UseEndFrame=true
    UseStartFrame=true
   )
   &DynamicMaterialParameters(
    Index 0 Param 2=1
    Index 0 Param 3=&FloatFromCurve(FloatCurve=DataInterface() {propertyValues {Curve {keys [
         <at 0.5 1 RCIM_Cubic>
         {arriveTangent=-5.4992, interpMode=RCIM_Cubic, leaveTangent=-5.4992, tangentMode=RCTM_User, time=0.6598, value=0.5089}
         <at 0.8 0 RCIM_Cubic>
        ]}}})
    Index 0 Param 4=&FloatFromCurve(FloatCurve=DataInterface() {propertyValues {Curve {keys [<at 0 1>,<at 0.5 0>]}}})
    Module.Param0WriteEnabled=false
    Module.Param1WriteEnabled=true
    Module.Param2WriteEnabled=true
    Module.Param3WriteEnabled=true
   ) {module=DynamicMaterialParameters001}
  ]
  rendererClasses [&NiagaraSpriteRendererProperties]
  renderers [{rendererClass=&NiagaraSpriteRendererProperties}]
 }
}
summary {
 systemName=NS_Contrail_GreenSmoke
 emitters [
  {emitterName=SmokeLoop, rendererClasses [&NiagaraSpriteRendererProperties]}
  {emitterName=SmokeLoopReversed, rendererClasses [&NiagaraSpriteRendererProperties]}
  {emitterName=SmokeLoopVelocity, rendererClasses [&NiagaraSpriteRendererProperties]}
  {emitterName=Darksmoke_hand_r, rendererClasses [&NiagaraRibbonRendererProperties]}
  {emitterName=Darksmoke_hand_l001, rendererClasses [&NiagaraRibbonRendererProperties]}
  {emitterName=Glow, rendererClasses [&NiagaraSpriteRendererProperties]}
  {emitterName=Light, rendererClasses [&NiagaraLightRendererProperties]}
  {emitterName=SmokeLoopIdle001, rendererClasses [&NiagaraSpriteRendererProperties]}
 ]
 userVariables [
  User.bIsFrontEnd {type=NiagaraBool}
  User.bIsFrontEndPreview {type=NiagaraBool}
 ]
}
