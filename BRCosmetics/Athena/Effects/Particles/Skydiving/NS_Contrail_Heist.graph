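# Niagara skydiving contrail for the Heist cosmetic: short-lived, faint ribbon trails plus tumbling V-Bucks meshes shed from the character's skeleton.
# The ribbon samples four bone or socket points via partitioning, while Body and Pockets mesh emitters spawn from offset skeletal points in looping bursts and by horizontal travel distance.
# Meshes inherit some velocity, then drag, spin, orbit and scale out.
# ParaGlideLeanAlpha boosts mesh spawn rate, and velocities switch between gameplay and front-end values.
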NiagaraSystem /BRCosmetics/Athena/Effects/Particles/Skydiving/NS_Contrail_Heist
dependencies {
 usedDataInterfaces [&NiagaraDataInterfaceArrayFloat3, &NiagaraDataInterfaceCurve, &NiagaraDataInterfaceSkeletalMesh]
 usedDynamicInputs []
 usedModules [
  &self:SystemSpawnScript.SystemScriptSource.SystemScriptGraph.NiagaraNodeAssignment_0.SetVariables_#1
  &SystemState
 ]
}
emitters {
 Ribbon {
  emitterUpdateScript [
   &EmitterState(Inactive Response=::NewEnumerator1, Life Cycle Mode=::NewEnumerator1, Loop Duration=1, UseLoopDelay=true)
   &SpawnRate(SpawnRate=240, Use Spawn Probability=false)
   /SetVariables(Emitter.RibbonPanningSpeed:NiagaraFloat=&Multiply_Float(
     A=&Multiply_Float(A=1, B=&Lerp_Float(A=1, B=2, Alpha=$User.ParaGlideLeanAlpha))
     B=-1
    )) {isSetParameters=true}
  ]
  particleSpawnScript [
   &PartitionParticles(Number Of Partitions=4)
   &InitializeParticle(Lifetime=0.15, Ribbon Width=32, Ribbon Width Mode=::NewEnumerator1, UsePositionOffset=false, Write Material Random=true)
   &Color(Color=&MakeLinearColorFromVectorAndFloat("Vector (RGB)"=<1 1 1>, "Float (Alpha)"=0.025))
   &SkeletalMeshLocation(
    Bone / Socket Index=&Add_Integer(A=2, B=$Particles.PartitionParticles.Partition)
    Bone Sampling Mode=::NewEnumerator4
    MaskSpawnGroup=false
    Orientation Sampling=::NewEnumerator1
    Skeletal Mesh=$System.Skeletal Mesh Reader
   )
   &AddVelocity(Velocity=&FN_FrontEndVector3Modifier(Gameplay=<0 0 -80>, Front End=<0 0 1200>))
   /CascadeToNiagaraConverter/NiagaraScripts/ModuleScripts/AddVectorToPosition(Vector:Vector3f=&TransformVector(Vector=<0 0 0>, Destination Space=::Simulation))
   &DynamicMaterialParameters(
    Index 0 Param 1=0.8
    Index 0 Param 2=0.4
    Index 0 Param 3=1
    Index 0 Param 4=1
    Module.Index0Float0Write=true
    Module.Index0Float1Write=true
    Module.Index0Float2Write=true
    Module.Index0Float3Write=true
   )
  ]
  particleUpdateScript [
   /SetVariables(Particles.ScaledRibbonWidth:NiagaraFloat=&Multiply_Float(
     A=$Particles.RibbonWidth
     B=&FloatFromCurve(FloatCurve=DataInterface() {propertyValues {Curve {keys [
          <at -0.0035 1 RCIM_Cubic>
          {arriveTangent=0.428, interpMode=RCIM_Cubic, leaveTangent=0.428, time=0.1014, value=1.2046}
          <at 0.9769 1.4197 RCIM_Cubic>
         ]}}})
    )) {isSetParameters=true}
   &ParticleState()
   &ScaleColor(
    Scale Alpha=&FloatFromCurve(FloatCurve=DataInterface() {propertyValues {Curve {keys [<at 0.1 0 RCIM_Cubic>,<at 0.3 1.3 RCIM_Cubic>,<at 1 0>]}}})
    ScaleA=true
    ScaleRGB=false
   )
   &SolveForcesAndVelocity(Clamp Velocity=false, Limit Acceleration=false)
  ]
  rendererClasses [&NiagaraRibbonRendererProperties]
  renderers [{rendererClass=&NiagaraRibbonRendererProperties}]
 }
 VBucks_Body {
  emitterUpdateScript [
   &EmitterState(
    Delay First Loop Only=true
    Inactive Response=::NewEnumerator1
    Life Cycle Mode=::NewEnumerator1
    Loop Delay=&RandomRangeFloat(Minimum=0.25, Maximum=0.4, Evaluation Type=::NewEnumerator1)
    Loop Duration=0.1042
    UseLoopDelay=true
   )
   &SpawnRate(
    SpawnRate=&Multiply_Float(A=3, B=&Lerp_Float(A=1, B=2, Alpha=$User.ParaGlideLeanAlpha))
    Use Spawn Probability=false
   )
   &SpawnPerUnit(
    Module.Use Max Movement Threshold=true
    Module.Use Movement Tolerance=true
    Movement Tolerance=0.1
    Spawn Spacing=200
    Use Spawn Probability=false
    Velocity Vector:Vector3f=&Multiply_Vector(A=$Engine.Owner.Velocity, B=<1 1 0>)
   )
  ]
  particleSpawnScript [
   /SetVariables(Particles.MeshRotationEuler:Vector3f=&RandomRangeVector(Minimum=<0 0 0>, Maximum=<1 1 1>, Evaluation Type=::NewEnumerator1)) {isSetParameters=true}
   /SetVariables(Particles.MeshOrientation=/CascadeToNiagaraConverter/NiagaraScripts/DynamicInputScripts/MakeQuatFromEuler(Euler:Vector3f=$Particles.MeshRotationEuler, Scale:NiagaraFloat=360)) {isSetParameters=true}
   /SetVariables(Particles.MeshRotationRateBase:Vector3f=&RandomRangeVector(Minimum=<-0.25 -0.25 -0.25>, Maximum=<0.25 0.25 0.25>, Evaluation Type=::NewEnumerator1)) {isSetParameters=true}
   /SetVariables(Particles.SkeletonSamplingIndex:NiagaraInt32=/CascadeToNiagaraConverter/NiagaraScripts/DynamicInputScripts/SequentialInt(Use Array Data:NiagaraBool=false, First Number:NiagaraInt32=0, Sequence:NiagaraDataInterfaceArrayInt32=?, Last Number:NiagaraInt32=0)) {isSetParameters=true}
   /SetVariables(Particles.ShapeLocationOffset:Vector3f=<0 -40 0>) {isSetParameters=true}
   /SetVariables(Particles.ShapeLocationOffsetSpace:ENiagaraCoordinateSpace=::Local) {isSetParameters=true}
   &InitializeParticle(
    Lifetime=&RandomRangeFloat(Minimum=1, Maximum=1.5, Evaluation Type=::NewEnumerator1)
    Mesh Scale=&RandomRangeVector(Minimum=<1 1 0.25>, Maximum=<1 1 1>, Evaluation Type=::NewEnumerator1)
    Mesh Scale Mode=::NewEnumerator4
    UsePositionOffset=false
    Write Material Random=true
   )
   &Color(Color=&MakeLinearColorFromVectorAndFloat("Vector (RGB)"=<15 15 15>, "Float (Alpha)"=1))
   &DynamicMaterialParameters(
    Index 0 Param 1=&RandomRangeFloat(Minimum=0, Maximum=1, Evaluation Type=::NewEnumerator1)
    Module.Param0WriteEnabled=true
    Module.Param1WriteEnabled=false
    Module.Param2WriteEnabled=false
    Module.Param3WriteEnabled=false
   )
   &SkeletalMeshLocation(
    Bone Sampling Mode=::NewEnumerator4
    MaskSpawnGroup=false
    Orientation Sampling=::NewEnumerator1
    Sampled Position Offset=&SelectVectorFromArray(Direct Array Index=$Particles.SkeletonSamplingIndex, Vector Selection Array=DataInterface() {propertyValues {FloatData [<50 -50 0>]}})
    Skeletal Mesh=$System.Skeletal Mesh Reader
   )
   &ShapeLocation(
    MaskSpawnGroup=false
    Offset=$Particles.ShapeLocationOffset
    Offset Coordinate Space=$Particles.ShapeLocationOffsetSpace
    Offset Mode=::NewEnumerator0
    Override Seed=false
    Sphere Radius=1
    UseOwnerScale=true
   )
   &AddVelocity(Velocity=&FN_FrontEndVector3Modifier(
     Gameplay=&RandomRangeVector(Minimum=<-250 0 0>, Maximum=<250 0 0>, Evaluation Type=::NewEnumerator1)
     Front End=&Multiply_VectorByFloat(
      Vector=&RandomRangeVector(Minimum=<0 0 50>, Maximum=<0 0 150>, Evaluation Type=::NewEnumerator1)
      Float=&Lerp_Float(A=1, B=2, Alpha=$User.ParaGlideLeanAlpha)
     )
    ))
   &InheritVelocity(Inherited Velocity Amount Scale=<0.1 0.1 -0.1>, Inherited Velocity Speed Limit=0, Speed Limit=false, UseSpeedThreshold=false)
  ]
  particleUpdateScript [
   /SetVariables(Particles.MeshRotationRate:Vector3f=$Particles.MeshRotationRateBase) {isSetParameters=true}
   /SetVariables(Particles.MeshRotationRate:Vector3f=&Add_Vector(
     A=$Particles.MeshRotationRate
     B=&VectorFromFloat(Value=&FloatFromCurve(FloatCurve=DataInterface() {propertyValues {Curve {keys [<at 0 2 RCIM_Cubic>,<at 1 1 RCIM_Cubic>]}}}))
    )) {isSetParameters=true}
   &ParticleState()
   &ScaleMeshSize(Scale Factor=&VectorFromFloat(Value=&FloatFromCurve(FloatCurve=DataInterface() {propertyValues {Curve {keys [
          {arriveTangent=10.2367, interpMode=RCIM_Cubic, leaveTangent=10.2367, tangentMode=RCTM_User, time=0, value=0}
          {arriveTangent=1.25, interpMode=RCIM_Cubic, leaveTangent=1.25, time=0.2993, value=1}
          {arriveTangent=-1.4272, interpMode=RCIM_Cubic, leaveTangent=-1.4272, time=0.8, value=1}
          {arriveTangent=-8.5201, interpMode=RCIM_Cubic, leaveTangent=-8.5201, tangentMode=RCTM_User, time=1, value=0}
         ]}}})))
   &Drag(Use Linear Drag=true, Use Rotational Drag=true)
   &SolveForcesAndVelocity(Clamp Velocity=false, Limit Acceleration=false)
   /CascadeToNiagaraConverter/NiagaraScripts/ModuleScripts/UpdateMeshOrientationByEuler(RotationRate:Vector3f=$Particles.MeshRotationRate)
   &DynamicMaterialParameters(
    Index 0 Param 2=&FloatFromCurve(FloatCurve=DataInterface() {propertyValues {Curve {keys [<at 0 0>,<at 0.25 1>]}}})
    Index 0 Param 3=&FloatFromCurve(FloatCurve=DataInterface() {propertyValues {Curve {keys [<at 0 3>,<at 0.1 1>,<at 1 0>]}}})
    Index 0 Param 4=&FloatFromCurve(FloatCurve=DataInterface() {propertyValues {Curve {keys [<at 0 1>,<at 1 0>]}}})
    Module.Param0WriteEnabled=false
    Module.Param1WriteEnabled=true
    Module.Param2WriteEnabled=true
    Module.Param3WriteEnabled=true
   ) {module=DynamicMaterialParameters001}
   /Niagara/Modules/CascadeConversion/CascadeConversion_Orbit(
    Rotation:Vector3f=&RandomRangeVector(Minimum=<-0.15 -0.15 -0.15>, Maximum=<0.15 0.15 0.15>)
    Rotation Rate:Vector3f=&RandomRangeVector(Minimum=<0 0 0>, Maximum=<1 1 1>)
   )
   /Niagara/Modules/CascadeConversion/CascadeConversion_SolveOrbit()
  ]
  rendererClasses [&NiagaraMeshRendererProperties]
  renderers [{rendererClass=&NiagaraMeshRendererProperties, rendererIndex=0}]
 }
 VBucks_Pockets {
  emitterUpdateScript [
   &EmitterState(
    Delay First Loop Only=true
    Inactive Response=::NewEnumerator1
    Life Cycle Mode=::NewEnumerator1
    Loop Delay=&RandomRangeFloat(Minimum=0.25, Maximum=0.4, Evaluation Type=::NewEnumerator1)
    Loop Duration=0.1042
    UseLoopDelay=true
   )
   &SpawnRate(
    SpawnRate=&Multiply_Float(A=3, B=&Lerp_Float(A=1, B=2, Alpha=$User.ParaGlideLeanAlpha))
    Use Spawn Probability=false
   )
   &SpawnPerUnit(
    Module.Use Max Movement Threshold=true
    Module.Use Movement Tolerance=true
    Movement Tolerance=0.1
    Spawn Spacing=200
    Use Spawn Probability=false
    Velocity Vector:Vector3f=&Multiply_Vector(A=$Engine.Owner.Velocity, B=<1 1 0>)
   )
  ]
  particleSpawnScript [
   /SetVariables(Particles.MeshRotationEuler:Vector3f=&RandomRangeVector(Minimum=<0 0 0>, Maximum=<1 1 1>, Evaluation Type=::NewEnumerator1)) {isSetParameters=true}
   /SetVariables(Particles.MeshOrientation=&MakeQuatFromEuler(Euler:Vector3f=$Particles.MeshRotationEuler, Scale:NiagaraFloat=360)) {isSetParameters=true}
   /SetVariables(Particles.MeshRotationRateBase:Vector3f=&RandomRangeVector(Minimum=<-0.25 -0.25 -0.25>, Maximum=<0.25 0.25 0.25>, Evaluation Type=::NewEnumerator1)) {isSetParameters=true}
   /SetVariables(Particles.SkeletonSamplingIndex:NiagaraInt32=&RandomRangeInt(Minimum=0, Maximum=1)) {isSetParameters=true}
   &InitializeParticle(
    Lifetime=&RandomRangeFloat(Minimum=1, Maximum=1.5, Evaluation Type=::NewEnumerator1)
    Mesh Scale=&RandomRangeVector(Minimum=<1 1 0.25>, Maximum=<1 1 1>, Evaluation Type=::NewEnumerator1)
    Mesh Scale Mode=::NewEnumerator4
    UsePositionOffset=false
    Write Material Random=true
   )
   &Color(Color=&MakeLinearColorFromVectorAndFloat("Vector (RGB)"=<15 15 15>, "Float (Alpha)"=1))
   &DynamicMaterialParameters(
    Index 0 Param 1=&RandomRangeFloat(Minimum=0, Maximum=1, Evaluation Type=::NewEnumerator1)
    Module.Param0WriteEnabled=true
    Module.Param1WriteEnabled=false
    Module.Param2WriteEnabled=false
    Module.Param3WriteEnabled=false
   )
   &SkeletalMeshLocation(
    Bone / Socket Index=1
    Bone Sampling Mode=::NewEnumerator4
    MaskSpawnGroup=false
    Orientation Sampling=::NewEnumerator1
    Sampled Position Offset=&SelectVectorFromArray(
     Direct Array Index=$Particles.SkeletonSamplingIndex
     Vector Selection Array=DataInterface() {propertyValues {FloatData [<0 0 25>, <0 0 -25>]}}
    )
    Skeletal Mesh=$System.Skeletal Mesh Reader
   )
   &AddVelocity(Velocity=&FN_FrontEndVector3Modifier(
     Gameplay=&RandomRangeVector(Minimum=<-250 0 0>, Maximum=<250 0 0>, Evaluation Type=::NewEnumerator1)
     Front End=&Multiply_VectorByFloat(
      Vector=&RandomRangeVector(Minimum=<0 0 50>, Maximum=<0 0 150>, Evaluation Type=::NewEnumerator1)
      Float=&Lerp_Float(A=1, B=2, Alpha=$User.ParaGlideLeanAlpha)
     )
    ))
   &InheritVelocity(Inherited Velocity Amount Scale=<0.1 0.1 -0.1>, Inherited Velocity Speed Limit=0, Speed Limit=false, UseSpeedThreshold=false)
  ]
  particleUpdateScript [
   /SetVariables(Particles.MeshRotationRate:Vector3f=$Particles.MeshRotationRateBase) {isSetParameters=true}
   /SetVariables(Particles.MeshRotationRate:Vector3f=&Add_Vector(
     A=$Particles.MeshRotationRate
     B=&VectorFromFloat(Value=&FloatFromCurve(FloatCurve=DataInterface() {propertyValues {Curve {keys [<at 0 2 RCIM_Cubic>,<at 1 1 RCIM_Cubic>]}}}))
    )) {isSetParameters=true}
   &ParticleState()
   &ScaleMeshSize(Scale Factor=&VectorFromFloat(Value=&FloatFromCurve(FloatCurve=DataInterface() {propertyValues {Curve {keys [
          {arriveTangent=10.2367, interpMode=RCIM_Cubic, leaveTangent=10.2367, tangentMode=RCTM_User, time=0, value=0}
          {arriveTangent=1.25, interpMode=RCIM_Cubic, leaveTangent=1.25, time=0.2993, value=1}
          {arriveTangent=-1.4272, interpMode=RCIM_Cubic, leaveTangent=-1.4272, time=0.8, value=1}
          {arriveTangent=-8.5201, interpMode=RCIM_Cubic, leaveTangent=-8.5201, tangentMode=RCTM_User, time=1, value=0}
         ]}}})))
   &Drag(Use Linear Drag=true, Use Rotational Drag=true)
   &SolveForcesAndVelocity(Clamp Velocity=false, Limit Acceleration=false)
   &UpdateMeshOrientationByEuler(RotationRate:Vector3f=$Particles.MeshRotationRate)
   &DynamicMaterialParameters(
    Index 0 Param 2=&FloatFromCurve(FloatCurve=DataInterface() {propertyValues {Curve {keys [<at 0 0>,<at 0.25 1>]}}})
    Index 0 Param 3=&FloatFromCurve(FloatCurve=DataInterface() {propertyValues {Curve {keys [<at 0 3>,<at 0.1 1>,<at 1 0>]}}})
    Index 0 Param 4=&FloatFromCurve(FloatCurve=DataInterface() {propertyValues {Curve {keys [<at 0 1>,<at 1 0>]}}})
    Module.Param0WriteEnabled=false
    Module.Param1WriteEnabled=true
    Module.Param2WriteEnabled=true
    Module.Param3WriteEnabled=true
   ) {module=DynamicMaterialParameters001}
   &CascadeConversion_Orbit(
    Offset:Vector3f=&RandomRangeVector(Minimum=<0 0 0>, Maximum=<0 0 25>)
    Rotation:Vector3f=&RandomRangeVector(Minimum=<-0.15 -0.15 -0.15>, Maximum=<0.15 0.15 0.15>)
    Rotation Rate:Vector3f=&RandomRangeVector(Minimum=<0 0 0>, Maximum=<1 1 1>)
   )
   &CascadeConversion_SolveOrbit()
  ]
  rendererClasses [&NiagaraMeshRendererProperties]
  renderers [{rendererClass=&NiagaraMeshRendererProperties, rendererIndex=0}]
 }
}
summary {
 systemName=NS_Contrail_Heist
 emitters [
  {emitterName=VBucks_Body, rendererClasses [&NiagaraMeshRendererProperties]}
  {emitterName=VBucks_Pockets, rendererClasses [&NiagaraMeshRendererProperties]}
  {emitterName=Ribbon, rendererClasses [&NiagaraRibbonRendererProperties]}
 ]
 userVariables [
  User.bIsFrontEnd {type=NiagaraBool}
  User.bIsFrontEndPreview {type=NiagaraBool}
  User.ParaGlideLeanAlpha {type=NiagaraFloat, defaultValue=0, description=""}
 ]
}
