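# Cascade-converted Niagara skydiving contrail for the Luck of the Irish cosmetic, with clover, rainbow and glow visuals attached to the character skeleton.
# Clover/star mesh and sprites, glows and smoke meshes sample both hand bones, merging variants through spawn-group switches.
# Persistent meshes follow indexed bones via a system skeletal mesh reader.
# Paired rainbow ribbons lerp offset and velocity by ParaGlideLeanAlpha, with separate front-end and gameplay values.
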NiagaraSystem /BRCosmetics/Athena/Effects/Particles/Skydiving/NS_Contrail_LuckOfTheIrish_Converted
dependencies {
 usedDataInterfaces [
  &NiagaraDataInterfaceArrayFloat3
  &NiagaraDataInterfaceVectorCurve
  &NiagaraDataInterfaceCurve
  &NiagaraDataInterfaceArrayInt32
  &NiagaraDataInterfaceVector2DCurve
 ]
 usedDynamicInputs []
 usedModules [
  &self:SystemSpawnScript.SystemScriptSource.SystemScriptGraph.NiagaraNodeAssignment_0.SetVariables_#1
  &SystemState
 ]
}
emitters {
 Clover_Leaf_Star {
  emitterUpdateScript [
   &EmitterState(Inactive Response=::NewEnumerator1, Life Cycle Mode=::NewEnumerator1, Loop Duration=1, UseLoopDelay=true)
   &SpawnRate(Use Spawn Probability=false)
   &SpawnBurst_Instantaneous(Spawn Count=6, Spawn Group=1, Use Spawn Probability=false, UseLoopCountLimit=false)
   &SpawnRate(SpawnGroup=1, SpawnRate=5, Use Spawn Probability=false) {module=SpawnRate001}
   &SpawnBurst_Instantaneous(
    Spawn Count=&RandomRangeInt(Minimum=1, Maximum=10)
    Spawn Group=2
    Use Spawn Probability=false
    UseLoopCountLimit=false
   ) {module=SpawnBurst_Instantaneous001}
   &SpawnRate(SpawnGroup=2, SpawnRate=20, Use Spawn Probability=false) {module=SpawnRate002}
  ]
  particleSpawnScript [
   /SetVariables(Particles.VisibilityTag=&SpawnGroup()) {isSetParameters=true}
   /SetVariables(Particles.MeshRotationEuler:Vector3f=&RandomRangeVector(Minimum=<0 0 0>, Maximum=<1 1 1>, Evaluation Type=::NewEnumerator1)) {isSetParameters=true}
   /SetVariables(Particles.MeshOrientation=/CascadeToNiagaraConverter/NiagaraScripts/DynamicInputScripts/MakeQuatFromEuler(Euler:Vector3f=$Particles.MeshRotationEuler, Scale:NiagaraFloat=360)) {isSetParameters=true}
   /SetVariables(Particles.MeshRotationRateBase:Vector3f=&RandomRangeVector(Minimum=<-0.4 -0.38 -0.38>, Maximum=<0.4 0.4 0.4>, Evaluation Type=::NewEnumerator1)) {isSetParameters=true}
   /SetVariables(Particles.SpriteRotationRate=&RandomRangeFloat(Minimum=-1, Maximum=1, Evaluation Type=::NewEnumerator1)) {isSetParameters=true}
   &InitializeParticle(
    Lifetime=&MergeHelper_FloatSwitch(
     0=&RandomRangeFloat(Minimum=1.1, Maximum=1.6, Evaluation Type=::NewEnumerator1)
     1=&RandomRangeFloat(Minimum=0.7, Maximum=1.2, Evaluation Type=::NewEnumerator1)
     2=&RandomRangeFloat(Minimum=0.16, Maximum=0.6, Evaluation Type=::NewEnumerator1)
     3=?
     Index=&SpawnGroup()
     Num=::NewEnumerator1
    )
    Mesh Scale=&VectorFromFloat(Value=&RandomRangeFloat(Minimum=0.8, Maximum=0.5, Evaluation Type=::NewEnumerator1))
    Mesh Scale Mode=::NewEnumerator4
    Sprite Size Mode=::NewEnumerator3
    Uniform Sprite Size=&MergeHelper_FloatSwitch(
     0=0
     1=&MakeFloatFromVector2d(Vector 2D=&Vector2DFromFloat(Value=&RandomRangeFloat(Minimum=15, Maximum=10, Evaluation Type=::NewEnumerator1)))
     2=&MakeFloatFromVector2d(Vector 2D=&Vector2DFromFloat(Value=&RandomRangeFloat(Minimum=1.5, Maximum=3.8, Evaluation Type=::NewEnumerator1)))
     3=?
     Index=&SpawnGroup()
     Num=::NewEnumerator1
    )
    UsePositionOffset=false
    Write Material Random=true
   )
   /CascadeToNiagaraConverter/NiagaraScripts/ModuleScripts/FixShapeVector()
   &AddVelocity(Velocity=&MergeHelper_VectorSwitch(
     0=&RandomRangeVector(Minimum=<-25 -25 0>, Maximum=<25 25 0>, Evaluation Type=::NewEnumerator1)
     1=&RandomRangeVector(Minimum=<-10 -10 150>, Maximum=<10 10 200>, Evaluation Type=::NewEnumerator1)
     2=&Multiply_VectorByFloat(
      Vector=$Output.FixShapeVector.ShapeVector
      Float=&RandomRangeFloat(Minimum=1.5, Maximum=4.5, Evaluation Type=::NewEnumerator1)
     )
     3=?
     Index=&SpawnGroup()
     Num=::NewEnumerator1
    ))
   &AddVelocity(Velocity=&MaskVector(Mask Source=0, Mask Parameter=?, Mask=2, Masked Vector=<0 0 80>, Passthrough Vector=<1 1 1>)) {module=AddVelocity001}
   &SkeletalMeshLocation(
    Bone / Socket Index=/CascadeToNiagaraConverter/NiagaraScripts/DynamicInputScripts/SequentialInt(Use Array Data:NiagaraBool=false, First Number:NiagaraInt32=0, Sequence:NiagaraDataInterfaceArrayInt32=?, Last Number:NiagaraInt32=1)
    Bone Sampling Mode=::NewEnumerator4
    MaskSpawnGroup=false
    Orientation Sampling=::NewEnumerator1
    Skeletal Mesh=DataInterface() {propertyValues {
      DefaultMesh=null
      ExcludeBoneName=root
      OptionalComponentClass=null
      PreviewMesh=null
      SkinningMode=SkinOnTheFly
      SoftSourceActor=null
      SourceMode=Default
      UvSetIndex=0
      WholeMeshLOD=-1
      bExcludeBone=false
      bReadDeformedGeometry=true
      bRequireCurrentFrameData=true
      ComponentTags []
      FilteredBones [hand_l,hand_r]
      FilteredSockets []
      MeshUserParameter {parameter {name=None, typeDefHandle {registeredTypeIndex=1}, varData []}}
      SamplingRegions []
     }}
   )
   &Color(Color=&MaskLinearColor(
     Mask Source=0
     Mask Parameter=?
     Mask=1
     Masked Linear Color=&MakeLinearColorFromVectorAndFloat("Vector (RGB)"=<0.756 0.8958 0.7419>, "Float (Alpha)"=1)
     Passthrough Linear Color=<rgba 1 1 1 1>
    )) {module=Color001}
   &Color(Color=&MaskLinearColor(
     Mask Source=0
     Mask Parameter=?
     Mask=2
     Masked Linear Color=&MakeLinearColorFromVectorAndFloat(
      "Vector (RGB)"=&RandomRangeVector(Minimum=<4 4 4>, Maximum=<10 10 10>, Evaluation Type=::NewEnumerator1)
      "Float (Alpha)"=1
     )
     Passthrough Linear Color=<rgba 1 1 1 1>
    )) {module=Color002}
   &ShapeLocation(
    MaskSpawnGroup=false
    Override Seed=false
    Sphere Radius=&MaskFloat(Mask Source=0, Mask Parameter=?, Mask=1, Masked Float=20, Passthrough Float=1)
    UseOwnerScale=true
   )
   &ShapeLocation(MaskSpawnGroup=false, Override Seed=false, Sphere Radius=10, UseOwnerScale=true) {module=ShapeLocation001}
   &InheritVelocity(
    Inherited Velocity Amount Scale=&MergeHelper_VectorSwitch(
     0=<-0.14 -0.14 -0.14>
     1=<-0.14 -0.14 -0.14>
     2=&RandomRangeVector(Minimum=<-0.22 -0.2 -0.22>, Maximum=<-0.08 -0.1 -0.1>, Evaluation Type=::NewEnumerator1)
     3=?
     Index=&SpawnGroup()
     Num=::NewEnumerator1
    )
    Inherited Velocity Speed Limit=0
    Speed Limit=false
    UseSpeedThreshold=false
   )
  ]
  particleUpdateScript [
   /SetVariables(Particles.MeshRotationRate:Vector3f=$Particles.MeshRotationRateBase) {isSetParameters=true}
   &ParticleState()
   &ScaleMeshSize(Scale Factor=&VectorFromFloat(Value=&FloatFromCurve(FloatCurve=DataInterface() {propertyValues {Curve {keys [<at 0 0>,<at 0.3 1>,<at 0.75 1>,<at 1 0>]}}})))
   &ScaleSpriteSize(
    Scale Factor=&MergeHelper_Vector2Switch(
     0=<0 0>
     1=&Vector2DFromCurve(
      Vector2Curve=DataInterface() {propertyValues {
        XCurve {keys [
          <at 0 1 RCIM_Cubic>
          {arriveTangent=-0.2248, interpMode=RCIM_Cubic, leaveTangent=-0.2248, time=0.2491, value=0.3516}
          {arriveTangent=-0.1247, interpMode=RCIM_Cubic, leaveTangent=-0.1247, time=0.4973, value=0.8882}
          {arriveTangent=0.2224, interpMode=RCIM_Cubic, leaveTangent=0.2224, time=0.7536, value=0.2887}
          <at 1 1 RCIM_Cubic>
         ]}
        YCurve {keys [<at 0 1 RCIM_Cubic>,<at 0.2491 1 RCIM_Cubic>,<at 0.4973 1 RCIM_Cubic>,<at 0.7536 1 RCIM_Cubic>,<at 1 1 RCIM_Cubic>]}
       }}
      ScaleCurve=<1 1>
     )
     2=&Vector2DFromFloat(Value=&FloatFromCurve(FloatCurve=DataInterface() {propertyValues {Curve {keys [<at 0 0 RCIM_Cubic>,<at 0.3 1 RCIM_Cubic>,<at 1 0>]}}}))
     3=?
     4=?
     Index=&SpawnGroup()
     Num=::NewEnumerator1
    )
    Scale Sprite Size Mode=::NewEnumerator2
   ) {module=ScaleSpriteSize001}
   &SpriteRotationRate(Rotation Rate=&AngleConversion(
     Angle:NiagaraFloat=$Particles.SpriteRotationRate
     Angle Input:ENiagara_AngleInput=::NewEnumerator1
     Angle Output:ENiagara_AngleInput=::NewEnumerator0
    ))
   &Drag(
    Drag=&MergeHelper_FloatSwitch(0=1.4, 1=2.4, 2=2, 3=?, Index=&SpawnGroup(), Num=::NewEnumerator1)
    Use Linear Drag=true
    Use Rotational Drag=true
   )
   &SolveForcesAndVelocity(Clamp Velocity=false, Limit Acceleration=false)
   &Color(Color=&MakeLinearColorFromVectorAndFloat(
     "Vector (RGB)"=&VectorFromCurve(VectorCurve=DataInterface() {propertyValues {XCurve {keys [<at 0 1>,<at 1 1>]}, YCurve {keys [<at 0 1>,<at 1 1>]}, ZCurve {keys [<at 0 1>,<at 1 1>]}}})
     "Float (Alpha)"=&ClampFloat(Float=&FloatFromCurve(FloatCurve=DataInterface() {propertyValues {Curve {keys [{arriveTangent=-1, time=0, value=1}, <at 1 0>]}}}))
    ))
   &ScaleColor(
    Scale Alpha=&MergeHelper_FloatSwitch(
     0=1
     1=&FloatFromCurve(FloatCurve=DataInterface() {propertyValues {Curve {keys [<at 0 1>,<at 1 0>]}}})
     2=&FloatFromCurve(FloatCurve=DataInterface() {propertyValues {Curve {keys [
          {arriveTangent=10.7452, interpMode=RCIM_Cubic, leaveTangent=10.7452, tangentMode=RCTM_User, time=0, value=0}
          {arriveTangent=0.3432, interpMode=RCIM_Cubic, leaveTangent=0.3432, tangentMode=RCTM_User, time=0.1334, value=0.9177}
          {arriveTangent=-0.4298, interpMode=RCIM_Cubic, leaveTangent=-0.4298, time=0.1811, value=0.291}
          {arriveTangent=-0.3692, interpMode=RCIM_Cubic, leaveTangent=-0.3692, time=0.3461, value=0.8263}
          {arriveTangent=0.1565, interpMode=RCIM_Cubic, leaveTangent=0.1565, time=0.4899, value=0.177}
          {arriveTangent=0.0773, interpMode=RCIM_Cubic, leaveTangent=0.0773, time=0.5914, value=0.8647}
          {arriveTangent=-0.9503, interpMode=RCIM_Cubic, leaveTangent=-0.9503, tangentMode=RCTM_User, time=0.7314, value=0.1956}
          {arriveTangent=-0.7283, interpMode=RCIM_Cubic, leaveTangent=-0.7283, time=0.8241, value=0.8895}
          <at 1 0>
         ]}}})
     3=?
     Index=&SpawnGroup()
     Num=::NewEnumerator1
    )
    Scale RGB=&MergeHelper_VectorSwitch(
     0=<1 1 1>
     1=&VectorFromCurve(VectorCurve=DataInterface() {propertyValues {XCurve {keys [<at 0 1>,<at 1 1>]}, YCurve {keys [<at 0 1>,<at 1 1>]}, ZCurve {keys [<at 0 1>,<at 1 1>]}}})
     2=&VectorFromCurve(VectorCurve=DataInterface() {propertyValues {
        XCurve {keys [
          <at 0 8 RCIM_Cubic>
          {arriveTangent=-14.44, interpMode=RCIM_Cubic, leaveTangent=-14.44, time=0.2, value=0.05}
          <at 0.5 0.78>
          <at 0.75 8>
          <at 1 10>
         ]}
        YCurve {keys [
          <at 0 0.8 RCIM_Cubic>
          {arriveTangent=-1.6, interpMode=RCIM_Cubic, leaveTangent=-1.6, time=0.2, value=7}
          <at 0.5 0>
          <at 0.75 0>
          <at 1 0.88>
         ]}
        ZCurve {keys [<at 0 0 RCIM_Cubic>, {arriveTangent=16, interpMode=RCIM_Cubic, leaveTangent=16, time=0.2, value=0}, <at 0.5 8>, <at 0.75 0.1853>, <at 1 0>]}
       }})
     3=?
     Index=&SpawnGroup()
     Num=::NewEnumerator1
    )
    ScaleA=true
    ScaleRGB=true
   )
   /CascadeToNiagaraConverter/NiagaraScripts/ModuleScripts/UpdateMeshOrientationByEuler(RotationRate:Vector3f=$Particles.MeshRotationRate)
   /Niagara/Modules/CascadeConversion/CascadeConversion_Orbit(
    Offset:Vector3f=&MergeHelper_VectorSwitch(
     0=&RandomRangeVector(Minimum=<-5 -4.5 -5>, Maximum=<5 4 5>)
     1=&RandomRangeVector(Minimum=<-5 -4.5 -5>, Maximum=<5 4 5>)
     2=&RandomRangeVector(Minimum=<-9 -7 0>, Maximum=<7 8 0>)
     3=?
     Index=&SpawnGroup()
     Num=::NewEnumerator1
    )
    Rotation:Vector3f=&RandomRangeVector(Minimum=<0 0 0>, Maximum=<1 1 1>)
    Rotation Rate:Vector3f=&RandomRangeVector(Minimum=<0 0 0>, Maximum=<1 1 1>)
   )
   /Niagara/Modules/CascadeConversion/CascadeConversion_SolveOrbit()
  ]
  rendererClasses [&NiagaraMeshRendererProperties, &NiagaraSpriteRendererProperties]
  renderers [
   {rendererClass=&NiagaraMeshRendererProperties, rendererIndex=0}
   {rendererClass=&NiagaraSpriteRendererProperties, rendererIndex=1}
   {rendererClass=&NiagaraSpriteRendererProperties, rendererIndex=2}
  ]
 }
 Glows {
  emitterUpdateScript [
   &EmitterState(Inactive Response=::NewEnumerator1, Life Cycle Mode=::NewEnumerator1, Loop Duration=0.125, UseLoopDelay=true)
   &SpawnBurst_Instantaneous(
    Spawn Count=&RandomRangeInt(Minimum=0, Maximum=1)
    Use Spawn Probability=false
    UseLoopCountLimit=false
   )
   &SpawnRate(SpawnRate=6, Use Spawn Probability=false)
   &SpawnRate(SpawnGroup=1, SpawnRate=20, Use Spawn Probability=false) {module=SpawnRate001}
  ]
  particleSpawnScript [
   /SetVariables(Particles.VisibilityTag=&SpawnGroup()) {isSetParameters=true}
   /SetVariables(Particles.SkeletonSamplingIndex:NiagaraInt32=&SequentialInt(Use Array Data:NiagaraBool=false, First Number:NiagaraInt32=0, Sequence:NiagaraDataInterfaceArrayInt32=?, Last Number:NiagaraInt32=1)) {isSetParameters=true}
   &InitializeParticle(
    Lifetime=&MergeHelper_FloatSwitch(
     0=&RandomRangeFloat(Minimum=0.6, Maximum=1, Evaluation Type=::NewEnumerator1)
     1=&RandomRangeFloat(Minimum=0.24, Maximum=0.45, Evaluation Type=::NewEnumerator1)
     2=&RandomRangeFloat(Minimum=0.16, Maximum=0.6, Evaluation Type=::NewEnumerator1)
     3=?
     Index=&SpawnGroup()
     Num=::NewEnumerator0
    )
    Sprite Size=&MergeHelper_Vector2Switch(
     0=<92 92>
     1=&Vector2DFromFloat(Value=&RandomRangeFloat(Minimum=12, Maximum=28, Evaluation Type=::NewEnumerator1))
     2=&Vector2DFromFloat(Value=&RandomRangeFloat(Minimum=1.5, Maximum=3.8, Evaluation Type=::NewEnumerator1))
     3=?
     4=?
     Index=&SpawnGroup()
     Num=::NewEnumerator0
    )
    Sprite Size Mode=::NewEnumerator4
    UsePositionOffset=false
    Write Material Random=true
   )
   &Color(Color=&MergeHelper_ColorSwitch(
     0=&MakeLinearColorFromVectorAndFloat(
      "Vector (RGB)"=&RandomRangeVector(Minimum=<0.2892 0.1039 1>, Maximum=<0.4444 1 0.0682>, Evaluation Type=::NewEnumerator1)
      "Float (Alpha)"=0.22
     )
     1=&MakeLinearColorFromVectorAndFloat("Vector (RGB)"=<1 1 1>, "Float (Alpha)"=0.52)
     2=&MakeLinearColorFromVectorAndFloat(
      "Vector (RGB)"=&RandomRangeVector(Minimum=<4 4 4>, Maximum=<10 10 10>, Evaluation Type=::NewEnumerator1)
      "Float (Alpha)"=1
     )
     3=?
     Index=&SpawnGroup()
     Num=::NewEnumerator0
    ))
   &ShapeLocation(
    MaskSpawnGroup=false
    Override Seed=false
    Sphere Radius=&MaskFloat(Mask Source=0, Mask Parameter=?, Mask=1, Masked Float=5, Passthrough Float=1)
    UseOwnerScale=true
   )
   &DynamicMaterialParameters(
    Index 0 Param 1=&MaskFloat(Mask Source=0, Mask Parameter=?, Mask=1, Masked Float=-2, Passthrough Float=1)
    Module.Param0WriteEnabled=true
    Module.Param1WriteEnabled=false
    Module.Param2WriteEnabled=false
    Module.Param3WriteEnabled=false
   )
  ]
  particleUpdateScript [
   &ParticleState()
   &SkeletalMeshLocation(
    Bone / Socket Index=$Particles.SkeletonSamplingIndex
    Bone Sampling Mode=::NewEnumerator4
    MaskSpawnGroup=false
    Orientation Sampling=::NewEnumerator1
    Sampled Position Offset=&MaskVector(
     Mask Source=0
     Mask Parameter=?
     Mask=&SelectIntFromArray(Direct Array Index=0, Int Selection Array=DataInterface() {propertyValues {IntData [0,1]}})
     Masked Vector=&SelectVectorFromArray(Direct Array Index=$Particles.SkeletonSamplingIndex, Vector Selection Array=DataInterface() {propertyValues {FloatData [<7 0 0>, <-7 0 0>]}})
     Passthrough Vector=<1 1 1>
    )
    Skeletal Mesh=DataInterface() {propertyValues {
      DefaultMesh=null
      ExcludeBoneName=root
      OptionalComponentClass=null
      PreviewMesh=null
      SkinningMode=SkinOnTheFly
      SoftSourceActor=null
      SourceMode=Default
      UvSetIndex=0
      WholeMeshLOD=-1
      bExcludeBone=false
      bReadDeformedGeometry=true
      bRequireCurrentFrameData=true
      ComponentTags []
      FilteredBones [hand_l,hand_r]
      FilteredSockets []
      MeshUserParameter {parameter {name=None, typeDefHandle {registeredTypeIndex=1}, varData []}}
      SamplingRegions []
     }}
   )
   &ScaleSpriteSize(
    Scale Sprite Size Mode=::NewEnumerator0
    Uniform Scale Factor=&MergeHelper_FloatSwitch(
     0=&FloatFromCurve(FloatCurve=DataInterface() {propertyValues {Curve {keys [<at 0 0.6>,<at 1 1.8>]}}})
     1=1
     2=&FloatFromCurve(FloatCurve=DataInterface() {propertyValues {Curve {keys [<at 0 0 RCIM_Cubic>,<at 0.3 1 RCIM_Cubic>,<at 1 0>]}}})
     3=?
     Index=&SpawnGroup()
     Num=::NewEnumerator0
    )
   )
   &CameraOffset(Camera Offset Amount=&MaskFloat(
     Mask Source=0
     Mask Parameter=?
     Mask=0
     Masked Float=&RandomRangeFloat(Minimum=10, Maximum=40, Evaluation Type=::NewEnumerator1)
     Passthrough Float=1
    ))
   &SolveForcesAndVelocity(Clamp Velocity=false, Limit Acceleration=false)
   &ScaleColor(
    Scale Alpha=&MergeHelper_FloatSwitch(
     0=&FloatFromCurve(FloatCurve=DataInterface() {propertyValues {Curve {keys [
          <at 0 0 RCIM_Cubic>
          {arriveTangent=0.5535, interpMode=RCIM_Cubic, leaveTangent=0.5535, tangentMode=RCTM_User, time=0.2, value=0.7783}
          {arriveTangent=-0.9729, interpMode=RCIM_Cubic, leaveTangent=-0.9729, time=0.6, value=1}
          <at 1 0>
         ]}}})
     1=&FloatFromCurve(FloatCurve=DataInterface() {propertyValues {Curve {keys [<at 0 0>,<at 0.3 1>,<at 1 0>]}}})
     2=?
     3=?
     Index=&SpawnGroup()
     Num=::NewEnumerator0
    )
    Scale RGB=&MergeHelper_VectorSwitch(
     0=&VectorFromCurve(VectorCurve=DataInterface() {propertyValues {
        XCurve {keys [<at 0 0.5826 RCIM_Cubic>,<at 0.8939 0.5473 RCIM_Cubic>]}
        YCurve {keys [<at 0 1.2 RCIM_Cubic>,<at 0.8939 0.081 RCIM_Cubic>]}
        ZCurve {keys [<at 0 0.2046 RCIM_Cubic>,<at 0.8939 0 RCIM_Cubic>]}
       }})
     1=&VectorFromCurve(VectorCurve=DataInterface() {propertyValues {XCurve {keys [<at 0 1>,<at 1 1>]}, YCurve {keys [<at 0 1>,<at 1 1>]}, ZCurve {keys [<at 0 1>,<at 1 1>]}}})
     2=&VectorFromCurve(VectorCurve=DataInterface() {propertyValues {
        XCurve {keys [
          <at 0 8 RCIM_Cubic>
          {arriveTangent=-14.44, interpMode=RCIM_Cubic, leaveTangent=-14.44, time=0.2, value=0.05}
          <at 0.5 0.78>
          <at 0.75 8>
          <at 1 10>
         ]}
        YCurve {keys [
          <at 0 0.8 RCIM_Cubic>
          {arriveTangent=-1.6, interpMode=RCIM_Cubic, leaveTangent=-1.6, time=0.2, value=7}
          <at 0.5 0>
          <at 0.75 0>
          <at 1 0.88>
         ]}
        ZCurve {keys [<at 0 0 RCIM_Cubic>, {arriveTangent=16, interpMode=RCIM_Cubic, leaveTangent=16, time=0.2, value=0}, <at 0.5 8>, <at 0.75 0.1853>, <at 1 0>]}
       }})
     3=?
     Index=&SpawnGroup()
     Num=::NewEnumerator0
    )
    ScaleA=true
    ScaleRGB=true
   )
   &DynamicMaterialParameters(
    Index 0 Param 2=&MaskFloat(Mask Source=0, Mask Parameter=?, Mask=1, Masked Float=1, Passthrough Float=1)
    Index 0 Param 3=&MaskFloat(Mask Source=0, Mask Parameter=?, Mask=1, Masked Float=1, Passthrough Float=1)
    Index 0 Param 4=&MaskFloat(Mask Source=0, Mask Parameter=?, Mask=1, Masked Float=1, Passthrough Float=1)
    Module.Param0WriteEnabled=false
    Module.Param1WriteEnabled=true
    Module.Param2WriteEnabled=true
    Module.Param3WriteEnabled=true
   ) {module=DynamicMaterialParameters001}
  ]
  rendererClasses [&NiagaraSpriteRendererProperties]
  renderers [
   {rendererClass=&NiagaraSpriteRendererProperties}
   {rendererClass=&NiagaraSpriteRendererProperties, rendererIndex=1}
  ]
 }
 Particle_Emitter004 {
  emitterUpdateScript [
   &EmitterState(
    Inactive Response=::NewEnumerator1
    Life Cycle Mode=::NewEnumerator1
    Loop Behavior=::NewEnumerator1
    Loop Duration Mode=::NewEnumerator1
    UseLoopDelay=true
   )
   &SpawnBurst_Instantaneous(Spawn Count=8, Use Spawn Probability=false, UseLoopCountLimit=false)
  ]
  particleSpawnScript [
   /SetVariables(Particles.SkeletonSamplingIndex:NiagaraInt32=&SequentialInt(Use Array Data:NiagaraBool=false, First Number:NiagaraInt32=0, Sequence:NiagaraDataInterfaceArrayInt32=?, Last Number:NiagaraInt32=7)) {isSetParameters=true}
   &InitializeParticle(
    Mesh Scale=/Niagara/DynamicInputs/CascadeConversion/SplitVector(
     Double Component Vector Channels:ECascadeNiagaraTwoVectorChannels=::NewEnumerator0
     Double Component:NiagaraFloat=&RandomRangeFloat(Minimum=0.78, Maximum=0.78, Evaluation Type=::NewEnumerator1)
     Single Component:NiagaraFloat=&RandomRangeFloat(Minimum=0.38, Maximum=0.65, Evaluation Type=::NewEnumerator1)
    )
    Mesh Scale Mode=::NewEnumerator4
    UsePositionOffset=false
    Write Material Random=true
   )
   &DynamicMaterialParameters(
    Index 0 Param 1=&RandomRangeFloat(Minimum=0, Maximum=50, Evaluation Type=::NewEnumerator1)
    Module.Param0WriteEnabled=true
    Module.Param1WriteEnabled=false
    Module.Param2WriteEnabled=false
    Module.Param3WriteEnabled=false
   )
   &InitialMeshOrientation(
    Mesh Orientation Mode=::NewEnumerator1
    Rotation=&SelectVectorFromArray(
     Direct Array Index=$Particles.SkeletonSamplingIndex
     Direct Array Mode=::NewEnumerator1
     Vector Selection Array=DataInterface() {propertyValues {FloatData [<0.05 0.1 0>, <0 0 0>, <0 0.1 0>, <0 0 0>, <0 -0.1 0.1>, <0 0.1 0>, <0 0.15 0>, <0 -0.15 0.5>]}}
    )
    Use Rotation Vector=true
   )
  ]
  particleUpdateScript [
   &ParticleState(Kill Particles When Lifetime Has Elapsed=false, Let Infinitely Lived Particles Die When Emitter Deactivates=true)
   &Color(Color=&MakeLinearColorFromVectorAndFloat(
     "Vector (RGB)"=&VectorFromCurve(VectorCurve=DataInterface() {propertyValues {XCurve {keys [<at 0 1>,<at 1 1>]}, YCurve {keys [<at 0 1>,<at 1 1>]}, ZCurve {keys [<at 0 1>,<at 1 1>]}}})
     "Float (Alpha)"=&ClampFloat(Float=&FloatFromCurve(FloatCurve=DataInterface() {propertyValues {Curve {keys [{arriveTangent=-1, time=0, value=1}, <at 1 0>]}}}))
    ))
   &SkeletalMeshLocation(
    Bone / Socket Index=$Particles.SkeletonSamplingIndex
    Bone Sampling Mode=::NewEnumerator4
    MaskSpawnGroup=false
    Orientation Sampling=::NewEnumerator1
    Sampled Position Offset=&SelectVectorFromArray(
     Direct Array Index=$Particles.SkeletonSamplingIndex
     Vector Selection Array=DataInterface() {propertyValues {FloatData [<0 0 0>, <-5 0 0>, <0 0 -2>, <0 0 0>, <-5 0 0>, <2.5 2 0>, <-7 0 0>, <5 0 1>]}}
    )
    Skeletal Mesh=$System.Skeletal Mesh Reader
   ) {module=SkeletalMeshLocation002}
   &DynamicMaterialParameters(
    Index 0 Param 2=1
    Index 0 Param 3=1
    Index 0 Param 4=1
    Module.Param0WriteEnabled=false
    Module.Param1WriteEnabled=true
    Module.Param2WriteEnabled=true
    Module.Param3WriteEnabled=true
   ) {module=DynamicMaterialParameters001}
  ]
  rendererClasses [&NiagaraMeshRendererProperties]
  renderers [{rendererClass=&NiagaraMeshRendererProperties, rendererIndex=0}]
 }
 Ribbon_Rainbow {
  emitterUpdateScript [
   &EmitterState()
   &FN_SpawnRibbon_v2(Number of Ribbons=2, Spawn Mode=::NewEnumerator0, SpawnRate=64)
  ]
  particleSpawnScript [
   &FN_SetSpawnGroup(Number of Spawn Groups=2)
   &PartitionParticles(Number Of Partitions=4)
   &InitializeParticle(
    Color Mode=::NewEnumerator0
    Lifetime=0.3
    Ribbon Width=&Multiply_Float(A=5, B=2)
    Ribbon Width Mode=::NewEnumerator1
    UsePositionOffset=false
    Write Material Random=true
   )
   &SkeletalMeshLocation(
    Bone / Socket Index=&Add_Integer(A=$Particles.SpawnGroup, B=10)
    Bone Sampling Mode=::NewEnumerator4
    MaskSpawnGroup=false
    Sampled Position Offset=&FN_FrontEndVector3Modifier(
     Gameplay=&Lerp_Vector(A=<0 0 -60>, B=<0 0 -80>, Alpha=$User.ParaGlideLeanAlpha)
     Front End=&Lerp_Vector(A=<0 40 -50>, B=<0 35 -70>, Alpha=$User.ParaGlideLeanAlpha)
    )
    Sampled Position Offset Coordinate Space:ENiagaraSkelMeshTransforms=::NewEnumerator1
    Skeletal Mesh=$User.New Skeletal Mesh
   ) {module=SkeletalMeshLocation001}
   /SetVariables(Particles.Acceleration:Vector3f=&RandomRangeVector(Minimum=<-3 -3 0>, Maximum=<3 3 0>)) {isSetParameters=true}
   &AddVelocity(Velocity=&FN_FrontEndVector3Modifier(
     Gameplay=&Lerp_Vector(A=<0 0 750>, B=<1 1 1200>, Alpha=$User.ParaGlideLeanAlpha)
     Front End=&Lerp_Vector(A=<0 0 700>, B=<1 1 999>, Alpha=$User.ParaGlideLeanAlpha)
    ))
   &AddVelocity(Velocity=<0 -500 0>, Velocity Speed Scale=&FN_FrontEndFloatModifier(Gameplay=0, Front End=1)) {module=AddVelocity002}
  ]
  particleUpdateScript [
   /SetVariables(Particles.ScaledRibbonWidth:NiagaraFloat=&Multiply_Float(
     A=$Particles.RibbonWidth
     B=&FloatFromCurve(FloatCurve=DataInterface() {propertyValues {Curve {keys [
          <at -0.0035 -0.0003 RCIM_Constant>
          {arriveTangent=10.4171, interpMode=RCIM_Cubic, leaveTangent=10.4171, time=0.0709, value=0.0294}
          {arriveTangent=4.2588, interpMode=RCIM_Cubic, leaveTangent=4.2588, time=0.1413, value=1.5081}
          {interpMode=RCIM_Cubic, leaveTangent=0.4859, time=1, value=3.9862}
         ]}}})
    )) {isSetParameters=true}
   &ParticleState()
   &AccelerationForce(Acceleration=$Particles.Acceleration)
   &SolveForcesAndVelocity(Clamp Velocity=false, Limit Acceleration=false)
   &ScaleColor(
    Scale Alpha=&FloatFromCurve(
     FloatCurve=DataInterface() {propertyValues {Curve {keys [
         {interpMode=RCIM_Cubic, tangentMode=RCTM_Break, time=0, value=0}
         {arriveTangent=15.623, interpMode=RCIM_Cubic, leaveTangent=-13.5483, tangentMode=RCTM_Break, time=0.5, value=2}
         {tangentMode=RCTM_User, time=1, value=0}
        ]}}}
     Scale Curve=&FN_FrontEndFloatModifier(Gameplay=0.2, Front End=0.3)
    )
    ScaleA=true
    ScaleRGB=false
   ) {module=ScaleColor001}
  ]
  rendererClasses [&NiagaraRibbonRendererProperties]
  renderers [{rendererClass=&NiagaraRibbonRendererProperties}]
 }
 Smoke_Mesh {
  emitterUpdateScript [
   &EmitterState(Inactive Response=::NewEnumerator1, Life Cycle Mode=::NewEnumerator1, Loop Duration=1, UseLoopDelay=true)
   &SpawnRate(SpawnRate=6, Use Spawn Probability=false)
  ]
  particleSpawnScript [
   /SetVariables(Particles.MeshRotationEuler:Vector3f=&RandomRangeVector(Minimum=<0 0 0>, Maximum=<0 0 1>, Evaluation Type=::NewEnumerator1)) {isSetParameters=true}
   /SetVariables(Particles.MeshOrientation=&MakeQuatFromEuler(Euler:Vector3f=$Particles.MeshRotationEuler, Scale:NiagaraFloat=360)) {isSetParameters=true}
   &InitializeParticle(
    Lifetime=&RandomRangeFloat(Minimum=0.4, Maximum=0.7, Evaluation Type=::NewEnumerator1)
    Mesh Scale=&RandomRangeVector(Minimum=<0.7 0.7 0.4>, Maximum=<1.05 1.05 0.25>, Evaluation Type=::NewEnumerator1)
    Mesh Scale Mode=::NewEnumerator4
    UsePositionOffset=false
    Write Material Random=true
   )
   &Color(Color=&MakeLinearColorFromVectorAndFloat("Vector (RGB)"=<1 1 1>, "Float (Alpha)"=0.08))
   &DynamicMaterialParameters(
    Index 0 Param 1=&RandomRangeFloat(Minimum=0, Maximum=1, Evaluation Type=::NewEnumerator1)
    Module.Param0WriteEnabled=true
    Module.Param1WriteEnabled=false
    Module.Param2WriteEnabled=false
    Module.Param3WriteEnabled=false
   )
  ]
  particleUpdateScript [
   &ParticleState()
   &SkeletalMeshLocation(
    Bone / Socket Index=&SequentialInt(Use Array Data:NiagaraBool=false, First Number:NiagaraInt32=0, Sequence:NiagaraDataInterfaceArrayInt32=?, Last Number:NiagaraInt32=1)
    Bone Sampling Mode=::NewEnumerator4
    MaskSpawnGroup=false
    Orientation Sampling=::NewEnumerator1
    Skeletal Mesh=DataInterface() {propertyValues {
      DefaultMesh=null
      ExcludeBoneName=root
      OptionalComponentClass=null
      PreviewMesh=null
      SkinningMode=SkinOnTheFly
      SoftSourceActor=null
      SourceMode=Default
      UvSetIndex=0
      WholeMeshLOD=-1
      bExcludeBone=false
      bReadDeformedGeometry=true
      bRequireCurrentFrameData=true
      ComponentTags []
      FilteredBones [hand_r,hand_l]
      FilteredSockets []
      MeshUserParameter {parameter {name=None, typeDefHandle {registeredTypeIndex=1}, varData []}}
      SamplingRegions []
     }}
   )
   &ScaleColor(
    Scale Alpha=&FloatFromCurve(FloatCurve=DataInterface() {propertyValues {Curve {keys [<at 0 0>,<at 0.31 1>,<at 1 0>]}}})
    ScaleA=true
    ScaleRGB=true
   )
   &DynamicMaterialParameters(
    Index 0 Param 2=1
    Index 0 Param 3=1
    Index 0 Param 4=1
    Module.Param0WriteEnabled=false
    Module.Param1WriteEnabled=true
    Module.Param2WriteEnabled=true
    Module.Param3WriteEnabled=true
   ) {module=DynamicMaterialParameters001}
  ]
  rendererClasses [&NiagaraMeshRendererProperties]
  renderers [{rendererClass=&NiagaraMeshRendererProperties, rendererIndex=0}]
 }
 leafclover {
  emitterUpdateScript [
   &EmitterState()
   &SpawnBurst_Instantaneous(Use Spawn Probability=false, UseLoopCountLimit=false)
   &SpawnBurst_Instantaneous(Spawn Group=1, Use Spawn Probability=false, UseLoopCountLimit=false) {module=SpawnBurst_Instantaneous001}
  ]
  particleSpawnScript [
   /SetVariables(Particles.VisibilityTag=&SpawnGroup()) {isSetParameters=true}
   &InitializeParticle(
    Mesh Scale=&MergeHelper_VectorSwitch(
     0=<0.45 0.45 0.68>
     1=&SplitVector(
      Double Component Vector Channels:ECascadeNiagaraTwoVectorChannels=::NewEnumerator0
      Double Component:NiagaraFloat=0.6
      Single Component:NiagaraFloat=0.8
     )
     2=?
     3=?
     Index=&SpawnGroup()
     Num=::NewEnumerator0
    )
    Mesh Scale Mode=::NewEnumerator4
    UsePositionOffset=false
    Write Material Random=true
   )
   &InitialMeshOrientation(
    Mesh Orientation Mode=::NewEnumerator1
    Rotation=&MergeHelper_VectorSwitch(0=<0 0 0.32>, 1=<0 0 0>, 2=?, 3=?, Index=&SpawnGroup(), Num=::NewEnumerator0)
    Use Rotation Vector=true
   )
  ]
  particleUpdateScript [
   &ParticleState(Kill Particles When Lifetime Has Elapsed=false, Loop Particles Lifetime=true)
   &Color(Color=&MakeLinearColorFromVectorAndFloat(
     "Vector (RGB)"=&VectorFromCurve(VectorCurve=DataInterface() {propertyValues {XCurve {keys [<at 0 1>,<at 1 1>]}, YCurve {keys [<at 0 1>,<at 1 1>]}, ZCurve {keys [<at 0 1>,<at 1 1>]}}})
     "Float (Alpha)"=&ClampFloat(Float=&FloatFromCurve(FloatCurve=DataInterface() {propertyValues {Curve {keys [{arriveTangent=-1, time=0, value=1}, <at 1 0>]}}}))
    ))
   &SkeletalMeshLocation(
    Bone / Socket Index=&SequentialInt(Use Array Data:NiagaraBool=false, First Number:NiagaraInt32=0, Sequence:NiagaraDataInterfaceArrayInt32=?, Last Number:NiagaraInt32=1)
    Bone Sampling Mode=::NewEnumerator4
    MaskSpawnGroup=false
    Orientation Sampling=::NewEnumerator1
    Sampled Position Offset Coordinate Space:ENiagaraSkelMeshTransforms=::NewEnumerator2
    Skeletal Mesh=$System.Skeletal Mesh Reader
   ) {module=SkeletalMeshLocation001}
   &DynamicMaterialParameters(
    Index 0 Param 1=&RandomRangeFloat(Minimum=0, Maximum=20, Evaluation Type=::NewEnumerator1)
    Index 0 Param 2=-15
    Index 0 Param 3=1
    Index 0 Param 4=1
    Module.Param0WriteEnabled=true
    Module.Param1WriteEnabled=true
    Module.Param2WriteEnabled=false
    Module.Param3WriteEnabled=false
   )
  ]
  rendererClasses [&NiagaraMeshRendererProperties]
  renderers [
   {rendererClass=&NiagaraMeshRendererProperties, rendererIndex=0}
   {rendererClass=&NiagaraMeshRendererProperties, rendererIndex=1}
  ]
 }
}
summary {
 systemName=NS_Contrail_LuckOfTheIrish_Converted
 emitters [
  {emitterName=Particle_Emitter004, rendererClasses [&NiagaraMeshRendererProperties]}
  {emitterName=Ribbon_Rainbow, rendererClasses [&NiagaraRibbonRendererProperties]}
  {emitterName=Glows, rendererClasses [&NiagaraSpriteRendererProperties]}
  {
   emitterName=Clover_Leaf_Star
   rendererClasses [&NiagaraMeshRendererProperties, &NiagaraSpriteRendererProperties]
  }
  {emitterName=Smoke_Mesh, rendererClasses [&NiagaraMeshRendererProperties]}
  {emitterName=leafclover, rendererClasses [&NiagaraMeshRendererProperties]}
 ]
 userVariables [
  User.bIsFrontEnd {type=NiagaraBool}
  User.bIsFrontEndPreview {type=NiagaraBool}
  User.New Skeletal Mesh {
   type=NiagaraDataInterfaceSkeletalMesh
   defaultValue {
    struct=NiagaraExt_VariableValue_DataInterface
    value {dataInterface=&self:NiagaraDataInterfaceSkeletalMesh_0}
   }
  }
  User.ParaGlideLeanAlpha {type=NiagaraFloat, defaultValue=0, description=""}
 ]
}
