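# Niagara skydiving contrail cosmetic that trails rainbow ribbons from the character's skeletal mesh.
# Only RainbowRibbons is active: it spawns four ribbons, each partition sampling its own bone or socket, with short-lived particles pushed along Z.
# ParaGlideLeanAlpha blends spawn offset and velocity, with separate front-end and gameplay values; curves widen the ribbons and fade alpha in and out.
# PresentRibbon and the near-duplicate Ribbon_Left_Foot emitters are disabled.
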NiagaraSystem /BRCosmetics/Athena/Effects/Particles/Skydiving/NS_Contrail_Rainbow
dependencies {
 usedDataInterfaces [&NiagaraDataInterfaceCurve]
 usedDynamicInputs [
  &SetBoolByFloatComparison
  &self:Abs_Z_Axis
  &TransformVector
  &Lerp_Float
  &NormalizeVector
  &Multiply_Vector
 ]
 usedModules [
  &SystemState
  &self:SystemSpawnScript.SystemScriptSource.SystemScriptGraph.NiagaraNodeAssignment_0.SetVariables_#1
  &Timeline
  &self:SystemSpawnScript.SystemScriptSource.SystemScriptGraph.NiagaraNodeAssignment_2.SetVariables_#2
  &self:SystemSpawnScript.SystemScriptSource.SystemScriptGraph.NiagaraNodeAssignment_1.SetVariables_#3
 ]
}
emitters {
 PresentRibbon {
  bEnabled=false
  emitterUpdateScript [
   &EmitterState()
   &FN_SpawnRibbon_v2(Number of Ribbons=4, Spawn Mode=::NewEnumerator0, SpawnRate=100)
  ]
  particleSpawnScript [
   &FN_InitializeRibbon_v3(
    Enable Skeletal Position=true
    Offset Axis:ENiagaraOrientationAxis=::ZAxis
    Ribbon Width:NiagaraFloat=&RandomRangeFloat(Minimum=9, Maximum=6)
    Sampled Position Coordinate Space:ENiagaraCoordinateSpace=::Local
    Sampled Rotation Coordinate Space:ENiagaraCoordinateSpace=::Local
    Sampled Velocity Coordinate Space:ENiagaraCoordinateSpace=::Local
   ) {module=FN_InitializeRibbon_v3001}
   &SkeletalMeshLocation(
    Apply Sampled Orientation to Mesh Particles=true
    Bone / Socket Index=$Particles.SpawnGroup
    Bone Sampling Mode=::NewEnumerator4
    Inherit Velocity=false
    Interpolate Mesh Sampling=false
    MaskSpawnGroup=false
    Sampled Position Offset=&MakeCustomVectorFromBool(
     Bool=&SetBoolByIntComparison(A=$Particles.SpawnGroup, B=1, Comparison Type=::NewEnumerator0)
     True Vector=<0 0 0>
     False Vector=<0 0 15>
    )
    Sampled Position Offset Coordinate Space:ENiagaraMeshTransforms=::NewEnumerator1
    Skeletal Mesh=$User.SkeletalMesh
   )
   &AddVelocity(Velocity=&FN_FrontEndVector3Modifier(Gameplay=&Lerp_Vector(A=<0 0 555>, B=<1 1 999>, Alpha=$User.ParaGlideLeanAlpha), Front End=<0 0 0>))
   &DynamicMaterialParameters(
    Index 0 Param 1=0.1
    Module.Param0WriteEnabled=true
    Module.Param1WriteEnabled=false
    Module.Param2WriteEnabled=false
    Module.Param3WriteEnabled=false
   ) {module=DynamicMaterialParameters002}
  ]
  particleUpdateScript [
   &Drag(Drag=3, Rotational Drag=0, Use Linear Drag=true, Use Rotational Drag=true)
   &GravityForce(Coordinate Space=::Local, Gravity=<0 0 500>)
   &SolveForcesAndVelocity(Clamp Velocity=false, Limit Acceleration=false)
   &FN_UpdateRibbon(
    Enable Skeletal Position=true
    Initial Color:LinearColor=$Particles.Initial.Color
    Initial RibbonWidth:NiagaraFloat=$Particles.Initial.RibbonWidth
   )
   &ScaleRibbonWidth(Ribbon Width Scale=&FloatFromCurve(FloatCurve=DataInterface() {propertyValues {Curve {keys [{tangentMode=RCTM_User, time=0, value=1}, {tangentMode=RCTM_User, time=0.5, value=2}]}}})) {enabled=false}
   &ScaleColor(
    Scale Alpha=&FloatFromCurve(FloatCurve=DataInterface() {propertyValues {Curve {keys [<at 0.1 0 RCIM_Cubic>,<at 0.2977 0.5 RCIM_Cubic>,<at 1 0>]}}})
    ScaleA=true
    ScaleRGB=false
   ) {module=ScaleColor001}
   &ScaleRibbonWidth(Ribbon Width Scale=&Multiply_Float(
     A=1
     B=&FloatFromCurve(FloatCurve=DataInterface() {propertyValues {Curve {keys [
          <at -0.0035 -0.0003 RCIM_Constant>
          {arriveTangent=10.4171, interpMode=RCIM_Cubic, leaveTangent=10.4171, time=0.0709, value=0.0294}
          {arriveTangent=4.2588, interpMode=RCIM_Cubic, leaveTangent=4.2588, time=0.1413, value=1.508}
          {interpMode=RCIM_Cubic, leaveTangent=0.4859, time=1, value=3.9862}
         ]}}})
    )) {module=ScaleRibbonWidth001}
  ]
  rendererClasses [&NiagaraRibbonRendererProperties]
  renderers [{rendererClass=&NiagaraRibbonRendererProperties}]
 }
 RainbowRibbons {
  emitterUpdateScript [
   &EmitterState()
   &FN_SpawnRibbon_v2(Number of Ribbons=4, Spawn Mode=::NewEnumerator0, SpawnRate=66)
  ]
  particleSpawnScript [
   &PartitionParticles(Number Of Partitions=4)
   &InitializeParticle(
    Color Mode=::NewEnumerator0
    Lifetime=0.3
    Ribbon Width=&Multiply_Float(A=5, B=2)
    Ribbon Width Mode=::NewEnumerator1
    UsePositionOffset=false
    Write Material Random=true
   )
   &SkeletalMeshLocation(
    Bone / Socket Index=$Particles.PartitionParticles.Partition
    Bone Sampling Mode=::NewEnumerator4
    MaskSpawnGroup=false
    Sampled Position Offset=&FN_FrontEndVector3Modifier(
     Gameplay=&Lerp_Vector(A=<0 0 -33>, B=<1 1 -80>, Alpha=$User.ParaGlideLeanAlpha)
     Front End=&Lerp_Vector(A=<0 0 -45>, B=<1 1 -70>, Alpha=$User.ParaGlideLeanAlpha)
    )
    Sampled Position Offset Coordinate Space:ENiagaraSkelMeshTransforms=::NewEnumerator1
    Skeletal Mesh=$User.New Skeletal Mesh
   ) {module=SkeletalMeshLocation001}
   /SetVariables(Particles.Acceleration:Vector3f=&RandomRangeVector(Minimum=<-3 -3 0>, Maximum=<3 3 0>)) {isSetParameters=true, module=SetVariables_#4}
   &AddVelocity(Velocity=&FN_FrontEndVector3Modifier(
     Gameplay=&Lerp_Vector(A=<0 0 750>, B=<1 1 1200>, Alpha=$User.ParaGlideLeanAlpha)
     Front End=&Lerp_Vector(A=<0 0 700>, B=<1 1 999>, Alpha=$User.ParaGlideLeanAlpha)
    ))
  ]
  particleUpdateScript [
   /SetVariables(Particles.ScaledRibbonWidth:NiagaraFloat=&Multiply_Float(
     A=$Particles.RibbonWidth
     B=&FloatFromCurve(FloatCurve=DataInterface() {propertyValues {Curve {keys [
          <at -0.0035 -0.0003 RCIM_Constant>
          {arriveTangent=10.4171, interpMode=RCIM_Cubic, leaveTangent=10.4171, time=0.0709, value=0.0294}
          {arriveTangent=4.2588, interpMode=RCIM_Cubic, leaveTangent=4.2588, time=0.1413, value=1.5081}
          {interpMode=RCIM_Cubic, leaveTangent=0.4859, time=1, value=3.9862}
         ]}}})
    )) {isSetParameters=true, module=SetVariables_#5}
   &ParticleState()
   &DynamicMaterialParameters(
    Index 0 Param 1=0.1
    Module.Param0WriteEnabled=true
    Module.Param1WriteEnabled=false
    Module.Param2WriteEnabled=false
    Module.Param3WriteEnabled=false
   )
   &AccelerationForce(Acceleration=$Particles.Acceleration)
   &SolveForcesAndVelocity(Clamp Velocity=false, Limit Acceleration=false)
   &ScaleColor(
    Scale Alpha=&FloatFromCurve(
     FloatCurve=DataInterface() {propertyValues {Curve {keys [
         {interpMode=RCIM_Cubic, tangentMode=RCTM_Break, time=0, value=0}
         {arriveTangent=15.623, interpMode=RCIM_Cubic, leaveTangent=-13.5483, tangentMode=RCTM_Break, time=0.5, value=2}
         {tangentMode=RCTM_User, time=1, value=0}
        ]}}}
     Scale Curve=&FN_FrontEndFloatModifier(Gameplay=0.4, Front End=1)
    )
    ScaleA=true
    ScaleRGB=false
   ) {module=ScaleColor001}
  ]
  rendererClasses [&NiagaraRibbonRendererProperties]
  renderers [{rendererClass=&NiagaraRibbonRendererProperties}]
 }
 Ribbon_Left_Foot {
  bEnabled=false
  emitterUpdateScript [
   &EmitterState()
   &FN_SpawnRibbon_v2(Number of Ribbons=4, Spawn Mode=::NewEnumerator0, SpawnRate=66)
  ]
  particleSpawnScript [
   &PartitionParticles(Number Of Partitions=4)
   &InitializeParticle(
    Color Mode=::NewEnumerator0
    Lifetime=0.3
    Ribbon Width=&Multiply_Float(A=5, B=2)
    Ribbon Width Mode=::NewEnumerator1
    UsePositionOffset=false
    Write Material Random=true
   )
   &SkeletalMeshLocation(
    Bone / Socket Index=$Particles.PartitionParticles.Partition
    Bone Sampling Mode=::NewEnumerator4
    MaskSpawnGroup=false
    Sampled Position Offset=&FN_FrontEndVector3Modifier(Gameplay=<0 0 -33>, Front End=<0 0 -45>)
    Sampled Position Offset Coordinate Space:ENiagaraSkelMeshTransforms=::NewEnumerator1
    Skeletal Mesh=$User.New Skeletal Mesh
   ) {module=SkeletalMeshLocation001}
   /SetVariables(Particles.Acceleration:Vector3f=&RandomRangeVector(Minimum=<-3 -3 0>, Maximum=<3 3 0>)) {isSetParameters=true, module=SetVariables_#4}
   &AddVelocity(Velocity=&FN_FrontEndVector3Modifier(
     Gameplay=&Lerp_Vector(A=<0 0 750>, B=<1 1 1200>, Alpha=$User.ParaGlideLeanAlpha)
     Front End=&Lerp_Vector(A=<0 0 700>, B=<1 1 999>, Alpha=$User.ParaGlideLeanAlpha)
    ))
  ]
  particleUpdateScript [
   /SetVariables(Particles.ScaledRibbonWidth:NiagaraFloat=&Multiply_Float(
     A=$Particles.RibbonWidth
     B=&FloatFromCurve(FloatCurve=DataInterface() {propertyValues {Curve {keys [
          <at -0.0035 -0.0003 RCIM_Constant>
          {arriveTangent=10.4171, interpMode=RCIM_Cubic, leaveTangent=10.4171, time=0.0709, value=0.0294}
          {arriveTangent=4.2588, interpMode=RCIM_Cubic, leaveTangent=4.2588, time=0.1413, value=1.5081}
          {interpMode=RCIM_Cubic, leaveTangent=0.4859, time=1, value=3.9862}
         ]}}})
    )) {isSetParameters=true, module=SetVariables_#5}
   &ParticleState()
   &DynamicMaterialParameters(
    Index 0 Param 1=0.1
    Module.Param0WriteEnabled=true
    Module.Param1WriteEnabled=false
    Module.Param2WriteEnabled=false
    Module.Param3WriteEnabled=false
   )
   &AccelerationForce(Acceleration=$Particles.Acceleration)
   &SolveForcesAndVelocity(Clamp Velocity=false, Limit Acceleration=false)
   &ScaleColor(
    Scale Alpha=&FloatFromCurve(
     FloatCurve=DataInterface() {propertyValues {Curve {keys [
         {interpMode=RCIM_Cubic, tangentMode=RCTM_Break, time=0, value=0}
         {arriveTangent=15.623, interpMode=RCIM_Cubic, leaveTangent=-13.5483, tangentMode=RCTM_Break, time=0.5, value=2}
         {tangentMode=RCTM_User, time=1, value=0}
        ]}}}
     Scale Curve=&FN_FrontEndFloatModifier(Gameplay=0.4, Front End=1)
    )
    ScaleA=true
    ScaleRGB=false
   ) {module=ScaleColor001}
  ]
  rendererClasses [&NiagaraRibbonRendererProperties]
  renderers [{rendererClass=&NiagaraRibbonRendererProperties}]
 }
}
summary {
 systemName=NS_Contrail_Rainbow
 emitters [
  {bEnabled=false, emitterName=PresentRibbon, rendererClasses [&NiagaraRibbonRendererProperties]}
  {bEnabled=false, emitterName=Ribbon_Left_Foot, rendererClasses [&NiagaraRibbonRendererProperties]}
  {emitterName=RainbowRibbons, rendererClasses [&NiagaraRibbonRendererProperties]}
 ]
 userVariables [
  User.bIsFrontEnd {type=NiagaraBool}
  User.bIsFrontEndPreview {type=NiagaraBool}
  User.New Skeletal Mesh {
   type=NiagaraDataInterfaceSkeletalMesh
   defaultValue {
    struct=NiagaraExt_VariableValue_DataInterface
    value {dataInterface=&self:NiagaraDataInterfaceSkeletalMesh_2}
   }
  }
  User.ParaGlideLeanAlpha {type=NiagaraFloat, defaultValue=0, description=""}
  User.PlayerSpeed {type=NiagaraFloat, defaultValue=0, description=""}
  User.RibbonIDOffset {type=NiagaraInt32}
  User.RibbonSpawnRate {type=NiagaraFloat, defaultValue=40, description=""}
  User.SkeletalMesh {
   type=NiagaraDataInterfaceSkeletalMesh
   defaultValue {
    struct=NiagaraExt_VariableValue_DataInterface
    value {dataInterface=&self:NiagaraDataInterfaceSkeletalMesh_0}
   }
  }
  User.TrailAlpha {type=NiagaraFloat, defaultValue=1, description=""}
  User.V2 {type=NiagaraFloat, defaultValue=0, description=""}
 ]
}
