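# Skydiving contrail for the SlurpMonster cosmetic, apparently converted from Cascade, attached to skeletal mesh bones or sockets.
# Four persistent meshes follow their sockets, short-lived ribbons trail from two, and sprite blobs spray out by rate and distance moved, with a distance-only Dive variant.
# Sequential-index MergeHelper switches let merged emitters vary size and velocity per socket.
# Ribbon velocity applies only in front end and bends with ParaGlideLeanAlpha.
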NiagaraSystem /BRCosmetics/Athena/Effects/Particles/Skydiving/NS_Contrail_SlurpMonster_Converted
dependencies {
 usedDataInterfaces [
  &NiagaraDataInterfaceCurve
  &NiagaraDataInterfaceVector2DCurve
  &NiagaraDataInterfaceVectorCurve
  &NiagaraDataInterfaceArrayInt32
  &NiagaraDataInterfaceSkeletalMesh
 ]
 usedDynamicInputs []
 usedModules [
  &self:SystemSpawnScript.SystemScriptSource.SystemScriptGraph.NiagaraNodeAssignment_0.SetVariables_#1
  &SystemState
 ]
}
emitters {
 Ribbon_Left_Hand {
  emitterUpdateScript [
   &EmitterState(Inactive Response=::NewEnumerator1, Life Cycle Mode=::NewEnumerator1, Loop Duration=0.1, UseLoopDelay=true)
   &SpawnRate(SpawnRate=100, Use Spawn Probability=false)
  ]
  particleSpawnScript [
   &FN_SetSpawnGroup()
   /SetVariables(Particles.RibbonID=/Niagara/DynamicInputs/Meshes/MakeRibbonIDFromSpawnGroup()) {isSetParameters=true}
   &InitializeParticle(
    Color=<rgba 1 1 1 0.7>
    Lifetime=0.35
    Ribbon Width=&Multiply_Float(A=24, B=2)
    Ribbon Width Mode=::NewEnumerator1
    UsePositionOffset=false
    Write Material Random=true
   )
   &SkeletalMeshLocation(
    Bone / Socket Index=/CascadeToNiagaraConverter/NiagaraScripts/DynamicInputScripts/SequentialInt(Use Array Data:NiagaraBool=false, First Number:NiagaraInt32=0, Sequence:NiagaraDataInterfaceArrayInt32=?, Last Number:NiagaraInt32=1)
    Bone Sampling Mode=::NewEnumerator4
    MaskSpawnGroup=false
    Orientation Sampling=::NewEnumerator1
    Sampled Position Offset=<0 0 -30>
    Sampled Position Offset Coordinate Space:ENiagaraSkelMeshTransforms=::NewEnumerator2
    Skeletal Mesh=$System.Skeletal Mesh Reader
   )
   &AddVelocity(
    Velocity=&Lerp_Vector(A=<0 0 1000>, B=<0 -300 1000>, Alpha=$User.ParaGlideLeanAlpha)
    Velocity Speed Scale=&FN_FrontEndFloatModifier(Gameplay=0, Front End=1)
   )
   &ApplyInitialForces()
  ]
  particleUpdateScript [
   &ParticleState()
   &ScaleColor(
    Scale Alpha=&FloatFromCurve(FloatCurve=DataInterface() {propertyValues {Curve {keys [
         {arriveTangent=1.6128, time=0, value=-1}
         {arriveTangent=0.3841, leaveTangent=0.3841, time=0.0919, value=0.5083}
         {arriveTangent=0.0031, leaveTangent=0.0031, time=0.4361, value=0.9969}
         {arriveTangent=-1, leaveTangent=-1, time=0.8111, value=1}
         <at 1 0>
        ]}}})
    ScaleA=true
    ScaleRGB=true
   )
   &DynamicMaterialParameters(
    Index 0 Param 1=0.02
    Index 0 Param 2=&FloatFromCurve(FloatCurve=DataInterface() {propertyValues {Curve {keys [{arriveTangent=-1.7534, interpMode=RCIM_Cubic, leaveTangent=-1.7534, tangentMode=RCTM_User, time=0, value=1}, <at 0.9971 0.2583>]}}})
    Index 0 Param 3=1
    Index 0 Param 4=1
    Module.Index0Float0Write=true
    Module.Index0Float1Write=true
    Module.Index0Float2Write=true
    Module.Index0Float3Write=true
   ) {module=DynamicMaterialParameters001}
   &SolveForcesAndVelocity(Clamp Velocity=false, Limit Acceleration=false)
  ]
  rendererClasses [&NiagaraRibbonRendererProperties]
  renderers [{rendererClass=&NiagaraRibbonRendererProperties}]
 }
 base_hand_L {
  emitterUpdateScript [
   &EmitterState(
    Inactive Response=::NewEnumerator1
    Life Cycle Mode=::NewEnumerator1
    Loop Behavior=::NewEnumerator1
    Loop Duration Mode=::NewEnumerator1
    UseLoopDelay=false
   )
   &SpawnBurst_Instantaneous(Spawn Count=4, Use Spawn Probability=false, UseLoopCountLimit=false)
  ]
  particleSpawnScript [
   &InitializeParticle(
    Color=<rgba 1 1 1 0.5>
    Lifetime=2
    Mesh Scale=&MergeHelper_VectorSwitch(
     0=<1.3 1.3 1.3>
     1=<1.3 1.3 1.3>
     2=<1 1 1>
     3=<1 1 1>
     Index=&SequentialInt(
      Use Array Data:NiagaraBool=false
      First Number:NiagaraInt32=0
      Sequence:NiagaraDataInterfaceArrayInt32=DataInterface() {propertyValues {IntData []}}
      Last Number:NiagaraInt32=3
     )
     Num=::NewEnumerator2
    )
    Mesh Scale Mode=::NewEnumerator4
    UsePositionOffset=false
    Write Material Random=true
   )
   &DynamicMaterialParameters(
    Index 0 Param 1=&RandomRangeFloat(Minimum=0, Maximum=1, Evaluation Type=::NewEnumerator1)
    Index 0 Param 2=1
    Index 0 Param 3=1
    Index 0 Param 4=1
    Module.Param0WriteEnabled=true
    Module.Param1WriteEnabled=true
    Module.Param2WriteEnabled=true
    Module.Param3WriteEnabled=true
   )
  ]
  particleUpdateScript [
   &ParticleState(Kill Particles When Lifetime Has Elapsed=false, Let Infinitely Lived Particles Die When Emitter Deactivates=true)
   &SkeletalMeshLocation(
    Bone / Socket Index=&SequentialInt(Use Array Data:NiagaraBool=false, First Number:NiagaraInt32=0, Sequence:NiagaraDataInterfaceArrayInt32=?, Last Number:NiagaraInt32=3)
    Bone Sampling Mode=::NewEnumerator4
    MaskSpawnGroup=false
    Orientation Sampling=::NewEnumerator1
    Skeletal Mesh=$System.Skeletal Mesh Reader
   )
  ]
  rendererClasses [&NiagaraMeshRendererProperties]
  renderers [{rendererClass=&NiagaraMeshRendererProperties, rendererIndex=0}]
 }
 hand_Blobs_L {
  emitterUpdateScript [
   &EmitterState(Inactive Response=::NewEnumerator1, Life Cycle Mode=::NewEnumerator1, Loop Duration=1, UseLoopDelay=true)
   &SpawnRate(SpawnRate=40, Use Spawn Probability=false)
   &SpawnPerUnit(
    Module.Use Max Movement Threshold=true
    Module.Use Movement Tolerance=true
    Spawn Spacing:NiagaraFloat=&Divide_Float(A=1, B=&Divide_Float(A=0.5, B=500))
    Use Spawn Probability=false
   )
  ]
  particleSpawnScript [
   &InitializeParticle(
    Color=<rgba 1 1 1 0.6>
    Lifetime=&RandomRangeFloat(Minimum=0.28, Maximum=0.4, Evaluation Type=::NewEnumerator1)
    Sprite Size=&MergeHelper_Vector2Switch(
     0=&RandomRangeVector2D(Minimum=<5 5>, Maximum=<6 10>, Evaluation Type=::NewEnumerator1)
     1=&RandomRangeVector2D(Minimum=<5 5>, Maximum=<6 10>, Evaluation Type=::NewEnumerator1)
     2=&RandomRangeVector2D(Minimum=<3 3>, Maximum=<4 8>, Evaluation Type=::NewEnumerator1)
     3=&RandomRangeVector2D(Minimum=<3 3>, Maximum=<4 8>, Evaluation Type=::NewEnumerator1)
     4=?
     Index=&SequentialInt(Use Array Data:NiagaraBool=false, First Number:NiagaraInt32=0, Sequence:NiagaraDataInterfaceArrayInt32=?, Last Number:NiagaraInt32=3)
     Num=::NewEnumerator2
    )
    Sprite Size Mode=::NewEnumerator4
    UsePositionOffset=false
    Write Material Random=true
   )
   &DynamicMaterialParameters(
    Index 0 Param 1=&RandomRangeFloat(Minimum=0, Maximum=1, Evaluation Type=::NewEnumerator1)
    Index 0 Param 2=1
    Index 0 Param 3=1
    Index 0 Param 4=1
    Module.Param0WriteEnabled=true
    Module.Param1WriteEnabled=true
    Module.Param2WriteEnabled=true
    Module.Param3WriteEnabled=true
   )
   &SkeletalMeshLocation(
    Bone / Socket Index=&SequentialInt(Use Array Data:NiagaraBool=false, First Number:NiagaraInt32=0, Sequence:NiagaraDataInterfaceArrayInt32=?, Last Number:NiagaraInt32=3)
    Bone Sampling Mode=::NewEnumerator4
    MaskSpawnGroup=false
    Orientation Sampling=::NewEnumerator1
    Skeletal Mesh=$System.Skeletal Mesh Reader
   )
   &InheritVelocity(Inherited Velocity Amount Scale=<-0.01 -0.01 -0.01>, Inherited Velocity Speed Limit=0, Speed Limit=false, UseSpeedThreshold=false)
   &AddVelocity(
    Rotation Coordinate Space=::World
    Velocity=&MergeHelper_VectorSwitch(
     0=&RandomRangeVector(Minimum=<0 0 30>, Maximum=<0 0 50>, Evaluation Type=::NewEnumerator1)
     1=&RandomRangeVector(Minimum=<0 0 30>, Maximum=<0 0 50>, Evaluation Type=::NewEnumerator1)
     2=&RandomRangeVector(Minimum=<0 0 15>, Maximum=<0 0 30>, Evaluation Type=::NewEnumerator1)
     3=&RandomRangeVector(Minimum=<0 0 15>, Maximum=<0 0 30>, Evaluation Type=::NewEnumerator1)
     Index=&SequentialInt(Use Array Data:NiagaraBool=false, First Number:NiagaraInt32=0, Sequence:NiagaraDataInterfaceArrayInt32=?, Last Number:NiagaraInt32=3)
     Num=::NewEnumerator2
    )
   )
   &ShapeLocation(
    Hemisphere Distribution=<0.5 1>
    MaskSpawnGroup=false
    Non Uniform Scale=<1 1 0>
    Override Seed=false
    Rotation Mode=::NewEnumerator2
    Sphere Radius=&MergeHelper_FloatSwitch(
     0=5
     1=5
     2=1
     3=1
     Index=&SequentialInt(Use Array Data:NiagaraBool=false, First Number:NiagaraInt32=0, Sequence:NiagaraDataInterfaceArrayInt32=?, Last Number:NiagaraInt32=3)
     Num=::NewEnumerator2
    )
    Sphere Surface Distribution=1
    UseOwnerScale=true
    Yaw / Pitch / Roll=<0.5 0 0>
   )
   &AddVelocity(
    Cone Angle=10
    Cone Angle Mode:ENiagara_AngleInput=::NewEnumerator1
    Cone Axis=&MergeHelper_VectorSwitch(
     0=<0.3 -0.3 1>
     1=<-0.3 -0.3 1>
     2=<0 0 1>
     3=<0 0 1>
     Index=&SequentialInt(Use Array Data:NiagaraBool=false, First Number:NiagaraInt32=0, Sequence:NiagaraDataInterfaceArrayInt32=?, Last Number:NiagaraInt32=3)
     Num=::NewEnumerator2
    )
    Module.UseConeFalloff=false
    Module.UseVelDistribution=false
    Override Seed=false
    Velocity Mode=::NewEnumerator2
    Velocity Speed=&MergeHelper_FloatSwitch(
     0=&RandomRangeFloat(Minimum=100, Maximum=300, Evaluation Type=::NewEnumerator1)
     1=&RandomRangeFloat(Minimum=100, Maximum=300, Evaluation Type=::NewEnumerator1)
     2=&RandomRangeFloat(Minimum=60, Maximum=100, Evaluation Type=::NewEnumerator1)
     3=&RandomRangeFloat(Minimum=60, Maximum=100, Evaluation Type=::NewEnumerator1)
     Index=&SequentialInt(Use Array Data:NiagaraBool=false, First Number:NiagaraInt32=0, Sequence:NiagaraDataInterfaceArrayInt32=?, Last Number:NiagaraInt32=3)
     Num=::NewEnumerator2
    )
   ) {module=AddVelocity001}
   &CameraOffset(Camera Offset Amount=5)
  ]
  particleUpdateScript [
   &ParticleState()
   &ScaleSpriteSize(
    Scale Factor=&Vector2DFromCurve(
     Vector2Curve=DataInterface() {propertyValues {XCurve {keys [<at 0.011 2.0221 RCIM_Cubic>,<at 0.9896 0.4674>]}, YCurve {keys [<at 0.011 4.0793 RCIM_Cubic>,<at 0.9896 0.2186>]}}}
     ScaleCurve=<1 1>
    )
    Scale Sprite Size Mode=::NewEnumerator2
   )
   &ScaleColor(
    Scale Alpha=&FloatFromCurve(FloatCurve=DataInterface() {propertyValues {Curve {keys [
         <at 0.003 0.0108 RCIM_Cubic>
         {arriveTangent=0.2323, interpMode=RCIM_Cubic, leaveTangent=0.2323, tangentMode=RCTM_User, time=0.0517, value=0.9978}
         {arriveTangent=-0.9627, leaveTangent=-0.9627, time=0.9582, value=0.9627}
         <at 1 0>
        ]}}})
    Scale RGB=&VectorFromCurve(VectorCurve=DataInterface() {propertyValues {XCurve {keys [<at 0 1>,<at 1 1>]}, YCurve {keys [<at 0 1>,<at 1 1>]}, ZCurve {keys [<at 0 1>,<at 1 1>]}}})
    ScaleA=true
    ScaleRGB=false
   )
   &SolveForcesAndVelocity(Clamp Velocity=false, Limit Acceleration=false)
  ]
  rendererClasses [&NiagaraSpriteRendererProperties]
  renderers [{rendererClass=&NiagaraSpriteRendererProperties}]
 }
 hand_Blobs_R_Dive {
  emitterUpdateScript [
   &EmitterState(Inactive Response=::NewEnumerator1, Life Cycle Mode=::NewEnumerator1, Loop Duration=1, UseLoopDelay=true)
   &SpawnPerUnit(
    Module.Use Max Movement Threshold=true
    Module.Use Movement Tolerance=true
    Movement Tolerance=0.1
    Spawn Spacing:NiagaraFloat=&Divide_Float(A=1, B=&Divide_Float(A=3, B=500))
    Use Spawn Probability=false
   )
  ]
  particleSpawnScript [
   &InitializeParticle(
    Color=<rgba 1 1 1 0.6>
    Lifetime=&RandomRangeFloat(Minimum=0.28, Maximum=0.4, Evaluation Type=::NewEnumerator1)
    Sprite Size=&RandomRangeVector2D(Minimum=<4 4>, Maximum=<8 7>, Evaluation Type=::NewEnumerator1)
    Sprite Size Mode=::NewEnumerator4
    UsePositionOffset=false
    Write Material Random=true
   )
   &DynamicMaterialParameters(
    Index 0 Param 1=&RandomRangeFloat(Minimum=0, Maximum=1, Evaluation Type=::NewEnumerator1)
    Index 0 Param 2=1
    Index 0 Param 3=1
    Index 0 Param 4=1
    Module.Index0Float0Write=true
    Module.Index0Float1Write=true
    Module.Index0Float2Write=true
    Module.Index0Float3Write=true
   )
   &SkeletalMeshLocation(
    Bone / Socket Index=&SequentialInt(Use Array Data:NiagaraBool=false, First Number:NiagaraInt32=0, Sequence:NiagaraDataInterfaceArrayInt32=?, Last Number:NiagaraInt32=1)
    Bone Sampling Mode=::NewEnumerator4
    MaskSpawnGroup=false
    Orientation Sampling=::NewEnumerator1
    Skeletal Mesh=$System.Skeletal Mesh Reader
   )
   &InheritVelocity(Inherited Velocity Amount Scale=<-0.01 -0.01 -0.01>, Inherited Velocity Speed Limit=0, Speed Limit=false, UseSpeedThreshold=false)
   &AddVelocity(
    Rotation Coordinate Space=::World
    Velocity=&RandomRangeVector(Minimum=<0 0 30>, Maximum=<0 0 50>, Evaluation Type=::NewEnumerator1)
   )
   &ShapeLocation(
    Hemisphere Distribution=<0.5 1>
    MaskSpawnGroup=false
    Non Uniform Scale=<1 1 0>
    Override Seed=false
    Rotation Mode=::NewEnumerator2
    Sphere Radius=5
    Sphere Surface Distribution=1
    UseOwnerScale=true
    Yaw / Pitch / Roll=<0.5 0 0>
   )
   &AddVelocity(
    Cone Angle=10
    Cone Angle Mode:ENiagara_AngleInput=::NewEnumerator1
    Cone Axis=&MergeHelper_VectorSwitch(
     0=<0.5 -0.6 1>
     1=<-0.3 -0.6 1>
     2=?
     3=?
     Index=&SequentialInt(Use Array Data:NiagaraBool=false, First Number:NiagaraInt32=0, Sequence:NiagaraDataInterfaceArrayInt32=?, Last Number:NiagaraInt32=1)
     Num=::NewEnumerator0
    )
    Module.UseConeFalloff=false
    Module.UseVelDistribution=false
    Override Seed=false
    Velocity Mode=::NewEnumerator2
    Velocity Speed=&RandomRangeFloat(Minimum=100, Maximum=300, Evaluation Type=::NewEnumerator1)
   ) {module=AddVelocity001}
   &CameraOffset(Camera Offset Amount=5)
  ]
  particleUpdateScript [
   &ParticleState()
   &ScaleSpriteSize(
    Scale Factor=&Vector2DFromCurve(
     Vector2Curve=DataInterface() {propertyValues {XCurve {keys [<at 0.0434 2.0221 RCIM_Cubic>,<at 0.9896 0.211>]}, YCurve {keys [<at 0.0434 7.9239 RCIM_Cubic>,<at 0.9896 0.2186>]}}}
     ScaleCurve=<1 1>
    )
    Scale Sprite Size Mode=::NewEnumerator2
   )
   &ScaleColor(
    Scale Alpha=&FloatFromCurve(FloatCurve=DataInterface() {propertyValues {Curve {keys [
         <at 0.003 0.0108 RCIM_Cubic>
         {arriveTangent=0.2323, interpMode=RCIM_Cubic, leaveTangent=0.2323, tangentMode=RCTM_User, time=0.0517, value=0.9978}
         {arriveTangent=-0.9627, leaveTangent=-0.9627, time=0.9582, value=0.9627}
         <at 1 0>
        ]}}})
    Scale RGB=&VectorFromCurve(VectorCurve=DataInterface() {propertyValues {XCurve {keys [<at 0 1>,<at 1 1>]}, YCurve {keys [<at 0 1>,<at 1 1>]}, ZCurve {keys [<at 0 1>,<at 1 1>]}}})
    ScaleA=true
    ScaleRGB=true
   )
   &SolveForcesAndVelocity(Clamp Velocity=false, Limit Acceleration=false)
  ]
  rendererClasses [&NiagaraSpriteRendererProperties]
  renderers [{rendererClass=&NiagaraSpriteRendererProperties}]
 }
}
summary {
 systemName=NS_Contrail_SlurpMonster_Converted
 emitters [
  {emitterName=hand_Blobs_L, rendererClasses [&NiagaraSpriteRendererProperties]}
  {emitterName=hand_Blobs_R_Dive, rendererClasses [&NiagaraSpriteRendererProperties]}
  {emitterName=base_hand_L, rendererClasses [&NiagaraMeshRendererProperties]}
  {emitterName=Ribbon_Left_Hand, rendererClasses [&NiagaraRibbonRendererProperties]}
 ]
 userVariables [
  User.bIsFrontEnd {type=NiagaraBool}
  User.bIsFrontEndPreview {type=NiagaraBool}
  User.ParaGlideLeanAlpha {type=NiagaraFloat, defaultValue=0, description=""}
 ]
}
