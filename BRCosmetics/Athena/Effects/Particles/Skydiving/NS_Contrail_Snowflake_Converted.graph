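# Skydiving cosmetic Niagara contrail, converted from Cascade, that trails snowflakes, powder snow and sparkles from skeletal mesh bones.
# Two snowflake emitters and a flipbook powder emitter run alongside a Source emitter, whose particles seed two noise-driven, flickering sparkle emitters through particle reads.
# ParaGlideLeanAlpha blends velocities, and front-end modifiers switch some rates, offsets and velocities between gameplay and menu.
# Emitters "8" and Minimal are disabled.
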NiagaraSystem /BRCosmetics/Athena/Effects/Particles/Skydiving/NS_Contrail_Snowflake_Converted
dependencies {
 usedDataInterfaces [
  &NiagaraDataInterfaceCurve
  &NiagaraDataInterfaceVectorCurve
  &NiagaraDataInterfaceSpriteRendererInfo
  &NiagaraDataInterfaceParticleRead
 ]
 usedDynamicInputs []
 usedModules [
  &self:SystemSpawnScript.SystemScriptSource.SystemScriptGraph.NiagaraNodeAssignment_0.SetVariables_#1
  &SystemState
 ]
}
emitters {
 8 {
  bEnabled=false
  emitterUpdateScript [
   &EmitterState(Inactive Response=::NewEnumerator1, Life Cycle Mode=::NewEnumerator1, Loop Duration=1, UseLoopDelay=true)
   &SpawnRate(SpawnRate=30, Use Spawn Probability=false) {module=SpawnRate001}
   &SpawnRate(SpawnRate=70, Use Spawn Probability=false) {enabled=false}
  ]
  particleSpawnScript [
   &InitializeParticle(
    Lifetime=&RandomRangeFloat(Minimum=0.12, Maximum=0.07, Evaluation Type=::NewEnumerator1)
    Sprite Size=&Vector2DFromFloat(Value=&RandomRangeFloat(Minimum=10, Maximum=10, Evaluation Type=::NewEnumerator1))
    Sprite Size Mode=::NewEnumerator4
    UsePositionOffset=false
    Write Material Random=true
   )
   &InheritVelocity(
    Inherited Velocity Amount Scale=&RandomRangeVector(Minimum=<-0.1 -0.1 -0.1>, Maximum=<-0.1 -0.1 -0.18>, Evaluation Type=::NewEnumerator1)
    Inherited Velocity Speed Limit=0
    Speed Limit=false
    UseSpeedThreshold=false
   )
   &SkeletalMeshLocation(
    Bone / Socket Index=/CascadeToNiagaraConverter/NiagaraScripts/DynamicInputScripts/SequentialInt(Use Array Data:NiagaraBool=false, First Number:NiagaraInt32=0, Sequence:NiagaraDataInterfaceArrayInt32=?, Last Number:NiagaraInt32=3)
    Bone Sampling Mode=::NewEnumerator4
    MaskSpawnGroup=false
    Orientation Sampling=::NewEnumerator1
    Sampled Position Offset=<0 0 -10>
    Sampled Position Offset Coordinate Space:ENiagaraSkelMeshTransforms=::NewEnumerator2
    Skeletal Mesh=$System.Skeletal Mesh Reader
   ) {module=SkeletalMeshLocation001}
   &ShapeLocation(MaskSpawnGroup=false, Override Seed=false, Sphere Radius=20, UseOwnerScale=true) {enabled=false, module=ShapeLocation001}
   &AddVelocity(Module.Use Falloff Distance=false, Override Seed=false, Velocity Mode=::NewEnumerator1, Velocity Speed=50) {module=AddVelocity001}
   &ShapeLocation(
    Cone Angle=900
    Cone Inner Angle=50
    Cone Length=250
    MaskSpawnGroup=false
    Override Seed=false
    Shape Primitive=::NewEnumerator5
    UseOwnerScale=true
   ) {enabled=false}
   /CascadeToNiagaraConverter/NiagaraScripts/ModuleScripts/FixShapeVector() {enabled=false, module=FixShapeVector001}
   &FixShapeVector() {enabled=false}
   &AddVelocity(Velocity=&Multiply_VectorByFloat(
     Vector=$Output.FixShapeVector.ShapeVector
     Float=&RandomRangeFloat(Minimum=5, Maximum=6, Evaluation Type=::NewEnumerator1)
    )) {enabled=false}
  ]
  particleUpdateScript [
   &ParticleState()
   &Color(Color=&MakeLinearColorFromVectorAndFloat(
     "Vector (RGB)"=&VectorFromCurve(VectorCurve=DataInterface() {propertyValues {XCurve {keys [<at 0 1>,<at 1 1>]}, YCurve {keys [<at 0 1>,<at 1 1>]}, ZCurve {keys [<at 0 1>,<at 1 1>]}}})
     "Float (Alpha)"=&ClampFloat(Float=&FloatFromCurve(FloatCurve=DataInterface() {propertyValues {Curve {keys [{arriveTangent=-1, time=0, value=1}, <at 1 0>]}}}))
    ))
   &Drag(Use Linear Drag=true, Use Rotational Drag=true)
   &SolveForcesAndVelocity(Clamp Velocity=false, Limit Acceleration=false)
  ]
  rendererClasses [&NiagaraSpriteRendererProperties]
  renderers [{rendererClass=&NiagaraSpriteRendererProperties}]
 }
 FN_SparkleDust {
  emitterUpdateScript [
   &EmitterState()
   &SpawnParticlesFromOtherEmitter(
    Attribute Reader=DataInterface() {propertyValues {EmitterBinding {bindingMode=Other, emitterName=Source}}}
    Spawn Rate=&FN_FrontEndFloatModifier(Gameplay=70, Front End=50)
    Spawn Rate Per Particle Cap=350
    Use Maximum Age=false
    Use Minimum Age=false
   )
   &SpawnRate(SpawnRate=16, Use Spawn Probability=false) {enabled=false}
   /Game/Effects/Niagara/Modules/Skydiving/FN_DivingAttributes/FN_EmitterDivingAttributes_v1(SkeletalMesh:NiagaraDataInterfaceSkeletalMesh=$User.SkeletalMesh) {enabled=false}
  ]
  particleSpawnScript [
   &FN_SetSpawnGroup_v2(Number of Spawn Groups=2) {enabled=false}
   /Niagara/Modules/Spawn/Initialization/InitializeParticle(
    Lifetime=&UniformRangedFloat(Minimum=0.3, Maximum=0.4)
    Mesh Scale=&Multiply_Vector(
     A=<1 1 1>
     B=&VectorFromFloat(Value=&UniformRangedFloat(Minimum=0.2, Maximum=1))
    )
    Module.Write Color=false
    Module.Write Lifetime=true
    Module.Write Mass=true
    Module.Write Material Random=true
    Module.Write Position=false
    Module.Write Scale=false
    Module.Write SpriteRotation=true
    Module.Write SpriteSize=true
    Position=&SimulationPosition()
    Sprite Rotation=&UniformAOrBFloat(A=0, B=360)
    Sprite Size=&Vector2DFromFloat(Value=&UniformRangedFloat(Minimum=5, Maximum=3))
   )
   &SkeletalMeshLocation(
    Bone / Socket Index=$Particles.SpawnGroup
    Bone Sampling Mode=::NewEnumerator4
    Inherit Velocity=false
    MaskSpawnGroup=false
    Skeletal Mesh=$User.SkeletalMesh
   )
   &Color(Color=&MakeLinearColorFromVectorAndFloat("Vector (RGB)"=<0.7994 1.2264 2>, "Float (Alpha)"=1))
   &AddVelocity(
    Coordinate Space=::World
    Scale Added Velocity=&Lerp_Vector(A=<1 1 1>, B=<1 1 3>, Alpha=$Emitter.Dive)
    Velocity=&UniformRangedVector(Minimum=<0 0 125>, Maximum=<1 1 225>)
   ) {enabled=false}
   &AddVelocity(Module.Use Falloff Distance=false, Override Seed=false, Velocity Mode=::NewEnumerator1, Velocity Speed=3) {module=AddVelocity001}
   &ShapeLocation(MaskSpawnGroup=false, Override Seed=false, Sphere Radius=1, UseOwnerScale=true)
   /Niagara/Modules/Spawn/Location/SphereLocation(MaskSpawnGroup=false, Module.Override Randomness=false, Sphere Radius=12.5, Surface Only=false) {enabled=false}
   &SampleParticlesFromOtherEmitter()
  ]
  particleUpdateScript [
   &ParticleState()
   &Drag(Drag=0.5, Use Linear Drag=true, Use Rotational Drag=false) {enabled=false}
   /Niagara/Modules/Update/Forces/CurlNoiseForce(
    Noise Frequency:NiagaraFloat=&UniformRangedFloat(Minimum=0.5, Maximum=1)
    Noise Strength=&UniformRangedFloat(Minimum=300, Maximum=500)
    Pan Noise=false
    Pan Noise Field=<1 1 0>
   ) {enabled=false}
   &VectorNoiseForce(Force Amount=450, Module.Use Random Noise Time Interval=false)
   /Niagara/Modules/Update/Forces/V2/CurlNoiseForce(
    Bias Noise=false
    Noise Frequency=&FN_FrontEndFloatModifier(Gameplay=200, Front End=300)
    Noise Strength=&FN_FrontEndFloatModifier(Gameplay=400, Front End=400)
    Pan Noise=false
   ) {module=CurlNoiseForce003}
   &SolveForcesAndVelocity(Clamp Velocity=false, Limit Acceleration=false)
   &ScaleSpriteSize(
    Scale Sprite Size Mode=::NewEnumerator0
    Uniform Scale Factor=&FloatFromCurve(FloatCurve=DataInterface() {propertyValues {Curve {keys [
         {arriveTangent=-6.1151, interpMode=RCIM_Cubic, leaveTangent=-6.1151, tangentMode=RCTM_User, time=0, value=1}
         {arriveTangent=0.2969, interpMode=RCIM_Cubic, leaveTangent=0.2969, tangentMode=RCTM_User, time=0.1354, value=0.2886}
         {interpMode=RCIM_Cubic, tangentMode=RCTM_User, time=0.2671, value=0.9417}
         {arriveTangent=0.057, interpMode=RCIM_Cubic, leaveTangent=0.057, time=0.412, value=0.1231}
         {arriveTangent=0.021, interpMode=RCIM_Cubic, leaveTangent=0.021, time=0.5725, value=0.9591}
         {arriveTangent=-0.1312, interpMode=RCIM_Cubic, leaveTangent=-0.1312, time=0.7149, value=0.1294}
         {arriveTangent=0.2476, interpMode=RCIM_Cubic, leaveTangent=0.2476, time=0.8709, value=0.9199}
         {interpMode=RCIM_Cubic, tangentMode=RCTM_User, time=1, value=0.2}
        ]}}})
   )
   &ScaleColor(
    Scale Alpha=&FloatFromCurve(FloatCurve=DataInterface() {propertyValues {Curve {keys [<at 0 0>,<at 0.12 1>,<at 0.6 0>,<at 1 0>]}}})
    Scale RGB=&VectorFromCurve(VectorCurve=DataInterface() {propertyValues {
       XCurve {keys [<at 0 3>,<at 0.15 4>,<at 0.5 0>]}
       YCurve {keys [<at 0 1.8323>,<at 0.15 1.2168>,<at 0.5 2.379>]}
       ZCurve {keys [<at 0 0>,<at 0.15 0>,<at 0.5 8>]}
      }})
    ScaleA=true
    ScaleRGB=true
   ) {enabled=false}
   &ScaleColor(
    Scale Alpha=&FloatFromCurve(FloatCurve=DataInterface() {propertyValues {Curve {keys [
         {arriveTangent=9.5296, interpMode=RCIM_Cubic, leaveTangent=9.5296, tangentMode=RCTM_User, time=0, value=0}
         {arriveTangent=0.5611, interpMode=RCIM_Cubic, leaveTangent=0.5611, tangentMode=RCTM_User, time=0.1068, value=0.9998}
         {arriveTangent=-0.6793, interpMode=RCIM_Cubic, leaveTangent=-0.6793, tangentMode=RCTM_User, time=0.2571, value=0.4216}
         {arriveTangent=0.7365, interpMode=RCIM_Cubic, leaveTangent=0.7365, tangentMode=RCTM_User, time=0.4096, value=0.9267}
         {arriveTangent=-1.0314, interpMode=RCIM_Cubic, leaveTangent=-1.0314, time=0.5903, value=0.346}
         {arriveTangent=0.4271, interpMode=RCIM_Cubic, leaveTangent=0.4271, tangentMode=RCTM_User, time=0.6717, value=0.6564}
         {arriveTangent=-1.1148, interpMode=RCIM_Cubic, leaveTangent=-1.1148, tangentMode=RCTM_User, time=0.7718, value=0.0893}
         <at 1 0>
        ]}}})
    Scale RGB=<3 3 3>
    ScaleA=true
    ScaleRGB=true
   ) {module=ScaleColor001}
  ]
  rendererClasses [&NiagaraSpriteRendererProperties]
  renderers [{rendererClass=&NiagaraSpriteRendererProperties}]
 }
 Minimal {
  bEnabled=false
  emitterUpdateScript [
   &EmitterState()
   &SpawnParticlesFromOtherEmitter(
    Attribute Reader=DataInterface() {propertyValues {EmitterBinding {bindingMode=Other, emitterName=Source}}}
    Spawn Rate=&FN_FrontEndFloatModifier(Gameplay=200, Front End=150)
    Spawn Rate Per Particle Cap=350
    Use Maximum Age=false
    Use Minimum Age=false
   )
  ]
  particleSpawnScript [
   /Niagara/Modules/Spawn/Initialization/V2/InitializeParticle(
    Color Mode=::NewEnumerator0
    Lifetime=&RandomRangeFloat(Minimum=0.1, Maximum=0.05)
    Sprite Size Mode=::NewEnumerator3
    Uniform Sprite Size=&RandomRangeFloat(Minimum=1.5, Maximum=3)
    UsePositionOffset=false
    Write Material Random=true
   )
   &AddVelocity(Module.Use Falloff Distance=false, Override Seed=false, Velocity Mode=::NewEnumerator1)
   &ShapeLocation(MaskSpawnGroup=false, Override Seed=false, Sphere Radius=50, UseOwnerScale=true)
   &Color(Color=&MakeLinearColorFromVectorAndFloat("Vector (RGB)"=<0.7994 1.2264 2>, "Float (Alpha)"=1))
   &SampleParticlesFromOtherEmitter()
   &AddVelocity(Velocity=&FN_FrontEndVector3Modifier(Gameplay=<0 0 0>, Front End=<0 0 300>)) {module=AddVelocity001}
  ]
  particleUpdateScript [
   &ParticleState()
   &ScaleSpriteSize(
    Scale Sprite Size Mode=::NewEnumerator0
    Uniform Scale Factor=&FloatFromCurve(FloatCurve=DataInterface() {propertyValues {Curve {keys [
         {arriveTangent=-6.1151, interpMode=RCIM_Cubic, leaveTangent=-6.1151, tangentMode=RCTM_User, time=0, value=1}
         {arriveTangent=0.2969, interpMode=RCIM_Cubic, leaveTangent=0.2969, tangentMode=RCTM_User, time=0.1354, value=0.2886}
         {interpMode=RCIM_Cubic, tangentMode=RCTM_User, time=0.2671, value=0.9417}
         {arriveTangent=0.057, interpMode=RCIM_Cubic, leaveTangent=0.057, time=0.412, value=0.1231}
         {arriveTangent=0.021, interpMode=RCIM_Cubic, leaveTangent=0.021, time=0.5725, value=0.9591}
         {arriveTangent=-0.1312, interpMode=RCIM_Cubic, leaveTangent=-0.1312, time=0.7149, value=0.1294}
         {arriveTangent=0.2476, interpMode=RCIM_Cubic, leaveTangent=0.2476, time=0.8709, value=0.9199}
         {interpMode=RCIM_Cubic, tangentMode=RCTM_User, time=1, value=0.2}
        ]}}})
   )
   &VortexForce(UseOriginOffset=false, Vortex Axis=<1 1 0>) {enabled=false}
   &VectorNoiseForce(Force Amount=350, Module.Use Random Noise Time Interval=false)
   &InheritSourceMovement(Applied Position Scale=<-0.2 -0.2 0.2>, ApplyForce=false, ApplyPosition=true, ApplyPositionFalloff=false) {enabled=false, module=InheritSourceMovement001}
   &InheritSourceMovement(
    Applied Position Scale=&Multiply_VectorByFloat(
     Vector=<0.2 0.2 1>
     Float=&FloatFromCurve(
      FloatCurve=DataInterface() {propertyValues {Curve {
         defaultValue=1
         keys [
          {interpMode=RCIM_Cubic, tangentMode=RCTM_User, time=0.1678, value=-0.1}
          {interpMode=RCIM_Cubic, tangentMode=RCTM_User, time=0.3286, value=0.4}
         ]
        }}}
      Scale Curve=-1
     )
    )
    ApplyForce=false
    ApplyPosition=true
    ApplyPositionFalloff=false
   )
   &VortexForce(UseOriginOffset=false, Vortex Axis=<0.5 0.5 1>) {enabled=false, module=VortexForce001}
   &CurlNoiseForce(Bias Noise=false, Noise Frequency=300, Noise Strength=450, Pan Noise=false)
   &SolveForcesAndVelocity(Clamp Velocity=false, Limit Acceleration=false)
   &ScaleColor(
    Scale Alpha=&FloatFromCurve(FloatCurve=DataInterface() {propertyValues {Curve {keys [
         {arriveTangent=9.5296, interpMode=RCIM_Cubic, leaveTangent=9.5296, tangentMode=RCTM_User, time=0, value=0}
         {arriveTangent=0.5611, interpMode=RCIM_Cubic, leaveTangent=0.5611, tangentMode=RCTM_User, time=0.1068, value=0.9998}
         {arriveTangent=-0.6793, interpMode=RCIM_Cubic, leaveTangent=-0.6793, tangentMode=RCTM_User, time=0.2571, value=0.4216}
         {arriveTangent=0.7365, interpMode=RCIM_Cubic, leaveTangent=0.7365, tangentMode=RCTM_User, time=0.4096, value=0.9267}
         {arriveTangent=-1.0314, interpMode=RCIM_Cubic, leaveTangent=-1.0314, time=0.5903, value=0.346}
         {arriveTangent=0.4271, interpMode=RCIM_Cubic, leaveTangent=0.4271, tangentMode=RCTM_User, time=0.6717, value=0.6564}
         {arriveTangent=-1.1148, interpMode=RCIM_Cubic, leaveTangent=-1.1148, tangentMode=RCTM_User, time=0.7718, value=0.0893}
         <at 1 0>
        ]}}})
    Scale RGB=<7 7 7>
    ScaleA=true
    ScaleRGB=true
   )
  ]
  rendererClasses [&NiagaraSpriteRendererProperties]
  renderers [{rendererClass=&NiagaraSpriteRendererProperties}]
 }
 Minimal001 {
  emitterUpdateScript [
   &EmitterState()
   &SpawnParticlesFromOtherEmitter(
    Attribute Reader=DataInterface() {propertyValues {EmitterBinding {bindingMode=Other, emitterName=Source}}}
    Spawn Rate=&FN_FrontEndFloatModifier(Gameplay=200, Front End=150)
    Spawn Rate Per Particle Cap=350
    Use Maximum Age=false
    Use Minimum Age=false
   )
  ]
  particleSpawnScript [
   &InitializeParticle(
    Color Mode=::NewEnumerator0
    Lifetime=&RandomRangeFloat(Minimum=0.1, Maximum=0.05)
    Sprite Size Mode=::NewEnumerator3
    Uniform Sprite Size=&RandomRangeFloat(Minimum=1.5, Maximum=3)
    UsePositionOffset=false
    Write Material Random=true
   )
   &AddVelocity(Module.Use Falloff Distance=false, Override Seed=false, Velocity Mode=::NewEnumerator1)
   &ShapeLocation(MaskSpawnGroup=false, Override Seed=false, Sphere Radius=50, UseOwnerScale=true)
   &Color(Color=&MakeLinearColorFromVectorAndFloat("Vector (RGB)"=<0.7994 1.2264 2>, "Float (Alpha)"=1))
   &SampleParticlesFromOtherEmitter()
   &AddVelocity(Velocity=&FN_FrontEndVector3Modifier(Gameplay=<0 0 0>, Front End=<0 0 300>)) {module=AddVelocity001}
  ]
  particleUpdateScript [
   &ParticleState()
   &ScaleSpriteSize(
    Scale Sprite Size Mode=::NewEnumerator0
    Uniform Scale Factor=&FloatFromCurve(FloatCurve=DataInterface() {propertyValues {Curve {keys [
         {arriveTangent=-6.1151, interpMode=RCIM_Cubic, leaveTangent=-6.1151, tangentMode=RCTM_User, time=0, value=1}
         {arriveTangent=0.2969, interpMode=RCIM_Cubic, leaveTangent=0.2969, tangentMode=RCTM_User, time=0.1354, value=0.2886}
         {interpMode=RCIM_Cubic, tangentMode=RCTM_User, time=0.2671, value=0.9417}
         {arriveTangent=0.057, interpMode=RCIM_Cubic, leaveTangent=0.057, time=0.412, value=0.1231}
         {arriveTangent=0.021, interpMode=RCIM_Cubic, leaveTangent=0.021, time=0.5725, value=0.9591}
         {arriveTangent=-0.1312, interpMode=RCIM_Cubic, leaveTangent=-0.1312, time=0.7149, value=0.1294}
         {arriveTangent=0.2476, interpMode=RCIM_Cubic, leaveTangent=0.2476, time=0.8709, value=0.9199}
         {interpMode=RCIM_Cubic, tangentMode=RCTM_User, time=1, value=0.2}
        ]}}})
   )
   &VortexForce(UseOriginOffset=false, Vortex Axis=<1 1 0>) {enabled=false}
   &VectorNoiseForce(Force Amount=350, Module.Use Random Noise Time Interval=false)
   &InheritSourceMovement(Applied Position Scale=<-0.2 -0.2 0.2>, ApplyForce=false, ApplyPosition=true, ApplyPositionFalloff=false) {enabled=false, module=InheritSourceMovement001}
   &InheritSourceMovement(
    Applied Position Scale=&Multiply_VectorByFloat(
     Vector=<0.2 0.2 1>
     Float=&FloatFromCurve(
      FloatCurve=DataInterface() {propertyValues {Curve {
         defaultValue=1
         keys [
          {interpMode=RCIM_Cubic, tangentMode=RCTM_User, time=0.1678, value=-0.1}
          {interpMode=RCIM_Cubic, tangentMode=RCTM_User, time=0.3286, value=0.4}
         ]
        }}}
      Scale Curve=-1
     )
    )
    ApplyForce=false
    ApplyPosition=true
    ApplyPositionFalloff=false
   )
   &VortexForce(UseOriginOffset=false, Vortex Axis=<0.5 0.5 1>) {enabled=false, module=VortexForce001}
   &CurlNoiseForce(Bias Noise=false, Noise Frequency=300, Noise Strength=450, Pan Noise=false)
   &SolveForcesAndVelocity(Clamp Velocity=false, Limit Acceleration=false)
   &ScaleColor(
    Scale Alpha=&FloatFromCurve(FloatCurve=DataInterface() {propertyValues {Curve {keys [
         {arriveTangent=9.5296, interpMode=RCIM_Cubic, leaveTangent=9.5296, tangentMode=RCTM_User, time=0, value=0}
         {arriveTangent=0.5611, interpMode=RCIM_Cubic, leaveTangent=0.5611, tangentMode=RCTM_User, time=0.1068, value=0.9998}
         {arriveTangent=-0.6793, interpMode=RCIM_Cubic, leaveTangent=-0.6793, tangentMode=RCTM_User, time=0.2571, value=0.4216}
         {arriveTangent=0.7365, interpMode=RCIM_Cubic, leaveTangent=0.7365, tangentMode=RCTM_User, time=0.4096, value=0.9267}
         {arriveTangent=-1.0314, interpMode=RCIM_Cubic, leaveTangent=-1.0314, time=0.5903, value=0.346}
         {arriveTangent=0.4271, interpMode=RCIM_Cubic, leaveTangent=0.4271, tangentMode=RCTM_User, time=0.6717, value=0.6564}
         {arriveTangent=-1.1148, interpMode=RCIM_Cubic, leaveTangent=-1.1148, tangentMode=RCTM_User, time=0.7718, value=0.0893}
         <at 1 0>
        ]}}})
    Scale RGB=<7 7 7>
    ScaleA=true
    ScaleRGB=true
   )
  ]
  rendererClasses [&NiagaraSpriteRendererProperties]
  renderers [{rendererClass=&NiagaraSpriteRendererProperties}]
 }
 Powderdsnow {
  emitterUpdateScript [
   &EmitterState(Inactive Response=::NewEnumerator1, Life Cycle Mode=::NewEnumerator1, Loop Duration=1, UseLoopDelay=true)
   &SpawnRate(SpawnRate=50, Use Spawn Probability=false) {module=SpawnRate001}
   &SpawnPerUnit(Module.Use Max Movement Threshold=true, Module.Use Movement Tolerance=false, Use Spawn Probability=false) {enabled=false}
   &SpawnRate(SpawnRate=20, Use Spawn Probability=false) {enabled=false}
  ]
  particleSpawnScript [
   &InitializeParticle(
    Lifetime=&RandomRangeFloat(Minimum=0.35, Maximum=1, Evaluation Type=::NewEnumerator1)
    Sprite Rotation Angle=&RandomRangeFloat(Minimum=0, Maximum=1, Evaluation Type=::NewEnumerator1)
    Sprite Rotation Mode=::NewEnumerator1
    Sprite Size=&Vector2DFromFloat(Value=&RandomRangeFloat(Minimum=50, Maximum=100, Evaluation Type=::NewEnumerator1))
    Sprite Size Mode=::NewEnumerator4
    UsePositionOffset=false
    Write Material Random=true
   )
   &Color(Color=&MakeLinearColorFromVectorAndFloat("Vector (RGB)"=<1 1 1>, "Float (Alpha)"=0.3))
   /CascadeToNiagaraConverter/NiagaraScripts/ModuleScripts/AddVectorToPosition(Vector:Vector3f=&TransformVector(
     Vector=&RandomRangeVector(Minimum=<0 0 0>, Maximum=<0 -10 0>, Evaluation Type=::NewEnumerator1)
     Destination Space=::Simulation
    ))
   &InheritVelocity(
    Inherited Velocity Amount Scale=&Lerp_Vector(A=<0.1 0.1 -0.058>, B=<0 0 0.1>, Alpha=$User.ParaGlideLeanAlpha)
    Inherited Velocity Speed Limit=0
    Speed Limit=false
    UseSpeedThreshold=false
   ) {enabled=false, module=InheritVelocity001}
   &SkeletalMeshLocation(
    Bone / Socket Index=&SequentialInt(Use Array Data:NiagaraBool=false, First Number:NiagaraInt32=0, Sequence:NiagaraDataInterfaceArrayInt32=?, Last Number:NiagaraInt32=3)
    Bone Sampling Mode=::NewEnumerator4
    MaskSpawnGroup=false
    Orientation Sampling=::NewEnumerator1
    Skeletal Mesh=DataInterface() {propertyValues {
      DefaultMesh=null
      ExcludeBoneName=root
      OptionalComponentClass=null
      PreviewMesh=null
      SkinningMode=SkinOnTheFly
      SoftSourceActor=null
      SourceMode=Default
      UvSetIndex=0
      WholeMeshLOD=-1
      bExcludeBone=false
      bReadDeformedGeometry=true
      bRequireCurrentFrameData=true
      ComponentTags []
      FilteredBones [hand_l,hand_r,foot_r,foot_l]
      FilteredSockets []
      MeshUserParameter {parameter {name=None, typeDefHandle {registeredTypeIndex=1}, varData []}}
      SamplingRegions []
     }}
   ) {enabled=false}
  ]
  particleUpdateScript [
   &ParticleState()
   &SkeletalMeshLocation(
    Bone / Socket Index=&SequentialInt(Use Array Data:NiagaraBool=false, First Number:NiagaraInt32=0, Sequence:NiagaraDataInterfaceArrayInt32=?, Last Number:NiagaraInt32=3)
    Bone Sampling Mode=::NewEnumerator4
    MaskSpawnGroup=false
    Orientation Sampling=::NewEnumerator1
    Sampled Position Offset=<0 0 -10>
    Sampled Position Offset Coordinate Space:ENiagaraSkelMeshTransforms=::NewEnumerator2
    Skeletal Mesh=$System.Skeletal Mesh Reader
   ) {module=SkeletalMeshLocation001}
   &InheritVelocity(
    Inherited Velocity Amount Scale=&Lerp_Vector(A=<-0.1 -0.1 -0.058>, B=<0 0 -0.05>, Alpha=$User.ParaGlideLeanAlpha)
    Inherited Velocity Speed Limit=0
    Speed Limit=false
    UseSpeedThreshold=false
   )
   &ScaleSpriteSize(
    Scale Sprite Size Mode=::NewEnumerator0
    Uniform Scale Factor=&FloatFromCurve(FloatCurve=DataInterface() {propertyValues {Curve {keys [<at 0 0.25>,<at 1.0134 0.9931 RCIM_Cubic>]}}})
   )
   &ScaleColor(
    Scale Alpha=&FloatFromCurve(FloatCurve=DataInterface() {propertyValues {Curve {keys [<at 0 0>,<at 0.2 1>,<at 0.75 0>]}}}, Scale Curve=0.25)
    Scale RGB=&VectorFromCurve(VectorCurve=DataInterface() {propertyValues {XCurve {keys [<at 0 0.6317>,<at 1 0.7999>]}, YCurve {keys [<at 0 0.7238>,<at 1 0.8711>]}, ZCurve {keys [<at 0 1>,<at 1 1>]}}})
    ScaleA=true
    ScaleRGB=true
   )
   &SubUVAnimation(
    End Frame Range Override=62
    "Sprite Renderer "=DataInterface() {propertyValues {SpriteRenderer=&self:Powderdsnow_0.NiagaraSpriteRendererProperties_2}}
    UseEndFrame=false
    UseStartFrame=false
   )
   &CameraOffset(Camera Offset Amount=13)
   &SolveForcesAndVelocity(Clamp Velocity=false, Limit Acceleration=false)
  ]
  rendererClasses [&NiagaraSpriteRendererProperties]
  renderers [{rendererClass=&NiagaraSpriteRendererProperties}]
 }
 Source {
  emitterUpdateScript [
   &EmitterState(Inactive Response=::NewEnumerator1, Life Cycle Mode=::NewEnumerator1, Loop Duration=1, UseLoopDelay=true)
   &SpawnRate(SpawnRate=&FN_FrontEndFloatModifier(Gameplay=15, Front End=20), Use Spawn Probability=false)
  ]
  particleSpawnScript [
   /SetVariables(Particles.SpriteRotationRate=&RandomRangeFloat(Minimum=-0.3, Maximum=0.35, Evaluation Type=::NewEnumerator1)) {enabled=false, isSetParameters=true, module=SetVariables_#2}
   &InitializeParticle(
    Lifetime=&RandomRangeFloat(Minimum=0.42, Maximum=0.9, Evaluation Type=::NewEnumerator1)
    Sprite Rotation Angle=&RandomRangeFloat(Minimum=0, Maximum=1, Evaluation Type=::NewEnumerator1)
    Sprite Rotation Mode=::NewEnumerator1
    Sprite Size=&Vector2DFromFloat(Value=&RandomRangeFloat(Minimum=16, Maximum=8, Evaluation Type=::NewEnumerator1))
    Sprite Size Mode=::NewEnumerator4
    UsePositionOffset=false
    Write Material Random=true
   )
   &SkeletalMeshLocation(
    Bone / Socket Index=&SequentialInt(Use Array Data:NiagaraBool=false, First Number:NiagaraInt32=0, Sequence:NiagaraDataInterfaceArrayInt32=?, Last Number:NiagaraInt32=1)
    Bone Sampling Mode=::NewEnumerator4
    MaskSpawnGroup=false
    Orientation Sampling=::NewEnumerator1
    Skeletal Mesh=DataInterface() {propertyValues {
      DefaultMesh=null
      ExcludeBoneName=root
      OptionalComponentClass=null
      PreviewMesh=null
      SkinningMode=SkinOnTheFly
      SoftSourceActor=null
      SourceMode=Default
      UvSetIndex=0
      WholeMeshLOD=-1
      bExcludeBone=false
      bReadDeformedGeometry=true
      bRequireCurrentFrameData=true
      ComponentTags []
      FilteredBones [hand_r,hand_l]
      FilteredSockets []
      MeshUserParameter {parameter {name=None, typeDefHandle {registeredTypeIndex=1}, varData []}}
      SamplingRegions []
     }}
   ) {enabled=false}
   &SkeletalMeshLocation(
    Bone / Socket Index=&SequentialInt(Use Array Data:NiagaraBool=false, First Number:NiagaraInt32=0, Sequence:NiagaraDataInterfaceArrayInt32=?, Last Number:NiagaraInt32=3)
    Bone Sampling Mode=::NewEnumerator4
    MaskSpawnGroup=false
    Orientation Sampling=::NewEnumerator1
    Sampled Position Offset=&FN_FrontEndVector3Modifier(Gameplay=<0 0 -10>, Front End=<0 0 20>)
    Sampled Position Offset Coordinate Space:ENiagaraSkelMeshTransforms=::NewEnumerator2
    Skeletal Mesh=$System.Skeletal Mesh Reader
   ) {module=SkeletalMeshLocation001}
   &ShapeLocation(MaskSpawnGroup=false, Override Seed=false, Sphere Radius=15, UseOwnerScale=true)
   &FixShapeVector()
   &AddVelocity(Velocity=&Multiply_VectorByFloat(
     Vector=$Output.FixShapeVector.ShapeVector
     Float=&RandomRangeFloat(Minimum=1.5, Maximum=3, Evaluation Type=::NewEnumerator1)
    ))
   &AddVelocity(Velocity=&FN_FrontEndVector3Modifier(Gameplay=<0 0 35>, Front End=<0 0 150>)) {module=AddVelocity001}
   &InheritVelocity(
    Inherited Velocity Amount Scale=&Lerp_Vector(A=<0 0 -0.12>, B=<0 0 -0.05>, Alpha=$User.ParaGlideLeanAlpha)
    Inherited Velocity Speed Limit=0
    Speed Limit=false
    UseSpeedThreshold=false
   )
   &Color(Color=&MakeLinearColorFromVectorAndFloat("Vector (RGB)"=<0 1 50>, "Float (Alpha)"=0.5)) {enabled=false}
   &DynamicMaterialParameters(
    Index 0 Param 4=&RandomRangeFloat(Minimum=0, Maximum=4, Evaluation Type=::NewEnumerator1)
    Module.Param0WriteEnabled=false
    Module.Param1WriteEnabled=false
    Module.Param2WriteEnabled=false
    Module.Param3WriteEnabled=true
   ) {enabled=false, module=DynamicMaterialParameters001}
  ]
  particleUpdateScript [
   &ParticleState()
   &ScaleSpriteSize(
    Scale Sprite Size Mode=::NewEnumerator0
    Uniform Scale Factor=&FloatFromCurve(FloatCurve=DataInterface() {propertyValues {Curve {keys [
         {arriveTangent=4.141, interpMode=RCIM_Cubic, leaveTangent=4.141, tangentMode=RCTM_User, time=0, value=0}
         {arriveTangent=10.421, interpMode=RCIM_Cubic, leaveTangent=10.421, tangentMode=RCTM_User, time=0.156, value=1.142}
         {arriveTangent=-1.1596, interpMode=RCIM_Cubic, leaveTangent=-1.1596, tangentMode=RCTM_User, time=0.42, value=1.3772}
         <at 1 0 RCIM_Cubic>
        ]}}})
   )
   &SpriteRotationRate(Rotation Rate=&AngleConversion(
     Angle:NiagaraFloat=$Particles.SpriteRotationRate
     Angle Input:ENiagara_AngleInput=::NewEnumerator1
     Angle Output:ENiagara_AngleInput=::NewEnumerator0
    ))
   &Drag(Use Linear Drag=true, Use Rotational Drag=true)
   &CurlNoiseForce(Bias Noise=false, Noise Frequency=100, Noise Strength=400, Pan Noise=false) {enabled=false}
   &CurlNoiseForce(Bias Noise=false, Noise Frequency=200, Noise Strength=400, Pan Noise=false) {module=CurlNoiseForce002}
   &SolveForcesAndVelocity(Clamp Velocity=false, Limit Acceleration=false)
   &ScaleColor(
    Scale Alpha=&FloatFromCurve(FloatCurve=DataInterface() {propertyValues {Curve {keys [<at 0 0>,<at 0.2 1.2>,<at 0.52 1>,<at 1 0>]}}})
    Scale RGB=&VectorFromFloat(Value=&FloatFromCurve(FloatCurve=DataInterface() {propertyValues {Curve {keys [<at 0 1>,<at 1 1>]}}}))
    ScaleA=true
    ScaleRGB=true
   )
   &DynamicMaterialParameters(
    Index 0 Param 2=&FloatFromCurve(FloatCurve=DataInterface() {propertyValues {Curve {keys [<at 0 3.5>,<at 0.85 2.8>,<at 1 0>]}}})
    Index 0 Param 3=&FloatFromCurve(FloatCurve=DataInterface() {propertyValues {Curve {keys [<at 0 0>,<at 0.2 0>,<at 0.5 4>,<at 0.88 0>]}}})
    Module.Param0WriteEnabled=true
    Module.Param1WriteEnabled=true
    Module.Param2WriteEnabled=true
    Module.Param3WriteEnabled=false
   ) {enabled=false}
  ]
  rendererClasses [&NiagaraSpriteRendererProperties]
  renderers [{rendererClass=&NiagaraSpriteRendererProperties}]
 }
 snowflakes {
  emitterUpdateScript [
   &EmitterState(Inactive Response=::NewEnumerator1, Life Cycle Mode=::NewEnumerator1, Loop Duration=1, UseLoopDelay=true)
   &SpawnRate(SpawnRate=5, Use Spawn Probability=false)
   &SpawnRate(SpawnGroup=1, SpawnRate=5, Use Spawn Probability=false) {module=SpawnRate001}
  ]
  particleSpawnScript [
   /SetVariables(Particles.SpriteRotationRate=&RandomRangeFloat(Minimum=-0.1, Maximum=0.12, Evaluation Type=::NewEnumerator1)) {isSetParameters=true}
   &InitializeParticle(
    Lifetime=&RandomRangeFloat(Minimum=1.7, Maximum=2.5, Evaluation Type=::NewEnumerator1)
    Sprite Rotation Angle=&RandomRangeFloat(Minimum=0, Maximum=1, Evaluation Type=::NewEnumerator1)
    Sprite Rotation Mode=::NewEnumerator1
    Sprite Size=&Vector2DFromFloat(Value=&RandomRangeFloat(Minimum=35, Maximum=7.5, Evaluation Type=::NewEnumerator1))
    Sprite Size Mode=::NewEnumerator4
    UsePositionOffset=false
    Write Material Random=true
   )
   &SkeletalMeshLocation(
    Bone / Socket Index=&SequentialInt(Use Array Data:NiagaraBool=false, First Number:NiagaraInt32=0, Sequence:NiagaraDataInterfaceArrayInt32=?, Last Number:NiagaraInt32=1)
    Bone Sampling Mode=::NewEnumerator4
    MaskSpawnGroup=false
    Orientation Sampling=::NewEnumerator1
    Skeletal Mesh=DataInterface() {propertyValues {
      DefaultMesh=null
      ExcludeBoneName=root
      OptionalComponentClass=null
      PreviewMesh=null
      SkinningMode=SkinOnTheFly
      SoftSourceActor=null
      SourceMode=Default
      UvSetIndex=0
      WholeMeshLOD=-1
      bExcludeBone=false
      bReadDeformedGeometry=true
      bRequireCurrentFrameData=true
      ComponentTags []
      FilteredBones [hand_r,hand_l]
      FilteredSockets []
      MeshUserParameter {parameter {name=None, typeDefHandle {registeredTypeIndex=1}, varData []}}
      SamplingRegions []
     }}
   ) {enabled=false}
   &SkeletalMeshLocation(
    Bone / Socket Index=&SequentialInt(Use Array Data:NiagaraBool=false, First Number:NiagaraInt32=0, Sequence:NiagaraDataInterfaceArrayInt32=?, Last Number:NiagaraInt32=1)
    Bone Sampling Mode=::NewEnumerator4
    MaskSpawnGroup=false
    Orientation Sampling=::NewEnumerator1
    Sampled Position Offset=<0 0 -10>
    Sampled Position Offset Coordinate Space:ENiagaraSkelMeshTransforms=::NewEnumerator2
    Skeletal Mesh=$System.Skeletal Mesh Reader
   ) {module=SkeletalMeshLocation001}
   &ShapeLocation(MaskSpawnGroup=false, Override Seed=false, Sphere Radius=20, UseOwnerScale=true)
   &FixShapeVector()
   &AddVelocity(Velocity=&Multiply_VectorByFloat(
     Vector=$Output.FixShapeVector.ShapeVector
     Float=&RandomRangeFloat(Minimum=0.5, Maximum=1.5, Evaluation Type=::NewEnumerator1)
    ))
   &AddVelocity(Velocity=&FN_FrontEndVector3Modifier(Gameplay=&Lerp_Vector(A=<0 0 40>, B=<0 0 -70>, Alpha=$User.ParaGlideLeanAlpha), Front End=<0 0 150>)) {module=AddVelocity001}
   &InheritVelocity(
    Inherited Velocity Amount Scale=&RandomRangeVector(Minimum=<0.05 0.05 -0.2>, Maximum=<0 0 -0.05>, Evaluation Type=::NewEnumerator1)
    Inherited Velocity Speed Limit=0
    Speed Limit=false
    UseSpeedThreshold=false
   )
   &Color(Color=&MakeLinearColorFromVectorAndFloat("Vector (RGB)"=<1 1 1>, "Float (Alpha)"=1))
   &DynamicMaterialParameters(
    Index 0 Param 4=&RandomRangeFloat(Minimum=0, Maximum=4)
    Module.Param0WriteEnabled=false
    Module.Param1WriteEnabled=false
    Module.Param2WriteEnabled=false
    Module.Param3WriteEnabled=true
   ) {module=DynamicMaterialParameters001}
  ]
  particleUpdateScript [
   &ParticleState()
   &SpriteRotationRate(Rotation Rate=&AngleConversion(
     Angle:NiagaraFloat=$Particles.SpriteRotationRate
     Angle Input:ENiagara_AngleInput=::NewEnumerator1
     Angle Output:ENiagara_AngleInput=::NewEnumerator0
    ))
   &ScaleSpriteSize(
    Scale Sprite Size Mode=::NewEnumerator0
    Uniform Scale Factor=&FloatFromCurve(FloatCurve=DataInterface() {propertyValues {Curve {keys [<at 0 1>,<at 0.75 1.2>,<at 1 1>]}}})
   )
   &Drag(Use Linear Drag=true, Use Rotational Drag=true)
   &SolveForcesAndVelocity(Clamp Velocity=false, Limit Acceleration=false)
   &ScaleColor(
    Scale Alpha=&FloatFromCurve(FloatCurve=DataInterface() {propertyValues {Curve {keys [<at 0 1>,<at 0.2 1>,<at 0.2 1>,<at 0.91 0>]}}})
    Scale RGB=&VectorFromCurve(VectorCurve=DataInterface() {propertyValues {XCurve {keys [<at 0 1>,<at 1 1>]}, YCurve {keys [<at 0 1>,<at 1 1>]}, ZCurve {keys [<at 0 1>,<at 1 1>]}}})
    ScaleA=true
    ScaleRGB=true
   )
   &DynamicMaterialParameters(
    Index 0 Param 1=&FloatFromCurve(FloatCurve=DataInterface() {propertyValues {Curve {keys [<at 0 1.2>,<at 0.45 0>,<at 0.6 -0.2>,<at 1 1>]}}})
    Index 0 Param 2=&FloatFromCurve(FloatCurve=DataInterface() {propertyValues {Curve {keys [<at 0 3.5>,<at 0.85 2.8>,<at 1 0>]}}})
    Index 0 Param 3=&FloatFromCurve(FloatCurve=DataInterface() {propertyValues {Curve {keys [<at 0 0>,<at 0.2 0>,<at 0.5 4>,<at 0.88 0>]}}})
    Module.Param0WriteEnabled=true
    Module.Param1WriteEnabled=true
    Module.Param2WriteEnabled=true
    Module.Param3WriteEnabled=false
   )
  ]
  rendererClasses [&NiagaraSpriteRendererProperties]
  renderers [
   {rendererClass=&NiagaraSpriteRendererProperties}
   {rendererClass=&NiagaraSpriteRendererProperties, rendererIndex=1}
  ]
 }
 snowflakes01 {
  emitterUpdateScript [
   &EmitterState(Inactive Response=::NewEnumerator1, Life Cycle Mode=::NewEnumerator1, Loop Duration=1, UseLoopDelay=true)
   &SpawnRate(SpawnRate=11, Use Spawn Probability=false)
  ]
  particleSpawnScript [
   /SetVariables(Particles.SpriteRotationRate=&RandomRangeFloat(Minimum=-0.3, Maximum=0.35, Evaluation Type=::NewEnumerator1)) {isSetParameters=true, module=SetVariables_#2}
   &InitializeParticle(
    Lifetime=&RandomRangeFloat(Minimum=0.42, Maximum=0.9, Evaluation Type=::NewEnumerator1)
    Sprite Rotation Angle=&RandomRangeFloat(Minimum=0, Maximum=1, Evaluation Type=::NewEnumerator1)
    Sprite Rotation Mode=::NewEnumerator1
    Sprite Size=&Vector2DFromFloat(Value=&RandomRangeFloat(Minimum=16, Maximum=8, Evaluation Type=::NewEnumerator1))
    Sprite Size Mode=::NewEnumerator4
    UsePositionOffset=false
    Write Material Random=true
   )
   &SkeletalMeshLocation(
    Bone / Socket Index=&SequentialInt(Use Array Data:NiagaraBool=false, First Number:NiagaraInt32=0, Sequence:NiagaraDataInterfaceArrayInt32=?, Last Number:NiagaraInt32=1)
    Bone Sampling Mode=::NewEnumerator4
    MaskSpawnGroup=false
    Orientation Sampling=::NewEnumerator1
    Skeletal Mesh=DataInterface() {propertyValues {
      DefaultMesh=null
      ExcludeBoneName=root
      OptionalComponentClass=null
      PreviewMesh=null
      SkinningMode=SkinOnTheFly
      SoftSourceActor=null
      SourceMode=Default
      UvSetIndex=0
      WholeMeshLOD=-1
      bExcludeBone=false
      bReadDeformedGeometry=true
      bRequireCurrentFrameData=true
      ComponentTags []
      FilteredBones [hand_r,hand_l]
      FilteredSockets []
      MeshUserParameter {parameter {name=None, typeDefHandle {registeredTypeIndex=1}, varData []}}
      SamplingRegions []
     }}
   ) {enabled=false}
   &SkeletalMeshLocation(
    Bone / Socket Index=&SequentialInt(Use Array Data:NiagaraBool=false, First Number:NiagaraInt32=0, Sequence:NiagaraDataInterfaceArrayInt32=?, Last Number:NiagaraInt32=1)
    Bone Sampling Mode=::NewEnumerator4
    MaskSpawnGroup=false
    Orientation Sampling=::NewEnumerator1
    Sampled Position Offset=<0 0 -10>
    Sampled Position Offset Coordinate Space:ENiagaraSkelMeshTransforms=::NewEnumerator2
    Skeletal Mesh=$System.Skeletal Mesh Reader
   ) {module=SkeletalMeshLocation001}
   &ShapeLocation(MaskSpawnGroup=false, Override Seed=false, Sphere Radius=20, UseOwnerScale=true)
   &FixShapeVector()
   &AddVelocity(Velocity=&Multiply_VectorByFloat(
     Vector=$Output.FixShapeVector.ShapeVector
     Float=&RandomRangeFloat(Minimum=1.5, Maximum=3, Evaluation Type=::NewEnumerator1)
    ))
   &AddVelocity(Velocity=&FN_FrontEndVector3Modifier(Gameplay=&Lerp_Vector(A=<0 0 -35>, B=<0 0 -400>, Alpha=$User.ParaGlideLeanAlpha), Front End=<0 0 150>)) {module=AddVelocity001}
   &InheritVelocity(
    Inherited Velocity Amount Scale=&Lerp_Vector(A=<0.05 0.05 -0.12>, B=<0 0 -0.1>, Alpha=$User.ParaGlideLeanAlpha)
    Inherited Velocity Speed Limit=0
    Speed Limit=false
    UseSpeedThreshold=false
   )
   &Color(Color=&MakeLinearColorFromVectorAndFloat("Vector (RGB)"=<0 1 50>, "Float (Alpha)"=0.5))
   &DynamicMaterialParameters(
    Index 0 Param 4=&RandomRangeFloat(Minimum=0, Maximum=4)
    Module.Param0WriteEnabled=false
    Module.Param1WriteEnabled=false
    Module.Param2WriteEnabled=false
    Module.Param3WriteEnabled=true
   ) {module=DynamicMaterialParameters001}
  ]
  particleUpdateScript [
   &ParticleState()
   &ScaleSpriteSize(
    Scale Sprite Size Mode=::NewEnumerator0
    Uniform Scale Factor=&FloatFromCurve(FloatCurve=DataInterface() {propertyValues {Curve {keys [
         {arriveTangent=4.141, interpMode=RCIM_Cubic, leaveTangent=4.141, tangentMode=RCTM_User, time=0, value=0}
         {arriveTangent=10.421, interpMode=RCIM_Cubic, leaveTangent=10.421, tangentMode=RCTM_User, time=0.156, value=1.142}
         {arriveTangent=-1.1596, interpMode=RCIM_Cubic, leaveTangent=-1.1596, tangentMode=RCTM_User, time=0.42, value=1.3772}
         <at 1 0 RCIM_Cubic>
        ]}}})
   )
   &SpriteRotationRate(Rotation Rate=&AngleConversion(
     Angle:NiagaraFloat=$Particles.SpriteRotationRate
     Angle Input:ENiagara_AngleInput=::NewEnumerator1
     Angle Output:ENiagara_AngleInput=::NewEnumerator0
    ))
   &Drag(Use Linear Drag=true, Use Rotational Drag=true)
   &SolveForcesAndVelocity(Clamp Velocity=false, Limit Acceleration=false)
   &ScaleColor(
    Scale Alpha=&FloatFromCurve(FloatCurve=DataInterface() {propertyValues {Curve {keys [<at 0 0>,<at 0.2 1.2>,<at 0.52 1>,<at 1 0>]}}})
    Scale RGB=&VectorFromFloat(Value=&FloatFromCurve(FloatCurve=DataInterface() {propertyValues {Curve {keys [<at 0 1>,<at 1 1>]}}}))
    ScaleA=true
    ScaleRGB=true
   )
   &DynamicMaterialParameters(
    Index 0 Param 2=&FloatFromCurve(FloatCurve=DataInterface() {propertyValues {Curve {keys [<at 0 3.5>,<at 0.85 2.8>,<at 1 0>]}}})
    Index 0 Param 3=&FloatFromCurve(FloatCurve=DataInterface() {propertyValues {Curve {keys [<at 0 0>,<at 0.2 0>,<at 0.5 4>,<at 0.88 0>]}}})
    Module.Param0WriteEnabled=true
    Module.Param1WriteEnabled=true
    Module.Param2WriteEnabled=true
    Module.Param3WriteEnabled=false
   )
  ]
  rendererClasses [&NiagaraSpriteRendererProperties]
  renderers [{rendererClass=&NiagaraSpriteRendererProperties}]
 }
}
summary {
 systemName=NS_Contrail_Snowflake_Converted
 emitters [
  {emitterName=snowflakes01, rendererClasses [&NiagaraSpriteRendererProperties]}
  {emitterName=snowflakes, rendererClasses [&NiagaraSpriteRendererProperties]}
  {emitterName=Powderdsnow, rendererClasses [&NiagaraSpriteRendererProperties]}
  {bEnabled=false, emitterName=8, rendererClasses [&NiagaraSpriteRendererProperties]}
  {bEnabled=false, emitterName=Minimal, rendererClasses [&NiagaraSpriteRendererProperties]}
  {emitterName=Source, rendererClasses [&NiagaraSpriteRendererProperties]}
  {emitterName=Minimal001, rendererClasses [&NiagaraSpriteRendererProperties]}
  {emitterName=FN_SparkleDust, rendererClasses [&NiagaraSpriteRendererProperties]}
 ]
 userVariables [
  User.bIsFrontEnd {type=NiagaraBool}
  User.bIsFrontEndPreview {type=NiagaraBool}
  User.ParaGlideLeanAlpha {type=NiagaraFloat, defaultValue=0.5, description=""}
  User.SkeletalMesh {
   type=NiagaraDataInterfaceSkeletalMesh
   defaultValue {
    struct=NiagaraExt_VariableValue_DataInterface
    value {dataInterface=&self:NiagaraDataInterfaceSkeletalMesh_0}
   }
  }
 ]
}
