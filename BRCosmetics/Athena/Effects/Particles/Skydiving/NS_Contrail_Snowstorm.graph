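# Niagara skydiving contrail cosmetic that surrounds the falling player with a snowstorm of clouds and flakes.
# Sparse, faint cloud sprites and two dense layers of white-to-blue flakes spawn in flattened spheres below the player, inheriting some velocity.
# Flakes are pushed upward, spinning, orbiting and scaling in and out over their lifetime.
# Four fading ribbons trail from sampled skeletal mesh positions, while two other ribbon emitters are disabled.
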NiagaraSystem /BRCosmetics/Athena/Effects/Particles/Skydiving/NS_Contrail_Snowstorm
dependencies {
 usedDataInterfaces [&NiagaraDataInterfaceVectorCurve, &NiagaraDataInterfaceCurve]
 usedDynamicInputs []
 usedModules [&SystemState]
}
emitters {
 Clouds {
  emitterSpawnScript [&FN_PreserveSkydiveUserVariables() {module=PreserveUserVariables}]
  emitterUpdateScript [
   &EmitterState(Enable Distance Culling=true, Module.bUseMaxDistance=true, Module.bUseMinDistance=false, Scalability Mode=::NewEnumerator1)
   &SpawnRate(SpawnRate=3, Use Spawn Probability=false)
  ]
  particleSpawnScript [
   /SetVariables(Particles.SpawnSeed:NiagaraFloat=$Engine.RealTime) {isSetParameters=true}
   /Game/Effects/Niagara/Modules/Groups/FN_SetSpawnGroup_v1/FN_SetSpawnGroup(Emitter Bone Count=1)
   /Niagara/Modules/Spawn/Initialization/InitializeParticle(
    Color=&UniformRangedLinearColor(
     Minimum=&MakeLinearColor(R:NiagaraFloat=1, G:NiagaraFloat=1, B:NiagaraFloat=1, A:NiagaraFloat=0.01)
     Maximum=&MakeLinearColor(R:NiagaraFloat=0.4, G:NiagaraFloat=0.8, B:NiagaraFloat=0.9, A:NiagaraFloat=0.2)
    )
    Lifetime=&UniformRangedFloat(Minimum=2, Maximum=3)
    Module.Write Color=true
    Module.Write Lifetime=true
    Module.Write Mass=true
    Module.Write Material Random=true
    Module.Write Position=false
    Module.Write Scale=false
    Module.Write SpriteRotation=true
    Module.Write SpriteSize=true
    Position=&SimulationPosition()
    Sprite Rotation=&UniformRangedFloat(Minimum=0, Maximum=360)
    Sprite Size=&Multiply_Vector2(
     A=&MakeVector2D(
      X=&UniformRangedFloat(Minimum=0.4, Maximum=1)
      Y=&UniformRangedFloat(Minimum=0.4, Maximum=1)
     )
     B=&Vector2DFromFloat(Value=&UniformRangedFloat(Minimum=1000, Maximum=2000))
    )
   )
   /SetVariables(Particles.Position=&Add_Vector(
     A=Linked() {linkedVariable {name=Engine.Owner.Position, type=NiagaraPosition}}
     B=&Multiply_Vector(
      A=&MakeVector(
       X=&Lerp_Float(A=-1, B=1, Alpha=HlslExpression() {hlslExpression="frac(Particles.SpawnSeed)"})
       Y=&Lerp_Float(A=-1, B=1, Alpha=HlslExpression() {hlslExpression="frac(Particles.SpawnSeed*3.14)"})
       Z=&Lerp_Float(A=-0.1, B=0.1, Alpha=HlslExpression() {hlslExpression="frac(Particles.SpawnSeed*4-3.222)"})
      )
      B=&VectorFromFloat(Value=&Lerp_Float(A=300, B=750, Alpha=HlslExpression() {hlslExpression="frac(Particles.SpawnSeed*.874252)"}))
     )
    )) {isSetParameters=true}
   &SampleSkeletalMeshSkeleton(Mesh:NiagaraDataInterfaceSkeletalMesh=$User.SkeletalMesh, Specific Bone Index:NiagaraInt32=$Particles.SpawnGroup) {enabled=false}
   /Niagara/Modules/Spawn/Location/SkeletalMeshSkeletonLocation(Module.Inherit Velocity=false, Module.Use Sampled Bone as Ribbon ID=true) {enabled=false}
   /Niagara/Modules/Spawn/Location/SphereLocation(
    MaskSpawnGroup=false
    Module.Override Randomness=false
    Non Uniform Scale=<1 0.75 0.2>
    Offset=<0 0 -2500>
    Sphere Radius=2000
    Surface Only=false
   )
   /SetVariables(Particles.DynamicMaterialParameter:Vector4f=<0.3 0.2 0 4>) {enabled=false, isSetParameters=true}
   /SetVariables(Particles.DynamicMaterialParameter1:Vector4f=&MakeVector4(X=&Multiply_Float(A=50, B=$Engine.RealTime), Y=0, Z=0, W=0)) {enabled=false, isSetParameters=true}
   &Color(Color=&MakeLinearColor(
     R:NiagaraFloat=1
     G:NiagaraFloat=1
     B:NiagaraFloat=1
     A:NiagaraFloat=&UniformRangedFloat(Minimum=0.01, Maximum=0.05)
    ))
   &InheritVelocity(Module.Limit Speed=false, Velocity Scale=<0.5 0.5 0>)
  ]
  particleUpdateScript [
   &ParticleState()
   &ScaleSpriteSize(Scale Factor=/Niagara/DynamicInputs/ScaleByCurve/ScaleVector2DByCurve(
     InitialValue:Vector2f=<1 1>
     ScaleCurve:NiagaraDataInterfaceVectorCurve=DataInterface() {propertyValues {
       XCurve {keys [
         <at 0.05 1>
         {arriveTangent=-0.3, interpMode=RCIM_Cubic, leaveTangent=-0.3, tangentMode=RCTM_User, time=0.4903, value=1}
         {arriveTangent=-3.5988, interpMode=RCIM_Cubic, leaveTangent=-3.5988, tangentMode=RCTM_User, time=1, value=0.4}
        ]}
       YCurve {keys [
         <at 0.05 1>
         {arriveTangent=-0.249, interpMode=RCIM_Cubic, leaveTangent=-0.249, tangentMode=RCTM_User, time=0.4903, value=1}
         {arriveTangent=-3.6328, interpMode=RCIM_Cubic, leaveTangent=-3.6328, tangentMode=RCTM_User, time=1, value=0.4}
        ]}
       ZCurve {keys [
         <at 0.05 1>
         {arriveTangent=-0.3348, interpMode=RCIM_Cubic, leaveTangent=-0.3348, tangentMode=RCTM_User, time=0.4903, value=1}
         {arriveTangent=-3.6678, interpMode=RCIM_Cubic, leaveTangent=-3.6678, tangentMode=RCTM_User, time=1, value=0.4}
        ]}
      }}
     CurveIndex:NiagaraFloat=$Particles.NormalizedAge
    ))
   &SolveForcesAndVelocity(Clamp Velocity=false, Limit Acceleration=false)
   &ScaleColor(
    Scale Alpha=&FloatFromCurve(FloatCurve=DataInterface() {propertyValues {Curve {keys [
         {arriveTangent=5.4379, interpMode=RCIM_Cubic, leaveTangent=5.4379, tangentMode=RCTM_User, time=-0.0046, value=0}
         <at 0.2 1>
         {arriveTangent=-0.717, interpMode=RCIM_Cubic, leaveTangent=-0.717, tangentMode=RCTM_User, time=0.792, value=1}
         {arriveTangent=-5.8207, interpMode=RCIM_Cubic, leaveTangent=-5.8207, tangentMode=RCTM_User, time=1, value=0}
        ]}}})
    ScaleA=true
    ScaleRGB=true
   )
  ]
  rendererClasses [&NiagaraSpriteRendererProperties]
  renderers [{rendererClass=&NiagaraSpriteRendererProperties}]
 }
 FN_Contrails_Ribbon_v4_A {
  emitterUpdateScript [
   &EmitterScalability(
    Module.bApplySpawnCountScale=false
    Module.bApplySpawnCountScaleByDistanceFraction=false
    Module.bUseMaxDistance=false
    Module.bUseMinDistance=false
   ) {enabled=false}
   &EmitterLifeCycle(CompleteOnInactive=true) {enabled=false}
   &EmitterState()
   /Game/Effects/Niagara/Modules/Spawn/FN_SpawnContrails(All limbs=false, Global Spawnrate=60) {enabled=false}
   /Game/Effects/Niagara/Modules/AnimTrail/FN_SpawnAnimTrail/FN_SpawnAnimTrail() {enabled=false}
   &FN_SpawnRibbon_v2(Number of Ribbons=2, Spawn Mode=::NewEnumerator0) {enabled=false, module=FN_SpawnRibbon_v2001}
   &SpawnRate(SpawnRate=30, Use Spawn Probability=false) {enabled=false}
   &FN_SpawnRibbon_v2(Number of Ribbons=4, Spawn Mode=::NewEnumerator0)
  ]
  particleSpawnScript [
   &FN_InitializeRibbon_v3(Enable Skeletal Position=true, Lifetime=0.7, Ribbon Width=60) {enabled=false}
   &FN_InitializeRibbon_v3(
    Color:LinearColor=/Niagara/DynamicInputs/TypeConversions/MakeLinearColorFromFloat(Value:NiagaraFloat=0.3)
    Enable Skeletal Position=true
    Lifetime=0.7
    Ribbon Width=50
    Sampled Position Coordinate Space:ENiagaraCoordinateSpace=::Local
    Sampled Rotation Coordinate Space:ENiagaraCoordinateSpace=::Local
    Sampled Velocity Coordinate Space:ENiagaraCoordinateSpace=::Local
   ) {module=FN_InitializeRibbon_v3001}
   &AddVelocity(Velocity=<0 0 410>) {enabled=false}
   &AddVelocity(Velocity=&MakeCustomVectorFromBool(
     Bool=$Particles.Locked
     True Vector=<0 0 0>
     False Vector=&FN_FrontEndVector3Modifier(Gameplay=<0 0 500>, Front End=<0 0 500>)
    )) {module=AddVelocity001}
   &SampleSkeletalMeshSkeleton(Mesh:NiagaraDataInterfaceSkeletalMesh=$User.SkeletalMesh)
  ]
  particleUpdateScript [
   &AccelerationForce(
    Acceleration=&VectorFromCurve(VectorCurve=DataInterface() {propertyValues {
       XCurve {keys [<at 0.0017 132.1921>,<at 0.9905 -65.3336>]}
       YCurve {keys [<at -0.0011 1.0067>,<at 0.9988 0>]}
       ZCurve {keys [<at -0.0011 -300>,<at 0.9957 -200.05>]}
      }})
    Coordinate Space=::Local
   ) {enabled=false}
   &Drag(Drag=0.9, Use Linear Drag=true, Use Rotational Drag=false) {enabled=false}
   &InheritVelocity(
    Module.Limit Speed=true
    Velocity Limit=10000
    Velocity Scale=&MakeCustomVectorFromBool(Bool=$Particles.Initial.Locked, True Vector=<0 0 0>, False Vector=<-0.03 -0.03 -0.01>)
    Velocity Source=&TransformVector(Vector=$Engine.Owner.Velocity, Source Space=::World, Destination Space=::Local)
   ) {enabled=false}
   /Game/Effects/Niagara/Modules/Skydiving/FN_DivingAttributes/FN_DivingAttributes() {enabled=false}
   &SolveForcesAndVelocity(Clamp Velocity=false, Limit Acceleration=false)
   &FN_UpdateRibbon(
    Enable Skeletal Position=true
    Initial Color:LinearColor=$Particles.Initial.Color
    Initial RibbonWidth:NiagaraFloat=$Particles.Initial.RibbonWidth
   )
   &ScaleColor(
    Scale Alpha=&FloatFromCurve(FloatCurve=DataInterface() {propertyValues {Curve {keys [
         {arriveTangent=1.8999, interpMode=RCIM_Cubic, leaveTangent=1.8999, tangentMode=RCTM_User, time=-0.0044, value=0.9894}
         {arriveTangent=-1.934, interpMode=RCIM_Cubic, leaveTangent=-1.934, tangentMode=RCTM_User, time=0.2157, value=1}
         {arriveTangent=-0.0191, interpMode=RCIM_Cubic, leaveTangent=-0.0191, tangentMode=RCTM_User, time=1.0011, value=0.002}
        ]}}})
    ScaleA=true
    ScaleRGB=false
   )
   &Material(Base Texture Time:NiagaraFloat=&Multiply_Float(A=$Emitter.Age, B=-1), PostMult=0.5455, Power=0.4408)
   &ScaleRibbonWidth(Ribbon Width Scale=&FloatFromCurve(FloatCurve=DataInterface() {propertyValues {Curve {keys [
         {arriveTangent=1.9737, interpMode=RCIM_Cubic, leaveTangent=1.9737, tangentMode=RCTM_User, time=0.0058, value=0.9363}
         {arriveTangent=-0.0131, interpMode=RCIM_Cubic, leaveTangent=-0.0131, time=0.4538, value=1.5093}
         <at 0.6635 0.9277>
        ]}}})) {enabled=false}
   &Color(
    Color=<rgba 1 1 1 0.8>
    Scale Alpha=&FloatFromCurve(
     FloatCurve=DataInterface() {propertyValues {Curve {keys [
         {arriveTangent=2.7939, interpMode=RCIM_Cubic, leaveTangent=2.7939, tangentMode=RCTM_User, time=-0.0046, value=1.0113}
         {arriveTangent=-1.0036, interpMode=RCIM_Cubic, leaveTangent=-4.3444, tangentMode=RCTM_Break, time=0.1387, value=1.232}
         <at 0.6 0.002>
        ]}}}
     Scale Curve=0.8
    )
   ) {enabled=false}
  ]
  rendererClasses [&NiagaraRibbonRendererProperties]
  renderers [{rendererClass=&NiagaraRibbonRendererProperties}]
 }
 GlowyFlakesLocalTiny {
  emitterSpawnScript [&FN_PreserveSkydiveUserVariables() {module=PreserveUserVariables}]
  emitterUpdateScript [
   &EmitterState(Enable Distance Culling=true, Module.bUseMaxDistance=true, Module.bUseMinDistance=false, Scalability Mode=::NewEnumerator1)
   &SpawnRate(SpawnRate=100, Use Spawn Probability=false)
  ]
  particleSpawnScript [
   &FN_SetSpawnGroup(Emitter Bone Count=1)
   &InitializeParticle(
    Lifetime=2
    Module.Write Color=true
    Module.Write Lifetime=true
    Module.Write Mass=true
    Module.Write Material Random=true
    Module.Write Position=false
    Module.Write Scale=false
    Module.Write SpriteRotation=true
    Module.Write SpriteSize=true
    Position=&SimulationPosition()
    Sprite Rotation=&UniformRangedFloat(Minimum=0, Maximum=360)
    Sprite Size=&Vector2DFromFloat(Value=&UniformRangedFloat(Minimum=10, Maximum=25))
   )
   &SampleSkeletalMeshSkeleton(Mesh:NiagaraDataInterfaceSkeletalMesh=$User.SkeletalMesh, Specific Bone Index:NiagaraInt32=$Particles.SpawnGroup) {enabled=false}
   &SkeletalMeshSkeletonLocation(Module.Inherit Velocity=false, Module.Use Sampled Bone as Ribbon ID=true) {enabled=false}
   &SphereLocation(MaskSpawnGroup=false, Module.Override Randomness=false, Non Uniform Scale=<2 2 1>, Offset=<0 0 -200>, Sphere Radius=150, Surface Only=false)
   &AddVelocity(Velocity=&UniformRangedVector(Minimum=<0 0 200>, Maximum=<0 0 750>))
   &InheritVelocity(
    Module.Limit Speed=false
    Velocity Scale=&UniformRangedVector(Minimum=<0.5 0.5 -0.4>, Maximum=<0.5 0.5 -0.1>)
   )
   &ApplyInitialForces()
   &Color(Color=&UniformRangedLinearColor(Minimum=<rgba 1 1 1 1>, Maximum=<rgba 0.2344 0.3273 1 1>, Color Channel Mode=::NewEnumerator1))
   /Game/Effects/Niagara/Modules/Orbit/FN_Orbit_Init_Variables_v2()
   /Niagara/Modules/Update/SubUV/SubUVAnimation(SubUV Lookup Index=&UniformRangedFloat(Minimum=0, Maximum=3))
  ]
  particleUpdateScript [
   &ParticleState()
   &ScaleSpriteSize(Scale Factor=&ScaleVector2DByCurve(
     InitialValue:Vector2f=<1 1>
     ScaleCurve:NiagaraDataInterfaceVectorCurve=DataInterface() {propertyValues {
       XCurve {keys [
         {arriveTangent=5.599, interpMode=RCIM_Cubic, leaveTangent=5.599, tangentMode=RCTM_User, time=0, value=0}
         <at 0.4903 1 RCIM_Cubic>
         {arriveTangent=-3.5988, interpMode=RCIM_Cubic, leaveTangent=-3.5988, tangentMode=RCTM_User, time=1, value=0}
        ]}
       YCurve {keys [
         {arriveTangent=5.5554, interpMode=RCIM_Cubic, leaveTangent=5.5554, tangentMode=RCTM_User, time=0, value=0}
         <at 0.4903 1 RCIM_Cubic>
         {arriveTangent=-3.6328, interpMode=RCIM_Cubic, leaveTangent=-3.6328, tangentMode=RCTM_User, time=1, value=0}
        ]}
       ZCurve {keys [
         {arriveTangent=5.4818, interpMode=RCIM_Cubic, leaveTangent=5.4818, tangentMode=RCTM_User, time=0, value=0}
         <at 0.4903 1 RCIM_Cubic>
         {arriveTangent=-3.6678, interpMode=RCIM_Cubic, leaveTangent=-3.6678, tangentMode=RCTM_User, time=1, value=0}
        ]}
      }}
     CurveIndex:NiagaraFloat=$Particles.NormalizedAge
    ))
   &SolveForcesAndVelocity(Clamp Velocity=false, Limit Acceleration=false)
   &SpriteRotationRate(Rotation Rate=-150)
   /BRCosmetics/Effects/Niagara/Modules/Orbit/FN_Orbit_v2()
  ]
  rendererClasses [&NiagaraSpriteRendererProperties]
  renderers [{rendererClass=&NiagaraSpriteRendererProperties}]
 }
 Ribbon {
  bEnabled=false
  emitterUpdateScript [
   &EmitterState(Life Cycle Mode=::NewEnumerator1, Loop Duration=4800, UseLoopDelay=false)
   &FN_SpawnContrails(Global Spawnrate=60)
  ]
  particleSpawnScript [
   /Game/Effects/Niagara/Modules/Contrails/FN_InitializeRibbon/FN_InitializeRibbon_v2(
    Color:LinearColor=&MakeLinearColorFromFloat(Value:NiagaraFloat=0.01)
    Enable Skeletal Position=true
    Lifetime=0.3
    Ribbon Width=50
   ) {module=FN_InitializeRibbon_v2001}
   &SampleSkeletalMeshSkeleton(Mesh:NiagaraDataInterfaceSkeletalMesh=$User.SkeletalMesh)
   &AddVelocity(Velocity=&FN_FrontEndVector3Modifier(Gameplay=<0 0 0>, Front End=<0 0 500>))
  ]
  particleUpdateScript [
   &SolveForcesAndVelocity(Clamp Velocity=false, Limit Acceleration=false)
   &FN_UpdateRibbon(
    Enable Skeletal Position=true
    Initial Color:LinearColor=$Particles.Initial.Color
    Initial RibbonWidth:NiagaraFloat=$Particles.Initial.RibbonWidth
   )
   &ScaleColor(
    Scale Alpha=&FloatFromCurve(FloatCurve=DataInterface() {propertyValues {Curve {keys [
         {arriveTangent=0.1037, interpMode=RCIM_Cubic, leaveTangent=0.1037, tangentMode=RCTM_User, time=-0.0044, value=0.9894}
         {arriveTangent=-0.9819, interpMode=RCIM_Cubic, leaveTangent=-0.9819, time=0.3394, value=1}
         {arriveTangent=-2.9148, interpMode=RCIM_Cubic, leaveTangent=-2.9148, tangentMode=RCTM_User, time=1.0011, value=0.002}
        ]}}})
    ScaleA=true
    ScaleRGB=true
   )
   &Material(Base Texture Time:NiagaraFloat=&Multiply_Float(A=$Emitter.Age, B=-1))
  ]
  rendererClasses [&NiagaraRibbonRendererProperties]
  renderers [{rendererClass=&NiagaraRibbonRendererProperties}]
 }
 Ribbon001 {
  bEnabled=false
  emitterUpdateScript [
   &EmitterState(Life Cycle Mode=::NewEnumerator1, Loop Duration=4800, UseLoopDelay=false)
   &FN_SpawnContrails(Global Spawnrate=60)
  ]
  particleSpawnScript [
   &FN_InitializeRibbon_v2(
    Color:LinearColor=&MakeLinearColorFromFloat(Value:NiagaraFloat=0.3)
    Enable Skeletal Position=true
    Lifetime=0.3
    Ribbon Width=50
   ) {module=FN_InitializeRibbon_v2001}
   &SampleSkeletalMeshSkeleton(Mesh:NiagaraDataInterfaceSkeletalMesh=$User.SkeletalMesh)
   &AddVelocity(Velocity=&FN_FrontEndVector3Modifier(Gameplay=<0 0 0>, Front End=<0 0 500>))
  ]
  particleUpdateScript [
   &SolveForcesAndVelocity(Clamp Velocity=false, Limit Acceleration=false)
   &FN_UpdateRibbon(
    Enable Skeletal Position=true
    Initial Color:LinearColor=$Particles.Initial.Color
    Initial RibbonWidth:NiagaraFloat=$Particles.Initial.RibbonWidth
   )
   &ScaleColor(
    Scale Alpha=&FloatFromCurve(FloatCurve=DataInterface() {propertyValues {Curve {keys [
         {arriveTangent=1.8999, interpMode=RCIM_Cubic, leaveTangent=1.8999, tangentMode=RCTM_User, time=-0.0044, value=0.9894}
         {arriveTangent=-1.934, interpMode=RCIM_Cubic, leaveTangent=-1.934, tangentMode=RCTM_User, time=0.2157, value=1}
         {arriveTangent=-0.0191, interpMode=RCIM_Cubic, leaveTangent=-0.0191, tangentMode=RCTM_User, time=1.0011, value=0.002}
        ]}}})
    ScaleA=true
    ScaleRGB=false
   )
   &Material(Base Texture Time:NiagaraFloat=&Multiply_Float(A=$Emitter.Age, B=-1), PostMult=0.5455, Power=0.4408)
  ]
  rendererClasses [&NiagaraRibbonRendererProperties]
  renderers [{rendererClass=&NiagaraRibbonRendererProperties}]
 }
 ShinyFlakes {
  emitterSpawnScript [&FN_PreserveSkydiveUserVariables() {module=PreserveUserVariables}]
  emitterUpdateScript [
   &EmitterState(Enable Distance Culling=true, Module.bUseMaxDistance=true, Module.bUseMinDistance=false, Scalability Mode=::NewEnumerator1)
   &SpawnRate(SpawnRate=150, Use Spawn Probability=false)
  ]
  particleSpawnScript [
   &FN_Orbit_Init_Variables_v2()
   &FN_SetSpawnGroup(Emitter Bone Count=1)
   &InitializeParticle(
    Lifetime=2
    Module.Write Color=true
    Module.Write Lifetime=true
    Module.Write Mass=true
    Module.Write Material Random=true
    Module.Write Position=false
    Module.Write Scale=false
    Module.Write SpriteRotation=true
    Module.Write SpriteSize=true
    Position=&SimulationPosition()
    Sprite Rotation=&UniformRangedFloat(Minimum=0, Maximum=360)
    Sprite Size=&Vector2DFromFloat(Value=&UniformRangedFloat(Minimum=10, Maximum=40))
   )
   &SampleSkeletalMeshSkeleton(Mesh:NiagaraDataInterfaceSkeletalMesh=$User.SkeletalMesh, Specific Bone Index:NiagaraInt32=$Particles.SpawnGroup) {enabled=false}
   &SkeletalMeshSkeletonLocation(Module.Inherit Velocity=false, Module.Use Sampled Bone as Ribbon ID=true) {enabled=false}
   &SphereLocation(MaskSpawnGroup=false, Module.Override Randomness=false, Non Uniform Scale=<2 2 1>, Offset=<0 0 -500>, Sphere Radius=300, Surface Only=false)
   &AddVelocity(Velocity=<0 0 500>)
   &InheritVelocity(
    Module.Limit Speed=false
    Velocity Scale=&UniformRangedVector(Minimum=<0.5 0.5 -0.4>, Maximum=<0.5 0.5 -0.1>)
   )
   &ApplyInitialForces()
   &Color(Color=&UniformRangedLinearColor(Minimum=<rgba 1 1 1 1>, Maximum=<rgba 0.1036 0.4193 1 1>, Color Channel Mode=::NewEnumerator1))
   &SubUVAnimation(SubUV Lookup Index=&UniformRangedFloat(Minimum=0, Maximum=3))
  ]
  particleUpdateScript [
   &ParticleState()
   &ScaleSpriteSize(Scale Factor=&ScaleVector2DByCurve(
     InitialValue:Vector2f=<1 1>
     ScaleCurve:NiagaraDataInterfaceVectorCurve=DataInterface() {propertyValues {
       XCurve {keys [
         {arriveTangent=5.599, interpMode=RCIM_Cubic, leaveTangent=5.599, tangentMode=RCTM_User, time=0, value=0}
         <at 0.4903 1 RCIM_Cubic>
         {arriveTangent=-3.5988, interpMode=RCIM_Cubic, leaveTangent=-3.5988, tangentMode=RCTM_User, time=1, value=0}
        ]}
       YCurve {keys [
         {arriveTangent=5.5554, interpMode=RCIM_Cubic, leaveTangent=5.5554, tangentMode=RCTM_User, time=0, value=0}
         <at 0.4903 1 RCIM_Cubic>
         {arriveTangent=-3.6328, interpMode=RCIM_Cubic, leaveTangent=-3.6328, tangentMode=RCTM_User, time=1, value=0}
        ]}
       ZCurve {keys [
         {arriveTangent=5.4818, interpMode=RCIM_Cubic, leaveTangent=5.4818, tangentMode=RCTM_User, time=0, value=0}
         <at 0.4903 1 RCIM_Cubic>
         {arriveTangent=-3.6678, interpMode=RCIM_Cubic, leaveTangent=-3.6678, tangentMode=RCTM_User, time=1, value=0}
        ]}
      }}
     CurveIndex:NiagaraFloat=$Particles.NormalizedAge
    ))
   &SolveForcesAndVelocity(Clamp Velocity=false, Limit Acceleration=false)
   &SpriteRotationRate(Rotation Rate=200)
   &FN_Orbit_v2()
  ]
  rendererClasses [&NiagaraSpriteRendererProperties]
  renderers [{rendererClass=&NiagaraSpriteRendererProperties}]
 }
}
summary {
 systemName=NS_Contrail_Snowstorm
 emitters [
  {emitterName=ShinyFlakes, rendererClasses [&NiagaraSpriteRendererProperties]}
  {emitterName=GlowyFlakesLocalTiny, rendererClasses [&NiagaraSpriteRendererProperties]}
  {emitterName=Clouds, rendererClasses [&NiagaraSpriteRendererProperties]}
  {bEnabled=false, emitterName=Ribbon, rendererClasses [&NiagaraRibbonRendererProperties]}
  {bEnabled=false, emitterName=Ribbon001, rendererClasses [&NiagaraRibbonRendererProperties]}
  {emitterName=FN_Contrails_Ribbon_v4_A, rendererClasses [&NiagaraRibbonRendererProperties]}
 ]
 userVariables [
  User.bIsFrontEnd {type=NiagaraBool}
  User.bIsFrontEndPreview {type=NiagaraBool}
  User.PlayerSpeed {type=NiagaraFloat, defaultValue=0, description=""}
  User.RibbonIDOffset {type=NiagaraInt32}
  User.SkeletalMesh {
   type=NiagaraDataInterfaceSkeletalMesh
   defaultValue {
    struct=NiagaraExt_VariableValue_DataInterface
    value {dataInterface=&self:NiagaraDataInterfaceSkeletalMesh_0}
   }
  }
 ]
}
