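# Niagara skydiving contrail "StringLights" cosmetic, attached to bones or sockets on the user-supplied skeletal mesh.
# Two burst emitters place persistent meshes by spawn group: four with per-group offsets, two oriented to the skeleton.
# The oriented pair drives dynamic material parameters by group, front-end versus gameplay values and ParaGlideLeanAlpha.
# Glow continuously spawns short-lived, color-shifting sprites that follow the skeleton and also render as lights.
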NiagaraSystem /BRCosmetics/Athena/Effects/Particles/Skydiving/NS_Contrail_StringLights
dependencies {
 usedDataInterfaces [
  &NiagaraDataInterfaceVectorCurve
  &NiagaraDataInterfaceCurve
  &NiagaraDataInterfaceArrayFloat3
  &NiagaraDataInterfaceArrayFloat
 ]
 usedDynamicInputs []
 usedModules [&SystemState, &InterpolateOverTime]
}
emitters {
 Glow {
  emitterUpdateScript [
   &EmitterState(Inactive Response=::NewEnumerator1, Life Cycle Mode=::NewEnumerator1, Loop Duration=1, UseLoopDelay=false)
   &SpawnRate(SpawnRate=8, Use Spawn Probability=false) {module=SpawnRate001}
  ]
  particleSpawnScript [
   &FN_SetSpawnGroup()
   &InitializeParticle(
    Lifetime=&RandomRangeFloat(Minimum=0.2, Maximum=0.4)
    Sprite Size=<50 72>
    Sprite Size Mode=::NewEnumerator4
    UsePositionOffset=false
    Write Material Random=true
   )
   &Color(Color=&MakeLinearColorFromVectorAndFloat("Vector (RGB)"=<1 1 1>, "Float (Alpha)"=&ClampFloat(Float=1.8, Max=1.8)))
  ]
  particleUpdateScript [
   /SetVariables(Particles.LightExponent=3) {isSetParameters=true}
   &ParticleState()
   &ScaleSpriteSize()
   &CameraOffset(Camera Offset Amount=&RandomRangeFloat(Minimum=10, Maximum=40, Evaluation Type=::NewEnumerator1))
   &ScaleColor(
    Scale Alpha=&FloatFromCurve(FloatCurve=DataInterface() {propertyValues {Curve {keys [
         {arriveTangent=10.7452, interpMode=RCIM_Cubic, time=0, value=0}
         {arriveTangent=1.0812, interpMode=RCIM_Cubic, leaveTangent=1.0812, time=0.1334, value=0.9177}
         {arriveTangent=-1.0589, interpMode=RCIM_Cubic, leaveTangent=-1.0589, time=0.7906, value=0.8548}
         <at 1 0>
        ]}}})
    Scale RGB=&VectorFromCurve(VectorCurve=DataInterface() {propertyValues {
       XCurve {keys [<at 0 4 RCIM_Cubic>, {arriveTangent=-3, interpMode=RCIM_Cubic, leaveTangent=-3, time=0.4, value=0}, <at 1 1>]}
       YCurve {keys [<at 0 1.0413 RCIM_Cubic>, {arriveTangent=-1.037, interpMode=RCIM_Cubic, leaveTangent=-1.037, time=0.4, value=0.5473}, <at 1 0.0043>]}
       ZCurve {keys [<at 0 0 RCIM_Cubic>, {arriveTangent=0.7018, interpMode=RCIM_Cubic, leaveTangent=0.7018, time=0.4, value=0.4181}, <at 1 0.7018>]}
      }})
    ScaleA=true
    ScaleRGB=true
   )
   &SkeletalMeshLocation(Bone / Socket Index=$Particles.SpawnGroup, Bone Sampling Mode=::NewEnumerator4, MaskSpawnGroup=false, Skeletal Mesh=$User.Skeletal Mesh)
   &CascadeConversion_LightProperties(LightBrightnessScale=1, LightColorScale=<10 20 1>, LightRadiusScale=1)
  ]
  rendererClasses [&NiagaraSpriteRendererProperties, &NiagaraLightRendererProperties]
  renderers [
   {rendererClass=&NiagaraSpriteRendererProperties}
   {rendererClass=&NiagaraLightRendererProperties, rendererIndex=1}
  ]
 }
 hand_lightR {
  emitterUpdateScript [
   &EmitterState(
    Inactive Response=::NewEnumerator1
    Life Cycle Mode=::NewEnumerator1
    Loop Behavior=::NewEnumerator1
    Loop Duration Mode=::NewEnumerator1
    UseLoopDelay=false
   )
   &SpawnBurst_Instantaneous(Spawn Count=4, Use Spawn Probability=false, UseLoopCountLimit=false)
  ]
  particleSpawnScript [
   &FN_SetSpawnGroup()
   &InitializeParticle(Mesh Scale=<1.2 1.2 1.2>, Mesh Scale Mode=::NewEnumerator4, UsePositionOffset=false, Write Material Random=true)
  ]
  particleUpdateScript [
   &ParticleState(Kill Particles When Lifetime Has Elapsed=false, Let Infinitely Lived Particles Die When Emitter Deactivates=true)
   &Color(Color=&MakeLinearColorFromVectorAndFloat(
     "Vector (RGB)"=&VectorFromCurve(VectorCurve=DataInterface() {propertyValues {XCurve {keys [<at 0 1>,<at 1 1>]}, YCurve {keys [<at 0 1>,<at 1 1>]}, ZCurve {keys [<at 0 1>,<at 1 1>]}}})
     "Float (Alpha)"=&ClampFloat(Float=&FloatFromCurve(FloatCurve=DataInterface() {propertyValues {Curve {keys [{arriveTangent=-1, time=0, value=1}, <at 1 0>]}}}))
    ))
   &SkeletalMeshLocation(
    Bone / Socket Index=$Particles.SpawnGroup
    Bone Sampling Mode=::NewEnumerator4
    MaskSpawnGroup=false
    Sampled Position Offset=&SelectVectorFromArray(
     Direct Array Index=$Particles.SpawnGroup
     Vector Selection Array=DataInterface() {propertyValues {FloatData [<0 -5 0>, <-4 -8 0>, <-6 0 0>, <5 -10 0>]}}
    )
    Skeletal Mesh=$User.Skeletal Mesh
   )
  ]
  rendererClasses [&NiagaraMeshRendererProperties]
  renderers [{rendererClass=&NiagaraMeshRendererProperties, rendererIndex=0}]
 }
 hand_lightR001 {
  emitterUpdateScript [
   &EmitterState(
    Inactive Response=::NewEnumerator1
    Life Cycle Mode=::NewEnumerator1
    Loop Behavior=::NewEnumerator1
    Loop Duration Mode=::NewEnumerator1
    UseLoopDelay=false
   )
   &SpawnBurst_Instantaneous(Spawn Count=2, Use Spawn Probability=false, UseLoopCountLimit=false)
  ]
  particleSpawnScript [
   &FN_SetSpawnGroup(Number of Spawn Groups=2)
   &InitializeParticle(Mesh Scale Mode=::NewEnumerator4, UsePositionOffset=false, Write Material Random=true)
  ]
  particleUpdateScript [
   &ParticleState(Kill Particles When Lifetime Has Elapsed=false, Let Infinitely Lived Particles Die When Emitter Deactivates=true)
   &Timeline(
    Play Rate=&Lerp_Float(A=1, B=1, Alpha=&ClampFloat(Float=$User.ParaGlideLeanAlpha))
    Timeline End=10000000
   ) {enabled=false, module=Timeline002}
   &Color(Color=&MakeLinearColorFromVectorAndFloat(
     "Vector (RGB)"=&VectorFromCurve(VectorCurve=DataInterface() {propertyValues {XCurve {keys [<at 0 1>,<at 1 1>]}, YCurve {keys [<at 0 1>,<at 1 1>]}, ZCurve {keys [<at 0 1>,<at 1 1>]}}})
     "Float (Alpha)"=&ClampFloat(Float=&FloatFromCurve(FloatCurve=DataInterface() {propertyValues {Curve {keys [{arriveTangent=-1, time=0, value=1}, <at 1 0>]}}}))
    ))
   &DynamicMaterialParameters(
    Index 0 Param 1=&Add_Float(A=0, B=0.4)
    Index 0 Param 2=0.25
    Index 0 Param 3=10
    Index 0 Param 4=1
    Index 1 Param 1=&SelectFloatFromArray(Direct Array Index=$Particles.SpawnGroup, Float Selection Array=DataInterface() {propertyValues {FloatData [0.25,0.5]}})
    Index 1 Param 2=&Multiply_Float(A=&FN_FrontEndFloatModifier(Gameplay=2.3, Front End=2.2), B=-1)
    Index 1 Param 3=&FN_FrontEndFloatModifier(Gameplay=200, Front End=150)
    Index 1 Param 4=&FN_FrontEndFloatModifier(Gameplay=&Lerp_Float(A=8, B=11, Alpha=$User.ParaGlideLeanAlpha), Front End=13)
    Module.Param0WriteEnabled=true
    Module.Param0WriteEnabled1=true
    Module.Param1WriteEnabled=true
    Module.Param1WriteEnabled1=true
    Module.Param2WriteEnabled=true
    Module.Param2WriteEnabled1=true
    Module.Param3WriteEnabled=true
    Module.Param3WriteEnabled1=true
    Write Parameter Index 1=true
   )
   &SkeletalMeshLocation(
    Bone / Socket Index=$Particles.SpawnGroup
    Bone Sampling Mode=::NewEnumerator4
    MaskSpawnGroup=false
    Orientation Sampling=::NewEnumerator1
    Skeletal Mesh=$User.Skeletal Mesh
   )
  ]
  rendererClasses [&NiagaraMeshRendererProperties]
  renderers [{rendererClass=&NiagaraMeshRendererProperties, rendererIndex=0}]
 }
}
summary {
 systemName=NS_Contrail_StringLights
 emitters [
  {emitterName=hand_lightR, rendererClasses [&NiagaraMeshRendererProperties]}
  {emitterName=hand_lightR001, rendererClasses [&NiagaraMeshRendererProperties]}
  {emitterName=Glow, rendererClasses [&NiagaraSpriteRendererProperties, &NiagaraLightRendererProperties]}
 ]
 userVariables [
  User.bIsFrontEnd {type=NiagaraBool}
  User.bIsFrontEndPreview {type=NiagaraBool}
  User.ParaGlideLeanAlpha {type=NiagaraFloat, defaultValue=0, description=""}
  User.Skeletal Mesh {
   type=NiagaraDataInterfaceSkeletalMesh
   defaultValue {
    struct=NiagaraExt_VariableValue_DataInterface
    value {dataInterface=&self:NiagaraDataInterfaceSkeletalMesh_0}
   }
  }
 ]
}
