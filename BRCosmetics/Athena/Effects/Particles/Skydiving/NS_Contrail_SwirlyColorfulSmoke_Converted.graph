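# Niagara skydiving contrail cosmetic that draws short-lived ribbon trails from the character's hands and feet, plus a purple sprite glow on the hands.
# Primary and secondary hand ribbons and foot ribbons inherit system velocity, blending from sampled skeletal mesh positions to simulated motion as they age.
# Glide lean and front-end modifiers adjust trail offsets, with particle partition flipping the Z offset sign.
# Curves fade alpha and scale ribbon width and glow.
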NiagaraSystem /BRCosmetics/Athena/Effects/Particles/Skydiving/NS_Contrail_SwirlyColorfulSmoke_Converted
dependencies {
 usedDataInterfaces [
  &NiagaraDataInterfaceCurve
  &NiagaraDataInterfaceVectorCurve
  &NiagaraDataInterfaceSkeletalMesh
  &NiagaraDataInterfaceArrayFloat3
 ]
 usedDynamicInputs []
 usedModules [
  &self:SystemSpawnScript.SystemScriptSource.SystemScriptGraph.NiagaraNodeAssignment_1.SetVariables_#1
  &SystemState
  /Game/Effects/Niagara/Modules/Contrails/FN_Contrail_Setup
  &FN_Integrator
  &self:SystemSpawnScript.SystemScriptSource.SystemScriptGraph.NiagaraNodeAssignment_0.SetVariables_#2
 ]
}
emitters {
 Foot_Trails {
  emitterUpdateScript [
   &EmitterState(Inactive Response=::NewEnumerator1, Life Cycle Mode=::NewEnumerator1, Loop Duration=1, UseLoopDelay=false)
   &SpawnRate(SpawnRate=150, Use Spawn Probability=false)
  ]
  particleSpawnScript [
   &PartitionParticles()
   &InitializeParticle(Lifetime=0.08, Ribbon Width=30, Ribbon Width Mode=::NewEnumerator1, UsePositionOffset=false, Write Material Random=true)
   &AddVelocity(Velocity=$System.GlobalVelocity)
  ]
  particleUpdateScript [
   &SkeletalMeshLocation(
    Bone / Socket Index=/CascadeToNiagaraConverter/NiagaraScripts/DynamicInputScripts/SequentialInt(Use Array Data:NiagaraBool=false, First Number:NiagaraInt32=2, Sequence:NiagaraDataInterfaceArrayInt32=?, Last Number:NiagaraInt32=3)
    MaskSpawnGroup=false
    Mesh Sampling Type=::NewEnumerator5
    Orientation Sampling=::NewEnumerator1
    Position Sampling=::NewEnumerator2
    Sampled Position Offset=&MakeVector(
     X=&Lerp_Float(
      A=&Lerp_Float(A=-3, B=0, Alpha=&FN_FrontEndFloatModifier(Gameplay=0, Front End=1))
      B=&Lerp_Float(A=15, B=5, Alpha=&FN_FrontEndFloatModifier(Gameplay=0, Front End=1))
      Alpha=$User.ParaGlideLeanAlpha
     )
     Y=&Lerp_Float(A=-2, B=1, Alpha=$User.ParaGlideLeanAlpha)
     Z=&Multiply_Float(
      A=&Lerp_Float(
       A=&Lerp_Float(A=0, B=-10, Alpha=&FN_FrontEndFloatModifier(Gameplay=0, Front End=1))
       B=&Lerp_Float(A=2, B=2, Alpha=&FN_FrontEndFloatModifier(Gameplay=0, Front End=1))
       Alpha=$User.ParaGlideLeanAlpha
      )
      B=&MakeFloatFromInt(Int=/Niagara/DynamicInputs/Helpers/MaskInt(
        Mask Source:NiagaraInt32=1
        Mask Parameter:NiagaraInt32=0
        Mask:NiagaraInt32=$Particles.PartitionParticles.Partition
        Masked Int:NiagaraInt32=1
        Passthrough Int:NiagaraInt32=-1
       ))
     )
    )
    Skeletal Mesh=$System.Skeletal Mesh
    Socket Sampling Mode=::NewEnumerator5
   )
   &SolveForcesAndVelocity(Clamp Velocity=false, Limit Acceleration=false)
   /SetVariables(Particles.Position=&Lerp_Vector(
     A=Linked() {linkedVariable {name=Particles.SkeletalMeshLocation.SampledPosition, type=NiagaraPosition}}
     B=Linked() {linkedVariable {name=Particles.Position, type=NiagaraPosition}}
     Alpha=&FloatFromCurve(FloatCurve=DataInterface() {propertyValues {Curve {defaultValue=1, keys [<at 0.05 0 RCIM_Cubic>,<at 0.2 1 RCIM_Cubic>]}}})
    )) {isSetParameters=true}
   /SetVariables(Particles.ScaledRibbonWidth:NiagaraFloat=&Multiply_Float(
     A=$Particles.RibbonWidth
     B=&FloatFromCurve(FloatCurve=DataInterface() {propertyValues {Curve {keys [
          {arriveTangent=0.1239, interpMode=RCIM_Cubic, leaveTangent=0.1239, tangentMode=RCTM_User, time=0, value=0.9627}
          {arriveTangent=-0.1002, interpMode=RCIM_Cubic, leaveTangent=-0.1002, tangentMode=RCTM_User, time=0.9983, value=0.857}
         ]}}})
    )) {isSetParameters=true}
   &ParticleState()
   &ScaleColor(
    Scale Alpha=&FloatFromCurve(FloatCurve=DataInterface() {propertyValues {Curve {keys [<at 0 1>,<at 1 0>]}}})
    Scale RGB=<1.18 1.18 1.18>
    ScaleA=true
    ScaleRGB=true
   )
   &DynamicMaterialParameters(
    Index 0 Param 1=0.5
    Index 0 Param 2=-65
    Index 0 Param 3=1
    Index 0 Param 4=1
    Module.Param0WriteEnabled=true
    Module.Param1WriteEnabled=true
    Module.Param2WriteEnabled=true
    Module.Param3WriteEnabled=true
   )
  ]
  rendererClasses [&NiagaraRibbonRendererProperties]
  renderers [{rendererClass=&NiagaraRibbonRendererProperties}]
 }
 HandGlow {
  emitterUpdateScript [
   &EmitterState(Inactive Response=::NewEnumerator1, Life Cycle Mode=::NewEnumerator1, Loop Duration=0.5, UseLoopDelay=false)
   &SpawnBurst_Instantaneous(Spawn Count=2, Use Spawn Probability=false, UseLoopCountLimit=false)
  ]
  particleSpawnScript [
   /SetVariables(Particles.SkeletonSamplingIndex:NiagaraInt32=&SequentialInt(Use Array Data:NiagaraBool=false, First Number:NiagaraInt32=0, Sequence:NiagaraDataInterfaceArrayInt32=?, Last Number:NiagaraInt32=1)) {isSetParameters=true}
   &InitializeParticle(Sprite Size=<60 60>, Sprite Size Mode=::NewEnumerator4, UsePositionOffset=false, Write Material Random=true)
   &Color(Color=&MakeLinearColorFromVectorAndFloat("Vector (RGB)"=<0.7035 0.427 1>, "Float (Alpha)"=0.14))
  ]
  particleUpdateScript [
   &ParticleState()
   &ScaleSpriteSize(
    Scale Factor=&Vector2DFromFloat(Value=&FloatFromCurve(FloatCurve=DataInterface() {propertyValues {Curve {keys [{arriveTangent=0.3342, time=0.0035, value=1.4658}, <at 1 1.8>]}}}))
    Scale Sprite Size Mode=::NewEnumerator2
   )
   &CameraOffset(Camera Offset Amount=40)
   &ScaleColor(
    Scale Alpha=&FloatFromCurve(FloatCurve=DataInterface() {propertyValues {Curve {keys [<at 0 0 RCIM_Cubic>,<at 0.5 1 RCIM_Cubic>,<at 1 0 RCIM_Cubic>]}}})
    Scale RGB=&VectorFromCurve(VectorCurve=DataInterface() {propertyValues {
       XCurve {keys [<at 0 4 RCIM_Cubic>,<at 0.8939 0.3518 RCIM_Cubic>]}
       YCurve {keys [<at 0 1.0581 RCIM_Cubic>,<at 0.8939 0.5607 RCIM_Cubic>]}
       ZCurve {keys [<at 0 3.6313 RCIM_Cubic>,<at 0.8939 0.8333 RCIM_Cubic>]}
      }})
    ScaleA=true
    ScaleRGB=true
   )
   &SkeletalMeshLocation(
    Bone / Socket Index=$Particles.SkeletonSamplingIndex
    Bone Sampling Mode=::NewEnumerator4
    MaskSpawnGroup=false
    Orientation Sampling=::NewEnumerator1
    Sampled Position Offset=&SelectVectorFromArray(
     Direct Array Index=$Particles.SkeletonSamplingIndex
     Vector Selection Array=DataInterface() {propertyValues {FloatData [<10 0 0>, <-10 0 0>]}}
    )
    Skeletal Mesh=DataInterface() {propertyValues {
      DefaultMesh=null
      ExcludeBoneName=root
      OptionalComponentClass=null
      PreviewMesh=null
      SkinningMode=SkinOnTheFly
      SoftSourceActor=null
      SourceMode=Default
      UvSetIndex=0
      WholeMeshLOD=-1
      bExcludeBone=false
      bReadDeformedGeometry=true
      bRequireCurrentFrameData=true
      ComponentTags []
      FilteredBones [hand_l,hand_r]
      FilteredSockets []
      MeshUserParameter {parameter {name=None, typeDefHandle {registeredTypeIndex=1}, varData []}}
      SamplingRegions []
     }}
   )
   &DynamicMaterialParameters(
    Index 0 Param 1=1
    Index 0 Param 2=1
    Index 0 Param 3=1
    Index 0 Param 4=1
    Module.Param0WriteEnabled=true
    Module.Param1WriteEnabled=true
    Module.Param2WriteEnabled=true
    Module.Param3WriteEnabled=true
   )
  ]
  rendererClasses [&NiagaraSpriteRendererProperties]
  renderers [{rendererClass=&NiagaraSpriteRendererProperties}]
 }
 Hand_Secondary_Trails {
  emitterUpdateScript [
   &EmitterState(Inactive Response=::NewEnumerator1, Life Cycle Mode=::NewEnumerator1, Loop Duration=1, UseLoopDelay=false)
   &SpawnRate(SpawnRate=120, Use Spawn Probability=false)
  ]
  particleSpawnScript [
   &PartitionParticles()
   &InitializeParticle(Lifetime=0.1, Ribbon Width=32, Ribbon Width Mode=::NewEnumerator1, UsePositionOffset=false, Write Material Random=true)
   &Color(Color=&MakeLinearColorFromVectorAndFloat("Vector (RGB)"=<1 1 1>, "Float (Alpha)"=6))
   &AddVelocity(Velocity=$System.GlobalVelocity, Velocity Speed Scale=0.7)
  ]
  particleUpdateScript [
   &SkeletalMeshLocation(
    Bone / Socket Index=&SequentialInt(Use Array Data:NiagaraBool=false, First Number:NiagaraInt32=0, Sequence:NiagaraDataInterfaceArrayInt32=?, Last Number:NiagaraInt32=1)
    MaskSpawnGroup=false
    Mesh Sampling Type=::NewEnumerator5
    Orientation Sampling=::NewEnumerator1
    Position Sampling=::NewEnumerator2
    Sampled Position Offset=&MakeVector(
     X=&Lerp_Float(A=5, B=0, Alpha=$User.ParaGlideLeanAlpha)
     Y=&Lerp_Float(
      A=-5
      B=&Lerp_Float(A=0, B=-20, Alpha=&FN_FrontEndFloatModifier(Gameplay=0, Front End=1))
      Alpha=$User.ParaGlideLeanAlpha
     )
     Z=&Multiply_Float(
      A=&Lerp_Float(A=5, B=0, Alpha=$User.ParaGlideLeanAlpha)
      B=&MakeFloatFromInt(Int=&MaskInt(
        Mask Source:NiagaraInt32=1
        Mask Parameter:NiagaraInt32=0
        Mask:NiagaraInt32=$Particles.PartitionParticles.Partition
        Masked Int:NiagaraInt32=1
        Passthrough Int:NiagaraInt32=-1
       ))
     )
    )
    Skeletal Mesh=$System.Skeletal Mesh
    Socket Sampling Mode=::NewEnumerator5
   ) {module=SkeletalMeshLocation001}
   &SolveForcesAndVelocity(Clamp Velocity=false, Limit Acceleration=false)
   /SetVariables(Particles.Position=&Lerp_Vector(
     A=Linked() {linkedVariable {name=Particles.SkeletalMeshLocation001.SampledPosition, type=NiagaraPosition}}
     B=Linked() {linkedVariable {name=Particles.Position, type=NiagaraPosition}}
     Alpha=&FloatFromCurve(FloatCurve=DataInterface() {propertyValues {Curve {defaultValue=1, keys [<at 0 0 RCIM_Cubic>,<at 0.3 1 RCIM_Cubic>]}}})
    )) {isSetParameters=true}
   /SetVariables(Particles.ScaledRibbonWidth:NiagaraFloat=&Multiply_Float(
     A=$Particles.RibbonWidth
     B=&FloatFromCurve(FloatCurve=DataInterface() {propertyValues {Curve {keys [
          {arriveTangent=0.1239, interpMode=RCIM_Cubic, leaveTangent=0.1239, tangentMode=RCTM_User, time=0, value=0.9627}
          {arriveTangent=-0.1002, interpMode=RCIM_Cubic, leaveTangent=-0.1002, tangentMode=RCTM_User, time=0.9983, value=0.857}
         ]}}})
    )) {isSetParameters=true}
   &ParticleState()
   &ScaleColor(
    Scale Alpha=&FloatFromCurve(FloatCurve=DataInterface() {propertyValues {Curve {keys [<at 0 1>,<at 1 0>]}}})
    Scale RGB=<4.2 4.2 4.2>
    ScaleA=true
    ScaleRGB=true
   )
   &DynamicMaterialParameters(
    Index 0 Param 2=1.7
    Index 0 Param 3=1
    Index 0 Param 4=1
    Module.Param0WriteEnabled=true
    Module.Param1WriteEnabled=true
    Module.Param2WriteEnabled=true
    Module.Param3WriteEnabled=true
   )
  ]
  rendererClasses [&NiagaraRibbonRendererProperties]
  renderers [{rendererClass=&NiagaraRibbonRendererProperties}]
 }
 Hand_Trails {
  emitterUpdateScript [
   &EmitterState(Inactive Response=::NewEnumerator1, Life Cycle Mode=::NewEnumerator1, Loop Duration=1, UseLoopDelay=false)
   &SpawnRate(SpawnRate=100, Use Spawn Probability=false)
  ]
  particleSpawnScript [
   &PartitionParticles()
   &InitializeParticle(Lifetime=0.135, Ribbon Width=48, Ribbon Width Mode=::NewEnumerator1, UsePositionOffset=false, Write Material Random=true)
   &AddVelocity(Velocity=$System.GlobalVelocity)
  ]
  particleUpdateScript [
   &SkeletalMeshLocation(
    Bone / Socket Index=&SequentialInt(Use Array Data:NiagaraBool=false, First Number:NiagaraInt32=0, Sequence:NiagaraDataInterfaceArrayInt32=?, Last Number:NiagaraInt32=1)
    MaskSpawnGroup=false
    Mesh Sampling Type=::NewEnumerator5
    Orientation Sampling=::NewEnumerator1
    Position Sampling=::NewEnumerator2
    Sampled Position Offset=&MakeVector(
     X=&Lerp_Float(A=5, B=5, Alpha=$User.ParaGlideLeanAlpha)
     Y=&Lerp_Float(
      A=-5
      B=&Lerp_Float(A=0, B=-10, Alpha=&FN_FrontEndFloatModifier(Gameplay=0, Front End=1))
      Alpha=$User.ParaGlideLeanAlpha
     )
     Z=&Multiply_Float(
      A=&Lerp_Float(
       A=5
       B=&Lerp_Float(A=2, B=2, Alpha=&FN_FrontEndFloatModifier(Gameplay=0, Front End=1))
       Alpha=$User.ParaGlideLeanAlpha
      )
      B=&MakeFloatFromInt(Int=&MaskInt(
        Mask Source:NiagaraInt32=1
        Mask Parameter:NiagaraInt32=0
        Mask:NiagaraInt32=$Particles.PartitionParticles.Partition
        Masked Int:NiagaraInt32=1
        Passthrough Int:NiagaraInt32=-1
       ))
     )
    )
    Skeletal Mesh=$System.Skeletal Mesh
    Socket Sampling Mode=::NewEnumerator5
   ) {module=SkeletalMeshLocation001}
   &SolveForcesAndVelocity(Clamp Velocity=false, Limit Acceleration=false)
   /SetVariables(Particles.Position=&Lerp_Vector(
     A=Linked() {linkedVariable {name=Particles.SkeletalMeshLocation001.SampledPosition, type=NiagaraPosition}}
     B=Linked() {linkedVariable {name=Particles.Position, type=NiagaraPosition}}
     Alpha=&FloatFromCurve(FloatCurve=DataInterface() {propertyValues {Curve {defaultValue=1, keys [<at 0 0 RCIM_Cubic>,<at 0.3 1 RCIM_Cubic>]}}})
    )) {isSetParameters=true}
   /SetVariables(Particles.ScaledRibbonWidth:NiagaraFloat=&Multiply_Float(
     A=$Particles.RibbonWidth
     B=&FloatFromCurve(FloatCurve=DataInterface() {propertyValues {Curve {keys [<at 0 1 RCIM_Cubic>,<at 0.9983 0.857 RCIM_Cubic>]}}})
    )) {isSetParameters=true}
   &ParticleState()
   &ScaleColor(
    Scale Alpha=&FloatFromCurve(FloatCurve=DataInterface() {propertyValues {Curve {keys [<at 0 1>,<at 1 0>]}}})
    Scale RGB=<1.18 1.18 1.18>
    ScaleA=true
    ScaleRGB=true
   )
   &DynamicMaterialParameters(
    Index 0 Param 1=0.5
    Index 0 Param 2=-35
    Index 0 Param 3=1.25
    Index 0 Param 4=1
    Module.Param0WriteEnabled=true
    Module.Param1WriteEnabled=true
    Module.Param2WriteEnabled=true
    Module.Param3WriteEnabled=false
   )
  ]
  rendererClasses [&NiagaraRibbonRendererProperties]
  renderers [{rendererClass=&NiagaraRibbonRendererProperties}]
 }
}
summary {
 systemName=NS_Contrail_SwirlyColorfulSmoke_Converted
 emitters [
  {emitterName=Hand_Trails, rendererClasses [&NiagaraRibbonRendererProperties]}
  {emitterName=Foot_Trails, rendererClasses [&NiagaraRibbonRendererProperties]}
  {emitterName=HandGlow, rendererClasses [&NiagaraSpriteRendererProperties]}
  {emitterName=Hand_Secondary_Trails, rendererClasses [&NiagaraRibbonRendererProperties]}
 ]
 userVariables [
  User.bIsFrontEnd {type=NiagaraBool}
  User.bIsFrontEndPreview {type=NiagaraBool}
  User.ParaGlideLeanAlpha {type=NiagaraFloat, defaultValue=0, description=""}
 ]
}
