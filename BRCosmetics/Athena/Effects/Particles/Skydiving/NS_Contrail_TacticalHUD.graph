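# Niagara skydiving contrail cosmetic that draws a tactical HUD around the player, built on the shared FN_Contrail_Setup module.
# DiskLocal bursts a mesh-and-sprite disc, while Pointers continuously spawns short-lived, randomly spinning orange or green meshes that fade in and out.
# handCounters places sprite readouts, some following hand, foot and spine bones through skeletal mesh sampling.
# Spawn groups set visibility tags and switch size, color and offset, letting one emitter hold several elements.
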NiagaraSystem /BRCosmetics/Athena/Effects/Particles/Skydiving/NS_Contrail_TacticalHUD
dependencies {
 usedDataInterfaces [
  &NiagaraDataInterfaceSkeletalMesh
  &NiagaraDataInterfaceArrayInt32
  &NiagaraDataInterfaceArrayFloat3
  &NiagaraDataInterfaceCurve
 ]
 usedDynamicInputs [
  &Lerp_Float
  &FN_FrontEndFloatModifier
  &RemapRange
 ]
 usedModules [
  &SystemState
  /Game/Effects/Niagara/Modules/Contrails/FN_Contrail_Setup
  &FN_Integrator
  &self:SystemSpawnScript.SystemScriptSource.SystemScriptGraph.NiagaraNodeAssignment_0.SetVariables_#1
 ]
}
emitters {
 DiskLocal {
  emitterUpdateScript [
   &EmitterState(
    Inactive Response=::NewEnumerator1
    Life Cycle Mode=::NewEnumerator1
    Loop Behavior=::NewEnumerator1
    Loop Duration Mode=::NewEnumerator1
    UseLoopDelay=true
   )
   &SpawnBurst_Instantaneous(Use Spawn Probability=false, UseLoopCountLimit=false)
   &SpawnBurst_Instantaneous(Spawn Group=1, Use Spawn Probability=false, UseLoopCountLimit=false) {module=SpawnBurst_Instantaneous001}
   &SpawnBurst_Instantaneous(Spawn Group=2, Use Spawn Probability=false, UseLoopCountLimit=false) {module=SpawnBurst_Instantaneous002}
   &SpawnBurst_Instantaneous(Spawn Group=3, Use Spawn Probability=false, UseLoopCountLimit=false) {module=SpawnBurst_Instantaneous004}
   &SpawnBurst_Instantaneous(Spawn Group=4, Use Spawn Probability=false, UseLoopCountLimit=false) {module=SpawnBurst_Instantaneous003}
  ]
  particleSpawnScript [
   /SetVariables(Particles.VisibilityTag=&SpawnGroup()) {isSetParameters=true}
   &PartitionParticles(Number Of Partitions=&MergeHelper_IntSwitch(0=0, 1=0, 2=0, 3=0, 4=2, 5=?, Index=&SpawnGroup(), Num=::NewEnumerator3))
   &InitializeParticle(
    Color=&MergeHelper_ColorSwitch(
     0=<rgba 0 0.946 1.5 1>
     1=<rgba 0 4.0298 5 1>
     2=<rgba 0 5 0.16 0.25>
     3=<rgba 0 0 0 0.75>
     4=<rgba 1 1 1 1>
     Index=&SpawnGroup()
     Num=::NewEnumerator3
    )
    Mesh Scale=&MergeHelper_VectorSwitch(
     0=<50 50 50>
     1=<4 4 7>
     2=<0 0 0>
     3=<0 0 0>
     4=&RandomRangeVector(Minimum=<3 2 4>, Maximum=<4.5 6 4>, Evaluation Type=::NewEnumerator1)
     Index=&SpawnGroup()
     Num=::NewEnumerator3
    )
    Mesh Scale Mode=::NewEnumerator4
    Position Offset=&MergeHelper_VectorSwitch(0=<0 0 0>, 1=<0 0 0>, 2=<0 0 -4>, 3=<0 0 -20>, 4=<0 0 0>, Index=&SpawnGroup(), Num=::NewEnumerator3)
    Sprite Size Mode=::NewEnumerator3
    Uniform Sprite Size=&MergeHelper_FloatSwitch(
     0=0
     1=0
     2=&MakeFloatFromVector2d(Vector 2D=<240 240>)
     3=260
     Index=&SpawnGroup()
     Num=::NewEnumerator2
    )
    UsePositionOffset=true
    Write Material Random=true
   )
   /CascadeToNiagaraConverter/NiagaraScripts/ModuleScripts/SpriteLockAxis(Lock Axis Type:ENiagara_SpriteLockAxis=::NewEnumerator2, UseAlignmentCoordinateSpace=false, UseFacingCoordinateSpace=false)
   /SetVariables(Particles.DiscWorldPos:NiagaraPosition=&ConvertVectorToPosition(Input Position=<0 0 4>, Conversion Operation:ENiagara_LWCConvertVecToPos=::NewEnumerator0)) {isSetParameters=true}
  ]
  particleUpdateScript [&ParticleState(Kill Particles When Lifetime Has Elapsed=false, Let Infinitely Lived Particles Die When Emitter Deactivates=true)]
  rendererClasses [&NiagaraMeshRendererProperties, &NiagaraSpriteRendererProperties]
  renderers [
   {rendererClass=&NiagaraMeshRendererProperties, rendererIndex=0}
   {rendererClass=&NiagaraMeshRendererProperties, rendererIndex=1}
   {rendererClass=&NiagaraSpriteRendererProperties, rendererIndex=2}
   {rendererClass=&NiagaraSpriteRendererProperties, rendererIndex=3}
   {rendererClass=&NiagaraSpriteRendererProperties, rendererIndex=4}
  ]
 }
 Pointers {
  emitterUpdateScript [
   &EmitterState(Inactive Response=::NewEnumerator1, Life Cycle Mode=::NewEnumerator1, Loop Duration=1, UseLoopDelay=true)
   &SpawnRate(Use Spawn Probability=false)
  ]
  particleSpawnScript [
   &PartitionParticles()
   /SetVariables(Particles.MeshRotationEuler:Vector3f=&RandomRangeVector(Minimum=<0 0 0>, Maximum=<0 0 1>, Evaluation Type=::NewEnumerator1)) {isSetParameters=true}
   /SetVariables(Particles.PointerMeshOrientation:Quat4f=/CascadeToNiagaraConverter/NiagaraScripts/DynamicInputScripts/MakeQuatFromEuler(Euler:Vector3f=$Particles.MeshRotationEuler, Scale:NiagaraFloat=360)) {isSetParameters=true}
   /SetVariables(Particles.MeshRotationRateBase:Vector3f=&RandomRangeVector(Minimum=<0 0 -0.5>, Maximum=<0 0 0.5>, Evaluation Type=::NewEnumerator1)) {isSetParameters=true}
   &InitializeParticle(
    Color=&MergeHelper_ColorSwitch(0=<rgba 20 5 0 1>, 1=<rgba 0 20 0 1>, 2=?, 3=?, Index=$Particles.PartitionParticles.Partition, Num=::NewEnumerator0)
    Lifetime=&RandomRangeFloat(Minimum=0.75, Maximum=1, Evaluation Type=::NewEnumerator1)
    Mesh Scale=&RandomRangeVector(Minimum=<3 2 4>, Maximum=<4.5 6 4>, Evaluation Type=::NewEnumerator1)
    Mesh Scale Mode=::NewEnumerator4
    UsePositionOffset=false
    Write Material Random=true
   )
  ]
  particleUpdateScript [
   /SetVariables(Particles.PointerRate:Vector3f=$Particles.MeshRotationRateBase) {isSetParameters=true}
   &ParticleState()
   &ScaleColor(
    Scale Alpha=&FloatFromCurve(FloatCurve=DataInterface() {propertyValues {Curve {keys [
         {arriveTangent=7.7051, interpMode=RCIM_Cubic, leaveTangent=7.7051, tangentMode=RCTM_User, time=0, value=0}
         <at 0.2 0.2>
         {arriveTangent=-0.25, interpMode=RCIM_Cubic, leaveTangent=-0.25, time=0.8, value=0.2}
         {arriveTangent=-7.0502, leaveTangent=-7.0502, time=1, value=0}
        ]}}})
    ScaleA=true
    ScaleRGB=true
   )
   /CascadeToNiagaraConverter/NiagaraScripts/ModuleScripts/UpdateMeshOrientationByEuler(RotationRate:Vector3f=$Particles.PointerRate)
  ]
  rendererClasses [&NiagaraMeshRendererProperties]
  renderers [{rendererClass=&NiagaraMeshRendererProperties, rendererIndex=0}]
 }
 handCounters {
  emitterUpdateScript [
   &EmitterState(
    Inactive Response=::NewEnumerator1
    Life Cycle Mode=::NewEnumerator1
    Loop Behavior=::NewEnumerator1
    Loop Duration Mode=::NewEnumerator1
    UseLoopDelay=true
   )
   &SpawnBurst_Instantaneous(Spawn Count=4, Spawn Group=4, Use Spawn Probability=false, UseLoopCountLimit=false) {module=SpawnBurst_Instantaneous004}
   &SpawnBurst_Instantaneous(Spawn Count=5, Use Spawn Probability=false, UseLoopCountLimit=false)
   &SpawnBurst_Instantaneous(Spawn Group=1, Use Spawn Probability=false, UseLoopCountLimit=false) {module=SpawnBurst_Instantaneous002}
   &SpawnBurst_Instantaneous(Spawn Group=2, Use Spawn Probability=false, UseLoopCountLimit=false) {module=SpawnBurst_Instantaneous001}
   &SpawnBurst_Instantaneous(Spawn Count=2, Spawn Group=3, Use Spawn Probability=false, UseLoopCountLimit=false) {module=SpawnBurst_Instantaneous003}
  ]
  particleSpawnScript [
   /SetVariables(Particles.VisibilityTag=&SpawnGroup()) {isSetParameters=true}
   &PartitionParticles(Number Of Partitions=&MergeHelper_IntSwitch(0=5, 1=0, 2=0, 3=2, 4=4, 5=?, Index=&SpawnGroup(), Num=::NewEnumerator3))
   &InitializeParticle(
    Sprite Rotation Angle=90
    Sprite Rotation Mode=::NewEnumerator0
    Sprite Size=&MergeHelper_Vector2Switch(
     0=&MergeHelper_Vector2Switch(0=<35 20>, 1=<35 20>, 2=<17.5 10>, 3=<17.5 10>, 4=<15 8>, Index=$Particles.PartitionParticles.Partition, Num=::NewEnumerator3)
     1=<15 8>
     2=<15 8>
     3=<2.1895 5>
     4=&MergeHelper_Vector2Switch(0=<35 20>, 1=<35 20>, 2=<17.5 10>, 3=<17.5 10>, 4=<15 8>, Index=$Particles.PartitionParticles.Partition, Num=::NewEnumerator3)
     Index=&SpawnGroup()
     Num=::NewEnumerator3
    )
    Sprite Size Mode=::NewEnumerator4
    UsePositionOffset=false
    Write Material Random=true
   )
   &Color(Color=&MakeLinearColorFromVectorAndFloat("Vector (RGB)"=<0 10 0.3333>, "Float (Alpha)"=0.8))
   /SetVariables(
    Particles.ContrailPos:NiagaraPosition=&ConvertPositionToVector(
     Input Position=&MergeHelper_VectorSwitch(
      0=<0 0 0>
      1=<-70 0 200>
      2=<-70 0 190>
      3=&MergeHelper_VectorSwitch(0=<-60 0 200>, 1=<-60 0 190>, 2=?, 3=?, Index=$Particles.PartitionParticles.Partition, Num=::NewEnumerator0)
      Index=&SpawnGroup()
      Num=::NewEnumerator2
     )
     Conversion Operation:ENiagara_LWCConvertPosToVec=::NewEnumerator0
    )
    Particles.TicRot=90
   ) {isSetParameters=true}
   /SetVariables(Particles.TicColor:LinearColor=&MergeHelper_ColorSwitch(0=<rgba 10 0.8487 0 1>, 1=<rgba 10 8.3446 0 1>, 2=?, 3=?, Index=$Particles.PartitionParticles.Partition, Num=::NewEnumerator0)) {isSetParameters=true}
   /SetVariables(Particles.TicAlignment:Vector3f=<0 0 1>) {enabled=false, isSetParameters=true}
   &ApplyInitialForces()
   &SpriteLockAxis(
    Align To Orientation=true
    Lock Axis Type:ENiagara_SpriteLockAxis=::NewEnumerator6
    UseAlignmentCoordinateSpace=false
    UseFacingCoordinateSpace=true
   )
   &SpriteFacingAndAlignment(Sprite Alignment=<1 0 0>, Sprite Facing=<0 1 0>, WriteAlignment=false, WriteFacing=false) {enabled=false}
   /SetVariables(Particles.BackingColor:LinearColor=<rgba 0 10 0.0435 0.3>) {isSetParameters=true}
  ]
  particleUpdateScript [
   &ParticleState(Kill Particles When Lifetime Has Elapsed=false, Let Infinitely Lived Particles Die When Emitter Deactivates=true)
   &SkeletalMeshLocation(
    Bone / Socket Index=&SelectIntFromArray(Direct Array Index=$Particles.PartitionParticles.Partition, Int Selection Array=DataInterface() {propertyValues {IntData [0,1,2,3,4]}})
    Bone Sampling Mode=::NewEnumerator4
    MaskSpawnGroup=false
    Orientation Sampling=::NewEnumerator1
    Sampled Position Offset=&SelectVectorFromArray(
     Direct Array Index=$Particles.PartitionParticles.Partition
     Vector Selection Array=DataInterface() {propertyValues {FloatData [<-20 -10 -20>, <20 10 20>, <-20 -10 30>, <20 10 -30>, <0 0 -40>]}}
    )
    Skeletal Mesh=DataInterface() {propertyValues {
      DefaultMesh=null
      ExcludeBoneName=root
      OptionalComponentClass=null
      PreviewMesh=/Game/Characters/Player/Male/Male_Avg_Base/Fortnite_M_Avg_Player
      SkinningMode=SkinOnTheFly
      SoftSourceActor=null
      SourceMode=Default
      UvSetIndex=0
      WholeMeshLOD=-1
      bExcludeBone=false
      bReadDeformedGeometry=true
      bRequireCurrentFrameData=true
      ComponentTags []
      FilteredBones [hand_r,hand_l,foot_r,foot_l,spine_03]
      FilteredSockets []
      MeshUserParameter {parameter {name=None, typeDefHandle {registeredTypeIndex=1}, varData []}}
      SamplingRegions []
     }}
   )
  ]
  rendererClasses [&NiagaraSpriteRendererProperties]
  renderers [
   {rendererClass=&NiagaraSpriteRendererProperties}
   {rendererClass=&NiagaraSpriteRendererProperties, rendererIndex=1}
   {rendererClass=&NiagaraSpriteRendererProperties, rendererIndex=2}
   {rendererClass=&NiagaraSpriteRendererProperties, rendererIndex=3}
   {rendererClass=&NiagaraSpriteRendererProperties, rendererIndex=4}
  ]
 }
}
summary {
 systemName=NS_Contrail_TacticalHUD
 emitters [
  {emitterName=handCounters, rendererClasses [&NiagaraSpriteRendererProperties]}
  {emitterName=DiskLocal, rendererClasses [&NiagaraMeshRendererProperties, &NiagaraSpriteRendererProperties]}
  {emitterName=Pointers, rendererClasses [&NiagaraMeshRendererProperties]}
 ]
 userVariables [
  User.bIsFrontEnd {type=NiagaraBool}
  User.bIsFrontEndPreview {type=NiagaraBool}
  User.ParaGlideLeanAlpha {type=NiagaraFloat, defaultValue=0, description=""}
 ]
}
