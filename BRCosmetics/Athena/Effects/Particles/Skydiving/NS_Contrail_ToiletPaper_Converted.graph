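# Niagara skydiving contrail cosmetic that trails toilet-paper ribbons from rolls attached to the character's hands and feet.
# Four mesh emitters re-sample bones or sockets on User.SK_Mesh each update to keep the rolls attached.
# Four ribbon emitters spawn at the same offsets, tinted over lifetime, with per-limb material parameters.
# Front-end modifiers switch ribbons from distance-based spawning in gameplay to constant-rate spawning with upward velocity in the front end.
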NiagaraSystem /BRCosmetics/Athena/Effects/Particles/Skydiving/NS_Contrail_ToiletPaper_Converted
dependencies {
 usedDataInterfaces [&NiagaraDataInterfaceVectorCurve]
 usedDynamicInputs []
 usedModules [&SystemState]
}
emitters {
 Ribbon_Left_Foot {
  emitterUpdateScript [
   &EmitterState()
   &SpawnRate(
    Spawn Probability=&FN_FrontEndFloatModifier(Gameplay=0, Front End=1)
    SpawnRate=120
    Use Spawn Probability=true
   )
   &SpawnPerUnit(
    Max Movement Threshold=50000
    Module.Use Max Movement Threshold=true
    Module.Use Movement Tolerance=true
    Spawn Probability:NiagaraFloat=&FN_FrontEndFloatModifier(Gameplay=1, Front End=0)
    Spawn Spacing=5
    Use Spawn Probability=true
   )
  ]
  particleSpawnScript [
   &InitializeParticle(Ribbon Width=15, Ribbon Width Mode=::NewEnumerator1, UsePositionOffset=false, Write Material Random=true)
   &SkeletalMeshLocation(
    Bone / Socket Index=2
    Bone Sampling Mode=::NewEnumerator4
    MaskSpawnGroup=false
    Sampled Position Offset=<0 -5 -20>
    Skeletal Mesh=$User.SK_Mesh
   )
   &Color(Color=&MakeLinearColorFromVectorAndFloat("Vector (RGB)"=<0.7 1 0.2>, "Float (Alpha)"=&ClampFloat(Float=1)))
   &DynamicMaterialParameters(
    Index 0 Param 1=5
    Index 0 Param 2=0.4
    Module.Param0WriteEnabled=true
    Module.Param1WriteEnabled=true
    Module.Param2WriteEnabled=false
    Module.Param3WriteEnabled=false
   )
   &AddVelocity(Velocity=&MakeVector(X=0, Y=0, Z=&FN_FrontEndFloatModifier(Gameplay=0, Front End=500)))
  ]
  particleUpdateScript [
   &ParticleState()
   &ScaleColor(
    Scale RGB=&VectorFromCurve(VectorCurve=DataInterface() {propertyValues {
       XCurve {keys [<at 0 1>,<at 0.03 1>,<at 0.8 2>,<at 1 2>]}
       YCurve {keys [<at 0 1>,<at 0.03 0>,<at 0.8 1>,<at 1 1>]}
       ZCurve {keys [<at 0 0>,<at 0.03 0>,<at 0.8 1>,<at 1 3>]}
      }})
    ScaleA=true
    ScaleRGB=true
   )
   &DynamicMaterialParameters(
    Index 0 Param 3=8
    Index 0 Param 4=1
    Module.Param0WriteEnabled=false
    Module.Param1WriteEnabled=false
    Module.Param2WriteEnabled=true
    Module.Param3WriteEnabled=true
   ) {module=DynamicMaterialParameters001}
   &CameraOffset(Camera Offset Amount=-30)
   &SolveForcesAndVelocity(Clamp Velocity=false, Limit Acceleration=false)
  ]
  rendererClasses [&NiagaraRibbonRendererProperties]
  renderers [{rendererClass=&NiagaraRibbonRendererProperties}]
 }
 Ribbon_Left_Hand {
  emitterUpdateScript [
   &EmitterState()
   &SpawnRate(
    Spawn Probability=&FN_FrontEndFloatModifier(Gameplay=0, Front End=1)
    SpawnRate=120
    Use Spawn Probability=true
   )
   &SpawnPerUnit(
    Max Movement Threshold=50000
    Module.Use Max Movement Threshold=true
    Module.Use Movement Tolerance=true
    Spawn Probability:NiagaraFloat=&FN_FrontEndFloatModifier(Gameplay=1, Front End=0)
    Spawn Spacing=5
    Use Spawn Probability=true
   )
  ]
  particleSpawnScript [
   &InitializeParticle(Ribbon Width=15, Ribbon Width Mode=::NewEnumerator1, UsePositionOffset=false, Write Material Random=true)
   &SkeletalMeshLocation(Bone Sampling Mode=::NewEnumerator4, MaskSpawnGroup=false, Sampled Position Offset=<15 -7 -3>, Skeletal Mesh=$User.SK_Mesh)
   &Color(Color=&MakeLinearColorFromVectorAndFloat("Vector (RGB)"=<0.7 1 0.2>, "Float (Alpha)"=&ClampFloat(Float=1)))
   &DynamicMaterialParameters(
    Index 0 Param 2=0.4
    Module.Param0WriteEnabled=true
    Module.Param1WriteEnabled=true
    Module.Param2WriteEnabled=false
    Module.Param3WriteEnabled=false
   )
   &AddVelocity(Velocity=&MakeVector(X=0, Y=0, Z=&FN_FrontEndFloatModifier(Gameplay=0, Front End=500)))
  ]
  particleUpdateScript [
   &ParticleState()
   &ScaleColor(
    Scale RGB=&VectorFromCurve(VectorCurve=DataInterface() {propertyValues {
       XCurve {keys [<at 0 1>,<at 0.03 1>,<at 0.8 2>,<at 1 2>]}
       YCurve {keys [<at 0 1>,<at 0.03 0>,<at 0.8 1>,<at 1 1>]}
       ZCurve {keys [<at 0 0>,<at 0.03 0>,<at 0.8 1>,<at 1 3>]}
      }})
    ScaleA=true
    ScaleRGB=true
   )
   &DynamicMaterialParameters(
    Index 0 Param 3=1
    Index 0 Param 4=1
    Module.Param0WriteEnabled=false
    Module.Param1WriteEnabled=false
    Module.Param2WriteEnabled=true
    Module.Param3WriteEnabled=true
   ) {module=DynamicMaterialParameters001}
   &CameraOffset(Camera Offset Amount=-30)
   &SolveForcesAndVelocity(Clamp Velocity=false, Limit Acceleration=false)
  ]
  rendererClasses [&NiagaraRibbonRendererProperties]
  renderers [{rendererClass=&NiagaraRibbonRendererProperties}]
 }
 Ribbon_Right_Foot {
  emitterUpdateScript [
   &EmitterState()
   &SpawnRate(
    Spawn Probability=&FN_FrontEndFloatModifier(Gameplay=0, Front End=1)
    SpawnRate=120
    Use Spawn Probability=true
   )
   &SpawnPerUnit(
    Max Movement Threshold=50000
    Module.Use Max Movement Threshold=true
    Module.Use Movement Tolerance=true
    Spawn Probability:NiagaraFloat=&FN_FrontEndFloatModifier(Gameplay=1, Front End=0)
    Spawn Spacing=5
    Use Spawn Probability=true
   )
  ]
  particleSpawnScript [
   &InitializeParticle(Ribbon Width=15, Ribbon Width Mode=::NewEnumerator1, UsePositionOffset=false, Write Material Random=true)
   &SkeletalMeshLocation(
    Bone / Socket Index=3
    Bone Sampling Mode=::NewEnumerator4
    MaskSpawnGroup=false
    Sampled Position Offset=<0 7 16>
    Skeletal Mesh=$User.SK_Mesh
   )
   &Color(Color=&MakeLinearColorFromVectorAndFloat("Vector (RGB)"=<0.7 1 0.2>, "Float (Alpha)"=&ClampFloat(Float=1)))
   &DynamicMaterialParameters(
    Index 0 Param 1=20
    Index 0 Param 2=0.8
    Module.Param0WriteEnabled=true
    Module.Param1WriteEnabled=true
    Module.Param2WriteEnabled=false
    Module.Param3WriteEnabled=false
   )
   &AddVelocity(Velocity=&MakeVector(X=0, Y=0, Z=&FN_FrontEndFloatModifier(Gameplay=0, Front End=500)))
  ]
  particleUpdateScript [
   &ParticleState()
   &ScaleColor(
    Scale RGB=&VectorFromCurve(VectorCurve=DataInterface() {propertyValues {
       XCurve {keys [<at 0 1>,<at 0.03 1>,<at 0.8 2>,<at 1 2>]}
       YCurve {keys [<at 0 1>,<at 0.03 0>,<at 0.8 1>,<at 1 1>]}
       ZCurve {keys [<at 0 0>,<at 0.03 0>,<at 0.8 1>,<at 1 3>]}
      }})
    ScaleA=true
    ScaleRGB=true
   )
   &DynamicMaterialParameters(
    Index 0 Param 3=15
    Index 0 Param 4=1
    Module.Param0WriteEnabled=false
    Module.Param1WriteEnabled=false
    Module.Param2WriteEnabled=true
    Module.Param3WriteEnabled=true
   ) {module=DynamicMaterialParameters001}
   &CameraOffset(Camera Offset Amount=-30)
   &SolveForcesAndVelocity(Clamp Velocity=false, Limit Acceleration=false)
  ]
  rendererClasses [&NiagaraRibbonRendererProperties]
  renderers [{rendererClass=&NiagaraRibbonRendererProperties}]
 }
 Ribbon_Right_Hand {
  emitterUpdateScript [
   &EmitterState()
   &SpawnRate(
    Spawn Probability=&FN_FrontEndFloatModifier(Gameplay=0, Front End=1)
    SpawnRate=120
    Use Spawn Probability=true
   )
   &SpawnPerUnit(
    Max Movement Threshold=50000
    Module.Use Max Movement Threshold=true
    Module.Use Movement Tolerance=true
    Spawn Probability:NiagaraFloat=&FN_FrontEndFloatModifier(Gameplay=1, Front End=0)
    Spawn Spacing=5
    Use Spawn Probability=true
   )
  ]
  particleSpawnScript [
   &InitializeParticle(Ribbon Width=15, Ribbon Width Mode=::NewEnumerator1, UsePositionOffset=false, Write Material Random=true)
   &SkeletalMeshLocation(
    Bone / Socket Index=1
    Bone Sampling Mode=::NewEnumerator4
    MaskSpawnGroup=false
    Sampled Position Offset=<-15 9 0>
    Skeletal Mesh=$User.SK_Mesh
   )
   &Color(Color=&MakeLinearColorFromVectorAndFloat("Vector (RGB)"=<0.7 1 0.2>, "Float (Alpha)"=&ClampFloat(Float=1)))
   &DynamicMaterialParameters(
    Index 0 Param 2=0.8
    Module.Param0WriteEnabled=false
    Module.Param1WriteEnabled=true
    Module.Param2WriteEnabled=false
    Module.Param3WriteEnabled=false
   ) {module=DynamicMaterialParameters001}
   &AddVelocity(Velocity=&MakeVector(X=0, Y=0, Z=&FN_FrontEndFloatModifier(Gameplay=0, Front End=500)))
  ]
  particleUpdateScript [
   &ParticleState()
   &ScaleColor(
    Scale RGB=&VectorFromCurve(VectorCurve=DataInterface() {propertyValues {
       XCurve {keys [<at 0 1>,<at 0.03 1>,<at 0.8 2>,<at 1 2>]}
       YCurve {keys [<at 0 1>,<at 0.03 0>,<at 0.8 1>,<at 1 1>]}
       ZCurve {keys [<at 0 0>,<at 0.03 0>,<at 0.8 1>,<at 1 3>]}
      }})
    ScaleA=true
    ScaleRGB=true
   )
   &DynamicMaterialParameters(
    Index 0 Param 1=15
    Index 0 Param 3=5
    Index 0 Param 4=1
    Module.Param0WriteEnabled=true
    Module.Param1WriteEnabled=false
    Module.Param2WriteEnabled=true
    Module.Param3WriteEnabled=true
   )
   &CameraOffset(Camera Offset Amount=-30)
   &SolveForcesAndVelocity(Clamp Velocity=false, Limit Acceleration=false)
  ]
  rendererClasses [&NiagaraRibbonRendererProperties]
  renderers [{rendererClass=&NiagaraRibbonRendererProperties}]
 }
 RollLeftFoot {
  emitterUpdateScript [
   &EmitterState(Life Cycle Mode=::NewEnumerator1, Loop Duration=1, UseLoopDelay=true)
   &SpawnBurst_Instantaneous(Use Spawn Probability=false, UseLoopCountLimit=false)
  ]
  particleSpawnScript [
   &InitializeParticle(
    Mesh Scale=<0.7 0.7 0.7>
    Mesh Scale Mode=::NewEnumerator4
    Sprite Rotation Mode=::NewEnumerator1
    UsePositionOffset=false
    Write Material Random=true
   )
   &Color(Color=&MakeLinearColorFromVectorAndFloat("Vector (RGB)"=<1 1 1>, "Float (Alpha)"=&ClampFloat(Float=1)))
   &InitialMeshOrientation(Mesh Orientation Mode=::NewEnumerator2, Rotation=<0 0.25 0>, Use Rotation Vector=true)
  ]
  particleUpdateScript [
   &ParticleState(Kill Particles When Lifetime Has Elapsed=false, Let Infinitely Lived Particles Die When Emitter Deactivates=true)
   &SkeletalMeshLocation(
    Bone / Socket Index=2
    Bone Sampling Mode=::NewEnumerator4
    MaskSpawnGroup=false
    Sampled Position Offset=<0 -5 -20>
    Skeletal Mesh=$User.SK_Mesh
   )
   &DynamicMaterialParameters(
    Index 0 Param 1=0.5
    Index 0 Param 3=1
    Index 0 Param 4=1
    Module.Param0WriteEnabled=true
    Module.Param1WriteEnabled=true
    Module.Param2WriteEnabled=true
    Module.Param3WriteEnabled=true
   )
  ]
  rendererClasses [&NiagaraMeshRendererProperties]
  renderers [{rendererClass=&NiagaraMeshRendererProperties, rendererIndex=0}]
 }
 RollLeftHand {
  emitterUpdateScript [
   &EmitterState(Life Cycle Mode=::NewEnumerator1, Loop Duration=1, UseLoopDelay=true)
   &SpawnBurst_Instantaneous(Use Spawn Probability=false, UseLoopCountLimit=false)
  ]
  particleSpawnScript [
   &InitializeParticle(
    Mesh Scale=<0.7 0.7 0.7>
    Mesh Scale Mode=::NewEnumerator4
    Sprite Rotation Mode=::NewEnumerator1
    UsePositionOffset=false
    Write Material Random=true
   )
   &Color(Color=&MakeLinearColorFromVectorAndFloat("Vector (RGB)"=<1 1 1>, "Float (Alpha)"=&ClampFloat(Float=1)))
   &InitialMeshOrientation(Mesh Orientation Mode=::NewEnumerator2, Rotation=<0 -0.25 0>, Use Rotation Vector=true)
  ]
  particleUpdateScript [
   &ParticleState(Kill Particles When Lifetime Has Elapsed=false, Let Infinitely Lived Particles Die When Emitter Deactivates=true)
   &SkeletalMeshLocation(Bone Sampling Mode=::NewEnumerator4, MaskSpawnGroup=false, Sampled Position Offset=<15 -7 -3>, Skeletal Mesh=$User.SK_Mesh)
   &DynamicMaterialParameters(
    Index 0 Param 2=1
    Index 0 Param 3=1
    Index 0 Param 4=1
    Module.Param0WriteEnabled=true
    Module.Param1WriteEnabled=true
    Module.Param2WriteEnabled=true
    Module.Param3WriteEnabled=true
   )
  ]
  rendererClasses [&NiagaraMeshRendererProperties]
  renderers [{rendererClass=&NiagaraMeshRendererProperties, rendererIndex=0}]
 }
 RollRightFoot {
  emitterUpdateScript [
   &EmitterState(Life Cycle Mode=::NewEnumerator1, Loop Duration=1, UseLoopDelay=true)
   &SpawnBurst_Instantaneous(Use Spawn Probability=false, UseLoopCountLimit=false)
  ]
  particleSpawnScript [
   &InitializeParticle(
    Mesh Scale=<0.7 0.7 0.7>
    Mesh Scale Mode=::NewEnumerator4
    Sprite Rotation Mode=::NewEnumerator1
    UsePositionOffset=false
    Write Material Random=true
   )
   &Color(Color=&MakeLinearColorFromVectorAndFloat("Vector (RGB)"=<1 1 1>, "Float (Alpha)"=&ClampFloat(Float=1)))
   &InitialMeshOrientation(Mesh Orientation Mode=::NewEnumerator2, Rotation=<0 0.25 0>, Use Rotation Vector=true)
  ]
  particleUpdateScript [
   &ParticleState(Kill Particles When Lifetime Has Elapsed=false, Let Infinitely Lived Particles Die When Emitter Deactivates=true)
   &SkeletalMeshLocation(
    Bone / Socket Index=3
    Bone Sampling Mode=::NewEnumerator4
    MaskSpawnGroup=false
    Sampled Position Offset=<0 7 16>
    Skeletal Mesh=$User.SK_Mesh
   )
   &DynamicMaterialParameters(
    Index 0 Param 1=0.75
    Index 0 Param 2=1
    Index 0 Param 3=1
    Index 0 Param 4=1
    Module.Param0WriteEnabled=true
    Module.Param1WriteEnabled=true
    Module.Param2WriteEnabled=true
    Module.Param3WriteEnabled=true
   )
  ]
  rendererClasses [&NiagaraMeshRendererProperties]
  renderers [{rendererClass=&NiagaraMeshRendererProperties, rendererIndex=0}]
 }
 RollRightHand {
  emitterUpdateScript [
   &EmitterState(Life Cycle Mode=::NewEnumerator1, Loop Duration=1, UseLoopDelay=true)
   &SpawnBurst_Instantaneous(Use Spawn Probability=false, UseLoopCountLimit=false)
  ]
  particleSpawnScript [
   &InitializeParticle(
    Mesh Scale=<0.7 0.7 0.7>
    Mesh Scale Mode=::NewEnumerator4
    Sprite Rotation Mode=::NewEnumerator1
    UsePositionOffset=false
    Write Material Random=true
   )
   &Color(Color=&MakeLinearColorFromVectorAndFloat("Vector (RGB)"=<1 1 1>, "Float (Alpha)"=&ClampFloat(Float=1)))
   &InitialMeshOrientation(Mesh Orientation Mode=::NewEnumerator2, Rotation=<0 0.25 0>, Use Rotation Vector=true)
  ]
  particleUpdateScript [
   &ParticleState(Kill Particles When Lifetime Has Elapsed=false, Let Infinitely Lived Particles Die When Emitter Deactivates=true)
   &SkeletalMeshLocation(
    Bone / Socket Index=1
    Bone Sampling Mode=::NewEnumerator4
    MaskSpawnGroup=false
    Sampled Position Offset=<-15 9 0>
    Skeletal Mesh=$User.SK_Mesh
   )
   &DynamicMaterialParameters(
    Index 0 Param 1=0.25
    Index 0 Param 2=1
    Index 0 Param 3=1
    Index 0 Param 4=1
    Module.Param0WriteEnabled=true
    Module.Param1WriteEnabled=true
    Module.Param2WriteEnabled=true
    Module.Param3WriteEnabled=true
   )
  ]
  rendererClasses [&NiagaraMeshRendererProperties]
  renderers [{rendererClass=&NiagaraMeshRendererProperties, rendererIndex=0}]
 }
}
summary {
 systemName=NS_Contrail_ToiletPaper_Converted
 emitters [
  {emitterName=RollLeftHand, rendererClasses [&NiagaraMeshRendererProperties]}
  {emitterName=RollRightHand, rendererClasses [&NiagaraMeshRendererProperties]}
  {emitterName=RollLeftFoot, rendererClasses [&NiagaraMeshRendererProperties]}
  {emitterName=RollRightFoot, rendererClasses [&NiagaraMeshRendererProperties]}
  {emitterName=Ribbon_Left_Hand, rendererClasses [&NiagaraRibbonRendererProperties]}
  {emitterName=Ribbon_Right_Hand, rendererClasses [&NiagaraRibbonRendererProperties]}
  {emitterName=Ribbon_Left_Foot, rendererClasses [&NiagaraRibbonRendererProperties]}
  {emitterName=Ribbon_Right_Foot, rendererClasses [&NiagaraRibbonRendererProperties]}
 ]
 userVariables [
  User.bIsFrontEnd {type=NiagaraBool}
  User.bIsFrontEndPreview {type=NiagaraBool}
  User.SK_Mesh {
   type=NiagaraDataInterfaceSkeletalMesh
   defaultValue {
    struct=NiagaraExt_VariableValue_DataInterface
    value {dataInterface=&self:NiagaraDataInterfaceSkeletalMesh_0}
   }
  }
 ]
}
