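# Skydiving contrail cosmetic that trails flames, smoke and light from the character's skeletal mesh, with green-tinted fire and purple-pink smoke colors.
# Flame and smoke sprites, fire and smoke ribbons, and lights spawn at two hand points via a shared mesh reader.
# Flame balls re-sample two other bone/socket points every update to stay attached.
# Spawn groups select the attachment point, and front-end modifiers add velocity to fake motion in menus.
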NiagaraSystem /BRCosmetics/Athena/Effects/Particles/Skydiving/NS_Contrail_VioletFlame
dependencies {
 usedDataInterfaces [
  &NiagaraDataInterfaceCurve
  &NiagaraDataInterfaceSpriteRendererInfo
  &NiagaraDataInterfaceVectorCurve
  &NiagaraDataInterfaceSkeletalMesh
 ]
 usedDynamicInputs []
 usedModules [
  &self:SystemSpawnScript.SystemScriptSource.SystemScriptGraph.NiagaraNodeAssignment_0.SetVariables_#1
  &SystemState
 ]
}
emitters {
 Fire_hand_Ribbon {
  emitterUpdateScript [
   &EmitterState(Inactive Response=::NewEnumerator1, Life Cycle Mode=::NewEnumerator1, Loop Duration=1, UseLoopDelay=true)
   &SpawnRate(SpawnRate=200, Use Spawn Probability=false)
   &SpawnRate(SpawnGroup=1, SpawnRate=200, Use Spawn Probability=false) {module=SpawnRate001}
  ]
  particleSpawnScript [
   &InitializeParticle(
    Lifetime=0.1
    Ribbon Width=&Multiply_Float(A=11, B=2)
    Ribbon Width Mode=::NewEnumerator1
    UsePositionOffset=false
    Write Material Random=true
   )
   /CascadeToNiagaraConverter/NiagaraScripts/ModuleScripts/AddVectorToPosition(Vector:Vector3f=&TransformVector(Vector=<0 0 0>, Destination Space=::Simulation))
   &SkeletalMeshLocation(
    Bone / Socket Index=&SpawnGroup()
    Bone Sampling Mode=::NewEnumerator4
    MaskSpawnGroup=false
    Orientation Sampling=::NewEnumerator1
    Skeletal Mesh=$System.Skeletal Mesh Reader
   )
   &Color(Color=&MakeLinearColorFromVectorAndFloat("Vector (RGB)"=<0.5903 0.9427 0.0121>, "Float (Alpha)"=3.5))
   &AddVelocity(Velocity=&FN_FrontEndVector3Modifier(Gameplay=<0 0 0>, Front End=<0 -300 1100>)) {module=AddVelocity001}
  ]
  particleUpdateScript [
   &ParticleState()
   &SolveForcesAndVelocity(Clamp Velocity=false, Limit Acceleration=false)
   &DynamicMaterialParameters(
    Index 0 Param 1=0.9
    Index 0 Param 2=2
    Index 0 Param 3=1
    Index 0 Param 4=1
    Module.Index0Float0Write=true
    Module.Index0Float1Write=true
    Module.Index0Float2Write=true
    Module.Index0Float3Write=true
   ) {module=DynamicMaterialParameters001}
  ]
  rendererClasses [&NiagaraRibbonRendererProperties]
  renderers [{rendererClass=&NiagaraRibbonRendererProperties}]
 }
 Flame_ball {
  emitterUpdateScript [
   &EmitterState(Inactive Response=::NewEnumerator1, Life Cycle Mode=::NewEnumerator1, Loop Duration=1, UseLoopDelay=true)
   &SpawnRate(SpawnGroup=2, SpawnRate=3, Use Spawn Probability=false)
   &SpawnRate(SpawnGroup=3, SpawnRate=3, Use Spawn Probability=false) {module=SpawnRate001}
  ]
  particleSpawnScript [
   &InitializeParticle(
    Lifetime=&RandomRangeFloat(Minimum=0.4, Maximum=0.75, Evaluation Type=::NewEnumerator1)
    Sprite Rotation Angle=&RandomRangeFloat(Minimum=0, Maximum=1, Evaluation Type=::NewEnumerator1)
    Sprite Rotation Mode=::NewEnumerator1
    Sprite Size=&Vector2DFromFloat(Value=&RandomRangeFloat(Minimum=25, Maximum=40, Evaluation Type=::NewEnumerator1))
    Sprite Size Mode=::NewEnumerator4
    UsePositionOffset=true
    Write Material Random=true
   )
   &Color(Color=&MakeLinearColorFromVectorAndFloat("Vector (RGB)"=<1.78 5 0.026>, "Float (Alpha)"=0.75))
   &ShapeLocation(MaskSpawnGroup=false, Override Seed=false, Sphere Radius=20, UseOwnerScale=true)
   /CascadeToNiagaraConverter/NiagaraScripts/ModuleScripts/FixShapeVector()
   &AddVelocity(Velocity=&Multiply_VectorByFloat(Vector=$Output.FixShapeVector.ShapeVector, Float=6))
   &AddVectorToPosition(Vector:Vector3f=&TransformVector(Vector=<0 0 -2>, Destination Space=::Simulation))
   &InheritVelocity(
    Inherited Velocity Amount Scale=&VectorFromFloat(Value=&RandomRangeFloat(Minimum=-0.028, Maximum=-0.15, Evaluation Type=::NewEnumerator1))
    Inherited Velocity Speed Limit=0
    Speed Limit=false
    UseSpeedThreshold=false
   )
  ]
  particleUpdateScript [
   &ParticleState()
   &ScaleSpriteSize(
    Scale Sprite Size Mode=::NewEnumerator0
    Uniform Scale Factor=&FloatFromCurve(FloatCurve=DataInterface() {propertyValues {Curve {keys [
         {arriveTangent=11.6538, interpMode=RCIM_Cubic, leaveTangent=11.6538, tangentMode=RCTM_User, time=0, value=0.0211}
         {arriveTangent=0.5535, interpMode=RCIM_Cubic, leaveTangent=0.5535, tangentMode=RCTM_User, time=0.1836, value=0.9874}
         <at 1 0.5 RCIM_Cubic>
        ]}}})
   )
   &SubUVAnimation(
    End Frame Range Override=62
    "Sprite Renderer "=DataInterface() {propertyValues {SpriteRenderer=&self:Flame_ball_0.NiagaraSpriteRendererProperties_24}}
    UseEndFrame=false
    UseStartFrame=false
   )
   &DynamicMaterialParameters(
    Index 0 Param 2=1
    Index 0 Param 3=1
    Index 0 Param 4=1
    Module.Index0Float0Write=true
    Module.Index0Float1Write=true
    Module.Index0Float2Write=true
    Module.Index0Float3Write=true
   ) {module=DynamicMaterialParameters001}
   &ScaleColor(
    Scale Alpha=&FloatFromCurve(FloatCurve=DataInterface() {propertyValues {Curve {keys [
         {arriveTangent=11.7282, interpMode=RCIM_Cubic, leaveTangent=11.7282, tangentMode=RCTM_User, time=0, value=0}
         <at 0.2 1 RCIM_Cubic>
         <at 1 0>
        ]}}})
    Scale RGB=&VectorFromCurve(VectorCurve=DataInterface() {propertyValues {XCurve {keys [<at 0 1.5098>,<at 1 0.6175>]}, YCurve {keys [<at 0 4>,<at 1 1>]}, ZCurve {keys [<at 0 0>,<at 1 0>]}}})
    ScaleA=true
    ScaleRGB=true
   )
   &AccelerationForce(Acceleration=<0 0 110>)
   &Drag(Use Linear Drag=true, Use Rotational Drag=true)
   &SolveForcesAndVelocity(Clamp Velocity=false, Limit Acceleration=false)
   &CameraOffset(Camera Offset Amount=5)
   &SkeletalMeshLocation(
    Bone / Socket Index=&SpawnGroup()
    Bone Sampling Mode=::NewEnumerator4
    MaskSpawnGroup=false
    Orientation Sampling=::NewEnumerator1
    Sampled Position Offset=&MergeHelper_VectorSwitch(0=<0 0 0>, 1=<0 0 0>, 2=<0 2 0>, 3=<0 -2 0>, Index=&SpawnGroup(), Num=::NewEnumerator2)
    Skeletal Mesh=$System.Skeletal Mesh Reader
   )
  ]
  rendererClasses [&NiagaraSpriteRendererProperties]
  renderers [{rendererClass=&NiagaraSpriteRendererProperties}]
 }
 Flame_hand {
  emitterUpdateScript [
   &EmitterState(Inactive Response=::NewEnumerator1, Life Cycle Mode=::NewEnumerator1, Loop Duration=1, UseLoopDelay=true)
   &SpawnRate(SpawnRate=8, Use Spawn Probability=false)
  ]
  particleSpawnScript [
   &InitializeParticle(
    Lifetime=&RandomRangeFloat(Minimum=0.4, Maximum=0.75, Evaluation Type=::NewEnumerator1)
    Sprite Rotation Angle=&RandomRangeFloat(Minimum=0, Maximum=1, Evaluation Type=::NewEnumerator1)
    Sprite Rotation Mode=::NewEnumerator1
    Sprite Size=&Vector2DFromFloat(Value=&RandomRangeFloat(Minimum=37, Maximum=62, Evaluation Type=::NewEnumerator1))
    Sprite Size Mode=::NewEnumerator4
    UsePositionOffset=false
    Write Material Random=true
   )
   &Color(Color=&MakeLinearColorFromVectorAndFloat("Vector (RGB)"=<1.78 5 0.026>, "Float (Alpha)"=0.75))
   &AddVelocity(Velocity=<0 0 -80>)
   &ShapeLocation(MaskSpawnGroup=false, Override Seed=false, Sphere Radius=15, UseOwnerScale=true)
   &FixShapeVector()
   &AddVelocity(Velocity=&Multiply_VectorByFloat(Vector=$Output.FixShapeVector.ShapeVector, Float=2)) {module=AddVelocity001}
   &AddVectorToPosition(Vector:Vector3f=&TransformVector(Vector=<0 0 -2>, Destination Space=::Simulation))
   &SkeletalMeshLocation(
    Bone / Socket Index=/CascadeToNiagaraConverter/NiagaraScripts/DynamicInputScripts/SequentialInt(Use Array Data:NiagaraBool=false, First Number:NiagaraInt32=0, Sequence:NiagaraDataInterfaceArrayInt32=?, Last Number:NiagaraInt32=1)
    Bone Sampling Mode=::NewEnumerator4
    MaskSpawnGroup=false
    Orientation Sampling=::NewEnumerator1
    Sampled Position Coordinate Space=::Local
    Skeletal Mesh=$System.Skeletal Mesh Reader
   )
   &InheritVelocity(
    Inherited Velocity Amount Scale=&VectorFromFloat(Value=&RandomRangeFloat(Minimum=-0.035, Maximum=-0.15, Evaluation Type=::NewEnumerator1))
    Inherited Velocity Speed Limit=0
    Speed Limit=false
    UseSpeedThreshold=false
   )
   &AddVelocity(Velocity=&FN_FrontEndVector3Modifier(Gameplay=<0 0 0>, Front End=<0 -20 80>)) {module=AddVelocity002}
  ]
  particleUpdateScript [
   &ParticleState()
   &ScaleSpriteSize(
    Scale Sprite Size Mode=::NewEnumerator0
    Uniform Scale Factor=&FloatFromCurve(FloatCurve=DataInterface() {propertyValues {Curve {keys [
         {arriveTangent=11.6538, interpMode=RCIM_Cubic, leaveTangent=11.6538, tangentMode=RCTM_User, time=0, value=0.0211}
         {arriveTangent=0.5535, interpMode=RCIM_Cubic, leaveTangent=0.5535, tangentMode=RCTM_User, time=0.1836, value=0.9874}
         <at 1 0.5 RCIM_Cubic>
        ]}}})
   )
   &Drag(Use Linear Drag=true, Use Rotational Drag=true)
   &SolveForcesAndVelocity(Clamp Velocity=false, Limit Acceleration=false)
   &SubUVAnimation(
    End Frame Range Override=62
    "Sprite Renderer "=DataInterface() {propertyValues {SpriteRenderer=&self:Flame_hand_0.NiagaraSpriteRendererProperties_23}}
    UseEndFrame=false
    UseStartFrame=false
   )
   &DynamicMaterialParameters(
    Index 0 Param 2=1
    Index 0 Param 3=1
    Index 0 Param 4=1
    Module.Index0Float0Write=true
    Module.Index0Float1Write=true
    Module.Index0Float2Write=true
    Module.Index0Float3Write=true
   ) {module=DynamicMaterialParameters001}
   &ScaleColor(
    Scale Alpha=&FloatFromCurve(FloatCurve=DataInterface() {propertyValues {Curve {keys [
         {arriveTangent=11.7282, interpMode=RCIM_Cubic, leaveTangent=11.7282, tangentMode=RCTM_User, time=0, value=0}
         <at 0.2 1 RCIM_Cubic>
         <at 1 0>
        ]}}})
    Scale RGB=&VectorFromCurve(VectorCurve=DataInterface() {propertyValues {XCurve {keys [<at 0 1.5098>,<at 1 0.6175>]}, YCurve {keys [<at 0 4>,<at 1 1>]}, ZCurve {keys [<at 0 0>,<at 1 0>]}}})
    ScaleA=true
    ScaleRGB=true
   )
   &CameraOffset(Camera Offset Amount=5)
  ]
  rendererClasses [&NiagaraSpriteRendererProperties]
  renderers [{rendererClass=&NiagaraSpriteRendererProperties}]
 }
 Smoke_hand {
  emitterUpdateScript [
   &EmitterState(Inactive Response=::NewEnumerator1, Life Cycle Mode=::NewEnumerator1, Loop Duration=0.075, UseLoopDelay=false)
   &SpawnBurst_Instantaneous(Use Spawn Probability=false, UseLoopCountLimit=false)
   &SpawnBurst_Instantaneous(Spawn Group=1, Use Spawn Probability=false, UseLoopCountLimit=false) {module=SpawnBurst_Instantaneous001}
  ]
  particleSpawnScript [
   /SetVariables(Particles.SpriteRotationRate=&RandomRangeFloat(Minimum=0.1, Maximum=0.3, Evaluation Type=::NewEnumerator1)) {isSetParameters=true}
   &InitializeParticle(
    Color=<rgba 0.1369 0.0262 0.1875 0.22>
    Lifetime=&RandomRangeFloat(Minimum=0.45, Maximum=0.8, Evaluation Type=::NewEnumerator1)
    Sprite Rotation Angle=&RandomRangeFloat(Minimum=0, Maximum=1, Evaluation Type=::NewEnumerator1)
    Sprite Rotation Mode=::NewEnumerator1
    Sprite Size=&Vector2DFromFloat(Value=&RandomRangeFloat(Minimum=30, Maximum=52, Evaluation Type=::NewEnumerator1))
    Sprite Size Mode=::NewEnumerator4
    UsePositionOffset=false
    Write Material Random=true
   )
   &SkeletalMeshLocation(
    Bone / Socket Index=&SpawnGroup()
    Bone Sampling Mode=::NewEnumerator4
    MaskSpawnGroup=false
    Orientation Sampling=::NewEnumerator1
    Sampled Position Coordinate Space=::Local
    Skeletal Mesh=$System.Skeletal Mesh Reader
   )
   &AddVelocity(Velocity=<0 0 -80>)
   &ShapeLocation(MaskSpawnGroup=false, Override Seed=false, Sphere Radius=2, Sphere Surface Distribution=1, UseOwnerScale=true)
   &FixShapeVector()
   &AddVelocity(Velocity=&Multiply_VectorByFloat(
     Vector=$Output.FixShapeVector.ShapeVector
     Float=&RandomRangeFloat(Minimum=6, Maximum=13, Evaluation Type=::NewEnumerator1)
    )) {module=AddVelocity001}
   &InheritVelocity(
    Inherited Velocity Amount Scale=&VectorFromFloat(Value=&RandomRangeFloat(Minimum=-0.08, Maximum=-0.2, Evaluation Type=::NewEnumerator1))
    Inherited Velocity Speed Limit=0
    Speed Limit=false
    UseSpeedThreshold=false
   )
   &AddVelocity(Velocity=&FN_FrontEndVector3Modifier(Gameplay=<0 0 0>, Front End=<0 0 80>)) {module=AddVelocity002}
  ]
  particleUpdateScript [
   &ParticleState()
   &SpriteRotationRate(Rotation Rate=&AngleConversion(
     Angle:NiagaraFloat=$Particles.SpriteRotationRate
     Angle Input:ENiagara_AngleInput=::NewEnumerator1
     Angle Output:ENiagara_AngleInput=::NewEnumerator0
    ))
   &ScaleSpriteSize(
    Scale Sprite Size Mode=::NewEnumerator0
    Uniform Scale Factor=&FloatFromCurve(FloatCurve=DataInterface() {propertyValues {Curve {keys [
         {arriveTangent=0.0274, interpMode=RCIM_Cubic, leaveTangent=0.0274, tangentMode=RCTM_User, time=0, value=0.4}
         {arriveTangent=1.3318, interpMode=RCIM_Cubic, leaveTangent=1.3318, tangentMode=RCTM_User, time=0.399, value=0.8}
         {arriveTangent=-0.1901, interpMode=RCIM_Cubic, leaveTangent=-0.1901, tangentMode=RCTM_User, time=1, value=1.1674}
        ]}}})
   )
   &Drag(Drag=2.5, Use Linear Drag=true, Use Rotational Drag=true)
   &SubUVAnimation(
    Evaluation Type=::NewEnumerator0
    "Sprite Renderer "=DataInterface() {propertyValues {SpriteRenderer=&self:Smoke_hand_0.NiagaraSpriteRendererProperties_22}}
    SubUV Animation Mode:ENiagara_SubUVLookupModeV2=::NewEnumerator2
    UseEndFrame=false
    UseStartFrame=false
   )
   &ScaleColor(
    Scale Alpha=&FloatFromCurve(FloatCurve=DataInterface() {propertyValues {Curve {keys [
         {arriveTangent=-24.7858, interpMode=RCIM_Cubic, leaveTangent=-0.2704, tangentMode=RCTM_Break, time=0, value=0}
         {arriveTangent=-2.5683, interpMode=RCIM_Cubic, leaveTangent=-2.5683, tangentMode=RCTM_User, time=0.1352, value=0.9884}
         <at 1 0>
        ]}}})
    Scale RGB=<3 3 3>
    ScaleA=true
    ScaleRGB=true
   )
   &DynamicMaterialParameters(
    Index 0 Param 1=&FloatFromCurve(FloatCurve=DataInterface() {propertyValues {Curve {keys [<at 0 1>,<at 0.5 1>,<at 1 0.12>]}}})
    Index 0 Param 2=&FloatFromCurve(FloatCurve=DataInterface() {propertyValues {Curve {keys [<at 0 0>,<at 1 1>]}}})
    Index 0 Param 3=1
    Index 0 Param 4=1
    Module.Index0Float0Write=true
    Module.Index0Float1Write=true
    Module.Index0Float2Write=true
    Module.Index0Float3Write=true
   ) {module=DynamicMaterialParameters001}
   &SolveForcesAndVelocity(Clamp Velocity=false, Limit Acceleration=false)
  ]
  rendererClasses [&NiagaraSpriteRendererProperties]
  renderers [{rendererClass=&NiagaraSpriteRendererProperties}]
 }
 Smoke_hand_Ribbon {
  emitterUpdateScript [
   &EmitterState(Inactive Response=::NewEnumerator1, Life Cycle Mode=::NewEnumerator1, Loop Duration=1, UseLoopDelay=true)
   &SpawnRate(SpawnRate=40, Use Spawn Probability=false)
   &SpawnRate(SpawnGroup=1, SpawnRate=40, Use Spawn Probability=false) {module=SpawnRate001}
  ]
  particleSpawnScript [
   &InitializeParticle(
    Lifetime=0.15
    Ribbon Width=&Multiply_Float(A=18, B=2)
    Ribbon Width Mode=::NewEnumerator1
    UsePositionOffset=false
    Write Material Random=true
   )
   &AddVelocity(Velocity=<0 0 -80>)
   &SkeletalMeshLocation(
    Bone / Socket Index=&SpawnGroup()
    Bone Sampling Mode=::NewEnumerator4
    MaskSpawnGroup=false
    Orientation Sampling=::NewEnumerator1
    Skeletal Mesh=$System.Skeletal Mesh Reader
   )
   &Color(Color=&MakeLinearColorFromVectorAndFloat("Vector (RGB)"=<1 0.5386 0.7562>, "Float (Alpha)"=0.89))
   &AddVelocity(Velocity=&FN_FrontEndVector3Modifier(Gameplay=<0 0 0>, Front End=<0 -300 1100>)) {module=AddVelocity001}
  ]
  particleUpdateScript [
   &ParticleState()
   &SolveForcesAndVelocity(Clamp Velocity=false, Limit Acceleration=false)
   &DynamicMaterialParameters(
    Index 0 Param 1=1
    Index 0 Param 2=2
    Index 0 Param 3=1
    Index 0 Param 4=1
    Module.Index0Float0Write=true
    Module.Index0Float1Write=true
    Module.Index0Float2Write=true
    Module.Index0Float3Write=true
   ) {module=DynamicMaterialParameters001}
  ]
  rendererClasses [&NiagaraRibbonRendererProperties]
  renderers [{rendererClass=&NiagaraRibbonRendererProperties}]
 }
 light_hand {
  emitterUpdateScript [
   &EmitterState(Inactive Response=::NewEnumerator1, Life Cycle Mode=::NewEnumerator1, Loop Duration=1, UseLoopDelay=true)
   &SpawnRate(SpawnRate=10, Use Spawn Probability=false)
  ]
  particleSpawnScript [
   &InitializeParticle(
    Lifetime=&RandomRangeFloat(Minimum=0.2, Maximum=0.55, Evaluation Type=::NewEnumerator1)
    Sprite Size=&Vector2DFromFloat(Value=&RandomRangeFloat(Minimum=200, Maximum=100, Evaluation Type=::NewEnumerator1))
    Sprite Size Mode=::NewEnumerator4
    UsePositionOffset=false
    Write Material Random=true
   )
   &SkeletalMeshLocation(
    Bone / Socket Index=&SequentialInt(Use Array Data:NiagaraBool=false, First Number:NiagaraInt32=0, Sequence:NiagaraDataInterfaceArrayInt32=?, Last Number:NiagaraInt32=1)
    Bone Sampling Mode=::NewEnumerator4
    MaskSpawnGroup=false
    Orientation Sampling=::NewEnumerator1
    Sampled Position Coordinate Space=::Local
    Skeletal Mesh=$System.Skeletal Mesh Reader
   )
   &InheritVelocity(
    Inherited Velocity Amount Scale=&VectorFromFloat(Value=&RandomRangeFloat(Minimum=-0.028, Maximum=-0.15, Evaluation Type=::NewEnumerator1))
    Inherited Velocity Speed Limit=0
    Speed Limit=false
    UseSpeedThreshold=false
   )
   &AddVectorToPosition(Vector:Vector3f=&TransformVector(Vector=<0 0 -2>, Destination Space=::Simulation))
   &ShapeLocation(MaskSpawnGroup=false, Override Seed=false, Sphere Radius=10, UseOwnerScale=true)
   &FixShapeVector()
   &AddVelocity(Velocity=&Multiply_VectorByFloat(Vector=$Output.FixShapeVector.ShapeVector, Float=4.5))
   &Color(Color=&MakeLinearColorFromVectorAndFloat("Vector (RGB)"=<0.6 1 0.2>, "Float (Alpha)"=1))
  ]
  particleUpdateScript [
   &ParticleState()
   &CascadeConversion_LightProperties(LightBrightnessScale=7.5, LightColorScale=<2 5 1>, LightRadiusScale=15)
   &AccelerationForce(Acceleration=<0 0 30>)
   &SolveForcesAndVelocity(Clamp Velocity=false, Limit Acceleration=false)
   &ScaleColor(
    Scale Alpha=&FloatFromCurve(FloatCurve=DataInterface() {propertyValues {Curve {keys [<at 0 0>,<at 0.31 1>,<at 1 0>]}}})
    Scale RGB=<180 180 180>
    ScaleA=true
    ScaleRGB=true
   )
  ]
  rendererClasses [&NiagaraLightRendererProperties]
  renderers [{rendererClass=&NiagaraLightRendererProperties, rendererIndex=0}]
 }
}
summary {
 systemName=NS_Contrail_VioletFlame
 emitters [
  {emitterName=Smoke_hand, rendererClasses [&NiagaraSpriteRendererProperties]}
  {emitterName=Smoke_hand_Ribbon, rendererClasses [&NiagaraRibbonRendererProperties]}
  {emitterName=Flame_hand, rendererClasses [&NiagaraSpriteRendererProperties]}
  {emitterName=Flame_ball, rendererClasses [&NiagaraSpriteRendererProperties]}
  {emitterName=Fire_hand_Ribbon, rendererClasses [&NiagaraRibbonRendererProperties]}
  {emitterName=light_hand, rendererClasses [&NiagaraLightRendererProperties]}
 ]
 userVariables [
  User.bIsFrontEnd {type=NiagaraBool}
  User.bIsFrontEndPreview {type=NiagaraBool}
 ]
}
