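# Niagara system for the Zero Point skydiving contrail cosmetic, layering glows, flames and ribbon trails on the player.
# Core, flare, shimmer, flame and ribbon emitters follow bones or sockets on User.PlayerMesh, with waveform-indexed curves flickering the persistent flares and shimmers.
# Two fragment emitters spawn drifting meshes in spheres, one adding sprite glints.
# Ribbons fake trailing velocity only in the front end, with ParaGlideLeanAlpha bending it and adjusting the material.
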NiagaraSystem /BRCosmetics/Athena/Effects/Particles/Skydiving/NS_Contrail_ZeroPoint
dependencies {
 usedDataInterfaces [
  &NiagaraDataInterfaceCurve
  &NiagaraDataInterfaceArrayInt32
  &NiagaraDataInterfaceVector2DCurve
  &NiagaraDataInterfaceArrayFloat3
  &NiagaraDataInterfaceArrayFloat
 ]
 usedDynamicInputs []
 usedModules [&SystemState]
}
emitters {
 Core {
  emitterUpdateScript [
   &EmitterState(
    Inactive Response=::NewEnumerator1
    Life Cycle Mode=::NewEnumerator1
    Loop Behavior=::NewEnumerator1
    Loop Duration Mode=::NewEnumerator1
    UseLoopDelay=false
   )
   &SpawnBurst_Instantaneous(Use Spawn Probability=false, UseLoopCountLimit=false)
  ]
  particleSpawnScript [
   &InitializeParticle(Sprite Size Mode=::NewEnumerator3, Uniform Sprite Size=60, UsePositionOffset=false, Write Material Random=true)
   &CameraOffset(Camera Offset Amount=20)
  ]
  particleUpdateScript [
   &ParticleState(Kill Particles When Lifetime Has Elapsed=false, Let Infinitely Lived Particles Die When Emitter Deactivates=true)
   &SkeletalMeshLocation(Bone Sampling Mode=::NewEnumerator4, MaskSpawnGroup=false, Skeletal Mesh=$User.PlayerMesh)
  ]
  rendererClasses [&NiagaraSpriteRendererProperties]
  renderers [{rendererClass=&NiagaraSpriteRendererProperties}]
 }
 Core_Flares {
  emitterUpdateScript [
   &EmitterState(
    Inactive Response=::NewEnumerator1
    Life Cycle Mode=::NewEnumerator1
    Loop Behavior=::NewEnumerator1
    Loop Duration Mode=::NewEnumerator1
    UseLoopDelay=false
   )
   &SpawnBurst_Instantaneous(Use Spawn Probability=false, UseLoopCountLimit=false)
  ]
  particleSpawnScript [
   &InitializeParticle(
    Color=&MakeLinearColorFromVectorAndFloat("Vector (RGB)"=<0.3 0.6 2>, "Float (Alpha)"=&RandomRangeFloat(Minimum=0.08, Maximum=0.1))
    Sprite Size=<105 62>
    Sprite Size Mode=::NewEnumerator4
    UsePositionOffset=false
    Write Material Random=true
   )
   &CameraOffset(Camera Offset Amount=30)
   &DynamicMaterialParameters(
    Index 0 Param 1=0.1
    Index 0 Param 2=2
    Index 0 Param 3=1
    Module.Param0WriteEnabled=true
    Module.Param1WriteEnabled=true
    Module.Param2WriteEnabled=true
    Module.Param3WriteEnabled=false
   )
  ]
  particleUpdateScript [
   &ParticleState(Kill Particles When Lifetime Has Elapsed=false, Let Infinitely Lived Particles Die When Emitter Deactivates=true)
   /SetVariables(
    Particles.Soft Core Flare Camera Offset=40
    Particles.Soft Core Flare Color:LinearColor=&MakeLinearColorFromVectorAndFloat(
     "Vector (RGB)"=&MakeVectorFromLinearColorRGB(Color=<rgba 0.0417 0.2275 1 1>)
     "Float (Alpha)"=&FloatFromCurve(
      FloatCurve=DataInterface() {propertyValues {Curve {keys [
          {interpMode=RCIM_Cubic, leaveTangentWeight=0.0553, tangentMode=RCTM_User, time=0, value=0}
          {arriveTangentWeight=0.1799, interpMode=RCIM_Cubic, leaveTangentWeight=0.0499, tangentMode=RCTM_User, time=0.166, value=1}
          {arriveTangentWeight=0.0811, interpMode=RCIM_Cubic, leaveTangentWeight=0.1022, tangentMode=RCTM_User, time=0.212, value=0.656}
          {arriveTangentWeight=0.0462, interpMode=RCIM_Cubic, leaveTangentWeight=0.0963, tangentMode=RCTM_User, time=0.27, value=1.54}
          {arriveTangentWeight=0.1657, interpMode=RCIM_Cubic, leaveTangentWeight=0.115, tangentMode=RCTM_User, time=0.391, value=0.268}
          {arriveTangentWeight=0.0418, interpMode=RCIM_Cubic, leaveTangentWeight=0.0483, tangentMode=RCTM_User, time=0.475, value=0.723}
          {arriveTangentWeight=0.0587, interpMode=RCIM_Cubic, leaveTangentWeight=0.052, tangentMode=RCTM_User, time=0.572, value=0.469}
          {arriveTangentWeight=0.0644, interpMode=RCIM_Cubic, leaveTangentWeight=0.0502, tangentMode=RCTM_User, time=0.658, value=1}
          {arriveTangentWeight=0.0224, interpMode=RCIM_Cubic, leaveTangentWeight=0.0154, tangentMode=RCTM_User, time=0.725, value=0.161}
          {arriveTangentWeight=0.0178, interpMode=RCIM_Cubic, leaveTangentWeight=0.0885, tangentMode=RCTM_User, time=0.771, value=1.01}
          {arriveTangentWeight=0.0763, interpMode=RCIM_Cubic, tangentMode=RCTM_User, time=1, value=0}
         ]}}}
      CurveIndex=&Waveform("Waveform [1]"=::NewEnumerator7, "[1] Frequency"=1, "[1] Amplitude Min/Max"=<0 1>)
      Scale Curve=&RandomRangeFloat(Minimum=0.5, Maximum=2)
     )
    )
    Particles.Soft Core Flare Size:Vector2f=&Multiply_Vector2D_ByFloat(
     Float=&FloatFromCurve(
      FloatCurve=DataInterface() {propertyValues {Curve {keys [
          {interpMode=RCIM_Cubic, leaveTangentWeight=0.0667, tangentMode=RCTM_User, time=0, value=0.8}
          {arriveTangentWeight=0.1577, interpMode=RCIM_Cubic, leaveTangentWeight=0.0442, tangentMode=RCTM_User, time=0.2, value=1}
          {arriveTangentWeight=0.0302, interpMode=RCIM_Cubic, leaveTangentWeight=0.182, tangentMode=RCTM_User, time=0.256, value=0.251}
          {arriveTangentWeight=0.1599, interpMode=RCIM_Cubic, leaveTangentWeight=0.192, tangentMode=RCTM_User, time=0.594, value=1.5}
          {arriveTangentWeight=0.064, interpMode=RCIM_Cubic, leaveTangentWeight=0.1507, tangentMode=RCTM_User, time=0.715, value=0.373}
          {arriveTangentWeight=0.1353, interpMode=RCIM_Cubic, tangentMode=RCTM_User, time=1, value=1}
         ]}}}
      CurveIndex=&Waveform(
       "Waveform [1]"=::NewEnumerator7
       "[1] Frequency"=&RandomRangeFloat(Minimum=0.5, Maximum=1.5)
       "[1] Amplitude Min/Max"=<0 1>
      )
     )
     Vector2D=<70 70>
    )
   ) {isSetParameters=true}
   &SkeletalMeshLocation(Bone Sampling Mode=::NewEnumerator4, MaskSpawnGroup=false, Skeletal Mesh=$User.PlayerMesh)
   &ScaleColor(
    Scale Alpha=&FloatFromCurve(
     FloatCurve=DataInterface() {propertyValues {Curve {keys [
         {interpMode=RCIM_Cubic, leaveTangentWeight=0.0553, tangentMode=RCTM_User, time=0, value=0}
         {arriveTangentWeight=0.1799, interpMode=RCIM_Cubic, leaveTangentWeight=0.0499, tangentMode=RCTM_User, time=0.166, value=1}
         {arriveTangentWeight=0.0811, interpMode=RCIM_Cubic, leaveTangentWeight=0.1022, tangentMode=RCTM_User, time=0.212, value=0.656}
         {arriveTangentWeight=0.0462, interpMode=RCIM_Cubic, leaveTangentWeight=0.0963, tangentMode=RCTM_User, time=0.27, value=1.54}
         {arriveTangentWeight=0.1657, interpMode=RCIM_Cubic, leaveTangentWeight=0.115, tangentMode=RCTM_User, time=0.391, value=0.268}
         {arriveTangentWeight=0.0418, interpMode=RCIM_Cubic, leaveTangentWeight=0.0483, tangentMode=RCTM_User, time=0.475, value=0.723}
         {arriveTangentWeight=0.0587, interpMode=RCIM_Cubic, leaveTangentWeight=0.052, tangentMode=RCTM_User, time=0.572, value=0.469}
         {arriveTangentWeight=0.0644, interpMode=RCIM_Cubic, leaveTangentWeight=0.0502, tangentMode=RCTM_User, time=0.658, value=1}
         {arriveTangentWeight=0.0224, interpMode=RCIM_Cubic, leaveTangentWeight=0.0154, tangentMode=RCTM_User, time=0.725, value=0.161}
         {arriveTangentWeight=0.0178, interpMode=RCIM_Cubic, leaveTangentWeight=0.0885, tangentMode=RCTM_User, time=0.771, value=1.01}
         {arriveTangentWeight=0.0763, interpMode=RCIM_Cubic, tangentMode=RCTM_User, time=1, value=0}
        ]}}}
     CurveIndex=&Waveform("Waveform [1]"=::NewEnumerator7, "[1] Frequency"=1, "[1] Amplitude Min/Max"=<0 1>)
     Scale Curve=&RandomRangeFloat(Minimum=0.5, Maximum=2)
    )
    Scale RGB=&VectorFromFloat(Value=1.2)
    ScaleA=true
    ScaleRGB=true
   )
  ]
  rendererClasses [&NiagaraSpriteRendererProperties]
  renderers [
   {rendererClass=&NiagaraSpriteRendererProperties}
   {rendererClass=&NiagaraSpriteRendererProperties, rendererIndex=1}
  ]
 }
 Flames {
  emitterUpdateScript [
   &EmitterState()
   &SpawnRate(SpawnRate=6, Use Spawn Probability=false)
  ]
  particleSpawnScript [
   &InitializeParticle(
    Color=&MakeLinearColorFromVectorAndFloat("Vector (RGB)"=<1 0.3 2>, "Float (Alpha)"=&RandomRangeFloat(Minimum=1, Maximum=3))
    Lifetime Max=1.5
    Lifetime Min=0.5
    Lifetime Mode=::NewEnumerator1
    Sprite Rotation Mode=::NewEnumerator2
    Sprite Size Mode=::NewEnumerator1
    Uniform Sprite Size Max=100
    Uniform Sprite Size Min=60
    UsePositionOffset=false
    Write Material Random=true
   )
   &DynamicMaterialParameters(
    Index 0 Param 1=&RandomRangeFloat(Minimum=0, Maximum=1)
    Module.Param0WriteEnabled=true
    Module.Param1WriteEnabled=false
    Module.Param2WriteEnabled=false
    Module.Param3WriteEnabled=false
   )
   &CameraOffset(Camera Offset Amount=12)
  ]
  particleUpdateScript [
   &ParticleState()
   &ScaleSpriteSize(Uniform Curve Sprite Scale=DataInterface() {propertyValues {Curve {keys [
        {arriveTangent=1.2331, interpMode=RCIM_Cubic, leaveTangent=1.2331, tangentMode=RCTM_User, time=0, value=0.1}
        {arriveTangent=1.1496, interpMode=RCIM_Cubic, leaveTangent=1.1496, tangentMode=RCTM_User, time=0.211, value=0.368}
        <at 1 1.02 RCIM_Cubic>
       ]}}})
   &SpriteRotationRate(Rotation Rate=&RandomRangeFloat(Minimum=-108, Maximum=108))
   &ScaleColor(
    Scale Alpha=&FloatFromCurve(FloatCurve=DataInterface() {propertyValues {Curve {keys [
         {arriveTangent=10.5415, interpMode=RCIM_Cubic, leaveTangent=10.5415, tangentMode=RCTM_User, time=0, value=0}
         <at 0.1 0.7 RCIM_Cubic>
         <at 1 0 RCIM_Cubic>
        ]}}})
    ScaleA=true
    ScaleRGB=true
   )
   &SkeletalMeshLocation(
    Bone / Socket Index=&SelectIntFromArray(
     Array Sampling Mode=::NewEnumerator1
     Direct Array Index=?
     Int Selection Array=DataInterface() {propertyValues {IntData [0,4,6]}}
     Random Seed=0
    )
    Bone Sampling Mode=::NewEnumerator4
    MaskSpawnGroup=false
    Position Sampling=::NewEnumerator1
    Skeletal Mesh=$User.PlayerMesh
   )
   &AccelerationForce(Acceleration=&MakeVector(
     X=0
     Y=0
     Z=&FloatFromCurve(FloatCurve=DataInterface() {propertyValues {Curve {keys [<at 0.15 0>,<at 1 1000>]}}})
    ))
   &SolveForcesAndVelocity(Clamp Velocity=false, Limit Acceleration=false)
  ]
  rendererClasses [&NiagaraSpriteRendererProperties]
  renderers [{rendererClass=&NiagaraSpriteRendererProperties}]
 }
 Hand_Flares {
  emitterUpdateScript [
   &EmitterState(
    Inactive Response=::NewEnumerator1
    Life Cycle Mode=::NewEnumerator1
    Loop Behavior=::NewEnumerator1
    Loop Duration Mode=::NewEnumerator1
    UseLoopDelay=false
   )
   &SpawnBurst_Instantaneous(Spawn Count=2, Use Spawn Probability=false, UseLoopCountLimit=false)
  ]
  particleSpawnScript [
   &FN_SetSpawnGroup(Number of Spawn Groups=2)
   &InitializeParticle(
    Color=<rgba 0 0.9822 1 0.1>
    Lifetime=5
    Sprite Size Mode=::NewEnumerator3
    Uniform Sprite Size=75
    UsePositionOffset=false
    Write Material Random=true
   )
   &DynamicMaterialParameters(
    Index 0 Param 1=1
    Index 0 Param 2=0.5
    Module.Param0WriteEnabled=true
    Module.Param1WriteEnabled=true
    Module.Param2WriteEnabled=false
    Module.Param3WriteEnabled=false
   )
   &CameraOffset(Camera Offset Amount=5)
   /SetVariables(Particles.Core Camera Offset=20, Particles.Core Size:Vector2f=<120 120>) {isSetParameters=true}
  ]
  particleUpdateScript [
   &ParticleState(Kill Particles When Lifetime Has Elapsed=false, Let Infinitely Lived Particles Die When Emitter Deactivates=true)
   &SkeletalMeshLocation(
    Bone / Socket Index=&Add_Integer(A=$Particles.SpawnGroup, B=8)
    Bone Sampling Mode=::NewEnumerator4
    MaskSpawnGroup=false
    Sampled Position Offset=&SelectVectorFromArray(Direct Array Index=$Particles.SpawnGroup, Vector Selection Array=DataInterface() {propertyValues {FloatData [<5 0 0>, <-5 0 0>]}})
    Skeletal Mesh=$User.PlayerMesh
   )
   /SetVariables(
    Particles.Animorphic Color:LinearColor=&MakeLinearColorFromVectorAndFloat(
     "Vector (RGB)"=&MakeVectorFromLinearColorRGB(Color=<rgba 0.05 0.273 1.2 1>)
     "Float (Alpha)"=&Multiply_Float(
      A=0.5
      B=&FloatFromCurve(
       FloatCurve=DataInterface() {propertyValues {Curve {keys [
           {interpMode=RCIM_Cubic, leaveTangentWeight=0.0553, tangentMode=RCTM_User, time=0, value=0.1}
           {arriveTangentWeight=0.1799, interpMode=RCIM_Cubic, leaveTangentWeight=0.0499, tangentMode=RCTM_User, time=0.166, value=1}
           {arriveTangentWeight=0.0811, interpMode=RCIM_Cubic, leaveTangentWeight=0.1022, tangentMode=RCTM_User, time=0.212, value=0.656}
           {arriveTangentWeight=0.0462, interpMode=RCIM_Cubic, leaveTangentWeight=0.0963, tangentMode=RCTM_User, time=0.27, value=1.54}
           {arriveTangentWeight=0.1657, interpMode=RCIM_Cubic, leaveTangentWeight=0.115, tangentMode=RCTM_User, time=0.391, value=0.268}
           {arriveTangentWeight=0.0418, interpMode=RCIM_Cubic, leaveTangentWeight=0.0483, tangentMode=RCTM_User, time=0.475, value=0.723}
           {arriveTangentWeight=0.0587, interpMode=RCIM_Cubic, leaveTangentWeight=0.052, tangentMode=RCTM_User, time=0.572, value=0.469}
           {arriveTangentWeight=0.0644, interpMode=RCIM_Cubic, leaveTangentWeight=0.0502, tangentMode=RCTM_User, time=0.658, value=1}
           {arriveTangentWeight=0.0224, interpMode=RCIM_Cubic, leaveTangentWeight=0.0154, tangentMode=RCTM_User, time=0.725, value=0.161}
           {arriveTangentWeight=0.0178, interpMode=RCIM_Cubic, leaveTangentWeight=0.0885, tangentMode=RCTM_User, time=0.771, value=1.01}
           {arriveTangentWeight=0.0763, interpMode=RCIM_Cubic, tangentMode=RCTM_User, time=1, value=0.1}
          ]}}}
       CurveIndex=&Waveform("Waveform [1]"=::NewEnumerator7, "[1] Frequency"=1, "[1] Amplitude Min/Max"=<0 1>)
      )
     )
    )
    Particles.Core Color:LinearColor=&MakeLinearColorFromVectorAndFloat(
     "Vector (RGB)"=&MakeVectorFromLinearColorRGB(Color=<rgba 0.05 0.273 1.2 1>)
     "Float (Alpha)"=&Multiply_Float(
      A=0.09
      B=&FloatFromCurve(
       FloatCurve=DataInterface() {propertyValues {Curve {keys [
           {interpMode=RCIM_Cubic, leaveTangentWeight=0.0553, tangentMode=RCTM_User, time=0, value=0}
           {arriveTangentWeight=0.1799, interpMode=RCIM_Cubic, leaveTangentWeight=0.0499, tangentMode=RCTM_User, time=0.166, value=1}
           {arriveTangentWeight=0.0811, interpMode=RCIM_Cubic, leaveTangentWeight=0.1022, tangentMode=RCTM_User, time=0.212, value=0.656}
           {arriveTangentWeight=0.0462, interpMode=RCIM_Cubic, leaveTangentWeight=0.0963, tangentMode=RCTM_User, time=0.27, value=1.54}
           {arriveTangentWeight=0.1657, interpMode=RCIM_Cubic, leaveTangentWeight=0.115, tangentMode=RCTM_User, time=0.391, value=0.268}
           {arriveTangentWeight=0.0418, interpMode=RCIM_Cubic, leaveTangentWeight=0.0483, tangentMode=RCTM_User, time=0.475, value=0.723}
           {arriveTangentWeight=0.0587, interpMode=RCIM_Cubic, leaveTangentWeight=0.052, tangentMode=RCTM_User, time=0.572, value=0.469}
           {arriveTangentWeight=0.0644, interpMode=RCIM_Cubic, leaveTangentWeight=0.0502, tangentMode=RCTM_User, time=0.658, value=1}
           {arriveTangentWeight=0.0224, interpMode=RCIM_Cubic, leaveTangentWeight=0.0154, tangentMode=RCTM_User, time=0.725, value=0.161}
           {arriveTangentWeight=0.0178, interpMode=RCIM_Cubic, leaveTangentWeight=0.0885, tangentMode=RCTM_User, time=0.771, value=1.01}
           {arriveTangentWeight=0.0763, interpMode=RCIM_Cubic, tangentMode=RCTM_User, time=1, value=0}
          ]}}}
       CurveIndex=&Waveform("Waveform [1]"=::NewEnumerator7, "[1] Frequency"=1, "[1] Amplitude Min/Max"=<0 1>)
      )
     )
    )
   ) {isSetParameters=true}
  ]
  rendererClasses [&NiagaraSpriteRendererProperties]
  renderers [
   {rendererClass=&NiagaraSpriteRendererProperties}
   {rendererClass=&NiagaraSpriteRendererProperties, rendererIndex=1}
   {rendererClass=&NiagaraSpriteRendererProperties, rendererIndex=2}
  ]
 }
 Ribbons {
  emitterUpdateScript [
   &EmitterState(Life Cycle Mode=::NewEnumerator1, Loop Duration=0.5, UseLoopDelay=false)
   &SpawnRate(SpawnRate=60, Use Spawn Probability=false)
   &SpawnRate(SpawnGroup=1, SpawnRate=60, Use Spawn Probability=false) {module=SpawnRate001}
  ]
  particleSpawnScript [
   &InitializeParticle(Lifetime=0.22, Ribbon Width=36, Ribbon Width Mode=::NewEnumerator1, UsePositionOffset=false, Write Material Random=true)
   &AddVelocity(Velocity=&FN_FrontEndVector3Modifier(Gameplay=<0 0 0>, Front End=&Lerp_Vector(A=<0 0 1000>, B=<0 -500 1000>, Alpha=$User.ParaGlideLeanAlpha)))
  ]
  particleUpdateScript [
   &ParticleState()
   &SkeletalMeshLocation(
    Bone / Socket Index=&Add_Integer(A=&SpawnGroup(), B=8)
    Bone Sampling Mode=::NewEnumerator4
    MaskSpawnGroup=false
    Position Sampling=::NewEnumerator1
    Sampled Position Offset=<0 0 -7>
    Sampled Position Offset Coordinate Space:ENiagaraSkelMeshTransforms=::NewEnumerator1
    Skeletal Mesh=$User.PlayerMesh
    Source Movement Scale:Vector3f=<0 0 0>
   )
   &DynamicMaterialParameters(
    Index 0 Param 1=&SelectFloatFromArray(Direct Array Index=&SpawnGroup(), Float Selection Array=DataInterface() {propertyValues {FloatData [0,0.1]}})
    Index 0 Param 3=&Lerp_Float(A=1, B=0.1, Alpha=$User.ParaGlideLeanAlpha)
    Index 0 Param 4=1
    Module.Param0WriteEnabled=true
    Module.Param1WriteEnabled=false
    Module.Param2WriteEnabled=true
    Module.Param3WriteEnabled=true
   )
   &ScaleRibbonWidth(Ribbon Width Scale=&FloatFromCurve(FloatCurve=DataInterface() {propertyValues {Curve {keys [{arriveTangent=0.6024, interpMode=RCIM_Cubic, leaveTangent=0.6024, tangentMode=RCTM_User, time=0, value=1}, <at 1 1.42 RCIM_Cubic>]}}}))
   &SolveForcesAndVelocity(Clamp Velocity=false, Limit Acceleration=false)
   &ScaleColor(
    Scale Alpha=&FloatFromCurve(FloatCurve=DataInterface() {propertyValues {Curve {keys [
         {arriveTangent=7.0419, interpMode=RCIM_Cubic, leaveTangent=7.0419, tangentMode=RCTM_User, time=0, value=0}
         {arriveTangentWeight=0.0778, interpMode=RCIM_Cubic, leaveTangentWeight=0.6243, tangentMode=RCTM_Break, time=0.2335, value=1}
         <at 1 0 RCIM_Cubic>
        ]}}})
    ScaleA=true
    ScaleRGB=true
   )
  ]
  rendererClasses [&NiagaraRibbonRendererProperties]
  renderers [{rendererClass=&NiagaraRibbonRendererProperties}]
 }
 Shimmers {
  emitterUpdateScript [
   &EmitterState(
    Inactive Response=::NewEnumerator1
    Life Cycle Mode=::NewEnumerator1
    Loop Behavior=::NewEnumerator1
    Loop Duration Mode=::NewEnumerator1
    UseLoopDelay=false
   )
   &SpawnBurst_Instantaneous(Spawn Count=7, Use Spawn Probability=false, UseLoopCountLimit=false)
  ]
  particleSpawnScript [
   &FN_SetSpawnGroup(Number of Spawn Groups=7)
   &InitializeParticle(
    Color=&MakeLinearColorFromVectorAndFloat("Vector (RGB)"=<0 0.8 1.6>, "Float (Alpha)"=0.55)
    Lifetime Max=0.9
    Lifetime Min=0.6
    Lifetime Mode=::NewEnumerator1
    Sprite Rotation Mode=::NewEnumerator2
    Sprite Size Mode=::NewEnumerator1
    Uniform Sprite Size Max=50
    Uniform Sprite Size Min=30
    UsePositionOffset=false
    Write Material Random=true
   )
   &CameraOffset(Camera Offset Amount=50)
   &DynamicMaterialParameters(
    Index 0 Param 1=&RandomRangeFloat(Minimum=0, Maximum=1)
    Index 0 Param 2=&RandomRangeFloat(Minimum=0, Maximum=1)
    Index 0 Param 3=1
    Module.Param0WriteEnabled=true
    Module.Param1WriteEnabled=true
    Module.Param2WriteEnabled=true
    Module.Param3WriteEnabled=false
   )
  ]
  particleUpdateScript [
   &ParticleState(Kill Particles When Lifetime Has Elapsed=false, Let Infinitely Lived Particles Die When Emitter Deactivates=true)
   &SkeletalMeshLocation(
    Bone / Socket Index=&Add_Integer(A=$Particles.SpawnGroup, B=1)
    Bone Sampling Mode=::NewEnumerator4
    MaskSpawnGroup=false
    Skeletal Mesh=$User.PlayerMesh
   )
   &SpriteRotationRate(Rotation Rate=&RandomRangeFloat(Minimum=-20, Maximum=20))
   &ScaleColor(
    Scale Alpha=&FloatFromCurve(
     FloatCurve=DataInterface() {propertyValues {Curve {keys [
         {interpMode=RCIM_Cubic, tangentMode=RCTM_User, time=0, value=0.1}
         {interpMode=RCIM_Cubic, tangentMode=RCTM_User, time=0.5, value=1}
         {tangentMode=RCTM_User, time=1, value=0.1}
        ]}}}
     CurveIndex=&Waveform(
      "Waveform [1]"=::NewEnumerator7
      "[1] Frequency"=&RandomRangeFloat(Minimum=0.5, Maximum=1.5)
      "[1] Amplitude Min/Max"=<0 1>
     )
    )
    ScaleA=true
    ScaleRGB=true
   )
  ]
  rendererClasses [&NiagaraSpriteRendererProperties]
  renderers [{rendererClass=&NiagaraSpriteRendererProperties}]
 }
 Solar_Flares {
  emitterUpdateScript [
   &EmitterState()
   &SpawnBurst_Instantaneous(Spawn Time=0.15, Use Spawn Probability=false, UseLoopCountLimit=false)
   &SpawnRate(Use Spawn Probability=false)
  ]
  particleSpawnScript [
   &InitializeParticle(
    Color Maximum=<rgba 0 0.3 0.6 0.7>
    Color Minimum=<rgba 0 0 0 0.4>
    Color Mode=::NewEnumerator2
    Lifetime Min=0.6
    Lifetime Mode=::NewEnumerator1
    Sprite Rotation Angle=&RandomRangeFloat(Minimum=0.1, Maximum=0.4)
    Sprite Rotation Mode=::NewEnumerator1
    Sprite Size=<60 20>
    Sprite Size Mode=::NewEnumerator4
    UsePositionOffset=false
    Write Material Random=true
   )
   &DynamicMaterialParameters(
    Index 0 Param 1=&RandomRangeFloat(Minimum=0, Maximum=1)
    Module.Param0WriteEnabled=true
    Module.Param1WriteEnabled=false
    Module.Param2WriteEnabled=false
    Module.Param3WriteEnabled=false
   )
   &CameraOffset(Camera Offset Amount=10)
  ]
  particleUpdateScript [
   &ParticleState()
   &ScaleSpriteSize(
    Non-Uniform Curve Sprite Scale=DataInterface() {propertyValues {
      XCurve {keys [{tangentMode=RCTM_User, time=0, value=0.1}, {tangentMode=RCTM_User, time=1, value=1}]}
      YCurve {keys [{tangentMode=RCTM_User, time=0, value=1}, {tangentMode=RCTM_User, time=1, value=2}]}
     }}
    Scale Sprite Size Mode=::NewEnumerator3
   )
   &SpriteRotationRate(Rotation Rate=&RandomRangeFloat(Minimum=-36, Maximum=36))
   &SkeletalMeshLocation(
    Bone / Socket Index=&RandomRangeInt(Minimum=1, Maximum=4)
    Bone Sampling Mode=::NewEnumerator4
    MaskSpawnGroup=false
    Skeletal Mesh=$User.PlayerMesh
   )
   &ScaleColor(
    Scale Alpha=&FloatFromCurve(FloatCurve=DataInterface() {propertyValues {Curve {keys [
         {interpMode=RCIM_Cubic, leaveTangentWeight=0.1033, tangentMode=RCTM_User, time=0, value=0}
         {arriveTangentWeight=0.2095, interpMode=RCIM_Cubic, leaveTangentWeight=0.1737, tangentMode=RCTM_User, time=0.31, value=1}
         {arriveTangentWeight=0.1508, interpMode=RCIM_Cubic, leaveTangentWeight=0.2541, tangentMode=RCTM_User, time=0.567, value=1}
         {arriveTangentWeight=0.1443, interpMode=RCIM_Cubic, tangentMode=RCTM_User, time=1, value=0}
        ]}}})
    ScaleA=true
    ScaleRGB=true
   )
  ]
  rendererClasses [&NiagaraSpriteRendererProperties]
  renderers [{rendererClass=&NiagaraSpriteRendererProperties}]
 }
 TintFragSwarm {
  emitterUpdateScript [
   &EmitterState(
    Life Cycle Mode=::NewEnumerator1
    Loop Duration=&RandomRangeFloat(Minimum=1, Maximum=2)
    Recalculate Duration Each Loop=true
    UseLoopDelay=false
   )
   &SpawnRate(Use Spawn Probability=false)
   &SpawnBurst_Instantaneous(
    Spawn Count=&RandomRangeInt(Minimum=1, Maximum=2)
    Use Spawn Probability=false
    UseLoopCountLimit=false
   )
   &SpawnBurst_Instantaneous(
    Spawn Count=&RandomRangeInt(Minimum=0, Maximum=2)
    Spawn Time=0.25
    Use Spawn Probability=false
    UseLoopCountLimit=false
   ) {module=SpawnBurst_Instantaneous001}
   &SpawnBurst_Instantaneous(
    Spawn Count=&RandomRangeInt(Minimum=0, Maximum=2)
    Spawn Time=1.4
    Use Spawn Probability=false
    UseLoopCountLimit=false
   ) {module=SpawnBurst_Instantaneous002}
  ]
  particleSpawnScript [
   &InitializeParticle(
    Color Mode=::NewEnumerator0
    Lifetime Max=3.5
    Lifetime Min=1.5
    Lifetime Mode=::NewEnumerator1
    Mesh Scale Mode=::NewEnumerator1
    Mesh Uniform Scale Max=1.5
    Mesh Uniform Scale Min=0.66
    UsePositionOffset=false
    Write Material Random=true
   )
   &ShapeLocation(
    MaskSpawnGroup=false
    Offset=<0 0 90>
    Offset Mode=::NewEnumerator0
    Override Seed=false
    Sphere Radius=70
    Sphere Surface Distribution=1
    UseOwnerScale=true
   )
   &InitialMeshOrientation(Use Rotation Vector=false)
   &AddVelocity(Velocity=&Multiply_VectorByFloat(Vector=$StackContext.ShapeLocation.ShapeNormal, Float=&RandomRangeFloat(Minimum=-0.5, Maximum=1)))
  ]
  particleUpdateScript [
   &ParticleState()
   &UpdateMeshOrientation(
    Rotation Rate=&RandomRangeFloat(Minimum=-0.15, Maximum=0.15)
    Rotation Vector=&RandomRangeVector(Minimum=<0 0 0>, Maximum=<1 1 1>)
   )
   &ScaleMeshSize(Scale Factor=&VectorFromFloat(Value=&FloatFromCurve(FloatCurve=DataInterface() {propertyValues {Curve {keys [
          {interpMode=RCIM_Cubic, leaveTangentWeight=0.0667, tangentMode=RCTM_User, time=0, value=0}
          <at 0.2 1>
          {arriveTangentWeight=0.3202, interpMode=RCIM_Cubic, leaveTangentWeight=0.1067, tangentMode=RCTM_User, time=0.8, value=1}
          <at 1 0>
         ]}}})))
   &ScaleColor(
    Scale Alpha=&FloatFromCurve(FloatCurve=DataInterface() {propertyValues {Curve {keys [
         {interpMode=RCIM_Cubic, leaveTangentWeight=0.0667, tangentMode=RCTM_User, time=0, value=0}
         {arriveTangentWeight=0.1163, interpMode=RCIM_Cubic, leaveTangentWeight=0.2906, tangentMode=RCTM_User, time=0.2, value=1}
         {arriveTangentWeight=0.2668, interpMode=RCIM_Cubic, leaveTangentWeight=0.1601, tangentMode=RCTM_User, time=0.7, value=1}
         {arriveTangentWeight=0.1, interpMode=RCIM_Cubic, tangentMode=RCTM_User, time=1, value=0}
        ]}}})
    ScaleA=true
    ScaleRGB=true
   )
   &SolveForcesAndVelocity(Clamp Velocity=false, Limit Acceleration=false)
  ]
  rendererClasses [&NiagaraMeshRendererProperties]
  renderers [{rendererClass=&NiagaraMeshRendererProperties, rendererIndex=0}]
 }
 TinyFrags_Glints {
  emitterUpdateScript [
   &EmitterState(
    Life Cycle Mode=::NewEnumerator1
    Loop Duration=&RandomRangeFloat(Minimum=1, Maximum=2)
    Recalculate Duration Each Loop=true
    UseLoopDelay=false
   )
   &SpawnRate(SpawnRate=2, Use Spawn Probability=false)
   &SpawnBurst_Instantaneous(
    Spawn Count=&RandomRangeInt(Minimum=1, Maximum=2)
    Use Spawn Probability=false
    UseLoopCountLimit=false
   )
   &SpawnBurst_Instantaneous(
    Spawn Count=&RandomRangeInt(Minimum=0, Maximum=2)
    Spawn Time=0.25
    Use Spawn Probability=false
    UseLoopCountLimit=false
   ) {module=SpawnBurst_Instantaneous001}
   &SpawnBurst_Instantaneous(
    Spawn Count=&RandomRangeInt(Minimum=0, Maximum=2)
    Spawn Time=1.4
    Use Spawn Probability=false
    UseLoopCountLimit=false
   ) {module=SpawnBurst_Instantaneous002}
  ]
  particleSpawnScript [
   &InitializeParticle(
    Color Mode=::NewEnumerator0
    Lifetime Max=3.5
    Lifetime Min=1.5
    Lifetime Mode=::NewEnumerator1
    Mesh Scale Mode=::NewEnumerator1
    Mesh Uniform Scale Max=1.5
    Mesh Uniform Scale Min=0.66
    Sprite Size Mode=::NewEnumerator1
    Uniform Sprite Size Max=100
    Uniform Sprite Size Min=50
    UsePositionOffset=false
    Write Material Random=true
   )
   &ShapeLocation(
    MaskSpawnGroup=false
    Offset=&FN_FrontEndVector3Modifier(Gameplay=<0 0 0>, Front End=<0 0 75>)
    Offset Mode=::NewEnumerator0
    Override Seed=false
    Sphere Surface Distribution=1
    UseOwnerScale=true
   )
   &InitialMeshOrientation(Use Rotation Vector=false)
   &AddVelocity(Velocity=&Multiply_VectorByFloat(Vector=$StackContext.ShapeLocation.ShapeNormal, Float=&RandomRangeFloat(Minimum=-0.5, Maximum=1)))
   &AddVelocity(Velocity=&RandomRangeVector(Minimum=<-8 -8 -8>, Maximum=<8 8 8>)) {module=AddVelocity001}
   /SetVariables(Particles.Flare_CameraOffset=5, Particles.Flare_Color:LinearColor=<rgba 0.4 0.9 1 1>) {isSetParameters=true}
  ]
  particleUpdateScript [
   &ParticleState()
   &UpdateMeshOrientation(
    Rotation Rate=&RandomRangeFloat(Minimum=-0.15, Maximum=0.15)
    Rotation Vector=&RandomRangeVector(Minimum=<0 0 0>, Maximum=<1 1 1>)
   )
   &ScaleMeshSize(Scale Factor=&VectorFromFloat(Value=&FloatFromCurve(FloatCurve=DataInterface() {propertyValues {Curve {keys [
          {interpMode=RCIM_Cubic, leaveTangentWeight=0.0667, tangentMode=RCTM_User, time=0, value=0}
          <at 0.2 1>
          {arriveTangentWeight=0.3202, interpMode=RCIM_Cubic, leaveTangentWeight=0.1067, tangentMode=RCTM_User, time=0.8, value=1}
          <at 1 0>
         ]}}})))
   &ScaleSpriteSize(Uniform Curve Sprite Scale=DataInterface() {propertyValues {Curve {keys [
        <at 0.2 0 RCIM_Cubic>
        {arriveTangent=12.1804, interpMode=RCIM_Cubic, leaveTangent=-7.333, tangentMode=RCTM_Break, time=0.4, value=1}
        <at 0.7 0 RCIM_Cubic>
       ]}}})
   &ScaleColor(
    Scale Alpha=&FloatFromCurve(FloatCurve=DataInterface() {propertyValues {Curve {keys [
         {interpMode=RCIM_Cubic, leaveTangentWeight=0.0667, tangentMode=RCTM_User, time=0, value=0}
         {arriveTangentWeight=0.1163, interpMode=RCIM_Cubic, leaveTangentWeight=0.2906, tangentMode=RCTM_User, time=0.2, value=1}
         {arriveTangentWeight=0.2668, interpMode=RCIM_Cubic, leaveTangentWeight=0.1601, tangentMode=RCTM_User, time=0.7, value=1}
         {arriveTangentWeight=0.1, interpMode=RCIM_Cubic, tangentMode=RCTM_User, time=1, value=0}
        ]}}})
    ScaleA=true
    ScaleRGB=true
   )
   &AccelerationForce(Acceleration=&MakeVector(
     X=0
     Y=0
     Z=&FloatFromCurve(FloatCurve=DataInterface() {propertyValues {Curve {keys [<at 0.15 0>,<at 1 800>]}}})
    ))
   &SolveForcesAndVelocity(Clamp Velocity=false, Limit Acceleration=false)
  ]
  rendererClasses [&NiagaraMeshRendererProperties, &NiagaraSpriteRendererProperties]
  renderers [
   {rendererClass=&NiagaraMeshRendererProperties, rendererIndex=0}
   {rendererClass=&NiagaraSpriteRendererProperties, rendererIndex=1}
  ]
 }
}
summary {
 systemName=NS_Contrail_ZeroPoint
 emitters [
  {emitterName=TintFragSwarm, rendererClasses [&NiagaraMeshRendererProperties]}
  {
   emitterName=TinyFrags_Glints
   rendererClasses [&NiagaraMeshRendererProperties, &NiagaraSpriteRendererProperties]
  }
  {emitterName=Core, rendererClasses [&NiagaraSpriteRendererProperties]}
  {emitterName=Shimmers, rendererClasses [&NiagaraSpriteRendererProperties]}
  {emitterName=Flames, rendererClasses [&NiagaraSpriteRendererProperties]}
  {emitterName=Core_Flares, rendererClasses [&NiagaraSpriteRendererProperties]}
  {emitterName=Solar_Flares, rendererClasses [&NiagaraSpriteRendererProperties]}
  {emitterName=Hand_Flares, rendererClasses [&NiagaraSpriteRendererProperties]}
  {emitterName=Ribbons, rendererClasses [&NiagaraRibbonRendererProperties]}
 ]
 userVariables [
  User.bIsFrontEnd {type=NiagaraBool}
  User.bIsFrontEndPreview {type=NiagaraBool, defaultValue=true}
  User.ParaGlideLeanAlpha {type=NiagaraFloat, defaultValue=0, description=""}
  User.PlayerMesh {
   type=NiagaraDataInterfaceSkeletalMesh
   defaultValue {
    struct=NiagaraExt_VariableValue_DataInterface
    value {dataInterface=&self:NiagaraDataInterfaceSkeletalMesh_0}
   }
  }
 ]
}
