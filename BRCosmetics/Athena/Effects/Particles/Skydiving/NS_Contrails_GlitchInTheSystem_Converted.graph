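# Niagara skydiving contrail cosmetic "Glitch In The System", converted from Cascade, drawing glitchy dark-green trails on the character.
# Ribbons sample skeletal mesh bones for hands and feet, alongside sprite dots, 8-bit bits, stretched lines and short-lived refraction sprites, mostly bone-attached.
# Stepped constant-key curves create alpha flicker and jumping ribbon widths, while spawn groups vary line behavior.
# FrontEnd modifiers retune offsets and velocities, enabling the Fountain emitter only in front end.
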NiagaraSystem /BRCosmetics/Athena/Effects/Particles/Skydiving/NS_Contrails_GlitchInTheSystem_Converted
dependencies {
 usedDataInterfaces [
  &NiagaraDataInterfaceCurve
  &NiagaraDataInterfaceVectorCurve
  &NiagaraDataInterfaceVector2DCurve
  &NiagaraDataInterfaceSkeletalMesh
 ]
 usedDynamicInputs []
 usedModules [
  &self:SystemSpawnScript.SystemScriptSource.SystemScriptGraph.NiagaraNodeAssignment_0.SetVariables_#1
  &SystemState
 ]
}
emitters {
 8BitDot {
  simTarget=GPUComputeSim
  emitterUpdateScript [
   &EmitterState(
    Delay First Loop Only=true
    Inactive Response=::NewEnumerator1
    Life Cycle Mode=::NewEnumerator1
    Loop Delay=0.2
    Loop Duration=1
    UseLoopDelay=true
   )
   &SpawnRate(SpawnRate=15, Use Spawn Probability=false)
   &SpawnRate(SpawnGroup=1, SpawnRate=15, Use Spawn Probability=false) {module=SpawnRate001}
   &SpawnPerUnit(
    Module.Use Max Movement Threshold=true
    Module.Use Movement Tolerance=true
    Movement Tolerance=0.1
    Spawn Spacing=90
    Use Spawn Probability=false
   ) {module=SpawnPerUnit001}
  ]
  particleSpawnScript [
   /SetVariables(Particles.VisibilityTag=&SpawnGroup()) {isSetParameters=true}
   &InitializeParticle(
    Lifetime=&RandomRangeFloat(Minimum=0.1, Maximum=0.7, Evaluation Type=::NewEnumerator1)
    Sprite Size=&Vector2DFromFloat(Value=&RandomRangeFloat(Minimum=2, Maximum=4, Evaluation Type=::NewEnumerator1))
    Sprite Size Mode=::NewEnumerator4
    UsePositionOffset=false
    Write Material Random=true
   )
   &InheritVelocity(Inherited Velocity Amount Scale=<-0.01 -0.01 -0.01>, Inherited Velocity Speed Limit=0, Speed Limit=false, UseSpeedThreshold=false)
   &ShapeLocation(
    MaskSpawnGroup=false
    Override Seed=false
    Sphere Radius=&FN_FrontEndFloatModifier(Gameplay=350, Front End=550)
    UseOwnerScale=true
   )
   &InheritVelocity(
    Inherited Velocity Amount Scale=<0.2 0.2 0.2>
    Inherited Velocity Speed Limit=0
    Speed Limit=false
    UseSpeedThreshold=false
    Velocity Source=&Multiply_VectorByFloat(
     Vector=&TransformVector(Vector=$Engine.Owner.Velocity, Source Space=::World, Destination Space=::Local)
     Float=&FN_FrontEndFloatModifier(Gameplay=-1, Front End=0)
    )
   ) {module=InheritVelocity002}
   &Color(Color=&MakeLinearColorFromVectorAndFloat(
     "Vector (RGB)"=&RandomRangeVector(Minimum=<0 0 0>, Maximum=<0.0522 0.1458 0>, Evaluation Type=::NewEnumerator1)
     "Float (Alpha)"=0
    ))
   /CascadeToNiagaraConverter/NiagaraScripts/ModuleScripts/FixShapeVector()
   &AddVelocity(Velocity=&Multiply_VectorByFloat(
     Vector=$Output.FixShapeVector.ShapeVector
     Float=&RandomRangeFloat(Minimum=0, Maximum=1, Evaluation Type=::NewEnumerator1)
    ))
  ]
  particleUpdateScript [
   &ParticleState()
   &SkeletalMeshLocation(
    Bone / Socket Index=/CascadeToNiagaraConverter/NiagaraScripts/DynamicInputScripts/SequentialInt(Use Array Data:NiagaraBool=false, First Number:NiagaraInt32=0, Sequence:NiagaraDataInterfaceArrayInt32=?, Last Number:NiagaraInt32=3)
    Bone Sampling Mode=::NewEnumerator4
    MaskSpawnGroup=false
    Orientation Sampling=::NewEnumerator1
    Sampled Position Offset=&FN_FrontEndVector3Modifier(
     Gameplay=&MergeHelper_VectorSwitch(0=<7 0 -10>, 1=<-7 0 -10>, 2=?, 3=?, Index=&SpawnGroup(), Num=::NewEnumerator0)
     Front End=<0 0 0>
    )
    Sampled Position Offset Coordinate Space:ENiagaraSkelMeshTransforms=::NewEnumerator2
    Skeletal Mesh=$System.Skeletal Mesh Reader
   ) {module=SkeletalMeshLocation001}
   &Drag(Drag=-3, Use Linear Drag=true, Use Rotational Drag=true)
   &SolveForcesAndVelocity(Clamp Velocity=false, Limit Acceleration=false)
   &Color(Color=&MakeLinearColorFromVectorAndFloat(
     "Vector (RGB)"=&VectorFromCurve(VectorCurve=DataInterface() {propertyValues {XCurve {keys [<at 0 1>,<at 1 0>]}, YCurve {keys [<at 0 1>,<at 1 0>]}, ZCurve {keys [<at 0 1>,<at 1 0>]}}})
     "Float (Alpha)"=&FloatFromCurve(FloatCurve=DataInterface() {propertyValues {Curve {keys [<at 0 1 RCIM_Constant>,<at 0.1223 10.5698 RCIM_Constant>,<at 0.2059 1.0005 RCIM_Constant>,<at 0.4553 10.5698 RCIM_Constant>,<at 0.5532 1.0005 RCIM_Constant>,<at 0.7824 10.5698 RCIM_Constant>,<at 0.8734 1 RCIM_Constant>,<at 1 1 RCIM_Constant>]}}})
    )) {module=Color001}
  ]
  rendererClasses [&NiagaraSpriteRendererProperties]
  renderers [
   {rendererClass=&NiagaraSpriteRendererProperties}
   {rendererClass=&NiagaraSpriteRendererProperties, rendererIndex=1}
  ]
 }
 Dots_Center {
  emitterUpdateScript [
   &EmitterState(
    Delay First Loop Only=true
    Inactive Response=::NewEnumerator1
    Life Cycle Mode=::NewEnumerator1
    Loop Delay=0.2
    Loop Duration=1
    UseLoopDelay=true
   )
   &SpawnRate(SpawnRate=2, Use Spawn Probability=false)
  ]
  particleSpawnScript [
   &InitializeParticle(Sprite Size=<100 100>, Sprite Size Mode=::NewEnumerator4, UsePositionOffset=false, Write Material Random=true)
   /CascadeToNiagaraConverter/NiagaraScripts/ModuleScripts/AddVectorToPosition(Vector:Vector3f=&TransformVector(Vector=<0 0 100>, Destination Space=::Simulation))
  ]
  particleUpdateScript [
   &ParticleState()
   &Color(Color=&MakeLinearColorFromVectorAndFloat(
     "Vector (RGB)"=&VectorFromCurve(VectorCurve=DataInterface() {propertyValues {XCurve {keys [<at 0 1>,<at 1 1>]}, YCurve {keys [<at 0 1>,<at 1 1>]}, ZCurve {keys [<at 0 1>,<at 1 1>]}}})
     "Float (Alpha)"=&ClampFloat(Float=&FloatFromCurve(FloatCurve=DataInterface() {propertyValues {Curve {keys [{arriveTangent=-1, time=0, value=1}, <at 1 0>]}}}))
    ))
   &CameraOffset(Camera Offset Amount=&FloatFromCurve(FloatCurve=DataInterface() {propertyValues {Curve {keys [<at 0 -200>,<at 1 200>]}}}))
  ]
  rendererClasses [&NiagaraSpriteRendererProperties]
  renderers [{rendererClass=&NiagaraSpriteRendererProperties}]
 }
 Dots_Hand {
  emitterUpdateScript [
   &EmitterState(
    Delay First Loop Only=true
    Inactive Response=::NewEnumerator1
    Life Cycle Mode=::NewEnumerator1
    Loop Delay=0.2
    Loop Duration=1
    UseLoopDelay=true
   )
   &SpawnRate(SpawnRate=20, Use Spawn Probability=false)
  ]
  particleSpawnScript [
   &InitializeParticle(Lifetime=0.5, Sprite Size=<50 50>, Sprite Size Mode=::NewEnumerator4, UsePositionOffset=false, Write Material Random=true)
   &AddVectorToPosition(Vector:Vector3f=&TransformVector(Vector=<0 50 0>, Destination Space=::Simulation))
  ]
  particleUpdateScript [
   &ParticleState()
   &Color(Color=&MakeLinearColorFromVectorAndFloat(
     "Vector (RGB)"=&VectorFromCurve(VectorCurve=DataInterface() {propertyValues {XCurve {keys [<at 0 1>,<at 1 1>]}, YCurve {keys [<at 0 1>,<at 1 1>]}, ZCurve {keys [<at 0 1>,<at 1 1>]}}})
     "Float (Alpha)"=&ClampFloat(Float=&FloatFromCurve(FloatCurve=DataInterface() {propertyValues {Curve {keys [{arriveTangent=-1, time=0, value=1}, <at 1 0>]}}}))
    ))
   &CameraOffset(Camera Offset Amount=&FloatFromCurve(FloatCurve=DataInterface() {propertyValues {Curve {keys [<at 0 -20>,<at 1 20>]}}}))
   &SkeletalMeshLocation(
    Bone / Socket Index=&SequentialInt(Use Array Data:NiagaraBool=false, First Number:NiagaraInt32=0, Sequence:NiagaraDataInterfaceArrayInt32=?, Last Number:NiagaraInt32=3)
    Bone Sampling Mode=::NewEnumerator4
    MaskSpawnGroup=false
    Orientation Sampling=::NewEnumerator1
    Sampled Position Offset=<0 0 -10>
    Sampled Position Offset Coordinate Space:ENiagaraSkelMeshTransforms=::NewEnumerator2
    Skeletal Mesh=$System.Skeletal Mesh Reader
   ) {module=SkeletalMeshLocation001}
  ]
  rendererClasses [&NiagaraSpriteRendererProperties]
  renderers [{rendererClass=&NiagaraSpriteRendererProperties}]
 }
 Fountain {
  emitterUpdateScript [
   &EmitterState(Life Cycle Mode=::NewEnumerator1, Loop Duration=2, UseLoopDelay=false)
   &SpawnRate(SpawnRate=&FN_FrontEndFloatModifier(Gameplay=0, Front End=7), Use Spawn Probability=false)
  ]
  particleSpawnScript [
   &InitializeParticle(
    Color=&RandomRangeLinearColor(Minimum=<rgba 0 0 0 0>, Maximum=<rgba 0.0522 0.1458 0 0>)
    Lifetime Max=1.75
    Lifetime Mode=::NewEnumerator1
    Mass Mode=::NewEnumerator3
    Position Offset=<0 30 30>
    Position Offset Coordinate Space=::Local
    Sprite Rotation Mode=::NewEnumerator2
    Sprite Size=&RandomRangeVector2D(Minimum=<1 1>, Maximum=<1.7 20>, Evaluation Type=::NewEnumerator1)
    Sprite Size Mode=::NewEnumerator4
    UsePositionOffset=true
    Write Material Random=true
   )
   &ShapeLocation(MaskSpawnGroup=false, Override Seed=false, Sphere Radius=15, UseOwnerScale=true)
   &AddVelocity(Velocity=&RandomRangeVector(Minimum=<0 0 250>, Maximum=<1 1 400>))
  ]
  particleUpdateScript [
   &ParticleState()
   &Drag(Drag=0.25, Use Linear Drag=true, Use Rotational Drag=false)
   &ScaleColor(
    Scale Alpha=&FloatFromCurve(FloatCurve=DataInterface() {propertyValues {Curve {keys [<at -0.0046 1.0113 RCIM_Cubic>,<at 1.0011 0.002 RCIM_Cubic>]}}})
    ScaleA=true
    ScaleRGB=true
   )
   &SolveForcesAndVelocity(Clamp Velocity=false, Limit Acceleration=false)
   &Color(Color=&MergeHelper_ColorSwitch(
     0=&MakeLinearColorFromVectorAndFloat(
      "Vector (RGB)"=&VectorFromCurve(VectorCurve=DataInterface() {propertyValues {XCurve {keys [<at 0 1>,<at 1 0>]}, YCurve {keys [<at 0 1>,<at 1 0>]}, ZCurve {keys [<at 0 1>,<at 1 0>]}}})
      "Float (Alpha)"=&FloatFromCurve(FloatCurve=DataInterface() {propertyValues {Curve {keys [<at 0 10.5493 RCIM_Constant>,<at 0.1773 0.1 RCIM_Constant>,<at 0.2971 10.5698 RCIM_Constant>,<at 0.3807 0.1 RCIM_Constant>,<at 0.5366 10.5698 RCIM_Constant>,<at 0.6334 0.1 RCIM_Constant>,<at 0.7824 10.5698 RCIM_Constant>,<at 0.8734 1 RCIM_Constant>,<at 1 1 RCIM_Constant>]}}})
     )
     1=&MakeLinearColorFromVectorAndFloat(
      "Vector (RGB)"=&VectorFromCurve(VectorCurve=DataInterface() {propertyValues {XCurve {keys [<at 0 1>,<at 1 0>]}, YCurve {keys [<at 0 1>,<at 1 0>]}, ZCurve {keys [<at 0 1>,<at 1 0>]}}})
      "Float (Alpha)"=&FloatFromCurve(FloatCurve=DataInterface() {propertyValues {Curve {keys [<at 0 1 RCIM_Constant>,<at 0.1223 10.5698 RCIM_Constant>,<at 0.2059 1.0005 RCIM_Constant>,<at 0.4553 10.5698 RCIM_Constant>,<at 0.5532 1.0005 RCIM_Constant>,<at 0.7824 10.5698 RCIM_Constant>,<at 0.8734 1 RCIM_Constant>,<at 1 1 RCIM_Constant>]}}})
     )
     2=&MakeLinearColorFromVectorAndFloat(
      "Vector (RGB)"=&VectorFromCurve(VectorCurve=DataInterface() {propertyValues {XCurve {keys [<at 0 1>,<at 1 0>]}, YCurve {keys [<at 0 1>,<at 1 0>]}, ZCurve {keys [<at 0 1>,<at 1 0>]}}})
      "Float (Alpha)"=&FloatFromCurve(FloatCurve=DataInterface() {propertyValues {Curve {keys [<at 0 10.5493 RCIM_Constant>,<at 0.1773 0.4624 RCIM_Constant>,<at 0.2971 10.5698 RCIM_Constant>,<at 0.3807 0.4624 RCIM_Constant>,<at 0.5366 10.5698 RCIM_Constant>,<at 0.6334 0.4624 RCIM_Constant>,<at 0.7824 10.5698 RCIM_Constant>,<at 0.8734 1 RCIM_Constant>,<at 1 1 RCIM_Constant>]}}})
     )
     3=?
     Index=&SpawnGroup()
     Num=::NewEnumerator1
    )) {module=Color001}
  ]
  rendererClasses [&NiagaraSpriteRendererProperties]
  renderers [
   {rendererClass=&NiagaraSpriteRendererProperties}
   {rendererClass=&NiagaraSpriteRendererProperties, rendererIndex=1}
  ]
 }
 Lines_3ways {
  emitterUpdateScript [
   &EmitterState(
    Delay First Loop Only=true
    Inactive Response=::NewEnumerator1
    Life Cycle Mode=::NewEnumerator1
    Loop Delay=0.2
    Loop Duration=1
    UseLoopDelay=true
   )
   &SpawnRate(SpawnRate=25, Use Spawn Probability=false)
   &SpawnRate(SpawnGroup=1, SpawnRate=25, Use Spawn Probability=false) {module=SpawnRate001}
   &SpawnRate(SpawnGroup=2, SpawnRate=10, Use Spawn Probability=false) {module=SpawnRate002}
  ]
  particleSpawnScript [
   /SetVariables(Particles.VisibilityTag=&SpawnGroup()) {isSetParameters=true}
   &InitializeParticle(
    Color=&MergeHelper_ColorSwitch(
     0=&RandomRangeLinearColor(Minimum=<rgba 0 0 0 0>, Maximum=<rgba 0.0522 0.1458 0 0>)
     1=&RandomRangeLinearColor(Minimum=<rgba 0 0 0 0>, Maximum=<rgba 0.0522 0.1458 0 0>)
     2=&RandomRangeLinearColor(Minimum=<rgba 0 0 0 0>, Maximum=<rgba 0.052 0.1458 0 0>)
     3=?
     Index=&SpawnGroup()
     Num=::NewEnumerator1
    )
    Lifetime=&MergeHelper_FloatSwitch(
     0=&RandomRangeFloat(Minimum=0.2, Maximum=0.4)
     1=&RandomRangeFloat(Minimum=0.2, Maximum=0.6, Evaluation Type=::NewEnumerator1)
     2=&RandomRangeFloat(Minimum=0.4, Maximum=0.6, Evaluation Type=::NewEnumerator1)
     3=?
     Index=&SpawnGroup()
     Num=::NewEnumerator1
    )
    Sprite Size=&MergeHelper_Vector2Switch(
     0=&RandomRangeVector2D(Minimum=<1 0.2>, Maximum=<1.5 0.7>, Evaluation Type=::NewEnumerator1)
     1=&RandomRangeVector2D(Minimum=<1 0.15>, Maximum=<1.5 0.8>, Evaluation Type=::NewEnumerator1)
     2=&RandomRangeVector2D(Minimum=<1 0.25>, Maximum=<1.5 0.5>, Evaluation Type=::NewEnumerator1)
     3=?
     4=?
     Index=&SpawnGroup()
     Num=::NewEnumerator1
    )
    Sprite Size Mode=::NewEnumerator4
    UsePositionOffset=false
    Write Material Random=true
   )
   &AddVectorToPosition(Vector:Vector3f=&TransformVector(Vector=<0 800 0>, Destination Space=::Simulation))
   &ShapeLocation(
    MaskSpawnGroup=false
    Override Seed=false
    Sphere Radius=&MergeHelper_FloatSwitch(0=5, 1=0, 2=0, 3=?, Index=&SpawnGroup(), Num=::NewEnumerator1)
    UseOwnerScale=true
   )
   &AddVelocity(
    Cone Angle=0
    Cone Angle Mode:ENiagara_AngleInput=::NewEnumerator1
    Cone Axis=&MergeHelper_VectorSwitch(0=<0 0 0>, 1=<0 0 1>, 2=<0 0 0>, 3=?, Index=&SpawnGroup(), Num=::NewEnumerator1)
    Module.UseConeFalloff=false
    Module.UseVelDistribution=false
    Override Seed=false
    Velocity Mode=::NewEnumerator2
    Velocity Speed=&MergeHelper_FloatSwitch(0=0, 1=50, 2=0, 3=?, Index=&SpawnGroup(), Num=::NewEnumerator1)
   )
   &ShapeLocation(
    Cylinder Height=&MergeHelper_FloatSwitch(0=0, 1=0, 2=400, 3=?, Index=&SpawnGroup(), Num=::NewEnumerator1)
    Cylinder Height Midpoint=0
    Cylinder Radius=&MergeHelper_FloatSwitch(0=0, 1=0, 2=500, 3=?, Index=&SpawnGroup(), Num=::NewEnumerator1)
    MaskSpawnGroup=false
    Non Uniform Scale=&MergeHelper_VectorSwitch(0=<0 0 0>, 1=<0 0 0>, 2=<1 1 0.5>, 3=?, Index=&SpawnGroup(), Num=::NewEnumerator1)
    Override Seed=false
    Rotation Mode=::NewEnumerator2
    Shape Primitive=::NewEnumerator1
    Surface Only=true
    Surface Only Band Thickness:NiagaraFloat=&MergeHelper_FloatSwitch(0=0, 1=0, 2=0, 3=?, Index=&SpawnGroup(), Num=::NewEnumerator1)
    Use Endcaps In Surface Only Mode=false
    UseOwnerScale=true
   ) {module=ShapeLocation001}
   &InheritVelocity(
    Inherited Velocity Amount Scale=<0.2 0.2 0.2>
    Inherited Velocity Speed Limit=0
    Speed Limit=false
    UseSpeedThreshold=false
    Velocity Source=&Multiply_VectorByFloat(
     Vector=&TransformVector(Vector=$Engine.Owner.Velocity, Source Space=::World, Destination Space=::Local)
     Float=-1
    )
   ) {module=InheritVelocity001}
   &AddVelocity(Velocity=&FN_FrontEndVector3Modifier(Gameplay=<0 0 0>, Front End=<0 -8 30>)) {module=AddVelocity002}
  ]
  particleUpdateScript [
   &ParticleState()
   &ScaleSpriteSize(
    Scale Factor=&MergeHelper_Vector2Switch(
     0=&Vector2DFromCurve(
      Vector2Curve=DataInterface() {propertyValues {
        XCurve {keys [<at 0 1 RCIM_Cubic>, {interpMode=RCIM_Cubic, tangentMode=RCTM_Break, time=0.2, value=1}, <at 1 1 RCIM_Cubic>]}
        YCurve {keys [
          <at 0 0 RCIM_Cubic>
          {arriveTangent=1105.7345, interpMode=RCIM_Cubic, leaveTangent=-373.3544, tangentMode=RCTM_Break, time=0.2, value=100}
          <at 1 0 RCIM_Cubic>
         ]}
       }}
      ScaleCurve=<1 1>
     )
     1=&Vector2DFromCurve(
      Vector2Curve=DataInterface() {propertyValues {
        XCurve {keys [<at 0 1 RCIM_Cubic>, {interpMode=RCIM_Cubic, tangentMode=RCTM_Break, time=0.2, value=1}, <at 1 1 RCIM_Cubic>]}
        YCurve {keys [
          <at 0 0 RCIM_Cubic>
          {arriveTangent=1105.7345, interpMode=RCIM_Cubic, leaveTangent=-373.3544, tangentMode=RCTM_Break, time=0.2, value=100}
          <at 1 0 RCIM_Cubic>
         ]}
       }}
      ScaleCurve=<1 1>
     )
     2=&Vector2DFromCurve(
      Vector2Curve=DataInterface() {propertyValues {
        XCurve {keys [<at 0 1 RCIM_Cubic>, {interpMode=RCIM_Cubic, tangentMode=RCTM_Break, time=0.2, value=1}, <at 1 1 RCIM_Cubic>]}
        YCurve {keys [
          <at 0 0 RCIM_Cubic>
          {arriveTangent=493.0888, interpMode=RCIM_Cubic, leaveTangent=-117.3663, tangentMode=RCTM_Break, time=0.2, value=38.8683}
          <at 1 0 RCIM_Cubic>
         ]}
       }}
      ScaleCurve=<1 1>
     )
     3=?
     4=?
     Index=&SpawnGroup()
     Num=::NewEnumerator1
    )
    Scale Sprite Size Mode=::NewEnumerator2
   )
   &SkeletalMeshLocation(
    Bone / Socket Index=&SequentialInt(Use Array Data:NiagaraBool=false, First Number:NiagaraInt32=0, Sequence:NiagaraDataInterfaceArrayInt32=?, Last Number:NiagaraInt32=3)
    Bone Sampling Mode=::NewEnumerator4
    MaskSpawnGroup=false
    Orientation Sampling=::NewEnumerator1
    Sampled Position Offset=&FN_FrontEndVector3Modifier(Gameplay=<0 0 -10>, Front End=<0 0 0>)
    Sampled Position Offset Coordinate Space:ENiagaraSkelMeshTransforms=::NewEnumerator2
    Skeletal Mesh=$System.Skeletal Mesh Reader
   ) {module=SkeletalMeshLocation001}
   &Drag(
    Drag=&MergeHelper_FloatSwitch(0=-5, 1=-3, 2=-1, 3=?, Index=&SpawnGroup(), Num=::NewEnumerator1)
    Use Linear Drag=true
    Use Rotational Drag=true
   )
   &SolveForcesAndVelocity(Clamp Velocity=false, Limit Acceleration=false)
   &Color(Color=&MergeHelper_ColorSwitch(
     0=&MakeLinearColorFromVectorAndFloat(
      "Vector (RGB)"=&VectorFromCurve(VectorCurve=DataInterface() {propertyValues {XCurve {keys [<at 0 1>,<at 1 0>]}, YCurve {keys [<at 0 1>,<at 1 0>]}, ZCurve {keys [<at 0 1>,<at 1 0>]}}})
      "Float (Alpha)"=&FloatFromCurve(FloatCurve=DataInterface() {propertyValues {Curve {keys [<at 0 10.5493 RCIM_Constant>,<at 0.1773 0.1 RCIM_Constant>,<at 0.2971 10.5698 RCIM_Constant>,<at 0.3807 0.1 RCIM_Constant>,<at 0.5366 10.5698 RCIM_Constant>,<at 0.6334 0.1 RCIM_Constant>,<at 0.7824 10.5698 RCIM_Constant>,<at 0.8734 1 RCIM_Constant>,<at 1 1 RCIM_Constant>]}}})
     )
     1=&MakeLinearColorFromVectorAndFloat(
      "Vector (RGB)"=&VectorFromCurve(VectorCurve=DataInterface() {propertyValues {XCurve {keys [<at 0 1>,<at 1 0>]}, YCurve {keys [<at 0 1>,<at 1 0>]}, ZCurve {keys [<at 0 1>,<at 1 0>]}}})
      "Float (Alpha)"=&FloatFromCurve(FloatCurve=DataInterface() {propertyValues {Curve {keys [<at 0 1 RCIM_Constant>,<at 0.1223 10.5698 RCIM_Constant>,<at 0.2059 1.0005 RCIM_Constant>,<at 0.4553 10.5698 RCIM_Constant>,<at 0.5532 1.0005 RCIM_Constant>,<at 0.7824 10.5698 RCIM_Constant>,<at 0.8734 1 RCIM_Constant>,<at 1 1 RCIM_Constant>]}}})
     )
     2=&MakeLinearColorFromVectorAndFloat(
      "Vector (RGB)"=&VectorFromCurve(VectorCurve=DataInterface() {propertyValues {XCurve {keys [<at 0 1>,<at 1 0>]}, YCurve {keys [<at 0 1>,<at 1 0>]}, ZCurve {keys [<at 0 1>,<at 1 0>]}}})
      "Float (Alpha)"=&FloatFromCurve(FloatCurve=DataInterface() {propertyValues {Curve {keys [<at 0 10.5493 RCIM_Constant>,<at 0.1773 0.4624 RCIM_Constant>,<at 0.2971 10.5698 RCIM_Constant>,<at 0.3807 0.4624 RCIM_Constant>,<at 0.5366 10.5698 RCIM_Constant>,<at 0.6334 0.4624 RCIM_Constant>,<at 0.7824 10.5698 RCIM_Constant>,<at 0.8734 1 RCIM_Constant>,<at 1 1 RCIM_Constant>]}}})
     )
     3=?
     Index=&SpawnGroup()
     Num=::NewEnumerator1
    )) {module=Color001}
  ]
  rendererClasses [&NiagaraSpriteRendererProperties]
  renderers [
   {rendererClass=&NiagaraSpriteRendererProperties}
   {rendererClass=&NiagaraSpriteRendererProperties, rendererIndex=1}
   {rendererClass=&NiagaraSpriteRendererProperties, rendererIndex=2}
  ]
 }
 Refraction {
  simTarget=GPUComputeSim
  emitterUpdateScript [
   &EmitterState(
    Delay First Loop Only=true
    Inactive Response=::NewEnumerator1
    Life Cycle Mode=::NewEnumerator1
    Loop Delay=0.2
    Loop Duration=1
    UseLoopDelay=true
   )
   &SpawnRate(SpawnRate=16, Use Spawn Probability=false)
   &SpawnPerUnit(
    Module.Use Max Movement Threshold=true
    Module.Use Movement Tolerance=true
    Movement Tolerance=0.1
    Spawn Spacing:NiagaraFloat=&Divide_Float(A=1, B=&Divide_Float(A=2, B=100))
    Use Spawn Probability=false
   )
  ]
  particleSpawnScript [
   &InitializeParticle(
    Lifetime=&RandomRangeFloat(Minimum=0.05, Maximum=0.1, Evaluation Type=::NewEnumerator1)
    Sprite Size=&RandomRangeVector2D(Minimum=<20 2>, Maximum=<40 4>, Evaluation Type=::NewEnumerator1)
    Sprite Size Mode=::NewEnumerator4
    UsePositionOffset=false
    Write Material Random=true
   )
   &AddVectorToPosition(Vector:Vector3f=&TransformVector(Vector=<0 50 0>, Destination Space=::Simulation))
   &Color(Color=&MakeLinearColorFromVectorAndFloat(
     "Vector (RGB)"=&RandomRangeVector(Minimum=<0 0 0>, Maximum=<0.0522 0.1458 0>, Evaluation Type=::NewEnumerator1)
     "Float (Alpha)"=0
    ))
   &ShapeLocation(MaskSpawnGroup=false, Override Seed=false, Sphere Radius=15, UseOwnerScale=true)
  ]
  particleUpdateScript [
   &ParticleState()
   &SkeletalMeshLocation(
    Bone / Socket Index=&SequentialInt(Use Array Data:NiagaraBool=false, First Number:NiagaraInt32=0, Sequence:NiagaraDataInterfaceArrayInt32=?, Last Number:NiagaraInt32=3)
    Bone Sampling Mode=::NewEnumerator4
    MaskSpawnGroup=false
    Orientation Sampling=::NewEnumerator1
    Sampled Position Offset=<0 0 -10>
    Sampled Position Offset Coordinate Space:ENiagaraSkelMeshTransforms=::NewEnumerator2
    Skeletal Mesh=$System.Skeletal Mesh Reader
   ) {module=SkeletalMeshLocation001}
  ]
  rendererClasses [&NiagaraSpriteRendererProperties]
  renderers [{rendererClass=&NiagaraSpriteRendererProperties}]
 }
 Ribbon_Feet {
  emitterUpdateScript [
   &EmitterState(Inactive Response=::NewEnumerator1, Life Cycle Mode=::NewEnumerator1, Loop Duration=1, UseLoopDelay=true)
   &SpawnRate(SpawnRate=120, Use Spawn Probability=false)
  ]
  particleSpawnScript [
   &FN_SetSpawnGroup()
   /SetVariables(Particles.RibbonID=/Niagara/DynamicInputs/Meshes/MakeRibbonIDFromSpawnGroup()) {enabled=false, isSetParameters=true, module=SetVariables_#2}
   &InitializeParticle(
    Lifetime=&MergeHelper_FloatSwitch(0=0.15, 1=0.15, 2=?, 3=?, Index=&SpawnGroup(), Num=::NewEnumerator0)
    Ribbon Width=&FN_FrontEndFloatModifier(Gameplay=15, Front End=18)
    Ribbon Width Mode=::NewEnumerator1
    UsePositionOffset=false
    Write Material Random=true
   ) {module=InitializeParticle001}
   &SkeletalMeshLocation(
    Bone / Socket Index=&SequentialInt(Use Array Data:NiagaraBool=false, First Number:NiagaraInt32=0, Sequence:NiagaraDataInterfaceArrayInt32=?, Last Number:NiagaraInt32=3)
    Bone Sampling Mode=::NewEnumerator4
    MaskSpawnGroup=false
    Orientation Sampling=::NewEnumerator1
    Sampled Position Offset=&FN_FrontEndVector3Modifier(Gameplay=<0 0 -10>, Front End=<0 0 -17>)
    Sampled Position Offset Coordinate Space:ENiagaraSkelMeshTransforms=::NewEnumerator2
    Skeletal Mesh=$System.Skeletal Mesh Reader
   ) {enabled=false, module=SkeletalMeshLocation001}
   &SkeletalMeshLocation(
    Bone / Socket Index=&SequentialInt(Use Array Data:NiagaraBool=false, First Number:NiagaraInt32=2, Sequence:NiagaraDataInterfaceArrayInt32=?, Last Number:NiagaraInt32=3)
    Bone Sampling Mode=::NewEnumerator4
    MaskSpawnGroup=false
    Orientation Sampling=::NewEnumerator1
    Sampled Position Offset=&FN_FrontEndVector3Modifier(Gameplay=<0 0 -11>, Front End=<0 0 -17>)
    Sampled Position Offset Coordinate Space:ENiagaraSkelMeshTransforms=::NewEnumerator2
    Skeletal Mesh=$System.Skeletal Mesh Reader
   )
   &AddVectorToPosition(Vector:Vector3f=&TransformVector(Vector=<0 0 0>, Destination Space=::Simulation))
   &AddVelocity(Velocity=&FN_FrontEndVector3Modifier(Gameplay=<0 0 0>, Front End=<0 -150 1000>))
   &ApplyInitialForces()
  ]
  particleUpdateScript [
   &ParticleState()
   &SolveForcesAndVelocity(Clamp Velocity=false, Limit Acceleration=false)
   &ScaleRibbonWidth(Ribbon Width Scale=&FloatFromCurve(FloatCurve=DataInterface() {propertyValues {Curve {keys [<at 0.05 0.3493 RCIM_Constant>,<at 0.125 0.5603 RCIM_Constant>,<at 0.2012 0.7777 RCIM_Constant>,<at 0.3146 0.5626 RCIM_Constant>,<at 0.4202 0.3649 RCIM_Constant>,<at 0.6988 0.258 RCIM_Constant>,<at 1.0081 0.2166 RCIM_Constant>]}}}))
  ]
  rendererClasses [&NiagaraRibbonRendererProperties]
  renderers [{rendererClass=&NiagaraRibbonRendererProperties}]
 }
 Ribbon_Hands {
  emitterUpdateScript [
   &EmitterState(Inactive Response=::NewEnumerator1, Life Cycle Mode=::NewEnumerator1, Loop Duration=1, UseLoopDelay=true)
   &SpawnRate(SpawnRate=120, Use Spawn Probability=false)
  ]
  particleSpawnScript [
   &FN_SetSpawnGroup()
   /SetVariables(Particles.RibbonID=&MakeRibbonIDFromSpawnGroup()) {enabled=false, isSetParameters=true, module=SetVariables_#2}
   &InitializeParticle(
    Lifetime=&MergeHelper_FloatSwitch(0=0.2, 1=0.2, 2=?, 3=?, Index=&SpawnGroup(), Num=::NewEnumerator0)
    Ribbon Width=&FN_FrontEndFloatModifier(Gameplay=15, Front End=18)
    Ribbon Width Mode=::NewEnumerator1
    UsePositionOffset=false
    Write Material Random=true
   ) {module=InitializeParticle001}
   &SkeletalMeshLocation(
    Bone / Socket Index=&SequentialInt(Use Array Data:NiagaraBool=false, First Number:NiagaraInt32=0, Sequence:NiagaraDataInterfaceArrayInt32=?, Last Number:NiagaraInt32=3)
    Bone Sampling Mode=::NewEnumerator4
    MaskSpawnGroup=false
    Orientation Sampling=::NewEnumerator1
    Sampled Position Offset=&FN_FrontEndVector3Modifier(Gameplay=<0 0 -10>, Front End=<0 0 -17>)
    Sampled Position Offset Coordinate Space:ENiagaraSkelMeshTransforms=::NewEnumerator2
    Skeletal Mesh=$System.Skeletal Mesh Reader
   ) {enabled=false, module=SkeletalMeshLocation001}
   &SkeletalMeshLocation(
    Bone / Socket Index=&SequentialInt(Use Array Data:NiagaraBool=false, First Number:NiagaraInt32=0, Sequence:NiagaraDataInterfaceArrayInt32=?, Last Number:NiagaraInt32=1)
    Bone Sampling Mode=::NewEnumerator4
    MaskSpawnGroup=false
    Orientation Sampling=::NewEnumerator1
    Sampled Position Offset=&FN_FrontEndVector3Modifier(Gameplay=<0 0 -11>, Front End=<0 0 -17>)
    Sampled Position Offset Coordinate Space:ENiagaraSkelMeshTransforms=::NewEnumerator2
    Skeletal Mesh=$System.Skeletal Mesh Reader
   )
   &AddVectorToPosition(Vector:Vector3f=&TransformVector(Vector=<0 0 0>, Destination Space=::Simulation))
   &AddVelocity(Velocity=&FN_FrontEndVector3Modifier(Gameplay=<0 0 0>, Front End=<0 -150 1000>))
   &ApplyInitialForces()
  ]
  particleUpdateScript [
   &ParticleState()
   &SolveForcesAndVelocity(Clamp Velocity=false, Limit Acceleration=false)
   &ScaleRibbonWidth(Ribbon Width Scale=&FloatFromCurve(FloatCurve=DataInterface() {propertyValues {Curve {keys [<at 0.05 0.3493 RCIM_Constant>,<at 0.125 0.5603 RCIM_Constant>,<at 0.2012 0.7777 RCIM_Constant>,<at 0.3146 0.5626 RCIM_Constant>,<at 0.4202 0.3649 RCIM_Constant>,<at 0.6988 0.258 RCIM_Constant>,<at 1.0081 0.2166 RCIM_Constant>]}}}))
  ]
  rendererClasses [&NiagaraRibbonRendererProperties]
  renderers [{rendererClass=&NiagaraRibbonRendererProperties}]
 }
}
summary {
 systemName=NS_Contrails_GlitchInTheSystem_Converted
 emitters [
  {emitterName=Ribbon_Hands, rendererClasses [&NiagaraRibbonRendererProperties]}
  {emitterName=Dots_Center, rendererClasses [&NiagaraSpriteRendererProperties]}
  {emitterName=Dots_Hand, rendererClasses [&NiagaraSpriteRendererProperties]}
  {emitterName=Refraction, simTarget=GPUComputeSim, rendererClasses [&NiagaraSpriteRendererProperties]}
  {emitterName=8BitDot, simTarget=GPUComputeSim, rendererClasses [&NiagaraSpriteRendererProperties]}
  {emitterName=Lines_3ways, rendererClasses [&NiagaraSpriteRendererProperties]}
  {emitterName=Fountain, rendererClasses [&NiagaraSpriteRendererProperties]}
  {emitterName=Ribbon_Feet, rendererClasses [&NiagaraRibbonRendererProperties]}
 ]
 userVariables [
  User.bIsFrontEnd {type=NiagaraBool}
  User.bIsFrontEndPreview {type=NiagaraBool}
  User.ParaGlideLeanAlpha {type=NiagaraFloat, defaultValue=0, description=""}
 ]
}
