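# Niagara skydiving contrail cosmetic "Retroraygun", apparently converted from Cascade, that trails effects from a character's skeletal mesh.
# A sprite glow follows both hand bones, while green-tinted mesh particles and ribbons spawn at sampled bones in two spawn groups, with ribbon IDs separating trails.
# Glide lean alpha scales mesh spawn rate and ribbon velocity.
# FrontEnd modifier inputs swap rates, lifetimes, velocities, offsets and curves between gameplay and front-end display.
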NiagaraSystem /BRCosmetics/Athena/Effects/Particles/Skydiving/NS_Contrails_Retroraygun_Converted
dependencies {
 usedDataInterfaces [&NiagaraDataInterfaceVectorCurve, &NiagaraDataInterfaceCurve, &NiagaraDataInterfaceSkeletalMesh]
 usedDynamicInputs []
 usedModules [
  &self:SystemSpawnScript.SystemScriptSource.SystemScriptGraph.NiagaraNodeAssignment_0.SetVariables_#1
  &SystemState
 ]
}
emitters {
 HandGlow_R {
  emitterUpdateScript [
   &EmitterState(Inactive Response=::NewEnumerator1, Life Cycle Mode=::NewEnumerator1, Loop Duration=0.1, UseLoopDelay=true)
   &SpawnRate(SpawnRate=10, Use Spawn Probability=false)
  ]
  particleSpawnScript [
   &FN_SetSpawnGroup(Number of Spawn Groups=2)
   &InitializeParticle(Lifetime=0.5, Sprite Size=<65 65>, Sprite Size Mode=::NewEnumerator4, UsePositionOffset=false, Write Material Random=true)
   &Color(Color=&MakeLinearColorFromVectorAndFloat("Vector (RGB)"=<1 1 1>, "Float (Alpha)"=0.1))
  ]
  particleUpdateScript [
   &ParticleState()
   &SkeletalMeshLocation(
    Bone / Socket Index=/CascadeToNiagaraConverter/NiagaraScripts/DynamicInputScripts/SequentialInt(Use Array Data:NiagaraBool=false, First Number:NiagaraInt32=0, Sequence:NiagaraDataInterfaceArrayInt32=?, Last Number:NiagaraInt32=1)
    Bone Sampling Mode=::NewEnumerator4
    MaskSpawnGroup=false
    Orientation Sampling=::NewEnumerator1
    Skeletal Mesh=DataInterface() {propertyValues {
      DefaultMesh=null
      ExcludeBoneName=root
      OptionalComponentClass=null
      PreviewMesh=null
      SkinningMode=SkinOnTheFly
      SoftSourceActor=null
      SourceMode=Default
      UvSetIndex=0
      WholeMeshLOD=-1
      bExcludeBone=false
      bReadDeformedGeometry=true
      bRequireCurrentFrameData=true
      ComponentTags []
      FilteredBones [hand_r,hand_l]
      FilteredSockets []
      MeshUserParameter {parameter {name=None, typeDefHandle {registeredTypeIndex=1}, varData []}}
      SamplingRegions []
     }}
   )
   &ScaleSpriteSize(
    Scale Sprite Size Mode=::NewEnumerator0
    Uniform Scale Factor=&FloatFromCurve(FloatCurve=DataInterface() {propertyValues {Curve {keys [<at 0 0.6>,<at 1 1.3>]}}})
   )
   &CameraOffset(Camera Offset Amount=15)
   &ScaleColor(
    Scale Alpha=&FloatFromCurve(FloatCurve=DataInterface() {propertyValues {Curve {keys [<at 0 0 RCIM_Cubic>,<at 0.2321 0.9177 RCIM_Cubic>,<at 1 0>]}}})
    ScaleA=true
    ScaleRGB=true
   )
  ]
  rendererClasses [&NiagaraSpriteRendererProperties]
  renderers [{rendererClass=&NiagaraSpriteRendererProperties}]
 }
 Particle_Emitter {
  emitterUpdateScript [
   &EmitterState(Inactive Response=::NewEnumerator1, Life Cycle Mode=::NewEnumerator1, Loop Duration=0.1, UseLoopDelay=true)
   &SpawnRate(
    SpawnGroup=1
    SpawnRate=&FN_FrontEndFloatModifier(Gameplay=&Lerp_Float(A=20, B=70, Alpha=$User.ParaGlideLeanAlpha), Front End=32)
    Use Spawn Probability=false
   ) {module=SpawnRate001}
   &SpawnRate(
    SpawnRate=&FN_FrontEndFloatModifier(Gameplay=&Lerp_Float(A=20, B=70, Alpha=$User.ParaGlideLeanAlpha), Front End=32)
    Use Spawn Probability=false
   ) {module=SpawnRate003}
  ]
  particleSpawnScript [
   &InitializeParticle(
    Lifetime=&FN_FrontEndFloatModifier(Gameplay=0.18, Front End=0.5)
    Mesh Scale Mode=::NewEnumerator4
    UsePositionOffset=false
    Write Material Random=true
   )
   &Color(Color=&MakeLinearColorFromVectorAndFloat("Vector (RGB)"=<2 2 2>, "Float (Alpha)"=1))
   &AddVelocity(Velocity=&FN_FrontEndVector3Modifier(Gameplay=&Lerp_Vector(A=<0 0 0>, B=<0 0 0>, Alpha=$User.ParaGlideLeanAlpha), Front End=<0 -250 800>))
   &SkeletalMeshLocation(
    Bone / Socket Index=&SpawnGroup()
    Bone Sampling Mode=::NewEnumerator4
    MaskSpawnGroup=false
    Orientation Sampling=::NewEnumerator1
    Sampled Position Offset=&FN_FrontEndVector3Modifier(Gameplay=<0 0 0>, Front End=<0 0 10>)
    Sampled Position Offset Coordinate Space:ENiagaraSkelMeshTransforms=::NewEnumerator2
    Skeletal Mesh=$System.Skeletal Mesh Reader
   )
   &ApplyInitialForces() {enabled=false}
  ]
  particleUpdateScript [
   &ParticleState()
   &ScaleMeshSize(Scale Factor=&VectorFromFloat(Value=&FN_FrontEndFloatModifier(
      Gameplay=&FloatFromCurve(FloatCurve=DataInterface() {propertyValues {Curve {
          defaultValue=0.5
          keys [{interpMode=RCIM_Cubic, tangentMode=RCTM_User, time=0, value=0.4}, {interpMode=RCIM_Cubic, tangentMode=RCTM_User, time=0.7, value=1}]
         }}})
      Front End=&FloatFromCurve(FloatCurve=DataInterface() {propertyValues {Curve {
          defaultValue=0.4
          keys [
           {interpMode=RCIM_Cubic, tangentMode=RCTM_User, time=0.0038, value=0.25}
           {arriveTangent=0.3013, interpMode=RCIM_Cubic, leaveTangent=0.3013, time=0.0452, value=0.5416}
           {arriveTangentWeight=0.0585, interpMode=RCIM_Cubic, leaveTangentWeight=0.1013, tangentMode=RCTM_User, time=0.1179, value=0.2844}
           {arriveTangentWeight=0.0254, interpMode=RCIM_Cubic, leaveTangentWeight=0.118, tangentMode=RCTM_User, time=0.1902, value=0.7482}
           {arriveTangent=0.0967, interpMode=RCIM_Cubic, leaveTangent=0.0967, time=0.3627, value=0.4437}
           {arriveTangentWeight=0.0517, interpMode=RCIM_Cubic, leaveTangentWeight=0.2159, tangentMode=RCTM_User, time=0.7735, value=0.8046}
           {interpMode=RCIM_Cubic, tangentMode=RCTM_User, time=1, value=0}
          ]
         }}})
     ))) {module=ScaleMeshSize002}
   &ScaleColor(
    Scale RGB=&VectorFromCurve(VectorCurve=DataInterface() {propertyValues {XCurve {keys [<at 0 0.6376>,<at 1 0.4852>]}, YCurve {keys [<at 0 0.8632>,<at 1 0.8469>]}, ZCurve {keys [<at 0 0.2307>,<at 1 0.006>]}}})
    ScaleA=true
    ScaleRGB=true
   )
   &DynamicMaterialParameters(
    Index 0 Param 1=&FN_FrontEndFloatModifier(
     Gameplay=&FloatFromCurve(FloatCurve=DataInterface() {propertyValues {Curve {keys [<at 0 0.3>,<at 0.5 1>]}}})
     Front End=&FloatFromCurve(FloatCurve=DataInterface() {propertyValues {Curve {keys [<at 0 0.3>,<at 0.3 1>]}}})
    )
    Index 0 Param 2=1
    Index 0 Param 3=1
    Index 0 Param 4=1
    Module.Index0Float0Write=true
    Module.Index0Float1Write=true
    Module.Index0Float2Write=true
    Module.Index0Float3Write=true
   ) {module=DynamicMaterialParameters001}
   &SolveForcesAndVelocity(Clamp Velocity=false, Limit Acceleration=false)
  ]
  rendererClasses [&NiagaraMeshRendererProperties]
  renderers [{rendererClass=&NiagaraMeshRendererProperties, rendererIndex=0}]
 }
 Ribbon_Left_Hand {
  emitterUpdateScript [
   &EmitterState(Inactive Response=::NewEnumerator1, Life Cycle Mode=::NewEnumerator1, Loop Duration=1, UseLoopDelay=true)
   &SpawnRate(SpawnRate=120, Use Spawn Probability=false)
  ]
  particleSpawnScript [
   &FN_SetSpawnGroup(Number of Spawn Groups=2)
   /SetVariables(Particles.RibbonID=/Niagara/DynamicInputs/Meshes/MakeRibbonIDFromSpawnGroup()) {isSetParameters=true}
   &InitializeParticle(Lifetime=0.25, Ribbon Width=35, Ribbon Width Mode=::NewEnumerator1, UsePositionOffset=false, Write Material Random=true)
   &SkeletalMeshLocation(
    Bone / Socket Index=&SequentialInt(Use Array Data:NiagaraBool=false, First Number:NiagaraInt32=0, Sequence:NiagaraDataInterfaceArrayInt32=?, Last Number:NiagaraInt32=1)
    Bone Sampling Mode=::NewEnumerator4
    MaskSpawnGroup=false
    Orientation Sampling=::NewEnumerator1
    Sampled Position Offset=&FN_FrontEndVector3Modifier(Gameplay=<0 0 -10>, Front End=<0 0 0>)
    Sampled Position Offset Coordinate Space:ENiagaraSkelMeshTransforms=::NewEnumerator2
    Skeletal Mesh=$System.Skeletal Mesh Reader
   ) {module=SkeletalMeshLocation001}
   &DynamicMaterialParameters(
    Index 0 Param 1=&RandomRangeFloat(Minimum=1, Maximum=2, Evaluation Type=::NewEnumerator1)
    Index 0 Param 3=&RandomRangeFloat(Minimum=1, Maximum=2, Evaluation Type=::NewEnumerator1)
    Module.Index0Float0Write=true
    Module.Index0Float1Write=true
    Module.Index0Float2Write=true
    Module.Index0Float3Write=true
   )
   /CascadeToNiagaraConverter/NiagaraScripts/ModuleScripts/AddVectorToPosition(Vector:Vector3f=&TransformVector(Vector=<0 0 0>, Destination Space=::Simulation))
   &Color(Color=&MakeLinearColorFromVectorAndFloat("Vector (RGB)"=<0.4852 0.8469 0.006>, "Float (Alpha)"=0.5))
   &AddVelocity(Velocity=&FN_FrontEndVector3Modifier(Gameplay=&Lerp_Vector(A=<0 0 -900>, B=<1 1 -1200>, Alpha=$User.ParaGlideLeanAlpha), Front End=<0 -400 1500>))
   &ApplyInitialForces()
  ]
  particleUpdateScript [
   &ParticleState()
   &DynamicMaterialParameters(
    Index 0 Param 2=1
    Index 0 Param 4=1
    Module.Index0Float0Write=false
    Module.Index0Float1Write=true
    Module.Index0Float2Write=false
    Module.Index0Float3Write=true
   ) {module=DynamicMaterialParameters001}
   &ScaleColor(
    Scale Alpha=&FloatFromCurve(FloatCurve=DataInterface() {propertyValues {Curve {keys [<at 0 0>,<at 0.15 1>,<at 1 0>]}}})
    Scale RGB=&VectorFromCurve(VectorCurve=DataInterface() {propertyValues {XCurve {keys [<at 0 10>,<at 1 1>]}, YCurve {keys [<at 0 10>,<at 1 1>]}, ZCurve {keys [<at 0 10>,<at 1 1>]}}})
    ScaleA=true
    ScaleRGB=true
   )
   &SolveForcesAndVelocity(
    Clamp Velocity=true
    Limit Acceleration=false
    Speed Limit:NiagaraFloat=&FN_FrontEndFloatModifier(Gameplay=0, Front End=1000)
   )
   &ScaleRibbonWidth(Ribbon Width Scale=&FloatFromCurve(FloatCurve=DataInterface() {propertyValues {Curve {
        defaultValue=0
        keys [
         {arriveTangent=4.5998, interpMode=RCIM_Cubic, leaveTangent=4.5998, tangentMode=RCTM_User, time=0, value=0}
         {arriveTangent=0.0224, interpMode=RCIM_Cubic, leaveTangent=0.0224, tangentMode=RCTM_User, time=0.5, value=1}
        ]
       }}})) {module=ScaleRibbonWidth001}
  ]
  rendererClasses [&NiagaraRibbonRendererProperties]
  renderers [{rendererClass=&NiagaraRibbonRendererProperties}]
 }
}
summary {
 systemName=NS_Contrails_Retroraygun_Converted
 emitters [
  {emitterName=Ribbon_Left_Hand, rendererClasses [&NiagaraRibbonRendererProperties]}
  {emitterName=Particle_Emitter, rendererClasses [&NiagaraMeshRendererProperties]}
  {emitterName=HandGlow_R, rendererClasses [&NiagaraSpriteRendererProperties]}
 ]
 userVariables [
  User.bIsFrontEnd {type=NiagaraBool}
  User.bIsFrontEndPreview {type=NiagaraBool}
  User.ParaGlideLeanAlpha {type=NiagaraFloat, defaultValue=0, description=""}
 ]
}
