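# Skydiving contrail Niagara system with spray-paint styling, converted from Cascade judging by its name and helper scripts, trailing four-colored streaks from the character's skeletal mesh.
# Short-lived ribbons and animated smoke sprites spawn at sequential bones or sockets 0-3, with spawn group selecting color and separating ribbons.
# FrontEnd modifiers give gameplay and front end different rates, velocities and fade or scale curves.
# User.ParaGlideLeanAlpha raises gameplay smoke spawn rate, supplemented by movement-based spawning.
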NiagaraSystem /BRCosmetics/Athena/Effects/Particles/Skydiving/NS_Contrails_SprayPaint_Master_Converted
dependencies {
 usedDataInterfaces [
  &NiagaraDataInterfaceCurve
  &NiagaraDataInterfaceSpriteRendererInfo
  &NiagaraDataInterfaceSkeletalMesh
 ]
 usedDynamicInputs []
 usedModules [
  &self:SystemSpawnScript.SystemScriptSource.SystemScriptGraph.NiagaraNodeAssignment_0.SetVariables_#1
  &SystemState
 ]
}
emitters {
 Ribbon_Left_Foot {
  emitterUpdateScript [
   &EmitterState(Inactive Response=::NewEnumerator1, Life Cycle Mode=::NewEnumerator1, Loop Duration=1, UseLoopDelay=true)
   &SpawnRate(SpawnRate=&FN_FrontEndFloatModifier(Gameplay=90, Front End=100), Use Spawn Probability=false)
  ]
  particleSpawnScript [
   &FN_SetSpawnGroup()
   /SetVariables(Particles.RibbonID=/Niagara/DynamicInputs/Meshes/MakeRibbonIDFromSpawnGroup()) {isSetParameters=true}
   &InitializeParticle(
    Color=&MergeHelper_ColorSwitch(
     0=<rgba 0.1559 0.7913 1 0.2>
     1=<rgba 0.4342 1 0.1714 0.2>
     2=<rgba 0.9823 0.1413 1 0.2>
     3=<rgba 0.6514 0.6514 0.6514 0.2>
     Index=$Particles.SpawnGroup
     Num=::NewEnumerator2
    )
    Lifetime=0.2
    Ribbon Width=&FN_FrontEndFloatModifier(Gameplay=&Multiply_Float(A=8, B=10), Front End=&Multiply_Float(A=3, B=10))
    Ribbon Width Mode=::NewEnumerator1
    UsePositionOffset=false
    Write Material Random=true
   )
   &SkeletalMeshLocation(
    Bone / Socket Index=/CascadeToNiagaraConverter/NiagaraScripts/DynamicInputScripts/SequentialInt(Use Array Data:NiagaraBool=false, First Number:NiagaraInt32=0, Sequence:NiagaraDataInterfaceArrayInt32=?, Last Number:NiagaraInt32=3)
    Bone Sampling Mode=::NewEnumerator4
    MaskSpawnGroup=false
    Orientation Sampling=::NewEnumerator1
    Sampled Position Offset=<0 0 -10>
    Sampled Position Offset Coordinate Space:ENiagaraSkelMeshTransforms=::NewEnumerator2
    Skeletal Mesh=$System.Skeletal Mesh Reader
   ) {module=SkeletalMeshLocation001}
   /CascadeToNiagaraConverter/NiagaraScripts/ModuleScripts/AddVectorToPosition(Vector:Vector3f=&TransformVector(Vector=<0 0 0>, Destination Space=::Simulation))
   &Color(Color=<rgba 1 1 1 3>) {enabled=false}
   &AddVelocity(Velocity=&FN_FrontEndVector3Modifier(Gameplay=<0 0 1000>, Front End=<0 -400 1500>))
  ]
  particleUpdateScript [
   /SetVariables(Particles.ScaledRibbonWidth:NiagaraFloat=&Multiply_Float(
     A=$Particles.RibbonWidth
     B=&FN_FrontEndFloatModifier(
      Gameplay=&FloatFromCurve(FloatCurve=DataInterface() {propertyValues {Curve {
          defaultValue=0
          keys [
           {arriveTangent=1.7452, interpMode=RCIM_Cubic, leaveTangent=1.7452, tangentMode=RCTM_User, time=0, value=0.46}
           {arriveTangent=0.8385, interpMode=RCIM_Cubic, leaveTangent=0.8385, tangentMode=RCTM_User, time=0.2044, value=0.8108}
           {arriveTangent=0.0224, interpMode=RCIM_Cubic, leaveTangent=0.0224, tangentMode=RCTM_User, time=1, value=1}
          ]
         }}})
      Front End=&FloatFromCurve(FloatCurve=DataInterface() {propertyValues {Curve {keys [
           <at 0 0.5>
           {interpMode=RCIM_Cubic, tangentMode=RCTM_User, time=0.6222, value=1.1001}
           {arriveTangent=0.4859, interpMode=RCIM_Cubic, leaveTangent=0.4859, tangentMode=RCTM_User, time=1, value=1.5701}
          ]}}})
     )
    )) {isSetParameters=true}
   &ParticleState()
   &ScaleColor(
    Scale Alpha=&FN_FrontEndFloatModifier(
     Gameplay=1
     Front End=&FloatFromCurve(FloatCurve=DataInterface() {propertyValues {Curve {keys [<at -0.0142 0.0267>,<at 0.0831 0.35>,<at 1 0>]}}})
    )
    ScaleA=true
    ScaleRGB=true
   )
   &SolveForcesAndVelocity(Clamp Velocity=false, Limit Acceleration=false)
  ]
  rendererClasses [&NiagaraRibbonRendererProperties]
  renderers [{rendererClass=&NiagaraRibbonRendererProperties}]
 }
 ThinSmokeLeftFoot {
  emitterUpdateScript [
   &EmitterState(
    Inactive Response=::NewEnumerator1
    Life Cycle Mode=::NewEnumerator1
    Loop Duration=&RandomRangeFloat(Minimum=0.2, Maximum=1, Evaluation Type=::NewEnumerator1)
    Recalculate Duration Each Loop=true
    UseLoopDelay=true
   )
   &SpawnRate(
    SpawnRate=&FN_FrontEndFloatModifier(Gameplay=&Lerp_Float(A=200, B=1100, Alpha=$User.ParaGlideLeanAlpha), Front End=300)
    Use Spawn Probability=false
   ) {module=SpawnRate002}
   &SpawnPerUnit(
    Module.Use Max Movement Threshold=true
    Module.Use Movement Tolerance=true
    Movement Tolerance=0.1
    Spawn Spacing:NiagaraFloat=&Divide_Float(A=1, B=&Divide_Float(A=1, B=30))
    Use Spawn Probability=false
   )
  ]
  particleSpawnScript [
   &FN_SetSpawnGroup()
   &InitializeParticle(
    Color=&MergeHelper_ColorSwitch(
     0=<rgba 0.1559 0.7913 1 0.1>
     1=<rgba 0.4342 1 0.1714 0.1>
     2=<rgba 0.9823 0.1413 1 0.1>
     3=<rgba 0.6514 0.6514 0.6514 0.1>
     Index=$Particles.SpawnGroup
     Num=::NewEnumerator2
    )
    Lifetime=&RandomRangeFloat(Minimum=0.18, Maximum=0.28, Evaluation Type=::NewEnumerator1)
    Sprite Rotation Angle=&RandomRangeFloat(Minimum=0, Maximum=1, Evaluation Type=::NewEnumerator1)
    Sprite Rotation Mode=::NewEnumerator1
    Sprite Size=&RandomRangeVector2D(Minimum=<33 43>, Maximum=<48 68>, Evaluation Type=::NewEnumerator1)
    Sprite Size Mode=::NewEnumerator4
    UsePositionOffset=false
    Write Material Random=true
   )
   &AddVectorToPosition(Vector:Vector3f=&TransformVector(Vector=<0 0 0>, Destination Space=::Simulation))
   &SkeletalMeshLocation(
    Bone / Socket Index=&SequentialInt(Use Array Data:NiagaraBool=false, First Number:NiagaraInt32=0, Sequence:NiagaraDataInterfaceArrayInt32=?, Last Number:NiagaraInt32=3)
    Bone Sampling Mode=::NewEnumerator4
    MaskSpawnGroup=false
    Orientation Sampling=::NewEnumerator1
    Sampled Position Offset=<0 0 -10>
    Sampled Position Offset Coordinate Space:ENiagaraSkelMeshTransforms=::NewEnumerator2
    Skeletal Mesh=$System.Skeletal Mesh Reader
   ) {module=SkeletalMeshLocation001}
   &AddVelocity(Velocity=&FN_FrontEndVector3Modifier(Gameplay=&Lerp_Vector(A=<0 0 0>, B=<0 0 0>, Alpha=$User.ParaGlideLeanAlpha), Front End=<0 -150 800>))
   &ApplyInitialForces()
  ]
  particleUpdateScript [
   &ParticleState()
   &ScaleSpriteSize(
    Scale Factor=&Vector2DFromFloat(Value=&FN_FrontEndFloatModifier(
      Gameplay=&FloatFromCurve(FloatCurve=DataInterface() {propertyValues {Curve {keys [
           {arriveTangent=0.9609, interpMode=RCIM_Cubic, leaveTangent=0.9609, tangentMode=RCTM_User, time=-0.008, value=0.3889}
           {arriveTangent=1.4688, interpMode=RCIM_Cubic, leaveTangent=1.4688, tangentMode=RCTM_User, time=0.3212, value=0.754}
           <at 0.8494 1.422>
          ]}}})
      Front End=&FloatFromCurve(FloatCurve=DataInterface() {propertyValues {Curve {keys [
           {arriveTangent=0.1776, interpMode=RCIM_Cubic, leaveTangent=0.1776, tangentMode=RCTM_User, time=-0.008, value=0.3889}
           {arriveTangent=0.3955, interpMode=RCIM_Cubic, leaveTangent=0.3955, tangentMode=RCTM_User, time=0.3323, value=0.6676}
           <at 0.8438 0.8358>
          ]}}})
     ))
    Scale Sprite Size Mode=::NewEnumerator2
   )
   &ScaleColor(
    Scale Alpha=&FN_FrontEndFloatModifier(
     Gameplay=&FloatFromCurve(FloatCurve=DataInterface() {propertyValues {Curve {
         defaultValue=1
         keys [{interpMode=RCIM_Cubic, tangentMode=RCTM_User, time=0, value=0}, {interpMode=RCIM_Cubic, tangentMode=RCTM_User, time=0.3, value=1}]
        }}})
     Front End=&FloatFromCurve(FloatCurve=DataInterface() {propertyValues {Curve {keys [<at 0 0>,<at 0.2 0.15>,<at 1 0>]}}})
    )
    ScaleA=true
    ScaleRGB=true
   )
   &CameraOffset(Camera Offset Amount=15)
   &DynamicMaterialParameters(
    Index 0 Param 1=0.5
    Index 0 Param 2=1
    Index 0 Param 3=1
    Index 0 Param 4=1
    Module.Index0Float0Write=true
    Module.Index0Float1Write=true
    Module.Index0Float2Write=true
    Module.Index0Float3Write=true
   )
   &SubUVAnimation(
    End Frame Range Override=49
    "Sprite Renderer "=DataInterface() {propertyValues {SpriteRenderer=&self:ThinSmokeLeftFoot_0.NiagaraSpriteRendererProperties_0}}
    Start Frame Range Override=10
    UseEndFrame=false
    UseStartFrame=false
   )
   &SolveForcesAndVelocity(Clamp Velocity=false, Limit Acceleration=false)
  ]
  rendererClasses [&NiagaraSpriteRendererProperties]
  renderers [{rendererClass=&NiagaraSpriteRendererProperties}]
 }
}
summary {
 systemName=NS_Contrails_SprayPaint_Master_Converted
 emitters [
  {emitterName=ThinSmokeLeftFoot, rendererClasses [&NiagaraSpriteRendererProperties]}
  {emitterName=Ribbon_Left_Foot, rendererClasses [&NiagaraRibbonRendererProperties]}
 ]
 userVariables [
  User.bIsFrontEnd {type=NiagaraBool}
  User.bIsFrontEndPreview {type=NiagaraBool}
  User.ParaGlideLeanAlpha {type=NiagaraFloat, defaultValue=0, description=""}
 ]
}
