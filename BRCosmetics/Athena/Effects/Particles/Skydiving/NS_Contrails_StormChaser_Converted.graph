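# Niagara skydiving contrail for the StormChaser cosmetic: repeating magenta-purple lightning strikes, body electricity and cone meshes around the player.
# Lightning bolts loop on a shared system interval, with impact and body-spark bursts timed by a user delay.
# Impact, sparks and cones follow spine or thigh bones; FN_Contrail_Setup dive value and glide lean shift offsets.
# Cascade-converted emitters merge layers through spawn-group switches, with body lightning split into High/Epic and Medium/Low variants.
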NiagaraSystem /BRCosmetics/Athena/Effects/Particles/Skydiving/NS_Contrails_StormChaser_Converted
dependencies {
 usedDataInterfaces [
  &NiagaraDataInterfaceSkeletalMesh
  &NiagaraDataInterfaceCurve
  &NiagaraDataInterfaceSocketReader
  &NiagaraDataInterfaceVectorCurve
  &NiagaraDataInterfaceSpriteRendererInfo
  &NiagaraDataInterfaceVector2DCurve
 ]
 usedDynamicInputs [&Lerp_Float]
 usedModules [
  &self:SystemSpawnScript.SystemScriptSource.SystemScriptGraph.NiagaraNodeAssignment_0.SetVariables_#1
  &SystemState
  &self:SystemSpawnScript.SystemScriptSource.SystemScriptGraph.NiagaraNodeAssignment_1.SetVariables_#2
  /Game/Effects/Niagara/Modules/Contrails/FN_Contrail_Setup
  &FN_Integrator
 ]
}
emitters {
 BodyLightningHighEpic {
  emitterUpdateScript [
   &EmitterState(
    Inactive Response=::NewEnumerator1
    Life Cycle Mode=::NewEnumerator1
    Loop Duration=$System.Lightnings Spawn
    Recalculate Duration Each Loop=true
    UseLoopDelay=false
   )
   &SpawnBurst_Instantaneous(Spawn Time=$User.Lightning_Impact_Delay, Use Spawn Probability=false, UseLoopCountLimit=false)
  ]
  particleSpawnScript [
   &InitializeParticle(
    Lifetime=&RandomRangeFloat(Minimum=0.85, Maximum=0.4, Evaluation Type=::NewEnumerator1)
    Sprite Rotation Angle=&RandomRangeFloat(Minimum=0, Maximum=1, Evaluation Type=::NewEnumerator1)
    Sprite Rotation Mode=::NewEnumerator1
    Sprite Size=&Vector2DFromFloat(Value=&RandomRangeFloat(Minimum=25, Maximum=40, Evaluation Type=::NewEnumerator1))
    Sprite Size Mode=::NewEnumerator4
    UsePositionOffset=false
    Write Material Random=true
   )
   &Color(Color=&MakeLinearColorFromVectorAndFloat("Vector (RGB)"=<20 0.2 25>, "Float (Alpha)"=1))
   &DynamicMaterialParameters(
    Index 0 Param 1=&RandomRangeFloat(Minimum=0, Maximum=1, Evaluation Type=::NewEnumerator1)
    Module.Param0WriteEnabled=true
    Module.Param1WriteEnabled=false
    Module.Param2WriteEnabled=false
    Module.Param3WriteEnabled=false
   )
   &AddVelocity(Velocity=&RandomRangeVector(Minimum=<0 0 0>, Maximum=<0 0 8>, Evaluation Type=::NewEnumerator1))
  ]
  particleUpdateScript [
   /SetVariables(Particles.LightExponent=16) {isSetParameters=true}
   &ParticleState()
   &ScaleSpriteSize(
    Scale Factor=&Vector2DFromFloat(Value=&FloatFromCurve(FloatCurve=DataInterface() {propertyValues {Curve {keys [
          {arriveTangent=0.9013, time=-0.0087, value=0.2033}
          {arriveTangent=-0.3413, leaveTangent=-0.3413, time=0.0437, value=1.1045}
          {arriveTangent=0.6499, leaveTangent=0.6499, time=0.1055, value=0.7632}
          {arriveTangent=-1.0159, leaveTangent=-1.0159, time=0.1723, value=1.4132}
          {arriveTangent=0.9301, leaveTangent=0.9301, time=0.2402, value=0.3973}
          {arriveTangent=-0.321, leaveTangent=-0.321, time=0.3722, value=1.3274}
          {arriveTangent=0.7115, leaveTangent=0.7115, time=0.4148, value=1.0064}
          {arriveTangent=-0.652, leaveTangent=-0.652, time=0.7199, value=1.7179}
          {arriveTangent=0.7017, leaveTangent=0.7017, time=0.7502, value=1.0659}
          <at 0.8751 1.7676>
         ]}}}))
    Scale Sprite Size Mode=::NewEnumerator2
   )
   &ScaleColor(
    Scale Alpha=&FloatFromCurve(FloatCurve=DataInterface() {propertyValues {Curve {keys [{arriveTangent=0.1913, time=-0.003, value=0.7986}, {arriveTangent=-0.9899, leaveTangent=-0.9899, time=0.6502, value=0.9899}, <at 1 0>]}}})
    Scale RGB=&VectorFromFloat(Value=&FloatFromCurve(FloatCurve=DataInterface() {propertyValues {Curve {keys [
          {arriveTangent=-0.7308, time=-0.0028, value=0.9692}
          {arriveTangent=0.7, leaveTangent=0.7, time=0.0444, value=0.2384}
          {arriveTangent=-0.7308, leaveTangent=-0.7308, time=0.1336, value=0.9384}
          {arriveTangent=1.2854, leaveTangent=1.2854, time=0.2789, value=0.2076}
          {arriveTangent=-1.3556, leaveTangent=-1.3556, time=0.4093, value=1.493}
          {arriveTangent=0.8178, leaveTangent=0.8178, time=0.5032, value=0.1373}
          {arriveTangent=-0.7783, leaveTangent=-0.7783, time=0.7182, value=0.9551}
          <at 0.9963 0.1768>
         ]}}}))
    ScaleA=true
    ScaleRGB=true
   )
   &DynamicMaterialParameters(
    Index 0 Param 2=1
    Index 0 Param 3=1
    Index 0 Param 4=1
    Module.Param0WriteEnabled=false
    Module.Param1WriteEnabled=true
    Module.Param2WriteEnabled=true
    Module.Param3WriteEnabled=true
   ) {module=DynamicMaterialParameters001}
   &SkeletalMeshLocation(
    Bone / Socket Index=/CascadeToNiagaraConverter/NiagaraScripts/DynamicInputScripts/SequentialInt(Use Array Data:NiagaraBool=false, First Number:NiagaraInt32=0, Sequence:NiagaraDataInterfaceArrayInt32=?, Last Number:NiagaraInt32=0)
    Bone Sampling Mode=::NewEnumerator4
    MaskSpawnGroup=false
    Orientation Sampling=::NewEnumerator1
    Sampled Position Offset=<0 -10 0>
    Skeletal Mesh=DataInterface() {propertyValues {
      DefaultMesh=null
      ExcludeBoneName=root
      OptionalComponentClass=null
      PreviewMesh=null
      SkinningMode=SkinOnTheFly
      SoftSourceActor=null
      SourceMode=Default
      UvSetIndex=0
      WholeMeshLOD=-1
      bExcludeBone=false
      bReadDeformedGeometry=true
      bRequireCurrentFrameData=true
      ComponentTags []
      FilteredBones [spine_03]
      FilteredSockets []
      MeshUserParameter {parameter {name=None, typeDefHandle {registeredTypeIndex=1}, varData []}}
      SamplingRegions []
     }}
   )
   &SolveForcesAndVelocity(Clamp Velocity=false, Limit Acceleration=false)
   &CascadeConversion_LightProperties(LightBrightnessScale=3, LightColorScale=<1 1 1>, LightRadiusScale=20)
  ]
  rendererClasses [&NiagaraSpriteRendererProperties, &NiagaraLightRendererProperties]
  renderers [
   {rendererClass=&NiagaraSpriteRendererProperties}
   {rendererClass=&NiagaraLightRendererProperties, rendererIndex=1}
  ]
 }
 BodyLightningMediumLow {
  emitterUpdateScript [
   &EmitterState(Inactive Response=::NewEnumerator1, Life Cycle Mode=::NewEnumerator1, Loop Duration=$System.Lightnings Spawn, UseLoopDelay=true)
   &SpawnBurst_Instantaneous(Spawn Count=6, Spawn Time=$User.Lightning_Impact_Delay, Use Spawn Probability=false, UseLoopCountLimit=false)
  ]
  particleSpawnScript [
   &InitializeParticle(
    Lifetime=&RandomRangeFloat(Minimum=0.85, Maximum=0.25, Evaluation Type=::NewEnumerator1)
    Sprite Rotation Angle=&RandomRangeFloat(Minimum=0, Maximum=1, Evaluation Type=::NewEnumerator1)
    Sprite Rotation Mode=::NewEnumerator1
    Sprite Size=&Vector2DFromFloat(Value=&RandomRangeFloat(Minimum=25, Maximum=40, Evaluation Type=::NewEnumerator1))
    Sprite Size Mode=::NewEnumerator4
    UsePositionOffset=false
    Write Material Random=true
   )
   &Color(Color=&MakeLinearColorFromVectorAndFloat("Vector (RGB)"=<20 0.2 25>, "Float (Alpha)"=1))
   &ShapeLocation(Hemisphere Distribution=<1 0.5>, MaskSpawnGroup=false, Override Seed=false, Sphere Radius=5, UseOwnerScale=true)
   &DynamicMaterialParameters(
    Index 0 Param 1=&RandomRangeFloat(Minimum=0, Maximum=1, Evaluation Type=::NewEnumerator1)
    Module.Param0WriteEnabled=true
    Module.Param1WriteEnabled=false
    Module.Param2WriteEnabled=false
    Module.Param3WriteEnabled=false
   )
   &AddVelocity(Velocity=&RandomRangeVector(Minimum=<0 0 0>, Maximum=<0 -30 15>, Evaluation Type=::NewEnumerator1))
  ]
  particleUpdateScript [
   &ParticleState()
   &ScaleSpriteSize(
    Scale Factor=&Vector2DFromFloat(Value=&FloatFromCurve(FloatCurve=DataInterface() {propertyValues {Curve {keys [
          {arriveTangent=0.9013, time=-0.0087, value=0.2033}
          {arriveTangent=-0.3413, leaveTangent=-0.3413, time=0.0437, value=1.1045}
          {arriveTangent=0.6499, leaveTangent=0.6499, time=0.1055, value=0.7632}
          {arriveTangent=-1.0159, leaveTangent=-1.0159, time=0.1723, value=1.4132}
          {arriveTangent=0.9301, leaveTangent=0.9301, time=0.2402, value=0.3973}
          {arriveTangent=-0.321, leaveTangent=-0.321, time=0.3722, value=1.3274}
          {arriveTangent=0.7115, leaveTangent=0.7115, time=0.4148, value=1.0064}
          {arriveTangent=-0.652, leaveTangent=-0.652, time=0.7199, value=1.7179}
          {arriveTangent=0.7017, leaveTangent=0.7017, time=0.7502, value=1.0659}
          <at 0.8751 1.7676>
         ]}}}))
    Scale Sprite Size Mode=::NewEnumerator2
   )
   &ScaleColor(
    Scale Alpha=&FloatFromCurve(FloatCurve=DataInterface() {propertyValues {Curve {keys [{arriveTangent=0.1913, time=-0.003, value=0.7986}, {arriveTangent=-0.9899, leaveTangent=-0.9899, time=0.6502, value=0.9899}, <at 1 0>]}}})
    Scale RGB=&VectorFromFloat(Value=&FloatFromCurve(FloatCurve=DataInterface() {propertyValues {Curve {keys [
          {arriveTangent=-0.7308, time=-0.0028, value=0.9692}
          {arriveTangent=0.7, leaveTangent=0.7, time=0.0444, value=0.2384}
          {arriveTangent=-0.7308, leaveTangent=-0.7308, time=0.1336, value=0.9384}
          {arriveTangent=1.2854, leaveTangent=1.2854, time=0.2789, value=0.2076}
          {arriveTangent=-1.3556, leaveTangent=-1.3556, time=0.4093, value=1.493}
          {arriveTangent=0.8178, leaveTangent=0.8178, time=0.5032, value=0.1373}
          {arriveTangent=-0.7783, leaveTangent=-0.7783, time=0.7182, value=0.9551}
          <at 0.9963 0.1768>
         ]}}}))
    ScaleA=true
    ScaleRGB=true
   )
   &DynamicMaterialParameters(
    Index 0 Param 2=1
    Index 0 Param 3=1
    Index 0 Param 4=1
    Module.Param0WriteEnabled=false
    Module.Param1WriteEnabled=true
    Module.Param2WriteEnabled=true
    Module.Param3WriteEnabled=true
   ) {module=DynamicMaterialParameters001}
   &SkeletalMeshLocation(
    Bone / Socket Index=&SequentialInt(Use Array Data:NiagaraBool=false, First Number:NiagaraInt32=0, Sequence:NiagaraDataInterfaceArrayInt32=?, Last Number:NiagaraInt32=5)
    Bone Sampling Mode=::NewEnumerator4
    MaskSpawnGroup=false
    Orientation Sampling=::NewEnumerator1
    Sampled Position Offset=<0 -10 0>
    Skeletal Mesh=DataInterface() {propertyValues {
      DefaultMesh=null
      ExcludeBoneName=root
      OptionalComponentClass=null
      PreviewMesh=null
      SkinningMode=SkinOnTheFly
      SoftSourceActor=null
      SourceMode=Default
      UvSetIndex=0
      WholeMeshLOD=-1
      bExcludeBone=false
      bReadDeformedGeometry=true
      bRequireCurrentFrameData=true
      ComponentTags []
      FilteredBones [spine_01,spine_02,spine_03,spine_05,thigh_l,thigh_r]
      FilteredSockets []
      MeshUserParameter {parameter {name=None, typeDefHandle {registeredTypeIndex=1}, varData []}}
      SamplingRegions []
     }}
   )
   &SolveForcesAndVelocity(Clamp Velocity=false, Limit Acceleration=false)
  ]
  rendererClasses [&NiagaraSpriteRendererProperties]
  renderers [{rendererClass=&NiagaraSpriteRendererProperties}]
 }
 Cones {
  emitterUpdateScript [
   &EmitterState(Life Cycle Mode=::NewEnumerator1, Loop Behavior=::NewEnumerator1, Loop Duration=1, UseLoopDelay=false)
   &SpawnBurst_Instantaneous(Use Spawn Probability=false, UseLoopCountLimit=false)
   &SpawnBurst_Instantaneous(Spawn Group=1, Use Spawn Probability=false, UseLoopCountLimit=false) {module=SpawnBurst_Instantaneous001}
  ]
  particleSpawnScript [
   /SetVariables(Particles.VisibilityTag=&SpawnGroup()) {isSetParameters=true}
   &InitializeParticle(
    Mesh Scale=&MergeHelper_VectorSwitch(0=<1 1 1>, 1=<0.6 0.6 0.6>, 2=?, 3=?, Index=&SpawnGroup(), Num=::NewEnumerator0)
    Mesh Scale Mode=::NewEnumerator4
    UsePositionOffset=false
    Write Material Random=true
   )
   &Color(Color=&MakeLinearColorFromVectorAndFloat(
     "Vector (RGB)"=&MergeHelper_VectorSwitch(0=<1 0.3 1>, 1=<1 1 1>, 2=?, 3=?, Index=&SpawnGroup(), Num=::NewEnumerator0)
     "Float (Alpha)"=&MergeHelper_FloatSwitch(0=1.5, 1=1.05, 2=?, 3=?, Index=&SpawnGroup(), Num=::NewEnumerator0)
    ))
  ]
  particleUpdateScript [
   &ParticleState(Kill Particles When Lifetime Has Elapsed=false)
   &SkeletalMeshLocation(
    Bone / Socket Index=&SequentialInt(Use Array Data:NiagaraBool=false, First Number:NiagaraInt32=0, Sequence:NiagaraDataInterfaceArrayInt32=?, Last Number:NiagaraInt32=0)
    Bone Sampling Mode=::NewEnumerator4
    MaskSpawnGroup=false
    Orientation Sampling=::NewEnumerator1
    Sampled Position Offset=&MergeHelper_VectorSwitch(
     0=&FN_FrontEndVector3Modifier(Gameplay=&Lerp_Vector(A=<-10 -25 0>, B=<-10 -10 0>, Alpha=$User.ParaGlideLeanAlpha), Front End=<0 -20 0>)
     1=<10 0 0>
     2=?
     3=?
     Index=&SpawnGroup()
     Num=::NewEnumerator0
    )
    Skeletal Mesh=DataInterface() {propertyValues {
      DefaultMesh=null
      ExcludeBoneName=root
      OptionalComponentClass=null
      PreviewMesh=null
      SkinningMode=SkinOnTheFly
      SoftSourceActor=null
      SourceMode=Default
      UvSetIndex=0
      WholeMeshLOD=-1
      bExcludeBone=false
      bReadDeformedGeometry=true
      bRequireCurrentFrameData=true
      ComponentTags []
      FilteredBones [spine_05]
      FilteredSockets []
      MeshUserParameter {parameter {name=None, typeDefHandle {registeredTypeIndex=1}, varData []}}
      SamplingRegions []
     }}
   )
  ]
  rendererClasses [&NiagaraMeshRendererProperties]
  renderers [
   {rendererClass=&NiagaraMeshRendererProperties, rendererIndex=0}
   {rendererClass=&NiagaraMeshRendererProperties, rendererIndex=1}
  ]
 }
 Lightning_Impact {
  emitterUpdateScript [
   &EmitterState(
    Inactive Response=::NewEnumerator1
    Life Cycle Mode=::NewEnumerator1
    Loop Duration=$System.Lightnings Spawn
    Recalculate Duration Each Loop=true
    UseLoopDelay=true
   )
   &SpawnBurst_Instantaneous(Spawn Count=3, Spawn Group=1, Spawn Time=$User.Lightning_Impact_Delay, Use Spawn Probability=false, UseLoopCountLimit=false) {module=SpawnBurst_Instantaneous001}
   &SpawnBurst_Instantaneous(Spawn Count=3, Spawn Group=2, Spawn Time=$User.Lightning_Impact_Delay, Use Spawn Probability=false, UseLoopCountLimit=false) {module=SpawnBurst_Instantaneous002}
   &SpawnBurst_Instantaneous(Spawn Group=4, Spawn Time=$User.Lightning_Impact_Delay, Use Spawn Probability=false, UseLoopCountLimit=false) {module=SpawnBurst_Instantaneous004}
   &SpawnBurst_Instantaneous(Spawn Group=5, Spawn Time=$User.Lightning_Impact_Delay, Use Spawn Probability=false, UseLoopCountLimit=false) {module=SpawnBurst_Instantaneous005}
  ]
  particleSpawnScript [
   /SetVariables(Particles.VisibilityTag=&SpawnGroup()) {isSetParameters=true, module=SetVariables_#3}
   /SetVariables(Particles.MeshRotationEuler:Vector3f=&MaskVector(
     Mask Source=0
     Mask Parameter=?
     Mask=5
     Masked Vector=<0 -0.25 0>
     Passthrough Vector=&RandomRangeVector(Minimum=<0 1 1>, Maximum=<1 1 1>, Evaluation Type=::NewEnumerator1)
    )) {isSetParameters=true, module=SetVariables_#4}
   /SetVariables(Particles.MeshOrientation=/CascadeToNiagaraConverter/NiagaraScripts/DynamicInputScripts/MakeQuatFromEuler(Euler:Vector3f=$Particles.MeshRotationEuler, Scale:NiagaraFloat=360)) {isSetParameters=true, module=SetVariables_#5}
   &InitializeParticle(
    Lifetime=&MergeHelper_FloatSwitch(
     0=0.15
     1=&RandomRangeFloat(Minimum=0.07, Maximum=0.2, Evaluation Type=::NewEnumerator1)
     2=0.08
     3=0.18
     4=0.06
     5=0.18
     Index=&SpawnGroup()
     Num=::NewEnumerator4
    )
    Mesh Scale=&MergeHelper_VectorSwitch(
     0=<0.3 0.15 0.3>
     1=&RandomRangeVector(Minimum=<0.15 0.1 0.75>, Maximum=<0.25 0.22 0.9>, Evaluation Type=::NewEnumerator1)
     2=<0 0 0>
     3=<0 0 0>
     4=<0 0 0>
     5=&FN_FrontEndVector3Modifier(Gameplay=<0.32 0.32 0.32>, Front End=<0.2 0.2 0.2>)
     Index=&SpawnGroup()
     Num=::NewEnumerator4
    )
    Mesh Scale Mode=::NewEnumerator4
    Sprite Size=&MergeHelper_Vector2Switch(0=<0 0>, 1=<0 0>, 2=<40 40>, 3=<300 1000>, 4=<300 300>, Index=&SpawnGroup(), Num=::NewEnumerator3)
    Sprite Size Mode=::NewEnumerator4
    UsePositionOffset=false
    Write Material Random=true
   )
   &Color(Color=&MakeLinearColorFromVectorAndFloat(
     "Vector (RGB)"=&MergeHelper_VectorSwitch(
      0=<2 0.1 1>
      1=<30 0.1 16>
      2=<0 1 0.9614>
      3=<0.8 0.1 1>
      4=<1 0.08 0.7478>
      5=<0.6 0 1>
      Index=&SpawnGroup()
      Num=::NewEnumerator4
     )
     "Float (Alpha)"=&MergeHelper_FloatSwitch(0=1, 1=1, 2=1, 3=0.1, 4=0.3, 5=1, Index=&SpawnGroup(), Num=::NewEnumerator4)
    ))
   &DynamicMaterialParameters(
    Index 0 Param 3=&RandomRangeFloat(Minimum=0, Maximum=1, Evaluation Type=::NewEnumerator1)
    Index 0 Param 4=&RandomRangeFloat(Minimum=0, Maximum=1, Evaluation Type=::NewEnumerator1)
    Module.Param0WriteEnabled=false
    Module.Param1WriteEnabled=false
    Module.Param2WriteEnabled=true
    Module.Param3WriteEnabled=true
   ) {module=DynamicMaterialParameters001}
   &AddVelocity(Velocity=&MergeHelper_VectorSwitch(
     0=<0 0 -1>
     1=&RandomRangeVector(Minimum=<-5 -1 2>, Maximum=<5 1 1>, Evaluation Type=::NewEnumerator1)
     2=?
     3=?
     Index=&SpawnGroup()
     Num=::NewEnumerator0
    ))
   /SetVariables(Particles.DynamicMaterialParameter_Burst:Vector4f=&MakeVector4(
     X=&RandomRangeFloat(Minimum=0.1, Maximum=0.5, Evaluation Type=::NewEnumerator1)
     Y=0
     Z=&RandomRangeFloat(Minimum=0, Maximum=5, Evaluation Type=::NewEnumerator1)
     W=&RandomRangeFloat(Minimum=1, Maximum=5, Evaluation Type=::NewEnumerator1)
    )) {isSetParameters=true, module=SetVariables_#6}
  ]
  particleUpdateScript [
   &ParticleState()
   &SkeletalMeshLocation(
    Bone Sampling Mode=::NewEnumerator4
    MaskSpawnGroup=false
    Sampled Position Offset=&MergeHelper_VectorSwitch(
     0=<0 0 0>
     1=<0 -15 0>
     2=<0 -30 0>
     3=<0 0 0>
     4=<0 -20 0>
     5=<0 -5 0>
     Index=&SpawnGroup()
     Num=::NewEnumerator4
    )
    Skeletal Mesh=DataInterface() {propertyValues {
      DefaultMesh=null
      ExcludeBoneName=root
      OptionalComponentClass=null
      PreviewMesh=null
      SkinningMode=SkinOnTheFly
      SoftSourceActor=null
      SourceMode=Default
      UvSetIndex=0
      WholeMeshLOD=-1
      bExcludeBone=false
      bReadDeformedGeometry=true
      bRequireCurrentFrameData=true
      ComponentTags [None]
      FilteredBones [spine_02]
      FilteredSockets []
      MeshUserParameter {parameter {name=None, typeDefHandle {registeredTypeIndex=1}, varData []}}
      SamplingRegions []
     }}
   )
   &ScaleMeshSize(Scale Factor=&MaskVector(
     Mask Source=0
     Mask Parameter=?
     Mask=5
     Masked Vector=&VectorFromCurve(VectorCurve=DataInterface() {propertyValues {XCurve {keys [<at 0 0.25>,<at 1 1>]}, YCurve {keys [<at 0 1>,<at 1 1.5>]}, ZCurve {keys [<at 0 1>,<at 1 1.5>]}}})
     Passthrough Vector=&VectorFromCurve(VectorCurve=DataInterface() {propertyValues {
        XCurve {keys [<at 0.0043 0>,<at 0.5 2>,<at 1 2>]}
        YCurve {keys [<at 0.0043 1>,<at 0.5 1>,<at 1 1>]}
        ZCurve {keys [<at 0.0043 1>,<at 0.5 1>,<at 1 1>]}
       }})
    ))
   &DynamicMaterialParameters(
    Index 0 Param 1=&FloatFromCurve(FloatCurve=DataInterface() {propertyValues {Curve {keys [
         {arriveTangent=45.9263, interpMode=RCIM_Cubic, leaveTangent=45.9263, tangentMode=RCTM_User, time=0, value=0}
         {arriveTangent=3.4958, interpMode=RCIM_Cubic, leaveTangent=3.4958, time=0.1045, value=14.9615}
         <at 0.9993 3.4932>
        ]}}})
    Index 0 Param 2=&FloatFromCurve(FloatCurve=DataInterface() {propertyValues {Curve {keys [<at 0 1 RCIM_Cubic>,<at 0.8 0 RCIM_Cubic>,<at 1 1>]}}})
    Module.Param0WriteEnabled=true
    Module.Param1WriteEnabled=true
    Module.Param2WriteEnabled=false
    Module.Param3WriteEnabled=false
   )
   &ScaleColor(
    Scale Alpha=&MergeHelper_FloatSwitch(
     0=&FloatFromCurve(FloatCurve=DataInterface() {propertyValues {Curve {keys [
          <at 0.4 1 RCIM_Cubic>
          {arriveTangent=-6.9248, interpMode=RCIM_Cubic, leaveTangent=-6.9248, time=0.4927, value=1.1208}
          {arriveTangent=0.3732, interpMode=RCIM_Cubic, leaveTangent=0.3732, time=0.5399, value=0.0315}
          {arriveTangent=-0.022, interpMode=RCIM_Cubic, leaveTangent=-0.022, time=0.5887, value=1.1566}
          {arriveTangent=0.0713, interpMode=RCIM_Cubic, leaveTangent=0.0713, time=0.6342, value=0.0294}
          {arriveTangent=-0.1642, interpMode=RCIM_Cubic, leaveTangent=-0.1642, time=0.6679, value=1.1623}
          {arriveTangent=-4.6994, interpMode=RCIM_Cubic, leaveTangent=-4.6994, time=0.7083, value=0.0172}
          {arriveTangent=-0.0103, interpMode=RCIM_Cubic, leaveTangent=-0.0103, time=0.7605, value=0.727}
          {arriveTangent=-0.453, interpMode=RCIM_Cubic, leaveTangent=-0.453, time=0.8161, value=0.0161}
          {arriveTangent=0.0387, interpMode=RCIM_Cubic, leaveTangent=0.0387, time=0.8565, value=0.6835}
          {arriveTangent=-2.9465, interpMode=RCIM_Cubic, leaveTangent=-2.9465, time=0.9045, value=0.0196}
          {arriveTangent=-0.2048, interpMode=RCIM_Cubic, leaveTangent=-0.2048, time=0.9441, value=0.4254}
          <at 1 0>
         ]}}})
     1=&FloatFromCurve(FloatCurve=DataInterface() {propertyValues {Curve {keys [
          <at 0.4 1 RCIM_Cubic>
          {arriveTangent=-6.9248, interpMode=RCIM_Cubic, leaveTangent=-6.9248, time=0.4927, value=1.1208}
          {arriveTangent=0.3732, interpMode=RCIM_Cubic, leaveTangent=0.3732, time=0.5399, value=0.0315}
          {arriveTangent=-0.022, interpMode=RCIM_Cubic, leaveTangent=-0.022, time=0.5887, value=1.1566}
          {arriveTangent=0.0713, interpMode=RCIM_Cubic, leaveTangent=0.0713, time=0.6342, value=0.0294}
          {arriveTangent=-0.1642, interpMode=RCIM_Cubic, leaveTangent=-0.1642, time=0.6679, value=1.1623}
          {arriveTangent=-4.6994, interpMode=RCIM_Cubic, leaveTangent=-4.6994, time=0.7083, value=0.0172}
          {arriveTangent=-0.0103, interpMode=RCIM_Cubic, leaveTangent=-0.0103, time=0.7605, value=0.727}
          {arriveTangent=-0.453, interpMode=RCIM_Cubic, leaveTangent=-0.453, time=0.8161, value=0.0161}
          {arriveTangent=0.0387, interpMode=RCIM_Cubic, leaveTangent=0.0387, time=0.8565, value=0.6835}
          {arriveTangent=-2.9465, interpMode=RCIM_Cubic, leaveTangent=-2.9465, time=0.9045, value=0.0196}
          {arriveTangent=-0.2048, interpMode=RCIM_Cubic, leaveTangent=-0.2048, time=0.9441, value=0.4254}
          <at 1 0>
         ]}}})
     2=&FloatFromCurve(FloatCurve=DataInterface() {propertyValues {Curve {keys [
          <at 0.4 1 RCIM_Cubic>
          {arriveTangent=-6.9248, interpMode=RCIM_Cubic, leaveTangent=-6.9248, time=0.4927, value=1.1208}
          {arriveTangent=0.3732, interpMode=RCIM_Cubic, leaveTangent=0.3732, time=0.5399, value=0.0315}
          {arriveTangent=-0.022, interpMode=RCIM_Cubic, leaveTangent=-0.022, time=0.5887, value=1.1566}
          {arriveTangent=0.0713, interpMode=RCIM_Cubic, leaveTangent=0.0713, time=0.6342, value=0.0294}
          {arriveTangent=-0.1642, interpMode=RCIM_Cubic, leaveTangent=-0.1642, time=0.6679, value=1.1623}
          {arriveTangent=-4.6994, interpMode=RCIM_Cubic, leaveTangent=-4.6994, time=0.7083, value=0.0172}
          {arriveTangent=-0.0103, interpMode=RCIM_Cubic, leaveTangent=-0.0103, time=0.7605, value=0.727}
          {arriveTangent=-0.453, interpMode=RCIM_Cubic, leaveTangent=-0.453, time=0.8161, value=0.0161}
          {arriveTangent=0.0387, interpMode=RCIM_Cubic, leaveTangent=0.0387, time=0.8565, value=0.6835}
          {arriveTangent=-2.9465, interpMode=RCIM_Cubic, leaveTangent=-2.9465, time=0.9045, value=0.0196}
          {arriveTangent=-0.2048, interpMode=RCIM_Cubic, leaveTangent=-0.2048, time=0.9441, value=0.4254}
          <at 1 0>
         ]}}})
     3=&FloatFromCurve(FloatCurve=DataInterface() {propertyValues {Curve {keys [<at 0 1>,<at 0.6 1>,<at 1 0>]}}})
     4=&FloatFromCurve(FloatCurve=DataInterface() {propertyValues {Curve {keys [<at 0 1>,<at 0.6 1>,<at 1 0>]}}})
     5=&FloatFromCurve(FloatCurve=DataInterface() {propertyValues {Curve {keys [<at 0 0>,<at 0.2 1>,<at 1 0>]}}})
     Index=&SpawnGroup()
     Num=::NewEnumerator4
    )
    Scale RGB=&MaskVector(
     Mask Source=0
     Mask Parameter=?
     Mask=5
     Masked Vector=&VectorFromCurve(VectorCurve=DataInterface() {propertyValues {XCurve {keys [<at 0 1>,<at 1 0>]}, YCurve {keys [<at 0 1>,<at 1 0>]}, ZCurve {keys [<at 0 1>,<at 1 0>]}}})
     Passthrough Vector=<1 1 1>
    )
    ScaleA=true
    ScaleRGB=true
   )
   &SolveForcesAndVelocity(Clamp Velocity=false, Limit Acceleration=false)
   &ScaleSpriteSize(
    Scale Factor=&MergeHelper_Vector2Switch(
     0=<0 0>
     1=<0 0>
     2=<1 1>
     3=&Vector2DFromCurve(
      Vector2Curve=DataInterface() {propertyValues {XCurve {keys [<at 0.0014 1>,<at 0.6 1>]}, YCurve {keys [{arriveTangent=0.8935, time=0.0014, value=0.1065}, <at 0.6 1>]}}}
      ScaleCurve=<1 1>
     )
     4=<1 1>
     Index=&SpawnGroup()
     Num=::NewEnumerator3
    )
    Scale Sprite Size Mode=::NewEnumerator2
   )
   /SetVariables(
    Particles.DynamicMaterialParameter_Impact:Vector4f=&MakeVector4(
     X=&FloatFromCurve(FloatCurve=DataInterface() {propertyValues {Curve {keys [
          {arriveTangent=45.9263, interpMode=RCIM_Cubic, leaveTangent=45.9263, tangentMode=RCTM_User, time=0, value=0}
          {arriveTangent=0.6153, interpMode=RCIM_Cubic, leaveTangent=0.6153, time=0.079, value=1.1796}
          <at 0.9972 0.6136>
         ]}}})
     Y=&FloatFromCurve(FloatCurve=DataInterface() {propertyValues {Curve {keys [
          <at -0.0084 -0.0167 RCIM_Cubic>
          {arriveTangent=0.0914, interpMode=RCIM_Cubic, leaveTangent=0.0914, time=0.2322, value=1.5}
          <at 1.0028 0.0757>
         ]}}})
     Z=&MakeFloatFromVector4(Vector 4=$Particles.Initial.DynamicMaterialParameter, Channel=::NewEnumerator2)
     W=&MakeFloatFromVector4(Vector 4=$Particles.Initial.DynamicMaterialParameter, Channel=::NewEnumerator3)
    )
    Particles.Scale_Impact:Vector3f=&Multiply_Vector(
     A=$Particles.Initial.Scale
     B=&RandomRangeVector(Minimum=<0.6 0.6 0.6>, Maximum=<0.8 0.8 0.8>, Evaluation Type=::NewEnumerator1)
    )
   ) {isSetParameters=true, module=SetVariables_#7}
   /SetVariables(Particles.LightExponent=16) {isSetParameters=true, module=SetVariables_#8}
   &CascadeConversion_LightProperties(LightBrightnessScale=30, LightColorScale=<1 1 1>, LightRadiusScale=10)
   /SetVariables(Particles.SpriteAlignment_Glow:Vector3f=<0 0 1>) {isSetParameters=true, module=SetVariables_#9}
   /SetVariables(Particles.DynamicMaterialParameter_Burst:Vector4f=&MakeVector4(
     X=&MakeFloatFromVector4(Vector 4=$Particles.Initial.DynamicMaterialParameter_Burst, Channel=::NewEnumerator0)
     Y=&FloatFromCurve(FloatCurve=DataInterface() {propertyValues {Curve {keys [<at 0 0>,<at 1 1>]}}})
     Z=&MakeFloatFromVector4(Vector 4=$Particles.Initial.DynamicMaterialParameter_Burst, Channel=::NewEnumerator2)
     W=&MakeFloatFromVector4(Vector 4=$Particles.Initial.DynamicMaterialParameter_Burst, Channel=::NewEnumerator3)
    )) {isSetParameters=true, module=SetVariables_#10}
   /SetVariables(Particles.MeshOrientation_Impact:Quat4f=<0 1 0 1>) {isSetParameters=true}
  ]
  rendererClasses [
   &NiagaraMeshRendererProperties
   &NiagaraLightRendererProperties
   &NiagaraSpriteRendererProperties
  ]
  renderers [
   {rendererClass=&NiagaraMeshRendererProperties, rendererIndex=0}
   {rendererClass=&NiagaraLightRendererProperties, rendererIndex=1}
   {rendererClass=&NiagaraSpriteRendererProperties, rendererIndex=2}
   {rendererClass=&NiagaraMeshRendererProperties, rendererIndex=3}
  ]
 }
 Lightnings {
  emitterUpdateScript [
   &EmitterState(
    Inactive Response=::NewEnumerator1
    Life Cycle Mode=::NewEnumerator1
    Loop Duration=$System.Lightnings Spawn
    Recalculate Duration Each Loop=true
    UseLoopDelay=true
   )
   &SpawnBurst_Instantaneous(Use Spawn Probability=false, UseLoopCountLimit=false)
   &SpawnBurst_Instantaneous(Spawn Group=3, Use Spawn Probability=false, UseLoopCountLimit=false) {module=SpawnBurst_Instantaneous003}
  ]
  particleSpawnScript [
   /SetVariables(Particles.VisibilityTag=&SpawnGroup()) {isSetParameters=true, module=SetVariables_#3}
   /SetVariables(Particles.MeshRotationEuler:Vector3f=&MaskVector(
     Mask Source=0
     Mask Parameter=?
     Mask=5
     Masked Vector=<0 -0.25 0>
     Passthrough Vector=&RandomRangeVector(Minimum=<0 1 1>, Maximum=<1 1 1>, Evaluation Type=::NewEnumerator1)
    )) {isSetParameters=true, module=SetVariables_#4}
   /SetVariables(Particles.MeshOrientation=&MakeQuatFromEuler(Euler:Vector3f=$Particles.MeshRotationEuler, Scale:NiagaraFloat=360)) {isSetParameters=true, module=SetVariables_#5}
   &InitializeParticle(
    Lifetime=&MergeHelper_FloatSwitch(
     0=0.15
     1=&RandomRangeFloat(Minimum=0.07, Maximum=0.2, Evaluation Type=::NewEnumerator1)
     2=0.08
     3=0.18
     4=0.06
     5=0.18
     Index=&SpawnGroup()
     Num=::NewEnumerator4
    )
    Mesh Scale=&MergeHelper_VectorSwitch(
     0=<0.3 0.15 0.3>
     1=&RandomRangeVector(Minimum=<0.2 0.15 1>, Maximum=<0.3 0.25 1>, Evaluation Type=::NewEnumerator1)
     2=<0 0 0>
     3=<0 0 0>
     4=<0 0 0>
     5=<0.2 0.2 0.2>
     Index=&SpawnGroup()
     Num=::NewEnumerator4
    )
    Mesh Scale Mode=::NewEnumerator4
    Sprite Size=&MergeHelper_Vector2Switch(0=<0 0>, 1=<0 0>, 2=<40 40>, 3=<300 1000>, 4=<300 300>, Index=&SpawnGroup(), Num=::NewEnumerator3)
    Sprite Size Mode=::NewEnumerator4
    UsePositionOffset=false
    Write Material Random=true
   )
   &Color(Color=&MakeLinearColorFromVectorAndFloat(
     "Vector (RGB)"=&MergeHelper_VectorSwitch(
      0=<2 0.1 1>
      1=<30 0.1 16>
      2=<0 1 0.9614>
      3=<0.8 0.1 1>
      4=<1 0.08 0.7478>
      5=<0.6 0 1>
      Index=&SpawnGroup()
      Num=::NewEnumerator4
     )
     "Float (Alpha)"=&MergeHelper_FloatSwitch(0=1, 1=1, 2=1, 3=0.1, 4=0.3, 5=1, Index=&SpawnGroup(), Num=::NewEnumerator4)
    ))
   &DynamicMaterialParameters(
    Index 0 Param 3=&RandomRangeFloat(Minimum=0, Maximum=1, Evaluation Type=::NewEnumerator1)
    Index 0 Param 4=&RandomRangeFloat(Minimum=0, Maximum=1, Evaluation Type=::NewEnumerator1)
    Module.Param0WriteEnabled=false
    Module.Param1WriteEnabled=false
    Module.Param2WriteEnabled=true
    Module.Param3WriteEnabled=true
   ) {module=DynamicMaterialParameters001}
   /CascadeToNiagaraConverter/NiagaraScripts/ModuleScripts/AddVectorToPosition(Vector:Vector3f=&TransformVector(
     Vector=&MergeHelper_VectorSwitch(
      0=&Lerp_Vector(A=<0 0 600>, B=<0 0 650>, Alpha=$User.ParaGlideLeanAlpha)
      1=&Lerp_Vector(A=<0 0 90>, B=<0 0 0>, Alpha=$System.FN_Contrail_Setup.Dive Value)
      2=<0 0 120>
      3=<0 0 600>
      4=&Lerp_Vector(A=<0 0 90>, B=<0 0 0>, Alpha=$System.FN_Contrail_Setup.Dive Value)
      5=&Lerp_Vector(A=<0 0 90>, B=<0 0 0>, Alpha=$System.FN_Contrail_Setup.Dive Value)
      Index=&SpawnGroup()
      Num=::NewEnumerator4
     )
     Destination Space=::Simulation
    ))
   &AddVelocity(Velocity=&MergeHelper_VectorSwitch(
     0=<0 0 -1>
     1=&RandomRangeVector(Minimum=<-5 -1 2>, Maximum=<5 1 1>, Evaluation Type=::NewEnumerator1)
     2=?
     3=?
     Index=&SpawnGroup()
     Num=::NewEnumerator0
    ))
   /SetVariables(Particles.DynamicMaterialParameter_Burst:Vector4f=&MakeVector4(
     X=&RandomRangeFloat(Minimum=0.1, Maximum=0.5, Evaluation Type=::NewEnumerator1)
     Y=0
     Z=&RandomRangeFloat(Minimum=0, Maximum=5, Evaluation Type=::NewEnumerator1)
     W=&RandomRangeFloat(Minimum=1, Maximum=5, Evaluation Type=::NewEnumerator1)
    )) {isSetParameters=true, module=SetVariables_#6}
  ]
  particleUpdateScript [
   &ParticleState()
   &ScaleMeshSize(Scale Factor=&MaskVector(
     Mask Source=0
     Mask Parameter=?
     Mask=5
     Masked Vector=&VectorFromCurve(VectorCurve=DataInterface() {propertyValues {XCurve {keys [<at 0 0.25>,<at 1 1>]}, YCurve {keys [<at 0 1>,<at 1 1.5>]}, ZCurve {keys [<at 0 1>,<at 1 1.5>]}}})
     Passthrough Vector=&VectorFromCurve(VectorCurve=DataInterface() {propertyValues {
        XCurve {keys [<at 0.0043 0>,<at 0.5 2>,<at 1 2>]}
        YCurve {keys [<at 0.0043 1>,<at 0.5 1>,<at 1 1>]}
        ZCurve {keys [<at 0.0043 1>,<at 0.5 1>,<at 1 1>]}
       }})
    ))
   &DynamicMaterialParameters(
    Index 0 Param 1=&FloatFromCurve(FloatCurve=DataInterface() {propertyValues {Curve {keys [
         {arriveTangent=45.9263, interpMode=RCIM_Cubic, leaveTangent=45.9263, tangentMode=RCTM_User, time=0, value=0}
         {arriveTangent=3.4958, interpMode=RCIM_Cubic, leaveTangent=3.4958, time=0.1045, value=14.9615}
         <at 0.9993 3.4932>
        ]}}})
    Index 0 Param 2=&FloatFromCurve(FloatCurve=DataInterface() {propertyValues {Curve {keys [<at 0 1 RCIM_Cubic>,<at 0.8 0 RCIM_Cubic>,<at 1 1>]}}})
    Module.Param0WriteEnabled=true
    Module.Param1WriteEnabled=true
    Module.Param2WriteEnabled=false
    Module.Param3WriteEnabled=false
   )
   &ScaleColor(
    Scale Alpha=&MergeHelper_FloatSwitch(
     0=&FloatFromCurve(FloatCurve=DataInterface() {propertyValues {Curve {keys [
          <at 0.4 1 RCIM_Cubic>
          {arriveTangent=-6.9248, interpMode=RCIM_Cubic, leaveTangent=-6.9248, time=0.4927, value=1.1208}
          {arriveTangent=0.3732, interpMode=RCIM_Cubic, leaveTangent=0.3732, time=0.5399, value=0.0315}
          {arriveTangent=-0.022, interpMode=RCIM_Cubic, leaveTangent=-0.022, time=0.5887, value=1.1566}
          {arriveTangent=0.0713, interpMode=RCIM_Cubic, leaveTangent=0.0713, time=0.6342, value=0.0294}
          {arriveTangent=-0.1642, interpMode=RCIM_Cubic, leaveTangent=-0.1642, time=0.6679, value=1.1623}
          {arriveTangent=-4.6994, interpMode=RCIM_Cubic, leaveTangent=-4.6994, time=0.7083, value=0.0172}
          {arriveTangent=-0.0103, interpMode=RCIM_Cubic, leaveTangent=-0.0103, time=0.7605, value=0.727}
          {arriveTangent=-0.453, interpMode=RCIM_Cubic, leaveTangent=-0.453, time=0.8161, value=0.0161}
          {arriveTangent=0.0387, interpMode=RCIM_Cubic, leaveTangent=0.0387, time=0.8565, value=0.6835}
          {arriveTangent=-2.9465, interpMode=RCIM_Cubic, leaveTangent=-2.9465, time=0.9045, value=0.0196}
          {arriveTangent=-0.2048, interpMode=RCIM_Cubic, leaveTangent=-0.2048, time=0.9441, value=0.4254}
          <at 1 0>
         ]}}})
     1=&FloatFromCurve(FloatCurve=DataInterface() {propertyValues {Curve {keys [
          <at 0.4 1 RCIM_Cubic>
          {arriveTangent=-6.9248, interpMode=RCIM_Cubic, leaveTangent=-6.9248, time=0.4927, value=1.1208}
          {arriveTangent=0.3732, interpMode=RCIM_Cubic, leaveTangent=0.3732, time=0.5399, value=0.0315}
          {arriveTangent=-0.022, interpMode=RCIM_Cubic, leaveTangent=-0.022, time=0.5887, value=1.1566}
          {arriveTangent=0.0713, interpMode=RCIM_Cubic, leaveTangent=0.0713, time=0.6342, value=0.0294}
          {arriveTangent=-0.1642, interpMode=RCIM_Cubic, leaveTangent=-0.1642, time=0.6679, value=1.1623}
          {arriveTangent=-4.6994, interpMode=RCIM_Cubic, leaveTangent=-4.6994, time=0.7083, value=0.0172}
          {arriveTangent=-0.0103, interpMode=RCIM_Cubic, leaveTangent=-0.0103, time=0.7605, value=0.727}
          {arriveTangent=-0.453, interpMode=RCIM_Cubic, leaveTangent=-0.453, time=0.8161, value=0.0161}
          {arriveTangent=0.0387, interpMode=RCIM_Cubic, leaveTangent=0.0387, time=0.8565, value=0.6835}
          {arriveTangent=-2.9465, interpMode=RCIM_Cubic, leaveTangent=-2.9465, time=0.9045, value=0.0196}
          {arriveTangent=-0.2048, interpMode=RCIM_Cubic, leaveTangent=-0.2048, time=0.9441, value=0.4254}
          <at 1 0>
         ]}}})
     2=&FloatFromCurve(FloatCurve=DataInterface() {propertyValues {Curve {keys [
          <at 0.4 1 RCIM_Cubic>
          {arriveTangent=-6.9248, interpMode=RCIM_Cubic, leaveTangent=-6.9248, time=0.4927, value=1.1208}
          {arriveTangent=0.3732, interpMode=RCIM_Cubic, leaveTangent=0.3732, time=0.5399, value=0.0315}
          {arriveTangent=-0.022, interpMode=RCIM_Cubic, leaveTangent=-0.022, time=0.5887, value=1.1566}
          {arriveTangent=0.0713, interpMode=RCIM_Cubic, leaveTangent=0.0713, time=0.6342, value=0.0294}
          {arriveTangent=-0.1642, interpMode=RCIM_Cubic, leaveTangent=-0.1642, time=0.6679, value=1.1623}
          {arriveTangent=-4.6994, interpMode=RCIM_Cubic, leaveTangent=-4.6994, time=0.7083, value=0.0172}
          {arriveTangent=-0.0103, interpMode=RCIM_Cubic, leaveTangent=-0.0103, time=0.7605, value=0.727}
          {arriveTangent=-0.453, interpMode=RCIM_Cubic, leaveTangent=-0.453, time=0.8161, value=0.0161}
          {arriveTangent=0.0387, interpMode=RCIM_Cubic, leaveTangent=0.0387, time=0.8565, value=0.6835}
          {arriveTangent=-2.9465, interpMode=RCIM_Cubic, leaveTangent=-2.9465, time=0.9045, value=0.0196}
          {arriveTangent=-0.2048, interpMode=RCIM_Cubic, leaveTangent=-0.2048, time=0.9441, value=0.4254}
          <at 1 0>
         ]}}})
     3=&FloatFromCurve(FloatCurve=DataInterface() {propertyValues {Curve {keys [<at 0 1>,<at 0.6 1>,<at 1 0>]}}})
     4=&FloatFromCurve(FloatCurve=DataInterface() {propertyValues {Curve {keys [<at 0 1>,<at 0.6 1>,<at 1 0>]}}})
     5=&FloatFromCurve(FloatCurve=DataInterface() {propertyValues {Curve {keys [<at 0 0>,<at 0.2 1>,<at 1 0>]}}})
     Index=&SpawnGroup()
     Num=::NewEnumerator4
    )
    Scale RGB=&MaskVector(
     Mask Source=0
     Mask Parameter=?
     Mask=5
     Masked Vector=&VectorFromCurve(VectorCurve=DataInterface() {propertyValues {XCurve {keys [<at 0 1>,<at 1 0>]}, YCurve {keys [<at 0 1>,<at 1 0>]}, ZCurve {keys [<at 0 1>,<at 1 0>]}}})
     Passthrough Vector=<1 1 1>
    )
    ScaleA=true
    ScaleRGB=true
   )
   &SolveForcesAndVelocity(Clamp Velocity=false, Limit Acceleration=false)
   &ScaleSpriteSize(
    Scale Factor=&MergeHelper_Vector2Switch(
     0=<0 0>
     1=<0 0>
     2=<1 1>
     3=&Vector2DFromCurve(
      Vector2Curve=DataInterface() {propertyValues {XCurve {keys [<at 0.0014 1>,<at 0.6 1>]}, YCurve {keys [{arriveTangent=0.8935, time=0.0014, value=0.1065}, <at 0.6 1>]}}}
      ScaleCurve=<1 1>
     )
     4=<1 1>
     Index=&SpawnGroup()
     Num=::NewEnumerator3
    )
    Scale Sprite Size Mode=::NewEnumerator2
   )
   /SetVariables(
    Particles.DynamicMaterialParameter_Impact:Vector4f=&MakeVector4(
     X=&FloatFromCurve(FloatCurve=DataInterface() {propertyValues {Curve {keys [
          {arriveTangent=45.9263, interpMode=RCIM_Cubic, leaveTangent=45.9263, tangentMode=RCTM_User, time=0, value=0}
          {arriveTangent=0.6153, interpMode=RCIM_Cubic, leaveTangent=0.6153, time=0.079, value=1.1796}
          <at 0.9972 0.6136>
         ]}}})
     Y=&FloatFromCurve(FloatCurve=DataInterface() {propertyValues {Curve {keys [
          <at -0.0084 -0.0167 RCIM_Cubic>
          {arriveTangent=0.0914, interpMode=RCIM_Cubic, leaveTangent=0.0914, time=0.2322, value=1.5}
          <at 1.0028 0.0757>
         ]}}})
     Z=&MakeFloatFromVector4(Vector 4=$Particles.Initial.DynamicMaterialParameter, Channel=::NewEnumerator2)
     W=&MakeFloatFromVector4(Vector 4=$Particles.Initial.DynamicMaterialParameter, Channel=::NewEnumerator3)
    )
    Particles.Scale_Impact:Vector3f=&Multiply_Vector(
     A=$Particles.Initial.Scale
     B=&RandomRangeVector(Minimum=<0.6 0.6 0.6>, Maximum=<0.8 0.8 0.8>, Evaluation Type=::NewEnumerator1)
    )
   ) {isSetParameters=true, module=SetVariables_#7}
   /SetVariables(Particles.LightExponent=16) {isSetParameters=true, module=SetVariables_#8}
   &CascadeConversion_LightProperties(LightBrightnessScale=30, LightColorScale=<1 1 1>, LightRadiusScale=10)
   /SetVariables(Particles.SpriteAlignment_Glow:Vector3f=<0 0 1>) {isSetParameters=true, module=SetVariables_#9}
   /SetVariables(Particles.DynamicMaterialParameter_Burst:Vector4f=&MakeVector4(
     X=&MakeFloatFromVector4(Vector 4=$Particles.Initial.DynamicMaterialParameter_Burst, Channel=::NewEnumerator0)
     Y=&FloatFromCurve(FloatCurve=DataInterface() {propertyValues {Curve {keys [<at 0 0>,<at 1 1>]}}})
     Z=&MakeFloatFromVector4(Vector 4=$Particles.Initial.DynamicMaterialParameter_Burst, Channel=::NewEnumerator2)
     W=&MakeFloatFromVector4(Vector 4=$Particles.Initial.DynamicMaterialParameter_Burst, Channel=::NewEnumerator3)
    )) {isSetParameters=true, module=SetVariables_#10}
  ]
  rendererClasses [&NiagaraMeshRendererProperties, &NiagaraSpriteRendererProperties]
  renderers [
   {rendererClass=&NiagaraMeshRendererProperties, rendererIndex=0}
   {rendererClass=&NiagaraSpriteRendererProperties, rendererIndex=1}
  ]
 }
 SparklesHands {
  emitterUpdateScript [
   &EmitterState(Inactive Response=::NewEnumerator1, Life Cycle Mode=::NewEnumerator1, Loop Duration=1, UseLoopDelay=true)
   &SpawnRate(SpawnRate=&FN_FrontEndFloatModifier(Gameplay=0, Front End=23), Use Spawn Probability=false)
   &SpawnPerUnit(
    Max Movement Threshold=1500
    Module.Use Max Movement Threshold=false
    Module.Use Movement Tolerance=true
    Movement Tolerance=0.1
    Spawn Spacing:NiagaraFloat=&FN_FrontEndFloatModifier(Gameplay=75, Front End=0)
    Use Spawn Probability=false
   )
  ]
  particleSpawnScript [
   &PartitionParticles(Number Of Partitions=4)
   &InitializeParticle(
    Lifetime=&RandomRangeFloat(Minimum=0.15, Maximum=0.3, Evaluation Type=::NewEnumerator1)
    Sprite Size=&RandomRangeVector2D(Minimum=<2 8>, Maximum=<3 12>, Evaluation Type=::NewEnumerator1)
    Sprite Size Mode=::NewEnumerator4
    UsePositionOffset=false
    Write Material Random=true
   )
   &Color(Color=&MakeLinearColorFromVectorAndFloat("Vector (RGB)"=<20 2 40>, "Float (Alpha)"=0.75))
   &SocketLocation(
    SamplePosOffset=false
    Socket Index=$Particles.PartitionParticles.Partition
    Socket Reader=DataInterface() {propertyValues {
      AttachComponentClass=null
      AttachComponentTag=None
      EditorPreviewAsset=null
      SourceActorRef=null
      SourceAsset=null
      SourceMode=Default
      bRequireCurrentFrameData=true
      bUpdateSocketsPerFrame=true
      FilteredSockets [FX_Contrail_Hand_L, FX_Contrail_Foot_L, FX_Contrail_Hand_R, FX_Contrail_Foot_R]
      ObjectParameterBinding {parameter {name=None, typeDefHandle {registeredTypeIndex=1}, varData []}}
     }}
    Socket Sampling Mode=::NewEnumerator3
    VelocitySpeedLimit=false
   )
   &ShapeLocation(MaskSpawnGroup=false, Override Seed=false, Sphere Radius=8, UseOwnerScale=true)
   &AddVelocity(Velocity=&FN_FrontEndVector3Modifier(Gameplay=<0 -500 200>, Front End=<0 -300 200>))
   &InheritVelocity(Inherited Velocity Amount Scale=<-0.1 -0.1 -0.1>, Inherited Velocity Speed Limit=0, Speed Limit=false, UseSpeedThreshold=false)
  ]
  particleUpdateScript [
   &ParticleState()
   &ScaleSpriteSize(
    Scale Factor=&Vector2DFromCurve(
     Vector2Curve=DataInterface() {propertyValues {XCurve {keys [<at 0 1>,<at 0.2 0.5>,<at 1 0>]}, YCurve {keys [<at 0 0>,<at 0.2 10>,<at 1 0>]}}}
     ScaleCurve=<1 1>
    )
    Scale Sprite Size Mode=::NewEnumerator2
   )
   &ScaleColor(
    Scale Alpha=&FloatFromCurve(FloatCurve=DataInterface() {propertyValues {Curve {keys [<at 0 0>,<at 0.2 1>,<at 1 0>]}}})
    Scale RGB=&VectorFromCurve(
     VectorCurve=DataInterface() {propertyValues {XCurve {keys [<at 0 15>,<at 0.1 1>,<at 1 0>]}, YCurve {keys [<at 0 15>,<at 0.1 1>,<at 1 0>]}, ZCurve {keys [<at 0 15>,<at 0.1 1>,<at 1 0>]}}}
     Scale Curve=<1.3 1.3 1.3>
    )
    ScaleA=true
    ScaleRGB=true
   )
   &SolveForcesAndVelocity(Clamp Velocity=false, Limit Acceleration=false)
  ]
  rendererClasses [&NiagaraSpriteRendererProperties]
  renderers [{rendererClass=&NiagaraSpriteRendererProperties}]
 }
 smokies {
  emitterUpdateScript [
   &EmitterState(Inactive Response=::NewEnumerator1, Life Cycle Mode=::NewEnumerator1, Loop Duration=1, UseLoopDelay=true)
   &SpawnRate(SpawnRate=13, Use Spawn Probability=false)
  ]
  particleSpawnScript [
   /SetVariables(Particles.Acceleration:Vector3f=&RandomRangeVector(Minimum=<0 0 230>, Maximum=<0 0 100>, Evaluation Type=::NewEnumerator1)) {isSetParameters=true}
   &InitializeParticle(
    Lifetime=&RandomRangeFloat(Minimum=0.4, Maximum=0.9, Evaluation Type=::NewEnumerator1)
    Sprite Rotation Angle=&RandomRangeFloat(Minimum=0, Maximum=1, Evaluation Type=::NewEnumerator1)
    Sprite Rotation Mode=::NewEnumerator1
    Sprite Size=&Vector2DFromFloat(Value=&RandomRangeFloat(Minimum=32, Maximum=18, Evaluation Type=::NewEnumerator1))
    Sprite Size Mode=::NewEnumerator4
    UsePositionOffset=false
    Write Material Random=true
   )
   &SocketLocation(
    OverrideRandomSeed=false
    SamplePosOffset=false
    Socket Reader=DataInterface() {propertyValues {
      AttachComponentClass=null
      AttachComponentTag=None
      EditorPreviewAsset=null
      SourceActorRef=null
      SourceAsset=null
      SourceMode=Default
      bRequireCurrentFrameData=true
      bUpdateSocketsPerFrame=true
      FilteredSockets [hand_l,spine_01,hand_r,spine_02,spine_03,lowerarm_r,lowerarm_l]
      ObjectParameterBinding {parameter {name=None, typeDefHandle {registeredTypeIndex=1}, varData []}}
     }}
    Socket Sampling Mode=::NewEnumerator1
    VelocitySpeedLimit=false
   )
   &Color(Color=&MakeLinearColorFromVectorAndFloat("Vector (RGB)"=<0.8 0.06 1>, "Float (Alpha)"=1))
  ]
  particleUpdateScript [
   &ParticleState()
   &ScaleSpriteSize(
    Scale Sprite Size Mode=::NewEnumerator0
    Uniform Scale Factor=&FloatFromCurve(FloatCurve=DataInterface() {propertyValues {Curve {keys [<at 0 1>,<at 1 3>]}}})
   )
   &ScaleColor(
    Scale Alpha=&FloatFromCurve(FloatCurve=DataInterface() {propertyValues {Curve {keys [
         <at 0 0 RCIM_Cubic>
         {arriveTangent=-2.4229, interpMode=RCIM_Cubic, leaveTangent=-2.4229, tangentMode=RCTM_User, time=0.2986, value=0.9692}
         <at 1 0>
        ]}}})
    Scale RGB=&VectorFromCurve(VectorCurve=DataInterface() {propertyValues {
       XCurve {keys [<at 0 1>,<at 0.4 1>,<at 1 0.2>]}
       YCurve {keys [<at 0 1>,<at 0.4 1>,<at 1 0.2>]}
       ZCurve {keys [<at 0 1>,<at 0.4 1>,<at 1 0.2>]}
      }})
    ScaleA=true
    ScaleRGB=true
   )
   &AccelerationForce(Acceleration=$Particles.Acceleration)
   &Drag(Drag=&RandomRangeFloat(Minimum=0.2, Maximum=0.5), Use Linear Drag=true, Use Rotational Drag=false)
   &SolveForcesAndVelocity(Clamp Velocity=false, Limit Acceleration=false)
   &SubUVAnimation(
    End Frame Range Override=3
    Evaluation Type=::NewEnumerator0
    "Sprite Renderer "=DataInterface() {propertyValues {SpriteRenderer=&self:smokies_0.NiagaraSpriteRendererProperties_5}}
    SubUV Animation Mode:ENiagara_SubUVLookupModeV2=::NewEnumerator2
    UseEndFrame=false
    UseStartFrame=false
   )
  ]
  rendererClasses [&NiagaraSpriteRendererProperties]
  renderers [{rendererClass=&NiagaraSpriteRendererProperties}]
 }
}
summary {
 systemName=NS_Contrails_StormChaser_Converted
 emitters [
  {emitterName=Cones, rendererClasses [&NiagaraMeshRendererProperties]}
  {
   emitterName=BodyLightningHighEpic
   rendererClasses [&NiagaraSpriteRendererProperties, &NiagaraLightRendererProperties]
  }
  {emitterName=BodyLightningMediumLow, rendererClasses [&NiagaraSpriteRendererProperties]}
  {emitterName=smokies, rendererClasses [&NiagaraSpriteRendererProperties]}
  {emitterName=Lightnings, rendererClasses [&NiagaraMeshRendererProperties, &NiagaraSpriteRendererProperties]}
  {emitterName=SparklesHands, rendererClasses [&NiagaraSpriteRendererProperties]}
  {
   emitterName=Lightning_Impact
   rendererClasses [
    &NiagaraMeshRendererProperties
    &NiagaraLightRendererProperties
    &NiagaraSpriteRendererProperties
   ]
  }
 ]
 userVariables [
  User.bIsFrontEnd {type=NiagaraBool}
  User.bIsFrontEndPreview {type=NiagaraBool}
  User.Lightning_Impact_Delay {type=NiagaraFloat, defaultValue=0.18, description=""}
  User.ParaGlideLeanAlpha {type=NiagaraFloat, defaultValue=0, description=""}
  User.RibbonIDOffset {type=NiagaraInt32}
 ]
}
