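# Viking skydiving contrail cosmetic that attaches blue glows, ribbons, pattern meshes and sparks to character bones.
# A shared system skeletal mesh reader places glow sprites and ribbons on four bones and curve-animated pattern meshes on the first two.
# GPU sparks spawn by distance moved from hand bones through their own mesh sampler and accelerate upward.
# Front-end modifiers give the ribbons velocity in menus, with none in gameplay.
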NiagaraSystem /BRCosmetics/Athena/Effects/Particles/Skydiving/NS_Contrails_Viking
dependencies {
 usedDataInterfaces [
  &NiagaraDataInterfaceArrayFloat3
  &NiagaraDataInterfaceCurve
  &NiagaraDataInterfaceVectorCurve
  &NiagaraDataInterfaceSkeletalMesh
  &NiagaraDataInterfaceVector2DCurve
 ]
 usedDynamicInputs []
 usedModules [
  &self:SystemSpawnScript.SystemScriptSource.SystemScriptGraph.NiagaraNodeAssignment_0.SetVariables_#1
  &SystemState
 ]
}
emitters {
 CirclePatternL {
  emitterUpdateScript [
   &EmitterState(Inactive Response=::NewEnumerator1, Life Cycle Mode=::NewEnumerator1, Loop Duration=1, UseLoopDelay=true)
   &SpawnRate(Use Spawn Probability=false)
   &SpawnRate(SpawnGroup=1, Use Spawn Probability=false) {module=SpawnRate001}
  ]
  particleSpawnScript [
   /SetVariables(Particles.SkeletonSamplingIndex:NiagaraInt32=/CascadeToNiagaraConverter/NiagaraScripts/DynamicInputScripts/SequentialInt(Use Array Data:NiagaraBool=false, First Number:NiagaraInt32=0, Sequence:NiagaraDataInterfaceArrayInt32=?, Last Number:NiagaraInt32=0)) {enabled=false, isSetParameters=true}
   &InitializeParticle(
    Lifetime=&RandomRangeFloat(Minimum=0.98, Maximum=1.4, Evaluation Type=::NewEnumerator1)
    Mesh Scale=/Niagara/DynamicInputs/CascadeConversion/SplitVector(
     Double Component Vector Channels:ECascadeNiagaraTwoVectorChannels=::NewEnumerator0
     Double Component:NiagaraFloat=0.25
     Single Component:NiagaraFloat=0.3
    )
    Mesh Scale Mode=::NewEnumerator4
    Sprite Rotation Angle=&RandomRangeFloat(Minimum=0, Maximum=1, Evaluation Type=::NewEnumerator1)
    Sprite Rotation Mode=::NewEnumerator1
    UsePositionOffset=false
    Write Material Random=true
   )
   &Color(Color=&MakeLinearColorFromVectorAndFloat("Vector (RGB)"=<1.1556 2.0161 2.5>, "Float (Alpha)"=3))
   &InitialMeshOrientation(Mesh Axis To Orient=<-1 0 0>, Mesh Orientation Mode=::NewEnumerator2, Orientation Vector=<0 0 1>, Use Rotation Vector=false) {module=InitialMeshOrientation001}
  ]
  particleUpdateScript [
   &ParticleState()
   &ScaleMeshSize(Scale Factor=&SplitVector(
     Double Component Vector Channels:ECascadeNiagaraTwoVectorChannels=::NewEnumerator0
     Double Component:NiagaraFloat=&FloatFromCurve(FloatCurve=DataInterface() {propertyValues {Curve {keys [{arriveTangent=0.3877, time=0, value=0.7}, <at 0.9959 1.0877>]}}})
     Single Component:NiagaraFloat=&FloatFromCurve(FloatCurve=DataInterface() {propertyValues {Curve {keys [{arriveTangent=-0.5, time=0, value=1}, <at 0.9959 0.5>]}}})
    ))
   &ScaleColor(
    Scale Alpha=&FloatFromCurve(FloatCurve=DataInterface() {propertyValues {Curve {keys [
         {arriveTangent=-0.2141, interpMode=RCIM_Cubic, leaveTangent=-0.2141, tangentMode=RCTM_User, time=-0.0038, value=0.0024}
         {arriveTangent=1.0075, interpMode=RCIM_Cubic, leaveTangent=-0.1412, tangentMode=RCTM_Break, time=0.5158, value=0.6557}
         {arriveTangent=-1.3543, interpMode=RCIM_Cubic, leaveTangent=-1.3543, time=0.7346, value=0.324}
         <at 1 0>
        ]}}})
    Scale RGB=&VectorFromCurve(VectorCurve=DataInterface() {propertyValues {
       XCurve {keys [{arriveTangent=-0.5045, time=-0.0015, value=0.5045}, <at 0.5 0>]}
       YCurve {keys [{arriveTangent=-1.1583, time=-0.0015, value=1.3209}, <at 0.5 0.1626>]}
       ZCurve {keys [{arriveTangent=-4, time=-0.0015, value=5}, <at 0.5 1>]}
      }})
    ScaleA=true
    ScaleRGB=true
   )
   &SkeletalMeshLocation(
    Bone / Socket Index=&SpawnGroup()
    Bone Sampling Mode=::NewEnumerator4
    MaskSpawnGroup=false
    Orientation Sampling=::NewEnumerator1
    Sampled Position Offset=&SelectVectorFromArray(
     Direct Array Index=&SpawnGroup()
     Vector Selection Array=DataInterface() {propertyValues {FloatData [<10 0 0>, <-10 0 0>]}}
    )
    Skeletal Mesh=$System.Skeletal Mesh Reader
   )
   &DynamicMaterialParameters(
    Index 0 Param 1=&FloatFromCurve(
     FloatCurve=DataInterface() {propertyValues {Curve {keys [{arriveTangent=-9.9922, interpMode=RCIM_Cubic, leaveTangent=-9.9922, tangentMode=RCTM_User, time=0, value=10}, <at 1 -10>]}}}
     Scale Curve=0.1
    )
    Index 0 Param 2=&FloatFromCurve(
     FloatCurve=DataInterface() {propertyValues {Curve {keys [
         {arriveTangent=139.4427, interpMode=RCIM_Cubic, leaveTangent=139.4427, tangentMode=RCTM_User, time=0, value=-16.7425}
         {arriveTangent=66.7952, interpMode=RCIM_Cubic, leaveTangent=66.7952, time=0.5, value=30}
         <at 1.001 50.1214>
        ]}}}
     Scale Curve=0.3
    )
    Index 0 Param 3=1
    Index 0 Param 4=1
    Module.Param0WriteEnabled=true
    Module.Param1WriteEnabled=true
    Module.Param2WriteEnabled=true
    Module.Param3WriteEnabled=true
   ) {module=DynamicMaterialParameters002}
   &DynamicMaterialParameters(
    Index 0 Param 1=&FloatFromCurve(
     FloatCurve=DataInterface() {propertyValues {Curve {keys [{arriveTangent=-9.9922, interpMode=RCIM_Cubic, leaveTangent=-9.9922, tangentMode=RCTM_User, time=0, value=10}, <at 1 -10>]}}}
     Scale Curve=0
    )
    Index 0 Param 2=&FloatFromCurve(FloatCurve=DataInterface() {propertyValues {Curve {keys [
         {interpMode=RCIM_Cubic, tangentMode=RCTM_User, time=0, value=-1.25}
         {arriveTangent=33.2876, interpMode=RCIM_Cubic, leaveTangent=33.2876, tangentMode=RCTM_User, time=0.4, value=8.75}
         {interpMode=RCIM_Cubic, tangentMode=RCTM_User, time=1, value=15}
        ]}}})
    Index 0 Param 3=1
    Index 0 Param 4=1
    Module.Param0WriteEnabled=true
    Module.Param1WriteEnabled=true
    Module.Param2WriteEnabled=true
    Module.Param3WriteEnabled=true
   ) {enabled=false}
   &DynamicMaterialParameters(
    Index 0 Param 2=&FloatFromCurve(FloatCurve=DataInterface() {propertyValues {Curve {keys [
         {arriveTangent=139.4427, interpMode=RCIM_Cubic, leaveTangent=139.4427, tangentMode=RCTM_User, time=0, value=-16.7425}
         {arriveTangent=66.7952, interpMode=RCIM_Cubic, leaveTangent=66.7952, time=0.4129, value=36.1549}
         <at 1.001 50.1214>
        ]}}})
    Index 0 Param 3=1
    Index 0 Param 4=1
    Module.Index0Float0Write=true
    Module.Index0Float1Write=true
    Module.Index0Float2Write=true
    Module.Index0Float3Write=true
   ) {enabled=false, module=DynamicMaterialParameters001}
  ]
  rendererClasses [&NiagaraMeshRendererProperties]
  renderers [{rendererClass=&NiagaraMeshRendererProperties, rendererIndex=0}]
 }
 HandGlowHands {
  emitterUpdateScript [
   &EmitterState(Inactive Response=::NewEnumerator1, Life Cycle Mode=::NewEnumerator1, Loop Duration=1, UseLoopDelay=true)
   &SpawnRate(SpawnRate=64, Use Spawn Probability=false)
  ]
  particleSpawnScript [
   /SetVariables(Particles.SkeletonSamplingIndex:NiagaraInt32=&SequentialInt(Use Array Data:NiagaraBool=false, First Number:NiagaraInt32=0, Sequence:NiagaraDataInterfaceArrayInt32=?, Last Number:NiagaraInt32=3)) {isSetParameters=true}
   &InitializeParticle(
    Lifetime=0.1
    Sprite Rotation Angle=-0.25
    Sprite Rotation Mode=::NewEnumerator1
    Sprite Size=<65 65>
    Sprite Size Mode=::NewEnumerator4
    UsePositionOffset=false
    Write Material Random=true
   )
   &Color(Color=&MakeLinearColorFromVectorAndFloat("Vector (RGB)"=<0 3.3396 50>, "Float (Alpha)"=0.5))
   &SkeletalMeshLocation(
    Bone / Socket Index=$Particles.SkeletonSamplingIndex
    Bone Sampling Mode=::NewEnumerator4
    MaskSpawnGroup=false
    Orientation Sampling=::NewEnumerator1
    Sampled Position Coordinate Space=::Local
    Sampled Position Offset=&SelectVectorFromArray(
     Direct Array Index=$Particles.SkeletonSamplingIndex
     Vector Selection Array=DataInterface() {propertyValues {FloatData [<10 0 0>, <-10 0 0>, <0 0 0>, <0 0 0>]}}
    )
    Skeletal Mesh=$System.Skeletal Mesh Reader
   )
  ]
  particleUpdateScript [
   &ParticleState()
   &ScaleSpriteSize(
    Scale Sprite Size Mode=::NewEnumerator0
    Uniform Scale Factor=&FloatFromCurve(FloatCurve=DataInterface() {propertyValues {Curve {keys [<at 0 0.6>,<at 1 1.3>]}}})
   )
   &CameraOffset(Camera Offset Amount=-15)
   &ScaleColor(
    Scale Alpha=&FloatFromCurve(FloatCurve=DataInterface() {propertyValues {Curve {keys [
         <at 0 0 RCIM_Cubic>
         {arriveTangent=-0.5737, interpMode=RCIM_Cubic, leaveTangent=-0.5737, tangentMode=RCTM_User, time=0.2321, value=0.9177}
         <at 1 0>
        ]}}})
    Scale RGB=&VectorFromCurve(VectorCurve=DataInterface() {propertyValues {XCurve {keys [<at 0 1>,<at 1 2>]}, YCurve {keys [<at 0 1>,<at 1 2>]}, ZCurve {keys [<at 0 1>,<at 1 2>]}}})
    ScaleA=true
    ScaleRGB=true
   )
  ]
  rendererClasses [&NiagaraSpriteRendererProperties]
  renderers [{rendererClass=&NiagaraSpriteRendererProperties}]
 }
 Ribbon_Right_Foot {
  emitterUpdateScript [
   &EmitterState(Inactive Response=::NewEnumerator1, Life Cycle Mode=::NewEnumerator1, Loop Duration=1, UseLoopDelay=true)
   &SpawnRate(SpawnRate=60, Use Spawn Probability=false)
   &SpawnRate(SpawnGroup=1, SpawnRate=60, Use Spawn Probability=false) {module=SpawnRate001}
   &SpawnRate(SpawnGroup=2, SpawnRate=60, Use Spawn Probability=false) {module=SpawnRate002}
   &SpawnRate(SpawnGroup=3, SpawnRate=60, Use Spawn Probability=false) {module=SpawnRate003}
  ]
  particleSpawnScript [
   /SetVariables(Particles.Acceleration:Vector3f=&RandomRangeVector(Minimum=<-3 -3 0>, Maximum=<3 3 0>, Evaluation Type=::NewEnumerator1)) {isSetParameters=true}
   &InitializeParticle(
    Lifetime=&MergeHelper_FloatSwitch(0=0.18, 1=0.18, 2=0.12, 3=0.12, Index=&SpawnGroup(), Num=::NewEnumerator2)
    Ribbon Width=&Multiply_Float(
     A=&MergeHelper_FloatSwitch(0=11, 1=11, 2=12, 3=12, Index=&SpawnGroup(), Num=::NewEnumerator2)
     B=2
    )
    Ribbon Width Mode=::NewEnumerator1
    UsePositionOffset=false
    Write Material Random=true
   )
   &Color(Color=&MakeLinearColorFromVectorAndFloat("Vector (RGB)"=<0.7838 3.0888 5>, "Float (Alpha)"=0.3))
   &SkeletalMeshLocation(
    Bone / Socket Index=&SpawnGroup()
    Bone Sampling Mode=::NewEnumerator4
    MaskSpawnGroup=false
    Orientation Sampling=::NewEnumerator1
    Sampled Position Offset=&MergeHelper_VectorSwitch(0=<10 0 0>, 1=<-10 0 0>, 2=<0 0 0>, 3=<0 0 0>, Index=&SpawnGroup(), Num=::NewEnumerator2)
    Skeletal Mesh=$System.Skeletal Mesh Reader
   )
   /CascadeToNiagaraConverter/NiagaraScripts/ModuleScripts/AddVectorToPosition(Vector:Vector3f=&TransformVector(Vector=<0 0 -4>, Destination Space=::Simulation))
   &AddVelocity(Velocity=&MakeVector(X=0, Y=0, Z=&FN_FrontEndFloatModifier(Gameplay=0, Front End=500))) {enabled=false, module=AddVelocity001}
   &AddVelocity(Velocity=&FN_FrontEndVector3Modifier(Gameplay=<0 0 0>, Front End=<0 -300 1200>))
  ]
  particleUpdateScript [
   &ParticleState()
   &ScaleColor(
    Scale Alpha=&FloatFromCurve(FloatCurve=DataInterface() {propertyValues {Curve {keys [
         {arriveTangent=-0.242, interpMode=RCIM_Cubic, leaveTangent=-0.242, tangentMode=RCTM_User, time=0.0008, value=0.9708}
         {arriveTangent=-1, leaveTangent=-1, time=0.5, value=1}
         <at 1 0>
        ]}}})
    Scale RGB=&VectorFromCurve(VectorCurve=DataInterface() {propertyValues {
       XCurve {keys [<at 0 0.1214>, {arriveTangent=-1, leaveTangent=-1, time=0.06, value=0.5312}, <at 0.6 0>]}
       YCurve {keys [<at 0 0.6886>, {arriveTangent=-0.9399, leaveTangent=-0.9399, time=0.06, value=0.712}, <at 0.6 0.0601>]}
       ZCurve {keys [<at 0 1.5>, {arriveTangent=-0.6302, leaveTangent=-0.6302, time=0.06, value=1}, <at 0.6 0.3698>]}
      }})
    ScaleA=true
    ScaleRGB=true
   )
   &AccelerationForce(Acceleration=$Particles.Acceleration)
   &SolveForcesAndVelocity(Clamp Velocity=false, Limit Acceleration=false)
   &DynamicMaterialParameters(
    Index 0 Param 1=&MergeHelper_FloatSwitch(0=0, 1=0.25, 2=0, 3=0.25, Index=&SpawnGroup(), Num=::NewEnumerator2)
    Index 0 Param 3=1
    Index 0 Param 4=0.75
    Module.Param0WriteEnabled=true
    Module.Param1WriteEnabled=true
    Module.Param2WriteEnabled=true
    Module.Param3WriteEnabled=true
   )
  ]
  rendererClasses [&NiagaraRibbonRendererProperties]
  renderers [{rendererClass=&NiagaraRibbonRendererProperties}]
 }
 SparksHands {
  simTarget=GPUComputeSim
  emitterUpdateScript [
   &EmitterState(Inactive Response=::NewEnumerator1, Life Cycle Mode=::NewEnumerator1, Loop Delay=0.1, Loop Duration=1, UseLoopDelay=true)
   &SpawnPerUnit(
    Max Movement Threshold=1500
    Module.Use Max Movement Threshold=true
    Module.Use Movement Tolerance=true
    Movement Tolerance=0.1
    Spawn Spacing:NiagaraFloat=&Divide_Float(A=1, B=&Divide_Float(A=1, B=62))
    Use Spawn Probability=false
   )
  ]
  particleSpawnScript [
   &InitializeParticle(
    Lifetime=&RandomRangeFloat(Minimum=0.15, Maximum=0.5, Evaluation Type=::NewEnumerator1)
    Sprite Size=&RandomRangeVector2D(Minimum=<0.5 7.5>, Maximum=<1.5 20>, Evaluation Type=::NewEnumerator1)
    Sprite Size Mode=::NewEnumerator4
    UsePositionOffset=false
    Write Material Random=true
   )
   &AddVelocity(
    Cone Angle=&RandomRangeFloat(Minimum=0, Maximum=0.2, Evaluation Type=::NewEnumerator1)
    Cone Angle Mode:ENiagara_AngleInput=::NewEnumerator1
    Cone Axis=<0 0 1>
    Module.UseConeFalloff=false
    Module.UseVelDistribution=false
    Override Seed=false
    Velocity Mode=::NewEnumerator2
    Velocity Speed=&RandomRangeFloat(Minimum=80, Maximum=400, Evaluation Type=::NewEnumerator1)
   )
   &Color(Color=&MakeLinearColorFromVectorAndFloat("Vector (RGB)"=<0.0398 0.5437 1>, "Float (Alpha)"=5))
   &SkeletalMeshLocation(
    Bone / Socket Index=&SequentialInt(Use Array Data:NiagaraBool=false, First Number:NiagaraInt32=0, Sequence:NiagaraDataInterfaceArrayInt32=?, Last Number:NiagaraInt32=5)
    Bone Sampling Mode=::NewEnumerator4
    MaskSpawnGroup=false
    Orientation Sampling=::NewEnumerator1
    Skeletal Mesh=DataInterface() {propertyValues {
      DefaultMesh=null
      ExcludeBoneName=root
      OptionalComponentClass=null
      PreviewMesh=&M_Med_Soldier_04_Preview
      SkinningMode=SkinOnTheFly
      SoftSourceActor=null
      SourceMode=Default
      UvSetIndex=0
      WholeMeshLOD=-1
      bExcludeBone=false
      bReadDeformedGeometry=true
      bRequireCurrentFrameData=true
      ComponentTags []
      FilteredBones [middle_01_r,middle_01_l,thumb_03_l,thumb_03_r,pinky_03_l,pinky_03_r]
      FilteredSockets []
      MeshUserParameter {parameter {name=None, typeDefHandle {registeredTypeIndex=1}, varData []}}
      SamplingRegions []
     }}
   )
  ]
  particleUpdateScript [
   &ParticleState()
   &ScaleSpriteSize(
    Scale Factor=&Vector2DFromCurve(
     Vector2Curve=DataInterface() {propertyValues {
       XCurve {keys [{arriveTangent=-0.0316, time=0, value=0.6559}, <at 0.5953 0.6243>]}
       YCurve {keys [{arriveTangent=-0.7726, time=0, value=1}, <at 0.5953 0.2274>]}
      }}
     ScaleCurve=<1 1>
    )
    Scale Sprite Size Mode=::NewEnumerator2
   )
   &AccelerationForce(Acceleration=<0 0 500>)
   &Drag(Drag=1.5, Use Linear Drag=true, Use Rotational Drag=true)
   &SolveForcesAndVelocity(Clamp Velocity=false, Limit Acceleration=false)
   &ScaleColor(
    Scale Alpha=&FloatFromCurve(FloatCurve=DataInterface() {propertyValues {Curve {keys [<at 0 1>,<at 1 0>]}}})
    ScaleA=true
    ScaleRGB=true
   )
  ]
  rendererClasses [&NiagaraSpriteRendererProperties]
  renderers [{rendererClass=&NiagaraSpriteRendererProperties}]
 }
}
summary {
 systemName=NS_Contrails_Viking
 emitters [
  {emitterName=HandGlowHands, rendererClasses [&NiagaraSpriteRendererProperties]}
  {emitterName=CirclePatternL, rendererClasses [&NiagaraMeshRendererProperties]}
  {emitterName=Ribbon_Right_Foot, rendererClasses [&NiagaraRibbonRendererProperties]}
  {emitterName=SparksHands, simTarget=GPUComputeSim, rendererClasses [&NiagaraSpriteRendererProperties]}
 ]
 userVariables [
  User.bIsFrontEnd {type=NiagaraBool}
  User.bIsFrontEndPreview {type=NiagaraBool}
 ]
}
