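# Niagara skydiving contrail cosmetic with a disco-ball theme, converted from Cascade.
# Emitters layer orbiting disco-ball meshes, ribbons, coronas, twinkling stars, GPU confetti, color-cycling glows and a pulsing light.
# Most attach to bones or sockets of a user skeletal mesh, several choosing the index by spawn group.
# Front-end modifiers add drift for menu previews, while ParaGlideLeanAlpha blends ribbon offset and velocity.
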NiagaraSystem /BRCosmetics/Athena/Effects/Particles/Skydiving/P_Contrail_DiscoBall_Converted
dependencies {
 usedDataInterfaces [&NiagaraDataInterfaceCurve, &NiagaraDataInterfaceVectorCurve, &NiagaraDataInterfaceArrayFloat]
 usedDynamicInputs []
 usedModules [&SystemState]
}
emitters {
 Confetti_Center {
  simTarget=GPUComputeSim
  emitterUpdateScript [
   &EmitterState(Life Cycle Mode=::NewEnumerator1, Loop Duration=0.5, UseLoopDelay=true)
   &SpawnRate(SpawnRate=12, Use Spawn Probability=false)
  ]
  particleSpawnScript [
   &FN_SetSpawnGroup(Number of Spawn Groups=2)
   &InitializeParticle(
    Lifetime=&RandomRangeFloat(Minimum=0.3, Maximum=0.6)
    Position=&ConvertVectorToPosition(
     Input Position=&RandomRangeVector(Minimum=<-90 -90 0>, Maximum=<90 90 0>)
     Conversion Operation:ENiagara_LWCConvertVecToPos=::NewEnumerator0
    )
    Position Mode=::NewEnumerator1
    Sprite Rotation Angle=&RandomRangeFloat(Minimum=0, Maximum=1)
    Sprite Rotation Mode=::NewEnumerator1
    Sprite Size=&Vector2DFromFloat(Value=&RandomRangeFloat(Minimum=3.2, Maximum=7))
    Sprite Size Mode=::NewEnumerator4
    UsePositionOffset=false
    Write Material Random=true
   )
   &Color(Color=&MakeLinearColorFromVectorAndFloat("Vector (RGB)"=<0 1 0>, "Float (Alpha)"=&ClampFloat(Float=1.4)))
   &InheritVelocity(
    Inherited Velocity Amount Scale=&RandomRangeVector(Minimum=<-0.05 -0.08 -0.12>, Maximum=<-0.1 -0.12 -0.3>)
    Inherited Velocity Speed Limit=0
    Speed Limit=false
    UseSpeedThreshold=false
   )
   &SkeletalMeshLocation(
    Bone / Socket Index=&Add_Integer(A=$Particles.SpawnGroup, B=10)
    Bone Sampling Mode=::NewEnumerator4
    MaskSpawnGroup=false
    Skeletal Mesh=$User.New Skeletal Mesh
   )
   &AddVelocity(
    Velocity=&RandomRangeVector(Minimum=<0 -150 150>, Maximum=<0 -300 300>)
    Velocity Speed Scale=&FN_FrontEndFloatModifier(Gameplay=0, Front End=1)
   ) {module=AddVelocity002}
   &ShapeLocation(MaskSpawnGroup=false, Override Seed=false, Sphere Radius=7, UseOwnerScale=true)
   &AddVelocity(Velocity=&Multiply_VectorByFloat(Vector=$Particles.ShapeLocation.ShapeVector, Float=&RandomRangeFloat(Minimum=1, Maximum=5)))
   &AddVelocity(Velocity=<0 0 -80>) {module=AddVelocity001}
   &ApplyInitialForces()
  ]
  particleUpdateScript [
   &ParticleState()
   &ScaleSpriteSize()
   &ScaleColor(
    Scale RGB=&VectorFromCurve(VectorCurve=DataInterface() {propertyValues {
       XCurve {keys [<at 0 0.3364>,<at 0.4 0.4664>,<at 1 0.6781>]}
       YCurve {keys [<at 0 0.4742>,<at 0.4 0.6562>,<at 1 0.4158>]}
       ZCurve {keys [<at 0 1>,<at 0.4 1>,<at 1 1>]}
      }})
    ScaleA=true
    ScaleRGB=true
   )
   &Drag(Drag=3, Use Linear Drag=true, Use Rotational Drag=true)
   &AccelerationForce(Acceleration=<0 0 -100>)
   &SolveForcesAndVelocity(Clamp Velocity=false, Limit Acceleration=false)
   /Niagara/Modules/CascadeConversion/CascadeConversion_Orbit(
    Offset:Vector3f=&RandomRangeVector(Minimum=<-16 -14 0>, Maximum=<15 18 0>)
    Rotation:Vector3f=&RandomRangeVector(Minimum=<0 0 0>, Maximum=<1 1 1>)
    Rotation Rate:Vector3f=&RandomRangeVector(Minimum=<0 0 0>, Maximum=<1 1 1>)
   )
   /Niagara/Modules/CascadeConversion/CascadeConversion_SolveOrbit()
  ]
  rendererClasses [&NiagaraSpriteRendererProperties]
  renderers [{rendererClass=&NiagaraSpriteRendererProperties}]
 }
 Confetti_g {
  simTarget=GPUComputeSim
  emitterUpdateScript [
   &EmitterState(Life Cycle Mode=::NewEnumerator1, Loop Duration=0.1, UseLoopDelay=true)
   &SpawnRate(SpawnRate=15, Use Spawn Probability=false)
  ]
  particleSpawnScript [
   &FN_SetSpawnGroup(Number of Spawn Groups=2)
   &InitializeParticle(
    Lifetime=&RandomRangeFloat(Minimum=0.3, Maximum=0.75)
    Position=&ConvertVectorToPosition(
     Input Position=&RandomRangeVector(Minimum=<-200 -200 20>, Maximum=<200 200 60>)
     Conversion Operation:ENiagara_LWCConvertVecToPos=::NewEnumerator0
    )
    Position Mode=::NewEnumerator1
    Sprite Rotation Angle=&RandomRangeFloat(Minimum=0, Maximum=1)
    Sprite Rotation Mode=::NewEnumerator1
    Sprite Size=&Vector2DFromFloat(Value=&RandomRangeFloat(Minimum=3.6, Maximum=7.5))
    Sprite Size Mode=::NewEnumerator4
    UsePositionOffset=false
    Write Material Random=true
   )
   &Color(Color=&MakeLinearColorFromVectorAndFloat("Vector (RGB)"=<0 1 0>, "Float (Alpha)"=&ClampFloat(Float=1.5)))
   &InheritVelocity(Inherited Velocity Amount Scale=<-0.12 -0.12 -0.2>, Inherited Velocity Speed Limit=0, Speed Limit=false, UseSpeedThreshold=false)
   &SkeletalMeshLocation(
    Bone / Socket Index=&Add_Integer(A=$Particles.SpawnGroup, B=10)
    Bone Sampling Mode=::NewEnumerator4
    MaskSpawnGroup=false
    Skeletal Mesh=$User.New Skeletal Mesh
   )
   &AddVelocity(
    Velocity=&RandomRangeVector(Minimum=<0 -150 150>, Maximum=<0 -300 300>)
    Velocity Speed Scale=&FN_FrontEndFloatModifier(Gameplay=0, Front End=1)
   ) {module=AddVelocity002}
   &ShapeLocation(MaskSpawnGroup=false, Override Seed=false, Sphere Radius=20, UseOwnerScale=true)
   &AddVelocity(Velocity=&Multiply_VectorByFloat(Vector=$Particles.ShapeLocation.ShapeVector, Float=&RandomRangeFloat(Minimum=1, Maximum=5)))
   &AddVelocity(Velocity=<0 0 -80>) {module=AddVelocity001}
   &ApplyInitialForces()
  ]
  particleUpdateScript [
   &ParticleState()
   &ScaleSpriteSize()
   &ScaleColor(
    Scale RGB=&VectorFromCurve(VectorCurve=DataInterface() {propertyValues {
       XCurve {keys [<at 0 0.3364>,<at 0.4 0.4664>,<at 1 0.6781>]}
       YCurve {keys [<at 0 0.4742>,<at 0.4 0.6562>,<at 1 0.4158>]}
       ZCurve {keys [<at 0 1>,<at 0.4 1>,<at 1 1>]}
      }})
    ScaleA=true
    ScaleRGB=true
   )
   &Drag(Drag=0.335, Use Linear Drag=true, Use Rotational Drag=true)
   &AccelerationForce(Acceleration=<0 0 -100>)
   &SolveForcesAndVelocity(Clamp Velocity=false, Limit Acceleration=false)
   &CascadeConversion_Orbit(
    Offset:Vector3f=&RandomRangeVector(Minimum=<-9 -11 0>, Maximum=<10 10 0>)
    Rotation:Vector3f=&RandomRangeVector(Minimum=<0 0 0>, Maximum=<1 1 1>)
    Rotation Rate:Vector3f=&RandomRangeVector(Minimum=<0 0 0>, Maximum=<1 1 1>)
   )
   &CascadeConversion_SolveOrbit()
  ]
  rendererClasses [&NiagaraSpriteRendererProperties]
  renderers [{rendererClass=&NiagaraSpriteRendererProperties}]
 }
 Corona {
  emitterUpdateScript [
   &EmitterState(Life Cycle Mode=::NewEnumerator1, Loop Duration=1.4, UseLoopDelay=true)
   &SpawnBurst_Instantaneous(Use Spawn Probability=false, UseLoopCountLimit=false)
   &SpawnRate(SpawnRate=50, Use Spawn Probability=false)
  ]
  particleSpawnScript [
   &FN_SetSpawnGroup()
   /SetVariables(Particles.SpriteRotationRate=&RandomRangeFloat(Minimum=-0.1, Maximum=0.1)) {isSetParameters=true}
   &InitializeParticle(
    Lifetime=&RandomRangeFloat(Minimum=0.6, Maximum=1)
    Sprite Rotation Angle=&RandomRangeFloat(Minimum=0, Maximum=1)
    Sprite Rotation Mode=::NewEnumerator1
    Sprite Size=&RandomRangeVector2D(Minimum=<25 25>, Maximum=<50 50>)
    Sprite Size Mode=::NewEnumerator4
    UsePositionOffset=false
    Write Material Random=true
   )
   &Color(Color=&MakeLinearColorFromVectorAndFloat("Vector (RGB)"=<2.5 2.5 2.5>, "Float (Alpha)"=&ClampFloat(Float=0.5)))
   &DynamicMaterialParameters(
    Index 0 Param 3=&RandomRangeFloat(Minimum=0.5, Maximum=0.7)
    Module.Param0WriteEnabled=false
    Module.Param1WriteEnabled=false
    Module.Param2WriteEnabled=true
    Module.Param3WriteEnabled=false
   ) {module=DynamicMaterialParameters001}
  ]
  particleUpdateScript [
   &ParticleState()
   &ScaleSpriteSize()
   &SpriteRotationRate(Rotation Rate=&AngleConversion(
     Angle:NiagaraFloat=$Particles.SpriteRotationRate
     Angle Input:ENiagara_AngleInput=::NewEnumerator1
     Angle Output:ENiagara_AngleInput=::NewEnumerator0
    ))
   &ScaleColor(
    Scale Alpha=&FloatFromCurve(FloatCurve=DataInterface() {propertyValues {Curve {keys [<at 0 0 RCIM_Cubic>,<at 0.5 1 RCIM_Cubic>,<at 1 0 RCIM_Cubic>]}}})
    Scale RGB=&VectorFromCurve(VectorCurve=DataInterface() {propertyValues {
       XCurve {keys [<at 0 1 RCIM_Cubic>, {arriveTangent=-0.95, interpMode=RCIM_Cubic, leaveTangent=-0.95, time=0.5, value=0.05}, <at 1 0.05 RCIM_Cubic>]}
       YCurve {keys [<at 0 0.05 RCIM_Cubic>, {arriveTangent=0.95, interpMode=RCIM_Cubic, leaveTangent=0.95, time=0.5, value=0.6122}, <at 1 1 RCIM_Cubic>]}
       ZCurve {keys [<at 0 0.77 RCIM_Cubic>, {arriveTangent=-0.4762, interpMode=RCIM_Cubic, leaveTangent=-0.4762, time=0.5, value=1}, <at 1 0.2939 RCIM_Cubic>]}
      }})
    ScaleA=true
    ScaleRGB=true
   )
   &CameraOffset(Camera Offset Amount=&RandomRangeFloat(Minimum=0, Maximum=15, Evaluation Type=::NewEnumerator1))
   &DynamicMaterialParameters(
    Index 0 Param 1=1
    Index 0 Param 2=1
    Index 0 Param 4=1
    Module.Param0WriteEnabled=true
    Module.Param1WriteEnabled=true
    Module.Param2WriteEnabled=false
    Module.Param3WriteEnabled=true
   )
   /Niagara/Modules/Update/SubUV/SubUVAnimation(Number Of Frames=3)
   &SkeletalMeshLocation(
    Bone / Socket Index=&Add_Integer(A=$Particles.SpawnGroup, B=8)
    Bone Sampling Mode=::NewEnumerator4
    MaskSpawnGroup=false
    Skeletal Mesh=$User.New Skeletal Mesh
   )
  ]
  rendererClasses [&NiagaraSpriteRendererProperties]
  renderers [{rendererClass=&NiagaraSpriteRendererProperties}]
 }
 RainbowRibbons {
  emitterUpdateScript [
   &EmitterState()
   &FN_SpawnRibbon_v2(Number of Ribbons=2, Spawn Mode=::NewEnumerator0, SpawnRate=64)
  ]
  particleSpawnScript [
   &FN_SetSpawnGroup(Number of Spawn Groups=2)
   &PartitionParticles(Number Of Partitions=4)
   &InitializeParticle(
    Color Mode=::NewEnumerator0
    Lifetime=0.3
    Ribbon Width=&Multiply_Float(A=5, B=2)
    Ribbon Width Mode=::NewEnumerator1
    UsePositionOffset=false
    Write Material Random=true
   )
   &SkeletalMeshLocation(
    Bone / Socket Index=&Add_Integer(A=$Particles.SpawnGroup, B=10)
    Bone Sampling Mode=::NewEnumerator4
    MaskSpawnGroup=false
    Sampled Position Offset=&FN_FrontEndVector3Modifier(
     Gameplay=&Lerp_Vector(A=<0 0 -60>, B=<0 0 -80>, Alpha=$User.ParaGlideLeanAlpha)
     Front End=&Lerp_Vector(A=<0 40 -60>, B=<0 35 -70>, Alpha=$User.ParaGlideLeanAlpha)
    )
    Sampled Position Offset Coordinate Space:ENiagaraSkelMeshTransforms=::NewEnumerator1
    Skeletal Mesh=$User.New Skeletal Mesh
   )
   /SetVariables(Particles.Acceleration:Vector3f=&RandomRangeVector(Minimum=<-3 -3 0>, Maximum=<3 3 0>)) {isSetParameters=true}
   &AddVelocity(Velocity=&FN_FrontEndVector3Modifier(
     Gameplay=&Lerp_Vector(A=<0 0 750>, B=<1 1 1200>, Alpha=$User.ParaGlideLeanAlpha)
     Front End=&Lerp_Vector(A=<0 0 700>, B=<1 1 999>, Alpha=$User.ParaGlideLeanAlpha)
    ))
   &AddVelocity(Velocity=<0 -500 0>, Velocity Speed Scale=&FN_FrontEndFloatModifier(Gameplay=0, Front End=1)) {module=AddVelocity002}
  ]
  particleUpdateScript [
   /SetVariables(Particles.ScaledRibbonWidth:NiagaraFloat=&Multiply_Float(
     A=$Particles.RibbonWidth
     B=&FloatFromCurve(FloatCurve=DataInterface() {propertyValues {Curve {keys [
          <at -0.0035 -0.0003 RCIM_Constant>
          {arriveTangent=10.4171, interpMode=RCIM_Cubic, leaveTangent=10.4171, time=0.0709, value=0.0294}
          {arriveTangent=4.2588, interpMode=RCIM_Cubic, leaveTangent=4.2588, time=0.1413, value=1.5081}
          {interpMode=RCIM_Cubic, leaveTangent=0.4859, time=1, value=3.9862}
         ]}}})
    )) {isSetParameters=true}
   &ParticleState()
   &DynamicMaterialParameters(
    Index 0 Param 1=&SelectFloatFromArray(Direct Array Index=$Particles.SpawnGroup, Float Selection Array=DataInterface() {propertyValues {FloatData [0,0.2]}})
    Module.Param0WriteEnabled=true
    Module.Param1WriteEnabled=false
    Module.Param2WriteEnabled=false
    Module.Param3WriteEnabled=false
   )
   &AccelerationForce(Acceleration=$Particles.Acceleration)
   &SolveForcesAndVelocity(Clamp Velocity=false, Limit Acceleration=false)
   &ScaleColor(
    Scale Alpha=&FloatFromCurve(
     FloatCurve=DataInterface() {propertyValues {Curve {keys [
         {interpMode=RCIM_Cubic, tangentMode=RCTM_Break, time=0, value=0}
         {arriveTangent=15.623, interpMode=RCIM_Cubic, leaveTangent=-13.5483, tangentMode=RCTM_Break, time=0.5, value=2}
         {tangentMode=RCTM_User, time=1, value=0}
        ]}}}
     Scale Curve=&FN_FrontEndFloatModifier(Gameplay=0.2, Front End=1)
    )
    ScaleA=true
    ScaleRGB=false
   ) {module=ScaleColor001}
  ]
  rendererClasses [&NiagaraRibbonRendererProperties]
  renderers [{rendererClass=&NiagaraRibbonRendererProperties}]
 }
 discoball {
  emitterUpdateScript [
   &EmitterState(Life Cycle Mode=::NewEnumerator1, Loop Duration=1, UseLoopDelay=true)
   &SpawnRate(SpawnRate=&Multiply_Float(A=3, B=0.45), Use Spawn Probability=false)
  ]
  particleSpawnScript [
   /SetVariables(Particles.MeshOrientation=&MakeQuatFromAxisAngle(Axis=&RandomRangeVector(Minimum=<0 0 0>, Maximum=<0 0 1>), AngleInDegrees=90)) {isSetParameters=true}
   &InitializeParticle(
    Lifetime=2
    Mesh Scale=&VectorFromFloat(Value=&RandomRangeFloat(Minimum=0.6, Maximum=0.9))
    Mesh Scale Mode=::NewEnumerator4
    Position=&ConvertVectorToPosition(Input Position=<0 0 0>, Conversion Operation:ENiagara_LWCConvertVecToPos=::NewEnumerator0)
    Position Mode=::NewEnumerator1
    UsePositionOffset=false
    Write Material Random=true
   )
   &InheritVelocity(Inherited Velocity Amount Scale=<-0.01 -0.01 -0.01>, Inherited Velocity Speed Limit=0, Speed Limit=false, UseSpeedThreshold=false)
   &Color(Color=&MakeLinearColorFromVectorAndFloat("Vector (RGB)"=<0.6684 0.017 1>, "Float (Alpha)"=&ClampFloat(Float=0.8)))
   &ApplyInitialForces()
  ]
  particleUpdateScript [
   &ParticleState()
   &DynamicMaterialParameters(
    Index 0 Param 1=&RandomRangeFloat(Minimum=0, Maximum=10)
    Index 0 Param 2=&RandomRangeFloat(Minimum=4, Maximum=7)
    Index 0 Param 3=&RandomRangeFloat(Minimum=0.8, Maximum=1.5)
    Index 0 Param 4=&RandomRangeFloat(Minimum=-0.6, Maximum=0.6)
    Module.Param0WriteEnabled=true
    Module.Param1WriteEnabled=true
    Module.Param2WriteEnabled=true
    Module.Param3WriteEnabled=true
   )
   &ScaleMeshSize(Scale Factor=&VectorFromFloat(Value=&FloatFromCurve(FloatCurve=DataInterface() {propertyValues {Curve {keys [
          <at 0 0 RCIM_Cubic>
          {arriveTangent=5.1804, interpMode=RCIM_Cubic, leaveTangent=5.1804, time=0.0947, value=1.391}
          {arriveTangent=-0.492, interpMode=RCIM_Cubic, leaveTangent=-0.492, time=0.1928, value=0.9989}
          {arriveTangent=-1.2317, interpMode=RCIM_Cubic, leaveTangent=-1.2317, time=0.8896, value=0.9999}
          {interpMode=RCIM_Cubic, leaveTangent=0.0709, time=1, value=0.0047}
         ]}}})))
   &SolveForcesAndVelocity(Clamp Velocity=false, Limit Acceleration=false)
   &ScaleColor(
    Scale Alpha=&FloatFromCurve(FloatCurve=DataInterface() {propertyValues {Curve {keys [<at 0 0 RCIM_Cubic>,<at 0.5 1 RCIM_Cubic>,<at 1 0 RCIM_Cubic>]}}})
    Scale RGB=&VectorFromCurve(VectorCurve=DataInterface() {propertyValues {XCurve {keys [<at 0 1>,<at 1 1>]}, YCurve {keys [<at 0 1>,<at 1 1>]}, ZCurve {keys [<at 0 1>,<at 1 1>]}}})
    ScaleA=true
    ScaleRGB=true
   )
   &CascadeConversion_Orbit(
    Offset:Vector3f=&RandomRangeVector(Minimum=<0 80 95>, Maximum=<0 95 110>)
    Rotation:Vector3f=&RandomRangeVector(Minimum=<0 0 0>, Maximum=<0 0 1>)
    Rotation Rate:Vector3f=&VectorFromCurve(VectorCurve=DataInterface() {propertyValues {
       XCurve {keys [<at 0 0 RCIM_Cubic>,<at 0.997 0 RCIM_Cubic>]}
       YCurve {keys [<at 0 0 RCIM_Cubic>,<at 0.997 0 RCIM_Cubic>]}
       ZCurve {keys [{arriveTangent=0.0683, interpMode=RCIM_Cubic, time=0, value=0.0998}, {interpMode=RCIM_Cubic, leaveTangent=0.0407, time=0.997, value=0.1254}]}
      }})
   )
   &CascadeConversion_SolveOrbit()
   &CameraOffset(Camera Offset Amount=-25)
  ]
  rendererClasses [&NiagaraMeshRendererProperties]
  renderers [{rendererClass=&NiagaraMeshRendererProperties, rendererIndex=0}]
 }
 glow {
  emitterUpdateScript [
   &EmitterState(Life Cycle Mode=::NewEnumerator1, Loop Duration=2, UseLoopDelay=true)
   &SpawnRate(SpawnRate=4, Use Spawn Probability=false)
  ]
  particleSpawnScript [
   &InitializeParticle(
    Lifetime=&RandomRangeFloat(Minimum=0.42, Maximum=0.8)
    Position=&ConvertVectorToPosition(
     Input Position=&RandomRangeVector(Minimum=<-55 -62 40>, Maximum=<55 62 120>)
     Conversion Operation:ENiagara_LWCConvertVecToPos=::NewEnumerator0
    )
    Position Mode=::NewEnumerator1
    Sprite Size=&Vector2DFromFloat(Value=&RandomRangeFloat(Minimum=80, Maximum=200))
    Sprite Size Mode=::NewEnumerator4
    UsePositionOffset=false
    Write Material Random=true
   )
   &AddVelocity(Velocity=&RandomRangeVector(Minimum=<-10 -10 20>, Maximum=<10 10 60>))
   &Color(Color=&MakeLinearColorFromVectorAndFloat("Vector (RGB)"=<1 1 1>, "Float (Alpha)"=&ClampFloat(Float=0.25)))
  ]
  particleUpdateScript [
   &ParticleState()
   &SolveForcesAndVelocity(Clamp Velocity=false, Limit Acceleration=false)
   &ScaleColor(
    Scale Alpha=&FloatFromCurve(FloatCurve=DataInterface() {propertyValues {Curve {keys [
         {arriveTangent=10.7452, interpMode=RCIM_Cubic, time=0, value=0}
         {arriveTangent=1.0812, interpMode=RCIM_Cubic, leaveTangent=1.0812, time=0.1334, value=0.9177}
         {arriveTangent=-1.0589, interpMode=RCIM_Cubic, leaveTangent=-1.0589, time=0.7906, value=0.8548}
         <at 1 0>
        ]}}})
    Scale RGB=&VectorFromCurve(VectorCurve=DataInterface() {propertyValues {
       XCurve {keys [
         <at 0 4 RCIM_Cubic>
         {arriveTangent=-3.6539, interpMode=RCIM_Cubic, leaveTangent=-3.6539, time=0.38, value=0.0782}
         <at 0.7 1.4423>
         <at 1 4>
        ]}
       YCurve {keys [<at 0 1.0413 RCIM_Cubic>, {arriveTangent=-1.4876, interpMode=RCIM_Cubic, leaveTangent=-1.4876, time=0.38, value=3}, <at 0.7 0>, <at 1 0>]}
       ZCurve {keys [<at 0 0 RCIM_Cubic>, {arriveTangent=5.7143, interpMode=RCIM_Cubic, leaveTangent=5.7143, time=0.38, value=0}, <at 0.7 4>, <at 1 0.1853>]}
      }})
    ScaleA=true
    ScaleRGB=true
   )
  ]
  rendererClasses [&NiagaraSpriteRendererProperties]
  renderers [{rendererClass=&NiagaraSpriteRendererProperties}]
 }
 light {
  emitterUpdateScript [
   &EmitterState(Life Cycle Mode=::NewEnumerator1, Loop Behavior=::NewEnumerator1, Loop Duration Mode=::NewEnumerator1, UseLoopDelay=true)
   &SpawnBurst_Instantaneous(Use Spawn Probability=false, UseLoopCountLimit=false)
  ]
  particleSpawnScript [
   &InitializeParticle(Lifetime=5, Position Mode=::NewEnumerator0, Sprite Size=<25 25>, Sprite Size Mode=::NewEnumerator4, Write Material Random=true)
   &Color(Color=&MakeLinearColorFromVectorAndFloat("Vector (RGB)"=<15 15 15>, "Float (Alpha)"=&ClampFloat(Float=1.95)))
  ]
  particleUpdateScript [
   /SetVariables(Particles.LightExponent=2) {isSetParameters=true}
   &ParticleState(
    Kill Particles When Lifetime Has Elapsed=false
    Let Infinitely Lived Particles Die When Emitter Deactivates=true
    Loop Particles Lifetime=true
   )
   &SkeletalMeshLocation(
    Bone Sampling Mode=::NewEnumerator4
    MaskSpawnGroup=false
    Position Sampling=::NewEnumerator1
    Sampled Position Offset=<0 -25 30>
    Sampled Position Offset Coordinate Space:ENiagaraSkelMeshTransforms=::NewEnumerator2
    Skeletal Mesh=$User.New Skeletal Mesh
   )
   &ScaleColor(
    Scale Alpha=&FloatFromCurve(
     FloatCurve=DataInterface() {propertyValues {Curve {keys [
         {arriveTangent=10.7452, interpMode=RCIM_Cubic, time=0, value=0}
         {arriveTangent=0.5628, interpMode=RCIM_Cubic, leaveTangent=0.5628, time=0.1123, value=0.9177}
         {arriveTangent=0.0538, interpMode=RCIM_Cubic, leaveTangent=0.0538, time=0.2234, value=0.1257}
         {arriveTangent=0.0703, interpMode=RCIM_Cubic, leaveTangent=0.0703, time=0.2699, value=0.9261}
         {arriveTangent=-0.0214, interpMode=RCIM_Cubic, leaveTangent=-0.0214, time=0.3799, value=0.1367}
         {arriveTangent=0.0081, interpMode=RCIM_Cubic, leaveTangent=0.0081, time=0.4307, value=0.9227}
         {arriveTangent=-0.1547, interpMode=RCIM_Cubic, leaveTangent=-0.1547, time=0.5745, value=0.1383}
         {arriveTangent=0.0294, interpMode=RCIM_Cubic, leaveTangent=0.0294, time=0.6337, value=0.8913}
         {arriveTangent=-0.2324, interpMode=RCIM_Cubic, leaveTangent=-0.2324, time=0.7733, value=0.1442}
         {arriveTangent=-0.636, interpMode=RCIM_Cubic, leaveTangent=-0.636, time=0.7906, value=0.8548}
         <at 1 0>
        ]}}}
     Scale Curve=0.5
    )
    Scale RGB=&VectorFromCurve(VectorCurve=DataInterface() {propertyValues {
       XCurve {keys [
         <at 0 4 RCIM_Cubic>
         {arriveTangent=-6.5577, interpMode=RCIM_Cubic, leaveTangent=-6.5577, time=0.2, value=0}
         <at 0.5 0.7211>
         <at 0.75 4>
         <at 1 3>
        ]}
       YCurve {keys [
         <at 0 1.0413 RCIM_Cubic>
         {arriveTangent=-2.0826, interpMode=RCIM_Cubic, leaveTangent=-2.0826, time=0.2, value=1.8}
         <at 0.5 0>
         <at 0.75 0>
         <at 1 1.2898>
        ]}
       ZCurve {keys [
         <at 0 0 RCIM_Cubic>
         {arriveTangent=4, interpMode=RCIM_Cubic, leaveTangent=4, time=0.2, value=0.4392}
         <at 0.5 2>
         <at 0.75 0.1853>
         <at 1 0>
        ]}
      }})
    ScaleA=true
    ScaleRGB=true
   )
   &CascadeConversion_LightProperties(LightBrightnessScale=0.25, LightColorScale=<1 1 10>, LightRadiusScale=1.4)
  ]
  rendererClasses [&NiagaraLightRendererProperties]
  renderers [{rendererClass=&NiagaraLightRendererProperties, rendererIndex=0}]
 }
 star {
  emitterUpdateScript [
   &EmitterState(Life Cycle Mode=::NewEnumerator1, Loop Duration=0.5, UseLoopDelay=true)
   &SpawnBurst_Instantaneous(
    Spawn Count=&RandomRangeInt(Minimum=1, Maximum=10)
    Use Spawn Probability=false
    UseLoopCountLimit=false
   )
   &SpawnRate(SpawnRate=50, Use Spawn Probability=false)
  ]
  particleSpawnScript [
   &FN_SetSpawnGroup()
   &InitializeParticle(
    Lifetime=&RandomRangeFloat(Minimum=0.16, Maximum=0.6)
    Sprite Size=&Vector2DFromFloat(Value=&RandomRangeFloat(Minimum=1.5, Maximum=3.8))
    Sprite Size Mode=::NewEnumerator4
    UsePositionOffset=false
    Write Material Random=true
   )
   &Color(Color=&MakeLinearColorFromVectorAndFloat(
     "Vector (RGB)"=&RandomRangeVector(Minimum=<4 4 4>, Maximum=<10 10 10>)
     "Float (Alpha)"=&ClampFloat(Float=1)
    ))
   &InheritVelocity(
    Inherited Velocity Amount Scale=&RandomRangeVector(Minimum=<-0.22 -0.2 -0.22>, Maximum=<-0.08 -0.1 -0.08>)
    Inherited Velocity Speed Limit=0
    Speed Limit=false
    UseSpeedThreshold=false
   )
   &AddVelocity(
    Velocity=&RandomRangeVector(Minimum=<0 -150 150>, Maximum=<0 -300 300>)
    Velocity Speed Scale=&FN_FrontEndFloatModifier(Gameplay=0, Front End=1)
   ) {module=AddVelocity002}
   &SkeletalMeshLocation(
    Bone / Socket Index=&Add_Integer(A=$Particles.SpawnGroup, B=8)
    Bone Sampling Mode=::NewEnumerator4
    MaskSpawnGroup=false
    Skeletal Mesh=$User.New Skeletal Mesh
   )
   &ShapeLocation(MaskSpawnGroup=false, Override Seed=false, Sphere Radius=20, UseOwnerScale=true)
   &AddVelocity(Velocity=<0 0 -80>) {module=AddVelocity001}
   &AddVelocity(Velocity=&Multiply_VectorByFloat(Vector=$Particles.ShapeLocation.ShapeVector, Float=&RandomRangeFloat(Minimum=1.5, Maximum=4.5)))
   &ApplyInitialForces()
  ]
  particleUpdateScript [
   &ParticleState()
   &ScaleSpriteSize()
   &Drag(Drag=2, Use Linear Drag=true, Use Rotational Drag=true)
   &SolveForcesAndVelocity(Clamp Velocity=false, Limit Acceleration=false)
   &ScaleColor(
    Scale Alpha=&FloatFromCurve(FloatCurve=DataInterface() {propertyValues {Curve {keys [
         {arriveTangent=10.7452, interpMode=RCIM_Cubic, time=0, value=0}
         {arriveTangent=1.6068, interpMode=RCIM_Cubic, leaveTangent=1.6068, time=0.1334, value=0.9177}
         {arriveTangent=-0.4298, interpMode=RCIM_Cubic, leaveTangent=-0.4298, time=0.1811, value=0.291}
         {arriveTangent=-0.3692, interpMode=RCIM_Cubic, leaveTangent=-0.3692, time=0.3461, value=0.8263}
         {arriveTangent=0.1565, interpMode=RCIM_Cubic, leaveTangent=0.1565, time=0.4899, value=0.177}
         {arriveTangent=0.0773, interpMode=RCIM_Cubic, leaveTangent=0.0773, time=0.5914, value=0.8647}
         {arriveTangent=0.1069, interpMode=RCIM_Cubic, leaveTangent=0.1069, time=0.7314, value=0.1956}
         {arriveTangent=-0.7283, interpMode=RCIM_Cubic, leaveTangent=-0.7283, time=0.8241, value=0.8895}
         <at 1 0>
        ]}}})
    Scale RGB=&VectorFromCurve(VectorCurve=DataInterface() {propertyValues {
       XCurve {keys [
         <at 0 8 RCIM_Cubic>
         {arriveTangent=-14.44, interpMode=RCIM_Cubic, leaveTangent=-14.44, time=0.2, value=0.05}
         <at 0.5 0.78>
         <at 0.75 8>
         <at 1 10>
        ]}
       YCurve {keys [
         <at 0 0.8 RCIM_Cubic>
         {arriveTangent=-1.6, interpMode=RCIM_Cubic, leaveTangent=-1.6, time=0.2, value=7}
         <at 0.5 0>
         <at 0.75 0>
         <at 1 0.88>
        ]}
       ZCurve {keys [<at 0 0 RCIM_Cubic>, {arriveTangent=16, interpMode=RCIM_Cubic, leaveTangent=16, time=0.2, value=0}, <at 0.5 8>, <at 0.75 0.1853>, <at 1 0>]}
      }})
    ScaleA=true
    ScaleRGB=true
   )
   &CameraOffset(Camera Offset Amount=60)
   &CascadeConversion_Orbit(
    Offset:Vector3f=&RandomRangeVector(Minimum=<-9 -7 0>, Maximum=<7 8 0>)
    Rotation:Vector3f=&RandomRangeVector(Minimum=<0 0 0>, Maximum=<1 1 1>)
    Rotation Rate:Vector3f=&RandomRangeVector(Minimum=<0 0 0>, Maximum=<1 1 1>)
   )
   &CascadeConversion_SolveOrbit()
  ]
  rendererClasses [&NiagaraSpriteRendererProperties]
  renderers [{rendererClass=&NiagaraSpriteRendererProperties}]
 }
}
summary {
 systemName=P_Contrail_DiscoBall_Converted
 emitters [
  {emitterName=discoball, rendererClasses [&NiagaraMeshRendererProperties]}
  {emitterName=Corona, rendererClasses [&NiagaraSpriteRendererProperties]}
  {emitterName=star, rendererClasses [&NiagaraSpriteRendererProperties]}
  {emitterName=Confetti_Center, simTarget=GPUComputeSim, rendererClasses [&NiagaraSpriteRendererProperties]}
  {emitterName=Confetti_g, simTarget=GPUComputeSim, rendererClasses [&NiagaraSpriteRendererProperties]}
  {emitterName=light, rendererClasses [&NiagaraLightRendererProperties]}
  {emitterName=glow, rendererClasses [&NiagaraSpriteRendererProperties]}
  {emitterName=RainbowRibbons, rendererClasses [&NiagaraRibbonRendererProperties]}
 ]
 userVariables [
  User.bIsFrontEnd {type=NiagaraBool}
  User.bIsFrontEndPreview {type=NiagaraBool}
  User.New Skeletal Mesh {
   type=NiagaraDataInterfaceSkeletalMesh
   defaultValue {
    struct=NiagaraExt_VariableValue_DataInterface
    value {dataInterface=&self:NiagaraDataInterfaceSkeletalMesh_0}
   }
  }
  User.ParaGlideLeanAlpha {type=NiagaraFloat, defaultValue=0, description=""}
 ]
}
