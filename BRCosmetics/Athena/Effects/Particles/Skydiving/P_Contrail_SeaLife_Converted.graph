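# Niagara skydiving contrail cosmetic that trails sea life from the player.
# Sprite bubbles, debris and randomly picked tumbling meshes from the Clam emitter surround the player, while two noise-driven seaweed ribbons and dive bubbles sample the user skeletal mesh.
# ParaGlideLeanAlpha scales ambient bubble and clam spawn rates, velocities and ribbon noise, and gates dive bubbles entirely.
# FrontEnd modifier inputs substitute separate rates, offsets and velocities for front-end display.
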NiagaraSystem /BRCosmetics/Athena/Effects/Particles/Skydiving/P_Contrail_SeaLife_Converted
dependencies {
 usedDataInterfaces [&NiagaraDataInterfaceCurve, &NiagaraDataInterfaceVectorCurve]
 usedDynamicInputs [
  &SetBoolByFloatComparison
  &self:Abs_Z_Axis
  &TransformVector
  &NormalizeVector
 ]
 usedModules [
  &SystemState
  &Timeline
  &self:SystemSpawnScript.SystemScriptSource.SystemScriptGraph.NiagaraNodeAssignment_1.SetVariables_#1
  &self:SystemSpawnScript.SystemScriptSource.SystemScriptGraph.NiagaraNodeAssignment_0.SetVariables_#2
 ]
}
emitters {
 Bubbles_AroundPlayer {
  emitterUpdateScript [
   &EmitterState()
   &SpawnRate(
    SpawnRate=&FN_FrontEndFloatModifier(Gameplay=&Lerp_Float(A=30, B=45, Alpha=$User.ParaGlideLeanAlpha), Front End=20)
    Use Spawn Probability=false
   )
  ]
  particleSpawnScript [
   &InitializeParticle(
    Lifetime Min=0.5
    Lifetime Mode=::NewEnumerator1
    Sprite Rotation Mode=::NewEnumerator2
    Sprite Size Max=<18 17>
    Sprite Size Min=<5 6>
    Sprite Size Mode=::NewEnumerator2
    UsePositionOffset=false
    Write Material Random=true
   ) {module=InitializeParticle001}
   &ShapeLocation(
    MaskSpawnGroup=false
    Offset=<0 0 -75>
    Offset Mode=::NewEnumerator0
    Sphere Distribution=::NewEnumerator1
    Sphere Radius=&RandomRangeFloat(Minimum=75, Maximum=150)
    U Position=&RandomRangeFloat(Minimum=0, Maximum=1)
    UseOwnerScale=true
    V Position=&RandomRangeFloat(Minimum=0, Maximum=1)
   )
   &InheritVelocity(Inherited Velocity Amount Scale=<0 0 -0.12>, Speed Limit=false, UseSpeedThreshold=false)
   &AddVelocity(Velocity=&FN_FrontEndVector3Modifier(
     Gameplay=&Multiply_Vector(A=&VectorFromFloat(Value=$User.ParaGlideLeanAlpha), B=<0 0 150>)
     Front End=/Niagara/DynamicInputs/Lerp/Lerp_Position(A:NiagaraPosition=<0 0 50>, B:NiagaraPosition=<1 -35 300>, Alpha:NiagaraFloat=$User.ParaGlideLeanAlpha)
    ))
  ]
  particleUpdateScript [
   &ParticleState()
   &ScaleColor(
    Scale Alpha=&FloatFromCurve(FloatCurve=DataInterface() {propertyValues {Curve {keys [
         {interpMode=RCIM_Cubic, tangentMode=RCTM_User, time=0, value=0}
         {interpMode=RCIM_Cubic, tangentMode=RCTM_User, time=0.5, value=1}
         {tangentMode=RCTM_User, time=1, value=0}
        ]}}})
    ScaleA=true
    ScaleRGB=true
   )
   &ScaleSpriteSize(Uniform Curve Sprite Scale=DataInterface() {propertyValues {Curve {keys [
        {interpMode=RCIM_Cubic, tangentMode=RCTM_User, time=0, value=0}
        {interpMode=RCIM_Cubic, tangentMode=RCTM_User, time=0.7, value=1}
        {tangentMode=RCTM_User, time=1, value=0}
       ]}}})
   &CurlNoiseForce(Noise Frequency=1000, Noise Strength=200, Pan Noise=false)
   &SolveForcesAndVelocity(Clamp Velocity=false, Limit Acceleration=false)
  ]
  rendererClasses [&NiagaraSpriteRendererProperties]
  renderers [{rendererClass=&NiagaraSpriteRendererProperties}]
 }
 Bubbles_Dive {
  emitterUpdateScript [
   &EmitterState()
   &SpawnRate(
    SpawnRate=&Multiply_Float(A=&RandomRangeFloat(Minimum=50, Maximum=75), B=$User.ParaGlideLeanAlpha)
    Use Spawn Probability=false
   )
  ]
  particleSpawnScript [
   &InitializeParticle(
    Lifetime Max=0.65
    Lifetime Min=0.25
    Lifetime Mode=::NewEnumerator1
    Sprite Rotation Mode=::NewEnumerator2
    Sprite Size Mode=::NewEnumerator1
    Uniform Sprite Size Max=15
    Uniform Sprite Size Min=10
    UsePositionOffset=false
    Write Material Random=true
   ) {module=InitializeParticle001}
   &SkeletalMeshLocation(
    Bone / Socket Index=4
    Bone Sampling Mode=::NewEnumerator4
    MaskSpawnGroup=false
    Sampled Position Offset=&FN_FrontEndVector3Modifier(Gameplay=<-60 0 0>, Front End=<-75 11 0>)
    Skeletal Mesh=$User.SkeletalMesh
   ) {module=SkeletalMeshLocation002}
   &ShapeLocation(MaskSpawnGroup=false, Non Uniform Scale=<1 1 0.15>, Override Seed=false, Sphere Radius=25, UseOwnerScale=true)
   &InheritVelocity(
    Inherited Velocity Amount Scale=&FN_FrontEndVector3Modifier(Gameplay=<0 0 -0.1>, Front End=<0 0 0>)
    Speed Limit=false
    UseSpeedThreshold=false
   )
   &AddVelocity(Velocity=&FN_FrontEndVector3Modifier(Gameplay=<0 0 0>, Front End=<0 -225 200>))
  ]
  particleUpdateScript [
   &ParticleState()
   &Drag(Drag=3, Use Linear Drag=true, Use Rotational Drag=false)
   &ScaleColor(
    Scale Alpha=&FloatFromCurve(FloatCurve=DataInterface() {propertyValues {Curve {keys [
         {interpMode=RCIM_Cubic, tangentMode=RCTM_User, time=0.15, value=0}
         {interpMode=RCIM_Cubic, tangentMode=RCTM_User, time=0.35, value=1}
         {tangentMode=RCTM_User, time=1, value=0.5}
        ]}}})
    ScaleA=true
    ScaleRGB=true
   )
   &ScaleSpriteSize(Uniform Curve Sprite Scale=DataInterface() {propertyValues {Curve {keys [{tangentMode=RCTM_User, time=0.5, value=1}, {tangentMode=RCTM_User, time=1, value=0}]}}})
   &SolveForcesAndVelocity(Clamp Velocity=false, Limit Acceleration=false)
  ]
  rendererClasses [&NiagaraSpriteRendererProperties]
  renderers [{rendererClass=&NiagaraSpriteRendererProperties}]
 }
 Clam {
  emitterUpdateScript [
   &EmitterState()
   &SpawnPerUnit(
    Module.Use Max Movement Threshold=true
    Module.Use Movement Tolerance=true
    Movement Tolerance=0.1
    Spawn Spacing=100
    Use Spawn Probability=false
   ) {enabled=false}
   &SpawnRate(
    SpawnRate=&FN_FrontEndFloatModifier(Gameplay=&Lerp_Float(A=4, B=8, Alpha=$User.ParaGlideLeanAlpha), Front End=3)
    Use Spawn Probability=false
   )
  ]
  particleSpawnScript [
   /SetVariables(Particles.MeshIndex=&RandomRangeInt(Minimum=0, Maximum=3)) {isSetParameters=true}
   &InitializeParticle(
    Lifetime Max=3
    Lifetime Mode=::NewEnumerator1
    Mesh Scale Mode=::NewEnumerator1
    Mesh Uniform Scale Max=0.9
    Mesh Uniform Scale Min=0.15
    UsePositionOffset=false
    Write Material Random=true
   ) {module=InitializeParticle001}
   &InheritVelocity(Inherited Velocity Amount Scale=<0 0 -0.15>, Inherited Velocity Speed Limit=500, Speed Limit=false, UseSpeedThreshold=false)
   &ShapeLocation(
    Handle Radius=50
    Large Radius=135
    MaskSpawnGroup=false
    Offset=&FN_FrontEndVector3Modifier(Gameplay=<0 0 -75>, Front End=<0 0 0>)
    Offset Mode=::NewEnumerator0
    Override Seed=false
    Shape Primitive=::NewEnumerator3
    UseOwnerScale=true
   )
   &InitialMeshOrientation(Use Rotation Vector=false)
   &AddVelocity(Velocity=&FN_FrontEndVector3Modifier(
     Gameplay=&Multiply_Vector(A=&VectorFromFloat(Value=$User.ParaGlideLeanAlpha), B=<0 0 150>)
     Front End=&Lerp_Position(A:NiagaraPosition=<0 0 50>, B:NiagaraPosition=<1 -25 250>, Alpha:NiagaraFloat=$User.ParaGlideLeanAlpha)
    ))
  ]
  particleUpdateScript [
   &ParticleState()
   &UpdateMeshOrientation(
    Rotation Rate=&FloatFromCurve(FloatCurve=DataInterface() {propertyValues {Curve {keys [<at 0 1>,<at 1 0>]}}})
    Rotation Vector=<1 1 1>
   )
   &WindForce(Wind Speed=&RandomRangeVector(Minimum=<-10 -10 -10>, Maximum=<10 10 10>))
   &SolveForcesAndVelocity(Clamp Velocity=false, Limit Acceleration=false)
   &ScaleColor(
    Scale Alpha=&FloatFromCurve(FloatCurve=DataInterface() {propertyValues {Curve {keys [
         {interpMode=RCIM_Cubic, tangentMode=RCTM_User, time=0, value=0}
         {interpMode=RCIM_Cubic, tangentMode=RCTM_User, time=0.1, value=1.125}
         {arriveTangent=-1.25, interpMode=RCIM_Cubic, leaveTangent=-1.25, time=0.9, value=1}
         {tangentMode=RCTM_User, time=1, value=0}
        ]}}})
    Scale RGB=&VectorFromCurve(VectorCurve=DataInterface() {propertyValues {XCurve {keys [<at 0 1>,<at 1 1>]}, YCurve {keys [<at 0 1>,<at 1 1>]}, ZCurve {keys [<at 0 1>,<at 1 1>]}}})
    ScaleA=true
    ScaleRGB=true
   )
   &ScaleMeshSize(Scale Factor=&VectorFromFloat(Value=&FloatFromCurve(FloatCurve=DataInterface() {propertyValues {Curve {keys [
          <at 0 0 RCIM_Cubic>
          {arriveTangent=5.1804, interpMode=RCIM_Cubic, leaveTangent=5.1804, time=0.0947, value=1.391}
          {arriveTangent=-0.492, interpMode=RCIM_Cubic, leaveTangent=-0.492, time=0.1928, value=0.9989}
          {arriveTangent=-1.2317, interpMode=RCIM_Cubic, leaveTangent=-1.2317, time=0.8896, value=0.9999}
          {interpMode=RCIM_Cubic, leaveTangent=0.0709, time=1, value=0.0047}
         ]}}})))
   &CameraOffset(Camera Offset Amount=-25) {enabled=false}
   &DynamicMaterialParameters(
    Index 0 Param 1=&RandomRangeFloat(Minimum=0, Maximum=10)
    Index 0 Param 2=&RandomRangeFloat(Minimum=4, Maximum=7)
    Index 0 Param 3=&RandomRangeFloat(Minimum=0.8, Maximum=1.5)
    Index 0 Param 4=&RandomRangeFloat(Minimum=-0.6, Maximum=0.6)
    Module.Param0WriteEnabled=true
    Module.Param1WriteEnabled=true
    Module.Param2WriteEnabled=true
    Module.Param3WriteEnabled=true
   ) {enabled=false}
   &SolveRotationalForcesAndVelocity()
  ]
  rendererClasses [&NiagaraMeshRendererProperties]
  renderers [{rendererClass=&NiagaraMeshRendererProperties, rendererIndex=0}]
 }
 SeaDebrisLoop {
  emitterUpdateScript [
   &EmitterState(Life Cycle Mode=::NewEnumerator1, Loop Duration=1, UseLoopDelay=true)
   &SpawnRate(SpawnRate=25, Use Spawn Probability=false)
  ]
  particleSpawnScript [
   &InitializeParticle(
    Lifetime Mode=::NewEnumerator1
    Sprite Rotation Mode=::NewEnumerator2
    Sprite Size Mode=::NewEnumerator1
    Uniform Sprite Size Max=6
    Uniform Sprite Size Min=2
    UsePositionOffset=false
    Write Material Random=true
   ) {module=InitializeParticle001}
   &ShapeLocation(MaskSpawnGroup=false, Offset=<0 0 -75>, Offset Mode=::NewEnumerator0, Override Seed=false, Sphere Radius=200, UseOwnerScale=true)
   &InheritVelocity(Inherited Velocity Amount Scale=<-0.1 -0.1 -0.1>, Speed Limit=false, UseSpeedThreshold=false)
   &AddVelocity(Velocity=&FN_FrontEndVector3Modifier(
     Gameplay=&Multiply_Vector(A=&VectorFromFloat(Value=$User.ParaGlideLeanAlpha), B=<0 0 150>)
     Front End=&Lerp_Position(A:NiagaraPosition=<0 0 50>, B:NiagaraPosition=<1 -35 300>, Alpha:NiagaraFloat=$User.ParaGlideLeanAlpha)
    ))
  ]
  particleUpdateScript [
   &ParticleState()
   &ScaleColor(
    Scale Alpha=&FloatFromCurve(FloatCurve=DataInterface() {propertyValues {Curve {keys [
         {interpMode=RCIM_Cubic, tangentMode=RCTM_User, time=0, value=0}
         {interpMode=RCIM_Cubic, tangentMode=RCTM_User, time=0.5, value=1}
         {tangentMode=RCTM_User, time=1, value=0}
        ]}}})
    ScaleA=true
    ScaleRGB=true
   )
   &ScaleSpriteSize()
   &CurlNoiseForce(Noise Frequency=1000, Noise Strength=200, Pan Noise=false)
   &SolveForcesAndVelocity(Clamp Velocity=false, Limit Acceleration=false)
  ]
  rendererClasses [&NiagaraSpriteRendererProperties]
  renderers [{rendererClass=&NiagaraSpriteRendererProperties}]
 }
 Seaweed_Ribbon {
  emitterUpdateScript [
   &EmitterState()
   &FN_SpawnRibbon_v2(Number of Ribbons=2, Spawn Mode=::NewEnumerator0, SpawnRate=100)
  ]
  particleSpawnScript [
   &FN_InitializeRibbon_v3(
    Color:LinearColor=<rgba 1 1 1 2>
    Enable Skeletal Position=true
    Offset Axis:ENiagaraOrientationAxis=::ZAxis
    Ribbon Width=35
    Sampled Position Coordinate Space:ENiagaraCoordinateSpace=::Local
    Sampled Rotation Coordinate Space:ENiagaraCoordinateSpace=::Local
    Sampled Velocity Coordinate Space:ENiagaraCoordinateSpace=::Local
   ) {module=FN_InitializeRibbon_v3001}
   &SkeletalMeshLocation(
    Bone Sampling Mode=::NewEnumerator4
    MaskSpawnGroup=false
    Skeletal Mesh=$User.SkeletalMesh
    Use Sampled Mesh Attribute as Ribbon ID=false
    Write To Particle Position=false
   )
   &AddVelocity(
    Scale Added Velocity=&Lerp_Vector(A=<1 1 1>, B=<1 1 2>, Alpha=$Emitter.Dive)
    Velocity=&FN_FrontEndVector3Modifier(Gameplay=<0 0 350>, Front End=&Lerp_Vector(A=<0 0 350>, B=<0 -300 350>, Alpha=$User.ParaGlideLeanAlpha))
   )
   &DynamicMaterialParameters(
    Index 0 Param 1=&RandomRangeFloat(Minimum=0, Maximum=1)
    Index 0 Param 2=1
    Index 0 Param 3=1
    Index 0 Param 4=&RandomRangeFloat(Minimum=0, Maximum=1)
    Module.Param0WriteEnabled=true
    Module.Param1WriteEnabled=true
    Module.Param2WriteEnabled=true
    Module.Param3WriteEnabled=true
   ) {enabled=false}
   &DynamicMaterialParameters(
    Index 0 Param 1=&Lerp_Float(A=0, B=0.2, Alpha=&MakeFloatFromInt(Int=$Particles.SpawnGroup))
    Index 0 Param 2=1.15
    Index 0 Param 3=1
    Index 0 Param 4=&Lerp_Float(A=-0.15, B=0.15, Alpha=&MakeFloatFromInt(Int=$Particles.SpawnGroup))
    Module.Param0WriteEnabled=true
    Module.Param1WriteEnabled=true
    Module.Param2WriteEnabled=true
    Module.Param3WriteEnabled=true
   ) {module=DynamicMaterialParameters001}
  ]
  particleUpdateScript [
   &CurlNoiseForce(
    Noise Frequency=25
    Noise Strength=&Multiply_Float(A=&Lerp_Float(A=0.25, B=0.75, Alpha=$User.ParaGlideLeanAlpha), B=750)
    Pan Noise=true
    Pan Noise Field=<0 0 5>
   )
   &SolveForcesAndVelocity(Clamp Velocity=false, Limit Acceleration=false)
   &FN_UpdateRibbon(
    Enable Skeletal Position=true
    Initial Color:LinearColor=$Particles.Initial.Color
    Initial RibbonWidth:NiagaraFloat=$Particles.Initial.RibbonWidth
   )
   &ScaleRibbonWidth(Ribbon Width Scale=&FloatFromCurve(FloatCurve=DataInterface() {propertyValues {Curve {keys [
         {interpMode=RCIM_Cubic, tangentMode=RCTM_User, time=0, value=0.75}
         {interpMode=RCIM_Cubic, tangentMode=RCTM_User, time=0.5, value=1}
         {tangentMode=RCTM_User, time=1, value=1}
        ]}}}))
  ]
  rendererClasses [&NiagaraRibbonRendererProperties]
  renderers [{rendererClass=&NiagaraRibbonRendererProperties}]
 }
 TinyBubbles_Dive {
  emitterUpdateScript [
   &EmitterState()
   &SpawnRate(
    SpawnRate=&Multiply_Float(A=&RandomRangeFloat(Minimum=350, Maximum=500), B=$User.ParaGlideLeanAlpha)
    Use Spawn Probability=false
   )
   &SpawnRate(
    SpawnGroup=1
    SpawnRate=&Multiply_Float(A=&RandomRangeFloat(Minimum=150, Maximum=200), B=$User.ParaGlideLeanAlpha)
    Use Spawn Probability=false
   ) {enabled=false, module=SpawnRate001}
  ]
  particleSpawnScript [
   &InitializeParticle(
    Lifetime Max=0.65
    Lifetime Min=0.25
    Lifetime Mode=::NewEnumerator1
    Sprite Rotation Mode=::NewEnumerator2
    Sprite Size Mode=::NewEnumerator1
    Uniform Sprite Size Max=6
    Uniform Sprite Size Min=4
    UsePositionOffset=false
    Write Material Random=true
   ) {module=InitializeParticle001}
   &SkeletalMeshLocation(
    Bone / Socket Index=4
    Bone Sampling Mode=::NewEnumerator4
    MaskSpawnGroup=false
    Sampled Position Offset=&FN_FrontEndVector3Modifier(Gameplay=<-70 0 0>, Front End=<-80 11 0>)
    Skeletal Mesh=$User.SkeletalMesh
   ) {module=SkeletalMeshLocation002}
   &ShapeLocation(MaskSpawnGroup=false, Non Uniform Scale=<1 1 0.15>, Override Seed=false, Sphere Radius=25, UseOwnerScale=true)
   &InheritVelocity(
    Inherited Velocity Amount Scale=&FN_FrontEndVector3Modifier(Gameplay=<0 0 -0.1>, Front End=<0 0 0>)
    Speed Limit=false
    UseSpeedThreshold=false
   )
   &AddVelocity(Velocity=&FN_FrontEndVector3Modifier(Gameplay=<0 0 0>, Front End=<0 -250 200>))
  ]
  particleUpdateScript [
   &ParticleState()
   &CurlNoiseForce(Noise Strength=600, Pan Noise=false)
   &Drag(Drag=3, Use Linear Drag=true, Use Rotational Drag=false)
   &ScaleColor(
    Scale Alpha=&FloatFromCurve(FloatCurve=DataInterface() {propertyValues {Curve {keys [
         {interpMode=RCIM_Cubic, tangentMode=RCTM_User, time=0.1, value=0}
         {interpMode=RCIM_Cubic, tangentMode=RCTM_User, time=0.2, value=0.95}
         {tangentMode=RCTM_User, time=1, value=0.5}
        ]}}})
    ScaleA=true
    ScaleRGB=true
   )
   &ScaleSpriteSize(Uniform Curve Sprite Scale=DataInterface() {propertyValues {Curve {keys [{tangentMode=RCTM_User, time=0.75, value=1}, {tangentMode=RCTM_User, time=1, value=0}]}}})
   &SolveForcesAndVelocity(Clamp Velocity=false, Limit Acceleration=false)
  ]
  rendererClasses [&NiagaraSpriteRendererProperties]
  renderers [{rendererClass=&NiagaraSpriteRendererProperties}]
 }
}
summary {
 systemName=P_Contrail_SeaLife_Converted
 emitters [
  {emitterName=Clam, rendererClasses [&NiagaraMeshRendererProperties]}
  {emitterName=Bubbles_AroundPlayer, rendererClasses [&NiagaraSpriteRendererProperties]}
  {emitterName=SeaDebrisLoop, rendererClasses [&NiagaraSpriteRendererProperties]}
  {emitterName=Seaweed_Ribbon, rendererClasses [&NiagaraRibbonRendererProperties]}
  {emitterName=Bubbles_Dive, rendererClasses [&NiagaraSpriteRendererProperties]}
  {emitterName=TinyBubbles_Dive, rendererClasses [&NiagaraSpriteRendererProperties]}
 ]
 userVariables [
  User.bIsFrontEnd {type=NiagaraBool}
  User.bIsFrontEndPreview {type=NiagaraBool}
  User.ParaGlideLeanAlpha {type=NiagaraFloat, defaultValue=0, description=""}
  User.RibbonIDOffset {type=NiagaraInt32}
  User.SkeletalMesh {
   type=NiagaraDataInterfaceSkeletalMesh
   defaultValue {
    struct=NiagaraExt_VariableValue_DataInterface
    value {dataInterface=&self:NiagaraDataInterfaceSkeletalMesh_0}
   }
  }
  User.WaterBodyIndex {type=NiagaraFloat, defaultValue=0, description=""}
  User.WaterDepth {type=NiagaraFloat, defaultValue=0, description=""}
  User.WaterPlaneLocation {type=Vector3f, description="", defaultValue {struct=Vector3f, value=<0 0 0>}}
  User.WaterPlaneNormal {type=Vector3f, description="", defaultValue {struct=Vector3f, value=<0 0 0>}}
  User.WaveAttenuationFactor {type=NiagaraFloat, defaultValue=0, description=""}
  User.WaveTime {type=NiagaraFloat, defaultValue=0, description=""}
 ]
}
